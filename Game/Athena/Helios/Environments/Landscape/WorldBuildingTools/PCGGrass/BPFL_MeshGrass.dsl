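; Function library for the PCG mesh grass world-building tools, with one check for whether the world has scalability data layers set up.
; HasScalabilityDLs finds the BP_MeshGrass_Partitioner actor and uses it to get the World Partition Data Layer Manager.
; It returns whether an instance of the DL_MeshGrass_MobileHighEnd asset exists, treating that single layer as the indicator for scalability layers.
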(fn HasScalabilityDLs
  (bind _returnvalue (Actor|GetActorOfClass 0 "/Game/Athena/Helios/Environments/Landscape/WorldBuildingTools/PCGGrass/BP_MeshGrass_Partitioner.BP_MeshGrass_Partitioner_C"))
  (bind _returnvalue_2 (WorldPartition|GetDataLayerManager _returnvalue))
  (return (Utilities|IsValid (DataLayers|GetDataLayerInstancefromAsset _returnvalue_2 "/Game/Athena/Helios/Environments/Landscape/WorldBuildingTools/PCGGrass/Resources/DL_MeshGrass_MobileHighEnd.DL_MeshGrass_MobileHighEnd") :!class /Script/Engine.KismetSystemLibrary)))
