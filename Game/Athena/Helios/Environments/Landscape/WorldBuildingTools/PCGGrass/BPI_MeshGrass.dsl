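; Blueprint interface for the PCG grass world-building tools in the Helios landscape, used with mesh grass.
; It declares one function, AddWorldTexture, which passes a world texture by name along with a Mesh Grass World Texture Data struct.
; The file holds only the signature, so implementing Blueprints define how the texture data is registered or used.
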(fn AddWorldTexture
  (params (WorldTextureName :type Name) (WorldTextureData :type Mesh%20Grass%20World%20Texture%20Data%20Structure)))
