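; BPI_MeshGrassUtility is a Blueprint interface for mesh grass utilities in the Helios landscape PCG grass world-building tools.
; It declares four functions without implementations: AddISM takes a target actor and static mesh, and PopulateBiomeData takes a biome index defaulting to -1.
; PostSpawn receives an array of instanced static mesh components, and UpdateSpline receives a spline component and a width.
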(fn AddISM
  (params (TargetActor :type Actor%20Object%20Reference) (Mesh :type Static%20Mesh%20Object%20Reference)))

(fn PopulateBiomeData
  (params (BiomeIndex :type Integer :default -1)))

(fn PostSpawn
  (params (ISMs :type Array%20of%20Instanced%20Static%20Mesh%20Component%20Object%20References)))

(fn UpdateSpline
  (params (SplineComponent :type Spline%20Component%20Object%20Reference) (Width :type Float%20%28double-precision%29)))
