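; BP_MeshGrass is a mesh-grass actor whose instanced static meshes can act as Simple Destructible Props (SDPs) with resource drops.
; "SDP-" component tags encode interactivity, resource type and drop, which select SDP components, collision profiles and AthenaResourceRates rows.
; Instance world locations are matched against stored attach locations to force attachment to building actors.
; PostSpawn is delegated to a utility component, and view distance quality changes toggle mesh visibility.
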(event EventBeginPlay
  (bind _returnvalue (Settings|GetFortGameUserSettings))
  (EventDispatchers|BindEventtoOnViewDistanceQualityChanged _returnvalue (delegate Custom|OnViewDistanceQualityChanged)))

(event Custom|OnViewDistanceQualityChanged
  (Rendering|SetVisibility (Actor|GetComponentsByClass "/Script/Engine.InstancedStaticMeshComponent") (!= (Settings|GetViewDistanceQuality (stale EventBeginPlay _returnvalue)) 0) false))

(event Custom|InitPostSpawn
  (bind _eubp_meshgrassutilitycomponent (Variables|Default|GetEUBP_MeshGrassUtilityComponent))
  (CallFunction|PrintDebugString(Unattended) "BP_MeshGrass::InitPostSpawn" "NewEnumerator1")
  (Utilities|IsValid _eubp_meshgrassutilitycomponent
    (:Is%20Valid
      (Class|BPIMeshGrassUtility|PostSpawn _eubp_meshgrassutilitycomponent (Actor|GetComponentsByClass "/Script/Engine.InstancedStaticMeshComponent")))
    (:Is%20Not%20Valid
      (CallFunction|PrintDebugString(Unattended) "Invalid MeshGrassUtilityComponent, aborting PostSpawn logic" "NewEnumerator1"))))

(fn GetForcedAttachment
  (params (StaticMeshComponent :type Static%20Mesh%20Component%20Object%20Reference) (MeshInstance :type Integer))
  (bind _attachlocations (Variables|Default|GetAttachLocations))
  (bind _meshinstance (Variables|GetMeshInstance))
  (bind _asinstanced_static_mesh_component (Utilities|Casting|CastToInstancedStaticMeshComponent StaticMeshComponent))
  (bind _outinstancetransform (Components|InstancedStaticMesh|GetInstanceTransform _asinstanced_static_mesh_component _meshinstance true))
  (bind _outinstancetransform_location (Math|Transform|BreakTransform _outinstancetransform))
  (if (!= (Utilities|Array|FindItem _attachlocations _outinstancetransform_location) -1)
    (LocalVariables|SetLocalForcedAttachment (Utilities|Casting|CastToBuildingActor (Utilities|ResolveSoftReference (Utilities|Array|Get(acopy) (Variables|Default|GetAttachTargets) (Utilities|Array|FindItem _attachlocations _outinstancetransform_location)))))
    (else
      (Development|PrintString "Warning: failed to match SDP location" true true "(R=1.000000,G=0.000000,B=0.163907,A=1.000000)" 2e+01)
      (LocalVariables|SetLocalForcedAttachment 0)))
  (if (== (Variables|GetMeshInstance) 0)
    (if (or (Variables|Default|GetDebug) (Variables|Default|GetDebugForceAll))
      (Development|PrintString (Utilities|String|Append (Utilities|GetDisplayName self) "
"))
      (Development|PrintString (Utilities|String|Append "   " (Utilities|GetDisplayName _asinstanced_static_mesh_component) "
"))
      (Development|PrintString (Utilities|String|Append "        " (Utilities|String|ToString(Integer) (Variables|GetMeshInstance)) ": " (Utilities|GetDisplayName (LocalVariables|GetLocalForcedAttachment))) :!label _label))
    (return (LocalVariables|GetLocalForcedAttachment)))
  (goto _label))

(fn GetSimpleDestructiblePropComponents
  (bind _collisionprofile_blocking_name (Utilities|Struct|BreakCollisionProfileName (Variables|GetCollisionProfile_Blocking_Name)))
  (for _array_element (Actor|GetComponentsByClass "/Script/Engine.InstancedStaticMeshComponent")
    (bind (_issdp _resourcetype _interactivity) (CallFunction|ParseSDPTags (Variables|/Script/Engine.ActorComponent|GetComponentTags _array_element)))
    (if _issdp
      (Utilities|Array|Add (LocalVariables|GetSDPComponents) _array_element)
      (elif (Utilities|Enum|Equal(Enum) _interactivity "NewEnumerator1")
        (Collision|SetCollisionProfileName _array_element _collisionprofile_blocking_name false :!class /Script/Engine.PrimitiveComponent))))
  (return (LocalVariables|GetSDPComponents)))

(fn GetSimpleDestructiblePropEntry
  (params (SourceIndex :type Integer) (SourceComponent :type Static%20Mesh%20Component%20Object%20Reference))
  (bind (_issdp _resourcetype _interactivity _resourcedrop) (CallFunction|ParseSDPTags (Variables|/Script/Engine.ActorComponent|GetComponentTags SourceComponent)))
  (bind _forcedattachment (CallFunction|GetForcedAttachment (Variables|GetSourceComponent) (Variables|GetSourceIndex)))
  (Utilities|IsValid _forcedattachment
    (:Is%20Valid
      (return 0 3 (Utilities|Select :Index _interactivity :NewEnumerator0 "(Name=\"NoCollision\")" :NewEnumerator1 "(Name=\"NoCollision\")" :NewEnumerator2 "(Name=\"MeshGrass_SDP\")" :NewEnumerator3 "(Name=\"MeshGrass_SDP_Blocking\")") _forcedattachment (Utilities|Struct|MakeCurveTableRowHandle "/Game/Athena/Balance/DataTables/AthenaResourceRates.AthenaResourceRates" (Utilities|String|StringToName (Utilities|String|SelectString "ResourceNone" (Utilities|String|Append "Prop" (Utilities|String|EnumtoString _resourcetype) (Utilities|String|EnumtoString _resourcedrop)) (Utilities|Enum|NotEqual(Enum) _resourcedrop "NewEnumerator0")))) (Utilities|Enum|BytetoEnumEDynamicDestructionResourceType (Math|Conversions|ToByte(Integer) (+ (Math|Conversions|ToInteger(Byte) _resourcetype) 1))) _resourcetype "ATTACH_None" "SLOT_None" true (Utilities|Enum|NotEqual(Enum) _resourcedrop "NewEnumerator0") false))
    (:Is%20Not%20Valid
      (return 0 3 (Utilities|Select :Index _interactivity :NewEnumerator0 "(Name=\"NoCollision\")" :NewEnumerator1 "(Name=\"NoCollision\")" :NewEnumerator2 "(Name=\"MeshGrass_SDP\")" :NewEnumerator3 "(Name=\"MeshGrass_SDP_Blocking\")") 0 (Utilities|Struct|MakeCurveTableRowHandle "/Game/Athena/Balance/DataTables/AthenaResourceRates.AthenaResourceRates" (Utilities|String|StringToName (Utilities|String|SelectString "ResourceNone" (Utilities|String|Append "Prop" (Utilities|String|EnumtoString _resourcetype) (Utilities|String|EnumtoString _resourcedrop)) (Utilities|Enum|NotEqual(Enum) _resourcedrop "NewEnumerator0")))) (Utilities|Enum|BytetoEnumEDynamicDestructionResourceType (Math|Conversions|ToByte(Integer) (+ (Math|Conversions|ToInteger(Byte) _resourcetype) 1))) _resourcetype "ATTACH_Floor" "SLOT_None" true (Utilities|Enum|NotEqual(Enum) _resourcedrop "NewEnumerator0") false))))

(fn ParseSDPTags
  (params (Array :type Array%20of%20Names))
  (for _array_element Array
    (if (Utilities|String|StartsWith (Utilities|String|ToString(Name) _array_element) "SDP-")
      (bind _returnvalue (Utilities|String|ParseIntoArray (Utilities|String|ToString(Name) _array_element) "-" false))
      (return (>= (Math|Conversions|ToByte(Integer) (Utilities|String|StringToInteger (Utilities|Array|Get(aref) _returnvalue 1))) 2) (Utilities|Enum|BytetoEnumEFortResourceType (Math|Conversions|ToByte(Integer) (Utilities|String|StringToInteger (Utilities|Array|Get(aref) _returnvalue 2)))) (Utilities|Enum|BytetoEnumEMeshGrassInteractivity (Math|Conversions|ToByte(Integer) (Utilities|String|StringToInteger (Utilities|Array|Get(aref) _returnvalue 1)))) (Utilities|Enum|BytetoEnumEMeshGrassResourceDrop (Math|Conversions|ToByte(Integer) (Utilities|String|StringToInteger (Utilities|Array|Get(aref) _returnvalue 3)))))))
  (return false "Wood" "NewEnumerator0" "NewEnumerator0"))

(fn PrintDebugString(Unattended)
  (params (DebugString :type String) (Verbosity :type EMeshGrassLogLevel%20Enum :default "NewEnumerator0") (Active :type Boolean :default true))
  (if (and Active (or (and (Utilities|IsUnattended) (Variables|Default|GetDebugUnattended)) (Variables|Default|GetDebugForceAll)))
    (if (Utilities|Enum|Equal(Enum) Verbosity "NewEnumerator0")
      (Development|PrintString (Variables|GetDebugString))
      (else
        (Fortnite|FortShippingLog self (Utilities|String|Append (Utilities|String|Append (Utilities|String|EnumtoString Verbosity) ": ") "[" (Utilities|GetObjectName self) "] " (Variables|GetDebugString)))))))

(fn ShouldKeepSourceActor
  (return true))

(fn ConstructionScript)
