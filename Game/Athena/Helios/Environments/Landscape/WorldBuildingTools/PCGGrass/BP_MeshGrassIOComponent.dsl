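; Mesh grass IO component that converts authored biome and texture-stamping setups into internal biome data for the PCG grass world-building tool.
; PopulateTextureStamping fills five color-channel biomes, taking the biome texture and base varieties from the GlobalControls actor.
; ProcessVariety flattens LandscapeGrassType varieties, optionally weighting by density or additively merging base-biome varieties, while syncing editable per-variety weights.
; ProcessMeshSetup converts extra meshes, capping material overrides at two; both encode interactivity, surface and resource-drop as SDP strings.
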(event EventBeginPlay)

(event EventTick
  (params (DeltaSeconds :type Float%20%28single-precision%29)))

(fn ConsolidateGrassVarieties
  (params (InGrassVarieties :type Array%20of%20Mesh%20Grass%20Variety%20Data%20Structures) (Debug :type Boolean))
  (for (_array_element _array_index) InGrassVarieties
    (bind _returnvalue (Utilities|Array|AddUnique (LocalVariables|GetGrassMeshes) (Utilities|Struct|BreakMeshGrassVarietyData _array_element)))
    (if (== _returnvalue -1)
      (if (not (!= (.OverrideMaterial_15_CE7779B24332523077826AA4F83D1800 (Utilities|Struct|BreakMeshGrassVarietyData _array_element)) 0))
        (Utilities|Array|Add (LocalVariables|GetLocalExcludedIndexes) _array_index)
        (if (Variables|GetDebug)
          (Development|PrintString (Utilities|String|Append "Consolidating Index " (Utilities|String|ToString(Integer) _array_index) ": " (Utilities|String|ToString(Object) (Utilities|Struct|BreakMeshGrassVarietyData _array_element)))))))
    (Utilities|Array|Add (LocalVariables|GetOutGrassVarieties) _array_element))
  (return (LocalVariables|GetOutGrassVarieties) (LocalVariables|GetLocalExcludedIndexes)))

(fn EncodeSDPData
  (params (Interactivity :type EMeshGrassInteractivity%20Enum :default "NewEnumerator0") (SurfaceType :type EMeshGrassResourceType%20Enum :default "NewEnumerator0") (ResourceDrop :type EMeshGrassResourceDrop%20Enum :default "NewEnumerator0"))
  (return (Utilities|String|Append "SDP-" (Utilities|String|ToString(Byte) Interactivity) "-" (Utilities|String|ToString(Byte) SurfaceType) "-" (Utilities|String|ToString(Byte) ResourceDrop))))

(fn InterpretGrassDensityAsWeight
  (params (GrassVariety :type Grass%20Variety%20Structure%20%28by%20ref%29) (PrintDebug :type Boolean))
  (bind _returnvalue (Grass|GetGrassVarietyDensity GrassVariety "Epic"))
  (bind _returnvalue_2 (Grass|GetGrassVarietyDensity GrassVariety))
  (bind _returnvalue_3 (Grass|GetGrassVarietyDensity GrassVariety "Medium"))
  (bind _returnvalue_4 (Grass|GetGrassVarietyDensity GrassVariety "High"))
  (bind _returnvalue_5 (Math|Float|Max(Float) _returnvalue_2 _returnvalue_3 _returnvalue_4 _returnvalue))
  (bind _returnvalue_6 (* _returnvalue_5 0.1))
  (if PrintDebug
    (Development|PrintString (Utilities|String|Append "DensityFromAsset for " (Utilities|String|ToString(Object) (Utilities|Struct|BreakGrassVariety GrassVariety)) ": " (Utilities|String|ToString(Float) _returnvalue_6))))
  (return _returnvalue_6))

(fn IsVarietyDuplicate
  (params (GrassVariety :type Grass%20Variety%20Structure%20%28by%20ref%29) (UniqueMeshes :type Array%20of%20Static%20Mesh%20Object%20References))
  (bind (_grassmesh _overridematerials) (Utilities|Struct|BreakGrassVariety GrassVariety))
  (bind _returnvalue (Utilities|Array|AddUnique UniqueMeshes _grassmesh))
  (return (and (== _returnvalue -1) (Utilities|Array|IsEmpty _overridematerials))))

(fn PopulateBiome
  (params (MeshGrass_MeshSetup_Combined :type Mesh%20Grass%20Mesh%20Setup%20Combined%20Structure%20%28by%20ref%29) (TargetBiome :type Mesh%20Grass%20Biome%20Data%20Structure%20%28by%20ref%29) (BiomeTexture :type Texture%202D%20Object%20Reference) (BiomeBaseMeshSetupVariety :type Mesh%20Grass%20Mesh%20Setup%20Grass%20Type%20Structure) (SourceBiomeSetup :type Mesh%20Grass%20Biome%20Setup%20Structure))
  (bind (_grasstype_14_54bc33c9418d25a265c3f093e9fdbd6a _secondarygrasstype_18_b2fa945b46b827645232db91db988401 _varietiespreviewweighting_13_bdad33c040d683be153b09b81c9dd9b4 _additionalmeshes_23_73d5b7ea4437df882c1aa8a473def8d1) (Utilities|Struct|BreakMeshGrassMeshSetupCombined MeshGrass_MeshSetup_Combined))
  (bind (_grasstype_14_54bc33c9418d25a265c3f093e9fdbd6a_2 _secondarygrasstype_18_b2fa945b46b827645232db91db988401_2 _varietiespreviewweighting_13_bdad33c040d683be153b09b81c9dd9b4_2) (Utilities|Struct|BreakMeshGrassMeshSetupGrassType (LocalVariables|GetVarieties)))
  (Variables|SetVarieties_GrassType_14_54BC33C9418D25A265C3F093E9FDBD6A (Utilities|Struct|MakeMeshGrassMeshSetupGrassType _grasstype_14_54bc33c9418d25a265c3f093e9fdbd6a _secondarygrasstype_18_b2fa945b46b827645232db91db988401 _varietiespreviewweighting_13_bdad33c040d683be153b09b81c9dd9b4))
  (Variables|SetMeshes_MeshSetup_12_16D9329C456F5D0B2C114D8B8AC0C078 (Utilities|Struct|MakeMeshGrassMeshSetupBasic _additionalmeshes_23_73d5b7ea4437df882c1aa8a473def8d1))
  (Default|TransferPropertiestoInternal BiomeTexture TargetBiome SourceBiomeSetup (LocalVariables|GetVarieties) (LocalVariables|GetMeshes) BiomeBaseMeshSetupVariety)
  (return (Utilities|Struct|MakeMeshGrassMeshSetupCombined _grasstype_14_54bc33c9418d25a265c3f093e9fdbd6a_2 _secondarygrasstype_18_b2fa945b46b827645232db91db988401_2 _varietiespreviewweighting_13_bdad33c040d683be153b09b81c9dd9b4_2 (Utilities|Struct|BreakMeshGrassMeshSetupBasic (LocalVariables|GetMeshes)))))

(fn PopulateTextureStamping
  (params (MeshGrass_TextureStampingSetup :type Mesh%20Grass%20Texture%20Stamping%20Setup%20Structure%20%28by%20ref%29) (BiomeRed :type Mesh%20Grass%20Biome%20Data%20Structure%20%28by%20ref%29) (BiomeGreen :type Mesh%20Grass%20Biome%20Data%20Structure%20%28by%20ref%29) (BiomeBlue :type Mesh%20Grass%20Biome%20Data%20Structure%20%28by%20ref%29) (BiomeWhite :type Mesh%20Grass%20Biome%20Data%20Structure%20%28by%20ref%29) (BiomeBlack :type Mesh%20Grass%20Biome%20Data%20Structure%20%28by%20ref%29) (SourceBiomeSetup :type Mesh%20Grass%20Biome%20Setup%20Structure))
  (bind (_exclusivechannels_62_23102ba140d2dffb7d4174b6044ec5dc _meshsetupred_8_e0556f6b44bb52696061c185d1ede805 _meshsetupgreen_10_f89b5c7748b961f3a233c08a44a2e254 _meshsetupblue_12_fbd29ddc49fd0559fd443788568cf28d _meshsetupblack_16_9f884cb64cfc86830d1e4ebae2b9756d _meshsetupwhite_14_c6a06f1a44eb0a82cb96f8aa895f8678) (Utilities|Struct|BreakMeshGrassTextureStampingSetup MeshGrass_TextureStampingSetup))
  (bind _biometexture (LocalVariables|GetBiomeTexture))
  (bind _biomebasevarietysetup (LocalVariables|GetBiomeBaseVarietySetup))
  (Utilities|FlowControl|Sequence
    (:then_0
      (bind _returnvalue (Actor|GetActorOfClass "/Game/Athena/Helios/Environments/Landscape/WorldBuildingTools/PCGGrass/BP_MeshGrass_GlobalControls.BP_MeshGrass_GlobalControls_C"))
      (LocalVariables|SetBiomeTexture (Variables|/Game/Athena/Helios/Environments/Landscape/WorldBuildingTools/PCGGrass/BP_MeshGrass_GlobalControls.BP_MeshGrass_GlobalControls_C|GetBiomeTexture _returnvalue))
      (LocalVariables|SetBiomeBaseVarietySetup (Variables|/Game/Athena/Helios/Environments/Landscape/WorldBuildingTools/PCGGrass/BP_MeshGrass_GlobalControls.BP_MeshGrass_GlobalControls_C|GetBiomeBase_MeshSetupVariety _returnvalue)))
    (:then_1
      (bind _meshsetupcombined (Default|PopulateBiome _meshsetupblack_16_9f884cb64cfc86830d1e4ebae2b9756d BiomeBlack _biometexture _biomebasevarietysetup SourceBiomeSetup))
      (bind _meshsetupcombined_2 (Default|PopulateBiome _meshsetupwhite_14_c6a06f1a44eb0a82cb96f8aa895f8678 BiomeWhite _biometexture _biomebasevarietysetup SourceBiomeSetup))
      (bind _meshsetupcombined_3 (Default|PopulateBiome _meshsetupblue_12_fbd29ddc49fd0559fd443788568cf28d BiomeBlue _biometexture _biomebasevarietysetup SourceBiomeSetup))
      (bind _meshsetupcombined_4 (Default|PopulateBiome _meshsetupgreen_10_f89b5c7748b961f3a233c08a44a2e254 BiomeGreen _biometexture _biomebasevarietysetup SourceBiomeSetup))
      (bind _meshsetupcombined_5 (Default|PopulateBiome _meshsetupred_8_e0556f6b44bb52696061c185d1ede805 BiomeRed _biometexture _biomebasevarietysetup SourceBiomeSetup))
      (Utilities|Struct|SetmembersinMeshGrassTextureStampingSetup MeshGrass_TextureStampingSetup _meshsetupcombined_5 _meshsetupcombined_4 _meshsetupcombined_3 _meshsetupcombined _meshsetupcombined_2))))

(fn ProcessMeshSetup
  (params (MeshSetup_MeshSetup_12_16D9329C456F5D0B2C114D8B8AC0C078 :type Array%20of%20Mesh%20Grass%20Mesh%20Setup%20Structures))
  (bind _outmeshsetup (LocalVariables|GetOutMeshSetup))
  (for _array_element MeshSetup_MeshSetup_12_16D9329C456F5D0B2C114D8B8AC0C078
    (Utilities|FlowControl|Sequence
      (:then_0
        (bind _returnvalue (Utilities|Array|Add _outmeshsetup _array_element))
        (if (> (Utilities|Array|Length (Utilities|Struct|BreakMeshGrassMeshSetup (Utilities|Array|Get(aref) _outmeshsetup _returnvalue))) 2)
          (Utilities|Array|Resize (Utilities|Struct|BreakMeshGrassMeshSetup (Utilities|Array|Get(aref) _outmeshsetup _returnvalue)) 2) ; Limit material overrides to 2 (UI)
        ))
      (:then_1
        (if (Utilities|Array|IsNotEmpty (.MaterialOverrides_70_20BFB541404CF2988D48EDBC7E7239A8 (Utilities|Struct|BreakMeshGrassMeshSetup _array_element)))
          (LocalVariables|SetOverrideMaterial (Utilities|Array|Get(acopy) (.MaterialOverrides_70_20BFB541404CF2988D48EDBC7E7239A8 (Utilities|Struct|BreakMeshGrassMeshSetup _array_element)))))
        (if (Utilities|Array|IsValidIndex (.MaterialOverrides_70_20BFB541404CF2988D48EDBC7E7239A8 (Utilities|Struct|BreakMeshGrassMeshSetup _array_element)) 1)
          (LocalVariables|SetOverrideMaterial1 (Utilities|Array|Get(acopy) (.MaterialOverrides_70_20BFB541404CF2988D48EDBC7E7239A8 (Utilities|Struct|BreakMeshGrassMeshSetup _array_element)) 1)))
        (Utilities|Array|Add (LocalVariables|GetOutMeshData) (Utilities|Struct|MakeMeshGrassMeshData (Utilities|Struct|BreakMeshGrassMeshSetup _array_element) (LocalVariables|GetOverrideMaterial) (LocalVariables|GetOverrideMaterial1) (.Weight_57_20DC87BC4F02B793B0DA1A85CC132F17 (Utilities|Struct|BreakMeshGrassMeshSetup _array_element)) (.Dominance_73_4EBE63E94CB058569CEB24B614E742CA (Utilities|Struct|BreakMeshGrassMeshSetup _array_element)) (.Interactivity_76_7EA054F24BE1C47AF613409404C93EE6 (Utilities|Struct|BreakMeshGrassMeshSetup _array_element)) (.Rendering_79_FDB572B04B9F5E7FFC6A1587947312D9 (Utilities|Struct|BreakMeshGrassMeshSetup _array_element)) (CallFunction|EncodeSDPData (.Interactivity_76_7EA054F24BE1C47AF613409404C93EE6 (Utilities|Struct|BreakMeshGrassMeshSetup _array_element)) (.SurfaceType_83_014E46CF49F65DFF1A5244B78941256E (Utilities|Struct|BreakMeshGrassMeshSetup _array_element)) (.ResourceDrop_85_3ACC02D34420471F70AB608BE83FF9D9 (Utilities|Struct|BreakMeshGrassMeshSetup _array_element)))))
        (LocalVariables|SetOverrideMaterial 0)
        (LocalVariables|SetOverrideMaterial1 0))))
  (Variables|SetBy-RefVar MeshSetup_MeshSetup_12_16D9329C456F5D0B2C114D8B8AC0C078 (LocalVariables|GetOutMeshSetup))
  (return (LocalVariables|GetOutMeshData)))

(fn ProcessVariety
  (params (GrassVarietySetup :type Mesh%20Grass%20Mesh%20Setup%20Grass%20Type%20Structure) (Additive :type Boolean) (AdditiveTarget :type Mesh%20Grass%20Mesh%20Setup%20Grass%20Type%20Structure) (WeightByGrassTypeDensity :type Boolean) (Debug :type Boolean))
  (bind (_grasstype_14_54bc33c9418d25a265c3f093e9fdbd6a _secondarygrasstype_18_b2fa945b46b827645232db91db988401 _varietiespreviewweighting_13_bdad33c040d683be153b09b81c9dd9b4) (Utilities|Struct|BreakMeshGrassMeshSetupGrassType GrassVarietySetup))
  (bind _currentgrassvariety (LocalVariables|GetCurrentGrassVariety))
  (bind (_grassmesh _overridematerials) (Utilities|Struct|BreakGrassVariety _currentgrassvariety))
  (bind (_grasstype_14_54bc33c9418d25a265c3f093e9fdbd6a_2 _secondarygrasstype_18_b2fa945b46b827645232db91db988401_2 _varietiespreviewweighting_13_bdad33c040d683be153b09b81c9dd9b4_2) (Utilities|Struct|BreakMeshGrassMeshSetupGrassType AdditiveTarget))
  (bind _biomeweights (LocalVariables|GetBiomeWeights))
  (bind _returnvalue (Utilities|Array|Length _biomeweights))
  (bind _biomeweights_2 (LocalVariables|GetBiomeWeights))
  (bind _returnvalue_2 (Utilities|Array|LastIndex _biomeweights_2))
  (bind _returnvalue_3 (Utilities|Array|IsEmpty _biomeweights))
  (bind _biomegrassvarieties (LocalVariables|GetBiomeGrassVarieties))
  (bind _returnvalue_4 (Utilities|Array|Length _biomegrassvarieties))
  (bind _basebiomeweights (LocalVariables|GetBaseBiomeWeights))
  (bind _returnvalue_5 (Utilities|Array|LastIndex _basebiomeweights))
  (bind _biomeweights_3 (LocalVariables|GetBiomeWeights))
  (bind _returnvalue_6 (Utilities|Array|LastIndex _biomeweights_3))
  (bind _iscurrentvarietyduplicate (LocalVariables|GetIsCurrentVarietyDuplicate))
  (bind _returnvalue_7 (not _iscurrentvarietyduplicate))
  (bind _returnvalue_8 (== _returnvalue_4 _returnvalue))
  (bind _returnvalue_9 (or _returnvalue_8 _returnvalue_3))
  (bind _additive (Variables|GetAdditive))
  (bind _returnvalue_10 (and _returnvalue_9 _additive))
  (bind (_startculldistance _endculldistance) (Grass|GetGrassVarietyCullDistanceRange (LocalVariables|GetCurrentGrassVariety) "Epic"))
  (bind (_startculldistance_2 _endculldistance_2) (Grass|GetGrassVarietyCullDistanceRange (LocalVariables|GetCurrentGrassVariety)))
  (bind (_startculldistance_3 _endculldistance_3) (Grass|GetGrassVarietyCullDistanceRange (LocalVariables|GetCurrentGrassVariety) "Medium"))
  (bind (_startculldistance_4 _endculldistance_4) (Grass|GetGrassVarietyCullDistanceRange (LocalVariables|GetCurrentGrassVariety) "High"))
  (bind (_scaling _scalex _scaley _scalez) (Utilities|Struct|BreakGrassVariety (LocalVariables|GetCurrentGrassVariety)))
  (bind (_scalex_min _scalex_max) (Utilities|Struct|BreakFloatInterval _scalex))
  (bind (_scaley_min _scaley_max) (Utilities|Struct|BreakFloatInterval _scaley))
  (bind (_scalez_min _scalez_max) (Utilities|Struct|BreakFloatInterval _scalez))
  (bind (_baligntotrianglenormals _baffectdistancefieldlighting _bcastdynamicshadow) (Utilities|Struct|BreakGrassVariety (LocalVariables|GetCurrentGrassVariety)))
  (bind _returnvalue_11 (- _returnvalue_6 _returnvalue_5))
  (bind _grassvarieties (Variables|/Script/Landscape.LandscapeGrassType|GetGrassVarieties _secondarygrasstype_18_b2fa945b46b827645232db91db988401))
  (bind _biomeweights_4 (LocalVariables|GetBiomeWeights))
  (bind _basebiomegrassvarieties (LocalVariables|GetBaseBiomeGrassVarieties))
  (bind _basebiomeweights_2 (LocalVariables|GetBaseBiomeWeights))
  (bind _weight (LocalVariables|GetWeight))
  (bind _initadditiveweights (LocalVariables|GetInitAdditiveWeights))
  (bind _iscurrentvarietyduplicate_2 (LocalVariables|GetIsCurrentVarietyDuplicate))
  (bind _grassvarieties_2 (Variables|/Script/Landscape.LandscapeGrassType|GetGrassVarieties _grasstype_14_54bc33c9418d25a265c3f093e9fdbd6a_2))
  (if (Utilities|IsValid _grasstype_14_54bc33c9418d25a265c3f093e9fdbd6a :!class /Script/Engine.KismetSystemLibrary)
    (Utilities|Array|Add (LocalVariables|GetValidLGTs) (Utilities|ToSoftObjectReference (Utilities|GetSoftObjectPath _grasstype_14_54bc33c9418d25a265c3f093e9fdbd6a)))
    (LocalVariables|SetBiomeGrassVarieties (Variables|/Script/Landscape.LandscapeGrassType|GetGrassVarieties _grasstype_14_54bc33c9418d25a265c3f093e9fdbd6a))
    (LocalVariables|SetBiomeWeights _varietiespreviewweighting_13_bdad33c040d683be153b09b81c9dd9b4)
    (Utilities|FlowControl|Sequence
      (:then_0
        (if (Utilities|IsValid _secondarygrasstype_18_b2fa945b46b827645232db91db988401 :!class /Script/Engine.KismetSystemLibrary)
          (Utilities|Array|Add (LocalVariables|GetValidLGTs) (Utilities|ToSoftObjectReference (Utilities|GetSoftObjectPath _secondarygrasstype_18_b2fa945b46b827645232db91db988401)))
          (if (Utilities|Array|IsNotEmpty _grassvarieties)
            (Utilities|Array|AppendArray (LocalVariables|GetBiomeGrassVarieties) _grassvarieties))))
      (:then_1
        (if (Utilities|IsValid _grasstype_14_54bc33c9418d25a265c3f093e9fdbd6a_2 :!class /Script/Engine.KismetSystemLibrary)
          (LocalVariables|SetBaseBiomeGrassVarieties _grassvarieties_2)
          (bind _output_get (LocalVariables|SetBaseBiomeWeights _varietiespreviewweighting_13_bdad33c040d683be153b09b81c9dd9b4_2))
          (Utilities|Array|Resize _output_get (Utilities|Array|Length _grassvarieties_2)) ; ignore secondary grass from base biome for now\nwe'll see if including it becomes a need in the future\ncurrently it seems like it would be more messy than helpful for UX
        ))
      (:then_2
        (LocalVariables|SetInitAdditiveWeights _returnvalue_10)
        (if (Variables|GetDebug)
          (Development|PrintString (Utilities|String|Append "InitAdditiveWeights: " (Utilities|String|ToString(Boolean) _returnvalue_10) ": " (Utilities|String|Append (Utilities|String|ToString(Integer) _returnvalue_4) "::" (Utilities|String|ToString(Integer) _returnvalue)))))
        (for (_array_element _array_index) (LocalVariables|GetBiomeGrassVarieties)
          (bind _output_get_2 (LocalVariables|SetCurrentGrassVariety _array_element))
          (bind _returnvalue_12 (Utilities|Array|AddUnique (LocalVariables|GetUniqueMeshes) (Utilities|Struct|BreakGrassVariety _array_element)))
          (LocalVariables|SetIsCurrentVarietyDuplicate (== _returnvalue_12 -1))
          (bind (_output_meshread_only_64_c756a69c4eaefe9be06fd5818b8bf6e1 _output_materialoverrideread_only_66_20bfb541404cf2988d48edbc7e7239a8 _output_weight_57_20dc87bc4f02b793b0da1a85cc132f17 _output_dominance_69_b12f024f45906840c4cc6c9ffa9dc73e _output_interactivity_72_a7b21e224abadfc760b45ebfb376bac5 _output_surfacetype_80_260a3d17449a87d173b686a733c75107 _output_resourcedrop_81_5720024544ad430f92aaa0b896911281) (Utilities|Struct|BreakMeshGrassVarietySetup (Utilities|Array|Get(aref) _biomeweights_2 _array_index)))
          (LocalVariables|SetDominance (Math|Float|SelectFloat 1.0 _output_dominance_69_b12f024f45906840c4cc6c9ffa9dc73e (or (== _output_weight_57_20dc87bc4f02b793b0da1a85cc132f17 -1.0) (> _array_index _returnvalue_2))))
          (LocalVariables|SetInteractivity (select (or (== _output_weight_57_20dc87bc4f02b793b0da1a85cc132f17 -1.0) (> _array_index _returnvalue_2)) "NewEnumerator0" _output_interactivity_72_a7b21e224abadfc760b45ebfb376bac5))
          (LocalVariables|SetSurfaceType (select false "NewEnumerator0" _output_surfacetype_80_260a3d17449a87d173b686a733c75107))
          (LocalVariables|SetResourceDrop (select false "NewEnumerator0" _output_resourcedrop_81_5720024544ad430f92aaa0b896911281))
          (if (Variables|GetWeightByGrassTypeDensity)
            (bind _weight_2 (Default|InterpretGrassDensityasWeight _output_get_2 (Variables|GetDebug)))
            (LocalVariables|SetWeight _weight_2)
            (else
              (LocalVariables|SetWeight (Math|Float|SelectFloat 1.0 _output_weight_57_20dc87bc4f02b793b0da1a85cc132f17 (or (== _output_weight_57_20dc87bc4f02b793b0da1a85cc132f17 -1.0) (> _array_index _returnvalue_2))))))
          (if (Variables|GetWeightByGrassTypeDensity)
            (if (and (Variables|GetDebug) (not _returnvalue_7))
              (Development|PrintString (Utilities|String|Append "Skipping duplicate: " (Utilities|String|ToString(Object) _grassmesh))))
            (if (and (or (and _returnvalue_7 (LocalVariables|GetIsCurrentVarietyAdditive)) (not (LocalVariables|GetIsCurrentVarietyAdditive))) (> (LocalVariables|GetWeight) 0))
              _))
          (if (Utilities|Array|IsNotEmpty _overridematerials)
            (LocalVariables|SetOverrideMaterial (Utilities|Array|Get(acopy) _overridematerials))
            (else
              (LocalVariables|SetOverrideMaterial 0)))
          (Utilities|Array|Add (LocalVariables|GetLocalMeshGrassWeights) (Utilities|Struct|MakeMeshGrassVarietySetup _grassmesh (LocalVariables|GetOverrideMaterial) (LocalVariables|GetWeight) (LocalVariables|GetDominance) (LocalVariables|GetInteractivity) (LocalVariables|GetSurfaceType) (LocalVariables|GetResourceDrop)))
          (if (Utilities|Array|IsValidIndex _overridematerials 1)
            (LocalVariables|SetOverrideMaterial1 (Utilities|Array|Get(acopy) _overridematerials 1)))
          (if (and (> _weight 0) (not (Math|Boolean|ANDBoolean (LocalVariables|GetIsCurrentVarietyAdditive) (== (LocalVariables|GetOverrideMaterial) 0) (LocalVariables|GetIsCurrentVarietyDuplicate))))
            (Utilities|Array|Add (LocalVariables|GetOutMeshGrassVariety) (Utilities|Struct|MakeMeshGrassVarietyData (Utilities|Struct|BreakGrassVariety (LocalVariables|GetCurrentGrassVariety)) (LocalVariables|GetOverrideMaterial) (LocalVariables|GetOverrideMaterial1) _weight (LocalVariables|GetDominance) (LocalVariables|GetInteractivity) (select (or _baffectdistancefieldlighting _bcastdynamicshadow) "NewEnumerator1" "NewEnumerator0") (Math|Integer|Max(Integer) _startculldistance_2 _startculldistance_3 _startculldistance_4 _startculldistance) (Math|Integer|Max(Integer) _endculldistance_2 _endculldistance_3 _endculldistance_4 _endculldistance) (Grass|GetInstanceWorldPositionOffsetDisableDistance (LocalVariables|GetCurrentGrassVariety)) (Utilities|Select :Index _scaling :Uniform (Utilities|Enum|LiteralenumEMeshGrassScalingMode "NewEnumerator0") :Free (Utilities|Enum|LiteralenumEMeshGrassScalingMode "NewEnumerator1") :LockXY (Utilities|Enum|LiteralenumEMeshGrassScalingMode "NewEnumerator2")) (Utilities|Select :Index _scaling :Uniform (Math|Conversions|ToVector(Float) _scalex_min) :Free (Math|Vector|MakeVector _scalex_min _scaley_min _scalez_min) :LockXY (Math|Vector|MakeVector _scalex_min _scalex_min _scalez_min)) (Utilities|Select :Index _scaling :Uniform (Math|Conversions|ToVector(Float) _scalex_max) :Free (Math|Vector|MakeVector _scalex_max _scaley_max _scalez_max) :LockXY (Math|Vector|MakeVector _scalex_max _scalex_max _scalez_max)) (Math|Integer|SelectInt 3 1 (and (Math|Boolean|ORBoolean (== (Grass|GetGrassVarietyDensity (LocalVariables|GetCurrentGrassVariety)) 0) (== (Grass|GetGrassVarietyDensity (LocalVariables|GetCurrentGrassVariety) "Medium") 0) (== (Grass|GetGrassVarietyDensity (LocalVariables|GetCurrentGrassVariety) "High") 0)) (> (Grass|GetGrassVarietyDensity (LocalVariables|GetCurrentGrassVariety) "Epic") 0))) (CallFunction|EncodeSDPData (LocalVariables|GetInteractivity) (LocalVariables|GetSurfaceType) (LocalVariables|GetResourceDrop))))))
        (if (Variables|GetAdditive)
          (LocalVariables|SetIsCurrentVarietyAdditive true)
          (for (_array_element_2 _array_index_2) _basebiomegrassvarieties
            (bind _output_get_3 (LocalVariables|SetCurrentGrassVariety _array_element_2))
            (bind _returnvalue_13 (Default|IsVarietyDuplicate _array_element_2 (LocalVariables|GetUniqueMeshes)))
            (LocalVariables|SetIsCurrentVarietyDuplicate _returnvalue_13)
            (if (Variables|GetWeightByGrassTypeDensity)
              (bind _weight_3 (Default|InterpretGrassDensityasWeight _output_get_3 (Variables|GetDebug)))
              (LocalVariables|SetWeight _weight_3)
              (else
                (bind (_output_meshread_only_64_c756a69c4eaefe9be06fd5818b8bf6e1_2 _output_materialoverrideread_only_66_20bfb541404cf2988d48edbc7e7239a8_2 _output_weight_57_20dc87bc4f02b793b0da1a85cc132f17_2 _output_dominance_69_b12f024f45906840c4cc6c9ffa9dc73e_2 _output_interactivity_72_a7b21e224abadfc760b45ebfb376bac5_2) (Utilities|Struct|BreakMeshGrassVarietySetup (Utilities|Array|Get(aref) _basebiomeweights_2 _array_index_2)))
                (bind (_output_meshread_only_64_c756a69c4eaefe9be06fd5818b8bf6e1_3 _output_materialoverrideread_only_66_20bfb541404cf2988d48edbc7e7239a8_3 _output_weight_57_20dc87bc4f02b793b0da1a85cc132f17_3 _output_dominance_69_b12f024f45906840c4cc6c9ffa9dc73e_3 _output_interactivity_72_a7b21e224abadfc760b45ebfb376bac5_3) (Utilities|Struct|BreakMeshGrassVarietySetup (Utilities|Array|Get(aref) _biomeweights_4 (+ _array_index_2 _returnvalue_11))))
                (LocalVariables|SetWeight (Math|Float|SelectFloat -1.0 (Math|Float|SelectFloat _output_weight_57_20dc87bc4f02b793b0da1a85cc132f17_2 _output_weight_57_20dc87bc4f02b793b0da1a85cc132f17_3 _initadditiveweights) _iscurrentvarietyduplicate_2))))
            (LocalVariables|SetDominance (Math|Float|SelectFloat -1.0 (Math|Float|SelectFloat _output_dominance_69_b12f024f45906840c4cc6c9ffa9dc73e_2 _output_dominance_69_b12f024f45906840c4cc6c9ffa9dc73e_3 _initadditiveweights) _iscurrentvarietyduplicate_2))
            (LocalVariables|SetInteractivity (select _iscurrentvarietyduplicate_2 "NewEnumerator0" (select _initadditiveweights _output_interactivity_72_a7b21e224abadfc760b45ebfb376bac5_2 _output_interactivity_72_a7b21e224abadfc760b45ebfb376bac5_3)))
            (if (and (Variables|GetAdditive) (Variables|GetDebug))
              (Development|PrintString (Utilities|String|Append (Utilities|String|ToString(Integer) (Math|Integer|SelectInt _array_index_2 (+ _array_index_2 _returnvalue_11) (LocalVariables|GetInitAdditiveWeights))) ":" (Utilities|String|ToString(Integer) (Math|Integer|SelectInt (Utilities|Array|LastIndex _basebiomegrassvarieties) _returnvalue_6 (LocalVariables|GetInitAdditiveWeights))) " = " (Utilities|String|ToString(Float) (Math|Float|SelectFloat _output_weight_57_20dc87bc4f02b793b0da1a85cc132f17_2 _output_weight_57_20dc87bc4f02b793b0da1a85cc132f17_3 _initadditiveweights)))))))
        (return (LocalVariables|GetLocalMeshGrassWeights) (LocalVariables|GetOutMeshGrassVariety) (LocalVariables|GetValidLGTs))))))

(fn TransferPropertiesToInternal
  (params (BiomeTexture :type Texture%202D%20Object%20Reference) (TargetBiome :type Mesh%20Grass%20Biome%20Data%20Structure%20%28by%20ref%29) (SourceBiomeSetup :type Mesh%20Grass%20Biome%20Setup%20Structure) (SourceMeshVarietySetup :type Mesh%20Grass%20Mesh%20Setup%20Grass%20Type%20Structure%20%28by%20ref%29) (SourceMeshBasicSetup :type Mesh%20Grass%20Mesh%20Setup%20Basic%20Structure%20%28by%20ref%29) (BiomeBaseMeshSetupVariety :type Mesh%20Grass%20Mesh%20Setup%20Grass%20Type%20Structure) (WeightByGrassTypeDensity :type Boolean) (Debug :type Boolean))
  (bind (_grasstype_14_54bc33c9418d25a265c3f093e9fdbd6a _secondarygrasstype_18_b2fa945b46b827645232db91db988401) (Utilities|Struct|BreakMeshGrassMeshSetupGrassType SourceMeshVarietySetup))
  (bind (_enabled_20_54bc33c9418d25a265c3f093e9fdbd6a _biometexturepage_38_e979ea1744ae6b361e60bfada1e24d94 _biometexturechannel_43_6ddfdfdf46f3ccf038b99cb06e487fe9 _biomestrength_42_80810f2341c54c319c668a82f53f5e8d _densityfactor_27_ad05867d402f8885736bd3bed9254add _additive_19_04a01db743f8255a71231ba33fcc81fb _usevariety_21_536335004eb40753fb3d34992005b360 _allowexclusions_24_04e7ba19427c2bed8063e58388c5f7ce) (Utilities|Struct|BreakMeshGrassBiomeSetup SourceBiomeSetup))
  (Utilities|FlowControl|Sequence
    (:then_0
      (Utilities|IsValid _grasstype_14_54bc33c9418d25a265c3f093e9fdbd6a
        (:Is%20Valid
          (Utilities|FlowControl|Sequence
            (:then_0
              (bind (_outmeshgrassweights _outmeshgrassvarieties _outlgts) (Default|ProcessVariety (Variables|GetSourceMeshVarietySetup) _additive_19_04a01db743f8255a71231ba33fcc81fb (Variables|GetBiomeBaseMeshSetupVariety) (Variables|GetWeightByGrassTypeDensity) (Variables|GetDebug)))
              (Utilities|Struct|SetmembersinMeshGrassMeshSetupGrassType SourceMeshVarietySetup _outmeshgrassweights)
              (LocalVariables|SetProcessedMeshGrassVarieties _outmeshgrassvarieties)
              (LocalVariables|SetObjectDependencies _outlgts))
            (:then_1
              (LocalVariables|SetConsolidatedMeshGrassVarieties (LocalVariables|GetProcessedMeshGrassVarieties)))))
        (:Is%20Not%20Valid
          (Variables|SetBy-RefVar (Variables|GetSourceMeshVarietySetup) (Utilities|Struct|MakeMeshGrassMeshSetupGrassType _grasstype_14_54bc33c9418d25a265c3f093e9fdbd6a _secondarygrasstype_18_b2fa945b46b827645232db91db988401)))))
    (:then_1
      (bind _meshdata (CallFunction|ProcessMeshSetup SourceMeshBasicSetup))
      (LocalVariables|SetMeshData _meshdata))
    (:then_2
      (Utilities|Struct|SetmembersinMeshGrassBiomeData TargetBiome _enabled_20_54bc33c9418d25a265c3f093e9fdbd6a (Variables|GetBiomeTexture) (Math|Vector2D|MakeVector2D (Math|Conversions|ToFloat(Integer) _biometexturepage_38_e979ea1744ae6b361e60bfada1e24d94) (- (Math|Conversions|ToFloat(Byte) _biometexturechannel_43_6ddfdfdf46f3ccf038b99cb06e487fe9) 1.0)) _biomestrength_42_80810f2341c54c319c668a82f53f5e8d _densityfactor_27_ad05867d402f8885736bd3bed9254add _additive_19_04a01db743f8255a71231ba33fcc81fb _usevariety_21_536335004eb40753fb3d34992005b360 _allowexclusions_24_04e7ba19427c2bed8063e58388c5f7ce (LocalVariables|GetConsolidatedMeshGrassVarieties) (LocalVariables|GetMeshData) (LocalVariables|GetObjectDependencies)))))
