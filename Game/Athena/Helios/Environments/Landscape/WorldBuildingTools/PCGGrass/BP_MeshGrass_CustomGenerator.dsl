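; BP_MeshGrass_CustomGenerator is a world-building actor that authors clusters of mesh grass instances and generates them through PCG.
; Construction seeds PCG from location or an override, prunes empty instanced mesh components, applies their render settings, and fits a box to bounds and projection distance.
; Authoring state controls cluster visibility and PCG activation.
; Forced refreshes regenerate locally, then use a one-shot graph-generated callback to regenerate the BP_MeshGrass_Partitioner.
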(fn AddISM
  (Utilities|IsValid (Variables|Authoring|GetMeshToAdd)
    (:Is%20Valid
      (Class|BPIMeshGrassUtility|AddISM (Variables|Default|GetEUBP_MeshGrassUtilityComponent) self (Variables|Authoring|GetMeshToAdd)))))

(fn ClearSeed
  (Utilities|Struct|SetmembersinMeshGrassCustomGeneratorParams (Variables|Generation|GetConfiguration) 0)
  (CallFunction|GenerateSeed)
  (Debug|ForceRefresh!))

(fn ForceRefresh!
  (bind _pcg (Variables|Default|GetPCG))
  (if (Variables|Generation|GetIsActive)
    (EventDispatchers|BindEventtoOnGraphGenerated _pcg (EventDispatchers|CreateEvent :!event RequestPartitionerUpdate))
    (PCG|GenerateLocal _pcg true)))

(fn GenerateSeed
  (bind _structout (Utilities|Struct|SetmembersinPCGPoint (Variables|Default|GetActorPoint) (Math|Transform|MakeTransform (Transformation|GetActorLocation))))
  (PCG|Helpers|SetSeedfromPosition _structout)
  (Variables|/Script/PCG.PCGComponent|SetSeed (Utilities|Struct|BreakPCGPoint _structout) (Variables|Default|GetPCG)))

(fn LockSeed
  (Utilities|Struct|SetmembersinMeshGrassCustomGeneratorParams (Variables|Generation|GetConfiguration) (Variables|/Script/PCG.PCGComponent|GetSeed (Variables|Default|GetPCG))))

(fn RandomizeSeed
  (bind _structout (Utilities|Struct|SetmembersinMeshGrassCustomGeneratorParams (Variables|Generation|GetConfiguration) (Math|Random|RandomIntegerinRange -214748364 214748364)))
  (bind (_structout_projectinstances_88_bd6264b0432cc7c90095b8981565c58e _structout_projectiondistance_87_5ce041ac44400dbfd67596a943fae585 _structout_spawnradius_74_c74a003b480f7048ddfe54ab523b94e1 _structout_rotatetosurfacenormal_84_122eab26433897a89962f4af587fdd6d _structout_preserveverticaloffset_83_c031594e41f9762388ad69981301f2c4 _structout_preserverendersettings_90_293237bb46a8523b471db1b19427e45a _structout_seedoverride_94_a3fad2fa44770ab1480f2bb46f857005) (Utilities|Struct|BreakMeshGrassCustomGeneratorParams _structout))
  (Variables|/Script/PCG.PCGComponent|SetSeed _structout_seedoverride_94_a3fad2fa44770ab1480f2bb46f857005 (Variables|Default|GetPCG))
  (Debug|ForceRefresh!))

(fn RequestPartitionerUpdate
  (params (PCGComponent :type PCGComponent%20Object%20Reference))
  (bind _returnvalue (Actor|GetActorOfClass "/Game/Athena/Helios/Environments/Landscape/WorldBuildingTools/PCGGrass/BP_MeshGrass_Partitioner.BP_MeshGrass_Partitioner_C"))
  (PCG|GenerateLocal (Variables|/Game/Athena/Helios/Environments/Landscape/WorldBuildingTools/PCGGrass/BP_MeshGrass_Partitioner.BP_MeshGrass_Partitioner_C|GetPCG_MeshGrass_Partitioner _returnvalue) true)
  (EventDispatchers|UnbindallEventsfromOnGraphGenerated PCGComponent))

(fn ShowHideCluster
  (params (Show :type Boolean))
  (Rendering|SetVisibility (Variables|Default|GetCurrentISMs) Show false))

(fn UpdateBounds
  (bind _configuration (Variables|Generation|GetConfiguration))
  (bind _projectiondistance_87_5ce041ac44400dbfd67596a943fae585 (Utilities|Struct|BreakMeshGrassCustomGeneratorParams _configuration))
  (bind _returnvalue (/ _projectiondistance_87_5ce041ac44400dbfd67596a943fae585 2.0))
  (bind _box (Variables|Default|GetBox))
  (bind _returnvalue_2 (PCG|Helpers|GetActorBoundsPCG self))
  (Components|Box|SetBoxExtent _box (Utilities|Operators|Add (/ (Math|Geometry|GetSize(Box) _returnvalue_2) 2.0) (Math|Vector|MakeVector 0.0 0.0 _returnvalue)) false)
  (Transformation|SetWorldLocationAndRotation _box (Utilities|Operators|Add (Math|Geometry|GetCenter(Box) _returnvalue_2) (Math|Vector|MakeVector 0.0 0.0 (* _returnvalue -1.0))) "0, 0, 0" false false))

(fn UpdateISMs
  (for _array_element (Variables|Default|GetCurrentISMs)
    (if (and (== (Components|InstancedStaticMesh|GetInstanceCount _array_element) 0) (not (Utilities|String|StartsWith (.RightS (Utilities|String|Split (Utilities|GetDisplayName _array_element) "." "IgnoreCase" "FromEnd")) (Utilities|String|LeftChop (Utilities|GetClassDisplayName (Utilities|GetClass _array_element)) 11) "CaseSensitive")))
      (Components|DestroyComponent _array_element)
      (else
        (Rendering|SetCastShadow _array_element (Utilities|Select :Index (Variables|/Game/Athena/Helios/Environments/Landscape/WorldBuildingTools/PCGGrass/BP_MeshGrassInstancedStaticMeshComponent.BP_MeshGrassInstancedStaticMeshComponent_C|GetRendering%20Quality _array_element) :NewEnumerator0 false :NewEnumerator1 true))
        (Rendering|SetVisibleinRayTracing _array_element (Utilities|Select :Index (Variables|/Game/Athena/Helios/Environments/Landscape/WorldBuildingTools/PCGGrass/BP_MeshGrassInstancedStaticMeshComponent.BP_MeshGrassInstancedStaticMeshComponent_C|GetRendering%20Quality _array_element) :NewEnumerator0 false :NewEnumerator1 true))
        (Rendering|LOD|SetWorldPositionOffsetDisableDistance _array_element (Variables|/Game/Athena/Helios/Environments/Landscape/WorldBuildingTools/PCGGrass/BP_MeshGrassInstancedStaticMeshComponent.BP_MeshGrassInstancedStaticMeshComponent_C|GetAnimation%20Disable%20Distance _array_element) :!class /Script/Engine.StaticMeshComponent)
        (bind (_start___end_cull_distance_x _start___end_cull_distance_y) (Utilities|Struct|BreakIntPoint (Variables|/Game/Athena/Helios/Environments/Landscape/WorldBuildingTools/PCGGrass/BP_MeshGrassInstancedStaticMeshComponent.BP_MeshGrassInstancedStaticMeshComponent_C|GetStart%20&%20End%20Cull%20Distance_X _array_element)))
        (Components|InstancedStaticMesh|SetCullDistances _array_element _start___end_cull_distance_x _start___end_cull_distance_y)))))

(fn ConstructionScript
  (bind _configuration (Variables|Generation|GetConfiguration))
  (bind _seedoverride_94_a3fad2fa44770ab1480f2bb46f857005 (Utilities|Struct|BreakMeshGrassCustomGeneratorParams _configuration))
  (Utilities|FlowControl|Sequence
    (:then_0
      (Transformation|SetActorScale3D "1.000000,1.000000,1.000000")
      (if (== _seedoverride_94_a3fad2fa44770ab1480f2bb46f857005 0)
        (CallFunction|GenerateSeed)
        (else
          (Variables|/Script/PCG.PCGComponent|SetSeed _seedoverride_94_a3fad2fa44770ab1480f2bb46f857005 (Variables|Default|GetPCG)))))
    (:then_1
      (Variables|Default|SetCurrentISMs (Actor|GetComponentsByClass "/Game/Athena/Helios/Environments/Landscape/WorldBuildingTools/PCGGrass/BP_MeshGrassInstancedStaticMeshComponent.BP_MeshGrassInstancedStaticMeshComponent_C"))
      (CallFunction|UpdateISMs)
      (if (Utilities|Operators|Equal(==) :A (Transformation|GetActorLocation))
        (CallFunction|ShowHideCluster true)
        (elif (Utilities|Array|IsEmpty (Variables|Default|GetCurrentISMs))
          (Variables|Authoring|SetAuthoring%20Mode true)
          (CallFunction|ShowHideCluster (Variables|Authoring|GetAuthoring%20Mode))
          (else
            (Variables|/Script/PCG.PCGComponent|SetbActivated (and (Variables|Generation|GetIsActive) (not (Variables|Authoring|GetAuthoring%20Mode))) (Variables|Default|GetPCG))))))
    (:then_2
      (Internal|UpdateBounds))))
