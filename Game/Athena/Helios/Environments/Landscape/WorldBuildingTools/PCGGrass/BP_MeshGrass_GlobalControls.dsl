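; Blueprint of global controls for procedural mesh grass, holding settings for a base biome plus biomes A to G.
; The construction script enforces biome indices and base setup, then populates biome data.
; UpdateInternalData copies world textures and uses BP_MeshGrassIOComponent to transfer each biome's setup into internal data, passing the base biome's texture and variety meshes to A–G.
; An add-texture event validates name, asset and scale before storing the texture and refreshing.
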(event EventBeginPlay)

(event Collision|EventActorBeginOverlap
  (params (OtherActor :type Actor%20Object%20Reference)))

(event EventTick
  (params (DeltaSeconds :type Float%20%28single-precision%29)))

(event EventAddWorldTexture
  (params (WorldTextureName :type Name :default "None") (WorldTextureData :type Mesh%20Grass%20World%20Texture%20Data%20Structure))
  (CallFunction|HandleAddWorldTexture WorldTextureName WorldTextureData))

(event EventPopulateBiomeData
  (params (BiomeIndex :type Integer))
  (CallFunction|UpdateInternalData BiomeIndex))

(fn EnforceProperties
  (bind _biomea (Variables|Internal|GetBiomeA))
  (bind _biomebase (Variables|Internal|GetBiomeBase))
  (bind _biomeb (Variables|Internal|GetBiomeB))
  (bind _biomec (Variables|Internal|GetBiomeC))
  (bind _biomed (Variables|Internal|GetBiomeD))
  (bind _biomee (Variables|Internal|GetBiomeE))
  (bind _biomef (Variables|Internal|GetBiomeF))
  (bind _biomeg (Variables|Internal|GetBiomeG))
  (Utilities|FlowControl|Sequence
    (:then_0
      (for _index (range (+ (Utilities|Array|LastIndex (Utilities|Array|MakeArray _biomebase _biomea _biomeb _biomec _biomed _biomee _biomef _biomeg)) 1))
        (Utilities|Struct|SetmembersinMeshGrassBiomeData (Utilities|Select :Index _index :Option%200 _biomebase :Option%201 _biomea :Option%202 _biomeb :Option%203 _biomec :Option%204 _biomed :Option%205 _biomee :Option%206 _biomef :Option%207 _biomeg) _index)))
    (:then_1
      (Utilities|Struct|SetmembersinMeshGrassBiomeSetup (Variables|BiomeBase|GetBiomeBase_BiomeSetup) -1 "NewEnumerator0" false))
    (:then_2
      (if (Editor|InEditor)
        (Variables|Information|SetDocumentation: "https://epicgames.atlassian.net/wiki/spaces/FORT/pages/75902954/MeshGrass+-+Procedural+Grass+On+Meshes")))))

(fn HandleAddWorldTexture
  (params (WorldTextureName :type Name) (WorldTextureData :type Mesh%20Grass%20World%20Texture%20Data%20Structure))
  (bind (_worldtexture_9_bc142def47edaa3e51cff1b4953a7c34 _texturechannel_22_0bdf1b9a47f0fedb61c33090963e15d9 _filterthreshold_19_79dda26e43c0613b1ac55ab2550f392b _invertfilter_17_f9c71d7b4d1b0f8139d3f9b3750eb87e _transform_13_672e88e242f1a0772b933b910defa397) (Utilities|Struct|BreakMeshGrassWorldTextureData WorldTextureData))
  (if (Math|Boolean|ANDBoolean (!= WorldTextureName "") (Utilities|IsValidSoftObjectReference _worldtexture_9_bc142def47edaa3e51cff1b4953a7c34) (> (Math|Vector|VectorLength (.scale _transform_13_672e88e242f1a0772b933b910defa397)) 0))
    (Utilities|Map|Add (Variables|GlobalParameters|GenerationFilters|GetWorldTextures) WorldTextureName WorldTextureData)
    (CallFunction|UpdateInternalData -1)))

(fn UpdateInternalData
  (params (BiomeIndex :type Integer :default -1))
  (if (== BiomeIndex -1)
    (bind _values (Utilities|Map|Values (Variables|GlobalParameters|GenerationFilters|GetWorldTextures)))
    (Variables|Internal|SetWorldTextureData _values)
    (Utilities|FlowControl|Sequence
      (:then_0
        _)
      (:then_1
        (Default|TransferPropertiestoInternal (Variables|Default|GetBP_MeshGrassIOComponent) (Variables|BiomeBase|GetBiomeTexture) (Variables|Internal|GetBiomeA) (Variables|BiomeA|GetBiomeA_BiomeSetup) (Variables|BiomeA|GetBiomeA_MeshSetupVariety) (Variables|BiomeA|GetBiomeA_MeshSetupBasic) (Variables|BiomeBase|GetBiomeBase_MeshSetupVariety) (Variables|GlobalParameters|GetWeightByGrassTypeDensity) false))
      (:then_2
        (Default|TransferPropertiestoInternal (Variables|Default|GetBP_MeshGrassIOComponent) (Variables|BiomeBase|GetBiomeTexture) (Variables|Internal|GetBiomeB) (Variables|BiomeB|GetBiomeB_BiomeSetup) (Variables|BiomeB|GetBiomeB_MeshSetupVariety) (Variables|BiomeB|GetBiomeB_MeshSetupBasic) (Variables|BiomeBase|GetBiomeBase_MeshSetupVariety) (Variables|GlobalParameters|GetWeightByGrassTypeDensity) false))
      (:then_3
        (Default|TransferPropertiestoInternal (Variables|Default|GetBP_MeshGrassIOComponent) (Variables|BiomeBase|GetBiomeTexture) (Variables|Internal|GetBiomeC) (Variables|BiomeC|GetBiomeC_BiomeSetup) (Variables|BiomeC|GetBiomeC_MeshSetupVariety) (Variables|BiomeC|GetBiomeC_MeshSetupBasic) (Variables|BiomeBase|GetBiomeBase_MeshSetupVariety) (Variables|GlobalParameters|GetWeightByGrassTypeDensity) false))
      (:then_4
        (Default|TransferPropertiestoInternal (Variables|Default|GetBP_MeshGrassIOComponent) (Variables|BiomeBase|GetBiomeTexture) (Variables|Internal|GetBiomeD) (Variables|BiomeD|GetBiomeD_BiomeSetup) (Variables|BiomeD|GetBiomeD_MeshSetupVariety) (Variables|BiomeD|GetBiomeD_MeshSetupBasic) (Variables|BiomeBase|GetBiomeBase_MeshSetupVariety) (Variables|GlobalParameters|GetWeightByGrassTypeDensity) false))
      (:then_5
        (Default|TransferPropertiestoInternal (Variables|Default|GetBP_MeshGrassIOComponent) (Variables|BiomeBase|GetBiomeTexture) (Variables|Internal|GetBiomeE) (Variables|BiomeE|GetBiomeE_BiomeSetup) (Variables|BiomeE|GetBiomeE_MeshSetupVariety) (Variables|BiomeE|GetBiomeE_MeshSetupBasic) (Variables|BiomeBase|GetBiomeBase_MeshSetupVariety) (Variables|GlobalParameters|GetWeightByGrassTypeDensity) false))
      (:then_6
        (Default|TransferPropertiestoInternal (Variables|Default|GetBP_MeshGrassIOComponent) (Variables|BiomeBase|GetBiomeTexture) (Variables|Internal|GetBiomeF) (Variables|BiomeF|GetBiomeF_BiomeSetup) (Variables|BiomeF|GetBiomeF_MeshSetupVariety) (Variables|BiomeF|GetBiomeF_MeshSetupBasic) (Variables|BiomeBase|GetBiomeBase_MeshSetupVariety) (Variables|GlobalParameters|GetWeightByGrassTypeDensity) false))
      (:then_7
        (Default|TransferPropertiestoInternal (Variables|Default|GetBP_MeshGrassIOComponent) (Variables|BiomeBase|GetBiomeTexture) (Variables|Internal|GetBiomeG) (Variables|BiomeG|GetBiomeG_BiomeSetup) (Variables|BiomeG|GetBiomeG_MeshSetupVariety) (Variables|BiomeG|GetBiomeG_MeshSetupBasic) (Variables|BiomeBase|GetBiomeBase_MeshSetupVariety) (Variables|GlobalParameters|GetWeightByGrassTypeDensity) false)))
    (else
      (switch int BiomeIndex
        (:0
          _)
        (:1
          _)
        (:2
          _)
        (:3
          _)
        (:4
          _)
        (:5
          _)
        (:6
          _)
        (:7
          _))))
  (Default|TransferPropertiestoInternal :self (Variables|Default|GetBP_MeshGrassIOComponent) :BiomeTexture (Variables|BiomeBase|GetBiomeTexture) :TargetBiome (Variables|Internal|GetBiomeBase) :SourceBiomeSetup (Variables|BiomeBase|GetBiomeBase_BiomeSetup) :SourceMeshVarietySetup (Variables|BiomeBase|GetBiomeBase_MeshSetupVariety) :SourceMeshBasicSetup (Variables|BiomeBase|GetBiomeBase_MeshSetupBasic) :WeightByGrassTypeDensity (Variables|GlobalParameters|GetWeightByGrassTypeDensity) :Debug false))

(fn ConstructionScript
  (CallFunction|EnforceProperties)
  (CallFunction|PopulateBiomeData -1))
