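; Landscape mesh grass generator Blueprint whose construction script maps landscape paint layers to grass biome data for a partitioned PCG component.
; It requires a level GlobalControls actor, resets ten layer biome slots, and fills an empty layer map from the Landscape's target layers.
; The IO component converts each layer's setup using global biome texture and base variety.
; Variety setups return to the map, and biome data gets layer names and indices offset by ten.
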(event Custom|CustomEvent)

(fn ConstructionScript
  (bind _layersetup (Variables|LandscapeLayerSetup|GetLayerSetup))
  (bind _varietysetups (Variables|Internal|GetVarietySetups))
  (bind _currentindex (LocalVariables|GetCurrentIndex))
  (bind _layersetup_2 (Variables|LandscapeLayerSetup|GetLayerSetup))
  (Utilities|FlowControl|Sequence
    (:then_0
      (Variables|/Script/PCG.PCGComponent|SetbIsComponentPartitioned true (Variables|Default|GetPCG))
      (bind _returnvalue (Actor|GetActorOfClass "/Game/Athena/Helios/Environments/Landscape/WorldBuildingTools/PCGGrass/BP_MeshGrass_GlobalControls.BP_MeshGrass_GlobalControls_C"))
      (LocalVariables|SetGlobalControls _returnvalue))
    (:then_1
      (Utilities|IsValid (LocalVariables|GetGlobalControls)
        (:Is%20Valid
          (Utilities|FlowControl|Sequence
            (:then_0
              ; ClearLayerBiomes
              (Variables|Internal|SetLayerBiome_0)
              (Variables|Internal|SetLayerBiome_1)
              (Variables|Internal|SetLayerBiome_2)
              (Variables|Internal|SetLayerBiome_3)
              (Variables|Internal|SetLayerBiome_4)
              (Variables|Internal|SetLayerBiome_5)
              (Variables|Internal|SetLayerBiome_6)
              (Variables|Internal|SetLayerBiome_7)
              (Variables|Internal|SetLayerBiome_8)
              (Variables|Internal|SetLayerBiome_9)
              (if (Utilities|Map|IsEmpty (Variables|LandscapeLayerSetup|GetLayerSetup))
                (Utilities|Map|Clear (Variables|LandscapeLayerSetup|GetLayerSetup))
                (bind _returnvalue_2 (Actor|GetActorOfClass "/Script/Landscape.Landscape"))
                (bind _returnvalue_3 (Landscape|Editor|GetTargetLayerNames _returnvalue_2 false))
                (for _array_element _returnvalue_3
                  (Utilities|Map|Add (Variables|LandscapeLayerSetup|GetLayerSetup) _array_element))))
            (:then_1
              (bind _values (Utilities|Map|Values (Variables|LandscapeLayerSetup|GetLayerSetup)))
              (Utilities|Array|Clear _varietysetups)
              (for _array_element_2 _values
                (bind (_array_element_enabled_1_b2c44e894dba4a5f44ba4ab94b52dac3 _array_element_strength_4_b5df7fcb47d3175d7c9e10965f4342a2 _array_element_grasstypesetup_8_02a677c84692fe6e20b385add65a5241) (Utilities|Struct|BreakMeshGrassLandscapeLayerSetup _array_element_2))
                (Utilities|Array|Add _varietysetups _array_element_grasstypesetup_8_02a677c84692fe6e20b385add65a5241)))
            (:then_2
              (bind _values_2 (Utilities|Map|Values _layersetup_2))
              (bind _keys (Utilities|Map|Keys _layersetup_2))
              (for _index (range (+ (Utilities|Array|LastIndex _values_2) 1))
                (LocalVariables|SetCurrentIndex _index)
                (if (Variables|Default|GetDebug)
                  (bind (_output_enabled_1_b2c44e894dba4a5f44ba4ab94b52dac3 _output_strength_4_b5df7fcb47d3175d7c9e10965f4342a2 _output_grasstypesetup_8_02a677c84692fe6e20b385add65a5241) (Utilities|Struct|BreakMeshGrassLandscapeLayerSetup (Utilities|Array|Get(aref) _values_2 _currentindex)))
                  (Development|PrintString (Utilities|String|Append (Utilities|String|ToString(Name) (Utilities|Array|Get(acopy) _keys (LocalVariables|GetCurrentIndex))) "
>>" (Utilities|GetDisplayName (Utilities|Struct|BreakMeshGrassMeshSetupGrassType _output_grasstypesetup_8_02a677c84692fe6e20b385add65a5241)))))
                (Default|TransferPropertiestoInternal (Variables|Default|GetBP_MeshGrassIOComponent) (Variables|/Game/Athena/Helios/Environments/Landscape/WorldBuildingTools/PCGGrass/BP_MeshGrass_GlobalControls.BP_MeshGrass_GlobalControls_C|GetBiomeTexture (LocalVariables|GetGlobalControls)) (Utilities|GetBiomeDatabyIndex (LocalVariables|GetCurrentIndex)) (Utilities|Struct|MakeMeshGrassBiomeSetup _output_enabled_1_b2c44e894dba4a5f44ba4ab94b52dac3 -1 "NewEnumerator0" _output_strength_4_b5df7fcb47d3175d7c9e10965f4342a2) (Utilities|Array|Get(aref) (Variables|Internal|GetVarietySetups) (LocalVariables|GetCurrentIndex)) (Utilities|Struct|MakeMeshGrassMeshSetupBasic) (Variables|/Game/Athena/Helios/Environments/Landscape/WorldBuildingTools/PCGGrass/BP_MeshGrass_GlobalControls.BP_MeshGrass_GlobalControls_C|GetBiomeBase_MeshSetupVariety (LocalVariables|GetGlobalControls)) false false)))
            (:then_3
              (bind _keys_2 (Utilities|Map|Keys _layersetup))
              (bind _values_3 (Utilities|Map|Values _layersetup))
              (for (_array_element_3 _array_index) _keys_2
                (bind (_output_enabled_1_b2c44e894dba4a5f44ba4ab94b52dac3_2 _output_strength_4_b5df7fcb47d3175d7c9e10965f4342a2_2) (Utilities|Struct|BreakMeshGrassLandscapeLayerSetup (Utilities|Array|Get(aref) _values_3 _array_index)))
                (Utilities|Map|Add _layersetup _array_element_3 (Utilities|Struct|MakeMeshGrassLandscapeLayerSetup _output_enabled_1_b2c44e894dba4a5f44ba4ab94b52dac3_2 _output_strength_4_b5df7fcb47d3175d7c9e10965f4342a2_2 (Utilities|Array|Get(aref) (Variables|Internal|GetVarietySetups) _array_index)))
                (Utilities|Struct|SetmembersinMeshGrassBiomeData (Utilities|GetBiomeDatabyIndex _array_index) (+ _array_index 10) (Utilities|String|ToString(Name) _array_element_3))))))))))
