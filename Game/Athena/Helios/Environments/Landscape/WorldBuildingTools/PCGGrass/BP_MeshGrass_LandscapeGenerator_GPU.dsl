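; Landscape grass generator actor that switches between runtime PCG mesh grass and standard landscape grass.
; Construction script caches the nearest Landscape; BeginPlay and a duplicate Refresh event check non-mobile platform and Fort.PCG.LandscapeGrass, then toggle grass, PCG and an MPC flag.
; User-settings changes refresh PCG to restore lost grass; a shutdown path deactivates PCG, flushes its pool and destroys the actor.
; HandleAutomation can activate PCG for replay-test runs.
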(event EventBeginPlay
  (bind _landscape (Variables|Default|GetLandscape))
  (Fortnite|FortShippingLog 0 "[GRASS] Grass shutdown logic begin")
  (if (not (or (Utilities|String|Equal,CaseInsensitive(String) (Game|GetPlatformName) "IOS") (Utilities|String|Equal,CaseInsensitive(String) (Game|GetPlatformName) "Android")))
    (Utilities|IsValid (Variables|Default|GetLandscape)
      (:Is%20Valid
        (bind _returnvalue (Development|GetConsoleVariableBoolValue "Fort.PCG.LandscapeGrass"))
        (if _returnvalue
          (Fortnite|FortShippingLog 0 "[GRASS] Runtime PCG grass activated")
          (Landscape|EnableGrass _landscape false true)
          (Variables|/Script/PCG.PCGComponent|SetbActivated true (Variables|Default|GetPCG))
          (Rendering|Material|SetScalarParameterValue "/Game/Athena/Helios/Environments/Landscape/WorldBuildingTools/PCGGrass/Resources/MPC_MeshGrass.MPC_MeshGrass" "IsPCGLandscapeGrassActive" 1.0 :!class /Script/Engine.KismetMaterialLibrary)
          ; Grass can get lost when changing graphics settings. Refresh after settings changes.
          (bind _returnvalue_2 (Settings|GetGameUserSettings))
          (Utilities|IsValid _returnvalue_2
            (:Is%20Valid
              (EventDispatchers|BindEventtoOnGameUserSettingsUINeedsUpdate _returnvalue_2 (delegate Custom|OnUserSettingsChnaged))))
          (else
            (Fortnite|FortShippingLog 0 "[GRASS] Fallback to LS grass")
            (Landscape|EnableGrass _landscape true false))))
      (:Is%20Not%20Valid
        (Fortnite|FortShippingLog 0 "[GRASS] Fallback failed due to missing landscape")))
    (Variables|/Script/PCG.PCGComponent|SetbActivated false (Variables|Default|GetPCG))
    (Development|ExecuteConsoleCommand "pcg.RuntimeGeneration.FlushActorPool")
    (Rendering|Material|SetScalarParameterValue "/Game/Athena/Helios/Environments/Landscape/WorldBuildingTools/PCGGrass/Resources/MPC_MeshGrass.MPC_MeshGrass" "IsPCGLandscapeGrassActive" :!class /Script/Engine.KismetMaterialLibrary)
    (Actor|DestroyActor)))

(event Custom|Refresh
  (bind _landscape (Variables|Default|GetLandscape))
  (Fortnite|FortShippingLog 0 "[GRASS] Grass shutdown logic begin")
  (if (not (or (Utilities|String|Equal,CaseInsensitive(String) (Game|GetPlatformName) "IOS") (Utilities|String|Equal,CaseInsensitive(String) (Game|GetPlatformName) "Android")))
    (Utilities|IsValid (Variables|Default|GetLandscape)
      (:Is%20Valid
        (bind _returnvalue (Development|GetConsoleVariableBoolValue "Fort.PCG.LandscapeGrass"))
        (if _returnvalue
          (Fortnite|FortShippingLog 0 "[GRASS] Runtime PCG grass activated")
          (Landscape|EnableGrass _landscape false true)
          (Variables|/Script/PCG.PCGComponent|SetbActivated true (Variables|Default|GetPCG))
          (Rendering|Material|SetScalarParameterValue "/Game/Athena/Helios/Environments/Landscape/WorldBuildingTools/PCGGrass/Resources/MPC_MeshGrass.MPC_MeshGrass" "IsPCGLandscapeGrassActive" 1.0 :!class /Script/Engine.KismetMaterialLibrary)
          ; Grass can get lost when changing graphics settings. Refresh after settings changes.
          (bind _returnvalue_2 (Settings|GetGameUserSettings))
          (Utilities|IsValid _returnvalue_2
            (:Is%20Valid
              (EventDispatchers|BindEventtoOnGameUserSettingsUINeedsUpdate _returnvalue_2 (delegate Custom|OnUserSettingsChnaged))))
          (else
            (Fortnite|FortShippingLog 0 "[GRASS] Fallback to LS grass")
            (Landscape|EnableGrass _landscape true false))))
      (:Is%20Not%20Valid
        (Fortnite|FortShippingLog 0 "[GRASS] Fallback failed due to missing landscape")))
    (Variables|/Script/PCG.PCGComponent|SetbActivated false (Variables|Default|GetPCG))
    (Development|ExecuteConsoleCommand "pcg.RuntimeGeneration.FlushActorPool")
    (Rendering|Material|SetScalarParameterValue "/Game/Athena/Helios/Environments/Landscape/WorldBuildingTools/PCGGrass/Resources/MPC_MeshGrass.MPC_MeshGrass" "IsPCGLandscapeGrassActive" :!class /Script/Engine.KismetMaterialLibrary)
    (Actor|DestroyActor)))

(event Custom|OnUserSettingsChnaged
  ; Grass can get lost when changing graphics settings. Refresh after settings changes.
  (PCG|Runtime|RefreshPCGRuntimeComponent (Variables|Default|GetPCG)))

(fn DetermineTargetLandscape
  (bind _landscapecandidates (LocalVariables|GetLandscapeCandidates))
  (bind _outactors (Actor|GetAllActorsOfClass "/Script/Landscape.Landscape"))
  (for _array_element _outactors
    (Utilities|Map|Add (LocalVariables|GetLandscapeCandidates) _array_element (Math|Vector|Distance(Vector) (Transformation|GetActorLocation _array_element) (Transformation|GetActorLocation))))
  (Utilities|FlowControl|Sequence
    (:then_0
      (if (Utilities|Map|IsNotEmpty _landscapecandidates)
        (bind _values (Utilities|Map|Values _landscapecandidates))
        (bind _keys (Utilities|Map|Keys _landscapecandidates))
        (LocalVariables|SetTargetLandscape (Utilities|Array|Get(acopy) _keys (Math|Float|MinOfFloatArray _values)))))
    (:then_1
      (return (LocalVariables|GetTargetLandscape) (Utilities|IsValid (LocalVariables|GetTargetLandscape) :!class /Script/Engine.KismetSystemLibrary)))))

(fn HandleAutomation
  (bind _self self)
  (bind _returnvalue (Utilities|GetCommandLine))
  (LocalVariables|SetCL _returnvalue)
  (if (Utilities|String|Contains (LocalVariables|GetCL) "ReplayRun_PCGLandscapeGPU")
    (bind _returnvalue_2 (Development|GetConsoleVariableIntValue "grass.Enable"))
    (if (== _returnvalue_2 0)
      (Fortnite|FortShippingLog _self "Enabling PCG landscape grass!")
      (Variables|/Script/PCG.PCGComponent|SetbActivated true (Variables|Default|GetPCG))
      (else
        (Fortnite|FortShippingLog _self "Leaving PCG landscape Grass disabled!")))))

(fn ConstructionScript
  (bind (_landscape _valid) (CallFunction|DetermineTargetLandscape))
  (if _valid
    (Variables|Default|SetLandscape _landscape)))
