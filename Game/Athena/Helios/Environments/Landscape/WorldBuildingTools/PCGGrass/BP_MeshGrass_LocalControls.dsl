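; Local control Blueprint for PCG mesh grass that supplies mesh overrides and volume biome data to its MeshGrass IO component.
; The construction script calls PopulateBiomeData, which looks up the BP_MeshGrass_GlobalControls actor for the biome texture and base variety mesh setup.
; It transfers these, with the local override meshes, into the component's internal properties.
; When texture stamping is enabled, it also populates stamping from the setup and five color-named biomes.
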(event EventPopulateBiomeData
  (params (BiomeIndex :type Integer))
  (bind (_overrideparameters_enablemeshoverride_40_e504d20441b27253f371c09ac54a62bf _overrideparameters_additivemeshes_51_6bd621cf48c9674d44729485de859474 _overrideparameters_usevarietysetup_54_f8a27b6740fffe8c8959648515fa014f _overrideparameters_meshsetupvariety_41_c0fa4f8f4e8200608a9b499f78d9d88f _overrideparameters_meshsetupbasic_42_925bbdb2402a218157d94c96daa18538) (Utilities|Struct|BreakMeshGrassOverrides (Variables|GetOverrideParameters_EnableMeshOverride_40_E504D20441B27253F371C09AC54A62BF)))
  (Utilities|FlowControl|Sequence
    (:then_0
      (bind _returnvalue (Actor|GetActorOfClass "/Game/Athena/Helios/Environments/Landscape/WorldBuildingTools/PCGGrass/BP_MeshGrass_GlobalControls.BP_MeshGrass_GlobalControls_C"))
      (Default|TransferPropertiestoInternal (Variables|Default|GetBP_MeshGrassIOComponent) (Variables|/Game/Athena/Helios/Environments/Landscape/WorldBuildingTools/PCGGrass/BP_MeshGrass_GlobalControls.BP_MeshGrass_GlobalControls_C|GetBiomeTexture _returnvalue) (Variables|Internal|GetVolumeBiome) (Utilities|Struct|MakeMeshGrassBiomeSetup true -1 "NewEnumerator0" 1.0 1.0 _overrideparameters_additivemeshes_51_6bd621cf48c9674d44729485de859474 _overrideparameters_usevarietysetup_54_f8a27b6740fffe8c8959648515fa014f) _overrideparameters_meshsetupvariety_41_c0fa4f8f4e8200608a9b499f78d9d88f _overrideparameters_meshsetupbasic_42_925bbdb2402a218157d94c96daa18538 (Variables|/Game/Athena/Helios/Environments/Landscape/WorldBuildingTools/PCGGrass/BP_MeshGrass_GlobalControls.BP_MeshGrass_GlobalControls_C|GetBiomeBase_MeshSetupVariety _returnvalue) false false))
    (:then_1
      (if (Variables|TextureStamping|GetEnableTextureStamping)
        (Default|PopulateTextureStamping :self (Variables|Default|GetBP_MeshGrassIOComponent) :MeshGrass_TextureStampingSetup (Variables|TextureStamping|GetTextureStampingSetup) :BiomeRed (Variables|Internal|GetBiomeRed) :BiomeGreen (Variables|Internal|GetBiomeGreen) :BiomeBlue (Variables|Internal|GetBiomeBlue) :BiomeWhite (Variables|Internal|GetBiomeWhite) :BiomeBlack (Variables|Internal|GetBiomeBlack))))))

(fn ConstructionScript
  (CallFunction|PopulateBiomeData -1))
