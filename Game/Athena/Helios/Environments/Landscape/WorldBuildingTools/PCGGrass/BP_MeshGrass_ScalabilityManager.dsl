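; Client-side manager actor that loads or unloads the high-end mobile PCG mesh grass data layer.
; On BeginPlay, view-distance quality changes or a custom load event, it checks for iOS/Android and maps Fort.PCG.MeshGrassQuality 0/1 to Unloaded/Activated.
; A force-toggle event flip-flops the layer regardless of that setting.
; PostSpawn destroys the actor if more than one instance exists, keeping it a singleton.
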(event Custom|LoadMobileHighGrass
  (bind _returnvalue (Game|GetPlatformName))
  (Networking|SwitchHasAuthority
    (:Remote
      (Fortnite|FortShippingLog self "Attempting to load mobile high end grass...")
      (switch string _returnvalue
        (:IOS
          _)
        (:Android
          _))
      (bind _returnvalue_2 (Development|GetConsoleVariableIntValue "Fort.PCG.MeshGrassQuality"))
      (Fortnite|FortShippingLog self (Utilities|String|Append _returnvalue " - " (Utilities|String|ToString(Integer) _returnvalue_2)))
      (switch int _returnvalue_2
        (:0
          (bind _returnvalue_3 (WorldPartition|GetDataLayerManager))
          (DataLayers|SetDataLayerRuntimeState _returnvalue_3 "/Game/Athena/Helios/Environments/Landscape/WorldBuildingTools/PCGGrass/Resources/DL_MeshGrass_MobileHighEnd.DL_MeshGrass_MobileHighEnd" "Unloaded" false :!class /Script/Engine.DataLayerManager))
        (:1
          (bind _returnvalue_4 (WorldPartition|GetDataLayerManager))
          (DataLayers|SetDataLayerRuntimeState _returnvalue_4 "/Game/Athena/Helios/Environments/Landscape/WorldBuildingTools/PCGGrass/Resources/DL_MeshGrass_MobileHighEnd.DL_MeshGrass_MobileHighEnd" "Activated" false :!class /Script/Engine.DataLayerManager))))))

(event Custom|ForceToggleMobileHighGrass
  (Networking|SwitchHasAuthority
    (:Remote
      (Fortnite|FortShippingLog self "Force toggle mobile high end grass...")
      (Utilities|FlowControl|FlipFlop
        (:A
          (bind _returnvalue (WorldPartition|GetDataLayerManager))
          (DataLayers|SetDataLayerRuntimeState _returnvalue "/Game/Athena/Helios/Environments/Landscape/WorldBuildingTools/PCGGrass/Resources/DL_MeshGrass_MobileHighEnd.DL_MeshGrass_MobileHighEnd" "Activated" false :!class /Script/Engine.DataLayerManager))
        (:B
          (bind _returnvalue_2 (WorldPartition|GetDataLayerManager))
          (DataLayers|SetDataLayerRuntimeState _returnvalue_2 "/Game/Athena/Helios/Environments/Landscape/WorldBuildingTools/PCGGrass/Resources/DL_MeshGrass_MobileHighEnd.DL_MeshGrass_MobileHighEnd" "Unloaded" false :!class /Script/Engine.DataLayerManager))))))

(event EventBeginPlay
  (bind _returnvalue (Game|GetPlatformName))
  (Networking|SwitchHasAuthority
    (:Remote
      (EventDispatchers|BindEventtoOnViewDistanceQualityChanged (Settings|GetFortGameUserSettings) (delegate Custom|OnQualityChanged))
      (switch string _returnvalue
        (:IOS
          _)
        (:Android
          _))
      (bind _returnvalue_2 (Development|GetConsoleVariableIntValue "Fort.PCG.MeshGrassQuality"))
      (Fortnite|FortShippingLog self (Utilities|String|Append _returnvalue " - " (Utilities|String|ToString(Integer) _returnvalue_2)))
      (switch int _returnvalue_2
        (:0
          (bind _returnvalue_3 (WorldPartition|GetDataLayerManager))
          (DataLayers|SetDataLayerRuntimeState _returnvalue_3 "/Game/Athena/Helios/Environments/Landscape/WorldBuildingTools/PCGGrass/Resources/DL_MeshGrass_MobileHighEnd.DL_MeshGrass_MobileHighEnd" "Unloaded" false :!class /Script/Engine.DataLayerManager))
        (:1
          (bind _returnvalue_4 (WorldPartition|GetDataLayerManager))
          (DataLayers|SetDataLayerRuntimeState _returnvalue_4 "/Game/Athena/Helios/Environments/Landscape/WorldBuildingTools/PCGGrass/Resources/DL_MeshGrass_MobileHighEnd.DL_MeshGrass_MobileHighEnd" "Activated" false :!class /Script/Engine.DataLayerManager))))))

(event Custom|OnQualityChanged
  (bind _returnvalue (Game|GetPlatformName))
  (switch string _returnvalue
    (:IOS
      _)
    (:Android
      _))
  (bind _returnvalue_2 (Development|GetConsoleVariableIntValue "Fort.PCG.MeshGrassQuality"))
  (Fortnite|FortShippingLog self (Utilities|String|Append _returnvalue " - " (Utilities|String|ToString(Integer) _returnvalue_2)))
  (switch int _returnvalue_2
    (:0
      (bind _returnvalue_3 (WorldPartition|GetDataLayerManager))
      (DataLayers|SetDataLayerRuntimeState _returnvalue_3 "/Game/Athena/Helios/Environments/Landscape/WorldBuildingTools/PCGGrass/Resources/DL_MeshGrass_MobileHighEnd.DL_MeshGrass_MobileHighEnd" "Unloaded" false :!class /Script/Engine.DataLayerManager))
    (:1
      (bind _returnvalue_4 (WorldPartition|GetDataLayerManager))
      (DataLayers|SetDataLayerRuntimeState _returnvalue_4 "/Game/Athena/Helios/Environments/Landscape/WorldBuildingTools/PCGGrass/Resources/DL_MeshGrass_MobileHighEnd.DL_MeshGrass_MobileHighEnd" "Activated" false :!class /Script/Engine.DataLayerManager))))

(fn PostSpawn
  (bind _outactors (Actor|GetAllActorsOfClass "/Game/Athena/Helios/Environments/Landscape/WorldBuildingTools/PCGGrass/BP_MeshGrass_ScalabilityManager.BP_MeshGrass_ScalabilityManager_C"))
  (if (> (Utilities|Array|Length _outactors) 1)
    (Actor|DestroyActor)))

(fn ConstructionScript)
