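; Spline-driven world-building actor that scatters PCG mesh grass, using biome data from the global mesh grass controls actor.
; Construction copies the spline setup and global biome texture and base setup into the MeshGrass IO component.
; It activates PCG only when active and the chosen variety or basic mesh setup is non-empty.
; When activated, it resets actor scale and sizes the bounds box from spline bounds, maximum point width scale and spawn distance.
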(event EventPopulateBiomeData
  (params (BiomeIndex :type Integer))
  (bind (_projectionmode_85_fd84146842fb9e8d41eef8aa89181479 _forcepreciseprojection_93_5a4f531549c126d66e410485f4bd9d04 _visualizeprojection_88_cfc5bc944875a49c9d8b6ab8ca4e41b0 _usevarietysetup_54_f8a27b6740fffe8c8959648515fa014f) (Utilities|Struct|BreakMeshGrassSplineSamplerParams (Variables|SplineSampler|GetSpline%20Setup)))
  (bind (_spawndistance_72_5ce041ac44400dbfd67596a943fae585 _spawnradius_74_c74a003b480f7048ddfe54ab523b94e1 _samplinginterval_79_122eab26433897a89962f4af587fdd6d _orientationmode_82_c416461340c8c75a4f615f8f4ef4616a _projectionmode_85_fd84146842fb9e8d41eef8aa89181479_2 _forcepreciseprojection_93_5a4f531549c126d66e410485f4bd9d04_2 _visualizeprojection_88_cfc5bc944875a49c9d8b6ab8ca4e41b0_2 _meshsetupvariety_41_c0fa4f8f4e8200608a9b499f78d9d88f _meshsetupbasic_42_925bbdb2402a218157d94c96daa18538) (Utilities|Struct|BreakMeshGrassSplineSamplerParams (Variables|SplineSampler|GetSpline%20Setup)))
  (bind _returnvalue (Actor|GetActorOfClass "/Game/Athena/Helios/Environments/Landscape/WorldBuildingTools/PCGGrass/BP_MeshGrass_GlobalControls.BP_MeshGrass_GlobalControls_C"))
  (Default|TransferPropertiestoInternal (Variables|Default|GetBP_MeshGrassIOComponent) (Variables|/Game/Athena/Helios/Environments/Landscape/WorldBuildingTools/PCGGrass/BP_MeshGrass_GlobalControls.BP_MeshGrass_GlobalControls_C|GetBiomeTexture _returnvalue) (Variables|Internal|GetBiome) (Utilities|Struct|MakeMeshGrassBiomeSetup (Variables|SplineSampler|GetIsActive) -1 "NewEnumerator0" 1.0 1.0 false _usevarietysetup_54_f8a27b6740fffe8c8959648515fa014f) _meshsetupvariety_41_c0fa4f8f4e8200608a9b499f78d9d88f _meshsetupbasic_42_925bbdb2402a218157d94c96daa18538 (Variables|/Game/Athena/Helios/Environments/Landscape/WorldBuildingTools/PCGGrass/BP_MeshGrass_GlobalControls.BP_MeshGrass_GlobalControls_C|GetBiomeBase_MeshSetupVariety _returnvalue) false false))

(fn DetermineSplineWidth
  (bind (_indexofmaxvalue _maxvalue) (Math|Float|MaxOfFloatArray (LocalVariables|GetSplinePointWidths)))
  (bind _splinepointwidths (LocalVariables|GetSplinePointWidths))
  (Utilities|Array|Resize _splinepointwidths (Spline|GetNumberOfSplinePoints (Variables|Default|GetSpline)))
  (for _index (range (+ (Utilities|Array|LastIndex _splinepointwidths) 1))
    (bind (_returnvalue_x _returnvalue_y) (Math|Vector|BreakVector (Spline|GetScaleatSplinePoint (Variables|Default|GetSpline) _index)))
    (Utilities|Array|SetArrayElem (LocalVariables|GetSplinePointWidths) _index _returnvalue_y))
  (return _maxvalue))

(fn GatherPCGData
  (bind (_projectionmode_85_fd84146842fb9e8d41eef8aa89181479 _forcepreciseprojection_93_5a4f531549c126d66e410485f4bd9d04 _visualizeprojection_88_cfc5bc944875a49c9d8b6ab8ca4e41b0 _usevarietysetup_54_f8a27b6740fffe8c8959648515fa014f) (Utilities|Struct|BreakMeshGrassSplineSamplerParams (Variables|SplineSampler|GetSpline%20Setup)))
  (bind (_spawndistance_72_5ce041ac44400dbfd67596a943fae585 _spawnradius_74_c74a003b480f7048ddfe54ab523b94e1 _samplinginterval_79_122eab26433897a89962f4af587fdd6d _orientationmode_82_c416461340c8c75a4f615f8f4ef4616a _projectionmode_85_fd84146842fb9e8d41eef8aa89181479_2 _forcepreciseprojection_93_5a4f531549c126d66e410485f4bd9d04_2 _visualizeprojection_88_cfc5bc944875a49c9d8b6ab8ca4e41b0_2 _meshsetupvariety_41_c0fa4f8f4e8200608a9b499f78d9d88f _meshsetupbasic_42_925bbdb2402a218157d94c96daa18538) (Utilities|Struct|BreakMeshGrassSplineSamplerParams (Variables|SplineSampler|GetSpline%20Setup)))
  (bind _returnvalue (Actor|GetActorOfClass "/Game/Athena/Helios/Environments/Landscape/WorldBuildingTools/PCGGrass/BP_MeshGrass_GlobalControls.BP_MeshGrass_GlobalControls_C"))
  (Default|TransferPropertiestoInternal (Variables|Default|GetBP_MeshGrassIOComponent) (Variables|/Game/Athena/Helios/Environments/Landscape/WorldBuildingTools/PCGGrass/BP_MeshGrass_GlobalControls.BP_MeshGrass_GlobalControls_C|GetBiomeTexture _returnvalue) (Variables|Internal|GetBiome) (Utilities|Struct|MakeMeshGrassBiomeSetup (Variables|SplineSampler|GetIsActive) -1 "NewEnumerator0" 1.0 1.0 false _usevarietysetup_54_f8a27b6740fffe8c8959648515fa014f) _meshsetupvariety_41_c0fa4f8f4e8200608a9b499f78d9d88f _meshsetupbasic_42_925bbdb2402a218157d94c96daa18538 (Variables|/Game/Athena/Helios/Environments/Landscape/WorldBuildingTools/PCGGrass/BP_MeshGrass_GlobalControls.BP_MeshGrass_GlobalControls_C|GetBiomeBase_MeshSetupVariety _returnvalue) false false))

(fn UpdateBounds
  (bind _spline_setup (Variables|SplineSampler|GetSpline%20Setup))
  (bind _spawndistance_72_5ce041ac44400dbfd67596a943fae585 (Utilities|Struct|BreakMeshGrassSplineSamplerParams _spline_setup))
  (bind _spline (Variables|Default|GetSpline))
  (bind (_returnvalue_origin _returnvalue_boxextent) (Math|BoxSphereBounds|BreakBoxSphereBounds (Fortnite|Utility|GetLocalBounds _spline)))
  (bind _z (.z _returnvalue_boxextent))
  (bind _returnvalue (/ _spawndistance_72_5ce041ac44400dbfd67596a943fae585 2.0))
  (bind _box (Variables|Default|GetBox))
  (bind _splinewidth (CallFunction|DetermineSplineWidth))
  (Components|Box|SetBoxExtent _box (Math|Vector|MakeVector (+ (.x _returnvalue_boxextent) _splinewidth) (+ (.y _returnvalue_boxextent) _splinewidth) (+ _z _returnvalue)) false)
  (Transformation|SetRelativeLocation _box (Math|Vector|MakeVector (.x _returnvalue_origin) (.y _returnvalue_origin) (- _z _returnvalue)) false true))

(fn ConstructionScript
  (bind (_spawndistance_72_5ce041ac44400dbfd67596a943fae585 _spawnradius_74_c74a003b480f7048ddfe54ab523b94e1 _samplinginterval_79_122eab26433897a89962f4af587fdd6d _orientationmode_82_c416461340c8c75a4f615f8f4ef4616a _projectionmode_85_fd84146842fb9e8d41eef8aa89181479 _forcepreciseprojection_93_5a4f531549c126d66e410485f4bd9d04 _visualizeprojection_88_cfc5bc944875a49c9d8b6ab8ca4e41b0 _usevarietysetup_54_f8a27b6740fffe8c8959648515fa014f _meshsetupvariety_41_c0fa4f8f4e8200608a9b499f78d9d88f _meshsetupbasic_42_925bbdb2402a218157d94c96daa18538) (Utilities|Struct|BreakMeshGrassSplineSamplerParams (Variables|SplineSampler|GetSpline%20Setup)))
  (bind (_grasstype_14_54bc33c9418d25a265c3f093e9fdbd6a _secondarygrasstype_18_b2fa945b46b827645232db91db988401 _varietiespreviewweighting_13_bdad33c040d683be153b09b81c9dd9b4) (Utilities|Struct|BreakMeshGrassMeshSetupGrassType _meshsetupvariety_41_c0fa4f8f4e8200608a9b499f78d9d88f))
  (Utilities|FlowControl|Sequence
    (:then_0
      (CallFunction|PopulateBiomeData -1))
    (:then_1
      (bind _output_get (Variables|/Script/PCG.PCGComponent|SetbActivated (and (Variables|SplineSampler|GetIsActive) (or (and (Utilities|Array|IsNotEmpty _varietiespreviewweighting_13_bdad33c040d683be153b09b81c9dd9b4) _usevarietysetup_54_f8a27b6740fffe8c8959648515fa014f) (and (not _usevarietysetup_54_f8a27b6740fffe8c8959648515fa014f) (Utilities|Array|IsNotEmpty (Utilities|Struct|BreakMeshGrassMeshSetupBasic _meshsetupbasic_42_925bbdb2402a218157d94c96daa18538))))) (Variables|Default|GetPCG)))
      (if _output_get
        (Transformation|SetActorScale3D "1.000000,1.000000,1.000000")
        (Internal|UpdateBounds)))))
