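; BP_MeshGrass_TextureGenerator is a world-building Blueprint that sets up PCG mesh grass from a texture stamping setup.
; Construction activates the PCG component according to IsActive, then has the IO component populate red, green, blue, white and black biome data.
; PopulateVarieties reads the biome texture and base variety from BP_MeshGrass_GlobalControls, runs each color's mesh setup through PopulateBiome, and writes results back to the stamping setup.
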(event EventPopulateBiomeData
  (params (BiomeIndex :type Integer))
  (Default|PopulateTextureStamping (Variables|Default|GetBP_MeshGrassIOComponent) (Variables|TextureSamplingSetup|GetTextureStampingSetup) (Variables|Internal|GetBiomeRed) (Variables|Internal|GetBiomeGreen) (Variables|Internal|GetBiomeBlue) (Variables|Internal|GetBiomeWhite) (Variables|Internal|GetBiomeBlack) (Utilities|Struct|MakeMeshGrassBiomeSetup true -1)))

(fn PopulateBiome
  (params (MeshGrass_MeshSetup_Combined :type Mesh%20Grass%20Mesh%20Setup%20Combined%20Structure%20%28by%20ref%29) (TargetBiome :type Mesh%20Grass%20Biome%20Data%20Structure%20%28by%20ref%29))
  (bind (_grasstype_14_54bc33c9418d25a265c3f093e9fdbd6a _secondarygrasstype_18_b2fa945b46b827645232db91db988401 _varietiespreviewweighting_13_bdad33c040d683be153b09b81c9dd9b4 _additionalmeshes_23_73d5b7ea4437df882c1aa8a473def8d1) (Utilities|Struct|BreakMeshGrassMeshSetupCombined MeshGrass_MeshSetup_Combined))
  (bind (_grasstype_14_54bc33c9418d25a265c3f093e9fdbd6a_2 _secondarygrasstype_18_b2fa945b46b827645232db91db988401_2 _varietiespreviewweighting_13_bdad33c040d683be153b09b81c9dd9b4_2) (Utilities|Struct|BreakMeshGrassMeshSetupGrassType (LocalVariables|GetVarieties)))
  (Variables|SetVarieties_GrassType_14_54BC33C9418D25A265C3F093E9FDBD6A (Utilities|Struct|MakeMeshGrassMeshSetupGrassType _grasstype_14_54bc33c9418d25a265c3f093e9fdbd6a _secondarygrasstype_18_b2fa945b46b827645232db91db988401 _varietiespreviewweighting_13_bdad33c040d683be153b09b81c9dd9b4))
  (Variables|SetMeshes_MeshSetup_12_16D9329C456F5D0B2C114D8B8AC0C078 (Utilities|Struct|MakeMeshGrassMeshSetupBasic _additionalmeshes_23_73d5b7ea4437df882c1aa8a473def8d1))
  (Default|TransferPropertiestoInternal (Variables|Default|GetBP_MeshGrassIOComponent) (Variables|Internal|GetBiomeTexture) TargetBiome (Utilities|Struct|MakeMeshGrassBiomeSetup true -1) (LocalVariables|GetVarieties) (LocalVariables|GetMeshes) (Variables|Internal|GetBiomeBaseVarietySetup) false false)
  (return (Utilities|Struct|MakeMeshGrassMeshSetupCombined _grasstype_14_54bc33c9418d25a265c3f093e9fdbd6a_2 _secondarygrasstype_18_b2fa945b46b827645232db91db988401_2 _varietiespreviewweighting_13_bdad33c040d683be153b09b81c9dd9b4_2 (Utilities|Struct|BreakMeshGrassMeshSetupBasic (LocalVariables|GetMeshes)))))

(fn PopulateVarieties
  (bind _texturestampingsetup (Variables|TextureSamplingSetup|GetTextureStampingSetup))
  (bind (_stamptexture_2_b94e06994c0d03725f8a64b671efa204 _luminanceasalpha_44_101ed4674bbbf9e97a4d15b0348871c7 _tiling_47_69d6a0814aba74972479c8bc806a5a5a _texelsize_36_66182b4945a872bc0f6b59b3d42e3a14 _masktexture_54_7b514b3d467c4fb4a50ef8a9bf3aaf76 _masktexturechannel_55_b727a1334cef9d5209a1ec9839d714ee _maskingthreshold_57_2a4447514b69f09186a695b2b20a0cec _invertmask_60_1353ff9f4811c570e18bdba226708c34 _exclusivechannels_62_23102ba140d2dffb7d4174b6044ec5dc _meshsetupred_8_e0556f6b44bb52696061c185d1ede805 _meshsetupgreen_10_f89b5c7748b961f3a233c08a44a2e254 _meshsetupblue_12_fbd29ddc49fd0559fd443788568cf28d _useblackwhitesetup_29_9305f69246d0379beaf30890c8c8aca9 _blackwhitethresholds_31_a1a19205447c1aa65ee64fb81f2973de _meshsetupblack_16_9f884cb64cfc86830d1e4ebae2b9756d _meshsetupwhite_14_c6a06f1a44eb0a82cb96f8aa895f8678) (Utilities|Struct|BreakMeshGrassTextureStampingSetup _texturestampingsetup))
  (Utilities|FlowControl|Sequence
    (:then_0
      (bind _returnvalue (Actor|GetActorOfClass "/Game/Athena/Helios/Environments/Landscape/WorldBuildingTools/PCGGrass/BP_MeshGrass_GlobalControls.BP_MeshGrass_GlobalControls_C"))
      (Variables|Internal|SetBiomeTexture (Variables|/Game/Athena/Helios/Environments/Landscape/WorldBuildingTools/PCGGrass/BP_MeshGrass_GlobalControls.BP_MeshGrass_GlobalControls_C|GetBiomeTexture _returnvalue))
      (Variables|Internal|SetBiomeBaseVarietySetup (Variables|/Game/Athena/Helios/Environments/Landscape/WorldBuildingTools/PCGGrass/BP_MeshGrass_GlobalControls.BP_MeshGrass_GlobalControls_C|GetBiomeBase_MeshSetupVariety _returnvalue)))
    (:then_1
      (bind _meshsetupcombined (CallFunction|PopulateBiome _meshsetupblack_16_9f884cb64cfc86830d1e4ebae2b9756d (Variables|Internal|GetBiomeBlack)))
      (bind _meshsetupcombined_2 (CallFunction|PopulateBiome _meshsetupwhite_14_c6a06f1a44eb0a82cb96f8aa895f8678 (Variables|Internal|GetBiomeWhite)))
      (bind _meshsetupcombined_3 (CallFunction|PopulateBiome _meshsetupblue_12_fbd29ddc49fd0559fd443788568cf28d (Variables|Internal|GetBiomeBlue)))
      (bind _meshsetupcombined_4 (CallFunction|PopulateBiome _meshsetupgreen_10_f89b5c7748b961f3a233c08a44a2e254 (Variables|Internal|GetBiomeGreen)))
      (bind _meshsetupcombined_5 (CallFunction|PopulateBiome _meshsetupred_8_e0556f6b44bb52696061c185d1ede805 (Variables|Internal|GetBiomeRed)))
      (Utilities|Struct|SetmembersinMeshGrassTextureStampingSetup (Variables|TextureSamplingSetup|GetTextureStampingSetup) _meshsetupcombined_5 _meshsetupcombined_4 _meshsetupcombined_3 _meshsetupcombined _meshsetupcombined_2))))

(fn ConstructionScript
  (Utilities|FlowControl|Sequence
    (:then_0
      (Variables|/Script/PCG.PCGComponent|SetbActivated (Variables|TextureSamplingSetup|GetIsActive) (Variables|Default|GetPCG)))
    (:then_1
      (CallFunction|PopulateBiomeData -1))))
