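# Editor utility behind the mesh grass world-building tool, post-processing spawned ISMs, adding ISM components and setting spline visualization width.
# PostSpawn uses custom depth stencil values to set SDP collision profiles and tags, then swaps in "_NoRVT" material variants where found.
# AddISM creates and renames components through SubobjectDataSubsystem, diffing component lists to find the new one.
# Helpers decode mesh-to-material strings and find non-Nanite materials via Asset Registry referencers.
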EditorUtilityBlueprint /Game/Athena/Helios/Environments/Landscape/WorldBuildingTools/PCGGrass/EUBP_MeshGrassUtilityComponent
graphs [
 DecodeMeshMaterialMapString {nodes [
   n1 = DecodeMeshMaterialMapString!() -> (then, InString:String) {is=title K2Node_FunctionEntry}
   > n2 = Set MeshMaterialStringArray(MeshMaterialStringArray:Array of Strings=ParseIntoArray(SourceString=@n1, Delimiter=|)) -> (Output_Get:Array of Strings)
   > n3 = For Each Loop(Array=@n2)
   > n4 = SplitStringCached!(execute, self, SourceString:String=@n3, InStr:String=:) -> (then, ReturnValue:bool, LeftS:String, RightS:String)
   > Branch(Condition=@n4)
   > n5 = LoadAsset_Blocking(Asset=To Soft Object Reference(SoftObjectPath=MakeSoftObjectPath!(PathString:String=@n4.LeftS) -> (ReturnValue:Soft Object Path Structure)))
   > n6 = Cast To StaticMesh(Object=@n5)
   > n7 = LoadAsset_Blocking(Asset=To Soft Object Reference(SoftObjectPath=MakeSoftObjectPath(PathString=@n4.RightS)))
   > n8 = Cast To MaterialInterface!(execute, Object=@n7) -> (then, CastFailed, AsMaterial Interface:Material Interface Object Reference) {is=title K2Node_DynamicCast}
   > Add!(
    execute
    TargetMap:Map of Static Mesh Object References to Material Interface Object References=Get MeshMaterialMap() -> (MeshMaterialMap:Map of Static Mesh Object References to Material Interface Object References)
    Key:"Static Mesh Object Reference (by ref)"=@n6
    Value:"Material Interface Object Reference (by ref)"=@n8
   ) -> (then) {is=title K2Node_CallFunction}
  ]}
 EventGraph {nodes [
   n1 = Event UpdateSpline!() -> (OutputDelegate, then, SplineComponent:Spline Component Object Reference, Width:double) {is=title K2Node_Event}
   > Is Valid(InputObject=@n1.SplineComponent)
   > SetEditorProperty(Object=@n1.SplineComponent, PropertyName=ScaleVisualizationWidth, PropertyValue:double=@n1.Width, ChangeNotifyMode=Never)
   n2 = Event AddISM!() -> (OutputDelegate, then, TargetActor:Actor Object Reference, Mesh:Static Mesh Object Reference) {is=title K2Node_Event}
   > HandleAddISM!(execute, self, TargetActor:Actor Object Reference=@n2.TargetActor, Mesh:Static Mesh Object Reference=@n2.Mesh) -> (then)
   n3 = Event PostSpawn!() -> (OutputDelegate, then, ISMs:Array of Instanced Static Mesh Component Object References) {is=title K2Node_Event}
   > "PrintDebugString(Unattended)"!(
    execute
    self
    DebugString:String=EUBP_MeshGrassUtilityComponent::EventPostSpawn
    Verbosity:EMeshGrassLogLevel Enum=NewEnumerator1
    Active:bool=true
   ) -> (then)
   > n4 = For Each Loop(Array:Array of Instanced Static Mesh Component Object References=@n3.ISMs) -> (Array Element:Instanced Static Mesh Component Object Reference)
   > HandleSDP!(execute, self, ISM:Instanced Static Mesh Component Object Reference=@n4) -> (then)
   n5 = GetNumMaterials!(self:Primitive Component Object Reference=@n4) -> (ReturnValue:int=0)
   > n6 = For Loop(LastIndex="Max (Integer)"(A=int - int(A=@n5), B=0))
   > TryReplaceMaterialNoRVT!(execute, self, TargetISM:Instanced Static Mesh Component Object Reference=@n4, MaterialIndex:int=@n6) -> (then)
  ]}
 FindNonNaniteMaterial {nodes [
   n1 = FindNonNaniteMaterial!() -> (then, NaniteMaterial:Object Reference) {is=title K2Node_FunctionEntry}
   > n2 = GetPathNameForLoadedAsset(LoadedAsset=@n1)
   > Set NaniteMaterialFolderPath(NaniteMaterialFolderPath:String=GetPath!(InPath:String=@n2) -> (ReturnValue:String)) -> (Output_Get:String)
   n3 = Get NaniteMaterialFolderPath() -> (NaniteMaterialFolderPath:String)
   n4 = FindSubstring(SearchIn=@n3, Substring=/, StartPosition=1)
   > n5 = Set NaniteMaterialFolderRoot(NaniteMaterialFolderRoot:String=Left!(SourceString:String=@n3, Count:int=int + int(A=@n4)) -> (ReturnValue:String)) -> (Output_Get:String)
   > Set NaniteMaterialIsPluginContent(NaniteMaterialIsPluginContent:bool="Not Equal, Case Insenstive (String)"!(A:String=@n5, B:String=/Game/) -> (ReturnValue:bool)) -> (Output_Get:bool)
   n6 = GetAssetRegistry()
   > n7 = Get Referencers!(
    execute
    self:Asset Registry Interface=@n6
    PackageName:Name=String To Name(InString=GetBaseFilename(InPath=@n2, bRemovePath=false))
    ReferenceOptions:"Asset Registry Dependency Options Structure (by ref)"=Make Asset Registry Dependency Options(bIncludeSoftPackageReferences=false)
   ) -> (then, OutReferencers:Array of Names, ReturnValue:bool) {is=title K2Node_CallFunction}
   > Branch(Condition=@n7.ReturnValue)
   > n8 = For Each Loop(Array:Array of Names=@n7) -> (Array Element:Name)
   n9 = For Each Loop(
    Exec=@n8.Completed
    Array:Array of Material Interface Object References=Get NonNaniteCandidates() -> (NonNaniteCandidates:Array of Material Interface Object References)
   ) -> (Array Element:Material Interface Object Reference)
   n10 = For Each Loop with Break!(
    Exec:Exec=@n9.Completed
    Array:Array of Material Interface Object References=Get NonNaniteMatches() -> (NonNaniteMatches:Array of Material Interface Object References)
    Break:Exec=@n19.then
   ) -> (LoopBody:Exec, Array Element:Material Interface Object Reference, Array Index:int, Completed:Exec)
   n11 = Get NonNaniteMatches()
   n12 = IsValid(Object=Get Material_NonNanite() -> (Material_NonNanite:Material Interface Object Reference))
   n13 = Branch(
    execute=@n10.Completed
    Condition=AND Boolean(A=Is Not Empty!(TargetArray:Array of Material Interface Object References=@n11) -> (ReturnValue:bool), B=NOT Boolean(A=@n12))
   )
   > Set Material_NonNanite(Material_NonNanite:Material Interface Object Reference="Get (a copy)"(Array:Array of Material Interface Object References=@n11) -> (Output:Material Interface Object Reference)) -> (Output_Get:Material Interface Object Reference)
   > n14 = Sequence()
   > Branch(Condition=Get Debug() -> (Debug:bool))
   n15 = Get NaniteMaterial() -> (NaniteMaterial:Object Reference)
   n16 = Get Material_NonNanite()
   > PrintString(InString=Append!(A="Found NonNanite material (", B=GetDisplayName(Object=@n15), C:String="->", D:String=GetDisplayName(Object=@n16), E:String=")") -> (ReturnValue))
   n17 = Get Material_NonNanite()
   Return Node!(execute=[@n14.then_1, @n13.else], Success:bool=IsValid(Object=@n17), NonNaniteMaterial:Material Interface Object Reference=@n17)
   n18 = GetPathNameForLoadedAsset(execute=@n10.LoopBody, LoadedAsset=@n10)
   > Branch(Condition="Equal Exactly (String)"(A=GetPath(InPath=@n18), B=Get NaniteMaterialFolderPath()))
   > n19 = Set Material_NonNanite(Material_NonNanite=@n10)
   Branch(
    execute=@n9.LoopBody
    Condition="Equal (Object)"(
     A=Get NaniteMaterial()
     B=GetNaniteOverideMaterial!(self:Material Interface Object Reference=@n9) -> (ReturnValue:Material Interface Object Reference)
    )
   )
   > Add(TargetArray:Array of Material Interface Object References=Get NonNaniteMatches(), NewItem:"Material Interface Object Reference (by ref)"=@n9) {is=title K2Node_CallArrayFunction}
   n20 = GetAssetsByPackageName!(
    execute=@n8.LoopBody
    self:Asset Registry Interface=@n6
    PackageName:Name=@n8
    bIncludeOnlyOnDiskAssets:bool=true
    bSkipARFilteredAssets:bool=false
   ) -> (then, OutAssetData:Array of Asset Data Structures, ReturnValue:bool)
   > Branch(Condition=@n20.ReturnValue)
   n21 = "Get (a copy)"!(Array:Array of Asset Data Structures=@n20, Dimension 1) -> (Output_PackageName:Name, Output_PackagePath:Name, Output_AssetName:Name, Output_AssetClassPath_PathString:String)
   > Branch(Condition="Equal Exactly (String)"(
     A=@n21.Output_AssetClassPath_PathString
     B="Get Class Path Name (Top Level Asset Path)"!(Class:Object Class Reference=/Script/Engine.MaterialInstanceConstant) -> (ReturnValue_PathString:String) {is=title K2Node_CallFunction}
    ))
   > n22 = Branch(Condition=Get NaniteMaterialIsPluginContent() -> (NaniteMaterialIsPluginContent:bool))
   n23 = "To String (Name)"(InName=@n21.Output_PackagePath)
   n24 = StartsWith(SourceString=@n23, InPrefix=/Game/)
   n25 = Get NaniteMaterialFolderRoot() -> (NaniteMaterialFolderRoot:String)
   > n26 = Branch(Condition=OR Boolean(A=@n24, B=StartsWith(SourceString=@n23, InPrefix=@n25)))
   n27 = Branch(execute=[@n32.then, @n26.then], Condition=Get Debug())
   n28 = "To String (Name)"(InName=@n8)
   > n29 = PrintString(InString=@n28)
   n30 = LoadAsset(execute=[@n29.then, @n27.else], AssetPath=@n28)
   > n31 = Cast To MaterialInterface(Object=@n30)
   > Add(TargetArray=Get NonNaniteCandidates(), NewItem=@n31)
   n32 = Branch(execute=@n22.else, Condition=@n24)
  ]}
 HandleAddISM {nodes [
   n1 = HandleAddISM!() -> (then, TargetActor:Actor Object Reference, Mesh:Static Mesh Object Reference) {is=title K2Node_FunctionEntry}
   > Set PreISMs(PreISMs:Array of Instanced Static Mesh Component Object References=Get Components By Class!(self:Actor Object Reference=@n1, ComponentClass:Actor Component Class Reference=/Script/Engine.InstancedStaticMeshComponent) -> (ReturnValue:Array of Instanced Static Mesh Component Object References)) -> (Output_Get:Array of Instanced Static Mesh Component Object References)
   n2 = Get SubobjectDataSubsystem()
   > n3 = Gather Subobject Data For Instance(self=@n2, Context=Get TargetActor() -> (TargetActor:Actor Object Reference))
   > n4 = AddNewSubobject!(
    execute
    self:Subobject Data Subsystem Object Reference=@n2
    Params:"Add New Subobject Params Structure (by ref)"=Make Add New Subobject Params!(
     ParentHandle:Subobject Data Handle Structure="Get (a ref)"(Array:Array of Subobject Data Handle Structures=@n3, Dimension 1=0) -> (Output:"Subobject Data Handle Structure (by ref)")
     NewClass:Object Class Reference=&/BP_MeshGrassInstancedStaticMeshComponent.BP_MeshGrassInstancedStaticMeshComponent_C
     BlueprintContext:Blueprint Object Reference
     bSkipMarkBlueprintModified:bool=False
     bConformTransformToParent:bool=True
    ) -> (AddNewSubobjectParams:Add New Subobject Params Structure) {is=title K2Node_MakeStruct}
   ) -> (then, FailReason:Text, ReturnValue:Subobject Data Handle Structure)
   > Set NewHandle(NewHandle:Subobject Data Handle Structure=@n4.ReturnValue) -> (Output_Get:Subobject Data Handle Structure)
   > n5 = For Each Loop with Break(
    Array:Array of Instanced Static Mesh Component Object References=Get Components By Class(self=Get TargetActor(), ComponentClass=&InstancedStaticMeshComponent)
    Break=@n7.then_1
   ) -> (Array Element:Instanced Static Mesh Component Object Reference)
   n6 = Get PreISMs() -> (PreISMs:Array of Instanced Static Mesh Component Object References)
   > Branch(Condition="Equal (Integer)"(
     A=Find Item!(
      TargetArray:Array of Instanced Static Mesh Component Object References=@n6
      ItemToFind:"Instanced Static Mesh Component Object Reference (by ref)"=@n5
     ) -> (ReturnValue)
     B=-1
    ))
   > Set NewISM(NewISM:Instanced Static Mesh Component Object Reference=@n5) -> (Output_Get:Instanced Static Mesh Component Object Reference)
   > n7 = Sequence()
   n8 = Get NewISM() -> (NewISM:Instanced Static Mesh Component Object Reference)
   > Set bHasPerInstanceHitProxies!(execute, bHasPerInstanceHitProxies:bool=true, self:Primitive Component Object Reference=@n8) -> (then, Output_Get:bool)
   > SetCastShadow!(execute, self:Primitive Component Object Reference=@n8, NewCastShadow:bool=false) -> (then)
   > SetStaticMesh(self=@n8, NewMesh=Get Mesh!() -> (Mesh:Static Mesh Object Reference))
   > AddInstance!(
    execute
    self:Instanced Static Mesh Component Object Reference=@n8
    InstanceTransform:"Transform (by ref)"=MakeTransform(Rotation="0, 0, 0")
    bWorldSpace:bool=false
   ) -> (then, ReturnValue:int=0)
   n9 = "To String (Object)"!(InObj:Object Reference=Get Mesh()) -> (ReturnValue:String)
   n10 = "To String (Integer)"(InInt=Length(TargetArray:Array of Instanced Static Mesh Component Object References=Get PreISMs()))
   > RenameSubobject!(
    execute
    self:Subobject Data Subsystem Object Reference=Get SubobjectDataSubsystem()
    Handle:"Subobject Data Handle Structure (by ref)"=Get NewHandle() -> (NewHandle:Subobject Data Handle Structure)
    InNewName:"Text (by ref)"="To Text (String)"(InString=Append!(A=@n10, B=_, C=@n9) -> (ReturnValue))
   ) -> (then, ReturnValue:bool)
  ]}
 HandleSDP {nodes [
   n1 = HandleSDP!() -> (then, ISM:Instanced Static Mesh Component Object Reference) {is=title K2Node_FunctionEntry}
   n2 = Get CustomDepthStencilValue!(self:Primitive Component Object Reference=@n1) -> (CustomDepthStencilValue:int=0)
   > "PrintDebugString(Unattended)"(
    DebugString=Append!(A="HandleSDP (", B="To String (Integer)"(InInt=@n2), C=")", D:String=" - ", E:String=GetDisplayName(Object=@n1)) -> (ReturnValue)
    Verbosity=NewEnumerator1
    Active=true
   )
   > n3 = Switch on Int!(execute:Exec, Selection:int=@n2) -> (Default:Exec, 2:Exec, 3:Exec) {startIndex=2}
   n4 = Get ISM() -> (ISM:Instanced Static Mesh Component Object Reference)
   n5 = SetCollisionProfileName!(execute=@n3.2, self:Primitive Component Object Reference=@n4, InCollisionProfileName:Name=MeshGrass_SDP, bUpdateOverlaps:bool=false) -> (then)
   n6 = Get ISM()
   n7 = Set Tags(
    execute=[@n5.then, @n11.then]
    Tags:Array of Names=Get ComponentTags!(self:Actor Component Object Reference=@n6) -> (ComponentTags:Array of Names)
   ) -> (Output_Get:Array of Names)
   > Add(TargetArray:Array of Names=@n7, NewItem:"Name (by ref)"=SDP)
   > Set ComponentTags!(execute, ComponentTags:Array of Names=Get Tags() -> (Tags:Array of Names), self:Actor Component Object Reference=@n6) -> (then, Output_Get:Array of Names)
   n8 = Get ISM()
   n9 = Contains Item(TargetArray=Get ComponentTags(self=@n8), ItemToFind=SDP)
   n10 = GetCollisionProfileName!(self:Primitive Component Object Reference=@n8) -> (ReturnValue:Name=None)
   > "PrintDebugString(Unattended)"(
    DebugString=Append!(
     A="SDP Conversion ("
     B="To String (Boolean)"(InBool=@n9)
     C=" - "
     D="To String (Name)"(InName=@n10)
     E=") "
     F:String=GetDisplayName(Object=@n8)
    ) -> (ReturnValue)
    Verbosity=NewEnumerator1
    Active=true
   )
   n11 = SetCollisionProfileName(execute=@n3.3, self=@n4, InCollisionProfileName=MeshGrass_SDP_Blocking, bUpdateOverlaps=false)
   SetCollisionProfileName(execute=@n3.Default, self=@n4, InCollisionProfileName=NoCollision, bUpdateOverlaps=false)
  ]}
 "PrintDebugString(Unattended)" {nodes [
   n1 = "PrintDebugString(Unattended)"!() -> (then, DebugString:String, Verbosity:EMeshGrassLogLevel Enum=NewEnumerator0, Active:bool=true) {is=title K2Node_FunctionEntry}
   n2 = AND Boolean(A=IsUnattended!() -> (ReturnValue:bool), B=Get DebugUnattended() -> (DebugUnattended:bool))
   n3 = Get DebugForceAll() -> (DebugForceAll:bool)
   > Branch(Condition=AND Boolean(A=@n1.Active, B=OR Boolean(A=@n2, B=@n3)))
   > n4 = Branch(Condition="Equal (Enum)"(A:EMeshGrassLogLevel Enum=@n1.Verbosity, B:EMeshGrassLogLevel Enum=NewEnumerator0))
   > PrintString(InString=Get DebugString() -> (DebugString:String))
   n5 = GetOwner(self=Self-Reference())
   n6 = Enum to String!(Enumerator:EMeshGrassLogLevel Enum=@n1.Verbosity) -> (ReturnValue:String)
   FortShippingLog!(
    execute=@n4.else
    WorldContextObject:Object Reference=GetOwner()
    InString:String=Append!(A=Append!(A=@n6, B=": ") -> (ReturnValue), B="[", C=GetObjectName(Object=@n5), D="] ", E=Get DebugString()) -> (ReturnValue)
    bLogAsWarning:bool=false
   ) -> (then)
  ]}
 SplitStringCached {nodes [
   n1 = SplitStringCached!() -> (then, SourceString:String, InStr:String) {is=title K2Node_FunctionEntry}
   n2 = Split(SourceString=@n1, InStr=@n1.InStr, SearchDir=FromStart)
   > Return Node!(execute, ReturnValue:bool=@n2.ReturnValue, LeftS:String=@n2, RightS:String=@n2.RightS)
  ]}
 TryReplaceMaterialNoRVT {nodes [
   TryReplaceMaterialNoRVT!() -> (then, TargetISM:Instanced Static Mesh Component Object Reference, MaterialIndex:int) {is=title K2Node_FunctionEntry}
   > "PrintDebugString(Unattended)"(DebugString=EUBP_MeshGrassUtilityComponent::TryReplaceMaterialNoRVT, Verbosity=NewEnumerator1, Active=true) {is=title K2Node_CallFunction}
   > n1 = Sequence!(execute) -> (then_0, then_1, then_2:Exec)
   > n2 = Set Material_Original(Material_Original:Material Interface Object Reference=GetEditorMaterial!(
     self:Primitive Component Object Reference=Get TargetISM() -> (TargetISM:Instanced Static Mesh Component Object Reference)
     ElementIndex:int=Get MaterialIndex() -> (MaterialIndex:int=0)
    ) -> (ReturnValue:Material Interface Object Reference)) -> (Output_Get:Material Interface Object Reference)
   > n3 = GetPathNameForLoadedAsset(LoadedAsset=@n2)
   n4 = GetBaseFilename(InPath=@n3, bRemovePath=false)
   > n5 = Branch(Condition=EndsWith(SourceString=@n4, InSuffix=_NoRVT))
   > Branch(Condition=Get Debug())
   > PrintString(InString=Skipping: already has NoRVT material)
   n6 = Set MaterialPath_Original(execute=@n5.else, MaterialPath_Original:String=@n4) -> (Output_Get:String)
   > n7 = Set MaterialPath_NoRVT(MaterialPath_NoRVT:String=Append(A=@n6, B=_NoRVT)) -> (Output_Get:String)
   > n8 = DoesAssetExist(AssetPath=@n7)
   > n9 = Set FoundNoRVTMaterial(FoundNoRVTMaterial:bool=@n8) -> (Output_Get:bool)
   > Branch(Condition=Get Debug())
   > PrintString(InString=Append!(A="DoesAssetExist: ", B=@n7, C:String=": ", D:String="To String (Boolean)"(InBool=@n9)) -> (ReturnValue))
   n10 = Get FoundNoRVTMaterial() -> (FoundNoRVTMaterial:bool)
   Branch(execute=@n1.then_1, Condition=@n10)
   > n11 = Branch(Condition=Get Debug())
   > n12 = PrintString(InString=Append!(A=Get MaterialPath_NoRVT() -> (MaterialPath_NoRVT:String), B=": ", C="To String (Boolean)"(InBool=@n10)) -> (ReturnValue))
   n13 = LoadAsset(execute=[@n12.then, @n11.else], AssetPath=Get MaterialPath_NoRVT())
   > Is Valid(InputObject=@n13)
   > n14 = Cast To MaterialInterface(Object=@n13)
   > n15 = Set Material_NoRVT(Material_NoRVT:Material Interface Object Reference=@n14) -> (Output_Get:Material Interface Object Reference)
   > Set ValidNoRVTMaterial(ValidNoRVTMaterial:bool=IsValid(Object=@n15)) -> (Output_Get:bool)
   n16 = Branch(execute=@n1.then_2, Condition=Get ValidNoRVTMaterial() -> (ValidNoRVTMaterial:bool))
   n17 = Get MaterialIndex() -> (MaterialIndex=0)
   n18 = Get TargetISM()
   > n19 = SetMaterial(self=@n18, ElementIndex=@n17, Material=Get Material_NoRVT() -> (Material_NoRVT:Material Interface Object Reference))
   n20 = GetDisplayName(Object=Get Material_Original() -> (Material_Original:Material Interface Object Reference))
   n21 = GetDisplayName(Object=GetMaterial!(self:Primitive Component Object Reference=@n18, ElementIndex:int=@n17) -> (ReturnValue:Material Interface Object Reference))
   n22 = "Not Equal, Case Insenstive (String)"(A=@n20, B=@n21)
   "PrintDebugString(Unattended)"(
    execute=[@n19.then, @n23.then]
    DebugString=Append!(
     A=EUBP_MeshGrassUtilityComponent::TryReplaceMaterialNoRVT::SetMaterial
     B="("
     C=@n20
     D:String="->"
     E:String=@n21
     F:String=")"
     G:String=SelectString(A=" @@@", bPickA=@n22)
    ) -> (ReturnValue)
    Verbosity=NewEnumerator1
    Active=false
   )
   n23 = SetMaterial(execute=@n16.else, self=@n18, ElementIndex=@n17, Material=Get Material_Original())
  ]}
]
variables [
 Debug {type=boolean}
 DebugForceAll {type=boolean}
 DebugUnattended {type=boolean, default=true}
 MeshMaterialMap {}
]
