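; Custom PCG Blueprint node, titled "Filter Point Data By Point Count", that sorts whole data sets by point count, in the PCGGrass world-building tools.
; It converts each spatial input to point data and compares the point count to the Threshold variable.
; Counts above the threshold route the original data, tags intact, to InsideFilter, and all others to OutsideFilter.
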(fn ExecuteWithContext
  (params (InContext :type PCGContext%20Structure%20%28by%20ref%29) (Input :type PCGData%20Collection%20Structure%20%28by%20ref%29))
  (Variables|Default|SetContext InContext)
  (bind _returnvalue (Data|GetAllSpatialInputs Input))
  (LocalVariables|SetSourceData _returnvalue)
  (if (Utilities|Array|IsNotEmpty _returnvalue)
    (for _array_element _returnvalue
      (bind _aspcgspatial_data (Utilities|Casting|CastToPCGSpatialData (PCG|TaggedData|BreakTaggedData _array_element)))
      (Utilities|Array|Add (Variables|Default|GetOutData) (PCG|TaggedData|MakeTaggedData (PCG|TaggedData|BreakTaggedData _array_element) (.Tags (PCG|TaggedData|BreakTaggedData _array_element)) (Utilities|String|SelectName "InsideFilter" "OutsideFilter" (> (Utilities|Array|Length (SpatialData|GetPoints (SpatialData|ToPointDatawithContext _aspcgspatial_data (Variables|Default|GetContext)))) (Variables|Default|GetThreshold))))))
    (return (Utilities|Struct|MakePCGDataCollection (Variables|Default|GetOutData)))))

(fn NodeTitleOverride
  (return "Filter Point Data By Point Count"))
