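; PCG Blueprint element that triggers biome data population on mesh grass actors used in landscape world building.
; When its soft-referenced target actor is valid, it tries casts to the global controls, called with a biome index, and local controls, called with -1.
; If the local cast fails, it tries the PCG component's owner as a spline, then texture, generator, also with -1.
; That valid-target branch returns the "In" pin data unchanged.
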(fn Execute
  (params (Input :type PCGData%20Collection%20Structure%20%28by%20ref%29))
  (bind _targetactor (Variables|Default|GetTargetActor))
  (bind _output (Utilities|ResolveSoftReference _targetactor))
  (bind _returnvalue (Data|GetInputsbyPinLabel Input "In"))
  (if (Utilities|IsValidSoftObjectReference _targetactor)
    (bind _asbp_mesh_grass_global_controls (Utilities|Casting|CastToBP_MeshGrass_GlobalControls _output)
      (:then
        (Class|BPMeshGrassGlobalControls|PopulateBiomeData _asbp_mesh_grass_global_controls (Variables|Default|GetBiomeIndex)))
      (:CastFailed
        _))
    (bind _asbp_mesh_grass_local_controls (Utilities|Casting|CastToBP_MeshGrass_LocalControls _output)
      (:then
        (Class|BPMeshGrassLocalControls|PopulateBiomeData _asbp_mesh_grass_local_controls -1))
      (:CastFailed
        (bind _returnvalue_2 (PCG|Helpers|GetComponentwithContext (PCG|Advanced|GetContextHandle)))
        (bind _asbp_mesh_grass_spline_generator (Utilities|Casting|CastToBP_MeshGrass_SplineGenerator (Components|GetOwner _returnvalue_2))
          (:then
            (Class|BPMeshGrassSplineGenerator|PopulateBiomeData _asbp_mesh_grass_spline_generator -1))
          (:CastFailed
            (bind _asbp_mesh_grass_texture_generator (Utilities|Casting|CastToBP_MeshGrass_TextureGenerator (Components|GetOwner _returnvalue_2))
              (:then
                (Class|BPMeshGrassTextureGenerator|PopulateBiomeData _asbp_mesh_grass_texture_generator -1))
              (:CastFailed
                _))))))
    (return _returnvalue false)))
