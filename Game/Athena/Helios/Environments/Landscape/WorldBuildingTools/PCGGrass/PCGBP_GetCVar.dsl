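; Custom PCG Blueprint node, titled GetCVar, that exposes a console variable's value to a PCG graph as attribute data.
; It creates a new param data object with one metadata entry and switches on the configured type to read a bool, int, float or string value.
; That value becomes the default of an attribute named after the CVar, with dots replaced by underscores, returned in the output collection.
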(fn ExecuteWithContext
  (params (InContext :type PCGContext%20Structure%20%28by%20ref%29) (Input :type PCGData%20Collection%20Structure%20%28by%20ref%29))
  (bind _returnvalue (Utilities|String|Replace (Variables|Default|GetCVarName) "." "_"))
  (bind _returnvalue_2 (Utilities|String|StringToName _returnvalue))
  (bind _outmetadata (LocalVariables|GetOutMetadata))
  (bind _cvarname (Variables|Default|GetCVarName))
  (Utilities|FlowControl|Sequence
    (:then_0
      (bind _returnvalue_3 (Game|ConstructObjectfromClass "/Script/PCG.PCGParamData"))
      (bind _returnvalue_4 (Metadata|MutableMetadata _returnvalue_3))
      (bind _output_get (LocalVariables|SetOutMetadata _returnvalue_4))
      (PCG|Metadata|AddEntry _output_get -1))
    (:then_1
      (switch Utilities|FlowControl|Switch|SwitchonECVarType (Variables|Default|GetCVarType)
        (:NewEnumerator0
          (bind _returnvalue_5 (Development|GetConsoleVariableBoolValue (Variables|Default|GetCVarName)))
          (PCG|Metadata|CreateBoolAttribute _outmetadata _returnvalue_2 _returnvalue_5 false true))
        (:NewEnumerator1
          (bind _returnvalue_6 (Development|GetConsoleVariableIntValue (Variables|Default|GetCVarName)))
          (PCG|Metadata|CreateInteger32Attribute _outmetadata _returnvalue_2 _returnvalue_6 false true))
        (:NewEnumerator2
          (bind _returnvalue_7 (Development|GetConsoleVariableFloatValue (Variables|Default|GetCVarName)))
          (PCG|Metadata|CreateFloatAttribute _outmetadata _returnvalue_2 _returnvalue_7 false true))
        (:NewEnumerator3
          (bind _returnvalue_8 (Development|GetConsoleVariableStringValue (Variables|Default|GetCVarName)))
          (PCG|Metadata|CreateStringAttribute _outmetadata _returnvalue_2 _returnvalue_8 false true)))
      (return (Utilities|Struct|MakePCGDataCollection (Utilities|Array|MakeArray (PCG|TaggedData|MakeTaggedData _returnvalue_3)))))))

(fn NodeColorOverride
  (return "(R=1.000000,G=0.600000,B=0.000000,A=1.000000)"))

(fn NodeTitleOverride
  (return "GetCVar"))
