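; Custom Blueprint PCG node, "Get Data by Index", in the Helios landscape PCG grass tools, that picks one entry from its input data collection.
; Execution reads the DataIndex variable and, if it is valid, returns a new collection holding only that tagged data.
; An invalid index returns no data.
; Dynamic pin typing follows the types connected to InData, and the node overrides its title and color.
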(fn DynamicPinTypesOverride
  (params (InSettings :type PCGSettings%20Object%20Reference) (InPin :type PCGPin%20Object%20Reference))
  (return (Settings|DynamicPins|GetTypeUnionOfIncidentEdges InSettings (Utilities|Name|MakeLiteralName "InData"))))

(fn ExecuteWithContext
  (params (InContext :type PCGContext%20Structure%20%28by%20ref%29) (Input :type PCGData%20Collection%20Structure%20%28by%20ref%29))
  (bind _taggeddata (Utilities|Struct|BreakPCGDataCollection Input))
  (bind _dataindex (Variables|Default|GetDataIndex))
  (if (Utilities|Array|IsValidIndex _taggeddata _dataindex)
    (return (Utilities|Struct|MakePCGDataCollection (Utilities|Array|MakeArray (Utilities|Array|Get(acopy) _taggeddata _dataindex))))))

(fn NodeColorOverride
  (return "(R=0.172868,G=0.483108,B=1.000000,A=1.000000)"))

(fn NodeTitleOverride
  (return "Get Data by Index"))
