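; PCG Blueprint helper that converts a Landscape Grass Type's varieties into metadata entries describing meshes, materials, scaling, culling and placement for PCG grass.
; PreProcessWeights totals densities at the chosen quality, then ProcessVarieties repeats each variety's entry in proportion to its density share, forming a weighted table.
; Grass console variables scale cull distances and, optionally, density.
; Effects-folder paths and one hard-coded wildflower mesh drive standalone flagging, and primitive indices accumulate across calls.
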(fn PreProcessWeights
  (params (Varieties :type Array%20of%20Grass%20Variety%20Structures) (Quality :type Integer))
  (for _array_element Varieties
    (LocalVariables|SetWeightSum (+ (Utilities|Select :Index Quality :Option%200 (Grass|GetGrassVarietyDensity _array_element) :Option%201 (Grass|GetGrassVarietyDensity _array_element "Medium") :Option%202 (Grass|GetGrassVarietyDensity _array_element "High") :Option%203 (Grass|GetGrassVarietyDensity _array_element "Epic")) (LocalVariables|GetWeightSum))))
  (return (LocalVariables|GetWeightSum)))

(fn ProcessVarieties
  (params (LGT :type Landscape%20Grass%20Type%20Object%20Reference) (NumElem :type Integer) (TargetMetadata :type PCGMetadata%20Object%20Reference%20%28by%20ref%29) (Quality :type Integer))
  (bind _grassvarieties (Variables|/Script/Landscape.LandscapeGrassType|GetGrassVarieties LGT))
  (bind _returnvalue (Utilities|Array|Length _grassvarieties))
  (bind _quality (Variables|Default|GetQuality))
  (bind _currentgrassvariety (LocalVariables|GetCurrentGrassVariety))
  (bind (_startculldistance _endculldistance) (Grass|GetGrassVarietyCullDistanceRange _currentgrassvariety "Cinematic"))
  (bind (_startculldistance_2 _endculldistance_2) (Grass|GetGrassVarietyCullDistanceRange _currentgrassvariety "Epic"))
  (bind (_startculldistance_3 _endculldistance_3) (Grass|GetGrassVarietyCullDistanceRange _currentgrassvariety "High"))
  (bind (_startculldistance_4 _endculldistance_4) (Grass|GetGrassVarietyCullDistanceRange _currentgrassvariety "Medium"))
  (bind (_startculldistance_5 _endculldistance_5) (Grass|GetGrassVarietyCullDistanceRange _currentgrassvariety))
  (bind _returnvalue_2 (Utilities|Select :Index _quality :Option%200 _endculldistance_5 :Option%201 _endculldistance_4 :Option%202 _endculldistance_3 :Option%203 _endculldistance_2 :Option%204 _endculldistance))
  (bind _returnvalue_3 (Utilities|Select :Index _quality :Option%200 _startculldistance_5 :Option%201 _startculldistance_4 :Option%202 _startculldistance_3 :Option%203 _startculldistance_2 :Option%204 _startculldistance))
  (bind _returnvalue_4 (+ _returnvalue_3 _returnvalue_2))
  (bind _returnvalue_5 (Math|Conversions|ToBoolean(Integer) _returnvalue_4))
  (bind (_instanceworldpositionoffsetdisabledistance _bfailed) (Grass|GetInstanceWorldPositionOffsetDisableDistance _currentgrassvariety))
  (bind _numelem (Variables|GetNumElem))
  (bind _returnvalue_6 (* _returnvalue _numelem))
  (bind _returnvalue_7 (Math|Conversions|ToFloat(Integer) _returnvalue_6))
  (bind _weightsum (LocalVariables|GetWeightSum))
  (bind _returnvalue_8 (Math|Float|Max(Float) _weightsum 0.001))
  (bind _returnvalue_9 (> _weightsum 0))
  (bind _targetmetadata (Variables|GetTargetMetadata))
  (bind _currententry (LocalVariables|GetCurrentEntry))
  (bind _fallbackculldistance (LocalVariables|GetFallbackCullDistance))
  (bind _grassculldistancescale (LocalVariables|GetGrassCullDistanceScale))
  (bind _targetmetadata_2 (Variables|GetTargetMetadata))
  (bind _quality_2 (Variables|Default|GetQuality))
  (bind _currententry_2 (LocalVariables|GetCurrentEntry))
  (bind _targetmetadata_3 (Variables|GetTargetMetadata))
  (bind _currententry_3 (LocalVariables|GetCurrentEntry))
  (bind _outweightsum (CallFunction|PreProcessWeights _grassvarieties Quality))
  (LocalVariables|SetWeightSum _outweightsum)
  (bind _returnvalue_10 (Development|GetConsoleVariableFloatValue "grass.densityscale"))
  (LocalVariables|SetGrassDensityScale _returnvalue_10)
  (bind _returnvalue_11 (Development|GetConsoleVariableFloatValue "grass.CullDistanceScale"))
  (LocalVariables|SetGrassCullDistanceScale _returnvalue_11)
  (for _array_element _grassvarieties
    (bind _output_get (LocalVariables|SetCurrentGrassVariety _array_element))
    (if (and (Utilities|IsValid (Utilities|Struct|BreakGrassVariety _output_get) :!class /Script/Engine.KismetSystemLibrary) (> (Utilities|Select :Index _quality_2 :Option%200 (Grass|GetGrassVarietyDensity _output_get) :Option%201 (Grass|GetGrassVarietyDensity _output_get "Medium") :Option%202 (Grass|GetGrassVarietyDensity _output_get "High") :Option%203 (Grass|GetGrassVarietyDensity _output_get "Epic")) 0))
      (if (Variables|Default|GetDebug)
        (Utilities|String|LogString (Utilities|String|Append (Utilities|GetDisplayName LGT) " >> " (Utilities|GetDisplayName (Utilities|Struct|BreakGrassVariety _output_get)) " >> " (Utilities|String|ToString(Float) (/ (Utilities|Select :Index _quality_2 :Option%200 (Grass|GetGrassVarietyDensity _output_get) :Option%201 (Grass|GetGrassVarietyDensity _output_get "Medium") :Option%202 (Grass|GetGrassVarietyDensity _output_get "High") :Option%203 (Grass|GetGrassVarietyDensity _output_get "Epic")) _returnvalue_8)) " == " (Utilities|String|ToString(Integer) (select _returnvalue_9 (Math|Float|Round (* _returnvalue_7 (/ (Utilities|Select :Index _quality_2 :Option%200 (Grass|GetGrassVarietyDensity _output_get) :Option%201 (Grass|GetGrassVarietyDensity _output_get "Medium") :Option%202 (Grass|GetGrassVarietyDensity _output_get "High") :Option%203 (Grass|GetGrassVarietyDensity _output_get "Epic")) _returnvalue_8))) 0)))))
      (if (Variables|Default|GetExcludeEffects)
        (if (Utilities|String|Contains (Utilities|Paths|GetPath (Utilities|GetObjectPathString (Utilities|Struct|BreakGrassVariety _output_get))) "effects")
          _
          (else
            _))
        (else
          (LocalVariables|SetIsStandalone (Utilities|String|Contains (Utilities|Paths|GetPath (Utilities|GetObjectPathString (Utilities|Struct|BreakGrassVariety _output_get))) "effects"))))
      (for _ (range (+ (Math|Integer|SelectInt (select _returnvalue_9 (Math|Float|Round (* _returnvalue_7 (/ (Utilities|Select :Index _quality_2 :Option%200 (Grass|GetGrassVarietyDensity _output_get) :Option%201 (Grass|GetGrassVarietyDensity _output_get "Medium") :Option%202 (Grass|GetGrassVarietyDensity _output_get "High") :Option%203 (Grass|GetGrassVarietyDensity _output_get "Epic")) _returnvalue_8))) 0) 1 (Math|Boolean|ANDBoolean (> _returnvalue 1) (> (Variables|GetNumElem) 1) _returnvalue_9)) 1))
        (bind _returnvalue_12 (PCG|Metadata|AddEntry (Variables|GetTargetMetadata) -1))
        (LocalVariables|SetCurrentEntry _returnvalue_12)
        (PCG|Metadata|SetStringAttributebyMetadataKey _currententry_3 _targetmetadata_2 "LandscapeGrassType" (Utilities|GetDisplayName LGT))
        (PCG|Metadata|SetSoftObjectPathAttributebyMetadataKey _currententry_3 _targetmetadata_2 "LGTPath" (Utilities|GetSoftObjectPath LGT))
        (PCG|Metadata|SetSoftObjectPathAttributebyMetadataKey (LocalVariables|GetCurrentEntry) (Variables|GetTargetMetadata) "MeshAsset" (Utilities|GetSoftObjectPath (Utilities|Struct|BreakGrassVariety _output_get)))
        (if (Utilities|Array|IsNotEmpty (.OverrideMaterials (Utilities|Struct|BreakGrassVariety _output_get)))
          (PCG|Metadata|SetSoftObjectPathAttributebyMetadataKey _currententry _targetmetadata "OverrideMaterial0" (Utilities|GetSoftObjectPath (Utilities|Array|Get(acopy) (.OverrideMaterials (Utilities|Struct|BreakGrassVariety _output_get)))))
          (if (Utilities|Array|IsValidIndex (.OverrideMaterials (Utilities|Struct|BreakGrassVariety _output_get)) 1)
            (PCG|Metadata|SetSoftObjectPathAttributebyMetadataKey _currententry _targetmetadata "OverrideMaterial1" (Utilities|GetSoftObjectPath (Utilities|Array|Get(acopy) (.OverrideMaterials (Utilities|Struct|BreakGrassVariety _output_get)) 1)))))
        (PCG|Metadata|SetInteger32AttributebyMetadataKey (LocalVariables|GetCurrentEntry) (Variables|GetTargetMetadata) "ScalingMode" (Math|Conversions|ToInteger(Byte) (.Scaling (Utilities|Struct|BreakGrassVariety _output_get))))
        (PCG|Metadata|SetVectorAttributebyMetadataKey (LocalVariables|GetCurrentEntry) (Variables|GetTargetMetadata) "ScaleMin" (Math|Vector|MakeVector (Utilities|Struct|BreakFloatInterval (.ScaleX (Utilities|Struct|BreakGrassVariety _output_get))) (Utilities|Struct|BreakFloatInterval (.ScaleY (Utilities|Struct|BreakGrassVariety _output_get))) (Utilities|Struct|BreakFloatInterval (.ScaleZ (Utilities|Struct|BreakGrassVariety _output_get)))))
        (PCG|Metadata|SetVectorAttributebyMetadataKey (LocalVariables|GetCurrentEntry) (Variables|GetTargetMetadata) "ScaleMax" (Math|Vector|MakeVector (.Max (Utilities|Struct|BreakFloatInterval (.ScaleX (Utilities|Struct|BreakGrassVariety _output_get)))) (.Max (Utilities|Struct|BreakFloatInterval (.ScaleY (Utilities|Struct|BreakGrassVariety _output_get)))) (.Max (Utilities|Struct|BreakFloatInterval (.ScaleZ (Utilities|Struct|BreakGrassVariety _output_get))))))
        (PCG|Metadata|SetFloatAttributebyMetadataKey (LocalVariables|GetCurrentEntry) (Variables|GetTargetMetadata) "ScaleByWeight" (Math|Float|SelectFloat (.MaxScaleWeightAttenuation (Utilities|Struct|BreakGrassVariety _output_get)) -1.0 (.bWeightAttenuatesMaxScale (Utilities|Struct|BreakGrassVariety _output_get))))
        (bind _returnvalue_13 (Utilities|LoadAssetBlocking (Utilities|ToSoftObjectReference (struct :PathString "/Hermes/Foliage/Grass/Mesh/SM_Hermes_Wildflower_A_A.SM_Hermes_Wildflower_A_A"))))
        (PCG|Metadata|SetBoolAttributebyMetadataKey (LocalVariables|GetCurrentEntry) (Variables|GetTargetMetadata) "IsStandalone" (or (LocalVariables|GetIsStandalone) (== (Utilities|Struct|BreakGrassVariety _output_get) _returnvalue_13)))
        (PCG|Metadata|SetInteger32AttributebyMetadataKey (LocalVariables|GetCurrentEntry) (Variables|GetTargetMetadata) "PrimitiveIndex" (+ (Variables|Default|GetBasePrimIndex) (LocalVariables|GetCurrentIndex)))
        (PCG|Metadata|SetInteger32AttributebyMetadataKey _currententry_2 _targetmetadata_3 "StartCullDistance" (Math|Float|Truncate (* (Math|Integer|SelectInt _returnvalue_3 _fallbackculldistance _returnvalue_5) _grassculldistancescale)))
        (PCG|Metadata|SetInteger32AttributebyMetadataKey _currententry_2 _targetmetadata_3 "EndCullDistance" (Math|Float|Truncate (* (Math|Integer|SelectInt _returnvalue_2 _fallbackculldistance _returnvalue_5) _grassculldistancescale)))
        (PCG|Metadata|SetInteger32AttributebyMetadataKey (LocalVariables|GetCurrentEntry) (Variables|GetTargetMetadata) "WorldPositionOffsetDisableDistance" (Math|Integer|SelectInt 2500 _instanceworldpositionoffsetdisabledistance _bfailed))
        (PCG|Metadata|SetFloatAttributebyMetadataKey (LocalVariables|GetCurrentEntry) (Variables|GetTargetMetadata) "SumDensities" (* (LocalVariables|GetWeightSum) (Math|Float|SelectFloat (LocalVariables|GetGrassDensityScale) (Utilities|Select :Index (Variables|Default|GetQuality) :Option%200 0.0 :Option%201 0.6 :Option%202 0.8 :Option%203 1.0) (Variables|Default|GetUseFoliageScalability)))) ; set sum densities in layer, multiplied by grss.densityscale if driven by foliage scalability, otherwise mock the cvar
        (PCG|Metadata|SetBoolAttributebyMetadataKey (LocalVariables|GetCurrentEntry) (Variables|GetTargetMetadata) "RandomRotation" (.RandomRotation (Utilities|Struct|BreakGrassVariety _output_get)))
        (PCG|Metadata|SetBoolAttributebyMetadataKey (LocalVariables|GetCurrentEntry) (Variables|GetTargetMetadata) "AlignToSurface" (.AlignToSurface (Utilities|Struct|BreakGrassVariety _output_get))))
      (Math|Integer|IncrementInt (LocalVariables|GetCurrentIndex))))
  (Variables|Default|SetBasePrimIndex (+ (Variables|Default|GetBasePrimIndex) (LocalVariables|GetCurrentIndex))))
