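; Custom PCG Blueprint node, titled "Get Material", that exposes the owning actor's material to a PCG graph as attribute data.
; It reads material slot 0 from the PrimitiveComponent found on the PCG component's owner and converts it to a soft object path.
; It outputs new param data with a "Material" attribute, left empty if the slot has no material.
; Without a valid primitive component, it builds no data.
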(fn ExecuteWithContext
  (params (InContext :type PCGContext%20Structure%20%28by%20ref%29) (Input :type PCGData%20Collection%20Structure%20%28by%20ref%29))
  (bind _returnvalue (PCG|Temporary|GetComponent InContext))
  (Utilities|IsValid (Actor|GetComponentbyClass (Components|GetOwner _returnvalue) "/Script/Engine.PrimitiveComponent")
    (:Is%20Valid
      (Utilities|IsValid (Rendering|Material|GetMaterial (Actor|GetComponentbyClass (Components|GetOwner _returnvalue) "/Script/Engine.PrimitiveComponent") 0)
        (:Is%20Valid
          (LocalVariables|SetOutMaterial (Utilities|GetSoftObjectPath (Rendering|Material|GetMaterial (Actor|GetComponentbyClass (Components|GetOwner _returnvalue) "/Script/Engine.PrimitiveComponent") 0))))
        (:Is%20Not%20Valid
          (LocalVariables|SetOutMaterial)))
      (bind _returnvalue_2 (Game|ConstructObjectfromClass "/Script/PCG.PCGParamData"))
      (bind _returnvalue_3 (Metadata|MutableMetadata _returnvalue_2))
      (PCG|Metadata|AddEntry _returnvalue_3 -1)
      (PCG|Metadata|CreateSoftObjectPathAttribute _returnvalue_3 "Material" (LocalVariables|GetOutMaterial) false true)
      (return (Utilities|Struct|MakePCGDataCollection (Utilities|Array|MakeArray (PCG|TaggedData|MakeTaggedData _returnvalue_2)))))
    (:Is%20Not%20Valid
      _)))

(fn NodeColorOverride
  (return "(R=1.000000,G=0.597202,B=0.000000,A=1.000000)"))

(fn NodeTitleOverride
  (return "Get Material"))
