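; PCG Blueprint node "Get Mesh Bounds" that outputs an attribute set of static mesh bounding-sphere radii.
; For each non-empty attribute set on the Meshes pin, it creates new param data with a double attribute, replacing the previous output, so only the last set is returned.
; It reads every attribute entry as a soft object path, blocking-loads the static mesh and adds an entry holding its sphere radius.
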(fn ExecuteWithContext
  (params (InContext :type PCGContext%20Structure%20%28by%20ref%29) (Input :type PCGData%20Collection%20Structure%20%28by%20ref%29))
  (bind _mutablemetadata (LocalVariables|GetMutableMetaData))
  (bind _returnvalue (Data|GetParamsbyPinLabel Input "Meshes"))
  (for _array_element _returnvalue
    (bind _aspcgattribute_set (Utilities|Casting|CastToPCGParamData (PCG|TaggedData|BreakTaggedData _array_element)))
    (bind _output_get (LocalVariables|SetInMetadata (Metadata|ConstMetadata _aspcgattribute_set)))
    (if (!= (PCG|Metadata|GetNumberofEntries _output_get) 0)
      (Utilities|FlowControl|Sequence
        (:then_0
          (bind _returnvalue_2 (Game|ConstructObjectfromClass "/Script/PCG.PCGParamData"))
          (bind _output_get_2 (LocalVariables|SetOutData _returnvalue_2))
          (bind _returnvalue_3 (Metadata|MutableMetadata _output_get_2))
          (bind _output_get_3 (LocalVariables|SetMutableMetaData _returnvalue_3))
          (PCG|Metadata|CreateDoubleAttribute _output_get_3 (LocalVariables|GetMeshBounds) 0.0 false true))
        (:then_1
          (for _array_element_2 (PCG|Metadata|GetAttributes _output_get)
            (for _index (range (+ (Math|Conversions|ToInteger(Integer64) (- (PCG|Metadata|GetNumberofEntries _output_get) 1)) 1))
              (bind _returnvalue_4 (PCG|Metadata|GetSoftObjectPathAttributebyMetadataKey (Math|Conversions|ToInteger64(Integer) _index) _output_get _array_element_2))
              (bind _returnvalue_pathstring (Utilities|BreakSoftObjectPath _returnvalue_4))
              (bind _returnvalue_5 (Utilities|LoadAssetBlocking (Utilities|ToSoftObjectReference (Utilities|MakeSoftObjectPath _returnvalue_pathstring))))
              (bind _asstatic_mesh (Utilities|Casting|CastToStaticMesh _returnvalue_5))
              (bind _returnvalue_6 (PCG|Metadata|AddEntry _mutablemetadata -1))
              (bind (_returnvalue_origin _returnvalue_boxextent _returnvalue_sphereradius) (Math|BoxSphereBounds|BreakBoxSphereBounds (StaticMesh|GetBounds _asstatic_mesh)))
              (PCG|Metadata|SetDoubleAttributebyMetadataKey _returnvalue_6 _mutablemetadata (LocalVariables|GetMeshBounds) _returnvalue_sphereradius)))))))
  (return (Utilities|Struct|MakePCGDataCollection (Utilities|Array|MakeArray (PCG|TaggedData|MakeTaggedData (LocalVariables|GetOutData))))))

(fn NodeColorOverride
  (return "(R=1.000000,G=0.597202,B=0.000000,A=1.000000)"))

(fn NodeTitleOverride
  (return "Get Mesh Bounds"))
