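; PCG Blueprint node titled "Get Num Triangles", with an orange color override, that outputs the triangle count of a static mesh.
; It takes the first StaticMeshComponent found on the PCG component's owning actor, ignoring input data, and reads LOD 0's triangle count.
; The count becomes the default value of an unnamed int32 attribute on new single-entry param data, returned only when that component is valid.
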(fn ExecuteWithContext
  (params (InContext :type PCGContext%20Structure%20%28by%20ref%29) (Input :type PCGData%20Collection%20Structure%20%28by%20ref%29))
  (bind _returnvalue (PCG|Temporary|GetComponent InContext))
  (Utilities|IsValid (Actor|GetComponentbyClass (Components|GetOwner _returnvalue) "/Script/Engine.StaticMeshComponent")
    (:Is%20Valid
      (bind _returnvalue_2 (Game|ConstructObjectfromClass "/Script/PCG.PCGParamData"))
      (bind _returnvalue_3 (Metadata|MutableMetadata _returnvalue_2))
      (PCG|Metadata|AddEntry _returnvalue_3 -1)
      (PCG|Metadata|CreateInteger32Attribute _returnvalue_3 "None" (StaticMesh|GetNumTriangles (Variables|/Script/Engine.StaticMeshComponent|GetStaticMesh (Actor|GetComponentbyClass (Components|GetOwner _returnvalue) "/Script/Engine.StaticMeshComponent")) 0) false true)
      (return (Utilities|Struct|MakePCGDataCollection (Utilities|Array|MakeArray (PCG|TaggedData|MakeTaggedData _returnvalue_2)))))))

(fn NodeColorOverride
  (return "(R=1.000000,G=0.597202,B=0.000000,A=1.000000)"))

(fn NodeTitleOverride
  (return "Get Num Triangles"))
