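; Custom PCG node "Get Partition Coords From Path" that derives partition coordinate strings from soft object paths in input attribute sets.
; It reads each SoftObjectPath attribute's first entry and writes the parsed text as a same-named string attribute's default on a new single-entry attribute set.
; ProcessString keeps the text from the second underscore after the last dot, trimming a fixed-length tail, presumably a hash, when the result is too long.
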(fn ExecuteWithContext
  (params (InContext :type PCGContext%20Structure%20%28by%20ref%29) (Input :type PCGData%20Collection%20Structure%20%28by%20ref%29))
  (Utilities|FlowControl|Sequence
    (:then_0
      (bind _returnvalue (Game|ConstructObjectfromClass "/Script/PCG.PCGParamData"))
      (bind _returnvalue_2 (Metadata|MutableMetadata _returnvalue))
      (bind _output_get (LocalVariables|SetOutMetadata _returnvalue_2))
      (bind _returnvalue_3 (PCG|Metadata|AddEntry _output_get -1))
      (LocalVariables|SetOutMetaDataEntry _returnvalue_3))
    (:then_1
      (bind _returnvalue_4 (Data|GetParams Input))
      (for _array_element _returnvalue_4
        (bind _aspcgattribute_set (Utilities|Casting|CastToPCGParamData (PCG|TaggedData|BreakTaggedData _array_element)))
        (bind _output_get_2 (LocalVariables|SetInMetadata (Metadata|ConstMetadata _aspcgattribute_set)))
        (for (_array_element_2 _array_index) (PCG|Metadata|GetAttributes _output_get_2)
          (if (Utilities|Enum|Equal(Enum) (Utilities|Array|Get(acopy) (.AttributeTypes (PCG|Metadata|GetAttributes _output_get_2)) _array_index) "SoftObjectPath")
            (bind _returnvalue_5 (PCG|Metadata|GetSoftObjectPathAttributebyMetadataKey 0 (LocalVariables|GetInMetadata) _array_element_2))
            (bind _returnvalue_pathstring (Utilities|BreakSoftObjectPath _returnvalue_5))
            (LocalVariables|SetSourceString _returnvalue_pathstring)
            (PCG|Metadata|CreateStringAttribute (LocalVariables|GetOutMetadata) _array_element_2 (CallFunction|ProcessString (LocalVariables|GetSourceString)) false true))))
      (return (Utilities|Struct|MakePCGDataCollection (Utilities|Array|MakeArray (PCG|TaggedData|MakeTaggedData _returnvalue)))))))

(fn NodeColorOverride
  (return "(R=1.000000,G=0.600000,B=0.000000,A=1.000000)"))

(fn NodeTitleOverride
  (return "Get Partition Coords From Path"))

(fn ProcessString
  (params (SourceString :type String))
  (bind _output_get (LocalVariables|SetUnresolvedPartitionString (Utilities|String|RightChop SourceString (Utilities|String|FindSubstring SourceString "_" false false (+ (Utilities|String|FindSubstring SourceString "_" false false (Utilities|String|FindSubstring SourceString "." false true)) 1))))) ; may contain hash at this stage
  (if (> (Utilities|String|Len _output_get) 8)
    (return (Utilities|String|LeftChop _output_get 35))
    (else
      (return _output_get))))
