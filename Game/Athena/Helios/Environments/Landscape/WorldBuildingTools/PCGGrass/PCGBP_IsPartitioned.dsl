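; Custom PCG Blueprint node, titled IsPartitioned and shown in orange, that reports whether the executing PCG component belongs to a partition actor.
; It builds a new param data object with one metadata entry, then casts the component's owner to PartitionActor.
; The cast result becomes the default value of an "IsPartitioned" bool attribute, and the param data is returned as the node's output, ignoring any input.
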(fn ExecuteWithContext
  (params (InContext :type PCGContext%20Structure%20%28by%20ref%29) (Input :type PCGData%20Collection%20Structure%20%28by%20ref%29))
  (Utilities|FlowControl|Sequence
    (:then_0
      (bind _returnvalue (Game|ConstructObjectfromClass "/Script/PCG.PCGParamData"))
      (bind _returnvalue_2 (Metadata|MutableMetadata _returnvalue))
      (bind _output_get (LocalVariables|SetOutMetadata _returnvalue_2))
      (PCG|Metadata|AddEntry _output_get -1))
    (:then_1
      (bind _returnvalue_3 (PCG|Temporary|GetComponent InContext))
      (PCG|Metadata|CreateBoolAttribute (LocalVariables|GetOutMetadata) "IsPartitioned" (.bSuccess (Utilities|Casting|CastToPartitionActor (Components|GetOwner _returnvalue_3))) false false)
      (return (Utilities|Struct|MakePCGDataCollection (Utilities|Array|MakeArray (PCG|TaggedData|MakeTaggedData _returnvalue)))))))

(fn NodeColorOverride
  (return "(R=1.000000,G=0.600000,B=0.000000,A=1.000000)"))

(fn NodeTitleOverride
  (return "IsPartitioned"))
