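; PCG Blueprint node "Join Strings" that concatenates a string attribute's values across all entries of an attribute set into one separated string.
; For each input param data, it builds a new param data with one entry holding the joined result under a configurable attribute name.
; Target attribute and separator are variables.
; A missing attribute or empty metadata writes "INVALID" instead.
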(fn ExecuteWithContext
  (params (InContext :type PCGContext%20Structure%20%28by%20ref%29) (Input :type PCGData%20Collection%20Structure%20%28by%20ref%29))
  (bind _inmetadata (LocalVariables|GetInMetadata))
  (bind _returnvalue (PCG|Metadata|GetNumberofEntries _inmetadata))
  (bind _returnvalue_2 (- _returnvalue 1))
  (bind _returnvalue_3 (Data|GetParams Input))
  (for _array_element _returnvalue_3
    (Utilities|FlowControl|Sequence
      (:then_0
        (bind _returnvalue_4 (Game|ConstructObjectfromClass "/Script/PCG.PCGParamData"))
        (bind _returnvalue_5 (Metadata|MutableMetadata _returnvalue_4))
        (bind _output_get (LocalVariables|SetOutMetadata _returnvalue_5))
        (bind _returnvalue_6 (PCG|Metadata|AddEntry _output_get -1))
        (LocalVariables|SetOutMetaDataEntry _returnvalue_6)
        (PCG|Metadata|CreateStringAttribute (LocalVariables|GetOutMetadata) (Variables|Default|GetOutAttributeName) "" false true))
      (:then_1
        (bind _aspcgattribute_set (Utilities|Casting|CastToPCGParamData (PCG|TaggedData|BreakTaggedData _array_element)))
        (bind _output_get_2 (LocalVariables|SetInMetadata (Metadata|ConstMetadata _aspcgattribute_set)))
        (if (and (PCG|Metadata|HasAttribute _output_get_2 (Variables|Default|GetTargetAttribute)) (!= _returnvalue_2 -1))
          (for _index (range (+ (Math|Conversions|ToInteger(Integer64) _returnvalue_2) 1))
            (bind _returnvalue_7 (PCG|Metadata|GetStringAttributebyMetadataKey (Math|Conversions|ToInteger64(Integer) _index) (LocalVariables|GetInMetadata) (Variables|Default|GetTargetAttribute)))
            (LocalVariables|SetOutString (Utilities|String|Append (LocalVariables|GetOutString) (Utilities|String|SelectString (Variables|Default|GetSeparator) "" (!= _index 0)) _returnvalue_7)))
          (else
            (LocalVariables|SetOutString "INVALID")))
        (PCG|Metadata|SetStringAttributebyMetadataKey (LocalVariables|GetOutMetaDataEntry) (LocalVariables|GetOutMetadata) (Variables|Default|GetOutAttributeName) (LocalVariables|GetOutString))
        (LocalVariables|SetOutString ""))
      (:then_2
        (Utilities|Array|Add (LocalVariables|GetOutData) (PCG|TaggedData|MakeTaggedData _returnvalue_4)))))
  (return (Utilities|Struct|MakePCGDataCollection (LocalVariables|GetOutData))))

(fn NodeColorOverride
  (return "(R=1.000000,G=0.600000,B=0.000000,A=1.000000)"))

(fn NodeTitleOverride
  (return "Join Strings"))
