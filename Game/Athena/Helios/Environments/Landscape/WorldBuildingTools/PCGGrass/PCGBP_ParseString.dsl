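; A PCG Blueprint node titled "Parse String" that splits a delimited string attribute into separate attribute set entries.
; For each "In" input, it reads the source attribute at metadata key 0 and parses it by the configured delimiter.
; Non-empty pieces, optionally trimmed, become entries in a new param data object, whose string attribute defaults to "No Strings Found!".
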(fn ExecuteWithContext
  (params (InContext :type PCGContext%20Structure%20%28by%20ref%29) (Input :type PCGData%20Collection%20Structure%20%28by%20ref%29))
  (bind _outattribute (LocalVariables|GetOutAttribute))
  (bind _outmetadata (LocalVariables|GetOutMetadata))
  (bind _returnvalue (Data|GetInputsbyPinLabel Input "In"))
  (for _array_element _returnvalue
    (Utilities|FlowControl|Sequence
      (:then_0
        (bind _aspcgattribute_set (Utilities|Casting|CastToPCGParamData (PCG|TaggedData|BreakTaggedData _array_element)))
        (LocalVariables|SetInMetadata (Metadata|ConstMetadata _aspcgattribute_set))
        (bind _returnvalue_2 (Game|ConstructObjectfromClass "/Script/PCG.PCGParamData"))
        (bind _returnvalue_3 (Metadata|MutableMetadata _returnvalue_2))
        (LocalVariables|SetOutMetadata _returnvalue_3)
        (PCG|Metadata|CreateStringAttribute (LocalVariables|GetOutMetadata) (LocalVariables|GetOutAttribute) "No Strings Found!" false true))
      (:then_1
        (bind _returnvalue_4 (PCG|Metadata|GetStringAttributebyMetadataKey 0 (LocalVariables|GetInMetadata) (Utilities|String|StringToName (Variables|Default|GetSourceAttribute))))
        (for _array_element_2 (Utilities|String|ParseIntoArray _returnvalue_4 (Variables|Default|GetDelimiter))
          (if (not (Utilities|String|IsEmpty _array_element_2))
            (bind _returnvalue_5 (PCG|Metadata|AddEntry (LocalVariables|GetOutMetadata) -1))
            (if (Variables|Default|GetTrim)
              (PCG|Metadata|SetStringAttributebyMetadataKey _returnvalue_5 _outmetadata _outattribute (Utilities|String|TrimTrailing (Utilities|String|Trim _array_element_2)))
              (else
                (PCG|Metadata|SetStringAttributebyMetadataKey _returnvalue_5 _outmetadata _outattribute _array_element_2)))))
        (Utilities|Array|Add (LocalVariables|GetOutData) (PCG|TaggedData|MakeTaggedData _returnvalue_2)))))
  (return (Utilities|Struct|MakePCGDataCollection (LocalVariables|GetOutData))))

(fn NodeColorOverride
  (return "(R=1.000000,G=0.600000,B=0.000000,A=1.000000)"))

(fn NodeTitleOverride
  (return "Parse String"))
