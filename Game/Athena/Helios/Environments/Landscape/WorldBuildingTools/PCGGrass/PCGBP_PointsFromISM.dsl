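; Custom PCG node "Points From ISM" that converts hand-authored instances on a BP_MeshGrass_CustomGenerator into point data.
; Each populated, non-PCG-generated ISM component yields one point data entry, with transforms composed with the component's relative transform and seeds derived from position.
; Metadata attributes carry mesh, weighting, changed materials for the first two slots, culling and WPO distances, interactivity and shadow casting.
; The generator's IO component encodes interactivity, surface and resource-drop settings into SDPTags.
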(fn ExecuteWithContext
  (params (InContext :type PCGContext%20Structure%20%28by%20ref%29) (Input :type PCGData%20Collection%20Structure%20%28by%20ref%29))
  (bind _currentism (LocalVariables|GetCurrentISM))
  (bind (_outstartculldistance _outendculldistance) (Components|InstancedStaticMesh|GetCullDistances _currentism))
  (LocalVariables|SetContext InContext)
  (Utilities|FlowControl|Sequence
    (:then_1
      (bind _returnvalue (PCG|Temporary|GetComponent (LocalVariables|GetContext)))
      (bind _asbp_mesh_grass_custom_generator (Utilities|Casting|CastToBP_MeshGrass_CustomGenerator (Components|GetOwner _returnvalue)))
      (LocalVariables|SetIOComponent (Variables|/Game/Athena/Helios/Environments/Landscape/WorldBuildingTools/PCGGrass/BP_MeshGrass_CustomGenerator.BP_MeshGrass_CustomGenerator_C|GetBP_MeshGrassIOComponent _asbp_mesh_grass_custom_generator))
      (for _array_element (Actor|GetComponentsByClass _asbp_mesh_grass_custom_generator "/Script/Engine.InstancedStaticMeshComponent")
        (if (and (not (Components|ComponentHasTag _array_element "PCG Generated Component")) (Math|Conversions|ToBoolean(Integer) (Components|InstancedStaticMesh|GetInstanceCount _array_element)))
          (bind _output_get (LocalVariables|SetCurrentISM _array_element))
          (bind _outtransforms (Components|InstancedStaticMesh|GetInstanceTransforms _output_get))
          (bind _output_get_2 (Variables|Default|SetCurrentInstanceTransforms _outtransforms))
          (Variables|Default|SetCurrentComponentTransform (Transformation|GetRelativeTransform _output_get))
          (Utilities|FlowControl|Sequence
            (:then_0
              (Utilities|Array|Resize (LocalVariables|GetPoints) (Utilities|Array|Length _output_get_2))
              (bind _returnvalue_2 (Game|ConstructObjectfromClass "/Script/PCG.PCGPointData"))
              (bind _output_get_3 (LocalVariables|SetOutPointData _returnvalue_2))
              (SpatialData|SetPoints _output_get_3 (LocalVariables|GetPoints))
              (bind _returnvalue_3 (Metadata|MutableMetadata _returnvalue_2))
              (PCG|Metadata|CreateSoftObjectPathAttribute _returnvalue_3 "Mesh" (Utilities|GetSoftObjectPath (Variables|/Script/Engine.StaticMeshComponent|GetStaticMesh (LocalVariables|GetCurrentISM))) false true)
              (PCG|Metadata|CreateDoubleAttribute _returnvalue_3 "Weight" 1.0 false true)
              (PCG|Metadata|CreateDoubleAttribute _returnvalue_3 "Dominance" 1.0 false true)
              (bind (_overridematerial _overridematerial1) (CallFunction|ProcessOverrideMaterials (LocalVariables|GetCurrentISM)))
              (PCG|Metadata|CreateSoftObjectPathAttribute _returnvalue_3 "OverrideMaterial" (Utilities|MakeSoftObjectPath _overridematerial) false true)
              (PCG|Metadata|CreateSoftObjectPathAttribute _returnvalue_3 "OverrideMaterial1" (Utilities|MakeSoftObjectPath _overridematerial1) false true)
              (PCG|Metadata|CreateInteger32Attribute _returnvalue_3 "InstanceStartCullDistance" _outstartculldistance false true)
              (PCG|Metadata|CreateInteger32Attribute _returnvalue_3 "InstanceEndCullDistance" _outendculldistance false true)
              (PCG|Metadata|CreateInteger32Attribute _returnvalue_3 "WPODisableDistance" (Variables|/Script/Engine.StaticMeshComponent|GetWorldPositionOffsetDisableDistance _currentism) false true)
              (PCG|Metadata|CreateInteger32Attribute _returnvalue_3 "Interactivity" (Math|Conversions|ToInteger(Byte) (Variables|/Game/Athena/Helios/Environments/Landscape/WorldBuildingTools/PCGGrass/BP_MeshGrassInstancedStaticMeshComponent.BP_MeshGrassInstancedStaticMeshComponent_C|GetInteractivity%20Type (Utilities|Casting|CastToBP_MeshGrassInstancedStaticMeshComponent (LocalVariables|GetCurrentISM)))) false true)
              (PCG|Metadata|CreateInteger32Attribute _returnvalue_3 "RenderingQuality" (Math|Conversions|ToInteger(Boolean) (Variables|/Script/Engine.PrimitiveComponent|GetCastShadow (LocalVariables|GetCurrentISM))) false true)
              (bind _asbp_mesh_grass_instanced_static_mesh_component (Utilities|Casting|CastToBP_MeshGrassInstancedStaticMeshComponent (LocalVariables|GetCurrentISM))
                (:then
                  (LocalVariables|SetSDPTags (Class|BPMeshGrassIOComponent|EncodeSDPData (LocalVariables|GetIOComponent) (Variables|/Game/Athena/Helios/Environments/Landscape/WorldBuildingTools/PCGGrass/BP_MeshGrassInstancedStaticMeshComponent.BP_MeshGrassInstancedStaticMeshComponent_C|GetInteractivity%20Type _asbp_mesh_grass_instanced_static_mesh_component) (Variables|/Game/Athena/Helios/Environments/Landscape/WorldBuildingTools/PCGGrass/BP_MeshGrassInstancedStaticMeshComponent.BP_MeshGrassInstancedStaticMeshComponent_C|GetSurface%20Type _asbp_mesh_grass_instanced_static_mesh_component) (Variables|/Game/Athena/Helios/Environments/Landscape/WorldBuildingTools/PCGGrass/BP_MeshGrassInstancedStaticMeshComponent.BP_MeshGrassInstancedStaticMeshComponent_C|GetResource%20Drop _asbp_mesh_grass_instanced_static_mesh_component))))
                (:CastFailed
                  (LocalVariables|SetSDPTags "")))
              (PCG|Metadata|CreateStringAttribute _returnvalue_3 "SDPTags" (LocalVariables|GetSDPTags) false true))
            (:then_1
              (PCG|FlowControl|IterationLoop (LocalVariables|GetContext) (Math|Conversions|ToInteger64(Integer) (Utilities|Array|Length (LocalVariables|GetPoints))) (LocalVariables|GetOutPointData) 0 (LocalVariables|GetOutPointData))
              (Utilities|Array|Add (LocalVariables|GetOutData) (PCG|TaggedData|MakeTaggedData :Data (LocalVariables|GetOutPointData)))))))
      (return (Utilities|Struct|MakePCGDataCollection (LocalVariables|GetOutData))))))

(fn IterationLoopBody
  (params (InContext :type PCGContext%20Structure%20%28by%20ref%29) (Iteration :type Integer64) (InA :type PCGSpatial%20Data%20Object%20Reference) (InB :type PCGSpatial%20Data%20Object%20Reference) (OutMetadata :type PCGMetadata%20Object%20Reference))
  (bind _returnvalue (Math|Conversions|ToInteger(Integer64) Iteration))
  (bind _output_get (LocalVariables|SetCurrentPoint (SpatialData|GetPoint (SpatialData|ToPointDatawithContext InA InContext) _returnvalue)))
  (bind _structout (Utilities|Struct|SetmembersinPCGPoint _output_get (Math|Transform|ComposeTransforms (Utilities|Array|Get(aref) (Variables|Default|GetCurrentInstanceTransforms) _returnvalue) (Variables|Default|GetCurrentComponentTransform))))
  (PCG|Helpers|SetSeedfromPosition _structout)
  (return (LocalVariables|GetCurrentPoint) true))

(fn NodeTitleOverride
  (return "Points From ISM"))

(fn ProcessOverrideMaterials
  (params (self2 :type Instanced%20Static%20Mesh%20Component%20Object%20Reference))
  (bind _staticmesh (Variables|/Script/Engine.StaticMeshComponent|GetStaticMesh self2))
  (bind _returnvalue (Utilities|String|SelectString "" (Utilities|GetObjectPathString (Rendering|Material|GetMaterial self2 0)) (== (StaticMesh|GetMaterial _staticmesh 0) (Rendering|Material|GetMaterial self2 0))))
  (if (== (Rendering|Material|GetNumMaterials self2) 1)
    (return _returnvalue "")
    (else
      (bind _returnvalue_2 (Utilities|String|SelectString "" (Utilities|GetObjectPathString (Rendering|Material|GetMaterial self2 1)) (== (StaticMesh|GetMaterial _staticmesh 1) (Rendering|Material|GetMaterial self2 1))))
      (return _returnvalue _returnvalue_2))))
