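; PCG Blueprint node, "Rename Attribute Set", that batch-renames the attributes of incoming attribute sets, located in the Helios landscape PCG grass tools.
; For each input, ExecuteWithContext creates a new param data with one entry, adds the source's attributes, renames each, and outputs the collection.
; Names are optionally truncated first through ProcessTruncation (left or right, keep or chop at an index), then given an optional prefix and suffix.
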(fn ExecuteWithContext
  (params (InContext :type PCGContext%20Structure%20%28by%20ref%29) (Input :type PCGData%20Collection%20Structure%20%28by%20ref%29))
  (bind _returnvalue (Data|GetParams Input))
  (for _array_element _returnvalue
    (Utilities|FlowControl|Sequence
      (:then_0
        (bind _returnvalue_2 (Game|ConstructObjectfromClass "/Script/PCG.PCGParamData"))
        (bind _output_get (LocalVariables|SetOutAttributeSet _returnvalue_2))
        (bind _returnvalue_3 (Metadata|MutableMetadata _output_get))
        (bind _output_get_2 (LocalVariables|SetOutMetadata _returnvalue_3))
        (bind _returnvalue_4 (PCG|Metadata|AddEntry _output_get_2 -1))
        (LocalVariables|SetOutMetaDataEntry _returnvalue_4))
      (:then_1
        (bind _aspcgattribute_set (Utilities|Casting|CastToPCGParamData (PCG|TaggedData|BreakTaggedData _array_element)))
        (PCG|Metadata|AddAttributes (LocalVariables|GetOutMetadata) (Metadata|ConstMetadata _aspcgattribute_set))
        (for _array_element_2 (PCG|Metadata|GetAttributes (Metadata|ConstMetadata _aspcgattribute_set))
          (LocalVariables|SetNewAttributeString (Utilities|String|ToString(Name) _array_element_2))
          (if (Variables|Truncation|GetUseTruncation)
            (LocalVariables|SetNewAttributeString (CallFunction|ProcessTruncation (LocalVariables|GetNewAttributeString))))
          (if (Variables|Prefix&Suffix|GetUsePrefix)
            (LocalVariables|SetNewAttributeString (Utilities|String|Append (Variables|Prefix&Suffix|GetPrefix) (LocalVariables|GetNewAttributeString))))
          (if (Variables|Prefix&Suffix|GetUseSuffix)
            (LocalVariables|SetNewAttributeString (Utilities|String|Append (LocalVariables|GetNewAttributeString) (Variables|Prefix&Suffix|GetSuffix))))
          (PCG|Metadata|RenameAttribute (LocalVariables|GetOutMetadata) _array_element_2 (Utilities|String|StringToName (LocalVariables|GetNewAttributeString))))
        (Utilities|Array|Add (LocalVariables|GetOutData) (PCG|TaggedData|MakeTaggedData (LocalVariables|GetOutAttributeSet))))))
  (return (Utilities|Struct|MakePCGDataCollection (LocalVariables|GetOutData))))

(fn NodeColorOverride
  (return "(R=1.000000,G=0.600000,B=0.000000,A=1.000000)"))

(fn NodeTitleOverride
  (return "Rename Attribute Set"))

(fn ProcessTruncation
  (params (InString :type String))
  (bind _chop (Variables|Truncation|GetChop))
  (bind _truncationindex (Variables|Truncation|GetTruncationIndex))
  (if (Variables|Truncation|GetReverseIndex)
    (if _chop
      (return (Utilities|String|RightChop InString _truncationindex))
      (else
        (return (Utilities|String|Right InString _truncationindex))))
    (elif _chop
      (return (Utilities|String|LeftChop InString _truncationindex))
      (else
        (return (Utilities|String|Left InString _truncationindex))))))
