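; Custom PCG Blueprint node, SoftObjectPathToString, that turns SoftObjectPath attributes in input param data into string attributes.
; For each input it builds new param data with one entry and a string attribute named by TargetAttribute.
; It writes entry 0's non-empty path from the matching attribute, or any soft-path attribute if TargetAttribute is empty.
; Other attribute names collect in a set that filters which attributes are added to the output metadata.
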(fn ExecuteWithContext
  (params (InContext :type PCGContext%20Structure%20%28by%20ref%29) (Input :type PCGData%20Collection%20Structure%20%28by%20ref%29))
  (bind _returnvalue (Data|GetParams Input))
  (for _array_element _returnvalue
    (Utilities|FlowControl|Sequence
      (:then_0
        (bind _returnvalue_2 (Game|ConstructObjectfromClass "/Script/PCG.PCGParamData"))
        (bind _returnvalue_3 (Metadata|MutableMetadata _returnvalue_2))
        (bind _output_get (LocalVariables|SetCurrentMetadata _returnvalue_3))
        (bind _returnvalue_4 (PCG|Metadata|AddEntry _output_get -1))
        (LocalVariables|SetCurrentMetaDataEntry _returnvalue_4)
        (Utilities|Array|Add (LocalVariables|GetOutData) (PCG|TaggedData|MakeTaggedData _returnvalue_2))
        (PCG|Metadata|CreateStringAttribute _output_get (Variables|Default|GetTargetAttribute) "" false true))
      (:then_1
        (bind _aspcgattribute_set (Utilities|Casting|CastToPCGParamData (PCG|TaggedData|BreakTaggedData _array_element)))
        (bind _output_get_2 (LocalVariables|SetInMetadata (Metadata|ConstMetadata _aspcgattribute_set)))
        (LocalVariables|SetUseTargetAttribute (not (== (Variables|Default|GetTargetAttribute) "")))
        (for (_array_element_2 _array_index) (PCG|Metadata|GetAttributes _output_get_2)
          (if (and (Utilities|Enum|Equal(Enum) (Utilities|Array|Get(acopy) (.AttributeTypes (PCG|Metadata|GetAttributes _output_get_2)) _array_index) "SoftObjectPath") (or (== _array_element_2 (Variables|Default|GetTargetAttribute)) (not (LocalVariables|GetUseTargetAttribute))))
            (bind _returnvalue_5 (PCG|Metadata|GetSoftObjectPathAttributebyMetadataKey 0 (LocalVariables|GetInMetadata) _array_element_2))
            (bind _returnvalue_pathstring (Utilities|BreakSoftObjectPath _returnvalue_5))
            (if (not (Utilities|String|IsEmpty _returnvalue_pathstring))
              (PCG|Metadata|SetStringAttributebyMetadataKey (LocalVariables|GetCurrentMetaDataEntry) (LocalVariables|GetCurrentMetadata) (Variables|Default|GetTargetAttribute) _returnvalue_pathstring))
            (else
              (Utilities|Set|Add (Variables|Default|GetIn%20Filtered%20Attributes) _array_element_2))))
        (PCG|Metadata|AddAttributesFiltered (LocalVariables|GetCurrentMetadata) (LocalVariables|GetInMetadata) (Variables|Default|GetIn%20Filtered%20Attributes) "IncludeAttributes" "Equal"))))
  (return (Utilities|Struct|MakePCGDataCollection (LocalVariables|GetOutData))))

(fn NodeColorOverride
  (return "(R=1.000000,G=0.600000,B=0.000000,A=1.000000)"))

(fn NodeTitleOverride
  (return "SoftObjectPathToString"))
