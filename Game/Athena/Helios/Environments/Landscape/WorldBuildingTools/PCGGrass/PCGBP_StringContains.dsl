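; Custom PCG Blueprint node "String Contains Substring" that tests string attributes in InString param data for a substring.
; Each input yields a new param data holding a ContainsSubstring bool.
; It reads String or SoftObjectPath values from the target attribute, or every attribute if none is named, where the last evaluation wins.
; With SeparateByComma set, the substring splits on commas and any match counts as true.
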(fn ExecuteWithContext
  (params (InContext :type PCGContext%20Structure%20%28by%20ref%29) (Input :type PCGData%20Collection%20Structure%20%28by%20ref%29))
  (bind _inmetadata (LocalVariables|GetInMetadata))
  (bind _outmetadataentry (LocalVariables|GetOutMetaDataEntry))
  (bind _returnvalue (Data|GetParamsbyPinLabel Input "InString"))
  (for _array_element _returnvalue
    (Utilities|FlowControl|Sequence
      (:then_0
        (bind _returnvalue_2 (Game|ConstructObjectfromClass "/Script/PCG.PCGParamData"))
        (bind _returnvalue_3 (Metadata|MutableMetadata _returnvalue_2))
        (bind _output_get (LocalVariables|SetOutMetadata _returnvalue_3))
        (bind _returnvalue_4 (PCG|Metadata|AddEntry _output_get -1))
        (LocalVariables|SetOutMetaDataEntry _returnvalue_4)
        (Utilities|Array|Add (LocalVariables|GetOutData) (PCG|TaggedData|MakeTaggedData _returnvalue_2)))
      (:then_1
        (bind _aspcgattribute_set (Utilities|Casting|CastToPCGParamData (PCG|TaggedData|BreakTaggedData _array_element)))
        (bind _output_get_2 (LocalVariables|SetInMetadata (Metadata|ConstMetadata _aspcgattribute_set)))
        (PCG|Metadata|CreateBoolAttribute (LocalVariables|GetOutMetadata) "ContainsSubstring" false false true)
        (for (_array_element_2 _array_index) (PCG|Metadata|GetAttributes _output_get_2)
          (if (or (== _array_element_2 (Variables|Default|GetTargetAttribute)) (== (Variables|Default|GetTargetAttribute) ""))
            (switch Utilities|FlowControl|Switch|SwitchonEPCGMetadataTypes (Utilities|Array|Get(acopy) (.AttributeTypes (PCG|Metadata|GetAttributes _output_get_2)) _array_index)
              (:String
                (bind _returnvalue_5 (PCG|Metadata|GetStringAttributebyMetadataKey _outmetadataentry _inmetadata _array_element_2))
                (LocalVariables|SetSourceString _returnvalue_5))
              (:SoftObjectPath
                (bind _returnvalue_6 (PCG|Metadata|GetSoftObjectPathAttributebyMetadataKey _outmetadataentry _inmetadata _array_element_2))
                (bind _returnvalue_pathstring (Utilities|BreakSoftObjectPath _returnvalue_6))
                (LocalVariables|SetSourceString _returnvalue_pathstring)))
            (if (Variables|Default|GetSeparateByComma)
              (CallFunction|SeparateSubstring)
              (bind _anymatch (CallFunction|TestSubstrings (LocalVariables|GetSourceString)))
              (PCG|Metadata|SetBoolAttributebyMetadataKey (LocalVariables|GetOutMetaDataEntry) (LocalVariables|GetOutMetadata) "ContainsSubstring" _anymatch)
              (else
                (PCG|Metadata|SetBoolAttributebyMetadataKey (LocalVariables|GetOutMetaDataEntry) (LocalVariables|GetOutMetadata) "ContainsSubstring" (Utilities|String|Contains (LocalVariables|GetSourceString) (Variables|Default|GetSubstring))))))))))
  (return (Utilities|Struct|MakePCGDataCollection (LocalVariables|GetOutData))))

(fn NodeColorOverride
  (return "(R=1.000000,G=0.600000,B=0.000000,A=1.000000)"))

(fn NodeTitleOverride
  (return "String Contains Substring"))

(fn SeparateSubstring
  (Variables|Default|SetSubstrings (Utilities|String|ParseIntoArray (Variables|Default|GetSubstring) ",")))

(fn TestSubstrings
  (params (SourceString :type String))
  (for _array_element (Variables|Default|GetSubstrings)
    (if (Utilities|String|Contains SourceString _array_element)
      (return true)))
  (return false))
