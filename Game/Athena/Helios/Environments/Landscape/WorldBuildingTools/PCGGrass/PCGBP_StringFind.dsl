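; PCG Blueprint node titled "String Find" that searches a string attribute for a substring, in the Helios landscape PCG grass tools.
; For each input attribute set, it reads the target attribute at metadata key 0 and runs Find Substring with configurable case and direction.
; The result becomes the default value of a "FoundIndex" integer attribute on a new single-entry attribute set.
; Missing params or a missing attribute end execution early.
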(fn ExecuteWithContext
  (params (InContext :type PCGContext%20Structure%20%28by%20ref%29) (Input :type PCGData%20Collection%20Structure%20%28by%20ref%29))
  (Utilities|FlowControl|Sequence
    (:then_0
      (bind _returnvalue (Game|ConstructObjectfromClass "/Script/PCG.PCGParamData"))
      (bind _returnvalue_2 (Metadata|MutableMetadata _returnvalue))
      (bind _output_get (LocalVariables|SetOutMetadata _returnvalue_2))
      (bind _returnvalue_3 (PCG|Metadata|AddEntry _output_get -1))
      (LocalVariables|SetOutMetaDataEntry _returnvalue_3))
    (:then_1
      (bind _returnvalue_4 (Data|GetParams Input))
      (if (Utilities|Array|IsNotEmpty _returnvalue_4)
        (for _array_element _returnvalue_4
          (bind _aspcgattribute_set (Utilities|Casting|CastToPCGParamData (PCG|TaggedData|BreakTaggedData _array_element)))
          (bind _output_get_2 (LocalVariables|SetInMetadata (Metadata|ConstMetadata _aspcgattribute_set)))
          (if (PCG|Metadata|HasAttribute _output_get_2 (Variables|Default|GetTargetAttribute))
            (bind _returnvalue_5 (PCG|Metadata|GetStringAttributebyMetadataKey 0 _output_get_2 (Variables|Default|GetTargetAttribute)))
            (PCG|Metadata|CreateInteger32Attribute (LocalVariables|GetOutMetadata) "FoundIndex" (Utilities|String|FindSubstring _returnvalue_5 (Variables|Default|GetSubString) (Variables|Default|GetUseCase) (Variables|Default|GetSearchFromEnd)) false true)
            (else
              (return 0))))
        (return (Utilities|Struct|MakePCGDataCollection (Utilities|Array|MakeArray (PCG|TaggedData|MakeTaggedData _returnvalue))))
        (else
          (return 0))))))

(fn NodeColorOverride
  (return "(R=1.000000,G=0.600000,B=0.000000,A=1.000000)"))

(fn NodeTitleOverride
  (return "String Find"))
