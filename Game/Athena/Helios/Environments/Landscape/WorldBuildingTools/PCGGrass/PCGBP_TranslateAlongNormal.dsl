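; PCG Blueprint node "Translate Along Normal", in the landscape grass tools, offsets points along their up axis by a per-point distance.
; ExecuteWithContext runs a point loop on each point data input that has the named distance attribute, omitting inputs without it from the output.
; PointLoopBody moves each point along its rotation's up vector by that attribute's value, keeping rotation and scale.
; Results are wrapped as new tagged data without copying input tags.
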(fn ExecuteWithContext
  (params (InContext :type PCGContext%20Structure%20%28by%20ref%29) (Input :type PCGData%20Collection%20Structure%20%28by%20ref%29))
  (LocalVariables|SetContext InContext)
  (bind (_outtaggeddata _returnvalue) (Data|GetTypedInputs Input "/Script/PCG.PCGPointData"))
  (for _array_element _returnvalue
    (if (PCG|Metadata|HasAttribute (Metadata|ConstMetadata _array_element) (Variables|Default|GetDistanceAttributeName))
      (bind _outdata (PCG|FlowControl|PointLoop (LocalVariables|GetContext) _array_element 0))
      (Utilities|Array|Add (LocalVariables|GetOutData) (PCG|TaggedData|MakeTaggedData :Data _outdata))))
  (return (Utilities|Struct|MakePCGDataCollection (LocalVariables|GetOutData))))

(fn NodeTitleOverride
  (return "Translate Along Normal"))

(fn PointLoopBody
  (params (InContext :type PCGContext%20Structure%20%28by%20ref%29) (InData :type PCGPoint%20Data%20Object%20Reference) (InPoint :type PCGPoint%20Structure%20%28by%20ref%29) (OutMetadata :type PCGMetadata%20Object%20Reference) (Iteration :type Integer64))
  (bind _returnvalue (PCG|Metadata|GetDoubleAttribute InPoint (Metadata|ConstMetadata InData) (Variables|Default|GetDistanceAttributeName)))
  (bind _structout (Utilities|Struct|SetmembersinPCGPoint InPoint (Math|Transform|MakeTransform (+ (* (Math|Vector|GetUpVector (.rotation (Utilities|Struct|BreakPCGPoint InPoint))) _returnvalue) (.location (Utilities|Struct|BreakPCGPoint InPoint))) (.rotation (Utilities|Struct|BreakPCGPoint InPoint)) (.scale (Utilities|Struct|BreakPCGPoint InPoint)))))
  (return _structout true))
