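; PCG Blueprint node "Trim String" that cleans whitespace from a string attribute in incoming attribute sets.
; For each input on "In", it reads the first entry of Source Attribute, falling back to the first string attribute found when the name is empty.
; Remove All Whitespaces chooses between stripping spaces and trimming both ends.
; The result goes into newly constructed param data, which is output only when the source attribute exists.
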(fn ExecuteWithContext
  (params (InContext :type PCGContext%20Structure%20%28by%20ref%29) (Input :type PCGData%20Collection%20Structure%20%28by%20ref%29))
  (bind _inmetadata (LocalVariables|GetInMetadata))
  (bind _source_attribute (Variables|Default|GetSource%20Attribute))
  (bind _returnvalue (Data|GetInputsbyPinLabel Input "In"))
  (for _array_element _returnvalue
    (Utilities|FlowControl|Sequence
      (:then_0
        (bind _aspcgattribute_set (Utilities|Casting|CastToPCGParamData (PCG|TaggedData|BreakTaggedData _array_element)))
        (LocalVariables|SetInMetadata (Metadata|ConstMetadata _aspcgattribute_set))
        (bind _returnvalue_2 (Game|ConstructObjectfromClass "/Script/PCG.PCGParamData"))
        (bind _returnvalue_3 (Metadata|MutableMetadata _returnvalue_2))
        (LocalVariables|SetOutMetadata _returnvalue_3)
        (PCG|Metadata|CreateStringAttribute (LocalVariables|GetOutMetadata) (LocalVariables|GetOutAttribute) "No Strings Found!" false true))
      (:then_1
        (if (== (Variables|Default|GetSource%20Attribute) "")
          (bind _stringattribute (CallFunction|FindStringAttribute (LocalVariables|GetInMetadata)))
          (Variables|Default|SetSource%20Attribute _stringattribute))
        (if (PCG|Metadata|HasAttribute _inmetadata _source_attribute)
          (bind _returnvalue_4 (PCG|Metadata|GetStringAttributebyMetadataKey 0 _inmetadata _source_attribute))
          (bind _returnvalue_5 (PCG|Metadata|AddEntry (LocalVariables|GetOutMetadata) -1))
          (PCG|Metadata|SetStringAttributebyMetadataKey _returnvalue_5 (LocalVariables|GetOutMetadata) (LocalVariables|GetOutAttribute) (Utilities|String|SelectString (Utilities|String|Replace _returnvalue_4 " ") (Utilities|String|TrimTrailing (Utilities|String|Trim _returnvalue_4)) (Variables|Default|GetRemove%20All%20Whitespaces)))
          (Utilities|Array|Add (LocalVariables|GetOutData) (PCG|TaggedData|MakeTaggedData _returnvalue_2))))))
  (return (Utilities|Struct|MakePCGDataCollection (LocalVariables|GetOutData))))

(fn FindStringAttribute
  (params (Metadata :type PCGMetadata%20Object%20Reference))
  (bind (_attributenames _attributetypes) (PCG|Metadata|GetAttributes Metadata))
  (for (_array_element _array_index) _attributetypes
    (if (Utilities|Enum|Equal(Enum) _array_element "String")
      (return (Utilities|Array|Get(aref) _attributenames _array_index) true)))
  (return "None" false))

(fn NodeColorOverride
  (return "(R=1.000000,G=0.600000,B=0.000000,A=1.000000)"))

(fn NodeTitleOverride
  (return "Trim String"))
