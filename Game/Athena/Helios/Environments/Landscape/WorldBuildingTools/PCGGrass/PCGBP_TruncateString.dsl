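; Custom PCG Blueprint node "Truncate String", in the PCG grass world-building tools, that shortens text values in attribute sets.
; It builds a new param data with one entry, copies attributes from each input set, and rewrites values read from key 0.
; String attributes are overwritten, while soft object paths are recreated as truncated string attributes.
; ProcessString applies Left, LeftChop, Right or RightChop at TruncationIndex, selected by the Chop and ReverseIndex variables.
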(fn ExecuteWithContext
  (params (InContext :type PCGContext%20Structure%20%28by%20ref%29) (Input :type PCGData%20Collection%20Structure%20%28by%20ref%29))
  (bind _sourcestring (LocalVariables|GetSourceString))
  (bind _outstring (CallFunction|ProcessString _sourcestring))
  (bind _outmetadata (LocalVariables|GetOutMetadata))
  (bind _inmetadata (LocalVariables|GetInMetadata))
  (Utilities|FlowControl|Sequence
    (:then_0
      (bind _returnvalue (Game|ConstructObjectfromClass "/Script/PCG.PCGParamData"))
      (bind _returnvalue_2 (Metadata|MutableMetadata _returnvalue))
      (bind _output_get (LocalVariables|SetOutMetadata _returnvalue_2))
      (bind _returnvalue_3 (PCG|Metadata|AddEntry _output_get -1))
      (LocalVariables|SetOutMetaDataEntry _returnvalue_3))
    (:then_1
      (bind _returnvalue_4 (Data|GetParams Input))
      (if (Utilities|Array|IsNotEmpty _returnvalue_4)
        (for _array_element _returnvalue_4
          (bind _aspcgattribute_set (Utilities|Casting|CastToPCGParamData (PCG|TaggedData|BreakTaggedData _array_element)))
          (bind _output_get_2 (LocalVariables|SetInMetadata (Metadata|ConstMetadata _aspcgattribute_set)))
          (PCG|Metadata|AddAttributes (LocalVariables|GetOutMetadata) _output_get_2)
          (for (_array_element_2 _array_index) (PCG|Metadata|GetAttributes _output_get_2)
            (switch Utilities|FlowControl|Switch|SwitchonEPCGMetadataTypes (Utilities|Array|Get(acopy) (.AttributeTypes (PCG|Metadata|GetAttributes _output_get_2)) _array_index)
              (:String
                (bind _returnvalue_5 (PCG|Metadata|GetStringAttributebyMetadataKey 0 _inmetadata _array_element_2))
                (LocalVariables|SetSourceString _returnvalue_5)
                (PCG|Metadata|SetStringAttributebyMetadataKey (LocalVariables|GetOutMetaDataEntry) _outmetadata _array_element_2 _outstring))
              (:SoftObjectPath
                (bind _returnvalue_6 (PCG|Metadata|GetSoftObjectPathAttributebyMetadataKey 0 _inmetadata _array_element_2))
                (bind _returnvalue_pathstring (Utilities|BreakSoftObjectPath _returnvalue_6))
                (LocalVariables|SetSourceString _returnvalue_pathstring)
                (PCG|Metadata|DeleteAttribute _outmetadata _array_element_2)
                (PCG|Metadata|CreateStringAttribute _outmetadata _array_element_2 _outstring false true)))))
        (return (Utilities|Struct|MakePCGDataCollection (Utilities|Array|MakeArray (PCG|TaggedData|MakeTaggedData _returnvalue))))
        (else
          (return 0))))))

(fn NodeColorOverride
  (return "(R=1.000000,G=0.600000,B=0.000000,A=1.000000)"))

(fn NodeTitleOverride
  (return "Truncate String"))

(fn ProcessString
  (params (InString :type String))
  (bind _chop (Variables|Default|GetChop))
  (bind _truncationindex (Variables|Default|GetTruncationIndex))
  (if (Variables|Default|GetReverseIndex)
    (if _chop
      (return (Utilities|String|RightChop InString _truncationindex))
      (else
        (return (Utilities|String|Right InString _truncationindex))))
    (elif _chop
      (return (Utilities|String|LeftChop InString _truncationindex))
      (else
        (return (Utilities|String|Left InString _truncationindex))))))
