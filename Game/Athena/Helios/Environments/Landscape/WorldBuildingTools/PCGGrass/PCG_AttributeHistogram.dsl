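; PCG subgraph in the PCG grass world-building tools that serves as a loop body for building a histogram of a named point attribute over a 0 to 1 range.
; It turns points into an attribute set, keeps only the target attribute and sorts it.
; The loop index and bucket data count define each iteration's equal-width bounds.
; A range filter keeps values inside those bounds, and their element count becomes that bucket's output.
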(graph PCG_AttributeHistogram
  (params (TargetAttribute :type String))
  (inputs (InPoint :type Point) (HistogramBucketsLooping :type Param) (HistogramBuckets :type Param))
  (outputs (OutBuckets :type Param))
  (bind PointToAttributeSet_15 (Point%20To%20Attribute%20Set))
  (bind DeleteAttributes_1 (Delete%20Attributes :Operation KeepSelectedAttributes :SelectedAttributes-param TargetAttribute))
  (bind SortAttributes_0 (Sort%20Attributes :bUseStableSort false))
  (bind CreateAttribute_33 (Create%20Constant :AttributeTypes (struct :DoubleValue 1.0)))
  (bind DataNum_34 (Data%20Count))
  (bind AttributeMathsOp_4 (Attribute%20Maths%20Op :Operation Divide))
  (bind CreateAttribute_10 (Create%20Constant :AttributeTypes (struct :DoubleValue 1.0)))
  (bind GetLoopIndex_8 (Get%20Loop%20Index))
  (bind AttributeMathsOp_9 (Attribute%20Maths%20Op))
  (bind AttributeMathsOp_7 (Attribute%20Maths%20Op :Operation Multiply))
  (bind AttributeMathsOp_13 (Attribute%20Maths%20Op :Operation Multiply))
  (bind AttributeFilterRange_2 (Filter%20Attribute%20Elements%20by%20Range :MaxThreshold (struct :ThresholdAttribute @Last :bUseSpatialQuery false) :MinThreshold (struct :ThresholdAttribute @Last :bUseSpatialQuery false) :TargetAttribute @Last))
  (bind GetEntriesCount_19 (Get%20Element%20Count))
  (connect AttributeMathsOp_7 Out AttributeFilterRange_2 FilterMax)
  (connect AttributeMathsOp_13 Out AttributeFilterRange_2 FilterMin)
  (connect SortAttributes_0 Out AttributeFilterRange_2 In)
  (connect CreateAttribute_33 Out AttributeMathsOp_4 InA)
  (connect DataNum_34 Out AttributeMathsOp_4 InB)
  (connect AttributeMathsOp_4 Out AttributeMathsOp_7 InA)
  (connect AttributeMathsOp_9 Out AttributeMathsOp_7 InB)
  (connect GetLoopIndex_8 Out AttributeMathsOp_9 InA)
  (connect CreateAttribute_10 Out AttributeMathsOp_9 InB)
  (connect AttributeMathsOp_4 Out AttributeMathsOp_13 InA)
  (connect GetLoopIndex_8 Out AttributeMathsOp_13 InB)
  (connect DefaultInputNode HistogramBuckets DataNum_34 In)
  (connect PointToAttributeSet_15 Out DeleteAttributes_1 In)
  (connect AttributeFilterRange_2 InsideFilter GetEntriesCount_19 In)
  (connect DefaultInputNode InPoint PointToAttributeSet_15 In)
  (connect DeleteAttributes_1 Out SortAttributes_0 In)
  (output OutBuckets GetEntriesCount_19))
