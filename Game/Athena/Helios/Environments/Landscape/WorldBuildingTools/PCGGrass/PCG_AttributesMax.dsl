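; PCG subgraph in the Helios landscape grass tools that merges attributes on points by taking their maximum, writing the result to LayerExclusion.
; SourceAttributes drives a filter that splits the incoming point data by attribute; data outside the filter passes straight to the output.
; Matching data supplies the attribute named by TargetAttribute to a Max operation.
; The second operand comes from a copy of the points stripped down to only the SourceAttributes selection.
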(graph PCG_AttributesMax
  (params (TargetAttribute :type String))
  (inputs (InPoints :type Point) (SourceAttributes :type Param))
  (outputs (InPoints :type Point))
  (bind DeleteAttributes_0 (Delete%20Attributes :Operation KeepSelectedAttributes))
  (bind FilterDataByAttribute_5 (Filter%20Data%20By%20Attribute))
  (bind AttributeMathsOp_2 (Attribute%20Maths%20Op :Operation Max :OutputTarget LayerExclusion :InputSource1-param TargetAttribute))
  (connect FilterDataByAttribute_5 InsideFilter AttributeMathsOp_2 InA)
  (connect DeleteAttributes_0 Out AttributeMathsOp_2 InB)
  (connect DefaultInputNode InPoints DeleteAttributes_0 In)
  (connect DefaultInputNode SourceAttributes DeleteAttributes_0 SelectedAttributes)
  (connect DefaultInputNode SourceAttributes FilterDataByAttribute_5 Attribute)
  (connect DefaultInputNode InPoints FilterDataByAttribute_5 In)
  (output InPoints AttributeMathsOp_2)
  (output InPoints FilterDataByAttribute_5.OutsideFilter))
