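; PCG utility subgraph in the Helios landscape grass tools that routes Payload point data to Success or Fail based on whether Condition contains any data.
; Data Count of Condition is compared NotEqual to a default-valued constant (zero), and the result drives the Branch's output selection.
; A nonzero count sends Payload to Success, otherwise to Fail.
; The check counts data entries on the pin, not the points inside them.
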(graph PCG_BranchValidCondition
  (inputs (Payload :type Point) (Condition :type Point))
  (outputs (Success :type Point) (Fail :type Point))
  (bind CreateAttribute_2 (Create%20Constant))
  (bind DataNum_0 (Data%20Count))
  (bind AttributeCompareOp_1 (Attribute%20Compare%20Op :Operation NotEqual))
  (bind Branch_0 (Branch))
  (connect DataNum_0 Out AttributeCompareOp_1 InA)
  (connect CreateAttribute_2 Out AttributeCompareOp_1 InB)
  (connect AttributeCompareOp_1 Out Branch_0 bOutputToB)
  (connect DefaultInputNode Payload Branch_0 In)
  (connect DefaultInputNode Condition DataNum_0 In)
  (output Fail Branch_0.Output%20A)
  (output Success Branch_0.Output%20B))
