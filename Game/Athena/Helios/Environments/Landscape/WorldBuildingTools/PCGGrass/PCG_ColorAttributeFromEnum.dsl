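; PCG subgraph for the Helios grass world-building tools that sets point density from chosen point color channels.
; An EMeshGrassColorChannel parameter selects a Vector4 mask for a single channel or channel combination, with red as default.
; The mask multiplies the point color, copied into a temporary MaxRGBA attribute.
; Chained Max operations write the largest remaining component to Density, then the temporary attribute is deleted.
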(graph PCG_ColorAttributeFromEnum
  (inputs (Color%20Enum :type Param))
  (outputs (Delete%20Attributes%20Out :type Point))
  (bind MakeVectorAttribute_0 (Make%20Vector%20Attribute :InputSource1 $Color.R :InputSource2 $Color.G :InputSource3 $Color.B :InputSource4 $Color.A :OutputTarget MaxRGBA :OutputType Vector4))
  (bind CreateConstant_0 (Create%20Constant :AttributeTypes (struct :Type Vector4 :Vector4Value (1.0 1.0 1.0 0.0)) :OutputTarget RGB))
  (bind CreateConstant_1 (Create%20Constant :AttributeTypes (struct :Type Vector4 :Vector4Value (1.0 0.0 0.0 0.0)) :OutputTarget R))
  (bind CreateConstant_2 (Create%20Constant :AttributeTypes (struct :Type Vector4 :Vector4Value (0.0 1.0 0.0 0.0)) :OutputTarget G))
  (bind CreateConstant_3 (Create%20Constant :AttributeTypes (struct :Type Vector4 :Vector4Value (0.0 0.0 1.0 0.0)) :OutputTarget B))
  (bind CreateConstant_4 (Create%20Constant :AttributeTypes (struct :Type Vector4 :Vector4Value (0.0 0.0 0.0 1.0)) :OutputTarget A))
  (bind CreateConstant_5 (Create%20Constant :AttributeTypes (struct :Type Vector4 :Vector4Value (1.0 1.0 1.0 1.0)) :OutputTarget RGBA))
  (bind CreateConstant_6 (Create%20Constant :AttributeTypes (struct :Type Vector4 :Vector4Value (1.0 0.0 1.0 0.0)) :OutputTarget RB))
  (bind CreateConstant_7 (Create%20Constant :AttributeTypes (struct :Type Vector4 :Vector4Value (1.0 0.0 1.0 1.0)) :OutputTarget RBA))
  (bind CreateConstant_8 (Create%20Constant :AttributeTypes (struct :Type Vector4 :Vector4Value (1.0 1.0 0.0 1.0)) :OutputTarget RGA))
  (bind CreateConstant_9 (Create%20Constant :AttributeTypes (struct :Type Vector4 :Vector4Value (0.0 1.0 1.0 1.0)) :OutputTarget GBA))
  (bind CreateConstant_10 (Create%20Constant :AttributeTypes (struct :Type Vector4 :Vector4Value (1.0 1.0 0.0 0.0)) :OutputTarget RG))
  (bind CreateConstant_11 (Create%20Constant :AttributeTypes (struct :Type Vector4 :Vector4Value (1.0 0.0 0.0 1.0)) :OutputTarget RA))
  (bind CreateConstant_12 (Create%20Constant :AttributeTypes (struct :Type Vector4 :Vector4Value (0.0 1.0 1.0 0.0)) :OutputTarget GB))
  (bind CreateConstant_13 (Create%20Constant :AttributeTypes (struct :Type Vector4 :Vector4Value (0.0 1.0 0.0 1.0)) :OutputTarget GA))
  (bind CreateConstant_14 (Create%20Constant :AttributeTypes (struct :Type Vector4 :Vector4Value (0.0 0.0 1.0 1.0)) :OutputTarget BA))
  (bind Select%20%28Multi%29_0 (Select%20%28Multi%29 :EnumSelection (struct :Class /Game/Athena/Helios/Environments/Landscape/WorldBuildingTools/PCGGrass/EMeshGrassColorChannel.EMeshGrassColorChannel) :IntOptions [] :SelectionMode Enum))
  (bind AttributeMathsOp_0 (Attribute%20Maths%20Op :Operation Multiply :OutputTarget MaxRGBA))
  (bind AttributeMathsOp_1 (Attribute%20Maths%20Op :InputSource1 MaxRGBA.X :InputSource2 MaxRGBA.Y :Operation Max :OutputTarget $Density))
  (bind AttributeMathsOp_2 (Attribute%20Maths%20Op :InputSource2 MaxRGBA.Z :Operation Max :OutputTarget $Density))
  (bind AttributeMathsOp_3 (Attribute%20Maths%20Op :InputSource2 MaxRGBA.W :Operation Max :OutputTarget $Density))
  (bind DeleteAttributes_0 (Delete%20Attributes :SelectedAttributes "MaxRGBA"))
  (connect MakeVectorAttribute_0 Out AttributeMathsOp_0 InA)
  (connect Select%20%28Multi%29_0 Out AttributeMathsOp_0 InB)
  (connect AttributeMathsOp_0 Out AttributeMathsOp_1 InA)
  (connect AttributeMathsOp_0 Out AttributeMathsOp_1 InB)
  (connect AttributeMathsOp_1 Out AttributeMathsOp_2 InA)
  (connect AttributeMathsOp_1 Out AttributeMathsOp_2 InB)
  (connect AttributeMathsOp_2 Out AttributeMathsOp_3 InA)
  (connect AttributeMathsOp_2 Out AttributeMathsOp_3 InB)
  (connect AttributeMathsOp_3 Out DeleteAttributes_0 In)
  (connect DefaultInputNode In MakeVectorAttribute_0 W)
  (connect DefaultInputNode In MakeVectorAttribute_0 X)
  (connect DefaultInputNode In MakeVectorAttribute_0 Y)
  (connect DefaultInputNode In MakeVectorAttribute_0 Z)
  (connect CreateConstant_4 Out Select%20%28Multi%29_0 Alpha)
  (connect CreateConstant_3 Out Select%20%28Multi%29_0 Blue)
  (connect CreateConstant_1 Out Select%20%28Multi%29_0 Default)
  (connect CreateConstant_2 Out Select%20%28Multi%29_0 Green)
  (connect CreateConstant_14 Out Select%20%28Multi%29_0 Max%20BA)
  (connect CreateConstant_13 Out Select%20%28Multi%29_0 Max%20GA)
  (connect CreateConstant_12 Out Select%20%28Multi%29_0 Max%20GB)
  (connect CreateConstant_11 Out Select%20%28Multi%29_0 Max%20RA)
  (connect CreateConstant_6 Out Select%20%28Multi%29_0 Max%20RB)
  (connect CreateConstant_10 Out Select%20%28Multi%29_0 Max%20RG)
  (connect CreateConstant_0 Out Select%20%28Multi%29_0 Max%20RGB)
  (connect CreateConstant_5 Out Select%20%28Multi%29_0 Max%20RGBA)
  (connect CreateConstant_9 Out Select%20%28Multi%29_0 MaxGBA)
  (connect CreateConstant_7 Out Select%20%28Multi%29_0 MaxRBA)
  (connect CreateConstant_8 Out Select%20%28Multi%29_0 MaxRGA)
  (connect CreateConstant_1 Out Select%20%28Multi%29_0 Red)
  (connect DefaultInputNode Color%20Enum Select%20%28Multi%29_0 Value)
  (output Delete%20Attributes%20Out DeleteAttributes_0))
