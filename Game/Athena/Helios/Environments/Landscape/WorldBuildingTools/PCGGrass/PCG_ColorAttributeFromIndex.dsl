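; PCG subgraph in the grass world-building tools that converts an integer Index into a color attribute selector string.
; An integer multi-select maps the index to R, G, B, A, MaxRGB or MaxRGBA, defaulting to R.
; Channel names get a "$Color." prefix, but a greater-than-3 comparison passes MaxRGB and MaxRGBA through unprefixed to the ColorAttribute output.
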(graph PCG_ColorAttributeFromIndex
  (inputs (Index :type Param))
  (outputs (ColorAttribute :type Param))
  (bind CreateAttribute_1 (Create%20Constant :AttributeTypes (struct :Type String :StringValue "R")))
  (bind CreateAttribute_2 (Create%20Constant :AttributeTypes (struct :Type String :StringValue "G")))
  (bind CreateAttribute_3 (Create%20Constant :AttributeTypes (struct :Type String :StringValue "B")))
  (bind CreateAttribute_4 (Create%20Constant :AttributeTypes (struct :Type String :StringValue "A")))
  (bind CreateAttribute_7 (Create%20Constant :AttributeTypes (struct :Type String :StringValue "MaxRGB")))
  (bind CreateAttribute_8 (Create%20Constant :AttributeTypes (struct :Type String :StringValue "MaxRGBA")))
  (bind Select%20%28Multi%29_0 (Select%20%28Multi%29 :EnumSelection (struct :Class /Game/Blueprints/Enum/2SidedEnum.2SidedEnum) :IntOptions [0 1 2 3 4 5]))
  (bind CreateAttribute_6 (Create%20Constant :AttributeTypes (struct :Type String :StringValue "$Color.")))
  (bind AttributeStringOp_5 (Attribute%20String%20Op))
  (bind CreateAttribute_5 (Create%20Constant :AttributeTypes (struct :Type Integer32 :Int32Value 3 :StringValue "A")))
  (bind AttributeCompareOp_0 (Attribute%20Compare%20Op :Operation Greater))
  (bind Select_0 (Select))
  (connect DefaultInputNode Index AttributeCompareOp_0 InA)
  (connect CreateAttribute_5 Out AttributeCompareOp_0 InB)
  (connect CreateAttribute_6 Out AttributeStringOp_5 InA)
  (connect Select%20%28Multi%29_0 Out AttributeStringOp_5 InB)
  (connect AttributeCompareOp_0 Out Select_0 bUseInputB)
  (connect AttributeStringOp_5 Out Select_0 Input%20A)
  (connect Select%20%28Multi%29_0 Out Select_0 Input%20B)
  (connect CreateAttribute_1 Out Select%20%28Multi%29_0 %30)
  (connect CreateAttribute_2 Out Select%20%28Multi%29_0 %31)
  (connect CreateAttribute_3 Out Select%20%28Multi%29_0 %32)
  (connect CreateAttribute_4 Out Select%20%28Multi%29_0 %33)
  (connect CreateAttribute_7 Out Select%20%28Multi%29_0 %34)
  (connect CreateAttribute_8 Out Select%20%28Multi%29_0 %35)
  (connect CreateAttribute_1 Out Select%20%28Multi%29_0 Default)
  (connect DefaultInputNode Index Select%20%28Multi%29_0 IntegerSelection)
  (output ColorAttribute Select_0))
