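; PCG subgraph in the Helios landscape grass tools that passes input point data through only when it is not empty.
; It compares the input's element count to a default-valued constant with NotEqual, and the result drives a Branch.
; Non-empty data routes to Branch output B as ValidPoints, while output A is left unconnected, so empty data goes no further.
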(graph PCG_GateValidPoints
  (outputs (ValidPoints :type Point))
  (bind GetPointsCount_0 (Get%20Element%20Count))
  (bind CreateAttribute_2 (Create%20Constant))
  (bind AttributeCompareOp_1 (Attribute%20Compare%20Op :Operation NotEqual))
  (bind Branch_3 (Branch))
  (connect GetPointsCount_0 Out AttributeCompareOp_1 InA)
  (connect CreateAttribute_2 Out AttributeCompareOp_1 InB)
  (connect AttributeCompareOp_1 Out Branch_3 bOutputToB)
  (connect DefaultInputNode In Branch_3 In)
  (connect DefaultInputNode In GetPointsCount_0 In)
  (output ValidPoints Branch_3.Output%20B))
