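; PCG_HasValidData is a small PCG utility graph that reports whether its input contains any data.
; It counts the incoming data and compares that count for inequality against a constant left at its default, presumably zero.
; The comparison result leaves through the IsValid attribute output as a boolean flag.
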(graph PCG_HasValidData
  (outputs (IsValid :type Param))
  (bind CreateAttribute_2 (Create%20Constant))
  (bind DataNum_0 (Data%20Count))
  (bind AttributeCompareOp_1 (Attribute%20Compare%20Op :Operation NotEqual))
  (connect DataNum_0 Out AttributeCompareOp_1 InA)
  (connect CreateAttribute_2 Out AttributeCompareOp_1 InB)
  (connect DefaultInputNode In DataNum_0 In)
  (output IsValid AttributeCompareOp_1))
