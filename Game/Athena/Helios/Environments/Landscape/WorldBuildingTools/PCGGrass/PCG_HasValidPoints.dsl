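; PCG utility subgraph in the Helios landscape PCG grass tools that reports whether its input data holds any points.
; It takes the element count of the input and compares it with a default-valued constant using a NotEqual attribute compare.
; The comparison result leaves through the IsValid attribute output as a validity flag.
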(graph PCG_HasValidPoints
  (outputs (IsValid :type Param))
  (bind GetPointsCount_0 (Get%20Element%20Count))
  (bind CreateAttribute_2 (Create%20Constant))
  (bind AttributeCompareOp_1 (Attribute%20Compare%20Op :Operation NotEqual))
  (connect GetPointsCount_0 Out AttributeCompareOp_1 InA)
  (connect CreateAttribute_2 Out AttributeCompareOp_1 InB)
  (connect DefaultInputNode In GetPointsCount_0 In)
  (output IsValid AttributeCompareOp_1))
