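; PCG subgraph in the PCGGrass world-building tools that re-expresses source points relative to a target point's transform.
; Position and rotation pass in turn through inverse-transform location and rotation ops against the target's transform.
; Scale is multiplied by the reciprocal of the target's scale, computed by dividing a constant one by it, since each component is handled by a separate attribute op.
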(graph PCG_InverseTransformPoints
  (inputs (SourcePoints :type Point) (TargetPoint :type Point))
  (outputs (OutPoints :type Point))
  (bind AttributeVectorOp_0 (Attribute%20Vector%20Op :InputSource1 $Position :InputSource2 $Transform :Operation InverseTransformLocation))
  (bind AttributeRotatorOp_1 (Attribute%20Rotator%20Op :InputSource1 $Rotation :InputSource2 $Transform :Operation InverseTransformRotation))
  (bind CreateAttribute_3 (Create%20Constant :AttributeTypes (struct :DoubleValue 1.0)))
  (bind AttributeMathsOp_2 (Attribute%20Maths%20Op :InputSource2 $Scale :Operation Divide))
  (bind AttributeMathsOp_4 (Attribute%20Maths%20Op :InputSource1 $Scale :InputSource2 $Scale :Operation Multiply))
  (connect CreateAttribute_3 Out AttributeMathsOp_2 InA)
  (connect DefaultInputNode TargetPoint AttributeMathsOp_2 InB)
  (connect AttributeRotatorOp_1 Out AttributeMathsOp_4 InA)
  (connect AttributeMathsOp_2 Out AttributeMathsOp_4 InB)
  (connect AttributeVectorOp_0 Out AttributeRotatorOp_1 In)
  (connect DefaultInputNode TargetPoint AttributeRotatorOp_1 Transform)
  (connect DefaultInputNode SourcePoints AttributeVectorOp_0 In)
  (connect DefaultInputNode TargetPoint AttributeVectorOp_0 Transform)
  (output OutPoints AttributeMathsOp_4))
