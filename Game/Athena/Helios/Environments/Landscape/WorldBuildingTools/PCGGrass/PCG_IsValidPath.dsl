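; PCG helper subgraph in the Helios landscape grass tools that reports whether a path attribute on the input is non-empty.
; It reads the attribute named by TargetAttribute, falling back to @Last through a Select when that name is empty.
; The value is cast to String and compared NotEqual to an empty string to produce IsValid.
; The check does not test whether the path resolves to an asset.
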(graph PCG_IsValidPath
  (params (TargetAttribute :type String))
  (outputs (IsValid :type Param))
  (bind CreateAttribute_3 (Create%20Constant :AttributeTypes (struct :Type String)))
  (bind AttributeCompareOp_2 (Attribute%20Compare%20Op :Operation NotEqual :InA TargetAttribute))
  (bind CreateAttribute_9 (Create%20Constant :AttributeTypes (struct :Type String :StringValue "@Last")))
  (bind Select_7 (Select :Input%20B TargetAttribute))
  (bind AttributeCast_6 (Attribute%20Cast :OutputType String))
  (bind CreateAttribute_2 (Create%20Constant :AttributeTypes (struct :Type String) :OutputTarget Valid))
  (bind AttributeCompareOp_1 (Attribute%20Compare%20Op :Operation NotEqual))
  (connect DefaultInputNode In AttributeCast_6 In)
  (connect Select_7 Out AttributeCast_6 InputSource)
  (connect AttributeCast_6 Out AttributeCompareOp_1 InA)
  (connect CreateAttribute_2 Out AttributeCompareOp_1 InB)
  (connect CreateAttribute_3 Out AttributeCompareOp_2 InB)
  (connect AttributeCompareOp_2 Out Select_7 bUseInputB)
  (connect CreateAttribute_9 Out Select_7 Input%20A)
  (output IsValid AttributeCompareOp_1))
