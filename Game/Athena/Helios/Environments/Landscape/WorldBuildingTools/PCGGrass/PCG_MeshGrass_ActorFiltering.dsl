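; PCG subgraph for the Helios mesh grass tools that filters an actor point by biome and material.
; It looks up the point's biome attributes and drops points in inactive biomes.
; When the graph parameter, global flag and biome all allow exclusion, it also drops points whose material name contains an exclusion string.
; Chained Branch nodes feed a count check, which gates the original point to the output.
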(graph PCG_MeshGrass_ActorFiltering
  (params (EnableMaterialExclusion :type Bool :default true))
  (inputs (ActorPoint :type Point) (BiomeAttributes :type Param) (GlobalAttributes :type Param))
  (outputs (ActorPoint :type Point))
  (bind GetPointsCount_7 (Get%20Element%20Count))
  (bind CreateAttribute_9 (Create%20Constant))
  (bind AttributeCompareOp_8 (Attribute%20Compare%20Op :Operation Greater))
  (bind Branch_10 (Branch))
  ;[0.05 0.05 0.05] Material Exclusion
  (bind ExecuteBlueprint_0 (Execute%20Blueprint :BlueprintElementInstance (class:/Game/Athena/Helios/Environments/Landscape/WorldBuildingTools/PCGGrass/PCGBP_GetMaterial.PCGBP_GetMaterial_C) :BlueprintElementType /Game/Athena/Helios/Environments/Landscape/WorldBuildingTools/PCGGrass/PCGBP_GetMaterial.PCGBP_GetMaterial_C))
  (bind ExecuteBlueprint_1 (Execute%20Blueprint :BlueprintElementInstance (class:/Game/Athena/Helios/Environments/Landscape/WorldBuildingTools/PCGGrass/PCGBP_SoftObjectPathToString.PCGBP_SoftObjectPathToString_C) :BlueprintElementType /Game/Athena/Helios/Environments/Landscape/WorldBuildingTools/PCGGrass/PCGBP_SoftObjectPathToString.PCGBP_SoftObjectPathToString_C))
  (bind DeleteAttributes_26 (Delete%20Attributes :bTokenizeOnWhiteSpace true :Operation KeepSelectedAttributes :SelectedAttributes "MaterialExclusionStrings"))
  (bind ExecuteBlueprint_2 (Execute%20Blueprint :BlueprintElementInstance (class:/Game/Athena/Helios/Environments/Landscape/WorldBuildingTools/PCGGrass/PCGBP_StringContains.PCGBP_StringContains_C :SeparateByComma true) :BlueprintElementType /Game/Athena/Helios/Environments/Landscape/WorldBuildingTools/PCGGrass/PCGBP_StringContains.PCGBP_StringContains_C))
  (bind DeleteAttributes_3 (Delete%20Attributes :bTokenizeOnWhiteSpace true :Operation KeepSelectedAttributes :SelectedAttributes "ContainsSubstring"))
  (bind DeleteAttributes_5 (Delete%20Attributes :bTokenizeOnWhiteSpace true :Operation KeepSelectedAttributes :SelectedAttributes "ExcludeByMaterialName"))
  (bind AttributeBooleanOp_1 (Attribute%20Boolean%20Op :InA EnableMaterialExclusion))
  (bind Subgraph_24 (Subgraph :Graph /Game/Athena/Helios/Environments/Landscape/WorldBuildingTools/PCGGrass/PCG_MeshGrass_FormatBiomeAttributes.PCG_MeshGrass_FormatBiomeAttributes))
  (bind ExecuteBlueprint_22 (Execute%20Blueprint :BlueprintElementInstance (class:/Game/Athena/Helios/Environments/Landscape/WorldBuildingTools/PCGGrass/PCGBP_GetDataByIndex.PCGBP_GetDataByIndex_C) :BlueprintElementType /Game/Athena/Helios/Environments/Landscape/WorldBuildingTools/PCGGrass/PCGBP_GetDataByIndex.PCGBP_GetDataByIndex_C))
  (bind DeleteAttributes_25 (Delete%20Attributes :bTokenizeOnWhiteSpace true :Operation KeepSelectedAttributes :SelectedAttributes "AllowExclusions"))
  (bind AttributeBooleanOp_3 (Attribute%20Boolean%20Op))
  (bind Branch_6 (Branch))
  (bind Branch_4 (Branch))
  ;[0.05 0.05 0.05] Biome Filtering
  (bind Subgraph_19 (Subgraph :Graph /Game/Athena/Helios/Environments/Landscape/WorldBuildingTools/PCGGrass/PCG_MeshGrass_FormatBiomeAttributes.PCG_MeshGrass_FormatBiomeAttributes))
  (bind ExecuteBlueprint_17 (Execute%20Blueprint :BlueprintElementInstance (class:/Game/Athena/Helios/Environments/Landscape/WorldBuildingTools/PCGGrass/PCGBP_GetDataByIndex.PCGBP_GetDataByIndex_C) :BlueprintElementType /Game/Athena/Helios/Environments/Landscape/WorldBuildingTools/PCGGrass/PCGBP_GetDataByIndex.PCGBP_GetDataByIndex_C))
  (bind DeleteAttributes_20 (Delete%20Attributes :bTokenizeOnWhiteSpace true :Operation KeepSelectedAttributes :SelectedAttributes "Active"))
  (bind Branch_12 (Branch))
  ;[0.05 0.05 0.05] Inputs
  (bind GetAttributeFromPointIndex_36 (Get%20Attribute%20From%20Point%20Index :bForceOutputAttributeToBeInElementsDomain true :InputSource Biome :OutputAttributeName Biome))
  (reroute Actor%20Biome GetAttributeFromPointIndex_36.Attribute)
  (reroute Actor%20Point ActorPoint)
  (reroute Biome%20Attributes BiomeAttributes)
  (reroute Global%20Attributes GlobalAttributes)
  (connect DeleteAttributes_5 Out AttributeBooleanOp_1 InB)
  (connect AttributeBooleanOp_1 Out AttributeBooleanOp_3 InA)
  (connect DeleteAttributes_25 Out AttributeBooleanOp_3 InB)
  (connect GetPointsCount_7 Out AttributeCompareOp_8 InA)
  (connect CreateAttribute_9 Out AttributeCompareOp_8 InB)
  (connect DeleteAttributes_3 Out Branch_4 bOutputToB)
  (connect Branch_6 Output%20B Branch_4 In)
  (connect AttributeBooleanOp_3 Out Branch_6 bOutputToB)
  (connect Branch_12 Output%20B Branch_6 In)
  (connect AttributeCompareOp_8 Out Branch_10 bOutputToB)
  (connect Actor%20Point Out Branch_10 In)
  (connect DeleteAttributes_20 Out Branch_12 bOutputToB)
  (connect Actor%20Point Out Branch_12 In)
  (connect ExecuteBlueprint_2 Out DeleteAttributes_3 In)
  (connect Global%20Attributes Out DeleteAttributes_5 In)
  (connect ExecuteBlueprint_17 OutData DeleteAttributes_20 In)
  (connect ExecuteBlueprint_22 OutData DeleteAttributes_25 In)
  (connect Global%20Attributes Out DeleteAttributes_26 In)
  (connect ExecuteBlueprint_0 Material ExecuteBlueprint_1 In)
  (connect ExecuteBlueprint_1 Out ExecuteBlueprint_2 InString)
  (connect DeleteAttributes_26 Out ExecuteBlueprint_2 Substring)
  (connect Actor%20Biome Out ExecuteBlueprint_17 DataIndex)
  (connect Subgraph_19 BiomeAttributes ExecuteBlueprint_17 InData)
  (connect Actor%20Biome Out ExecuteBlueprint_22 DataIndex)
  (connect Subgraph_24 BiomeAttributes ExecuteBlueprint_22 InData)
  (connect Actor%20Point Out GetAttributeFromPointIndex_36 In)
  (connect Branch_4 Output%20A GetPointsCount_7 In)
  (connect Branch_6 Output%20A GetPointsCount_7 In)
  (connect Biome%20Attributes Out Subgraph_19 InAttributes)
  (connect Biome%20Attributes Out Subgraph_24 InAttributes)
  (output ActorPoint Branch_10.Output%20B))
