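; PCG graph for mesh grass that assigns a biome to an actor point.
; It projects the point against the largest BP_MeshGrass_Partitioner volume by XY extents, giving points outside it the global FallbackBiome.
; For points inside, active biome entries, minus the first, loop through PCG_MeshGrass_SampleBiome.
; Non-zero results are sorted by BiomeStrength and the strongest wins, with biome 0 used when none remain.
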(graph PCG_MeshGrass_BiomeDetermination
  (inputs (ActorPoint :type Point) (BiomeAttributes :type Param) (GlobalAttributes :type Param))
  (outputs (ActorPoint :type Point))
  ;[0.05 0.05 0.05] Biome Assignment
  (bind ExtentsModifier_9 (Extents%20Modifier :Extents (1e+03 1e+03 1e+03)))
  (bind AddAttribute_0 (Add%20Attribute :AttributeTypes (struct :Type Boolean) :OutputTarget IsInMainVolume))
  (bind GetActorData_4 (Get%20Actor%20Data :ActorSelector (struct :ActorFilter AllWorldActors :ActorSelection ByClass :ActorSelectionClass /Game/Athena/Helios/Environments/Landscape/WorldBuildingTools/PCGGrass/BP_MeshGrass_Partitioner.BP_MeshGrass_Partitioner_C :bSelectMultiple true) :bComponentsMustOverlapSelf false :bMergeSinglePointData true :Mode GetSinglePoint))
  (bind ApplyScaleToBounds_0 (Apply%20Scale%20To%20Bounds))
  (bind AttributeSelect_2 (Attribute%20Select :Axis CustomAxis :CustomAxis (1.0 1.0 0.0 0.0) :InputSource $Extents :Operation Max))
  (bind AddAttribute_1 (Add%20Attribute :AttributeTypes (struct :Type Boolean :BoolValue true) :OutputTarget IsInMainVolume))
  (bind Projection_10 (Projection :bKeepZeroDensityPoints true :ProjectionParams (struct :bProjectPositions false :bProjectRotations false :AttributeList "IsInMainVolume" :AttributeMode IncludeAttributes)))
  (bind GetAttributeFromPointIndex_5 (Get%20Attribute%20From%20Point%20Index :bForceOutputAttributeToBeInElementsDomain true :InputSource IsInMainVolume :OutputAttributeName IsInMainVolume))
  (bind AddAttribute_22 (Add%20Attribute :AttributeTypes (struct :Type Integer32) :OutputTarget Biome))
  (bind DeleteAttributes_7 (Delete%20Attributes :bTokenizeOnWhiteSpace true :Operation KeepSelectedAttributes :SelectedAttributes "FallbackBiome"))
  (bind AddAttribute_6 (Add%20Attribute :AttributeTypes (struct :DoubleValue 0.3) :OutputTarget Biome))
  (bind CreateAttribute_16 (Create%20Constant :AttributeTypes (struct :Type Integer32) :OutputTarget IsEmpty))
  (bind AddAttribute_23 (Add%20Attribute :AttributeTypes (struct :DoubleValue 1.0) :OutputTarget BiomeStrength))
  (bind FilterDataByIndex_25 (Filter%20Data%20By%20Index :bInvertFilter true :SelectedIndices "0"))
  (bind DeleteAttributes_0 (Delete%20Attributes :Operation KeepSelectedAttributes :Operator Matches :SelectedAttributes "*_Active"))
  (bind CopyAttributes_0 (Copy%20Attributes :InputSource @Last :OutputTarget ShouldSample))
  (bind AttributeFilter_0 (Filter%20Attribute%20Elements :AttributeTypes (struct :Type Boolean :BoolValue true) :bGenerateOutputDataEvenIfEmpty false :bUseConstantThreshold true :Operator Equal :TargetAttribute @Last :ThresholdAttribute @Last))
  (bind Loop_24 (Loop :bUseGraphDefaultPinUsage false :LoopPins "Biome Attributes" :Graph /Game/Athena/Helios/Environments/Landscape/WorldBuildingTools/PCGGrass/PCG_MeshGrass_SampleBiome.PCG_MeshGrass_SampleBiome))
  (bind MergePoints_26 (Merge%20Points))
  (bind AttributeFilter_12 (Filter%20Attribute%20Elements :AttributeTypes (struct :Type Integer32) :bUseConstantThreshold true :Operator NotEqual :TargetAttribute Biome))
  (bind SortAttributes_11 (Sort%20Attributes :bUseStableSort false :SortAttributes [(struct :InputSource BiomeStrength :SortMethod Descending)]))
  (bind AttributeFilter_13 (Filter%20Attribute%20Elements :AttributeTypes (struct :Type Integer32) :bUseConstantThreshold true :bUseSpatialQuery true :Operator Equal :TargetAttribute $Index))
  (bind GetPointsCount_14 (Get%20Element%20Count))
  (bind AttributeCompareOp_15 (Attribute%20Compare%20Op))
  (bind Select_17 (Select))
  (bind Select_8 (Select))
  (reroute Actor%20Point ActorPoint)
  (reroute Biome%20Attributes BiomeAttributes)
  (reroute Global%20Attributes GlobalAttributes)
  (connect ExtentsModifier_9 Out AddAttribute_0 In)
  (connect AttributeSelect_2 Point AddAttribute_1 In)
  (connect DeleteAttributes_7 Out AddAttribute_6 Attributes)
  (connect AddAttribute_22 Out AddAttribute_6 In)
  (connect Actor%20Point Out AddAttribute_22 In)
  (connect AddAttribute_22 Out AddAttribute_23 In)
  (connect GetActorData_4 Out ApplyScaleToBounds_0 In)
  (connect GetPointsCount_14 Out AttributeCompareOp_15 InA)
  (connect CreateAttribute_16 Out AttributeCompareOp_15 InB)
  (connect CopyAttributes_0 Out AttributeFilter_0 In)
  (connect MergePoints_26 Out AttributeFilter_12 In)
  (connect SortAttributes_11 Out AttributeFilter_13 In)
  (connect ApplyScaleToBounds_0 Out AttributeSelect_2 In)
  (connect DeleteAttributes_0 Out CopyAttributes_0 Source)
  (connect FilterDataByIndex_25 InsideFilter CopyAttributes_0 Target)
  (connect FilterDataByIndex_25 InsideFilter DeleteAttributes_0 In)
  (connect Global%20Attributes Out DeleteAttributes_7 In)
  (connect Actor%20Point Out ExtentsModifier_9 In)
  (connect Biome%20Attributes Out FilterDataByIndex_25 In)
  (connect Projection_10 Out GetAttributeFromPointIndex_5 In)
  (connect AttributeFilter_13 InsideFilter GetPointsCount_14 In)
  (connect AttributeFilter_0 InsideFilter Loop_24 Biome%20Attributes)
  (connect Global%20Attributes Out Loop_24 Global%20Attributes)
  (connect AddAttribute_23 Out Loop_24 Target%20Point)
  (connect Loop_24 Out MergePoints_26 In)
  (connect AddAttribute_0 Out Projection_10 In)
  (connect AddAttribute_1 Out Projection_10 Projection%20Target)
  (connect GetAttributeFromPointIndex_5 Attribute Select_8 bUseInputB)
  (connect AddAttribute_6 Out Select_8 Input%20A)
  (connect Select_17 Out Select_8 Input%20B)
  (connect AttributeCompareOp_15 Out Select_17 bUseInputB)
  (connect AttributeFilter_13 InsideFilter Select_17 Input%20A)
  (connect AddAttribute_22 Out Select_17 Input%20B)
  (connect AttributeFilter_12 InsideFilter SortAttributes_11 In)
  (output ActorPoint Select_8))
