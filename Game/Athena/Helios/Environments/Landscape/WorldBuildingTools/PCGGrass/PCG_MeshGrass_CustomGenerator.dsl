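; PCG graph that generates custom mesh grass points for an actor, with parameters for mesh, density, transform and filtering, and outputs them as CustomPoints.
; It packs parameters and constant topology-filter attributes into actor attributes, then resolves them with the actor's point through the shared Inputs subgraph.
; Points flow through custom sampling, local filtering, grass assignment, point processing, world filtering and metadata processing.
; Validity-check branches gate the actor point and both filtering results.
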(graph PCG_MeshGrass_CustomGenerator
  (params (WorldSpaceTransform :type Bool) (LocationOffsetZ :type Double) (RotationOffset :type Rotator) (RandomizeRotationOffset :type Bool :default true) (ScaleMinMax :type Vector2 :default (1.0 1.0)) (EnableMeshOverride :type Bool) (GrassMesh :type (SoftObject /Script/Engine.StaticMesh)) (MaterialOverride :type (Object /Script/Engine.MaterialInterface)) (GrassDensity :type Double :default 1.0) (FilterNormals :type Bool) (NormalOffset :type Double) (FilterWorldIntersection :type Bool) (ExtendNegativeBounds :type Bool) (WorldIntersectionDistance :type Double :default 1e+02) (FilterBelowTerrain :type Bool) (BelowTerrainBias :type Double :default -1.0) (TraceZOffset :type Double) (UsePruning :type Bool) (FilterLandscapeLayers :type Bool) (LandscapeLayers :type String) (LandscapeLayerThreshold :type Double) (LandscapeLayerDilation :type Double) (FilterWaterBodies :type Bool) (WaterBodyDistance :type Double) (WaterBodyInvertFilter :type Bool) (WaterBodyPrecision :type (Enum /Game/Athena/Helios/Environments/Landscape/WorldBuildingTools/PCGGrass/EMeshGrassGenericQuality.EMeshGrassGenericQuality) :default NewEnumerator1) (DetailModeOverride :type (Enum /Game/Athena/Helios/Environments/Landscape/WorldBuildingTools/PCGGrass/EMeshGrassDetailMode.EMeshGrassDetailMode)) (IgnoreCullVolumes :type Bool))
  (outputs (CustomPoints :type Point))
  (bind Subgraph_14 (Subgraph :Graph /Game/Athena/Helios/Environments/Landscape/WorldBuildingTools/PCGGrass/PCG_MeshGrass_Inputs.PCG_MeshGrass_Inputs))
  (bind Subgraph_4 (Subgraph :Graph /Game/Athena/Helios/Environments/Landscape/WorldBuildingTools/PCGGrass/PCG_HasValidData.PCG_HasValidData))
  (bind Branch_6 (Branch))
  (bind Subgraph_7 (Subgraph :Graph /Game/Athena/Helios/Environments/Landscape/WorldBuildingTools/PCGGrass/PCG_MeshGrass_CustomSampling.PCG_MeshGrass_CustomSampling))
  (bind Subgraph_11 (Subgraph :Graph /Game/Athena/Helios/Environments/Landscape/WorldBuildingTools/PCGGrass/PCG_MeshGrass_LocalFiltering.PCG_MeshGrass_LocalFiltering))
  (bind Subgraph_24 (Subgraph :Graph /Game/Athena/Helios/Environments/Landscape/WorldBuildingTools/PCGGrass/PCG_HasValidData.PCG_HasValidData))
  (bind Branch_26 (Branch))
  (bind Subgraph_6 (Subgraph :Graph /Game/Athena/Helios/Environments/Landscape/WorldBuildingTools/PCGGrass/PCG_MeshGrass_GrassAssignment.PCG_MeshGrass_GrassAssignment))
  (bind Subgraph_13 (Subgraph :Graph /Game/Athena/Helios/Environments/Landscape/WorldBuildingTools/PCGGrass/PCG_MeshGrass_PointProcessing.PCG_MeshGrass_PointProcessing))
  (bind Subgraph_1 (Subgraph :Graph /Game/Athena/Helios/Environments/Landscape/WorldBuildingTools/PCGGrass/PCG_MeshGrass_WorldFiltering.PCG_MeshGrass_WorldFiltering))
  (bind Subgraph_3 (Subgraph :Graph /Game/Athena/Helios/Environments/Landscape/WorldBuildingTools/PCGGrass/PCG_HasValidPoints.PCG_HasValidPoints))
  (bind Branch_2 (Branch))
  (bind Subgraph_0 (Subgraph :Graph /Game/Athena/Helios/Environments/Landscape/WorldBuildingTools/PCGGrass/PCG_MeshGrass_MetadataProcessing.PCG_MeshGrass_MetadataProcessing))
  ;[0.05 0.05 0.05] Actor Attributes
  (bind CreateAttribute_1 (Create%20Constant :AttributeTypes (struct :Type Boolean) :OutputTarget FilterNonSelfHits))
  (bind CreateAttribute_3 (Create%20Constant :AttributeTypes (struct :Type Boolean) :OutputTarget FilterBorders))
  (bind CreateAttribute_84 (Create%20Constant :AttributeTypes (struct :Type Boolean) :OutputTarget FilterHeightGradient))
  (bind CreateAttribute_85 (Create%20Constant :OutputTarget EdgeDistance))
  (bind CreateAttribute_87 (Create%20Constant :AttributeTypes (struct :Type Boolean) :OutputTarget FilterLocalZ))
  (bind CreateAttribute_89 (Create%20Constant :OutputTarget LocalZMin))
  (bind CreateAttribute_91 (Create%20Constant :OutputTarget LocalZMax))
  (bind CreateConstant_0 (Create%20Constant :AttributeTypes (struct :Type Boolean) :OutputTarget ScaleByGradient))
  (bind CreateConstant_1 (Create%20Constant :AttributeTypes (struct :Type Boolean) :OutputTarget FilterTopologyInWorldSpace))
  (bind AddAttribute_0 (Add%20Attribute :AttributeTypes (struct :Type Integer64) :bCopyAllAttributes true :Attributes UsePruning :Attributes GrassDensity :Attributes ScaleMinMax :Attributes NormalOffset :Attributes RotationOffset :Attributes EnableMeshOverride :Attributes FilterWorldIntersection :Attributes TraceZOffset :Attributes WorldIntersectionDistance :Attributes ExtendNegativeBounds :Attributes FilterLandscapeLayers :Attributes MaterialOverride :Attributes WorldSpaceTransform :Attributes LocationOffsetZ :Attributes RotationOffset :Attributes LandscapeLayers :Attributes LandscapeLayerThreshold :Attributes LandscapeLayerDilation :Attributes FilterNormals :In GrassMesh))
  (bind GetActorData_0 (Get%20Actor%20Data :bComponentsMustOverlapSelf false :Mode GetSinglePoint :PropertyName Scale))
  (reroute Actor%20Attributes AddAttribute_0)
  (reroute Actor%20Point GetActorData_0)
  (connect CreateAttribute_1 Out AddAttribute_0 Attributes)
  (connect CreateAttribute_3 Out AddAttribute_0 Attributes)
  (connect CreateAttribute_84 Out AddAttribute_0 Attributes)
  (connect CreateAttribute_85 Out AddAttribute_0 Attributes)
  (connect CreateAttribute_87 Out AddAttribute_0 Attributes)
  (connect CreateAttribute_89 Out AddAttribute_0 Attributes)
  (connect CreateAttribute_91 Out AddAttribute_0 Attributes)
  (connect CreateConstant_0 Out AddAttribute_0 Attributes)
  (connect CreateConstant_1 Out AddAttribute_0 Attributes)
  (connect Subgraph_3 IsValid Branch_2 bOutputToB)
  (connect Subgraph_1 OutPoints Branch_2 In)
  (connect Subgraph_4 IsValid Branch_6 bOutputToB)
  (connect Subgraph_14 ActorPoint Branch_6 In)
  (connect Subgraph_24 IsValid Branch_26 bOutputToB)
  (connect Subgraph_11 OutPoints Branch_26 In)
  (connect Branch_6 Output%20B Subgraph_0 ActorPoint)
  (connect Branch_2 Output%20B Subgraph_0 InPoints)
  (connect Subgraph_14 ActorAttributes Subgraph_1 ActorAttributes)
  (connect Branch_6 Output%20B Subgraph_1 ActorPoint)
  (connect Subgraph_14 BiomeAttributes Subgraph_1 BiomeAttributes)
  (connect Subgraph_14 GlobalAttributes Subgraph_1 GlobalAttributes)
  (connect Subgraph_13 OutPoints Subgraph_1 InPoints)
  (connect Subgraph_14 VolumeAttributes Subgraph_1 VolumeAttributes)
  (connect Subgraph_1 OutPoints Subgraph_3 In)
  (connect Subgraph_14 ActorPoint Subgraph_4 In)
  (connect Subgraph_14 ActorAttributes Subgraph_6 ActorAttributes)
  (connect Branch_6 Output%20B Subgraph_6 ActorPoint)
  (connect Subgraph_14 BiomeAttributes Subgraph_6 BiomeAttributes)
  (connect Subgraph_14 GlobalAttributes Subgraph_6 GlobalAttributes)
  (connect Branch_26 Output%20B Subgraph_6 InPoints)
  (connect Subgraph_14 VolumeAttributes Subgraph_6 VolumeAttributes)
  (connect Branch_6 Output%20B Subgraph_7 ActorPoint)
  (connect Subgraph_14 ActorAttributes Subgraph_11 ActorAttributes)
  (connect Branch_6 Output%20B Subgraph_11 ActorPoint)
  (connect Subgraph_14 BiomeAttributes Subgraph_11 BiomeAttributes)
  (connect Subgraph_14 GlobalAttributes Subgraph_11 GlobalAttributes)
  (connect Subgraph_7 OutPoints Subgraph_11 InPoints)
  (connect Subgraph_14 VolumeAttributes Subgraph_11 VolumeAttributes)
  (connect Subgraph_14 ActorAttributes Subgraph_13 ActorAttributes)
  (connect Branch_6 Output%20B Subgraph_13 ActorPoint)
  (connect Subgraph_14 BiomeAttributes Subgraph_13 BiomeAttributes)
  (connect Subgraph_14 GlobalAttributes Subgraph_13 GlobalAttributes)
  (connect Subgraph_6 OutPoints Subgraph_13 InPoints)
  (connect Subgraph_14 VolumeAttributes Subgraph_13 VolumeAttributes)
  (connect Actor%20Attributes Out Subgraph_14 ActorAttributes)
  (connect Actor%20Point Out Subgraph_14 ActorPoint)
  (connect Subgraph_11 OutPoints Subgraph_24 In)
  (output CustomPoints Subgraph_0.OutPoints))
