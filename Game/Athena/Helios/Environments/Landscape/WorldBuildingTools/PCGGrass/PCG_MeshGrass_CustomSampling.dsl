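; PCG graph that scatters mesh grass around a custom generator actor, using its instanced meshes as source points and its Configuration as settings.
; Source points get mesh and optional render attributes, are copied to the actor point and are jittered in XY by SpawnRadius.
; Optional projection raycasts from raised origins with configurable target filtering, processes hits, restores scale, reapplies rotation with optional surface alignment and vertical offset, then filters results.
; Unprojected points output directly.
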(graph PCG_MeshGrass_CustomSampling
  (inputs (ActorPoint :type Point))
  (outputs (OutPoints :type Point))
  (bind ExecuteBlueprint_0 (Execute%20Blueprint :BlueprintElementInstance (class:/Game/Athena/Helios/Environments/Landscape/WorldBuildingTools/PCGGrass/PCGBP_PointsFromISM.PCGBP_PointsFromISM_C) :BlueprintElementType /Game/Athena/Helios/Environments/Landscape/WorldBuildingTools/PCGGrass/PCGBP_PointsFromISM.PCGBP_PointsFromISM_C))
  (bind Subgraph_1 (Subgraph :Graph /Game/Athena/Helios/Environments/Landscape/WorldBuildingTools/PCGGrass/PCG_MeshGrass_ExtraMeshAttributes.PCG_MeshGrass_ExtraMeshAttributes))
  (bind MergePoints_0 (Merge%20Points))
  (bind AddAttribute_2 (Add%20Attribute :bCopyAllAttributes true))
  (bind CreateAttribute_19 (Create%20Constant :AttributeTypes (struct :DoubleValue -1.0)))
  (bind DeleteAttributes_18 (Delete%20Attributes :Operation KeepSelectedAttributes :SelectedAttributes "SpawnRadius"))
  (bind AttributeMathsOp_20 (Attribute%20Maths%20Op :Operation Multiply))
  (bind CreateAttribute_21 (Create%20Constant :AttributeTypes (struct :DoubleValue 1.0)))
  (bind AttributeMathsOp_22 (Attribute%20Maths%20Op :Operation Multiply))
  (bind DeleteAttributes_1 (Delete%20Attributes :Operation KeepSelectedAttributes :SelectedAttributes "InstanceStartCullDistance,InstanceEndCullDistance,WPODisableDistance"))
  (bind CopyAttributes_14 (Copy%20Attributes :bCopyAllAttributes true :InputSource @Last))
  (bind DeleteAttributes_5 (Delete%20Attributes :Operation KeepSelectedAttributes :SelectedAttributes "PreserveRenderSettings"))
  (bind Select_9 (Select))
  (bind CopyPoints_5 (Copy%20Points :AttributeInheritance SourceOnly :bCopyEachSourceOnEveryTarget false :ColorInheritance Source :ScaleInheritance Source :SeedInheritance Source :TagInheritance Source))
  (bind ExtentsModifier_0 (Extents%20Modifier :Extents (1e+01 1e+01 1e+01)))
  (bind AttributeNoise_23 (Attribute%20Noise :InputSource $Position.XY :Mode Add :NoiseMin -1.0))
  (bind DeleteAttributes_4 (Delete%20Attributes :Operation KeepSelectedAttributes :SelectedAttributes "ProjectInstances"))
  (bind Branch_1 (Branch))
  (bind CreateAttribute_22 (Create%20Constant :AttributeTypes (struct :DoubleValue 1e+02)))
  (bind AttributeMathsOp_21 (Attribute%20Maths%20Op :InputSource1 $Position.Z))
  (bind DeleteAttributes_14 (Delete%20Attributes :Operation KeepSelectedAttributes :SelectedAttributes "ProjectionDistance"))
  (bind BoundsModifier_4 (Bounds%20Modifier :BoundsMin (1.0 1.0 1e+02)))
  (bind MakeVectorAttribute_38 (Make%20Vector%20Attribute :OutputType Vector))
  (bind ExtentsModifier_37 (Extents%20Modifier :Extents (1.0 1.0 0.0) :Mode Add))
  (bind Subgraph_0 (Subgraph :Graph /Game/Athena/Helios/Environments/Landscape/WorldBuildingTools/PCGGrass/PCG_MeshGrass_InterpretProjectionSetup.PCG_MeshGrass_InterpretProjectionSetup))
  (bind WorldRaycast_3 (World%20Raycast :bKeepOriginalPointOnMiss true :RaycastMode NormalizedWithLength :RayLength 1e+04 :WorldQueryParams (struct :bGetReferenceToStaticMesh true :ActorTagFilter Exclude :ActorTagsList "PCG_MeshGrass_Out,PCG_MeshGrass_NoTrace" :ActorClassFilter Exclude :ActorClass /Game/Athena/Helios/Environments/Landscape/WorldBuildingTools/PCGGrass/BP_MeshGrass_CustomGenerator.BP_MeshGrass_CustomGenerator_C :ActorFilterFromInput Include :bGetReferenceToActorHit true)))
  (bind Branch_0 (Branch))
  (bind Subgraph_3 (Subgraph :Graph /Game/Athena/Helios/Environments/Landscape/WorldBuildingTools/PCGGrass/PCG_ProcessTraceResults.PCG_ProcessTraceResults))
  (bind Gather_0 (Gather))
  (bind CopyAttributes_3 (Copy%20Attributes :InputSource $Scale :OutputTarget $Scale))
  (bind DeleteAttributes_16 (Delete%20Attributes :Operation KeepSelectedAttributes :SelectedAttributes "RotatetoSurfaceNormal"))
  (bind Branch_17 (Branch))
  (bind CopyAttributes_8 (Copy%20Attributes :InputSource $Rotation :OutputTarget $Rotation))
  (bind DeleteAttributes_7 (Delete%20Attributes :Operation KeepSelectedAttributes :SelectedAttributes "PreserveVerticalOffset"))
  (bind AttributeRotatorOp_7 (Attribute%20Rotator%20Op :InputSource1 $Rotation :InputSource2 $Rotation))
  (bind CopyAttributes_12 (Copy%20Attributes :InputSource $Rotation :OutputTarget $Rotation))
  (bind Gather_15 (Gather))
  (bind Branch_9 (Branch))
  (bind CopyAttributes_29 (Copy%20Attributes :InputSource $Position.Z :OutputTarget VerticalOffset))
  (bind ExecuteBlueprint_28 (Execute%20Blueprint :BlueprintElementInstance (class:/Game/Athena/Helios/Environments/Landscape/WorldBuildingTools/PCGGrass/PCGBP_TranslateAlongNormal.PCGBP_TranslateAlongNormal_C :DistanceAttributeName VerticalOffset) :BlueprintElementType /Game/Athena/Helios/Environments/Landscape/WorldBuildingTools/PCGGrass/PCGBP_TranslateAlongNormal.PCGBP_TranslateAlongNormal_C))
  (bind MergePoints_20 (Merge%20Points))
  (bind AttributeFilter_17 (Filter%20Attribute%20Elements :AttributeTypes (struct :DoubleValue 1.0) :bUseConstantThreshold true :Operator Equal :ThresholdAttribute @Last))
  ;[0.05 0.05 0.05] Inputs
  (bind GetPropertyFromObjectPath_10 (Get%20Property%20From%20Object%20Path :OutputAttributeName None :PropertyName Configuration))
  (bind GetAttributeFromPointIndex_0 (Get%20Attribute%20From%20Point%20Index :bForceOutputAttributeToBeInElementsDomain true :InputSource ActorReference :OutputAttributeName None))
  (reroute Actor%20Point ActorPoint)
  (reroute Actor%20Reference GetAttributeFromPointIndex_0.Attribute)
  (reroute CustomAttributes GetPropertyFromObjectPath_10)
  (reroute IgnoredMeshes Subgraph_0.IgnoredMeshes)
  (reroute OriginalPoints MergePoints_0)
  (reroute ProjectionSetup Subgraph_0.ProjectionSetup)
  (reroute RequiresProcessing Subgraph_0.RequiresProcessing)
  (connect Subgraph_1 MeshAttributes AddAttribute_2 Attributes)
  (connect MergePoints_0 Out AddAttribute_2 In)
  (connect MergePoints_20 Out AttributeFilter_17 In)
  (connect CreateAttribute_19 Out AttributeMathsOp_20 InA)
  (connect DeleteAttributes_18 Out AttributeMathsOp_20 InB)
  (connect Branch_1 Output%20B AttributeMathsOp_21 InA)
  (connect CreateAttribute_22 Out AttributeMathsOp_21 InB)
  (connect DeleteAttributes_18 Out AttributeMathsOp_22 InA)
  (connect CreateAttribute_21 Out AttributeMathsOp_22 InB)
  (connect ExtentsModifier_0 Out AttributeNoise_23 In)
  (connect AttributeMathsOp_22 Out AttributeNoise_23 NoiseMax)
  (connect AttributeMathsOp_20 Out AttributeNoise_23 NoiseMin)
  (connect CopyAttributes_8 Out AttributeRotatorOp_7 InA)
  (connect Branch_17 Output%20B AttributeRotatorOp_7 InB)
  (connect Actor%20Point Out BoundsModifier_4 In)
  (connect RequiresProcessing Out Branch_0 bOutputToB)
  (connect WorldRaycast_3 Out Branch_0 In)
  (connect DeleteAttributes_4 Out Branch_1 bOutputToB)
  (connect AttributeNoise_23 Out Branch_1 In)
  (connect DeleteAttributes_7 Out Branch_9 bOutputToB)
  (connect Gather_15 Out Branch_9 In)
  (connect DeleteAttributes_16 Out Branch_17 bOutputToB)
  (connect CopyAttributes_3 Out Branch_17 In)
  (connect OriginalPoints Out CopyAttributes_3 Source)
  (connect Gather_0 Out CopyAttributes_3 Target)
  (connect OriginalPoints Out CopyAttributes_8 Source)
  (connect Branch_17 Output%20B CopyAttributes_8 Target)
  (connect OriginalPoints Out CopyAttributes_12 Source)
  (connect Branch_17 Output%20A CopyAttributes_12 Target)
  (connect DeleteAttributes_1 Out CopyAttributes_14 Source)
  (connect AddAttribute_2 Out CopyAttributes_14 Target)
  (connect OriginalPoints Out CopyAttributes_29 Source)
  (connect Branch_9 Output%20B CopyAttributes_29 Target)
  (connect Select_9 Out CopyPoints_5 Source)
  (connect Actor%20Point Out CopyPoints_5 Target)
  (connect MergePoints_0 Out DeleteAttributes_1 In)
  (connect CustomAttributes Out DeleteAttributes_4 In)
  (connect CustomAttributes Out DeleteAttributes_5 In)
  (connect CustomAttributes Out DeleteAttributes_7 In)
  (connect CustomAttributes Out DeleteAttributes_14 In)
  (connect CustomAttributes Out DeleteAttributes_16 In)
  (connect CustomAttributes Out DeleteAttributes_18 In)
  (connect Actor%20Point Out ExecuteBlueprint_0 In)
  (connect CopyAttributes_29 Out ExecuteBlueprint_28 InPoints)
  (connect CopyPoints_5 Out ExtentsModifier_0 In)
  (connect MakeVectorAttribute_38 Out ExtentsModifier_37 Extents)
  (connect BoundsModifier_4 Out ExtentsModifier_37 In)
  (connect Branch_0 Output%20A Gather_0 In)
  (connect Subgraph_3 OutPoints Gather_0 In)
  (connect AttributeRotatorOp_7 Out Gather_15 In)
  (connect CopyAttributes_12 Out Gather_15 In)
  (connect Actor%20Point Out GetAttributeFromPointIndex_0 In)
  (connect Actor%20Reference Out GetPropertyFromObjectPath_10 In)
  (connect DeleteAttributes_18 Out MakeVectorAttribute_38 X)
  (connect DeleteAttributes_18 Out MakeVectorAttribute_38 Y)
  (connect ExecuteBlueprint_0 OutPoints MergePoints_0 In)
  (connect Branch_9 Output%20A MergePoints_20 In)
  (connect ExecuteBlueprint_28 OutPoints MergePoints_20 In)
  (connect DeleteAttributes_5 Out Select_9 bUseInputB)
  (connect AddAttribute_2 Out Select_9 Input%20A)
  (connect CopyAttributes_14 Out Select_9 Input%20B)
  (connect Actor%20Reference Out Subgraph_0 ActorReference)
  (connect Branch_0 Output%20B Subgraph_3 InPoints)
  (connect ProjectionSetup Out Subgraph_3 ProjectionSetup)
  (connect IgnoredMeshes Out Subgraph_3 ProjectionSetup)
  (connect Subgraph_0 CustomTarget WorldRaycast_3 ActorClass)
  (connect Subgraph_0 FilterFromInput WorldRaycast_3 ActorClassFilter)
  (connect Subgraph_0 FilterFromInput WorldRaycast_3 ActorFilterFromInput)
  (connect ExtentsModifier_37 Out WorldRaycast_3 Bounding%20Shape)
  (connect Subgraph_0 FilterActors WorldRaycast_3 FilterActors)
  (connect AttributeMathsOp_21 Out WorldRaycast_3 Origins)
  (connect DeleteAttributes_14 Out WorldRaycast_3 RayLength)
  (connect Subgraph_0 SelectLandscapeHits WorldRaycast_3 SelectLandscapeHits)
  (output OutPoints AttributeFilter_17.InsideFilter)
  (output OutPoints Branch_1.Output%20A))
