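; PCG debug graph for mesh grass that prints partition coordinates for the input actor point on screen.
; It gathers all world PCGPartitionActors as candidates and loops a DeterminePartition subgraph to find the partition matching the point.
; A Blueprint element extracts coordinates from that partition's path.
; A HasValidData check selects between those coordinates and a "_?_?" placeholder before printing with an owner prefix.
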(graph PCG_MeshGrass_DebugPartition
  (inputs (ActorPoint :type Point))
  (bind GetActorData_0 (Get%20Actor%20Data :ActorSelector (struct :ActorFilter AllWorldActors :bMustOverlapSelf true :ActorSelection ByClass :ActorSelectionClass /Script/PCG.PCGPartitionActor :bSelectMultiple true) :Mode GetSinglePoint))
  (bind Loop_1 (Loop :Graph /Game/Athena/Helios/Environments/Landscape/WorldBuildingTools/PCGGrass/PCG_MeshGrass_DeterminePartition.PCG_MeshGrass_DeterminePartition))
  (bind ExecuteBlueprint_2 (Execute%20Blueprint :BlueprintElementInstance (class:/Game/Athena/Helios/Environments/Landscape/WorldBuildingTools/PCGGrass/PCGBP_GetPartitionCoordsFromPath.PCGBP_GetPartitionCoordsFromPath_C) :BlueprintElementType /Game/Athena/Helios/Environments/Landscape/WorldBuildingTools/PCGGrass/PCGBP_GetPartitionCoordsFromPath.PCGBP_GetPartitionCoordsFromPath_C))
  (bind Subgraph_3 (Subgraph :Graph /Game/Athena/Helios/Environments/Landscape/WorldBuildingTools/PCGGrass/PCG_HasValidData.PCG_HasValidData))
  (bind CreateAttribute_5 (Create%20Constant :AttributeTypes (struct :Type String :StringValue "_?_?")))
  (bind Select_4 (Select))
  (bind Print%20String_6 (Print%20String :bPrefixWithNode false :bPrefixWithOwner true :bPrintToScreen true :PrintToScreenColor (struct :B 107 :G 189 :R 255 :A 255) :PrintToScreenDuration 4.0))
  (connect Loop_1 PartitionActor ExecuteBlueprint_2 In)
  (connect GetActorData_0 Out Loop_1 PartitionActorCandidates)
  (connect DefaultInputNode ActorPoint Loop_1 TargetActor)
  (connect DefaultInputNode ActorPoint Print%20String_6 In)
  (connect Select_4 Out Print%20String_6 PrintString)
  (connect Subgraph_3 IsValid Select_4 bUseInputB)
  (connect CreateAttribute_5 Out Select_4 Input%20A)
  (connect ExecuteBlueprint_2 Out Select_4 Input%20B)
  (connect Loop_1 PartitionActor Subgraph_3 In))
