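; PCG subgraph for mesh grass that decodes biome texture coordinates into a texture offset, tiling and color channel attribute name.
; The input's X component picks one of four quadrant offsets, and Y picks R, G, B or A, appended to "$Color." as a selector string.
; The IsAtlas parameter, true by default, uses the quadrant offset with unit tiling; otherwise offset is zero and tiling doubles.
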(graph PCG_MeshGrass_DecodeBiomeCoordinates
  (params (IsAtlas :type Bool :default true))
  (inputs (Biome%20Texture%20Coordinates :type Param))
  (outputs (Texture%20Offset :type Param) (Tiling :type Param) (Color%20Attribute :type Param))
  (bind BreakVectorAttribute_5 (Break%20Vector%20Attribute))
  (bind CreateAttribute_1 (Create%20Constant :AttributeTypes (struct :Type Vector2 :Vector2Value (-0.25 0.25))))
  (bind CreateAttribute_2 (Create%20Constant :AttributeTypes (struct :Type Vector2 :Vector2Value (0.25 0.25))))
  (bind CreateAttribute_3 (Create%20Constant :AttributeTypes (struct :Type Vector2 :Vector2Value (-0.25 -0.25))))
  (bind CreateAttribute_4 (Create%20Constant :AttributeTypes (struct :Type Vector2 :Vector2Value (0.25 -0.25))))
  (bind Select%20%28Multi%29_0 (Select%20%28Multi%29 :EnumSelection (struct :Class /Script/ACLPlugin.ACLCompressionLevel) :IntOptions [0 1 2 3]))
  (bind CreateAttribute_7 (Create%20Constant :AttributeTypes (struct :Type String :StringValue "R")))
  (bind CreateAttribute_8 (Create%20Constant :AttributeTypes (struct :Type String :StringValue "G")))
  (bind CreateAttribute_9 (Create%20Constant :AttributeTypes (struct :Type String :StringValue "B")))
  (bind CreateAttribute_10 (Create%20Constant :AttributeTypes (struct :Type String :StringValue "A")))
  (bind Select%20%28Multi%29_6 (Select%20%28Multi%29 :EnumSelection (struct :Class /Script/ACLPlugin.ACLCompressionLevel) :IntOptions [0 1 2 3]))
  (bind CreateAttribute_12 (Create%20Constant :AttributeTypes (struct :Type String :StringValue "$Color.")))
  (bind AttributeStringOp_11 (Attribute%20String%20Op))
  (bind CreateConstant_0 (Create%20Constant :AttributeTypes (struct :Type Vector2)))
  (bind Select_0 (Select :bUseInputB-param IsAtlas))
  (bind CreateConstant_1 (Create%20Constant :AttributeTypes (struct :Type Vector2 :Vector2Value (1.0 1.0))))
  (bind CreateConstant_2 (Create%20Constant :AttributeTypes (struct :Type Vector2 :Vector2Value (2.0 2.0))))
  (bind Select_1 (Select :bUseInputB-param IsAtlas))
  (connect CreateAttribute_12 Out AttributeStringOp_11 InA)
  (connect Select%20%28Multi%29_6 Out AttributeStringOp_11 InB)
  (connect DefaultInputNode Biome%20Texture%20Coordinates BreakVectorAttribute_5 In)
  (connect CreateConstant_0 Out Select_0 Input%20A)
  (connect Select%20%28Multi%29_0 Out Select_0 Input%20B)
  (connect CreateConstant_2 Out Select_1 Input%20A)
  (connect CreateConstant_1 Out Select_1 Input%20B)
  (connect CreateAttribute_2 Out Select%20%28Multi%29_0 %30)
  (connect CreateAttribute_1 Out Select%20%28Multi%29_0 %31)
  (connect CreateAttribute_4 Out Select%20%28Multi%29_0 %32)
  (connect CreateAttribute_3 Out Select%20%28Multi%29_0 %33)
  (connect CreateAttribute_2 Out Select%20%28Multi%29_0 Default)
  (connect BreakVectorAttribute_5 X Select%20%28Multi%29_0 IntegerSelection)
  (connect CreateAttribute_7 Out Select%20%28Multi%29_6 %30)
  (connect CreateAttribute_8 Out Select%20%28Multi%29_6 %31)
  (connect CreateAttribute_9 Out Select%20%28Multi%29_6 %32)
  (connect CreateAttribute_10 Out Select%20%28Multi%29_6 %33)
  (connect BreakVectorAttribute_5 Y Select%20%28Multi%29_6 IntegerSelection)
  (output Color%20Attribute AttributeStringOp_11)
  (output Texture%20Offset Select_0)
  (output Tiling Select_1))
