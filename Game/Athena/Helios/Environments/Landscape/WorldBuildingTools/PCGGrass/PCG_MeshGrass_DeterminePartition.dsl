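; PCG mesh-grass subgraph that checks whether a candidate partition actor contains a target actor and, if so, outputs that partition's actor reference.
; It adds each candidate's bounds to its position to get absolute min and max corners, then compares the target position against them on X and Y only.
; The combined InBox result from the first point drives a Branch that passes the first candidate's ActorReference to the PartitionActor output.
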(graph PCG_MeshGrass_DeterminePartition
  (inputs (PartitionActorCandidates :type Point) (TargetActor :type Point))
  (outputs (PartitionActor :type Param))
  (bind AttributeMathsOp_1 (Attribute%20Maths%20Op :InputSource1 $Position :InputSource2 $BoundsMax :OutputTarget AbsoluteBoundsMax))
  (bind AttributeMathsOp_2 (Attribute%20Maths%20Op :InputSource1 $Position :InputSource2 $BoundsMin :OutputTarget AbsoluteBoundsMin))
  (bind AttributeCompareOp_0 (Attribute%20Compare%20Op :InputSource1 $Position.X :InputSource2 AbsoluteBoundsMin.X :Operation Greater :OutputTarget InBox))
  (bind AttributeCompareOp_5 (Attribute%20Compare%20Op :InputSource1 $Position.X :InputSource2 AbsoluteBoundsMax.X :Operation Less :OutputTarget InBox))
  (bind AttributeCompareOp_6 (Attribute%20Compare%20Op :InputSource1 $Position.Y :InputSource2 AbsoluteBoundsMin.Y :Operation Greater :OutputTarget InBox))
  (bind AttributeCompareOp_7 (Attribute%20Compare%20Op :InputSource1 $Position.Y :InputSource2 AbsoluteBoundsMax.Y :Operation Less :OutputTarget InBox))
  (bind AttributeBooleanOp_8 (Attribute%20Boolean%20Op))
  (bind AttributeBooleanOp_9 (Attribute%20Boolean%20Op))
  (bind AttributeBooleanOp_10 (Attribute%20Boolean%20Op))
  (bind GetAttributeFromPointIndex_11 (Get%20Attribute%20From%20Point%20Index :bForceOutputAttributeToBeInElementsDomain true :OutputAttributeName InBox))
  (bind GetAttributeFromPointIndex_19 (Get%20Attribute%20From%20Point%20Index :bForceOutputAttributeToBeInElementsDomain true :InputSource ActorReference :OutputAttributeName ActorReference))
  (bind Branch_18 (Branch))
  (connect AttributeCompareOp_0 Out AttributeBooleanOp_8 InA)
  (connect AttributeCompareOp_5 Out AttributeBooleanOp_8 InB)
  (connect AttributeBooleanOp_8 Out AttributeBooleanOp_9 InA)
  (connect AttributeCompareOp_6 Out AttributeBooleanOp_9 InB)
  (connect AttributeBooleanOp_9 Out AttributeBooleanOp_10 InA)
  (connect AttributeCompareOp_7 Out AttributeBooleanOp_10 InB)
  (connect DefaultInputNode TargetActor AttributeCompareOp_0 InA)
  (connect AttributeMathsOp_2 Out AttributeCompareOp_0 InB)
  (connect DefaultInputNode TargetActor AttributeCompareOp_5 InA)
  (connect AttributeMathsOp_2 Out AttributeCompareOp_5 InB)
  (connect DefaultInputNode TargetActor AttributeCompareOp_6 InA)
  (connect AttributeMathsOp_2 Out AttributeCompareOp_6 InB)
  (connect DefaultInputNode TargetActor AttributeCompareOp_7 InA)
  (connect AttributeMathsOp_2 Out AttributeCompareOp_7 InB)
  (connect DefaultInputNode PartitionActorCandidates AttributeMathsOp_1 InA)
  (connect DefaultInputNode PartitionActorCandidates AttributeMathsOp_1 InB)
  (connect AttributeMathsOp_1 Out AttributeMathsOp_2 InA)
  (connect AttributeMathsOp_1 Out AttributeMathsOp_2 InB)
  (connect GetAttributeFromPointIndex_11 Attribute Branch_18 bOutputToB)
  (connect GetAttributeFromPointIndex_19 Attribute Branch_18 In)
  (connect AttributeBooleanOp_10 Out GetAttributeFromPointIndex_11 In)
  (connect DefaultInputNode PartitionActorCandidates GetAttributeFromPointIndex_19 In)
  (output PartitionActor Branch_18.Output%20B))
