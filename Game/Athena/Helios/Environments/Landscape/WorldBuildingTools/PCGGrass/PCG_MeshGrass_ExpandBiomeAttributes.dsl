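; PCG graph for mesh grass that expands an incoming collection of attribute data into separate biome outputs, BiomeBase and BiomeA through BiomeG.
; A chain of seven Filter Data By Index nodes each takes the first remaining entry as one output and passes the rest to the next.
; Biome assignment therefore depends on input order, and BiomeG receives everything left after BiomeF, not just a single entry.
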(graph PCG_MeshGrass_ExpandBiomeAttributes
  (inputs (InAttributes :type Param))
  (outputs (BiomeBase :type Param) (BiomeA :type Param) (BiomeB :type Param) (BiomeC :type Param) (BiomeD :type Param) (BiomeE :type Param) (BiomeF :type Param) (BiomeG :type Param))
  (bind FilterDataByIndex_0 (Filter%20Data%20By%20Index :SelectedIndices "0"))
  (bind FilterDataByIndex_2 (Filter%20Data%20By%20Index :SelectedIndices "0"))
  (bind FilterDataByIndex_4 (Filter%20Data%20By%20Index :SelectedIndices "0"))
  (bind FilterDataByIndex_6 (Filter%20Data%20By%20Index :SelectedIndices "0"))
  (bind FilterDataByIndex_8 (Filter%20Data%20By%20Index :SelectedIndices "0"))
  (bind FilterDataByIndex_10 (Filter%20Data%20By%20Index :SelectedIndices "0"))
  (bind FilterDataByIndex_1 (Filter%20Data%20By%20Index :SelectedIndices "0"))
  (connect DefaultInputNode InAttributes FilterDataByIndex_0 In)
  (connect FilterDataByIndex_10 OutsideFilter FilterDataByIndex_1 In)
  (connect FilterDataByIndex_0 OutsideFilter FilterDataByIndex_2 In)
  (connect FilterDataByIndex_2 OutsideFilter FilterDataByIndex_4 In)
  (connect FilterDataByIndex_4 OutsideFilter FilterDataByIndex_6 In)
  (connect FilterDataByIndex_6 OutsideFilter FilterDataByIndex_8 In)
  (connect FilterDataByIndex_8 OutsideFilter FilterDataByIndex_10 In)
  (output BiomeA FilterDataByIndex_2.InsideFilter)
  (output BiomeB FilterDataByIndex_4.InsideFilter)
  (output BiomeBase FilterDataByIndex_0.InsideFilter)
  (output BiomeC FilterDataByIndex_6.InsideFilter)
  (output BiomeD FilterDataByIndex_8.InsideFilter)
  (output BiomeE FilterDataByIndex_10.InsideFilter)
  (output BiomeF FilterDataByIndex_1.InsideFilter)
  (output BiomeG FilterDataByIndex_1.OutsideFilter))
