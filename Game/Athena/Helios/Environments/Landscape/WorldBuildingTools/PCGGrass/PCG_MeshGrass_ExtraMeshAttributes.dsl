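; PCG graph that builds a default set of extra per-mesh attributes for mesh grass, output as the MeshAttributes param data.
; A Create Constant node seeds ScaleMin, then a linear chain of Add Attribute nodes appends ScaleMax, start and end instance cull distances, WPO disable distance and DetailMode.
; All values are fixed constants with no graph inputs, with scale limits as vectors and the rest as integers.
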(graph PCG_MeshGrass_ExtraMeshAttributes
  (outputs (MeshAttributes :type Param))
  (bind CreateAttribute_3 (Create%20Constant :AttributeTypes (struct :Type Vector :VectorValue (1.0 1.0 1.0)) :OutputTarget ScaleMin))
  (bind AddAttribute_0 (Add%20Attribute :AttributeTypes (struct :Type Vector :VectorValue (1.0 1.0 1.0)) :OutputTarget ScaleMax))
  (bind AddAttribute_1 (Add%20Attribute :AttributeTypes (struct :Type Integer32 :Int32Value 4000 :DoubleValue 4e+03 :IntValue 4000 :VectorValue (1.0 1.0 1.0)) :OutputTarget InstanceStartCullDistance))
  (bind AddAttribute_2 (Add%20Attribute :AttributeTypes (struct :Type Integer32 :Int32Value 5000 :DoubleValue 5e+03 :IntValue 5000 :VectorValue (1.0 1.0 1.0)) :OutputTarget InstanceEndCullDistance))
  (bind AddAttribute_3 (Add%20Attribute :AttributeTypes (struct :Type Integer32 :Int32Value 5000 :DoubleValue 5e+03 :IntValue 3000 :VectorValue (1.0 1.0 1.0)) :OutputTarget WPODisableDistance))
  (bind AddAttribute_5 (Add%20Attribute :AttributeTypes (struct :Type Integer32 :Int32Value 3 :IntValue 3 :VectorValue (1.0 1.0 1.0)) :OutputTarget DetailMode))
  (connect CreateAttribute_3 Out AddAttribute_0 In)
  (connect AddAttribute_0 Out AddAttribute_1 In)
  (connect AddAttribute_1 Out AddAttribute_2 In)
  (connect AddAttribute_2 Out AddAttribute_3 In)
  (connect AddAttribute_3 Out AddAttribute_5 In)
  (output MeshAttributes AddAttribute_5))
