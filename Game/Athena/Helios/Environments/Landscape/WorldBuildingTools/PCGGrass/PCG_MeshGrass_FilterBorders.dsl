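; PCG graph that trims the borders of a point-covered area, used by the mesh grass world-building tools.
; It builds a local grid over the actor bounds, padded in XY in proportion to their size, and colors it black.
; Projection onto white OriginalPoints, expanded by BorderDistance, marks covered cells through color.
; Uncovered cells get full density and BorderSize extents, then are subtracted from InSpatial with a binary Difference, eroding its edges.
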(graph PCG_MeshGrass_FilterBorders
  (inputs (InSpatial :type Spatial) (OriginalPoints :type Point) (ActorPoint :type Point) (BorderSize :type Param) (BorderDistance :type Param))
  (outputs (OutSpatial :type Spatial))
  (bind CreateAttribute_12 (Create%20Constant :AttributeTypes (struct :Type Vector)))
  (bind ApplyScaleToBounds_14 (Apply%20Scale%20To%20Bounds))
  (bind GetAttributeFromPointIndex_8 (Get%20Attribute%20From%20Point%20Index :bForceOutputAttributeToBeInElementsDomain true :InputSource $Extents :OutputAttributeName None))
  (bind CreateAttribute_21 (Create%20Constant :AttributeTypes (struct :DoubleValue 0.3)))
  (bind AttributeVectorOp_18 (Attribute%20Vector%20Op :InputSource1 $BoundsMax :Operation Length :OutputTarget Lenght))
  (bind GetAttributeFromPointIndex_19 (Get%20Attribute%20From%20Point%20Index :bForceOutputAttributeToBeInElementsDomain true :OutputAttributeName None))
  (bind AttributeMathsOp_20 (Attribute%20Maths%20Op :Operation Multiply))
  (bind CreateAttribute_17 (Create%20Constant))
  (bind MakeVectorAttribute_16 (Make%20Vector%20Attribute :OutputType Vector))
  (bind AttributeMathsOp_15 (Attribute%20Maths%20Op))
  (bind CreateAttribute_22 (Create%20Constant :AttributeTypes (struct :DoubleValue 1e+02) :OutputTarget CellSize))
  (bind CreatePointsGrid_5 (Create%20Points%20Grid :CellSize (1e+02 1e+02 2e+02) :CoordinateSpace LocalComponent :GridExtents (1e+02 1e+02 5e+01) :PointSteepness 0.5))
  (bind GetAttributeFromPointIndex_7 (Get%20Attribute%20From%20Point%20Index :bForceOutputAttributeToBeInElementsDomain true :InputSource $LocalCenter :OutputAttributeName None))
  (bind TransformPoints_6 (Transform%20Points))
  (bind AttributeMathsOp_11 (Attribute%20Maths%20Op :InputSource1 $Color :Operation Set))
  (bind CreateAttribute_10 (Create%20Constant :AttributeTypes (struct :Type Vector :VectorValue (1.0 1.0 1.0))))
  (bind AttributeMathsOp_9 (Attribute%20Maths%20Op :InputSource1 $Color :Operation Set))
  (bind ExtentsModifier_13 (Extents%20Modifier :Extents (2e+01 2e+01 2e+01)))
  (bind Projection_0 (Projection :bKeepZeroDensityPoints true :ProjectionParams (struct :bProjectPositions false :bProjectRotations false :ColorBlendMode TargetValue)))
  (bind AttributeFilter_1 (Filter%20Attribute%20Elements :AttributeTypes (struct :Type Vector) :bUseConstantThreshold true :bUseSpatialQuery true :Operator Equal :TargetAttribute $Color))
  (bind CreateAttribute_4 (Create%20Constant :AttributeTypes (struct :DoubleValue 1.0)))
  (bind AttributeMathsOp_3 (Attribute%20Maths%20Op :InputSource1 $Density :Operation Set))
  (bind ExtentsModifier_2 (Extents%20Modifier :Extents (75.0 75.0 5e+01)))
  (bind Difference_23 (Difference :bDiffMetadata false :DensityFunction Binary :Mode Discrete))
  (connect DefaultInputNode ActorPoint ApplyScaleToBounds_14 In)
  (connect Projection_0 Out AttributeFilter_1 In)
  (connect AttributeFilter_1 InsideFilter AttributeMathsOp_3 InA)
  (connect CreateAttribute_4 Out AttributeMathsOp_3 InB)
  (connect DefaultInputNode OriginalPoints AttributeMathsOp_9 InA)
  (connect CreateAttribute_10 Out AttributeMathsOp_9 InB)
  (connect TransformPoints_6 Out AttributeMathsOp_11 InA)
  (connect CreateAttribute_12 Out AttributeMathsOp_11 InB)
  (connect GetAttributeFromPointIndex_8 Attribute AttributeMathsOp_15 InA)
  (connect MakeVectorAttribute_16 Out AttributeMathsOp_15 InB)
  (connect GetAttributeFromPointIndex_19 Attribute AttributeMathsOp_20 InA)
  (connect CreateAttribute_21 Out AttributeMathsOp_20 InB)
  (connect ApplyScaleToBounds_14 Out AttributeVectorOp_18 In)
  (connect CreateAttribute_22 Out CreatePointsGrid_5 CellSize)
  (connect AttributeMathsOp_15 Out CreatePointsGrid_5 GridExtents)
  (connect ExtentsModifier_2 Out Difference_23 Differences)
  (connect DefaultInputNode InSpatial Difference_23 Source)
  (connect DefaultInputNode BorderSize ExtentsModifier_2 Extents)
  (connect AttributeMathsOp_3 Out ExtentsModifier_2 In)
  (connect DefaultInputNode BorderDistance ExtentsModifier_13 Extents)
  (connect AttributeMathsOp_9 Out ExtentsModifier_13 In)
  (connect ApplyScaleToBounds_14 Out GetAttributeFromPointIndex_7 In)
  (connect ApplyScaleToBounds_14 Out GetAttributeFromPointIndex_8 In)
  (connect AttributeVectorOp_18 Out GetAttributeFromPointIndex_19 In)
  (connect AttributeMathsOp_20 Out MakeVectorAttribute_16 X)
  (connect AttributeMathsOp_20 Out MakeVectorAttribute_16 Y)
  (connect CreateAttribute_17 Out MakeVectorAttribute_16 Z)
  (connect AttributeMathsOp_11 Out Projection_0 In)
  (connect ExtentsModifier_13 Out Projection_0 Projection%20Target)
  (connect CreatePointsGrid_5 Out TransformPoints_6 In)
  (connect GetAttributeFromPointIndex_7 Attribute TransformPoints_6 OffsetMax)
  (connect GetAttributeFromPointIndex_7 Attribute TransformPoints_6 OffsetMin)
  (output OutSpatial Difference_23))
