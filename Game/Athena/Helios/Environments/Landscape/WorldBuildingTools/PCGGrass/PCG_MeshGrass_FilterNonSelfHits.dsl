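; PCG subgraph for mesh grass that removes input spatial data near surfaces belonging to actors other than the owning actor.
; A world ray query, sized from the actor point's bounds plus a Z offset, feeds a surface sampler when its length passes a check.
; Samples whose hit actor reference differs from the owner are expanded by NonSelfHitDistance and subtracted from the input.
; Surviving output has its density reset to a constant.
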(graph PCG_MeshGrass_FilterNonSelfHits
  (inputs (InSpatial :type Spatial) (ActorPoint :type Point) (NonSelfHitDistance :type Param) (NonSelfTraceZOffset :type Param) (NonSelfTraceComplex :type Param))
  (outputs (OutSpatial :type Spatial))
  (bind AttributeMathsOp_18 (Attribute%20Maths%20Op :InputSource1 $BoundsMin.Z :InputSource2 $BoundsMin.Z :Operation Abs))
  (bind AttributeMathsOp_19 (Attribute%20Maths%20Op :InputSource1 $BoundsMin.Z :InputSource2 $BoundsMax.Z :OutputTarget RayLenght))
  (bind GetAttributeFromPointIndex_11 (Get%20Attribute%20From%20Point%20Index :bForceOutputAttributeToBeInElementsDomain true :OutputAttributeName None))
  (bind AttributeMathsOp_15 (Attribute%20Maths%20Op))
  (bind CreateAttribute_13 (Create%20Constant :AttributeTypes (struct :DoubleValue 1.0 :VectorValue (0.0 0.0 1e+02))))
  (bind AttributeMathsOp_12 (Attribute%20Maths%20Op))
  (bind CreateAttribute_1 (Create%20Constant))
  (bind AttributeCompareOp_0 (Attribute%20Compare%20Op :Operation Greater))
  (bind CreateAttribute_16 (Create%20Constant :AttributeTypes (struct :Type Vector) :OutputTarget Vector))
  (bind AttributeMathsOp_17 (Attribute%20Maths%20Op :InputSource1 Vector.Z :Operation Set))
  (bind AttributeMathsOp_20 (Attribute%20Maths%20Op :InputSource1 $Position.Z :InputSource2 $BoundsMax.Z))
  (bind GetAttributeFromPointIndex_10 (Get%20Attribute%20From%20Point%20Index :bForceOutputAttributeToBeInElementsDomain true :InputSource $Position :OutputAttributeName None))
  (bind AttributeMathsOp_14 (Attribute%20Maths%20Op))
  (bind WorldRayHitQuery_7 (World%20Ray%20Hit%20Query :QueryParams (struct :bOverrideDefaultParams true :RayOrigin (0.0 0.0 8e+03) :RayLength 1.5e+02 :bIgnoreSelfHits false :ActorTagsList "PCG_MeshGrass" :bGetReferenceToActorHit true)))
  (bind Branch_8 (Branch))
  (bind SurfaceSampler_8 (Surface%20Sampler :bApplyDensityToPoints false :Looseness 0.5 :PointExtents (1e+01 1e+01 1e+01) :PointsPerSquaredMeter 8.0 :PointSteepness 0.5))
  (bind Subgraph_9 (Subgraph :Graph /Game/Athena/Helios/Environments/Landscape/WorldBuildingTools/PCGGrass/PCG_HasValidPoints.PCG_HasValidPoints))
  (bind Branch_6 (Branch))
  (bind AttributeCompareOp_3 (Attribute%20Compare%20Op :InputSource1 ActorReference :InputSource2 ActorReference :OutputTarget IsSelf))
  (bind AttributeFilter_4 (Filter%20Attribute%20Elements :AttributeTypes (struct :Type Boolean :BoolValue true) :bUseConstantThreshold true :bUseSpatialQuery true :Operator NotEqual :TargetAttribute IsSelf))
  (bind ExtentsModifier_5 (Extents%20Modifier :Extents (2e+01 2e+01 2e+01)))
  (bind Difference_0 (Difference :bDiffMetadata false :DensityFunction ClampedSubstraction :Mode Discrete))
  (bind CreateAttribute_2 (Create%20Constant :AttributeTypes (struct :DoubleValue 1.0)))
  (bind AttributeMathsOp_1 (Attribute%20Maths%20Op :InputSource1 $Density :Operation Set))
  (reroute Actor%20Point ActorPoint)
  (reroute in%20Spatial InSpatial)
  (reroute Non%20Self%20Hit%20Distance NonSelfHitDistance)
  (reroute Non%20Self%20Trace%20Complex NonSelfTraceComplex)
  (reroute Non%20Self%20Trace%20ZOffset NonSelfTraceZOffset)
  (connect AttributeMathsOp_12 Out AttributeCompareOp_0 InA)
  (connect CreateAttribute_1 Out AttributeCompareOp_0 InB)
  (connect Branch_6 Output%20B AttributeCompareOp_3 InA)
  (connect Actor%20Point Out AttributeCompareOp_3 InB)
  (connect AttributeCompareOp_3 Out AttributeFilter_4 In)
  (connect Difference_0 Out AttributeMathsOp_1 InA)
  (connect CreateAttribute_2 Out AttributeMathsOp_1 InB)
  (connect AttributeMathsOp_15 Out AttributeMathsOp_12 InA)
  (connect CreateAttribute_13 Out AttributeMathsOp_12 InB)
  (connect GetAttributeFromPointIndex_10 Attribute AttributeMathsOp_14 InA)
  (connect AttributeMathsOp_17 Out AttributeMathsOp_14 InB)
  (connect Non%20Self%20Trace%20ZOffset Out AttributeMathsOp_15 InA)
  (connect GetAttributeFromPointIndex_11 Attribute AttributeMathsOp_15 InB)
  (connect CreateAttribute_16 Out AttributeMathsOp_17 InA)
  (connect Non%20Self%20Trace%20ZOffset Out AttributeMathsOp_17 InB)
  (connect Actor%20Point Out AttributeMathsOp_18 In)
  (connect AttributeMathsOp_18 Out AttributeMathsOp_19 InA)
  (connect AttributeMathsOp_18 Out AttributeMathsOp_19 InB)
  (connect Actor%20Point Out AttributeMathsOp_20 InA)
  (connect Actor%20Point Out AttributeMathsOp_20 InB)
  (connect Subgraph_9 IsValid Branch_6 bOutputToB)
  (connect SurfaceSampler_8 Out Branch_6 In)
  (connect AttributeCompareOp_0 Out Branch_8 bOutputToB)
  (connect WorldRayHitQuery_7 Out Branch_8 In)
  (connect ExtentsModifier_5 Out Difference_0 Differences)
  (connect in%20Spatial Out Difference_0 Source)
  (connect Non%20Self%20Hit%20Distance Out ExtentsModifier_5 Extents)
  (connect AttributeFilter_4 InsideFilter ExtentsModifier_5 In)
  (connect AttributeMathsOp_20 Out GetAttributeFromPointIndex_10 In)
  (connect AttributeMathsOp_19 Out GetAttributeFromPointIndex_11 In)
  (connect SurfaceSampler_8 Out Subgraph_9 In)
  (connect Branch_8 Output%20B SurfaceSampler_8 Surface)
  (connect Non%20Self%20Trace%20Complex Out WorldRayHitQuery_7 bTraceComplex)
  (connect AttributeMathsOp_12 Out WorldRayHitQuery_7 RayLength)
  (connect AttributeMathsOp_14 Out WorldRayHitQuery_7 RayOrigin)
  (output OutSpatial AttributeMathsOp_1))
