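; PCG graph for mesh grass that culls points by surface normal and by closeness to points failing that normal test.
; Normal To Density flags points using an actor offset, optionally biased by volume attributes.
; When FilterHeightGradient is on, a Distance pass flags edge points, with EdgeDistance optionally volume-scaled.
; FilterNormals controls whether normal flags join edge flags and any existing Discard; survivors lose temporary attributes and pass through PCG_GateValidPoints.
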(graph PCG_MeshGrass_FilterTopology
  (inputs (InPoints :type Point) (ActorAttributes :type Param) (VolumeAttributes :type Param))
  (outputs (OutPoints :type Point))
  (bind FilterDataByAttribute_0 (Filter%20Data%20By%20Attribute :Attribute Discard))
  (bind AddAttribute_1 (Add%20Attribute :AttributeTypes (struct :Type Boolean) :OutputTarget Discard))
  (bind MergePoints_0 (Merge%20Points))
  ;[0.05 0.05 0.05] Normal Filtering
  (bind AddAttribute_2 (Add%20Attribute :AttributeTypes (struct :Type Boolean) :OutputTarget DiscardNormals))
  (bind DeleteAttributes_5 (Delete%20Attributes :Operation KeepSelectedAttributes :SelectedAttributes "NormalOffset"))
  (bind DeleteAttributes_6 (Delete%20Attributes :Operation KeepSelectedAttributes :SelectedAttributes "Normal Offset Bias"))
  (bind AttributeMathsOp_2 (Attribute%20Maths%20Op))
  (bind DeleteAttributes_7 (Delete%20Attributes :bTokenizeOnWhiteSpace true :Operation KeepSelectedAttributes :SelectedAttributes "VolumeOverride"))
  (bind Select_2 (Select))
  (bind NormalToDensity_1 (Normal%20To%20Density))
  (bind AttributeCompareOp_2 (Attribute%20Compare%20Op :InputSource1 $Density :Operation Less :OutputTarget DiscardNormals))
  ;[0.05 0.05 0.05] Edge Distance Filtering
  (bind AddAttribute_0 (Add%20Attribute :AttributeTypes (struct :Type Boolean) :OutputTarget DiscardEdgeDistance))
  (bind DeleteAttributes_3 (Delete%20Attributes :bTokenizeOnWhiteSpace true :Operation KeepSelectedAttributes :SelectedAttributes "FilterHeightGradient"))
  (bind Branch_0 (Branch))
  (bind AttributeFilter_0 (Filter%20Attribute%20Elements :AttributeTypes (struct :Type Boolean) :bUseConstantThreshold true :Operator Equal :TargetAttribute DiscardNormals :ThresholdAttribute @Last))
  (bind CreateConstant_0 (Create%20Constant :AttributeTypes (struct :DoubleValue 1e-05)))
  (bind DeleteAttributes_0 (Delete%20Attributes :Operation KeepSelectedAttributes :SelectedAttributes "Edge Distance Factor"))
  (bind DeleteAttributes_1 (Delete%20Attributes :Operation KeepSelectedAttributes :SelectedAttributes "EdgeDistance"))
  (bind AttributeMathsOp_1 (Attribute%20Maths%20Op :Operation Multiply))
  (bind DeleteAttributes_2 (Delete%20Attributes :bTokenizeOnWhiteSpace true :Operation KeepSelectedAttributes :SelectedAttributes "VolumeOverride"))
  (bind Select_0 (Select))
  (bind AttributeMathsOp_0 (Attribute%20Maths%20Op :Operation Max))
  (bind Distance_0 (Distance :bOutputToAttribute false :bSetDensity true :MaximumDistance 1.0))
  (bind AttributeCompareOp_0 (Attribute%20Compare%20Op :InputSource1 $Density :Operation LessOrEqual :OutputTarget DiscardEdgeDistance))
  (bind MergePoints_1 (Merge%20Points))
  (bind Gather_0 (Gather))
  (bind AttributeFilter_2 (Filter%20Attribute%20Elements :AttributeTypes (struct :Type Boolean) :bUseConstantThreshold true :Operator Equal :TargetAttribute DiscardEdgeDistance :ThresholdAttribute @Last))
  ;[0.05 0.05 0.05] Discard
  (bind AttributeBooleanOp_0 (Attribute%20Boolean%20Op :InputSource1 DiscardNormals :InputSource2 DiscardEdgeDistance :Operation Or :OutputTarget DiscardLocalFiltering))
  (bind AttributeBooleanOp_2 (Attribute%20Boolean%20Op :InputSource1 DiscardEdgeDistance :InputSource2 DiscardEdgeDistance :OutputTarget DiscardLocalFiltering))
  (bind Select_1 (Select :bUseInputB-param FilterNormals))
  (bind AttributeBooleanOp_1 (Attribute%20Boolean%20Op :InputSource1 DiscardLocalFiltering :InputSource2 Discard :Operation Or :OutputTarget Discard))
  (bind AttributeFilter_1 (Filter%20Attribute%20Elements :AttributeTypes (struct :Type Boolean) :bUseConstantThreshold true :Operator Equal :TargetAttribute Discard :ThresholdAttribute @Last))
  (bind DeleteAttributes_4 (Delete%20Attributes :SelectedAttributes "Discard,DiscardNormals,DiscardEdgeDistance,DiscardLocalFiltering"))
  (bind Subgraph_0 (Subgraph :Graph /Game/Athena/Helios/Environments/Landscape/WorldBuildingTools/PCGGrass/PCG_GateValidPoints.PCG_GateValidPoints))
  (reroute NamedRerouteDeclaration_0 ActorAttributes)
  (reroute NamedRerouteDeclaration_1 VolumeAttributes)
  (connect AttributeCompareOp_2 Out AddAttribute_0 In)
  (connect FilterDataByAttribute_0 OutsideFilter AddAttribute_1 In)
  (connect MergePoints_0 Out AddAttribute_2 In)
  (connect Gather_0 Out AttributeBooleanOp_0 InA)
  (connect Gather_0 Out AttributeBooleanOp_0 InB)
  (connect Select_1 Out AttributeBooleanOp_1 InA)
  (connect Select_1 Out AttributeBooleanOp_1 InB)
  (connect Gather_0 Out AttributeBooleanOp_2 InA)
  (connect Distance_0 Out AttributeCompareOp_0 InA)
  (connect NormalToDensity_1 Out AttributeCompareOp_2 InA)
  (connect Branch_0 Output%20B AttributeFilter_0 In)
  (connect AttributeBooleanOp_1 Out AttributeFilter_1 In)
  (connect AttributeCompareOp_0 Out AttributeFilter_2 In)
  (connect Select_0 Out AttributeMathsOp_0 InA)
  (connect CreateConstant_0 Out AttributeMathsOp_0 InB)
  (connect DeleteAttributes_1 Out AttributeMathsOp_1 InA)
  (connect DeleteAttributes_0 Out AttributeMathsOp_1 InB)
  (connect DeleteAttributes_5 Out AttributeMathsOp_2 InA)
  (connect DeleteAttributes_6 Out AttributeMathsOp_2 InB)
  (connect DeleteAttributes_3 Out Branch_0 bOutputToB)
  (connect AddAttribute_0 Out Branch_0 In)
  (connect NamedRerouteDeclaration_1 Out DeleteAttributes_0 In)
  (connect NamedRerouteDeclaration_0 Out DeleteAttributes_1 In)
  (connect NamedRerouteDeclaration_1 Out DeleteAttributes_2 In)
  (connect NamedRerouteDeclaration_0 Out DeleteAttributes_3 In)
  (connect AttributeFilter_1 InsideFilter DeleteAttributes_4 In)
  (connect NamedRerouteDeclaration_0 Out DeleteAttributes_5 In)
  (connect NamedRerouteDeclaration_1 Out DeleteAttributes_6 In)
  (connect NamedRerouteDeclaration_1 Out DeleteAttributes_7 In)
  (connect AttributeMathsOp_0 Out Distance_0 MaximumDistance)
  (connect AttributeFilter_0 InsideFilter Distance_0 Source)
  (connect AttributeFilter_0 OutsideFilter Distance_0 Target)
  (connect DefaultInputNode InPoints FilterDataByAttribute_0 In)
  (connect Branch_0 Output%20A Gather_0 In)
  (connect MergePoints_1 Out Gather_0 In)
  (connect AddAttribute_1 Out MergePoints_0 In)
  (connect FilterDataByAttribute_0 InsideFilter MergePoints_0 In)
  (connect AttributeCompareOp_0 Out MergePoints_1 In)
  (connect AttributeFilter_0 OutsideFilter MergePoints_1 In)
  (connect AddAttribute_2 Out NormalToDensity_1 In)
  (connect Select_2 Out NormalToDensity_1 Offset)
  (connect DeleteAttributes_2 Out Select_0 bUseInputB)
  (connect DeleteAttributes_1 Out Select_0 Input%20A)
  (connect AttributeMathsOp_1 Out Select_0 Input%20B)
  (connect AttributeBooleanOp_2 Out Select_1 Input%20A)
  (connect AttributeBooleanOp_0 Out Select_1 Input%20B)
  (connect DeleteAttributes_7 Out Select_2 bUseInputB)
  (connect DeleteAttributes_5 Out Select_2 Input%20A)
  (connect AttributeMathsOp_2 Out Select_2 Input%20B)
  (connect DeleteAttributes_4 Out Subgraph_0 In)
  (output OutPoints Subgraph_0.ValidPoints))
