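; PCG subgraph that removes mesh grass areas from the input spatial data where non-landscape world geometry intersects them.
; A downward-offset ray origin, built from actor bounds top plus WorldIntersectionDistance, feeds a world ray hit surface excluding PCG_MeshGrass actors, which is point-sampled.
; Samples project onto the input, optionally bounds-modified, and those at or above it are subtracted via binary Difference.
; With no samples, the input passes through.
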(graph PCG_MeshGrass_FilterWorldIntersection
  (inputs (InSpatial :type Spatial) (ActorPoint :type Point) (WorldIntersectionDistance :type Param) (ExtendNegativeBounds :type Param))
  (outputs (OutSpatial :type Spatial))
  (bind BoundsModifier_18 (Bounds%20Modifier :BoundsMin (1e+01 1e+01 1e+01)))
  (bind Select_19 (Select))
  (bind CreateAttribute_7 (Create%20Constant))
  (bind AttributeMathsOp_15 (Attribute%20Maths%20Op :InputSource1 $Position.Z :InputSource2 $BoundsMax.Z :OutputTarget RayOrigin))
  (bind GetAttributeFromPointIndex_16 (Get%20Attribute%20From%20Point%20Index :bForceOutputAttributeToBeInElementsDomain true :OutputAttributeName None))
  (bind AttributeMathsOp_12 (Attribute%20Maths%20Op))
  (bind CreateAttribute_11 (Create%20Constant))
  (bind MakeVectorAttribute_10 (Make%20Vector%20Attribute :OutputType Vector))
  (bind Subgraph_4 (Subgraph :Graph /Game/Athena/Helios/Environments/Landscape/WorldBuildingTools/PCGGrass/PCG_HasValidData.PCG_HasValidData))
  (bind WorldRayHitQuery_9 (World%20Ray%20Hit%20Query :QueryParams (struct :bOverrideDefaultParams true :RayOrigin (0.0 0.0 8e+03) :RayLength 1e+04 :bIgnorePCGHits true :ActorTagFilter Exclude :ActorTagsList "PCG_MeshGrass" :SelectLandscapeHits Exclude)))
  (bind Branch_5 (Branch))
  (bind CreateAttribute_14 (Create%20Constant :AttributeTypes (struct :DoubleValue 5e+01)))
  (bind MakeVectorAttribute_13 (Make%20Vector%20Attribute :OutputType Vector))
  (bind SurfaceSampler_4 (Surface%20Sampler :bApplyDensityToPoints false :Looseness 0.0 :PointExtents (5e+01 5e+01 5e+02) :PointsPerSquaredMeter 1.0 :PointSteepness 0.5))
  (bind GetPointsCount_5 (Get%20Element%20Count))
  (bind AttributeCompareOp_6 (Attribute%20Compare%20Op :Operation NotEqual))
  (bind Branch_8 (Branch))
  (bind TransformPoints_1 (Transform%20Points :bAbsoluteRotation true))
  (bind Projection_2 (Projection :bKeepZeroDensityPoints true :ProjectionParams (struct :bProjectRotations false)))
  (bind AttributeFilter_3 (Filter%20Attribute%20Elements :AttributeTypes (struct :BoolValue true) :Operator Lesser :TargetAttribute $Position.Z :ThresholdAttribute $Position.Z))
  (bind Branch_17 (Branch))
  (bind Difference_0 (Difference :bDiffMetadata false :DensityFunction Binary :Mode Discrete))
  (connect GetPointsCount_5 Out AttributeCompareOp_6 InA)
  (connect CreateAttribute_7 Out AttributeCompareOp_6 InB)
  (connect Projection_2 Out AttributeFilter_3 Filter)
  (connect TransformPoints_1 Out AttributeFilter_3 In)
  (connect GetAttributeFromPointIndex_16 Attribute AttributeMathsOp_12 InA)
  (connect DefaultInputNode WorldIntersectionDistance AttributeMathsOp_12 InB)
  (connect DefaultInputNode ActorPoint AttributeMathsOp_15 InA)
  (connect DefaultInputNode ActorPoint AttributeMathsOp_15 InB)
  (connect DefaultInputNode InSpatial BoundsModifier_18 In)
  (connect Subgraph_4 IsValid Branch_5 bOutputToB)
  (connect WorldRayHitQuery_9 Out Branch_5 In)
  (connect AttributeCompareOp_6 Out Branch_8 bOutputToB)
  (connect SurfaceSampler_4 Out Branch_8 In)
  (connect AttributeCompareOp_6 Out Branch_17 bOutputToB)
  (connect DefaultInputNode InSpatial Branch_17 In)
  (connect AttributeFilter_3 OutsideFilter Difference_0 Differences)
  (connect Branch_17 Output%20B Difference_0 Source)
  (connect AttributeMathsOp_15 Out GetAttributeFromPointIndex_16 In)
  (connect SurfaceSampler_4 Out GetPointsCount_5 In)
  (connect CreateAttribute_11 Out MakeVectorAttribute_10 X)
  (connect CreateAttribute_11 Out MakeVectorAttribute_10 Y)
  (connect AttributeMathsOp_12 Out MakeVectorAttribute_10 Z)
  (connect CreateAttribute_14 Out MakeVectorAttribute_13 X)
  (connect CreateAttribute_14 Out MakeVectorAttribute_13 Y)
  (connect DefaultInputNode WorldIntersectionDistance MakeVectorAttribute_13 Z)
  (connect TransformPoints_1 Out Projection_2 In)
  (connect Select_19 Out Projection_2 Projection%20Target)
  (connect DefaultInputNode ExtendNegativeBounds Select_19 bUseInputB)
  (connect DefaultInputNode InSpatial Select_19 Input%20A)
  (connect BoundsModifier_18 Out Select_19 Input%20B)
  (connect MakeVectorAttribute_10 Out Subgraph_4 In)
  (connect MakeVectorAttribute_13 Out SurfaceSampler_4 PointExtents)
  (connect Branch_5 Output%20B SurfaceSampler_4 Surface)
  (connect Branch_8 Output%20B TransformPoints_1 In)
  (connect MakeVectorAttribute_10 Out WorldRayHitQuery_9 RayOrigin)
  (output OutSpatial Branch_17.Output%20A)
  (output OutSpatial Difference_0))
