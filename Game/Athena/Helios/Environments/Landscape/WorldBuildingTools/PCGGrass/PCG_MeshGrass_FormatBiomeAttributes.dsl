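; PCG subgraph for MeshGrass world-building that formats biome attribute names, adding, keeping or stripping per-biome prefixes.
; With AddBiomePrefix on, it expands attributes through PCG_MeshGrass_ExpandBiomeAttributes, prefixes BiomeBase and BiomeA through BiomeG via PCGBP_RenameAttributeSet, then gathers them.
; Otherwise, IncludeBiomePrefix off truncates names by the short prefix length, with a second chop on data index 0 for the longer BiomeBase_ prefix.
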(graph PCG_MeshGrass_FormatBiomeAttributes
  (params (IncludeBiomePrefix :type Bool) (AddBiomePrefix :type Bool))
  (inputs (InAttributes :type Param))
  (outputs (BiomeAttributes :type Param))
  (bind Branch_3 (Branch :bOutputToB-param AddBiomePrefix))
  (bind Branch_17 (Branch :bOutputToB-param IncludeBiomePrefix))
  (bind ExecuteBlueprint_2 (Execute%20Blueprint :BlueprintElementInstance (class:/Game/Athena/Helios/Environments/Landscape/WorldBuildingTools/PCGGrass/PCGBP_RenameAttributeSet.PCGBP_RenameAttributeSet_C :Chop true :ReverseIndex true :TruncationIndex 7 :UseTruncation true) :BlueprintElementType /Game/Athena/Helios/Environments/Landscape/WorldBuildingTools/PCGGrass/PCGBP_RenameAttributeSet.PCGBP_RenameAttributeSet_C))
  (bind FilterDataByIndex_0 (Filter%20Data%20By%20Index :SelectedIndices "0"))
  (bind Branch_23 (Branch :bOutputToB-param IncludeBiomePrefix))
  (bind ExecuteBlueprint_12 (Execute%20Blueprint :BlueprintElementInstance (class:/Game/Athena/Helios/Environments/Landscape/WorldBuildingTools/PCGGrass/PCGBP_RenameAttributeSet.PCGBP_RenameAttributeSet_C :Chop true :ReverseIndex true :TruncationIndex 3 :UseTruncation true) :BlueprintElementType /Game/Athena/Helios/Environments/Landscape/WorldBuildingTools/PCGGrass/PCGBP_RenameAttributeSet.PCGBP_RenameAttributeSet_C))
  (bind Gather_31 (Gather))
  (bind Subgraph_20 (Subgraph :Graph /Game/Athena/Helios/Environments/Landscape/WorldBuildingTools/PCGGrass/PCG_MeshGrass_ExpandBiomeAttributes.PCG_MeshGrass_ExpandBiomeAttributes))
  (bind ExecuteBlueprint_14 (Execute%20Blueprint :BlueprintElementInstance (class:/Game/Athena/Helios/Environments/Landscape/WorldBuildingTools/PCGGrass/PCGBP_RenameAttributeSet.PCGBP_RenameAttributeSet_C :Prefix "BiomeBase_" :UsePrefix true) :BlueprintElementType /Game/Athena/Helios/Environments/Landscape/WorldBuildingTools/PCGGrass/PCGBP_RenameAttributeSet.PCGBP_RenameAttributeSet_C))
  (bind ExecuteBlueprint_22 (Execute%20Blueprint :BlueprintElementInstance (class:/Game/Athena/Helios/Environments/Landscape/WorldBuildingTools/PCGGrass/PCGBP_RenameAttributeSet.PCGBP_RenameAttributeSet_C :Prefix "BiomeA_" :UsePrefix true) :BlueprintElementType /Game/Athena/Helios/Environments/Landscape/WorldBuildingTools/PCGGrass/PCGBP_RenameAttributeSet.PCGBP_RenameAttributeSet_C))
  (bind ExecuteBlueprint_24 (Execute%20Blueprint :BlueprintElementInstance (class:/Game/Athena/Helios/Environments/Landscape/WorldBuildingTools/PCGGrass/PCGBP_RenameAttributeSet.PCGBP_RenameAttributeSet_C :Prefix "BiomeB_" :UsePrefix true) :BlueprintElementType /Game/Athena/Helios/Environments/Landscape/WorldBuildingTools/PCGGrass/PCGBP_RenameAttributeSet.PCGBP_RenameAttributeSet_C))
  (bind ExecuteBlueprint_26 (Execute%20Blueprint :BlueprintElementInstance (class:/Game/Athena/Helios/Environments/Landscape/WorldBuildingTools/PCGGrass/PCGBP_RenameAttributeSet.PCGBP_RenameAttributeSet_C :Prefix "BiomeC_" :UsePrefix true) :BlueprintElementType /Game/Athena/Helios/Environments/Landscape/WorldBuildingTools/PCGGrass/PCGBP_RenameAttributeSet.PCGBP_RenameAttributeSet_C))
  (bind ExecuteBlueprint_28 (Execute%20Blueprint :BlueprintElementInstance (class:/Game/Athena/Helios/Environments/Landscape/WorldBuildingTools/PCGGrass/PCGBP_RenameAttributeSet.PCGBP_RenameAttributeSet_C :Prefix "BiomeD_" :UsePrefix true) :BlueprintElementType /Game/Athena/Helios/Environments/Landscape/WorldBuildingTools/PCGGrass/PCGBP_RenameAttributeSet.PCGBP_RenameAttributeSet_C))
  (bind ExecuteBlueprint_30 (Execute%20Blueprint :BlueprintElementInstance (class:/Game/Athena/Helios/Environments/Landscape/WorldBuildingTools/PCGGrass/PCGBP_RenameAttributeSet.PCGBP_RenameAttributeSet_C :Prefix "BiomeE_" :UsePrefix true) :BlueprintElementType /Game/Athena/Helios/Environments/Landscape/WorldBuildingTools/PCGGrass/PCGBP_RenameAttributeSet.PCGBP_RenameAttributeSet_C))
  (bind ExecuteBlueprint_32 (Execute%20Blueprint :BlueprintElementInstance (class:/Game/Athena/Helios/Environments/Landscape/WorldBuildingTools/PCGGrass/PCGBP_RenameAttributeSet.PCGBP_RenameAttributeSet_C :Prefix "BiomeF_" :UsePrefix true) :BlueprintElementType /Game/Athena/Helios/Environments/Landscape/WorldBuildingTools/PCGGrass/PCGBP_RenameAttributeSet.PCGBP_RenameAttributeSet_C))
  (bind ExecuteBlueprint_35 (Execute%20Blueprint :BlueprintElementInstance (class:/Game/Athena/Helios/Environments/Landscape/WorldBuildingTools/PCGGrass/PCGBP_RenameAttributeSet.PCGBP_RenameAttributeSet_C :Prefix "BiomeG_" :UsePrefix true) :BlueprintElementType /Game/Athena/Helios/Environments/Landscape/WorldBuildingTools/PCGGrass/PCGBP_RenameAttributeSet.PCGBP_RenameAttributeSet_C))
  (bind Gather_33 (Gather))
  (connect DefaultInputNode InAttributes Branch_3 In)
  (connect Branch_3 Output%20A Branch_17 In)
  (connect FilterDataByIndex_0 InsideFilter Branch_23 In)
  (connect Branch_17 Output%20A ExecuteBlueprint_2 In)
  (connect Branch_23 Output%20A ExecuteBlueprint_12 In)
  (connect Subgraph_20 BiomeBase ExecuteBlueprint_14 In)
  (connect Subgraph_20 BiomeA ExecuteBlueprint_22 In)
  (connect Subgraph_20 BiomeB ExecuteBlueprint_24 In)
  (connect Subgraph_20 BiomeC ExecuteBlueprint_26 In)
  (connect Subgraph_20 BiomeD ExecuteBlueprint_28 In)
  (connect Subgraph_20 BiomeE ExecuteBlueprint_30 In)
  (connect Subgraph_20 BiomeF ExecuteBlueprint_32 In)
  (connect Subgraph_20 BiomeG ExecuteBlueprint_35 In)
  (connect Branch_17 Output%20B FilterDataByIndex_0 In)
  (connect ExecuteBlueprint_2 Out FilterDataByIndex_0 In)
  (connect Branch_23 Output%20B Gather_31 In)
  (connect ExecuteBlueprint_12 Out Gather_31 In)
  (connect ExecuteBlueprint_14 Out Gather_33 In)
  (connect ExecuteBlueprint_22 Out Gather_33 In)
  (connect ExecuteBlueprint_24 Out Gather_33 In)
  (connect ExecuteBlueprint_26 Out Gather_33 In)
  (connect ExecuteBlueprint_28 Out Gather_33 In)
  (connect ExecuteBlueprint_30 Out Gather_33 In)
  (connect ExecuteBlueprint_32 Out Gather_33 In)
  (connect ExecuteBlueprint_35 Out Gather_33 In)
  (connect Branch_3 Output%20B Subgraph_20 InAttributes)
  (output BiomeAttributes FilterDataByIndex_0.OutsideFilter)
  (output BiomeAttributes Gather_31)
  (output BiomeAttributes Gather_33))
