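; PCG graph that generates mesh grass points for an actor, using exposed sampling, filter and transform parameters.
; An Inputs subgraph takes the actor point and attributes, then supplies actor, biome, global and volume attributes to later stages.
; Subgraphs chain biome determination, actor filtering, mesh sampling, local filtering, grass assignment, point processing, world filtering and metadata processing into the MeshPoints output.
; Validity-check branches gate key stages, and a side branch feeds debug partitioning.
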(graph PCG_MeshGrass_Generator
  (params (WorldSpaceTransform :type Bool) (LocationOffsetZ :type Double) (RotationOffset :type Rotator :default (struct :Yaw 1.8e+02)) (RandomizeRotationOffset :type Bool :default true) (ForceSurfaceOrientation :type Bool) (ScaleMinMax :type Vector2 :default (1.0 1.0)) (SourceMeshIndex :type Int) (FilterTopologyInWorldSpace :type Bool) (EnableMeshOverride :type Bool) (GrassMesh :type (SoftObject /Script/Engine.StaticMesh)) (MaterialOverride :type (Object /Script/Engine.MaterialInterface)) (SamplingRadius :type Double :default 25.0) (GrassDensity :type Double :default 1.0) (UsePruning :type Bool) (FilterNormals :type Bool :default true) (NormalOffset :type Double) (ScaleByGradient :type Bool) (GradientScaleMinMax :type Vector2 :default (0.5 1.0)) (GradientScalePower :type Double :default 1.0) (FilterHeightGradient :type Bool) (EdgeDistance :type Double :default 2.5e+02) (FilterNonSelfHits :type Bool) (NonSelfHitDistance :type Double :default 5e+01) (NonSelfTraceZOffset :type Double) (NonSelfTraceComplex :type Bool) (FilterBorders :type Bool) (BorderDistance :type Double :default 25.0) (BorderSize :type Double :default 1e+02) (FilterWorldIntersection :type Bool) (WorldIntersectionDistance :type Double :default 1e+02) (ExtendNegativeBounds :type Bool) (FilterBelowTerrain :type Bool :default true) (BelowTerrainBias :type Double :default -1.0) (FilterLocalZ :type Bool) (LocalZMin :type Double) (LocalZMax :type Double :default 999.0) (FilterVertexColor :type Bool) (VertexColorChannel :type (Enum /Game/Athena/Helios/Environments/Landscape/WorldBuildingTools/PCGGrass/EMeshGrassColorChannel.EMeshGrassColorChannel)) (VertexColorFromInstance :type Bool) (VertexColorInvert :type Bool) (VertexColorBias :type Double :default 0.5) (VertexColorDilation :type Double) (VertexColorForcePrecision :type Bool) (FilterLandscapeLayers :type Bool) (LandscapeLayers :type String) (LandscapeLayerThreshold :type Double) (LandscapeLayerDilation :type Double) (FilterMaterialID :type Bool) (MaterialIDTarget :type Int) (MaterialIDInvert :type Bool) (FilterWaterBodies :type Bool) (WaterBodyDistance :type Double) (WaterBodyInvertFilter :type Bool) (WaterBodyPrecision :type (Enum /Game/Athena/Helios/Environments/Landscape/WorldBuildingTools/PCGGrass/EMeshGrassGenericQuality.EMeshGrassGenericQuality) :default NewEnumerator1) (DetailModeOverride :type (Enum /Game/Athena/Helios/Environments/Landscape/WorldBuildingTools/PCGGrass/EMeshGrassDetailMode.EMeshGrassDetailMode)) (IgnoreCullVolumes :type Bool) (IgnoreMaxPointsPerMesh :type Bool) (MeshOverrideByLGT :type Bool) (OverrideLGT :type (Object /Script/Landscape.LandscapeGrassType)))
  (outputs (MeshPoints :type Point))
  (bind Subgraph_14 (Subgraph :Graph /Game/Athena/Helios/Environments/Landscape/WorldBuildingTools/PCGGrass/PCG_MeshGrass_Inputs.PCG_MeshGrass_Inputs))
  (bind Subgraph_0 (Subgraph :Graph /Game/Athena/Helios/Environments/Landscape/WorldBuildingTools/PCGGrass/PCG_HasValidData.PCG_HasValidData))
  (bind Branch_2 (Branch))
  (bind Subgraph_4 (Subgraph :Graph /Game/Athena/Helios/Environments/Landscape/WorldBuildingTools/PCGGrass/PCG_MeshGrass_BiomeDetermination.PCG_MeshGrass_BiomeDetermination))
  (bind Subgraph_5 (Subgraph :Graph /Game/Athena/Helios/Environments/Landscape/WorldBuildingTools/PCGGrass/PCG_MeshGrass_ActorFiltering.PCG_MeshGrass_ActorFiltering))
  (bind Subgraph_22 (Subgraph :Graph /Game/Athena/Helios/Environments/Landscape/WorldBuildingTools/PCGGrass/PCG_HasValidData.PCG_HasValidData))
  (bind Branch_23 (Branch))
  (bind CreateAttribute_0 (Create%20Constant :AttributeTypes (struct :Type SoftObjectPath :SoftObjectPathValue (struct :AssetPath (struct :PackageName /Game/Athena/Helios/Environments/Landscape/WorldBuildingTools/PCGGrass/PCG_MeshGrass_MeshSampling :AssetName PCG_MeshGrass_MeshSampling)))))
  (bind Subgraph_7 (Subgraph :Graph /Game/Athena/Helios/Environments/Landscape/WorldBuildingTools/PCGGrass/PCG_MeshGrass_MeshSampling.PCG_MeshGrass_MeshSampling))
  (bind Subgraph_9 (Subgraph :Graph /Game/Athena/Helios/Environments/Landscape/WorldBuildingTools/PCGGrass/PCG_HasValidData.PCG_HasValidData))
  (bind Branch_11 (Branch))
  (bind Subgraph_11 (Subgraph :Graph /Game/Athena/Helios/Environments/Landscape/WorldBuildingTools/PCGGrass/PCG_MeshGrass_LocalFiltering.PCG_MeshGrass_LocalFiltering))
  (bind Subgraph_24 (Subgraph :Graph /Game/Athena/Helios/Environments/Landscape/WorldBuildingTools/PCGGrass/PCG_HasValidData.PCG_HasValidData))
  (bind Branch_26 (Branch))
  (bind Subgraph_6 (Subgraph :Graph /Game/Athena/Helios/Environments/Landscape/WorldBuildingTools/PCGGrass/PCG_MeshGrass_GrassAssignment.PCG_MeshGrass_GrassAssignment))
  (bind Subgraph_13 (Subgraph :Graph /Game/Athena/Helios/Environments/Landscape/WorldBuildingTools/PCGGrass/PCG_MeshGrass_PointProcessing.PCG_MeshGrass_PointProcessing))
  (bind Subgraph_1 (Subgraph :Graph /Game/Athena/Helios/Environments/Landscape/WorldBuildingTools/PCGGrass/PCG_MeshGrass_WorldFiltering.PCG_MeshGrass_WorldFiltering))
  (bind Subgraph_8 (Subgraph :Graph /Game/Athena/Helios/Environments/Landscape/WorldBuildingTools/PCGGrass/PCG_HasValidPoints.PCG_HasValidPoints))
  (bind Branch_5 (Branch))
  (bind Subgraph_2 (Subgraph :Graph /Game/Athena/Helios/Environments/Landscape/WorldBuildingTools/PCGGrass/PCG_MeshGrass_DebugPartition.PCG_MeshGrass_DebugPartition))
  (bind Subgraph_3 (Subgraph :Graph /Game/Athena/Helios/Environments/Landscape/WorldBuildingTools/PCGGrass/PCG_MeshGrass_MetadataProcessing.PCG_MeshGrass_MetadataProcessing))
  ;[0.05 0.05 0.05] Actor Attributes
  (bind AddAttribute_0 (Add%20Attribute :AttributeTypes (struct :Type Integer64) :bCopyAllAttributes true :Attributes WorldIntersectionDistance :Attributes GrassDensity :Attributes ScaleMinMax :Attributes NormalOffset :Attributes EdgeDistance :Attributes EnableMeshOverride :Attributes SamplingRadius :Attributes FilterNonSelfHits :Attributes FilterBorders :Attributes FilterHeightGradient :Attributes FilterWorldIntersection :Attributes FilterLocalZ :Attributes LocalZMin :Attributes LocalZMax :Attributes BorderSize :Attributes BorderDistance :Attributes NonSelfHitDistance :Attributes LocationOffsetZ :Attributes ExtendNegativeBounds :Attributes NonSelfTraceZOffset :Attributes NonSelfTraceComplex :Attributes UsePruning :Attributes MaterialOverride :Attributes ScaleByGradient :Attributes GradientScaleMinMax :Attributes WorldSpaceTransform :Attributes LandscapeLayers :Attributes RotationOffset :Attributes GradientScalePower :Attributes FilterLandscapeLayers :Attributes LandscapeLayerThreshold :Attributes LandscapeLayerDilation :Attributes ForceSurfaceOrientation :Attributes FilterTopologyInWorldSpace :In GrassMesh))
  (bind GetActorData_0 (Get%20Actor%20Data :bComponentsMustOverlapSelf false :Mode GetSinglePoint :PropertyName Scale))
  (reroute Actor%20Attributes AddAttribute_0)
  (reroute Actor%20Point GetActorData_0)
  (connect Subgraph_0 IsValid Branch_2 bOutputToB)
  (connect Subgraph_14 ActorPoint Branch_2 In)
  (connect Subgraph_8 IsValid Branch_5 bOutputToB)
  (connect Subgraph_1 OutPoints Branch_5 In)
  (connect Subgraph_9 IsValid Branch_11 bOutputToB)
  (connect Subgraph_7 OutPoints Branch_11 In)
  (connect Subgraph_22 IsValid Branch_23 bOutputToB)
  (connect Subgraph_5 ActorPoint Branch_23 In)
  (connect Subgraph_24 IsValid Branch_26 bOutputToB)
  (connect Subgraph_11 OutPoints Branch_26 In)
  (connect Subgraph_14 ActorPoint Subgraph_0 In)
  (connect Subgraph_14 ActorAttributes Subgraph_1 ActorAttributes)
  (connect Subgraph_7 ActorPoint Subgraph_1 ActorPoint)
  (connect Subgraph_14 BiomeAttributes Subgraph_1 BiomeAttributes)
  (connect Subgraph_14 GlobalAttributes Subgraph_1 GlobalAttributes)
  (connect Subgraph_13 OutPoints Subgraph_1 InPoints)
  (connect Subgraph_14 VolumeAttributes Subgraph_1 VolumeAttributes)
  (connect Branch_2 Output%20B Subgraph_2 ActorPoint)
  (connect Subgraph_7 ActorPoint Subgraph_3 ActorPoint)
  (connect Branch_5 Output%20B Subgraph_3 InPoints)
  (connect Branch_2 Output%20B Subgraph_4 ActorPoint)
  (connect Subgraph_14 BiomeAttributes Subgraph_4 BiomeAttributes)
  (connect Subgraph_14 GlobalAttributes Subgraph_4 GlobalAttributes)
  (connect Subgraph_4 ActorPoint Subgraph_5 ActorPoint)
  (connect Subgraph_14 BiomeAttributes Subgraph_5 BiomeAttributes)
  (connect Subgraph_14 GlobalAttributes Subgraph_5 GlobalAttributes)
  (connect Subgraph_14 ActorAttributes Subgraph_6 ActorAttributes)
  (connect Subgraph_7 ActorPoint Subgraph_6 ActorPoint)
  (connect Subgraph_14 BiomeAttributes Subgraph_6 BiomeAttributes)
  (connect Subgraph_14 GlobalAttributes Subgraph_6 GlobalAttributes)
  (connect Branch_26 Output%20B Subgraph_6 InPoints)
  (connect Subgraph_14 VolumeAttributes Subgraph_6 VolumeAttributes)
  (connect Subgraph_14 ActorAttributes Subgraph_7 ActorAttributes)
  (connect Branch_23 Output%20B Subgraph_7 ActorPoint)
  (connect Subgraph_14 BiomeAttributes Subgraph_7 BiomeAttributes)
  (connect Subgraph_14 GlobalAttributes Subgraph_7 GlobalAttributes)
  (connect CreateAttribute_0 Out Subgraph_7 SubgraphOverride)
  (connect Subgraph_14 VolumeAttributes Subgraph_7 VolumeAttributes)
  (connect Subgraph_1 OutPoints Subgraph_8 In)
  (connect Subgraph_7 OutPoints Subgraph_9 In)
  (connect Subgraph_14 ActorAttributes Subgraph_11 ActorAttributes)
  (connect Subgraph_7 ActorPoint Subgraph_11 ActorPoint)
  (connect Subgraph_14 BiomeAttributes Subgraph_11 BiomeAttributes)
  (connect Subgraph_14 GlobalAttributes Subgraph_11 GlobalAttributes)
  (connect Branch_11 Output%20B Subgraph_11 InPoints)
  (connect Subgraph_14 VolumeAttributes Subgraph_11 VolumeAttributes)
  (connect Subgraph_14 ActorAttributes Subgraph_13 ActorAttributes)
  (connect Subgraph_7 ActorPoint Subgraph_13 ActorPoint)
  (connect Subgraph_14 BiomeAttributes Subgraph_13 BiomeAttributes)
  (connect Subgraph_14 GlobalAttributes Subgraph_13 GlobalAttributes)
  (connect Subgraph_6 OutPoints Subgraph_13 InPoints)
  (connect Subgraph_14 VolumeAttributes Subgraph_13 VolumeAttributes)
  (connect Actor%20Attributes Out Subgraph_14 ActorAttributes)
  (connect Actor%20Point Out Subgraph_14 ActorPoint)
  (connect Subgraph_5 ActorPoint Subgraph_22 In)
  (connect Subgraph_11 OutPoints Subgraph_24 In)
  (output MeshPoints Subgraph_3.OutPoints))
