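; PCG subgraph for mesh grass that reads biome settings from the BP_MeshGrass_GlobalControls actor supplied as a reference input.
; Eight Get Actor Property nodes fetch BiomeBase and BiomeA through BiomeG.
; PCGBP_RenameAttributeSet prefixes each set's attributes with its biome name, and all eight feed the BiomeAttributes output.
; A temporary builder-only diagnostic validates the reference and prints a warning when global controls are missing, without gating the output.
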(graph PCG_MeshGrass_GetBiomeAttributes
  (inputs (Exec :type Any) (GlobalControlsReference :type Param))
  (outputs (BiomeAttributes :type Param))
  (bind GetActorProperty_0 (Get%20Actor%20Property :ActorSelector (struct :ActorFilter FromInput :ActorSelection ByClass :ActorSelectionClass /Game/Athena/Helios/Environments/Landscape/WorldBuildingTools/PCGGrass/BP_MeshGrass_GlobalControls.BP_MeshGrass_GlobalControls_C) :bAlwaysRequeryActors false :bProcessAllComponents false :bSanitizeOutputAttributeName false :OutputAttributeName None :PropertyName BiomeA))
  (bind GetActorProperty_1 (Get%20Actor%20Property :ActorSelector (struct :ActorFilter FromInput :ActorSelection ByClass :ActorSelectionClass /Game/Athena/Helios/Environments/Landscape/WorldBuildingTools/PCGGrass/BP_MeshGrass_GlobalControls.BP_MeshGrass_GlobalControls_C) :bAlwaysRequeryActors false :bProcessAllComponents false :bSanitizeOutputAttributeName false :OutputAttributeName None :PropertyName BiomeBase))
  (bind GetActorProperty_2 (Get%20Actor%20Property :ActorSelector (struct :ActorFilter FromInput :ActorSelection ByClass :ActorSelectionClass /Game/Athena/Helios/Environments/Landscape/WorldBuildingTools/PCGGrass/BP_MeshGrass_GlobalControls.BP_MeshGrass_GlobalControls_C) :bAlwaysRequeryActors false :bProcessAllComponents false :bSanitizeOutputAttributeName false :OutputAttributeName None :PropertyName BiomeB))
  (bind GetActorProperty_3 (Get%20Actor%20Property :ActorSelector (struct :ActorFilter FromInput :ActorSelection ByClass :ActorSelectionClass /Game/Athena/Helios/Environments/Landscape/WorldBuildingTools/PCGGrass/BP_MeshGrass_GlobalControls.BP_MeshGrass_GlobalControls_C) :bAlwaysRequeryActors false :bProcessAllComponents false :bSanitizeOutputAttributeName false :OutputAttributeName None :PropertyName BiomeC))
  (bind GetActorProperty_4 (Get%20Actor%20Property :ActorSelector (struct :ActorFilter FromInput :ActorSelection ByClass :ActorSelectionClass /Game/Athena/Helios/Environments/Landscape/WorldBuildingTools/PCGGrass/BP_MeshGrass_GlobalControls.BP_MeshGrass_GlobalControls_C) :bAlwaysRequeryActors false :bProcessAllComponents false :bSanitizeOutputAttributeName false :OutputAttributeName None :PropertyName BiomeD))
  (bind GetActorProperty_5 (Get%20Actor%20Property :ActorSelector (struct :ActorFilter FromInput :ActorSelection ByClass :ActorSelectionClass /Game/Athena/Helios/Environments/Landscape/WorldBuildingTools/PCGGrass/BP_MeshGrass_GlobalControls.BP_MeshGrass_GlobalControls_C) :bAlwaysRequeryActors false :bProcessAllComponents false :bSanitizeOutputAttributeName false :OutputAttributeName None :PropertyName BiomeE))
  (bind GetActorProperty_6 (Get%20Actor%20Property :ActorSelector (struct :ActorFilter FromInput :ActorSelection ByClass :ActorSelectionClass /Game/Athena/Helios/Environments/Landscape/WorldBuildingTools/PCGGrass/BP_MeshGrass_GlobalControls.BP_MeshGrass_GlobalControls_C) :bAlwaysRequeryActors false :bProcessAllComponents false :bSanitizeOutputAttributeName false :OutputAttributeName None :PropertyName BiomeF))
  (bind ExecuteBlueprint_7 (Execute%20Blueprint :BlueprintElementInstance (class:/Game/Athena/Helios/Environments/Landscape/WorldBuildingTools/PCGGrass/PCGBP_RenameAttributeSet.PCGBP_RenameAttributeSet_C :Prefix "BiomeBase_" :UsePrefix true) :BlueprintElementType /Game/Athena/Helios/Environments/Landscape/WorldBuildingTools/PCGGrass/PCGBP_RenameAttributeSet.PCGBP_RenameAttributeSet_C))
  (bind ExecuteBlueprint_8 (Execute%20Blueprint :BlueprintElementInstance (class:/Game/Athena/Helios/Environments/Landscape/WorldBuildingTools/PCGGrass/PCGBP_RenameAttributeSet.PCGBP_RenameAttributeSet_C :Prefix "BiomeA_" :UsePrefix true) :BlueprintElementType /Game/Athena/Helios/Environments/Landscape/WorldBuildingTools/PCGGrass/PCGBP_RenameAttributeSet.PCGBP_RenameAttributeSet_C))
  (bind ExecuteBlueprint_9 (Execute%20Blueprint :BlueprintElementInstance (class:/Game/Athena/Helios/Environments/Landscape/WorldBuildingTools/PCGGrass/PCGBP_RenameAttributeSet.PCGBP_RenameAttributeSet_C :Prefix "BiomeB_" :UsePrefix true) :BlueprintElementType /Game/Athena/Helios/Environments/Landscape/WorldBuildingTools/PCGGrass/PCGBP_RenameAttributeSet.PCGBP_RenameAttributeSet_C))
  (bind ExecuteBlueprint_10 (Execute%20Blueprint :BlueprintElementInstance (class:/Game/Athena/Helios/Environments/Landscape/WorldBuildingTools/PCGGrass/PCGBP_RenameAttributeSet.PCGBP_RenameAttributeSet_C :Prefix "BiomeC_" :UsePrefix true) :BlueprintElementType /Game/Athena/Helios/Environments/Landscape/WorldBuildingTools/PCGGrass/PCGBP_RenameAttributeSet.PCGBP_RenameAttributeSet_C))
  (bind ExecuteBlueprint_11 (Execute%20Blueprint :BlueprintElementInstance (class:/Game/Athena/Helios/Environments/Landscape/WorldBuildingTools/PCGGrass/PCGBP_RenameAttributeSet.PCGBP_RenameAttributeSet_C :Prefix "BiomeD_" :UsePrefix true) :BlueprintElementType /Game/Athena/Helios/Environments/Landscape/WorldBuildingTools/PCGGrass/PCGBP_RenameAttributeSet.PCGBP_RenameAttributeSet_C))
  (bind ExecuteBlueprint_12 (Execute%20Blueprint :BlueprintElementInstance (class:/Game/Athena/Helios/Environments/Landscape/WorldBuildingTools/PCGGrass/PCGBP_RenameAttributeSet.PCGBP_RenameAttributeSet_C :Prefix "BiomeE_" :UsePrefix true) :BlueprintElementType /Game/Athena/Helios/Environments/Landscape/WorldBuildingTools/PCGGrass/PCGBP_RenameAttributeSet.PCGBP_RenameAttributeSet_C))
  (bind ExecuteBlueprint_13 (Execute%20Blueprint :BlueprintElementInstance (class:/Game/Athena/Helios/Environments/Landscape/WorldBuildingTools/PCGGrass/PCGBP_RenameAttributeSet.PCGBP_RenameAttributeSet_C :Prefix "BiomeF_" :UsePrefix true) :BlueprintElementType /Game/Athena/Helios/Environments/Landscape/WorldBuildingTools/PCGGrass/PCGBP_RenameAttributeSet.PCGBP_RenameAttributeSet_C))
  (bind GetActorProperty_14 (Get%20Actor%20Property :ActorSelector (struct :ActorFilter FromInput :ActorSelection ByClass :ActorSelectionClass /Game/Athena/Helios/Environments/Landscape/WorldBuildingTools/PCGGrass/BP_MeshGrass_GlobalControls.BP_MeshGrass_GlobalControls_C) :bAlwaysRequeryActors false :bProcessAllComponents false :bSanitizeOutputAttributeName false :OutputAttributeName None :PropertyName BiomeG))
  (bind ExecuteBlueprint_15 (Execute%20Blueprint :BlueprintElementInstance (class:/Game/Athena/Helios/Environments/Landscape/WorldBuildingTools/PCGGrass/PCGBP_RenameAttributeSet.PCGBP_RenameAttributeSet_C :Prefix "BiomeG_" :UsePrefix true) :BlueprintElementType /Game/Athena/Helios/Environments/Landscape/WorldBuildingTools/PCGGrass/PCGBP_RenameAttributeSet.PCGBP_RenameAttributeSet_C))
  ;[0.15 0.15 0.15 0.5] Temp diagnostic
  (bind GetExecutionContext_0 (Get%20Execution%20Context%20Info :Mode IsBuilder))
  (bind Branch_0 (Branch))
  (bind Subgraph_0 (Subgraph :Graph /Game/Athena/Helios/Environments/Landscape/WorldBuildingTools/PCGGrass/PCG_HasValidEntries.PCG_HasValidEntries))
  (bind Subgraph_1 (Subgraph :Graph /Game/Athena/Helios/Environments/Landscape/WorldBuildingTools/PCGGrass/PCG_HasValidData.PCG_HasValidData))
  (bind AttributeBooleanOp_0 (Attribute%20Boolean%20Op))
  (bind CreateConstant_0 (Create%20Constant :AttributeTypes (struct :Type SoftObjectPath)))
  (bind AttributeCompareOp_0 (Attribute%20Compare%20Op :Operation NotEqual))
  (bind AttributeBooleanOp_2 (Attribute%20Boolean%20Op))
  (bind Branch_1 (Branch))
  (bind Print%20String_0 (Print%20String :bPrefixWithComponent true :bPrefixWithOwner true :PrintString "@@@ FAILED TO FIND GLOBAL CONTROLS" :Verbosity Warning))
  (connect Subgraph_0 IsValid AttributeBooleanOp_0 InA)
  (connect Subgraph_1 IsValid AttributeBooleanOp_0 InB)
  (connect AttributeBooleanOp_0 Out AttributeBooleanOp_2 InA)
  (connect AttributeCompareOp_0 Out AttributeBooleanOp_2 InB)
  (connect DefaultInputNode GlobalControlsReference AttributeCompareOp_0 InA)
  (connect CreateConstant_0 Out AttributeCompareOp_0 InB)
  (connect GetExecutionContext_0 Out Branch_0 bOutputToB)
  (connect DefaultInputNode GlobalControlsReference Branch_0 In)
  (connect AttributeBooleanOp_2 Out Branch_1 bOutputToB)
  (connect Branch_0 Output%20B Branch_1 In)
  (connect GetActorProperty_1 Out ExecuteBlueprint_7 In)
  (connect GetActorProperty_0 Out ExecuteBlueprint_8 In)
  (connect GetActorProperty_2 Out ExecuteBlueprint_9 In)
  (connect GetActorProperty_3 Out ExecuteBlueprint_10 In)
  (connect GetActorProperty_4 Out ExecuteBlueprint_11 In)
  (connect GetActorProperty_5 Out ExecuteBlueprint_12 In)
  (connect GetActorProperty_6 Out ExecuteBlueprint_13 In)
  (connect GetActorProperty_14 Out ExecuteBlueprint_15 In)
  (connect DefaultInputNode GlobalControlsReference GetActorProperty_0 In)
  (connect DefaultInputNode GlobalControlsReference GetActorProperty_1 In)
  (connect DefaultInputNode GlobalControlsReference GetActorProperty_2 In)
  (connect DefaultInputNode GlobalControlsReference GetActorProperty_3 In)
  (connect DefaultInputNode GlobalControlsReference GetActorProperty_4 In)
  (connect DefaultInputNode GlobalControlsReference GetActorProperty_5 In)
  (connect DefaultInputNode GlobalControlsReference GetActorProperty_6 In)
  (connect DefaultInputNode GlobalControlsReference GetActorProperty_14 In)
  (connect Branch_1 Output%20A Print%20String_0 In)
  (connect DefaultInputNode GlobalControlsReference Subgraph_0 In)
  (connect DefaultInputNode GlobalControlsReference Subgraph_1 In)
  (output BiomeAttributes ExecuteBlueprint_7)
  (output BiomeAttributes ExecuteBlueprint_8)
  (output BiomeAttributes ExecuteBlueprint_9)
  (output BiomeAttributes ExecuteBlueprint_10)
  (output BiomeAttributes ExecuteBlueprint_11)
  (output BiomeAttributes ExecuteBlueprint_12)
  (output BiomeAttributes ExecuteBlueprint_13)
  (output BiomeAttributes ExecuteBlueprint_15))
