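; PCG subgraph that supplies mesh grass settings for the biome selected by BiomeIndex (Base, A–G).
; It finds BP_MeshGrass_GlobalControls and switches on the index to output that biome's MeshData as Mesh.
; VarietyData becomes Variety, with mesh and cull-distance attributes renamed.
; An AutoSyncLGT-gated branch runs PCGBP_GatherBiomeData, then reads a property from the biome's deduplicated ObjectDependencies without feeding either output.
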(graph PCG_MeshGrass_GetBiomeGrassData
  (params (BiomeIndex :type Int))
  (inputs (Exec :type Any))
  (outputs (Mesh :type Param) (Variety :type Param))
  (bind GetActorData_1 (Get%20Actor%20Data :ActorSelector (struct :ActorFilter AllWorldActors :ActorSelection ByClass :ActorSelectionClass /Game/Athena/Helios/Environments/Landscape/WorldBuildingTools/PCGGrass/BP_MeshGrass_GlobalControls.BP_MeshGrass_GlobalControls_C) :bTrackActorsOnlyWithinBounds false :Mode GetActorReference))
  (bind Switch_0 (Switch :EnumSelection (struct :Class /Game/Blueprints/Enum/2SidedEnum.2SidedEnum) :IntOptions [0 1 2 3 4 5 6 7] :IntegerSelection-param BiomeIndex))
  (bind GetActorProperty_0 (Get%20Actor%20Property :ActorSelector (struct :ActorFilter FromInput :ActorSelection ByClass :ActorSelectionClass /Game/Athena/Helios/Environments/Landscape/WorldBuildingTools/PCGGrass/BP_MeshGrass_GlobalControls.BP_MeshGrass_GlobalControls_C) :bAlwaysRequeryActors false :bProcessAllComponents false :bSanitizeOutputAttributeName false :OutputAttributeName None :PropertyName BiomeBase.VarietyData))
  (bind GetActorProperty_1 (Get%20Actor%20Property :ActorSelector (struct :ActorFilter FromInput :ActorSelection ByClass :ActorSelectionClass /Game/Athena/Helios/Environments/Landscape/WorldBuildingTools/PCGGrass/BP_MeshGrass_GlobalControls.BP_MeshGrass_GlobalControls_C) :bAlwaysRequeryActors false :bProcessAllComponents false :bSanitizeOutputAttributeName false :OutputAttributeName None :PropertyName BiomeA.VarietyData))
  (bind GetActorProperty_2 (Get%20Actor%20Property :ActorSelector (struct :ActorFilter FromInput :ActorSelection ByClass :ActorSelectionClass /Game/Athena/Helios/Environments/Landscape/WorldBuildingTools/PCGGrass/BP_MeshGrass_GlobalControls.BP_MeshGrass_GlobalControls_C) :bAlwaysRequeryActors false :bProcessAllComponents false :bSanitizeOutputAttributeName false :OutputAttributeName None :PropertyName BiomeB.VarietyData))
  (bind GetActorProperty_3 (Get%20Actor%20Property :ActorSelector (struct :ActorFilter FromInput :ActorSelection ByClass :ActorSelectionClass /Game/Athena/Helios/Environments/Landscape/WorldBuildingTools/PCGGrass/BP_MeshGrass_GlobalControls.BP_MeshGrass_GlobalControls_C) :bAlwaysRequeryActors false :bProcessAllComponents false :bSanitizeOutputAttributeName false :OutputAttributeName None :PropertyName BiomeC.VarietyData))
  (bind GetActorProperty_4 (Get%20Actor%20Property :ActorSelector (struct :ActorFilter FromInput :ActorSelection ByClass :ActorSelectionClass /Game/Athena/Helios/Environments/Landscape/WorldBuildingTools/PCGGrass/BP_MeshGrass_GlobalControls.BP_MeshGrass_GlobalControls_C) :bAlwaysRequeryActors false :bProcessAllComponents false :bSanitizeOutputAttributeName false :OutputAttributeName None :PropertyName BiomeD.VarietyData))
  (bind GetActorProperty_5 (Get%20Actor%20Property :ActorSelector (struct :ActorFilter FromInput :ActorSelection ByClass :ActorSelectionClass /Game/Athena/Helios/Environments/Landscape/WorldBuildingTools/PCGGrass/BP_MeshGrass_GlobalControls.BP_MeshGrass_GlobalControls_C) :bAlwaysRequeryActors false :bProcessAllComponents false :bSanitizeOutputAttributeName false :OutputAttributeName None :PropertyName BiomeE.VarietyData))
  (bind GetActorProperty_6 (Get%20Actor%20Property :ActorSelector (struct :ActorFilter FromInput :ActorSelection ByClass :ActorSelectionClass /Game/Athena/Helios/Environments/Landscape/WorldBuildingTools/PCGGrass/BP_MeshGrass_GlobalControls.BP_MeshGrass_GlobalControls_C) :bAlwaysRequeryActors false :bProcessAllComponents false :bSanitizeOutputAttributeName false :OutputAttributeName None :PropertyName BiomeF.VarietyData))
  (bind GetActorProperty_10 (Get%20Actor%20Property :ActorSelector (struct :ActorFilter FromInput :ActorSelection ByClass :ActorSelectionClass /Game/Athena/Helios/Environments/Landscape/WorldBuildingTools/PCGGrass/BP_MeshGrass_GlobalControls.BP_MeshGrass_GlobalControls_C) :bAlwaysRequeryActors false :bProcessAllComponents false :bSanitizeOutputAttributeName false :OutputAttributeName None :PropertyName BiomeG.VarietyData))
  (bind Gather_7 (Gather))
  (bind AddAttribute_9 (Add%20Attribute :AttributeTypes (struct :Type SoftObjectPath) :OutputTarget Mesh))
  (bind AttributeRename_8 (Attribute%20Rename :AttributeToRename GrassMesh :NewAttributeName Mesh))
  (bind GetActorProperty_11 (Get%20Actor%20Property :ActorSelector (struct :ActorFilter FromInput :ActorSelection ByClass :ActorSelectionClass /Game/Athena/Helios/Environments/Landscape/WorldBuildingTools/PCGGrass/BP_MeshGrass_GlobalControls.BP_MeshGrass_GlobalControls_C) :bAlwaysRequeryActors false :bProcessAllComponents false :bSanitizeOutputAttributeName false :OutputAttributeName None :PropertyName BiomeBase.MeshData))
  (bind GetActorProperty_12 (Get%20Actor%20Property :ActorSelector (struct :ActorFilter FromInput :ActorSelection ByClass :ActorSelectionClass /Game/Athena/Helios/Environments/Landscape/WorldBuildingTools/PCGGrass/BP_MeshGrass_GlobalControls.BP_MeshGrass_GlobalControls_C) :bAlwaysRequeryActors false :bProcessAllComponents false :bSanitizeOutputAttributeName false :OutputAttributeName None :PropertyName BiomeA.MeshData))
  (bind GetActorProperty_13 (Get%20Actor%20Property :ActorSelector (struct :ActorFilter FromInput :ActorSelection ByClass :ActorSelectionClass /Game/Athena/Helios/Environments/Landscape/WorldBuildingTools/PCGGrass/BP_MeshGrass_GlobalControls.BP_MeshGrass_GlobalControls_C) :bAlwaysRequeryActors false :bProcessAllComponents false :bSanitizeOutputAttributeName false :OutputAttributeName None :PropertyName BiomeB.MeshData))
  (bind GetActorProperty_14 (Get%20Actor%20Property :ActorSelector (struct :ActorFilter FromInput :ActorSelection ByClass :ActorSelectionClass /Game/Athena/Helios/Environments/Landscape/WorldBuildingTools/PCGGrass/BP_MeshGrass_GlobalControls.BP_MeshGrass_GlobalControls_C) :bAlwaysRequeryActors false :bProcessAllComponents false :bSanitizeOutputAttributeName false :OutputAttributeName None :PropertyName BiomeC.MeshData))
  (bind GetActorProperty_15 (Get%20Actor%20Property :ActorSelector (struct :ActorFilter FromInput :ActorSelection ByClass :ActorSelectionClass /Game/Athena/Helios/Environments/Landscape/WorldBuildingTools/PCGGrass/BP_MeshGrass_GlobalControls.BP_MeshGrass_GlobalControls_C) :bAlwaysRequeryActors false :bProcessAllComponents false :bSanitizeOutputAttributeName false :OutputAttributeName None :PropertyName BiomeD.MeshData))
  (bind GetActorProperty_16 (Get%20Actor%20Property :ActorSelector (struct :ActorFilter FromInput :ActorSelection ByClass :ActorSelectionClass /Game/Athena/Helios/Environments/Landscape/WorldBuildingTools/PCGGrass/BP_MeshGrass_GlobalControls.BP_MeshGrass_GlobalControls_C) :bAlwaysRequeryActors false :bProcessAllComponents false :bSanitizeOutputAttributeName false :OutputAttributeName None :PropertyName BiomeE.MeshData))
  (bind GetActorProperty_17 (Get%20Actor%20Property :ActorSelector (struct :ActorFilter FromInput :ActorSelection ByClass :ActorSelectionClass /Game/Athena/Helios/Environments/Landscape/WorldBuildingTools/PCGGrass/BP_MeshGrass_GlobalControls.BP_MeshGrass_GlobalControls_C) :bAlwaysRequeryActors false :bProcessAllComponents false :bSanitizeOutputAttributeName false :OutputAttributeName None :PropertyName BiomeF.MeshData))
  (bind GetActorProperty_18 (Get%20Actor%20Property :ActorSelector (struct :ActorFilter FromInput :ActorSelection ByClass :ActorSelectionClass /Game/Athena/Helios/Environments/Landscape/WorldBuildingTools/PCGGrass/BP_MeshGrass_GlobalControls.BP_MeshGrass_GlobalControls_C) :bAlwaysRequeryActors false :bProcessAllComponents false :bSanitizeOutputAttributeName false :OutputAttributeName None :PropertyName BiomeG.MeshData))
  (bind Gather_19 (Gather))
  (bind AttributeRename_20 (Attribute%20Rename :AttributeToRename StartCullDistance :NewAttributeName InstanceStartCullDistance))
  (bind AttributeRename_21 (Attribute%20Rename :AttributeToRename EndCullDistance :NewAttributeName InstanceEndCullDistance))
  (bind GetActorProperty_25 (Get%20Actor%20Property :ActorSelector (struct :ActorFilter FromInput :ActorSelection ByClass :ActorSelectionClass /Game/Athena/Helios/Environments/Landscape/WorldBuildingTools/PCGGrass/BP_MeshGrass_GlobalControls.BP_MeshGrass_GlobalControls_C) :bAlwaysRequeryActors false :bProcessAllComponents false :bSanitizeOutputAttributeName false :OutputAttributeName None :PropertyName AutoSyncLGT))
  (bind Branch_0 (Branch))
  (bind ExecuteBlueprint_1 (Execute%20Blueprint :BlueprintElementInstance (class:/Game/Athena/Helios/Environments/Landscape/WorldBuildingTools/PCGGrass/PCGBP_GatherBiomeData.PCGBP_GatherBiomeData_C) :BlueprintElementType /Game/Athena/Helios/Environments/Landscape/WorldBuildingTools/PCGGrass/PCGBP_GatherBiomeData.PCGBP_GatherBiomeData_C :BiomeIndex BiomeIndex))
  (bind Switch_1 (Switch :EnumSelection (struct :Class /Game/Blueprints/Enum/2SidedEnum.2SidedEnum) :IntOptions [0 1 2 3 4 5 6 7] :IntegerSelection-param BiomeIndex))
  (bind GetActorProperty_7 (Get%20Actor%20Property :ActorSelector (struct :ActorFilter FromInput :ActorSelection ByClass :ActorSelectionClass /Game/Athena/Helios/Environments/Landscape/WorldBuildingTools/PCGGrass/BP_MeshGrass_GlobalControls.BP_MeshGrass_GlobalControls_C) :bAlwaysRequeryActors false :bProcessAllComponents false :bSanitizeOutputAttributeName false :OutputAttributeName None :PropertyName BiomeBase.ObjectDependencies))
  (bind GetActorProperty_9 (Get%20Actor%20Property :ActorSelector (struct :ActorFilter FromInput :ActorSelection ByClass :ActorSelectionClass /Game/Athena/Helios/Environments/Landscape/WorldBuildingTools/PCGGrass/BP_MeshGrass_GlobalControls.BP_MeshGrass_GlobalControls_C) :bAlwaysRequeryActors false :bProcessAllComponents false :bSanitizeOutputAttributeName false :OutputAttributeName None :PropertyName BiomeA.ObjectDependencies))
  (bind GetActorProperty_19 (Get%20Actor%20Property :ActorSelector (struct :ActorFilter FromInput :ActorSelection ByClass :ActorSelectionClass /Game/Athena/Helios/Environments/Landscape/WorldBuildingTools/PCGGrass/BP_MeshGrass_GlobalControls.BP_MeshGrass_GlobalControls_C) :bAlwaysRequeryActors false :bProcessAllComponents false :bSanitizeOutputAttributeName false :OutputAttributeName None :PropertyName BiomeB.ObjectDependencies))
  (bind GetActorProperty_20 (Get%20Actor%20Property :ActorSelector (struct :ActorFilter FromInput :ActorSelection ByClass :ActorSelectionClass /Game/Athena/Helios/Environments/Landscape/WorldBuildingTools/PCGGrass/BP_MeshGrass_GlobalControls.BP_MeshGrass_GlobalControls_C) :bAlwaysRequeryActors false :bProcessAllComponents false :bSanitizeOutputAttributeName false :OutputAttributeName None :PropertyName BiomeC.ObjectDependencies))
  (bind GetActorProperty_21 (Get%20Actor%20Property :ActorSelector (struct :ActorFilter FromInput :ActorSelection ByClass :ActorSelectionClass /Game/Athena/Helios/Environments/Landscape/WorldBuildingTools/PCGGrass/BP_MeshGrass_GlobalControls.BP_MeshGrass_GlobalControls_C) :bAlwaysRequeryActors false :bProcessAllComponents false :bSanitizeOutputAttributeName false :OutputAttributeName None :PropertyName BiomeD.ObjectDependencies))
  (bind GetActorProperty_22 (Get%20Actor%20Property :ActorSelector (struct :ActorFilter FromInput :ActorSelection ByClass :ActorSelectionClass /Game/Athena/Helios/Environments/Landscape/WorldBuildingTools/PCGGrass/BP_MeshGrass_GlobalControls.BP_MeshGrass_GlobalControls_C) :bAlwaysRequeryActors false :bProcessAllComponents false :bSanitizeOutputAttributeName false :OutputAttributeName None :PropertyName BiomeE.ObjectDependencies))
  (bind GetActorProperty_23 (Get%20Actor%20Property :ActorSelector (struct :ActorFilter FromInput :ActorSelection ByClass :ActorSelectionClass /Game/Athena/Helios/Environments/Landscape/WorldBuildingTools/PCGGrass/BP_MeshGrass_GlobalControls.BP_MeshGrass_GlobalControls_C) :bAlwaysRequeryActors false :bProcessAllComponents false :bSanitizeOutputAttributeName false :OutputAttributeName None :PropertyName BiomeF.ObjectDependencies))
  (bind GetActorProperty_24 (Get%20Actor%20Property :ActorSelector (struct :ActorFilter FromInput :ActorSelection ByClass :ActorSelectionClass /Game/Athena/Helios/Environments/Landscape/WorldBuildingTools/PCGGrass/BP_MeshGrass_GlobalControls.BP_MeshGrass_GlobalControls_C) :bAlwaysRequeryActors false :bProcessAllComponents false :bSanitizeOutputAttributeName false :OutputAttributeName None :PropertyName BiomeG.ObjectDependencies))
  (bind MergeAttributes_0 (Merge%20Attributes))
  (bind AttributeRemoveDuplicates_0 (Attribute%20Remove%20Duplicates))
  (bind GetPropertyFromObjectPath_0 (Get%20Property%20From%20Object%20Path :PropertyName bEnableDensityScaling))
  (connect Gather_7 Out AddAttribute_9 In)
  (connect MergeAttributes_0 Out AttributeRemoveDuplicates_0 In)
  (connect AddAttribute_9 Out AttributeRename_8 In)
  (connect AttributeRename_8 Out AttributeRename_20 In)
  (connect AttributeRename_20 Out AttributeRename_21 In)
  (connect GetActorProperty_25 Out Branch_0 bOutputToB)
  (connect GetActorData_1 Out Branch_0 In)
  (connect Branch_0 Output%20B ExecuteBlueprint_1 In)
  (connect Branch_0 Output%20B ExecuteBlueprint_1 TargetActor)
  (connect GetActorProperty_0 Out Gather_7 In)
  (connect GetActorProperty_1 Out Gather_7 In)
  (connect GetActorProperty_2 Out Gather_7 In)
  (connect GetActorProperty_3 Out Gather_7 In)
  (connect GetActorProperty_4 Out Gather_7 In)
  (connect GetActorProperty_5 Out Gather_7 In)
  (connect GetActorProperty_6 Out Gather_7 In)
  (connect GetActorProperty_10 Out Gather_7 In)
  (connect GetActorProperty_11 Out Gather_19 In)
  (connect GetActorProperty_12 Out Gather_19 In)
  (connect GetActorProperty_13 Out Gather_19 In)
  (connect GetActorProperty_14 Out Gather_19 In)
  (connect GetActorProperty_15 Out Gather_19 In)
  (connect GetActorProperty_16 Out Gather_19 In)
  (connect GetActorProperty_17 Out Gather_19 In)
  (connect GetActorProperty_18 Out Gather_19 In)
  (connect Switch_0 %30 GetActorProperty_0 In)
  (connect Switch_0 %31 GetActorProperty_1 In)
  (connect Switch_0 %32 GetActorProperty_2 In)
  (connect Switch_0 %33 GetActorProperty_3 In)
  (connect Switch_0 %34 GetActorProperty_4 In)
  (connect Switch_0 %35 GetActorProperty_5 In)
  (connect Switch_0 %36 GetActorProperty_6 In)
  (connect Switch_1 %30 GetActorProperty_7 In)
  (connect Switch_1 %31 GetActorProperty_9 In)
  (connect Switch_0 %37 GetActorProperty_10 In)
  (connect Switch_0 %30 GetActorProperty_11 In)
  (connect Switch_0 %31 GetActorProperty_12 In)
  (connect Switch_0 %32 GetActorProperty_13 In)
  (connect Switch_0 %33 GetActorProperty_14 In)
  (connect Switch_0 %34 GetActorProperty_15 In)
  (connect Switch_0 %35 GetActorProperty_16 In)
  (connect Switch_0 %36 GetActorProperty_17 In)
  (connect Switch_0 %37 GetActorProperty_18 In)
  (connect Switch_1 %32 GetActorProperty_19 In)
  (connect Switch_1 %33 GetActorProperty_20 In)
  (connect Switch_1 %34 GetActorProperty_21 In)
  (connect Switch_1 %35 GetActorProperty_22 In)
  (connect Switch_1 %36 GetActorProperty_23 In)
  (connect Switch_1 %37 GetActorProperty_24 In)
  (connect GetActorData_1 Out GetActorProperty_25 In)
  (connect AttributeRemoveDuplicates_0 Out GetPropertyFromObjectPath_0 In)
  (connect GetActorProperty_7 Out MergeAttributes_0 In)
  (connect GetActorProperty_9 Out MergeAttributes_0 In)
  (connect GetActorProperty_19 Out MergeAttributes_0 In)
  (connect GetActorProperty_20 Out MergeAttributes_0 In)
  (connect GetActorProperty_21 Out MergeAttributes_0 In)
  (connect GetActorProperty_22 Out MergeAttributes_0 In)
  (connect GetActorProperty_23 Out MergeAttributes_0 In)
  (connect GetActorProperty_24 Out MergeAttributes_0 In)
  (connect GetActorData_1 Out Switch_0 In)
  (connect ExecuteBlueprint_1 Out Switch_1 In)
  (output Mesh Gather_19)
  (output Variety AttributeRename_21))
