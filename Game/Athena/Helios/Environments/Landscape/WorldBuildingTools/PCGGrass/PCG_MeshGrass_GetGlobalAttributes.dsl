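; PCG subgraph that reads the mesh grass global settings from a BP_MeshGrass_GlobalControls actor into a single GlobalAttributes param for PCG grass generation.
; Get Actor Property nodes, fed by the GlobalControlsReference input, fetch the per-mesh point cap, fallback biome, biome atlas flag, and landscape-layer and material-name exclusion settings.
; An Add Attribute node with copy-all merges them onto the FallbackBiome set, renaming only MaxPointsPerMesh, which becomes MaxSamples.
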(graph PCG_MeshGrass_GetGlobalAttributes
  (inputs (Exec :type Any) (GlobalControlsReference :type Param))
  (outputs (GlobalAttributes :type Param))
  (bind GetActorProperty_0 (Get%20Actor%20Property :ActorSelector (struct :ActorFilter FromInput :ActorSelection ByClass :ActorSelectionClass /Game/Athena/Helios/Environments/Landscape/WorldBuildingTools/PCGGrass/BP_MeshGrass_GlobalControls.BP_MeshGrass_GlobalControls_C) :bAlwaysRequeryActors false :bProcessAllComponents false :bSanitizeOutputAttributeName false :OutputAttributeName MaxSamples :PropertyName MaxPointsPerMesh))
  (bind GetActorProperty_1 (Get%20Actor%20Property :ActorSelector (struct :ActorFilter FromInput :ActorSelection ByClass :ActorSelectionClass /Game/Athena/Helios/Environments/Landscape/WorldBuildingTools/PCGGrass/BP_MeshGrass_GlobalControls.BP_MeshGrass_GlobalControls_C) :bAlwaysRequeryActors false :bProcessAllComponents false :bSanitizeOutputAttributeName false :OutputAttributeName FallbackBiome :PropertyName FallbackBiome))
  (bind GetActorProperty_2 (Get%20Actor%20Property :ActorSelector (struct :ActorFilter FromInput :ActorSelection ByClass :ActorSelectionClass /Game/Athena/Helios/Environments/Landscape/WorldBuildingTools/PCGGrass/BP_MeshGrass_GlobalControls.BP_MeshGrass_GlobalControls_C) :bAlwaysRequeryActors false :bProcessAllComponents false :bSanitizeOutputAttributeName false :OutputAttributeName ExcludeByLandscapeLayers :PropertyName ExcludeByLandscapeLayers))
  (bind GetActorProperty_3 (Get%20Actor%20Property :ActorSelector (struct :ActorFilter FromInput :ActorSelection ByClass :ActorSelectionClass /Game/Athena/Helios/Environments/Landscape/WorldBuildingTools/PCGGrass/BP_MeshGrass_GlobalControls.BP_MeshGrass_GlobalControls_C) :bAlwaysRequeryActors false :bProcessAllComponents false :bSanitizeOutputAttributeName false :OutputAttributeName ExcludedLayerNames :PropertyName ExcludedLayerNames))
  (bind GetActorProperty_4 (Get%20Actor%20Property :ActorSelector (struct :ActorFilter FromInput :ActorSelection ByClass :ActorSelectionClass /Game/Athena/Helios/Environments/Landscape/WorldBuildingTools/PCGGrass/BP_MeshGrass_GlobalControls.BP_MeshGrass_GlobalControls_C) :bAlwaysRequeryActors false :bProcessAllComponents false :bSanitizeOutputAttributeName false :OutputAttributeName ExcludeByMaterialName :PropertyName ExcludeByMaterialName))
  (bind GetActorProperty_5 (Get%20Actor%20Property :ActorSelector (struct :ActorFilter FromInput :ActorSelection ByClass :ActorSelectionClass /Game/Athena/Helios/Environments/Landscape/WorldBuildingTools/PCGGrass/BP_MeshGrass_GlobalControls.BP_MeshGrass_GlobalControls_C) :bAlwaysRequeryActors false :bProcessAllComponents false :bSanitizeOutputAttributeName false :OutputAttributeName MaterialExclusionStrings :PropertyName MaterialExclusionStrings))
  (bind GetActorProperty_6 (Get%20Actor%20Property :ActorSelector (struct :ActorFilter FromInput :ActorSelection ByClass :ActorSelectionClass /Game/Athena/Helios/Environments/Landscape/WorldBuildingTools/PCGGrass/BP_MeshGrass_GlobalControls.BP_MeshGrass_GlobalControls_C) :bAlwaysRequeryActors false :bProcessAllComponents false :bSanitizeOutputAttributeName false :OutputAttributeName BiomeTextureIsAtlas :PropertyName BiomeTextureIsAtlas))
  (bind AddAttribute_6 (Add%20Attribute :bCopyAllAttributes true))
  (connect GetActorProperty_0 Out AddAttribute_6 Attributes)
  (connect GetActorProperty_2 Out AddAttribute_6 Attributes)
  (connect GetActorProperty_3 Out AddAttribute_6 Attributes)
  (connect GetActorProperty_4 Out AddAttribute_6 Attributes)
  (connect GetActorProperty_5 Out AddAttribute_6 Attributes)
  (connect GetActorProperty_6 Out AddAttribute_6 Attributes)
  (connect GetActorProperty_1 Out AddAttribute_6 In)
  (connect DefaultInputNode GlobalControlsReference GetActorProperty_0 In)
  (connect DefaultInputNode GlobalControlsReference GetActorProperty_1 In)
  (connect DefaultInputNode GlobalControlsReference GetActorProperty_2 In)
  (connect DefaultInputNode GlobalControlsReference GetActorProperty_3 In)
  (connect DefaultInputNode GlobalControlsReference GetActorProperty_4 In)
  (connect DefaultInputNode GlobalControlsReference GetActorProperty_5 In)
  (connect DefaultInputNode GlobalControlsReference GetActorProperty_6 In)
  (output GlobalAttributes AddAttribute_6))
