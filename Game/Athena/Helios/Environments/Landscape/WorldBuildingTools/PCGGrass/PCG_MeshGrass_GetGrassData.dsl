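; PCG graph that gathers mesh grass data from three sources, outputting separate mesh and variety sets for biome, volume and spline.
; Biome data comes from a subgraph gated on the points' Biome index, volume data from overlapping LocalControls actors, spline data from SplineGenerator actors.
; For volume and spline, GatherBiomeData blueprints run before variety reads, which get renamed mesh and cull-distance attributes.
; A shared subgraph appends extra attributes to all mesh outputs.
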(graph PCG_MeshGrass_GetGrassData
  (inputs (ActorPoints :type Point))
  (outputs (OutBiomeMeshData :type Param) (OutBiomeVarietyData :type Param) (OutVolumeMeshData :type Param) (OutVolumeVarietyData :type Param) (OutSplineMeshData :type Param) (OutSplineVarietyData :type Param))
  (bind Subgraph_2 (Subgraph :Graph /Game/Athena/Helios/Environments/Landscape/WorldBuildingTools/PCGGrass/PCG_HasValidData.PCG_HasValidData))
  (bind Branch_0 (Branch))
  ;[0.05 0.05 0.05] Volume
  (bind GetActorData_0 (Get%20Actor%20Data :ActorSelector (struct :ActorFilter AllWorldActors :bMustOverlapSelf true :ActorSelection ByClass :ActorSelectionClass /Game/Athena/Helios/Environments/Landscape/WorldBuildingTools/PCGGrass/BP_MeshGrass_LocalControls.BP_MeshGrass_LocalControls_C) :bAlwaysRequeryActors true :bTrackActorsOnlyWithinBounds false :Mode GetActorReference))
  (bind ExecuteBlueprint_3 (Execute%20Blueprint :BlueprintElementInstance (class:/Game/Athena/Helios/Environments/Landscape/WorldBuildingTools/PCGGrass/PCGBP_GatherBiomeData.PCGBP_GatherBiomeData_C) :BlueprintElementType /Game/Athena/Helios/Environments/Landscape/WorldBuildingTools/PCGGrass/PCGBP_GatherBiomeData.PCGBP_GatherBiomeData_C))
  (bind GetPropertyFromObjectPath_1 (Get%20Property%20From%20Object%20Path :PropertyName VolumeBiome.VarietyData))
  (bind AddAttribute_13 (Add%20Attribute :AttributeTypes (struct :Type SoftObjectPath) :OutputTarget Mesh))
  (bind AttributeRename_11 (Attribute%20Rename :AttributeToRename GrassMesh :NewAttributeName Mesh))
  (bind AttributeRename_15 (Attribute%20Rename :AttributeToRename StartCullDistance :NewAttributeName InstanceStartCullDistance))
  (bind AttributeRename_17 (Attribute%20Rename :AttributeToRename EndCullDistance :NewAttributeName InstanceEndCullDistance))
  (bind GetPropertyFromObjectPath_0 (Get%20Property%20From%20Object%20Path :PropertyName VolumeBiome.MeshData))
  (bind AddAttribute_19 (Add%20Attribute :bCopyAllAttributes true))
  (bind GetPropertyFromObjectPath_2 (Get%20Property%20From%20Object%20Path :PropertyName VolumeBiome.ObjectDependencies))
  (bind GetPropertyFromObjectPath_3 (Get%20Property%20From%20Object%20Path :PropertyName bEnableDensityScaling))
  ;[0.05 0.05 0.05] Spline
  (bind ExecuteBlueprint_1 (Execute%20Blueprint :BlueprintElementInstance (class:/Game/Athena/Helios/Environments/Landscape/WorldBuildingTools/PCGGrass/PCGBP_IsTargetClass.PCGBP_IsTargetClass_C :TargetClass /Game/Athena/Helios/Environments/Landscape/WorldBuildingTools/PCGGrass/BP_MeshGrass_SplineGenerator.BP_MeshGrass_SplineGenerator_C) :BlueprintElementType /Game/Athena/Helios/Environments/Landscape/WorldBuildingTools/PCGGrass/PCGBP_IsTargetClass.PCGBP_IsTargetClass_C))
  (bind GetActorData_28 (Get%20Actor%20Data :Mode GetActorReference))
  (bind Branch_32 (Branch))
  (bind ExecuteBlueprint_4 (Execute%20Blueprint :BlueprintElementInstance (class:/Game/Athena/Helios/Environments/Landscape/WorldBuildingTools/PCGGrass/PCGBP_GatherBiomeData.PCGBP_GatherBiomeData_C) :BlueprintElementType /Game/Athena/Helios/Environments/Landscape/WorldBuildingTools/PCGGrass/PCGBP_GatherBiomeData.PCGBP_GatherBiomeData_C))
  (bind GetPropertyFromObjectPath_34 (Get%20Property%20From%20Object%20Path :bSanitizeOutputAttributeName false :OutputAttributeName Mesh :PropertyName Biome.VarietyData))
  (bind AddAttribute_16 (Add%20Attribute :AttributeTypes (struct :Type SoftObjectPath) :OutputTarget Mesh))
  (bind AttributeRename_13 (Attribute%20Rename :AttributeToRename GrassMesh :NewAttributeName Mesh))
  (bind AttributeRename_18 (Attribute%20Rename :AttributeToRename StartCullDistance :NewAttributeName InstanceStartCullDistance))
  (bind AttributeRename_20 (Attribute%20Rename :AttributeToRename EndCullDistance :NewAttributeName InstanceEndCullDistance))
  (bind GetPropertyFromObjectPath_29 (Get%20Property%20From%20Object%20Path :bSanitizeOutputAttributeName false :OutputAttributeName Mesh :PropertyName Biome.MeshData))
  (bind AddAttribute_27 (Add%20Attribute :bCopyAllAttributes true))
  (bind GetPropertyFromObjectPath_4 (Get%20Property%20From%20Object%20Path :PropertyName Biome.ObjectDependencies))
  (bind GetPropertyFromObjectPath_5 (Get%20Property%20From%20Object%20Path :PropertyName bEnableDensityScaling))
  ;[0.05 0.05 0.05] Inputs
  (bind Subgraph_0 (Subgraph :Graph /Game/Athena/Helios/Environments/Landscape/WorldBuildingTools/PCGGrass/PCG_MeshGrass_ExtraMeshAttributes.PCG_MeshGrass_ExtraMeshAttributes))
  ;[0.05 0.05 0.05] Biome
  (bind CreateAttribute_8 (Create%20Constant :AttributeTypes (struct :Type Integer32)))
  (bind GetAttributeFromPointIndex_5 (Get%20Attribute%20From%20Point%20Index :bForceOutputAttributeToBeInElementsDomain true :InputSource Biome :OutputAttributeName None))
  (bind AttributeCompareOp_7 (Attribute%20Compare%20Op :Operation GreaterOrEqual))
  (bind Branch_9 (Branch))
  (bind Subgraph_1 (Subgraph :Graph /Game/Athena/Helios/Environments/Landscape/WorldBuildingTools/PCGGrass/PCG_MeshGrass_GetBiomeGrassData.PCG_MeshGrass_GetBiomeGrassData))
  (bind AddAttribute_1 (Add%20Attribute :bCopyAllAttributes true))
  (reroute Additional%20Mesh%20Attributes Subgraph_0.MeshAttributes)
  (connect Additional%20Mesh%20Attributes Out AddAttribute_1 Attributes)
  (connect Subgraph_1 Mesh AddAttribute_1 In)
  (connect GetPropertyFromObjectPath_1 Out AddAttribute_13 In)
  (connect GetPropertyFromObjectPath_34 Out AddAttribute_16 In)
  (connect Additional%20Mesh%20Attributes Out AddAttribute_19 Attributes)
  (connect GetPropertyFromObjectPath_0 Out AddAttribute_19 In)
  (connect Additional%20Mesh%20Attributes Out AddAttribute_27 Attributes)
  (connect GetPropertyFromObjectPath_29 Out AddAttribute_27 In)
  (connect GetAttributeFromPointIndex_5 Attribute AttributeCompareOp_7 InA)
  (connect CreateAttribute_8 Out AttributeCompareOp_7 InB)
  (connect AddAttribute_13 Out AttributeRename_11 In)
  (connect AddAttribute_16 Out AttributeRename_13 In)
  (connect AttributeRename_11 Out AttributeRename_15 In)
  (connect AttributeRename_15 Out AttributeRename_17 In)
  (connect AttributeRename_13 Out AttributeRename_18 In)
  (connect AttributeRename_18 Out AttributeRename_20 In)
  (connect Subgraph_2 IsValid Branch_0 bOutputToB)
  (connect GetActorData_0 Out Branch_0 In)
  (connect AttributeCompareOp_7 Out Branch_9 bOutputToB)
  (connect AttributeCompareOp_7 Out Branch_9 In)
  (connect ExecuteBlueprint_1 Out Branch_32 bOutputToB)
  (connect GetActorData_28 Out Branch_32 In)
  (connect Branch_0 Output%20B ExecuteBlueprint_3 In)
  (connect GetActorData_0 Out ExecuteBlueprint_3 TargetActor)
  (connect Branch_32 Output%20B ExecuteBlueprint_4 In)
  (connect DefaultInputNode ActorPoints GetAttributeFromPointIndex_5 In)
  (connect GetActorData_0 Out GetPropertyFromObjectPath_0 In)
  (connect ExecuteBlueprint_3 Out GetPropertyFromObjectPath_1 Execution%20Dependency)
  (connect GetActorData_0 Out GetPropertyFromObjectPath_1 In)
  (connect GetActorData_0 Out GetPropertyFromObjectPath_2 In)
  (connect GetPropertyFromObjectPath_2 Out GetPropertyFromObjectPath_3 In)
  (connect Branch_32 Output%20B GetPropertyFromObjectPath_4 In)
  (connect GetPropertyFromObjectPath_4 Out GetPropertyFromObjectPath_5 In)
  (connect Branch_32 Output%20B GetPropertyFromObjectPath_29 In)
  (connect ExecuteBlueprint_4 Out GetPropertyFromObjectPath_34 Execution%20Dependency)
  (connect Branch_32 Output%20B GetPropertyFromObjectPath_34 In)
  (connect GetAttributeFromPointIndex_5 Attribute Subgraph_1 BiomeIndex)
  (connect Branch_9 Output%20B Subgraph_1 Exec)
  (connect GetActorData_0 Out Subgraph_2 In)
  (output OutBiomeMeshData AddAttribute_1)
  (output OutBiomeVarietyData Subgraph_1.Variety)
  (output OutSplineMeshData AddAttribute_27)
  (output OutSplineVarietyData AttributeRename_20)
  (output OutVolumeMeshData AddAttribute_19)
  (output OutVolumeVarietyData AttributeRename_17))
