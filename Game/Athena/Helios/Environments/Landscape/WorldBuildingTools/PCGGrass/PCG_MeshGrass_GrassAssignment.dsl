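; PCG subgraph that assigns mesh grass varieties to scattered points, choosing among biome, volume, spline and actor-LGT data.
; Projection from LocalControls actors flags locally controlled points, then tiled spatial noise drives weighted Match And Set selection, with a loop for Dominance-filtered additive varieties.
; Texture and custom generators skip biome matching.
; Later stages apply experimental priority-sorted texture stamping, actor mesh/material overrides, and detail-mode overrides outside locally controlled points.
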(graph PCG_MeshGrass_GrassAssignment
  (params (WorldSpace :type Bool) (AllowPerActorVolumeOverride :type Bool :default true) (AllowActorOverrideLGT :type Bool))
  (inputs (InPoints :type Point) (ActorPoint :type Point) (ActorAttributes :type Param) (VolumeAttributes :type Param) (BiomeAttributes :type Param) (GlobalAttributes :type Param))
  (outputs (OutPoints :type Point))
  ;[0.05 0.05 0.05] Mesh Selection
  ;;[0.05 0.05 0.05] Actor Override
  (bind FilterAttribute_0 (Delete%20Attributes :bTokenizeOnWhiteSpace true :Operation KeepSelectedAttributes :SelectedAttributes "GrassMesh"))
  (bind AttributeBooleanOp_5 (Attribute%20Boolean%20Op :Operation Not :In MeshOverridebyLGT))
  (bind FilterAttribute_4 (Delete%20Attributes :bTokenizeOnWhiteSpace true :Operation KeepSelectedAttributes :SelectedAttributes "EnableMeshOverride"))
  (bind AttributeBooleanOp_6 (Attribute%20Boolean%20Op))
  (bind Branch_33 (Branch))
  (bind AddAttribute_34 (Add%20Attribute :OutputTarget Mesh))
  (bind DeleteAttributes_1 (Delete%20Attributes :bTokenizeOnWhiteSpace true :Operation KeepSelectedAttributes :SelectedAttributes "MaterialOverride"))
  (bind AttributeRename_2 (Attribute%20Rename :AttributeToRename MaterialOverride :NewAttributeName OverrideMaterial))
  (bind AddAttribute_5 (Add%20Attribute :bCopyAllAttributes true :OutputTarget Mesh))
  (bind Subgraph_5 (Subgraph :Graph /Game/Athena/Helios/Environments/Landscape/WorldBuildingTools/PCGGrass/PCG_MeshGrass_ExtraMeshAttributes.PCG_MeshGrass_ExtraMeshAttributes))
  (bind AddAttribute_6 (Add%20Attribute :bCopyAllAttributes true))
  (bind Gather_1 (Gather))
  (bind AttributeCast_0 (Attribute%20Cast :OutputType Integer32 :In DetailModeOverride))
  (bind CreateAttribute_4 (Create%20Constant :AttributeTypes (struct :Type Integer32)))
  (bind AttributeCompareOp_3 (Attribute%20Compare%20Op :Operation NotEqual))
  (bind AttributeFilter_2 (Filter%20Attribute%20Elements :AttributeTypes (struct :Type Boolean :BoolValue true) :bUseConstantThreshold true :Operator Equal :TargetAttribute IsLocallyControlled :ThresholdAttribute @Last))
  (bind Branch_9 (Branch))
  (bind CreateAttribute_1 (Create%20Constant :AttributeTypes (struct :Type Integer32 :DoubleValue 1.0)))
  (bind CreateAttribute_16 (Create%20Constant :AttributeTypes (struct :Type Integer32 :Int32Value 1 :DoubleValue 1.0 :IntValue 1)))
  (bind CreateAttribute_20 (Create%20Constant :AttributeTypes (struct :Type Integer32 :Int32Value 3 :DoubleValue 3.0 :IntValue 3)))
  (bind CreateConstant_0 (Create%20Constant :AttributeTypes (struct :Type Integer32 :Int32Value 2 :DoubleValue 3.0 :IntValue 3)))
  (bind Select%20%28Multi%29_21 (Select%20%28Multi%29 :EnumSelection (struct :Class /Game/Blueprints/Enum/2SidedEnum.2SidedEnum) :IntOptions [4 3 1]))
  (bind AttributeMathsOp_10 (Attribute%20Maths%20Op :InputSource1 DetailMode :Operation Set))
  (bind MergePoints_0 (Merge%20Points))
  ;;[0.05 0.05 0.05] Texture Stamping (Experimental)
  (bind DeleteAttributes_0 (Delete%20Attributes :bTokenizeOnWhiteSpace true :Operation KeepSelectedAttributes :SelectedAttributes "EnableMeshOverride"))
  (bind AttributeBooleanOp_9 (Attribute%20Boolean%20Op :Operation Not))
  (bind CreateAttribute_19 (Create%20Constant))
  (bind GetActorData_0 (Get%20Actor%20Data :ActorSelector (struct :ActorFilter AllWorldActors :bMustOverlapSelf true :ActorSelection ByClass :ActorSelectionClass /Game/Athena/Helios/Environments/Landscape/WorldBuildingTools/PCGGrass/BP_MeshGrass_LocalControls.BP_MeshGrass_LocalControls_C :bSelectMultiple true) :Mode GetSinglePoint))
  (bind GetAttributeFromPointIndex_6 (Get%20Attribute%20From%20Point%20Index :bForceOutputAttributeToBeInElementsDomain true :InputSource ActorReference :OutputAttributeName ActorReference))
  (bind GetPropertyFromObjectPath_4 (Get%20Property%20From%20Object%20Path :bSanitizeOutputAttributeName false :InputSource ActorReference :OutputAttributeName Priority :PropertyName Priority))
  (bind MergeAttributes_16 (Merge%20Attributes))
  (bind MergePoints_15 (Merge%20Points))
  (bind AddAttribute_18 (Add%20Attribute :bCopyAllAttributes true))
  (bind GetPropertyFromObjectPath_23 (Get%20Property%20From%20Object%20Path :bSanitizeOutputAttributeName false :InputSource ActorReference :OutputAttributeName EnableTextureStamping :PropertyName EnableTextureStamping))
  (bind MergeAttributes_25 (Merge%20Attributes))
  (bind AddAttribute_29 (Add%20Attribute :bCopyAllAttributes true))
  (bind GetPropertyFromObjectPath_5 (Get%20Property%20From%20Object%20Path :bSanitizeOutputAttributeName false :InputSource ActorReference :OutputAttributeName IsActive :PropertyName IsActive))
  (bind MergeAttributes_14 (Merge%20Attributes))
  (bind AddAttribute_12 (Add%20Attribute :bCopyAllAttributes true))
  (bind AttributeFilter_15 (Filter%20Attribute%20Elements :AttributeTypes (struct :Type Boolean :BoolValue true) :bUseConstantThreshold true :Operator Equal :TargetAttribute IsActive))
  (bind AttributeFilter_27 (Filter%20Attribute%20Elements :AttributeTypes (struct :Type Boolean :BoolValue true) :bUseConstantThreshold true :Operator Equal :TargetAttribute EnableTextureStamping))
  (bind GetPointsCount_16 (Get%20Element%20Count))
  (bind AttributeCompareOp_17 (Attribute%20Compare%20Op :Operation Greater))
  (bind AttributeBooleanOp_13 (Attribute%20Boolean%20Op))
  (bind Branch_5 (Branch))
  (bind AddAttribute_21 (Add%20Attribute :AttributeTypes (struct :Type Boolean) :OutputTarget TextureStamped))
  (bind AttributeFilter_0 (Filter%20Attribute%20Elements :AttributeTypes (struct :Type Boolean :BoolValue true) :bGenerateOutputDataEvenIfEmpty false :bUseConstantThreshold true :bUseSpatialQuery true :Operator Equal :TargetAttribute IsLocallyControlled :ThresholdAttribute @Last))
  (bind SortAttributes_2 (Sort%20Attributes :bUseStableSort false :SortAttributes [(struct :InputSource Priority)]))
  (bind AttributePartition_19 (Attribute%20Partition :PartitionAttributeSelectors [$Index]))
  (bind Subgraph_12 (Subgraph :Graph /Game/Athena/Helios/Environments/Landscape/WorldBuildingTools/PCGGrass/PCG_HasValidData.PCG_HasValidData))
  (bind Branch_2 (Branch))
  (bind GetAttributeFromPointIndex_91 (Get%20Attribute%20From%20Point%20Index :bForceOutputAttributeToBeInElementsDomain true :InputSource $Transform :OutputAttributeName None))
  (bind Loop_0 (Loop :Graph /Game/Athena/Helios/Environments/Landscape/WorldBuildingTools/PCGGrass/PCG_MeshGrass_TextureStamping.PCG_MeshGrass_TextureStamping :WorldSpace WorldSpace))
  (bind MergePoints_1 (Merge%20Points))
  (bind Gather_0 (Gather))
  ;;[0.05 0.05 0.05] Mesh Biome Matching
  (bind AttributeBooleanOp_2 (Attribute%20Boolean%20Op))
  (bind Subgraph_0 (Subgraph :Graph /Game/Athena/Helios/Environments/Landscape/WorldBuildingTools/PCGGrass/PCG_HasValidEntries.PCG_HasValidEntries))
  (bind AttributeBooleanOp_3 (Attribute%20Boolean%20Op))
  (bind DeleteAttributes_3 (Delete%20Attributes :bTokenizeOnWhiteSpace true :Operation KeepSelectedAttributes :SelectedAttributes "EnableMeshOverride"))
  (bind AttributeBooleanOp_14 (Attribute%20Boolean%20Op :InA MeshOverridebyLGT))
  (bind Subgraph_16 (Subgraph :Graph /Game/Athena/Helios/Environments/Landscape/WorldBuildingTools/PCGGrass/PCG_HasValidData.PCG_HasValidData))
  (bind AttributeBooleanOp_25 (Attribute%20Boolean%20Op))
  (bind AttributeBooleanOp_41 (Attribute%20Boolean%20Op :Operation Not))
  (bind AttributeBooleanOp_40 (Attribute%20Boolean%20Op))
  (bind Select_28 (Select))
  (bind Select_18 (Select))
  (bind Select_0 (Select))
  (bind AttributeFilter_1 (Filter%20Attribute%20Elements :bUseConstantThreshold true :bWarnOnDataMissingAttribute false :Operator Equal :TargetAttribute Dominance :ThresholdAttribute @Last))
  (bind AttributePartition_8 (Attribute%20Partition :PartitionAttributeSelectors [$Index]))
  (bind ExecuteBlueprint_4 (Execute%20Blueprint :BlueprintElementInstance (class:/Game/Athena/Helios/Environments/Landscape/WorldBuildingTools/PCGGrass/PCGBP_IsTargetClass.PCGBP_IsTargetClass_C :TargetClass /Game/Athena/Helios/Environments/Landscape/WorldBuildingTools/PCGGrass/BP_MeshGrass_TextureGenerator.BP_MeshGrass_TextureGenerator_C) :BlueprintElementType /Game/Athena/Helios/Environments/Landscape/WorldBuildingTools/PCGGrass/PCGBP_IsTargetClass.PCGBP_IsTargetClass_C))
  (bind ExecuteBlueprint_5 (Execute%20Blueprint :BlueprintElementInstance (class:/Game/Athena/Helios/Environments/Landscape/WorldBuildingTools/PCGGrass/PCGBP_IsTargetClass.PCGBP_IsTargetClass_C :TargetClass /Game/Athena/Helios/Environments/Landscape/WorldBuildingTools/PCGGrass/BP_MeshGrass_CustomGenerator.BP_MeshGrass_CustomGenerator_C) :BlueprintElementType /Game/Athena/Helios/Environments/Landscape/WorldBuildingTools/PCGGrass/PCGBP_IsTargetClass.PCGBP_IsTargetClass_C))
  (bind AttributeBooleanOp_0 (Attribute%20Boolean%20Op :Operation Or))
  (bind Branch_7 (Branch))
  (bind Loop_7 (Loop :Graph /Game/Athena/Helios/Environments/Landscape/WorldBuildingTools/PCGGrass/PCG_MeshGrass_ProcessAdditiveVarieties.PCG_MeshGrass_ProcessAdditiveVarieties))
  (bind Subgraph_3 (Subgraph :Graph /Game/Athena/Helios/Environments/Landscape/WorldBuildingTools/PCGGrass/PCG_HasValidData.PCG_HasValidData))
  (bind Subgraph_10 (Subgraph :Graph /Game/Athena/Helios/Environments/Landscape/WorldBuildingTools/PCGGrass/PCG_HasValidEntries.PCG_HasValidEntries))
  (bind AttributeBooleanOp_12 (Attribute%20Boolean%20Op))
  (bind Branch_4 (Branch))
  (bind MatchAndSetAttributes_35 (Match%20And%20Set%20Attributes :bUseInputWeightAttribute true :bUseWeightAttribute true :InputAttribute Biome :InputWeightAttribute Variety :MatchAttribute Biome :WeightAttribute Weight))
  (bind MergePoints_9 (Merge%20Points))
  (bind Branch_42 (Branch))
  (bind AddAttribute_45 (Add%20Attribute :AttributeTypes (struct :Type Boolean :BoolValue true) :OutputTarget IsLocallyControlled))
  (bind MatchAndSetAttributes_44 (Match%20And%20Set%20Attributes :bMatchAttributes true :bUseInputWeightAttribute true :bUseWeightAttribute true :InputAttribute IsLocallyControlled :InputWeightAttribute Variety :MatchAttribute IsLocallyControlled :WeightAttribute Weight))
  (bind Gather_29 (Gather))
  ;;[0.05 0.05 0.05] Biome Varieties
  (bind Subgraph_8 (Subgraph :Graph /Game/Athena/Helios/Environments/Landscape/WorldBuildingTools/PCGGrass/PCG_MeshGrass_FormatBiomeAttributes.PCG_MeshGrass_FormatBiomeAttributes))
  (bind ExecuteBlueprint_1 (Execute%20Blueprint :BlueprintElementInstance (class:/Game/Athena/Helios/Environments/Landscape/WorldBuildingTools/PCGGrass/PCGBP_GetDataByIndex.PCGBP_GetDataByIndex_C) :BlueprintElementType /Game/Athena/Helios/Environments/Landscape/WorldBuildingTools/PCGGrass/PCGBP_GetDataByIndex.PCGBP_GetDataByIndex_C))
  (bind DeleteAttributes_14 (Delete%20Attributes :bTokenizeOnWhiteSpace true :Operation KeepSelectedAttributes :SelectedAttributes "UseVariety"))
  (bind Subgraph_2 (Subgraph :Graph /Game/Athena/Helios/Environments/Landscape/WorldBuildingTools/PCGGrass/PCG_HasValidData.PCG_HasValidData))
  (bind Select_4 (Select))
  (bind Select_5 (Select))
  ;;[0.05 0.05 0.05] Volume Varieties
  (bind DeleteAttributes_8 (Delete%20Attributes :Operation KeepSelectedAttributes :SelectedAttributes "Use Variety Setup"))
  (bind Select_16 (Select))
  (bind DeleteAttributes_4 (Delete%20Attributes :Operation KeepSelectedAttributes :SelectedAttributes "Additive Meshes"))
  (bind MergeAttributes_36 (Merge%20Attributes))
  (bind Select_35 (Select))
  ;;[0.05 0.05 0.05] Spline Varieties
  (bind GetActorData_2 (Get%20Actor%20Data :Mode GetActorReference))
  (bind ExecuteBlueprint_2 (Execute%20Blueprint :BlueprintElementInstance (class:/Game/Athena/Helios/Environments/Landscape/WorldBuildingTools/PCGGrass/PCGBP_IsTargetClass.PCGBP_IsTargetClass_C :TargetClass /Game/Athena/Helios/Environments/Landscape/WorldBuildingTools/PCGGrass/BP_MeshGrass_SplineGenerator.BP_MeshGrass_SplineGenerator_C) :BlueprintElementType /Game/Athena/Helios/Environments/Landscape/WorldBuildingTools/PCGGrass/PCGBP_IsTargetClass.PCGBP_IsTargetClass_C))
  (bind Branch_6 (Branch))
  (bind GetPropertyFromObjectPath_3 (Get%20Property%20From%20Object%20Path :bSanitizeOutputAttributeName false :OutputAttributeName Mesh :PropertyName Biome))
  (bind DeleteAttributes_9 (Delete%20Attributes :Operation KeepSelectedAttributes :SelectedAttributes "UseVariety"))
  (bind Select_13 (Select))
  (bind DeleteAttributes_21 (Delete%20Attributes :Operation KeepSelectedAttributes :SelectedAttributes "Active"))
  ;;[0.15 0 0 0.5] TEMP DEBUG
  (bind Subgraph_6 (Subgraph :Graph /Game/Athena/Helios/Environments/Landscape/WorldBuildingTools/PCGGrass/PCG_HasValidData.PCG_HasValidData))
  (bind AttributeBooleanOp_4 (Attribute%20Boolean%20Op :Operation Not))
  (bind ExecuteBlueprint_3 (Execute%20Blueprint :BlueprintElementInstance (class:/Game/Athena/Helios/Environments/Landscape/WorldBuildingTools/PCGGrass/PCGBP_IsUnattended.PCGBP_IsUnattended_C) :BlueprintElementType /Game/Athena/Helios/Environments/Landscape/WorldBuildingTools/PCGGrass/PCGBP_IsUnattended.PCGBP_IsUnattended_C))
  (bind AttributeBooleanOp_1 (Attribute%20Boolean%20Op))
  (bind CreateConstant_1 (Create%20Constant :AttributeTypes (struct :Type String :StringValue "NoMatchData: ")))
  (bind Branch_0 (Branch))
  (bind DataNum_2 (Data%20Count))
  (bind AttributeStringOp_1 (Attribute%20String%20Op))
  (bind DataNum_0 (Data%20Count))
  (bind AttributeStringOp_3 (Attribute%20String%20Op))
  (bind AttributeStringOp_2 (Attribute%20String%20Op))
  (bind DataNum_1 (Data%20Count))
  (bind AttributeStringOp_5 (Attribute%20String%20Op))
  (bind AttributeStringOp_4 (Attribute%20String%20Op))
  (bind AttributeStringOp_7 (Attribute%20String%20Op))
  (bind AttributeStringOp_6 (Attribute%20String%20Op))
  (bind AttributeStringOp_9 (Attribute%20String%20Op))
  (bind AttributeStringOp_8 (Attribute%20String%20Op))
  (bind Subgraph_7 (Subgraph :Graph /Game/Athena/Helios/Environments/Landscape/WorldBuildingTools/PCGGrass/PCG_HasValidEntries.PCG_HasValidEntries))
  (bind AttributeStringOp_10 (Attribute%20String%20Op))
  (bind AttributeStringOp_11 (Attribute%20String%20Op))
  (bind AttributeStringOp_0 (Attribute%20String%20Op))
  (bind Print%20String_0 (Print%20String :bPrefixWithGraph false :bPrefixWithNode false :PrintToScreenColor (struct :G 184 :R 255 :A 255) :Verbosity Display))
  ;;[0.05 0.05 0.05] Actor Varieties
  (bind CreateConstant_2 (Create%20Constant :AttributeTypes (struct :Type SoftObjectPath)))
  (bind AttributeCompareOp_0 (Attribute%20Compare%20Op :Operation NotEqual :InA OverrideLGT))
  (bind Branch_3 (Branch :In OverrideLGT))
  (bind ExecuteBlueprint_6 (Execute%20Blueprint :BlueprintElementInstance (class:/Game/Athena/Helios/Environments/Landscape/WorldBuildingTools/PCGGrass/PCGBP_ProcessLGT.PCGBP_ProcessLGT_C :SourceAttributeName OverrideLGT) :BlueprintElementType /Game/Athena/Helios/Environments/Landscape/WorldBuildingTools/PCGGrass/PCGBP_ProcessLGT.PCGBP_ProcessLGT_C))
  (bind MergeAttributes_0 (Merge%20Attributes))
  (bind AttributeRename_0 (Attribute%20Rename :AttributeToRename GrassMesh :NewAttributeName Mesh))
  (bind AttributeRename_1 (Attribute%20Rename :AttributeToRename StartCullDistance :NewAttributeName InstanceStartCullDistance))
  (bind AttributeRename_3 (Attribute%20Rename :AttributeToRename EndCullDistance :NewAttributeName InstanceEndCullDistance))
  (bind GetPropertyFromObjectPath_0 (Get%20Property%20From%20Object%20Path :PropertyName bEnableDensityScaling))
  ;[0.05 0.05 0.05] LocalControls overlap check
  (bind AddAttribute_7 (Add%20Attribute :AttributeTypes (struct :Type Boolean) :OutputTarget IsLocallyControlled))
  (bind GetActorData_1 (Get%20Actor%20Data :ActorSelector (struct :ActorFilter AllWorldActors :ActorSelection ByClass :ActorSelectionClass /Game/Athena/Helios/Environments/Landscape/WorldBuildingTools/PCGGrass/BP_MeshGrass_LocalControls.BP_MeshGrass_LocalControls_C) :bComponentsMustOverlapSelf false :bMergeSinglePointData true :Mode GetSinglePoint)) ; single point
  (bind Subgraph_13 (Subgraph :Graph /Game/Athena/Helios/Environments/Landscape/WorldBuildingTools/PCGGrass/PCG_HasValidData.PCG_HasValidData))
  (bind AttributeBooleanOp_11 (Attribute%20Boolean%20Op))
  (bind Branch_10 (Branch))
  (bind CreateAttribute_18 (Create%20Constant :AttributeTypes (struct :DoubleValue 1.0) :OutputTarget Desnity))
  (bind BoundsModifier_3 (Bounds%20Modifier))
  (bind Subgraph_1 (Subgraph :Graph /Game/Athena/Helios/Environments/Landscape/WorldBuildingTools/PCGGrass/PCG_InverseTransformPoints.PCG_InverseTransformPoints))
  (bind Select_2 (Select :bUseInputB-param WorldSpace))
  (bind AddAttribute_2 (Add%20Attribute :AttributeTypes (struct :Type Boolean :BoolValue true) :OutputTarget IsLocallyControlled))
  (bind AttributeRename_4 (Attribute%20Rename :AttributeToRename ActorReference :NewAttributeName LocalControlsReference))
  (bind Projection_4 (Projection :bKeepZeroDensityPoints true :ProjectionParams (struct :bProjectPositions false :bProjectRotations false :AttributeList "IsLocallyControlled,LocalControlsReference" :AttributeMode IncludeAttributes)))
  (bind AttributeMathsOp_17 (Attribute%20Maths%20Op :InputSource1 $Density :Operation Set))
  (bind PointToAttributeSet_1 (Point%20To%20Attribute%20Set))
  (bind GetPropertyFromObjectPath_1 (Get%20Property%20From%20Object%20Path :InputSource LocalControlsReference :OutputAttributeName AttachTarget :PropertyName TargetActor))
  (bind MergeAttributes_1 (Merge%20Attributes))
  (bind AddAttribute_1 (Add%20Attribute :bCopyAllAttributes true))
  (bind DeleteAttributes_5 (Delete%20Attributes :Operation KeepSelectedAttributes :SelectedAttributes "VolumeOverride"))
  (bind GetActorProperty_1 (Get%20Actor%20Property :ActorSelector (struct :ActorFilter FromInput :bMustOverlapSelf true :ActorSelection ByClass :ActorSelectionClass /Game/Athena/Helios/Environments/Landscape/WorldBuildingTools/PCGGrass/BP_MeshGrass_LocalControls.BP_MeshGrass_LocalControls_C) :bAlwaysRequeryActors false :bProcessAllComponents false :bSanitizeOutputAttributeName false :PropertyName Enable%20Attachment%20Targeting))
  (bind AttributeBooleanOp_7 (Attribute%20Boolean%20Op))
  (bind Select_1 (Select))
  ;[0.05 0.05 0.05] Variety
  (bind Spatial%20Noise_36 (Spatial%20Noise :bTiling true :Contrast 3.4 :Iterations 3 :Transform (struct :Scale3D (8e+01 8e+01 1.0)) :ValueTarget Variety))
  (bind AttributeReduce_19 (Attribute%20Reduce :Operation Max :OutputAttributeName VarietyMax))
  (bind AttributeReduce_21 (Attribute%20Reduce :Operation Min :OutputAttributeName VarietyMax))
  (bind AttributeRemap_9 (Attribute%20Remap :bAllowInverseRange false :bClampToUnitRange true))
  ;[0.05 0.05 0.05] Inputs
  (bind DeleteAttributes_20 (Delete%20Attributes :bTokenizeOnWhiteSpace true :Operation KeepSelectedAttributes :SelectedAttributes "VolumeOverride"))
  (bind DeleteAttributes_22 (Delete%20Attributes :Operation KeepSelectedAttributes :SelectedAttributes "Enable Mesh Override"))
  (bind AttributeBooleanOp_24 (Attribute%20Boolean%20Op))
  (bind GetAttributeFromPointIndex_5 (Get%20Attribute%20From%20Point%20Index :bForceOutputAttributeToBeInElementsDomain true :InputSource Biome :OutputAttributeName Biome))
  (bind DeleteAttributes_2 (Delete%20Attributes :bTokenizeOnWhiteSpace true :Operation KeepSelectedAttributes :SelectedAttributes "VolumePerInstanceOverride"))
  (bind AttributeBooleanOp_10 (Attribute%20Boolean%20Op :Operation Not :In AllowPerActorVolumeOverride))
  (bind AttributeBooleanOp_8 (Attribute%20Boolean%20Op :Operation Or))
  (bind Branch_1 (Branch :bOutputToB-param AllowActorOverrideLGT :In AllowActorOverrideLGT))
  ;;[0.05 0.05 0.05] Grass Data
  (bind Subgraph_4 (Subgraph :Graph /Game/Athena/Helios/Environments/Landscape/WorldBuildingTools/PCGGrass/PCG_MeshGrass_GetGrassData.PCG_MeshGrass_GetGrassData))
  (reroute Actor%20Attributes ActorAttributes)
  (reroute Actor%20Biome GetAttributeFromPointIndex_5.Attribute)
  (reroute Actor%20Point ActorPoint)
  (reroute Actor%20Varieties AttributeRename_3)
  (reroute Allow%20Volume%20Mesh AttributeBooleanOp_24)
  (reroute AllowSplineMesh DeleteAttributes_21)
  (reroute Biome%20Attributes BiomeAttributes)
  (reroute Biome%20Varieties Select_5)
  (reroute BiomeMeshData Subgraph_4.OutBiomeMeshData)
  (reroute BiomeVarietyData Subgraph_4.OutBiomeVarietyData)
  (reroute Dependency%20LGT Branch_1.Output%20B)
  (reroute Global%20Attributes GlobalAttributes)
  (reroute In%20Points InPoints)
  (reroute SkippedPoints Branch_7.Output%20B)
  (reroute Spline%20Overridden%20Varieties Select_13)
  (reroute SplineMeshData Subgraph_4.OutSplineMeshData)
  (reroute SplineVarietyData Subgraph_4.OutSplineVarietyData)
  (reroute Volume%20Attributes VolumeAttributes)
  (reroute Volume%20Overridden%20Varieties Select_35)
  (reroute VolumeInstanceOverride AttributeBooleanOp_8)
  (reroute VolumeMeshData Subgraph_4.OutVolumeMeshData)
  (reroute VolumeVarietyData Subgraph_4.OutVolumeVarietyData)
  (connect MergeAttributes_1 Out AddAttribute_1 Attributes)
  (connect AttributeMathsOp_17 Out AddAttribute_1 In)
  (connect Select_2 Out AddAttribute_2 In)
  (connect AttributeRename_2 Out AddAttribute_5 Attributes)
  (connect AddAttribute_34 Out AddAttribute_5 In)
  (connect Subgraph_5 MeshAttributes AddAttribute_6 Attributes)
  (connect AddAttribute_5 Out AddAttribute_6 In)
  (connect In%20Points Out AddAttribute_7 In)
  (connect MergeAttributes_14 Out AddAttribute_12 Attributes)
  (connect AddAttribute_29 Out AddAttribute_12 In)
  (connect MergeAttributes_16 Out AddAttribute_18 Attributes)
  (connect MergePoints_15 Out AddAttribute_18 In)
  (connect Branch_5 Output%20B AddAttribute_21 In)
  (connect MergeAttributes_25 Out AddAttribute_29 Attributes)
  (connect AddAttribute_18 Out AddAttribute_29 In)
  (connect FilterAttribute_0 Out AddAttribute_34 Attributes)
  (connect Branch_33 Output%20B AddAttribute_34 In)
  (connect Volume%20Overridden%20Varieties Out AddAttribute_45 In)
  (connect ExecuteBlueprint_4 Out AttributeBooleanOp_0 InA)
  (connect ExecuteBlueprint_5 Out AttributeBooleanOp_0 InB)
  (connect AttributeBooleanOp_4 Out AttributeBooleanOp_1 InA)
  (connect ExecuteBlueprint_3 Out AttributeBooleanOp_1 InB)
  (connect VolumeInstanceOverride Out AttributeBooleanOp_2 InA)
  (connect Allow%20Volume%20Mesh Out AttributeBooleanOp_2 InB)
  (connect AttributeBooleanOp_2 Out AttributeBooleanOp_3 InA)
  (connect Subgraph_0 IsValid AttributeBooleanOp_3 InB)
  (connect Subgraph_6 IsValid AttributeBooleanOp_4 In)
  (connect FilterAttribute_4 Out AttributeBooleanOp_6 InA)
  (connect AttributeBooleanOp_5 Out AttributeBooleanOp_6 InB)
  (connect DeleteAttributes_5 Out AttributeBooleanOp_7 InA)
  (connect GetActorProperty_1 Out AttributeBooleanOp_7 InB)
  (connect DeleteAttributes_2 Out AttributeBooleanOp_8 InA)
  (connect AttributeBooleanOp_10 Out AttributeBooleanOp_8 InB)
  (connect DeleteAttributes_0 Out AttributeBooleanOp_9 In)
  (connect Subgraph_13 IsValid AttributeBooleanOp_11 InA)
  (connect VolumeInstanceOverride Out AttributeBooleanOp_11 InB)
  (connect Subgraph_3 IsValid AttributeBooleanOp_12 InA)
  (connect Subgraph_10 IsValid AttributeBooleanOp_12 InB)
  (connect AttributeBooleanOp_9 Out AttributeBooleanOp_13 InA)
  (connect AttributeCompareOp_17 Out AttributeBooleanOp_13 InB)
  (connect DeleteAttributes_3 Out AttributeBooleanOp_14 InB)
  (connect DeleteAttributes_20 Out AttributeBooleanOp_24 InA)
  (connect DeleteAttributes_22 Out AttributeBooleanOp_24 InB)
  (connect Subgraph_16 IsValid AttributeBooleanOp_25 InA)
  (connect AllowSplineMesh Out AttributeBooleanOp_25 InB)
  (connect Allow%20Volume%20Mesh Out AttributeBooleanOp_40 InA)
  (connect AttributeBooleanOp_41 Out AttributeBooleanOp_40 InB)
  (connect VolumeInstanceOverride Out AttributeBooleanOp_41 In)
  (connect CreateConstant_2 Out AttributeCompareOp_0 InB)
  (connect AttributeCast_0 Out AttributeCompareOp_3 InA)
  (connect CreateAttribute_4 Out AttributeCompareOp_3 InB)
  (connect GetPointsCount_16 Out AttributeCompareOp_17 InA)
  (connect CreateAttribute_19 Out AttributeCompareOp_17 InB)
  (connect AddAttribute_21 Out AttributeFilter_0 In)
  (connect Select_0 Out AttributeFilter_1 In)
  (connect Gather_1 Out AttributeFilter_2 In)
  (connect AddAttribute_12 Out AttributeFilter_15 In)
  (connect AttributeFilter_15 InsideFilter AttributeFilter_27 In)
  (connect Branch_9 Output%20B AttributeMathsOp_10 InA)
  (connect Select%20%28Multi%29_21 Out AttributeMathsOp_10 InB)
  (connect Projection_4 Out AttributeMathsOp_17 InA)
  (connect CreateAttribute_18 Out AttributeMathsOp_17 InB)
  (connect AttributeFilter_1 InsideFilter AttributePartition_8 In)
  (connect SortAttributes_2 Out AttributePartition_19 In)
  (connect Spatial%20Noise_36 Out AttributeReduce_19 In)
  (connect Spatial%20Noise_36 Out AttributeReduce_21 In)
  (connect Spatial%20Noise_36 Out AttributeRemap_9 In)
  (connect AttributeReduce_19 Out AttributeRemap_9 InRangeMax)
  (connect AttributeReduce_21 Out AttributeRemap_9 InRangeMin)
  (connect MergeAttributes_0 Out AttributeRename_0 In)
  (connect AttributeRename_0 Out AttributeRename_1 In)
  (connect DeleteAttributes_1 Out AttributeRename_2 In)
  (connect AttributeRename_1 Out AttributeRename_3 In)
  (connect AddAttribute_2 Out AttributeRename_4 In)
  (connect Branch_0 Output%20B AttributeStringOp_0 InA)
  (connect AttributeStringOp_11 Out AttributeStringOp_0 InB)
  (connect DataNum_2 Out AttributeStringOp_1 InB)
  (connect AttributeStringOp_1 Out AttributeStringOp_2 InA)
  (connect AttributeStringOp_3 Out AttributeStringOp_2 InB)
  (connect DataNum_0 Out AttributeStringOp_3 InB)
  (connect AttributeStringOp_2 Out AttributeStringOp_4 InA)
  (connect AttributeStringOp_5 Out AttributeStringOp_4 InB)
  (connect DataNum_1 Out AttributeStringOp_5 InB)
  (connect AttributeStringOp_4 Out AttributeStringOp_6 InA)
  (connect AttributeStringOp_7 Out AttributeStringOp_6 InB)
  (connect Actor%20Biome Out AttributeStringOp_7 InB)
  (connect AttributeStringOp_6 Out AttributeStringOp_8 InA)
  (connect AttributeStringOp_9 Out AttributeStringOp_8 InB)
  (connect DeleteAttributes_14 Out AttributeStringOp_9 InB)
  (connect Subgraph_7 IsValid AttributeStringOp_10 InB)
  (connect AttributeStringOp_8 Out AttributeStringOp_11 InA)
  (connect AttributeStringOp_10 Out AttributeStringOp_11 InB)
  (connect GetActorData_1 Out BoundsModifier_3 In)
  (connect AttributeBooleanOp_1 Out Branch_0 bOutputToB)
  (connect CreateConstant_1 Out Branch_0 In)
  (connect Subgraph_12 IsValid Branch_2 bOutputToB)
  (connect AttributeFilter_0 InsideFilter Branch_2 In)
  (connect AttributeCompareOp_0 Out Branch_3 bOutputToB)
  (connect AttributeBooleanOp_12 Out Branch_4 bOutputToB)
  (connect Branch_7 Output%20A Branch_4 In)
  (connect AttributeBooleanOp_13 Out Branch_5 bOutputToB)
  (connect Gather_29 Out Branch_5 In)
  (connect ExecuteBlueprint_2 Out Branch_6 bOutputToB)
  (connect GetActorData_2 Out Branch_6 In)
  (connect AttributeBooleanOp_0 Out Branch_7 bOutputToB)
  (connect AttributeRemap_9 Out Branch_7 In)
  (connect AttributeCompareOp_3 Out Branch_9 bOutputToB)
  (connect AttributeFilter_2 OutsideFilter Branch_9 In)
  (connect AttributeBooleanOp_11 Out Branch_10 bOutputToB)
  (connect AddAttribute_7 Out Branch_10 In)
  (connect AttributeBooleanOp_6 Out Branch_33 bOutputToB)
  (connect Gather_0 Out Branch_33 In)
  (connect AttributeBooleanOp_3 Out Branch_42 bOutputToB)
  (connect MergePoints_9 Out Branch_42 In)
  (connect SkippedPoints Out Branch_42 In)
  (connect BiomeMeshData Out DataNum_0 In)
  (connect BiomeVarietyData Out DataNum_1 In)
  (connect Biome%20Attributes Out DataNum_2 In)
  (connect Actor%20Attributes Out DeleteAttributes_0 In)
  (connect Actor%20Attributes Out DeleteAttributes_1 In)
  (connect Volume%20Attributes Out DeleteAttributes_2 In)
  (connect Actor%20Attributes Out DeleteAttributes_3 In)
  (connect Volume%20Attributes Out DeleteAttributes_4 In)
  (connect Volume%20Attributes Out DeleteAttributes_5 In)
  (connect Volume%20Attributes Out DeleteAttributes_8 In)
  (connect GetPropertyFromObjectPath_3 Out DeleteAttributes_9 In)
  (connect ExecuteBlueprint_1 OutData DeleteAttributes_14 In)
  (connect Volume%20Attributes Out DeleteAttributes_20 In)
  (connect GetPropertyFromObjectPath_3 Out DeleteAttributes_21 In)
  (connect Volume%20Attributes Out DeleteAttributes_22 In)
  (connect Actor%20Biome Out ExecuteBlueprint_1 DataIndex)
  (connect Subgraph_8 BiomeAttributes ExecuteBlueprint_1 InData)
  (connect Branch_3 Output%20B ExecuteBlueprint_6 LGTPath)
  (connect Actor%20Attributes Out FilterAttribute_0 In)
  (connect Actor%20Attributes Out FilterAttribute_4 In)
  (connect Branch_5 Output%20A Gather_0 In)
  (connect MergePoints_1 Out Gather_0 In)
  (connect AddAttribute_6 Out Gather_1 In)
  (connect Branch_33 Output%20A Gather_1 In)
  (connect Branch_42 Output%20A Gather_29 In)
  (connect MatchAndSetAttributes_44 Out Gather_29 In)
  (connect GetActorData_1 Out GetActorProperty_1 In)
  (connect Actor%20Point Out GetAttributeFromPointIndex_5 In)
  (connect GetActorData_0 Out GetAttributeFromPointIndex_6 In)
  (connect Actor%20Point Out GetAttributeFromPointIndex_91 In)
  (connect Dependency%20LGT Out GetGenericGraphParameter_0 Execution%20Dependency)
  (connect Dependency%20LGT Out GetGenericGraphParameter_1 Execution%20Dependency)
  (connect Dependency%20LGT Out GetGenericGraphParameter_3 Execution%20Dependency)
  (connect AttributeFilter_27 InsideFilter GetPointsCount_16 In)
  (connect Branch_3 Output%20B GetPropertyFromObjectPath_0 In)
  (connect PointToAttributeSet_1 Out GetPropertyFromObjectPath_1 In)
  (connect Branch_6 Output%20B GetPropertyFromObjectPath_3 In)
  (connect GetAttributeFromPointIndex_6 Attribute GetPropertyFromObjectPath_4 In)
  (connect GetAttributeFromPointIndex_6 Attribute GetPropertyFromObjectPath_5 In)
  (connect GetAttributeFromPointIndex_6 Attribute GetPropertyFromObjectPath_23 In)
  (connect AttributePartition_19 Out Loop_0 LocalControls)
  (connect Branch_2 Output%20B Loop_0 TargetPoints)
  (connect GetAttributeFromPointIndex_91 Attribute Loop_0 TargetTransform)
  (connect Branch_7 Output%20A Loop_7 SourcePoints)
  (connect AttributePartition_8 Out Loop_7 VarietyData)
  (connect Branch_4 Output%20B MatchAndSetAttributes_35 In)
  (connect AttributeFilter_1 OutsideFilter MatchAndSetAttributes_35 Match%20Data)
  (connect Branch_42 Output%20B MatchAndSetAttributes_44 In)
  (connect AddAttribute_45 Out MatchAndSetAttributes_44 Match%20Data)
  (connect ExecuteBlueprint_6 Out MergeAttributes_0 In)
  (connect GetPropertyFromObjectPath_1 Out MergeAttributes_1 In)
  (connect GetPropertyFromObjectPath_5 Out MergeAttributes_14 In)
  (connect GetPropertyFromObjectPath_4 Out MergeAttributes_16 In)
  (connect GetPropertyFromObjectPath_23 Out MergeAttributes_25 In)
  (connect Biome%20Varieties Out MergeAttributes_36 In)
  (connect Select_16 Out MergeAttributes_36 In)
  (connect AttributeFilter_2 InsideFilter MergePoints_0 In)
  (connect AttributeMathsOp_10 Out MergePoints_0 In)
  (connect Branch_9 Output%20A MergePoints_0 In)
  (connect AttributeFilter_0 OutsideFilter MergePoints_1 In)
  (connect Loop_0 TargetPoints MergePoints_1 In)
  (connect Loop_7 OutPoints MergePoints_9 In)
  (connect MatchAndSetAttributes_35 Out MergePoints_9 In)
  (connect GetActorData_0 Out MergePoints_15 In)
  (connect Projection_4 Out PointToAttributeSet_1 In)
  (connect AttributeStringOp_0 Out Print%20String_0 In)
  (connect AttributeStringOp_0 Out Print%20String_0 PrintString)
  (connect Branch_10 Output%20B Projection_4 In)
  (connect AttributeRename_4 Out Projection_4 Projection%20Target)
  (connect AttributeBooleanOp_14 Out Select_0 bUseInputB)
  (connect Select_18 Out Select_0 Input%20A)
  (connect Actor%20Varieties Out Select_0 Input%20B)
  (connect AttributeBooleanOp_7 Out Select_1 bUseInputB)
  (connect AttributeMathsOp_17 Out Select_1 Input%20A)
  (connect AddAttribute_1 Out Select_1 Input%20B)
  (connect Subgraph_1 OutPoints Select_2 Input%20A)
  (connect BoundsModifier_3 Out Select_2 Input%20B)
  (connect Subgraph_2 IsValid Select_4 bUseInputB)
  (connect DeleteAttributes_14 Out Select_4 Input%20B)
  (connect Select_4 Out Select_5 bUseInputB)
  (connect BiomeMeshData Out Select_5 Input%20A)
  (connect BiomeVarietyData Out Select_5 Input%20B)
  (connect DeleteAttributes_9 Out Select_13 bUseInputB)
  (connect SplineMeshData Out Select_13 Input%20A)
  (connect SplineVarietyData Out Select_13 Input%20B)
  (connect DeleteAttributes_8 Out Select_16 bUseInputB)
  (connect VolumeMeshData Out Select_16 Input%20A)
  (connect VolumeVarietyData Out Select_16 Input%20B)
  (connect AttributeBooleanOp_25 Out Select_18 bUseInputB)
  (connect Select_28 Out Select_18 Input%20A)
  (connect Spline%20Overridden%20Varieties Out Select_18 Input%20B)
  (connect AttributeBooleanOp_40 Out Select_28 bUseInputB)
  (connect Biome%20Varieties Out Select_28 Input%20A)
  (connect Volume%20Overridden%20Varieties Out Select_28 Input%20B)
  (connect DeleteAttributes_4 Out Select_35 bUseInputB)
  (connect Select_16 Out Select_35 Input%20A)
  (connect MergeAttributes_36 Out Select_35 Input%20B)
  (connect CreateAttribute_1 Out Select%20%28Multi%29_21 %31)
  (connect CreateConstant_0 Out Select%20%28Multi%29_21 %33)
  (connect CreateAttribute_20 Out Select%20%28Multi%29_21 %34)
  (connect CreateAttribute_16 Out Select%20%28Multi%29_21 Default)
  (connect AttributeCast_0 Out Select%20%28Multi%29_21 IntegerSelection)
  (connect AttributeFilter_27 InsideFilter SortAttributes_2 In)
  (connect Branch_10 Output%20A Spatial%20Noise_36 In)
  (connect Select_1 Out Spatial%20Noise_36 In)
  (connect Volume%20Overridden%20Varieties Out Subgraph_0 In)
  (connect BoundsModifier_3 Out Subgraph_1 SourcePoints)
  (connect Actor%20Point Out Subgraph_1 TargetPoint)
  (connect Biome%20Attributes Out Subgraph_2 In)
  (connect Branch_7 Output%20A Subgraph_3 In)
  (connect Actor%20Point Out Subgraph_4 ActorPoints)
  (connect Select_0 Out Subgraph_6 In)
  (connect BiomeVarietyData Out Subgraph_7 In)
  (connect Biome%20Attributes Out Subgraph_8 InAttributes)
  (connect AttributeFilter_1 OutsideFilter Subgraph_10 In)
  (connect AttributeFilter_0 InsideFilter Subgraph_12 In)
  (connect GetActorData_1 Out Subgraph_13 In)
  (connect Spline%20Overridden%20Varieties Out Subgraph_16 In)
  (output OutPoints MergePoints_0))
