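; PCG graph that gathers the inputs for mesh grass generation: actor point plus volume, biome and global attributes.
; An overlapping LocalControls volume supplies override properties, with built-in neutral defaults otherwise.
; Toggleable subgraphs read biome and global attributes from the GlobalControls actor.
; Points pass only with a Partitioner present and valid global data unless either toggle is off, in which case they get Biome -1.
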(graph PCG_MeshGrass_Inputs
  (params (OutputBiomeAttributes :type Bool :default true) (OutputGlobalAttributes :type Bool :default true))
  (inputs (ActorPoint :type Point) (ActorAttributes :type Param))
  (outputs (ActorPoint :type Point) (ActorAttributes :type Param) (VolumeAttributes :type Param) (BiomeAttributes :type Param) (GlobalAttributes :type Param))
  (bind Branch_4 (Branch))
  (bind Branch_22 (Branch))
  (bind AddAttribute_19 (Add%20Attribute :AttributeTypes (struct :Type Integer32 :Int32Value -1) :OutputTarget Biome))
  (bind Gather_23 (Gather))
  ;[0.05 0.05 0.05] Volume Properties
  (bind GetActorData_0 (Get%20Actor%20Data :ActorSelector (struct :ActorFilter AllWorldActors :bMustOverlapSelf true :ActorSelection ByClass :ActorSelectionClass /Game/Athena/Helios/Environments/Landscape/WorldBuildingTools/PCGGrass/BP_MeshGrass_LocalControls.BP_MeshGrass_LocalControls_C) :Mode GetSinglePoint))
  (bind GetAttributeFromPointIndex_0 (Get%20Attribute%20From%20Point%20Index :bForceOutputAttributeToBeInElementsDomain true :InputSource ActorReference))
  (bind GetActorProperty_11 (Get%20Actor%20Property :ActorSelector (struct :ActorFilter FromInput :bMustOverlapSelf true :ActorSelection ByClass :ActorSelectionClass /Game/Athena/Helios/Environments/Landscape/WorldBuildingTools/PCGGrass/BP_MeshGrass_LocalControls.BP_MeshGrass_LocalControls_C) :bAlwaysRequeryActors false :bProcessAllComponents false :bSanitizeOutputAttributeName false :OutputAttributeName VolumeOverride :PropertyName IsActive))
  (bind GetActorProperty_36 (Get%20Actor%20Property :ActorSelector (struct :ActorFilter FromInput :bMustOverlapSelf true :ActorSelection ByClass :ActorSelectionClass /Game/Athena/Helios/Environments/Landscape/WorldBuildingTools/PCGGrass/BP_MeshGrass_LocalControls.BP_MeshGrass_LocalControls_C) :bAlwaysRequeryActors false :bProcessAllComponents false :bSanitizeOutputAttributeName false :OutputAttributeName VolumePerInstanceOverride :PropertyName PerInstanceOverride))
  (bind GetActorProperty_8 (Get%20Actor%20Property :ActorSelector (struct :ActorFilter FromInput :bMustOverlapSelf true :ActorSelection ByClass :ActorSelectionClass /Game/Athena/Helios/Environments/Landscape/WorldBuildingTools/PCGGrass/BP_MeshGrass_LocalControls.BP_MeshGrass_LocalControls_C) :bAlwaysRequeryActors false :bProcessAllComponents false :bSanitizeOutputAttributeName false :OutputAttributeName None :PropertyName OverrideParameters))
  (bind AddAttribute_15 (Add%20Attribute :AttributeTypes (struct :Type Boolean) :OutputTarget IsOverlappingVolume))
  (bind AddAttribute_17 (Add%20Attribute :AttributeTypes (struct :Type Boolean :BoolValue true) :OutputTarget IsOverlappingVolume))
  (bind Projection_18 (Projection :bKeepZeroDensityPoints true))
  (bind GetAttributeFromPointIndex_19 (Get%20Attribute%20From%20Point%20Index :bForceOutputAttributeToBeInElementsDomain true :InputSource IsOverlappingVolume :OutputAttributeName IsOverlappingVolume))
  (bind Subgraph_7 (Subgraph :Graph /Game/Athena/Helios/Environments/Landscape/WorldBuildingTools/PCGGrass/PCG_HasValidData.PCG_HasValidData))
  (bind AttributeBooleanOp_22 (Attribute%20Boolean%20Op))
  (bind Select_10 (Select))
  (bind AttributeRename_9 (Attribute%20Rename :AttributeToRename MaterialOverride :NewAttributeName OverrideMaterial))
  (bind AddAttribute_11 (Add%20Attribute :AttributeTypes (struct :Type Integer64) :bCopyAllAttributes true))
  ;;[0.05 0.05 0.05] Default Values
  (bind CreateAttribute_0 (Create%20Constant :AttributeTypes (struct :Type Boolean) :OutputTarget VolumeOverride))
  (bind AddAttribute_0 (Add%20Attribute :AttributeTypes (struct :Type Boolean) :OutputTarget "Enable Mesh Override"))
  (bind AddAttribute_9 (Add%20Attribute :AttributeTypes (struct :Type Boolean :BoolValue true) :OutputTarget "Use Variety Setup"))
  (bind AddAttribute_12 (Add%20Attribute :AttributeTypes (struct :Type Boolean :BoolValue true) :OutputTarget "Additive Meshes"))
  (bind AddAttribute_1 (Add%20Attribute :AttributeTypes (struct :Type SoftObjectPath) :OutputTarget Mesh))
  (bind AddAttribute_8 (Add%20Attribute :AttributeTypes (struct :Type SoftObjectPath) :OutputTarget OverrideMaterial))
  (bind AddAttribute_10 (Add%20Attribute :AttributeTypes (struct :Type SoftObjectPath) :OutputTarget OverrideMaterial1))
  (bind AddAttribute_2 (Add%20Attribute :AttributeTypes (struct :Type Boolean) :OutputTarget VolumePerInstanceOverride))
  (bind AddAttribute_13 (Add%20Attribute :AttributeTypes (struct :DoubleValue 1.0) :OutputTarget "Sampling Radius Factor"))
  (bind AddAttribute_16 (Add%20Attribute :AttributeTypes (struct :DoubleValue 1.0) :OutputTarget "Edge Distance Factor"))
  (bind AddAttribute_3 (Add%20Attribute :AttributeTypes (struct :DoubleValue 1.0) :OutputTarget "Density Factor"))
  (bind AddAttribute_4 (Add%20Attribute :OutputTarget "Normal Offset Bias"))
  (bind AddAttribute_5 (Add%20Attribute :AttributeTypes (struct :DoubleValue 1.0) :OutputTarget "Scale Min Factor"))
  (bind AddAttribute_6 (Add%20Attribute :AttributeTypes (struct :DoubleValue 1.0) :OutputTarget "Scale Max Factor"))
  ;[0.05 0.05 0.05] Biome Properties
  (bind Branch_5 (Branch :bOutputToB-param OutputBiomeAttributes :In OutputBiomeAttributes))
  (bind GetActorData_2 (Get%20Actor%20Data :ActorSelector (struct :ActorFilter AllWorldActors :ActorSelection ByClass :ActorSelectionClass /Game/Athena/Helios/Environments/Landscape/WorldBuildingTools/PCGGrass/BP_MeshGrass_GlobalControls.BP_MeshGrass_GlobalControls_C) :bTrackActorsOnlyWithinBounds false :Mode GetActorReference))
  (bind Subgraph_3 (Subgraph :Graph /Game/Athena/Helios/Environments/Landscape/WorldBuildingTools/PCGGrass/PCG_MeshGrass_GetBiomeAttributes.PCG_MeshGrass_GetBiomeAttributes))
  ;[0.05 0.05 0.05] Global Properties
  (bind Branch_8 (Branch :bOutputToB-param OutputGlobalAttributes :In OutputGlobalAttributes))
  (bind Subgraph_6 (Subgraph :Graph /Game/Athena/Helios/Environments/Landscape/WorldBuildingTools/PCGGrass/PCG_MeshGrass_GetGlobalAttributes.PCG_MeshGrass_GetGlobalAttributes))
  (bind AttributeBooleanOp_16 (Attribute%20Boolean%20Op :InA OutputGlobalAttributes :InB OutputBiomeAttributes))
  (bind AttributeBooleanOp_17 (Attribute%20Boolean%20Op :Operation Not))
  (bind Subgraph_2 (Subgraph :Graph /Game/Athena/Helios/Environments/Landscape/WorldBuildingTools/PCGGrass/PCG_HasValidData.PCG_HasValidData))
  (bind AttributeBooleanOp_11 (Attribute%20Boolean%20Op :Operation Or))
  (bind GetActorData_1 (Get%20Actor%20Data :ActorSelector (struct :ActorFilter AllWorldActors :ActorSelection ByClass :ActorSelectionClass /Game/Athena/Helios/Environments/Landscape/WorldBuildingTools/PCGGrass/BP_MeshGrass_Partitioner.BP_MeshGrass_Partitioner_C) :Mode GetActorReference))
  (bind Subgraph_4 (Subgraph :Graph /Game/Athena/Helios/Environments/Landscape/WorldBuildingTools/PCGGrass/PCG_HasValidData.PCG_HasValidData))
  (bind AttributeBooleanOp_5 (Attribute%20Boolean%20Op))
  (reroute AllowOutput AttributeBooleanOp_5)
  (reroute NamedRerouteDeclaration_0 ActorPoint)
  (reroute NoBiome AttributeBooleanOp_17)
  (connect CreateAttribute_0 Out AddAttribute_0 In)
  (connect AddAttribute_12 Out AddAttribute_1 In)
  (connect AddAttribute_10 Out AddAttribute_2 In)
  (connect AddAttribute_16 Out AddAttribute_3 In)
  (connect AddAttribute_3 Out AddAttribute_4 In)
  (connect AddAttribute_4 Out AddAttribute_5 In)
  (connect AddAttribute_5 Out AddAttribute_6 In)
  (connect AddAttribute_1 Out AddAttribute_8 In)
  (connect AddAttribute_0 Out AddAttribute_9 In)
  (connect AddAttribute_8 Out AddAttribute_10 In)
  (connect AttributeRename_9 Out AddAttribute_11 Attributes)
  (connect GetActorProperty_36 Out AddAttribute_11 Attributes)
  (connect Select_10 Out AddAttribute_11 In)
  (connect AddAttribute_9 Out AddAttribute_12 In)
  (connect AddAttribute_2 Out AddAttribute_13 In)
  (connect NamedRerouteDeclaration_0 Out AddAttribute_15 In)
  (connect AddAttribute_13 Out AddAttribute_16 In)
  (connect GetActorData_0 Out AddAttribute_17 In)
  (connect Branch_22 Output%20B AddAttribute_19 In)
  (connect AttributeBooleanOp_11 Out AttributeBooleanOp_5 InA)
  (connect Subgraph_4 IsValid AttributeBooleanOp_5 InB)
  (connect Subgraph_2 IsValid AttributeBooleanOp_11 InA)
  (connect AttributeBooleanOp_17 Out AttributeBooleanOp_11 InB)
  (connect AttributeBooleanOp_16 Out AttributeBooleanOp_17 In)
  (connect Subgraph_7 IsValid AttributeBooleanOp_22 InA)
  (connect GetAttributeFromPointIndex_19 Attribute AttributeBooleanOp_22 InB)
  (connect GetActorProperty_8 Out AttributeRename_9 In)
  (connect AllowOutput Out Branch_4 bOutputToB)
  (connect DefaultInputNode ActorPoint Branch_4 In)
  (connect NoBiome Out Branch_22 bOutputToB)
  (connect Branch_4 Output%20B Branch_22 In)
  (connect AddAttribute_19 Out Gather_23 In)
  (connect Branch_22 Output%20A Gather_23 In)
  (connect GetAttributeFromPointIndex_0 Attribute GetActorProperty_8 In)
  (connect GetAttributeFromPointIndex_0 Attribute GetActorProperty_11 In)
  (connect GetAttributeFromPointIndex_0 Attribute GetActorProperty_36 In)
  (connect GetActorData_0 Out GetAttributeFromPointIndex_0 In)
  (connect Projection_18 Out GetAttributeFromPointIndex_19 In)
  (connect AddAttribute_15 Out Projection_18 In)
  (connect AddAttribute_17 Out Projection_18 Projection%20Target)
  (connect AttributeBooleanOp_22 Out Select_10 bUseInputB)
  (connect AddAttribute_6 Out Select_10 Input%20A)
  (connect GetActorProperty_11 Out Select_10 Input%20B)
  (connect Subgraph_6 GlobalAttributes Subgraph_2 In)
  (connect Branch_5 Output%20B Subgraph_3 Exec)
  (connect GetActorData_2 Out Subgraph_3 GlobalControlsReference)
  (connect GetActorData_1 Out Subgraph_4 In)
  (connect Branch_8 Output%20B Subgraph_6 Exec)
  (connect GetActorData_2 Out Subgraph_6 GlobalControlsReference)
  (connect GetActorProperty_11 Out Subgraph_7 In)
  (output ActorAttributes ActorAttributes)
  (output ActorPoint Gather_23)
  (output BiomeAttributes Subgraph_3.BiomeAttributes)
  (output GlobalAttributes Subgraph_6.GlobalAttributes)
  (output VolumeAttributes AddAttribute_11))
