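; PCG subgraph for mesh grass that reads an actor's Projection Setup and turns it into control values for projection.
; It outputs the setup, Ignored Meshes and Custom Target read from the actor reference.
; ProjectionTarget drives an integer multi-select for SelectLandscapeHits, which is compared to a constant for RequiresProcessing.
; FilterFromInput compares ProjectionTarget with option 3, while FilterActors independently gathers references to Landscape Streaming Proxies.
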(graph PCG_MeshGrass_InterpretProjectionSetup
  (inputs (ActorReference :type Param))
  (outputs (FilterActors :type Param) (FilterFromInput :type Param) (CustomTarget :type Param) (SelectLandscapeHits :type Param) (ProjectionSetup :type Param) (IgnoredMeshes :type Param) (RequiresProcessing :type Param))
  (bind GetPropertyFromObjectPath_0 (Get%20Property%20From%20Object%20Path :OutputAttributeName None :PropertyName Projection%20Setup))
  (bind GetPropertyFromObjectPath_1 (Get%20Property%20From%20Object%20Path :OutputAttributeName "Ignored Meshes" :PropertyName Projection%20Setup.Ignored%20Meshes))
  (bind CreateConstant_0 (Create%20Constant :AttributeTypes (struct :Type Integer32)))
  (bind CreateConstant_1 (Create%20Constant :AttributeTypes (struct :Type Integer32 :Int32Value 1)))
  (bind CreateConstant_2 (Create%20Constant :AttributeTypes (struct :Type Integer32 :Int32Value 2)))
  (bind DeleteAttributes_0 (Delete%20Attributes :Operation KeepSelectedAttributes :SelectedAttributes "ProjectionTarget"))
  (bind Select%20%28Multi%29_0 (Select%20%28Multi%29 :EnumSelection (struct :Class /Game/Blueprints/Enum/2SidedEnum.2SidedEnum) :IntOptions [0 1 2 3]))
  (bind CreateConstant_3 (Create%20Constant :AttributeTypes (struct :DoubleValue 1.0)))
  (bind AttributeCompareOp_0 (Attribute%20Compare%20Op))
  (bind CreateConstant_4 (Create%20Constant :AttributeTypes (struct :Type Integer32 :Int32Value 3 :DoubleValue 3.0)))
  (bind AttributeCompareOp_1 (Attribute%20Compare%20Op))
  (bind GetPropertyFromObjectPath_2 (Get%20Property%20From%20Object%20Path :PropertyName Projection%20Setup.Custom%20Target))
  (bind GetActorData_0 (Get%20Actor%20Data :ActorSelector (struct :ActorFilter AllWorldActors :bMustOverlapSelf true :ActorSelection ByClass :ActorSelectionClass /Script/Landscape.LandscapeStreamingProxy :bSelectMultiple true) :bMergeSinglePointData true :Mode GetActorReference))
  (connect Select%20%28Multi%29_0 Out AttributeCompareOp_0 InA)
  (connect CreateConstant_3 Out AttributeCompareOp_0 InB)
  (connect CreateConstant_4 Out AttributeCompareOp_1 InA)
  (connect DeleteAttributes_0 Out AttributeCompareOp_1 InB)
  (connect GetPropertyFromObjectPath_0 Out DeleteAttributes_0 In)
  (connect DefaultInputNode ActorReference GetPropertyFromObjectPath_0 In)
  (connect DefaultInputNode ActorReference GetPropertyFromObjectPath_1 In)
  (connect DefaultInputNode ActorReference GetPropertyFromObjectPath_2 In)
  (connect CreateConstant_1 Out Select%20%28Multi%29_0 %30)
  (connect CreateConstant_0 Out Select%20%28Multi%29_0 %31)
  (connect CreateConstant_2 Out Select%20%28Multi%29_0 %32)
  (connect CreateConstant_1 Out Select%20%28Multi%29_0 %33)
  (connect DeleteAttributes_0 Out Select%20%28Multi%29_0 IntegerSelection)
  (output CustomTarget GetPropertyFromObjectPath_2)
  (output FilterActors GetActorData_0)
  (output FilterFromInput AttributeCompareOp_1)
  (output IgnoredMeshes GetPropertyFromObjectPath_1)
  (output ProjectionSetup GetPropertyFromObjectPath_0)
  (output RequiresProcessing AttributeCompareOp_0)
  (output SelectLandscapeHits Select%20%28Multi%29_0))
