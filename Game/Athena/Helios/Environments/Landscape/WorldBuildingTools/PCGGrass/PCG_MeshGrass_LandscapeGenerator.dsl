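; PCG graph that scatters landscape grass meshes on the GPU from the landscape's grass types, gated by platform and shader feature level to exclude mobile.
; On a 128 grid, it generates grass map and height textures, then builds a max-downsampled coverage mask hierarchy.
; Custom HLSL kernels derive layer info, seed groups and generate points, which spawn meshes through a per-primitive table.
; Debug branches visualise grass maps, masks and seeds.
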(graph PCG_MeshGrass_LandscapeGenerator
  (params (LandscapeMaterial :type String :default "/Hera/Environment/Terrain/Landscape/Materials/M_Hera_Base.M_Hera_Base"))
  ;[0.15 0.15 0.15 0.5] UB
  (bind DeleteAttributes_1 (Delete%20Attributes :Operation KeepSelectedAttributes :SelectedAttributes "SumDensities,PrimitiveIndex,ScaleMin,ScaleMax,ScalingMode,ScaleByWeight,RandomRotation,AlignToSurface,CDF"))
  (bind AttributeMathsOp_1 (Attribute%20Maths%20Op :InputSource1 SumDensities :Operation Multiply)) ; Density conversion
  (bind CustomHLSL_0 (Custom%20HLSL :InputPins [(struct :Label In :AllowedTypes (struct :Ids [(struct :Struct /Script/PCG.PCGDataTypeInfoParam)]) :PinStatus Required :bAllowMultipleConnections false)] :KernelType AttributeSetProcessor :OutputPins [(struct :PropertiesGPU (struct :PinsToInititalizeFrom [In]) :Label Out :AllowedTypes (struct :Ids [(struct :Struct /Script/PCG.PCGDataTypeInfoParam)]))]))
  (bind DeleteAttributes_0 (Delete%20Attributes :Operation KeepSelectedAttributes :SelectedAttributes "LandscapeGrassType"))
  (bind GetAttributeFromIndex_1 (Get%20Attribute%20Set%20from%20Index))
  (bind MergeAttributes_0 (Merge%20Attributes))
  (bind AttributePartition_2 (Attribute%20Partition :PartitionAttributeSelectors [PrimitiveIndex]))
  (bind GetAttributeFromIndex_4 (Get%20Attribute%20Set%20from%20Index))
  (bind MergeAttributes_2 (Merge%20Attributes))
  ;;[0.15 0.15 0.15 0.5] Inputs
  (bind GetLandscapeGrassTypes_0 (Get%20Landscape%20Grass%20Types))
  (bind ProcessGrassTypes_0 (Process%20Grass%20Types :bEmitUniqueRowsWithCDF true))
  (bind GetLandscapeData_1 (Get%20Landscape%20Data :SamplingProperties (struct :bGetHeightOnly true :bGetLayerWeights false :bSampleVirtualTextures false)))
  (bind RemoveEmptyData_1 (Remove%20Empty%20Data))
  ;[0.15 0.15 0.15 0.5 10] Platform whitelist. Disabled on mobile until we have runtime grass maps
  (bind PlatformSwitch_1 (Platform%20Switch :PlatformOutputs [PS4 XB1 Switch]))
  (bind ShaderFeatureLevelSwitch_1 (Max%20Shader%20Feature%20Level%20Switch :RHIFeatureLevelOutputs [SM5 SM6]))
  ;[0.15 0.15 0.15 0.5] 128
  (bind HiGenGridSize_1 (Change%20Grid%20Size :HiGenGridSize Grid128))
  ;;[0.15 0.15 0.15 0.5] Gen LS textures
  (bind GenerateLandscapeTextures_0 (Generate%20Landscape%20Textures :bExcludeSelectedGrassTypes false :bOutputDataForEachSelectedGrassType false :bOutputTextureArray true :bOverrideFromInput true))
  ;;[0.15 0.15 0.15 0.5] Generate weight hierarchy
  (bind CustomHLSL_1 (Custom%20HLSL :bMuteUnwrittenPinDataErrors true :InputPins [(struct :Label In :AllowedTypes (struct :Ids [(struct :Struct /Script/PCG.PCGDataTypeInfoTexture2DArray)]) :bAllowMultipleData false :PinStatus Required :bAllowMultipleConnections false)] :KernelSourceOverride /Game/Athena/Helios/Environments/Landscape/WorldBuildingTools/PCGGrass/Resources/ComputeSources/PCGCS_GenerateMask.PCGCS_GenerateMask :KernelType Custom :OutputPins [(struct :PropertiesGPU (struct :PinsToInititalizeFrom [In] :TextureFormat R32f) :Label Out :AllowedTypes (struct :Ids [(struct :Struct /Script/PCG.PCGDataTypeInfoTexture2DArray)]) :bAllowMultipleData false)]))
  (bind DownsampleTexture_1 (Downsample%20Texture :Mode Max))
  ;;[0.15 0.15 0.15 0.5] Analysis
  (bind CustomHLSL_2 (Custom%20HLSL :InputPins [(struct :Label GrassMaps :AllowedTypes (struct :Ids [(struct :Struct /Script/PCG.PCGDataTypeInfoTexture2DArray)]) :bAllowMultipleData false :bAllowMultipleConnections false)] :KernelType AttributeSetGenerator :NumElements 1 :OutputPins [(struct :PropertiesGPU (struct :CreatedKernelAttributeKeys [(struct :Identifier (struct :Name NumSlices) :Type Int :Name NumSlices)]) :Label NumSlices :AllowedTypes (struct :Ids [(struct :Struct /Script/PCG.PCGDataTypeInfoParam)]) :bAllowMultipleData false)]))
  (bind CustomHLSL_3 (Custom%20HLSL :InputPins [(struct :Label Mask :AllowedTypes (struct :Ids [(struct :Struct /Script/PCG.PCGDataTypeInfoTexture2DArray)]) :bAllowMultipleData false :bAllowMultipleConnections false) (struct :Label LayerData :AllowedTypes (struct :Ids [(struct :Struct /Script/PCG.PCGDataTypeInfoParam)]) :bAllowMultipleConnections false)] :KernelSourceOverride /Game/Athena/Helios/Environments/Landscape/WorldBuildingTools/PCGGrass/Resources/ComputeSources/PCGCS_LayerInfo.PCGCS_LayerInfo :KernelType AttributeSetGenerator :OutputPins [(struct :PropertiesGPU (struct :InitializationMode Custom :PinsToInititalizeFrom [Mask] :CreatedKernelAttributeKeys [(struct :Identifier (struct :Name TargetLod) :Type Int :Name TargetLod) (struct :Identifier (struct :Name MaxCellCount) :Type Int :Name MaxCellCount)]) :Label LayerInfo :AllowedTypes (struct :Ids [(struct :Struct /Script/PCG.PCGDataTypeInfoParam)]) :bAllowMultipleData false) (struct :PropertiesGPU (struct :InitializationMode Custom :PinsToInititalizeFrom [Mask] :bInitializeToZero true :CreatedKernelAttributeKeys [(struct :Identifier (struct :Name TotalMaxNumCells) :Type Int :Name TotalMaxNumCells)]) :Label TotalMaxNumCells :AllowedTypes (struct :Ids [(struct :Struct /Script/PCG.PCGDataTypeInfoParam)]) :bAllowMultipleData false)]))
  (bind CustomHLSL_4 (Custom%20HLSL :InputPins [(struct :Label Mask :AllowedTypes (struct :Ids [(struct :Struct /Script/PCG.PCGDataTypeInfoTexture2DArray)]) :bAllowMultipleData false :bAllowMultipleConnections false) (struct :Label LayerInfo :AllowedTypes (struct :Ids [(struct :Struct /Script/PCG.PCGDataTypeInfoParam)]) :bAllowMultipleData false :PinStatus Required :bAllowMultipleConnections false)] :KernelSourceOverride /Game/Athena/Helios/Environments/Landscape/WorldBuildingTools/PCGGrass/Resources/ComputeSources/PCGCS_CreateGroupSeeds.PCGCS_CreateGroupSeeds :KernelType AttributeSetGenerator :OutputPins [(struct :PropertiesGPU (struct :PinsToInititalizeFrom [LayerInfo] :ElementCount 256 :bInitializeToZero true :bAutoInitializeOutput false :AttributeInheritanceMode None :CreatedKernelAttributeKeys [(struct :Identifier (struct :Name BoundsXYExtent_LayerIndex) :Type Float4 :Name BoundsXYExtent_LayerIndex)]) :Label OutSeeds :AllowedTypes (struct :Ids [(struct :Struct /Script/PCG.PCGDataTypeInfoParam)]) :bAllowMultipleData false) (struct :PropertiesGPU (struct :InitializationMode Custom :PinsToInititalizeFrom [LayerInfo] :bInitializeToZero true :CreatedKernelAttributeKeys [(struct :Identifier (struct :Name NumThreads) :Type Int :Name NumThreads)]) :Label NumThreads :AllowedTypes (struct :Ids [(struct :Struct /Script/PCG.PCGDataTypeInfoParam)]) :bAllowMultipleData false)]))
  ;;[0.15 0.15 0.15 0.5] Instancing
  (bind CustomHLSL_6 (Custom%20HLSL :AdditionalSources [/Game/Athena/Helios/Environments/Landscape/WorldBuildingTools/PCGGrass/Resources/ComputeSources/PCGCS_InstanceGrass_Functions.PCGCS_InstanceGrass_Functions] :InputPins [(struct :Label Seeds :AllowedTypes (struct :Ids [(struct :Struct /Script/PCG.PCGDataTypeInfoParam)]) :bAllowMultipleData false :PinStatus Required :bAllowMultipleConnections false) (struct :Label GrassMap :AllowedTypes (struct :Ids [(struct :Struct /Script/PCG.PCGDataTypeInfoTexture2DArray)]) :bAllowMultipleData false :bAllowMultipleConnections false) (struct :Label Height :AllowedTypes (struct :Ids [(struct :Struct /Script/PCG.PCGDataTypeInfoTexture2DSingleBase)]) :bAllowMultipleData false :bAllowMultipleConnections false) (struct :Label LayerData :AllowedTypes (struct :Ids [(struct :Struct /Script/PCG.PCGDataTypeInfoParam)]) :bAllowMultipleConnections false)] :KernelSourceOverride /Game/Athena/Helios/Environments/Landscape/WorldBuildingTools/PCGGrass/Resources/ComputeSources/PCGCS_InstanceGrass.PCGCS_InstanceGrass :KernelType PointGenerator :OutputPins [(struct :PropertiesGPU (struct :bAutoInitializeOutput false :CreatedKernelAttributeKeys [(struct :Identifier (struct :Name PrimitiveIndex) :Type Int :Name PrimitiveIndex)]) :Label Out :AllowedTypes (struct :Ids [(struct :Struct /Script/PCG.PCGDataTypeInfoPoint)]))]))
  (bind StaticMeshSpawner_1 (Static%20Mesh%20Spawner :bApplyMeshBoundsToPoints false :MeshSelectorParameters (class:/Script/PCG.PCGMeshSelectorPrimitiveData :MaterialOverrideAttributes [OverrideMaterial0 OverrideMaterial1] :MeshAttribute MeshAsset :PrimitiveIndexAttribute PrimitiveIndex :PrimitiveOverrideAttributes [(struct :InputSource StartCullDistance :PropertyTarget "InstanceStartCullDistance") (struct :InputSource EndCullDistance :PropertyTarget "InstanceEndCullDistance") (struct :InputSource WorldPositionOffsetDisableDistance :PropertyTarget "WorldPositionOffsetDisableDistance")] :TemplateDescriptor (struct :InstanceStartCullDistance 2500 :InstanceEndCullDistance 3000 :bCastShadow false :bAffectDistanceFieldLighting false :bReceivesDecals false :bVisibleInRayTracing false)) :MeshSelectorType /Script/PCG.PCGMeshSelectorPrimitiveData))
  ;[0.4375 0 0.2969 0.5] Group seed debug (world)
  (bind CustomHLSL_5 (Custom%20HLSL :InputPins [(struct :Label Seeds :AllowedTypes (struct :Ids [(struct :Struct /Script/PCG.PCGDataTypeInfoParam)]) :bAllowMultipleData false :PinStatus Required :bAllowMultipleConnections false) (struct :Label GrassMap :AllowedTypes (struct :Ids [(struct :Struct /Script/PCG.PCGDataTypeInfoTexture2DArray)]) :bAllowMultipleData false :bAllowMultipleConnections false) (struct :Label Height :AllowedTypes (struct :Ids [(struct :Struct /Script/PCG.PCGDataTypeInfoTexture2DSingleBase)]) :bAllowMultipleData false :bAllowMultipleConnections false) (struct :Label Layer :AllowedTypes (struct :Ids [(struct :Struct /Script/PCG.PCGDataTypeInfoParam)]) :bAllowMultipleData false :bAllowMultipleConnections false)] :KernelType Custom :OutputPins [(struct :PropertiesGPU (struct :PinsToInititalizeFrom [Seeds] :AttributeInheritanceMode None) :Label Out :AllowedTypes (struct :Ids [(struct :Struct /Script/PCG.PCGDataTypeInfoPoint)]) :bAllowMultipleData false)]))
  (bind CustomHLSL_12 (Custom%20HLSL :OutputPins [(struct :PropertiesGPU (struct :PinsToInititalizeFrom [In] :CreatedKernelAttributeKeys [(struct :Identifier (struct :Name MyDensity) :Name MyDensity)]) :Label Out :AllowedTypes (struct :Ids [(struct :Struct /Script/PCG.PCGDataTypeInfoPoint)]))]))
  (bind StaticMeshSpawner_2 (Static%20Mesh%20Spawner :InstanceDataPackerParameters (class:/Script/PCG.PCGInstanceDataPackerByAttribute :AttributeSelectors [MyDensity]) :InstanceDataPackerType /Script/PCG.PCGInstanceDataPackerByAttribute :MeshSelectorParameters (class:/Script/PCG.PCGMeshSelectorWeighted :MeshEntries [(struct :Descriptor (struct :StaticMesh /PCG/DebugObjects/PCG_CubeWireframe.PCG_CubeWireframe :OverrideMaterials [/PCG/DebugObjects/PCG_DebugMaterial.PCG_DebugMaterial]))])))
  ;[0.4375 0 0.2969 0.5] inspect grass map (viewport)
  (bind CustomHLSL_7 (Custom%20HLSL :InputPins [(struct :Label In :AllowedTypes (struct :Ids [(struct :Struct /Script/PCG.PCGDataTypeInfoTexture2DArray)]) :bAllowMultipleData false :PinStatus Required :bAllowMultipleConnections false) (struct :Label SliceIndex :AllowedTypes (struct :Ids [(struct :Struct /Script/PCG.PCGDataTypeInfoParam)]) :bAllowMultipleData false :bAllowMultipleConnections false)] :KernelType Custom :OutputPins [(struct :PropertiesGPU (struct :PinsToInititalizeFrom [In]) :Label Out :AllowedTypes (struct :Ids [(struct :Struct /Script/PCG.PCGDataTypeInfoTexture2DSingleBase)]) :bAllowMultipleData false)]))
  ;[0.4375 0 0.2969 0.5] debug grass maps (world)
  (bind CustomHLSL_8 (Custom%20HLSL :InputPins [(struct :Label In :AllowedTypes (struct :Ids [(struct :Struct /Script/PCG.PCGDataTypeInfoTexture2DArray)]) :bAllowMultipleData false :PinStatus Required :bAllowMultipleConnections false) (struct :Label SliceIndex :AllowedTypes (struct :Ids [(struct :Struct /Script/PCG.PCGDataTypeInfoParam)]) :bAllowMultipleData false :bAllowMultipleConnections false)] :KernelType Custom :OutputPins [(struct :PropertiesGPU (struct :PinsToInititalizeFrom [In]) :Label Out :AllowedTypes (struct :Ids [(struct :Struct /Script/PCG.PCGDataTypeInfoTexture2DSingleBase)]) :bAllowMultipleData false)]))
  (bind CustomHLSL_9 (Custom%20HLSL :bMuteUnwrittenPinDataErrors true :InputPins [(struct :Label In :AllowedTypes (struct :Ids [(struct :Struct /Script/PCG.PCGDataTypeInfoTexture2DSingleBase)]) :bAllowMultipleData false :PinStatus Required :bAllowMultipleConnections false) (struct :Label Height :AllowedTypes (struct :Ids [(struct :Struct /Script/PCG.PCGDataTypeInfoTexture2DSingleBase)]) :bAllowMultipleData false :bAllowMultipleConnections false)] :KernelType Custom :OutputPins [(struct :PropertiesGPU (struct :PinsToInititalizeFrom [In]) :Label Out :AllowedTypes (struct :Ids [(struct :Struct /Script/PCG.PCGDataTypeInfoPoint)]) :bAllowMultipleData false)]))
  (bind CustomHLSL_11 (Custom%20HLSL :OutputPins [(struct :PropertiesGPU (struct :PinsToInititalizeFrom [In] :CreatedKernelAttributeKeys [(struct :Identifier (struct :Name MyDensity) :Name MyDensity)]) :Label Out :AllowedTypes (struct :Ids [(struct :Struct /Script/PCG.PCGDataTypeInfoPoint)]))]))
  (bind StaticMeshSpawner_0 (Static%20Mesh%20Spawner :InstanceDataPackerParameters (class:/Script/PCG.PCGInstanceDataPackerByAttribute :AttributeSelectors [MyDensity]) :InstanceDataPackerType /Script/PCG.PCGInstanceDataPackerByAttribute :MeshSelectorParameters (class:/Script/PCG.PCGMeshSelectorWeighted :MeshEntries [(struct :Descriptor (struct :StaticMesh /PCG/DebugObjects/PCG_Cube.PCG_Cube :OverrideMaterials [/PCG/DebugObjects/PCG_DebugMaterial.PCG_DebugMaterial]))])))
  ;[0.4375 0 0.2969 0.5] Debug grass index - ENABLE HERE
  (bind CreateConstant_0 (Create%20Constant :AttributeTypes (struct :Type Integer32 :Int32Value 1) :OutputTarget SliceIndex))
  ;[0.4375 0 0.2969 0.5] inspect coverage mask (viewport)
  (bind CreateConstant_1 (Create%20Constant :AttributeTypes (struct :Type Integer32) :OutputTarget MipIndex))
  (bind CustomHLSL_10 (Custom%20HLSL :InputPins [(struct :Label In :AllowedTypes (struct :Ids [(struct :Struct /Script/PCG.PCGDataTypeInfoTexture2DArray)]) :bAllowMultipleData false :PinStatus Required :bAllowMultipleConnections false) (struct :Label SliceIndex :AllowedTypes (struct :Ids [(struct :Struct /Script/PCG.PCGDataTypeInfoParam)]) :bAllowMultipleData false :bAllowMultipleConnections false) (struct :Label MipIndex :AllowedTypes (struct :Ids [(struct :Struct /Script/PCG.PCGDataTypeInfoParam)]) :bAllowMultipleData false :bAllowMultipleConnections false)] :KernelType Custom :OutputPins [(struct :PropertiesGPU (struct :PinsToInititalizeFrom [In]) :Label Out :AllowedTypes (struct :Ids [(struct :Struct /Script/PCG.PCGDataTypeInfoTexture2DSingleBase)]) :bAllowMultipleData false)]))
  (reroute DebugGrassIndex CreateConstant_0)
  (reroute GrassData RemoveEmptyData_1)
  (reroute GrassMapArray GenerateLandscapeTextures_0.GrassMaps)
  (reroute Height GenerateLandscapeTextures_0.Height)
  (reroute Landscape GetLandscapeData_1)
  (reroute Layer%20Data CustomHLSL_0)
  (reroute Mask DownsampleTexture_1)
  (reroute PrimData MergeAttributes_2)
  (connect DeleteAttributes_1 Out AttributeMathsOp_1 InA)
  (connect GrassData Out AttributePartition_2 In)
  (connect DebugGrassIndex Out CreateConstant_1 Execution%20Dependency)
  (connect AttributeMathsOp_1 Out CustomHLSL_0 In)
  (connect GrassMapArray Out CustomHLSL_1 In)
  (connect GrassMapArray Out CustomHLSL_2 GrassMaps)
  (connect Layer%20Data Out CustomHLSL_3 LayerData)
  (connect Mask Out CustomHLSL_3 Mask)
  (connect CustomHLSL_2 NumSlices CustomHLSL_3 NumElements)
  (connect CustomHLSL_3 LayerInfo CustomHLSL_4 LayerInfo)
  (connect Mask Out CustomHLSL_4 Mask)
  (connect CustomHLSL_3 TotalMaxNumCells CustomHLSL_4 NumElements)
  (connect GrassMapArray Out CustomHLSL_5 GrassMap)
  (connect Height Out CustomHLSL_5 Height)
  (connect CustomHLSL_4 OutSeeds CustomHLSL_5 Seeds)
  (connect GrassMapArray Out CustomHLSL_6 GrassMap)
  (connect Height Out CustomHLSL_6 Height)
  (connect Layer%20Data Out CustomHLSL_6 LayerData)
  (connect CustomHLSL_4 NumThreads CustomHLSL_6 NumElements)
  (connect CustomHLSL_4 OutSeeds CustomHLSL_6 Seeds)
  (connect GrassMapArray Out CustomHLSL_7 In)
  (connect DebugGrassIndex Out CustomHLSL_7 SliceIndex)
  (connect GrassMapArray Out CustomHLSL_8 In)
  (connect DebugGrassIndex Out CustomHLSL_8 SliceIndex)
  (connect Height Out CustomHLSL_9 Height)
  (connect CustomHLSL_8 Out CustomHLSL_9 In)
  (connect Mask Out CustomHLSL_10 In)
  (connect CreateConstant_1 Out CustomHLSL_10 MipIndex)
  (connect DebugGrassIndex Out CustomHLSL_10 SliceIndex)
  (connect CustomHLSL_9 Out CustomHLSL_11 In)
  (connect CustomHLSL_5 Out CustomHLSL_12 In)
  (connect GrassData Out DeleteAttributes_0 In)
  (connect GrassData Out DeleteAttributes_1 In)
  (connect CustomHLSL_1 Out DownsampleTexture_1 In)
  (connect MergeAttributes_0 Out GenerateLandscapeTextures_0 Grass%20Types)
  (connect HiGenGridSize_1 Out GenerateLandscapeTextures_0 Landscape)
  (connect DeleteAttributes_0 Out GetAttributeFromIndex_1 In)
  (connect AttributePartition_2 Out GetAttributeFromIndex_4 In)
  (connect PlatformSwitch_1 PS4 GetLandscapeData_1 Execution%20Dependency)
  (connect PlatformSwitch_1 Switch GetLandscapeData_1 Execution%20Dependency)
  (connect PlatformSwitch_1 XB1 GetLandscapeData_1 Execution%20Dependency)
  (connect ShaderFeatureLevelSwitch_1 SM5 GetLandscapeData_1 Execution%20Dependency)
  (connect ShaderFeatureLevelSwitch_1 SM6 GetLandscapeData_1 Execution%20Dependency)
  (connect Landscape Out GetLandscapeGrassTypes_0 Landscape)
  (connect Landscape Out HiGenGridSize_1 In)
  (connect GetAttributeFromIndex_1 Out MergeAttributes_0 In)
  (connect GetAttributeFromIndex_4 Out MergeAttributes_2 In)
  (connect GetLandscapeGrassTypes_0 Out ProcessGrassTypes_0 In)
  (connect ProcessGrassTypes_0 Out RemoveEmptyData_1 In)
  (connect CustomHLSL_11 Out StaticMeshSpawner_0 In)
  (connect CustomHLSL_6 Out StaticMeshSpawner_1 In)
  (connect PrimData Out StaticMeshSpawner_1 PrimitiveTable)
  (connect CustomHLSL_12 Out StaticMeshSpawner_2 In))
