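; PCG graph that assigns weighted grass mesh varieties to points by landscape layer, for the mesh grass tool.
; It reads ten LayerBiome slots and their VarietyData from BP_MeshGrass_LandscapeGenerator, keeping active layers.
; A looped processor subgraph runs per layer to resolve each point's dominant layer, which Match And Set uses for weighted variety selection.
; TextureStamped points skip this step and merge back before empty meshes are dropped and PCG_GateValidPoints runs.
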(graph PCG_MeshGrass_LandscapeLayers
  (inputs (InPoints :type Point))
  (outputs (OutPoints :type Point))
  (bind FilterDataByAttribute_0 (Filter%20Data%20By%20Attribute :Attribute TextureStamped))
  (bind AttributeFilter_0 (Filter%20Attribute%20Elements :AttributeTypes (struct :Type Boolean :BoolValue true) :bUseConstantThreshold true :bUseSpatialQuery true :Operator Equal :TargetAttribute TextureStamped :ThresholdAttribute @Last))
  (bind AddAttribute_10 (Add%20Attribute :OutputTarget DominantLayer))
  (bind AddAttribute_11 (Add%20Attribute :AttributeTypes (struct :Type String) :OutputTarget DominantLayerName))
  (bind AttributePartition_9 (Attribute%20Partition :PartitionAttributeSelectors [$Index]))
  (bind Loop_16 (Loop :Graph /Game/Athena/Helios/Environments/Landscape/WorldBuildingTools/PCGGrass/PCG_MeshGrass_LandscapeLayers_Processor.PCG_MeshGrass_LandscapeLayers_Processor))
  (bind DeleteAttributes_13 (Delete%20Attributes :SelectedAttributes "DominantLayer"))
  (bind AttributeRename_14 (Attribute%20Rename :AttributeToRename DominantLayerName :NewAttributeName Layer))
  (bind DeleteAttributes_15 (Delete%20Attributes :SelectedAttributes "OriginalIndex"))
  (bind CreateAttribute_78 (Create%20Constant :AttributeTypes (struct :Type Integer64)))
  (bind AttributeCompareOp_77 (Attribute%20Compare%20Op :Operation NotEqual :InA DetailModeOverride))
  (bind GetActorProperty_73 (Get%20Actor%20Property :ActorSelector (struct :ActorSelection ByClass :ActorSelectionClass /Game/Athena/Helios/Environments/Landscape/WorldBuildingTools/PCGGrass/BP_MeshGrass_LandscapeGenerator.BP_MeshGrass_LandscapeGenerator_C) :bProcessAllComponents false :bSanitizeOutputAttributeName false :OutputAttributeName None :PropertyName LayerBiome_1.VarietyData))
  (bind GetActorProperty_83 (Get%20Actor%20Property :ActorSelector (struct :ActorSelection ByClass :ActorSelectionClass /Game/Athena/Helios/Environments/Landscape/WorldBuildingTools/PCGGrass/BP_MeshGrass_LandscapeGenerator.BP_MeshGrass_LandscapeGenerator_C) :bProcessAllComponents false :bSanitizeOutputAttributeName false :OutputAttributeName None :PropertyName LayerBiome_1))
  (bind AddAttribute_104 (Add%20Attribute :OutputTarget Layer))
  (bind GetActorProperty_72 (Get%20Actor%20Property :ActorSelector (struct :ActorSelection ByClass :ActorSelectionClass /Game/Athena/Helios/Environments/Landscape/WorldBuildingTools/PCGGrass/BP_MeshGrass_LandscapeGenerator.BP_MeshGrass_LandscapeGenerator_C) :bProcessAllComponents false :bSanitizeOutputAttributeName false :OutputAttributeName None :PropertyName LayerBiome_0.VarietyData))
  (bind GetActorProperty_82 (Get%20Actor%20Property :ActorSelector (struct :ActorSelection ByClass :ActorSelectionClass /Game/Athena/Helios/Environments/Landscape/WorldBuildingTools/PCGGrass/BP_MeshGrass_LandscapeGenerator.BP_MeshGrass_LandscapeGenerator_C) :bProcessAllComponents false :bSanitizeOutputAttributeName false :OutputAttributeName None :PropertyName LayerBiome_0))
  (bind AddAttribute_105 (Add%20Attribute :OutputTarget Layer))
  (bind GetActorProperty_74 (Get%20Actor%20Property :ActorSelector (struct :ActorSelection ByClass :ActorSelectionClass /Game/Athena/Helios/Environments/Landscape/WorldBuildingTools/PCGGrass/BP_MeshGrass_LandscapeGenerator.BP_MeshGrass_LandscapeGenerator_C) :bProcessAllComponents false :bSanitizeOutputAttributeName false :OutputAttributeName None :PropertyName LayerBiome_2.VarietyData))
  (bind GetActorProperty_84 (Get%20Actor%20Property :ActorSelector (struct :ActorSelection ByClass :ActorSelectionClass /Game/Athena/Helios/Environments/Landscape/WorldBuildingTools/PCGGrass/BP_MeshGrass_LandscapeGenerator.BP_MeshGrass_LandscapeGenerator_C) :bProcessAllComponents false :bSanitizeOutputAttributeName false :OutputAttributeName None :PropertyName LayerBiome_2))
  (bind AddAttribute_106 (Add%20Attribute :OutputTarget Layer))
  (bind GetActorProperty_75 (Get%20Actor%20Property :ActorSelector (struct :ActorSelection ByClass :ActorSelectionClass /Game/Athena/Helios/Environments/Landscape/WorldBuildingTools/PCGGrass/BP_MeshGrass_LandscapeGenerator.BP_MeshGrass_LandscapeGenerator_C) :bProcessAllComponents false :bSanitizeOutputAttributeName false :OutputAttributeName None :PropertyName LayerBiome_3.VarietyData))
  (bind GetActorProperty_85 (Get%20Actor%20Property :ActorSelector (struct :ActorSelection ByClass :ActorSelectionClass /Game/Athena/Helios/Environments/Landscape/WorldBuildingTools/PCGGrass/BP_MeshGrass_LandscapeGenerator.BP_MeshGrass_LandscapeGenerator_C) :bProcessAllComponents false :bSanitizeOutputAttributeName false :OutputAttributeName None :PropertyName LayerBiome_3))
  (bind AddAttribute_107 (Add%20Attribute :OutputTarget Layer))
  (bind GetActorProperty_76 (Get%20Actor%20Property :ActorSelector (struct :ActorSelection ByClass :ActorSelectionClass /Game/Athena/Helios/Environments/Landscape/WorldBuildingTools/PCGGrass/BP_MeshGrass_LandscapeGenerator.BP_MeshGrass_LandscapeGenerator_C) :bProcessAllComponents false :bSanitizeOutputAttributeName false :OutputAttributeName None :PropertyName LayerBiome_4.VarietyData))
  (bind GetActorProperty_86 (Get%20Actor%20Property :ActorSelector (struct :ActorSelection ByClass :ActorSelectionClass /Game/Athena/Helios/Environments/Landscape/WorldBuildingTools/PCGGrass/BP_MeshGrass_LandscapeGenerator.BP_MeshGrass_LandscapeGenerator_C) :bProcessAllComponents false :bSanitizeOutputAttributeName false :OutputAttributeName None :PropertyName LayerBiome_4))
  (bind AddAttribute_108 (Add%20Attribute :OutputTarget Layer))
  (bind GetActorProperty_77 (Get%20Actor%20Property :ActorSelector (struct :ActorSelection ByClass :ActorSelectionClass /Game/Athena/Helios/Environments/Landscape/WorldBuildingTools/PCGGrass/BP_MeshGrass_LandscapeGenerator.BP_MeshGrass_LandscapeGenerator_C) :bProcessAllComponents false :bSanitizeOutputAttributeName false :OutputAttributeName None :PropertyName LayerBiome_5.VarietyData))
  (bind GetActorProperty_87 (Get%20Actor%20Property :ActorSelector (struct :ActorSelection ByClass :ActorSelectionClass /Game/Athena/Helios/Environments/Landscape/WorldBuildingTools/PCGGrass/BP_MeshGrass_LandscapeGenerator.BP_MeshGrass_LandscapeGenerator_C) :bProcessAllComponents false :bSanitizeOutputAttributeName false :OutputAttributeName None :PropertyName LayerBiome_5))
  (bind AddAttribute_109 (Add%20Attribute :OutputTarget Layer))
  (bind GetActorProperty_78 (Get%20Actor%20Property :ActorSelector (struct :ActorSelection ByClass :ActorSelectionClass /Game/Athena/Helios/Environments/Landscape/WorldBuildingTools/PCGGrass/BP_MeshGrass_LandscapeGenerator.BP_MeshGrass_LandscapeGenerator_C) :bProcessAllComponents false :bSanitizeOutputAttributeName false :OutputAttributeName None :PropertyName LayerBiome_6.VarietyData))
  (bind GetActorProperty_88 (Get%20Actor%20Property :ActorSelector (struct :ActorSelection ByClass :ActorSelectionClass /Game/Athena/Helios/Environments/Landscape/WorldBuildingTools/PCGGrass/BP_MeshGrass_LandscapeGenerator.BP_MeshGrass_LandscapeGenerator_C) :bProcessAllComponents false :bSanitizeOutputAttributeName false :OutputAttributeName None :PropertyName LayerBiome_6))
  (bind AddAttribute_110 (Add%20Attribute :OutputTarget Layer))
  (bind GetActorProperty_79 (Get%20Actor%20Property :ActorSelector (struct :ActorSelection ByClass :ActorSelectionClass /Game/Athena/Helios/Environments/Landscape/WorldBuildingTools/PCGGrass/BP_MeshGrass_LandscapeGenerator.BP_MeshGrass_LandscapeGenerator_C) :bProcessAllComponents false :bSanitizeOutputAttributeName false :OutputAttributeName None :PropertyName LayerBiome_7.VarietyData))
  (bind GetActorProperty_89 (Get%20Actor%20Property :ActorSelector (struct :ActorSelection ByClass :ActorSelectionClass /Game/Athena/Helios/Environments/Landscape/WorldBuildingTools/PCGGrass/BP_MeshGrass_LandscapeGenerator.BP_MeshGrass_LandscapeGenerator_C) :bProcessAllComponents false :bSanitizeOutputAttributeName false :OutputAttributeName None :PropertyName LayerBiome_7))
  (bind AddAttribute_111 (Add%20Attribute :OutputTarget Layer))
  (bind GetActorProperty_80 (Get%20Actor%20Property :ActorSelector (struct :ActorSelection ByClass :ActorSelectionClass /Game/Athena/Helios/Environments/Landscape/WorldBuildingTools/PCGGrass/BP_MeshGrass_LandscapeGenerator.BP_MeshGrass_LandscapeGenerator_C) :bProcessAllComponents false :bSanitizeOutputAttributeName false :OutputAttributeName None :PropertyName LayerBiome_8.VarietyData))
  (bind GetActorProperty_90 (Get%20Actor%20Property :ActorSelector (struct :ActorSelection ByClass :ActorSelectionClass /Game/Athena/Helios/Environments/Landscape/WorldBuildingTools/PCGGrass/BP_MeshGrass_LandscapeGenerator.BP_MeshGrass_LandscapeGenerator_C) :bProcessAllComponents false :bSanitizeOutputAttributeName false :OutputAttributeName None :PropertyName LayerBiome_8))
  (bind AddAttribute_112 (Add%20Attribute :OutputTarget Layer))
  (bind GetActorProperty_81 (Get%20Actor%20Property :ActorSelector (struct :ActorSelection ByClass :ActorSelectionClass /Game/Athena/Helios/Environments/Landscape/WorldBuildingTools/PCGGrass/BP_MeshGrass_LandscapeGenerator.BP_MeshGrass_LandscapeGenerator_C) :bProcessAllComponents false :bSanitizeOutputAttributeName false :OutputAttributeName None :PropertyName LayerBiome_9.VarietyData))
  (bind GetActorProperty_91 (Get%20Actor%20Property :ActorSelector (struct :ActorSelection ByClass :ActorSelectionClass /Game/Athena/Helios/Environments/Landscape/WorldBuildingTools/PCGGrass/BP_MeshGrass_LandscapeGenerator.BP_MeshGrass_LandscapeGenerator_C) :bProcessAllComponents false :bSanitizeOutputAttributeName false :OutputAttributeName None :PropertyName LayerBiome_9))
  (bind AddAttribute_113 (Add%20Attribute :OutputTarget Layer))
  (bind Gather_114 (Gather))
  (bind AttributeRename_103 (Attribute%20Rename :AttributeToRename GrassMesh :NewAttributeName Mesh))
  (bind AttributeRename_1 (Attribute%20Rename :AttributeToRename StartCullDistance :NewAttributeName InstanceStartCullDistance))
  (bind AttributeRename_5 (Attribute%20Rename :AttributeToRename EndCullDistance :NewAttributeName InstanceEndCullDistance))
  (bind CreateAttribute_20 (Create%20Constant :AttributeTypes (struct :Type Integer64 :DoubleValue 1e+01 :IntValue 10)))
  (bind AttributeMathsOp_18 (Attribute%20Maths%20Op :InputSource1 BiomeIndex :Operation Subtract))
  (bind AttributeFilter_21 (Filter%20Attribute%20Elements :AttributeTypes (struct :Type Boolean :BoolValue true) :bUseConstantThreshold true :Operator Equal :TargetAttribute Active))
  (bind DeleteAttributes_24 (Delete%20Attributes :Operation KeepSelectedAttributes :SelectedAttributes "BiomeIndex"))
  (bind AttributeCast_28 (Attribute%20Cast :OutputType String))
  (bind ExecuteBlueprint_26 (Execute%20Blueprint :BlueprintElementInstance (class:/Game/Athena/Helios/Environments/Landscape/WorldBuildingTools/PCGGrass/PCGBP_JoinStrings.PCGBP_JoinStrings_C :TargetAttribute BiomeIndex) :BlueprintElementType /Game/Athena/Helios/Environments/Landscape/WorldBuildingTools/PCGGrass/PCGBP_JoinStrings.PCGBP_JoinStrings_C))
  (bind FilterDataByIndex_96 (Filter%20Data%20By%20Index))
  (bind FilterDataByAttribute_34 (Filter%20Data%20By%20Attribute :Attribute Mesh))
  (bind MergeAttributes_53 (Merge%20Attributes))
  (bind DeleteAttributes_79 (Delete%20Attributes :SelectedAttributes "DetailMode"))
  (bind Select_80 (Select))
  (bind MatchAndSetAttributes_0 (Match%20And%20Set%20Attributes :bKeepUnmatched false :bMatchAttributes true :bUseInputWeightAttribute true :bUseWeightAttribute true :InputAttribute Layer :InputWeightAttribute Variety :MatchAttribute Layer :WeightAttribute Weight))
  (bind MergeAttributes_94 (Merge%20Attributes))
  (bind AttributeRename_95 (Attribute%20Rename :AttributeToRename CustomAttribute0 :NewAttributeName Layer))
  (bind MergePoints_4 (Merge%20Points))
  (bind AttributeFilter_75 (Filter%20Attribute%20Elements :AttributeTypes (struct :Type SoftObjectPath) :bUseConstantThreshold true :Operator NotEqual :TargetAttribute Mesh :ThresholdAttribute @Last))
  (bind Subgraph_82 (Subgraph :Graph /Game/Athena/Helios/Environments/Landscape/WorldBuildingTools/PCGGrass/PCG_GateValidPoints.PCG_GateValidPoints))
  (reroute ActiveLayers AttributeFilter_21.InsideFilter)
  (reroute LayerAttributes AttributeRename_95)
  (reroute SkippedPoints AttributeFilter_0.InsideFilter)
  (connect AttributeFilter_0 OutsideFilter AddAttribute_10 In)
  (connect FilterDataByAttribute_0 OutsideFilter AddAttribute_10 In)
  (connect AddAttribute_10 Out AddAttribute_11 In)
  (connect GetActorProperty_83 Out AddAttribute_104 Attributes)
  (connect GetActorProperty_73 Out AddAttribute_104 In)
  (connect GetActorProperty_82 Out AddAttribute_105 Attributes)
  (connect GetActorProperty_72 Out AddAttribute_105 In)
  (connect GetActorProperty_84 Out AddAttribute_106 Attributes)
  (connect GetActorProperty_74 Out AddAttribute_106 In)
  (connect GetActorProperty_85 Out AddAttribute_107 Attributes)
  (connect GetActorProperty_75 Out AddAttribute_107 In)
  (connect GetActorProperty_86 Out AddAttribute_108 Attributes)
  (connect GetActorProperty_76 Out AddAttribute_108 In)
  (connect GetActorProperty_87 Out AddAttribute_109 Attributes)
  (connect GetActorProperty_77 Out AddAttribute_109 In)
  (connect GetActorProperty_88 Out AddAttribute_110 Attributes)
  (connect GetActorProperty_78 Out AddAttribute_110 In)
  (connect GetActorProperty_89 Out AddAttribute_111 Attributes)
  (connect GetActorProperty_79 Out AddAttribute_111 In)
  (connect GetActorProperty_90 Out AddAttribute_112 Attributes)
  (connect GetActorProperty_80 Out AddAttribute_112 In)
  (connect GetActorProperty_91 Out AddAttribute_113 Attributes)
  (connect GetActorProperty_81 Out AddAttribute_113 In)
  (connect DeleteAttributes_24 Out AttributeCast_28 In)
  (connect CreateAttribute_78 Out AttributeCompareOp_77 InB)
  (connect FilterDataByAttribute_0 InsideFilter AttributeFilter_0 In)
  (connect AttributeMathsOp_18 Out AttributeFilter_21 In)
  (connect MergePoints_4 Out AttributeFilter_75 In)
  (connect LayerAttributes Out AttributeMathsOp_18 InA)
  (connect CreateAttribute_20 Out AttributeMathsOp_18 InB)
  (connect ActiveLayers Out AttributePartition_9 In)
  (connect AttributeRename_103 Out AttributeRename_1 In)
  (connect AttributeRename_1 Out AttributeRename_5 In)
  (connect DeleteAttributes_13 Out AttributeRename_14 In)
  (connect MergeAttributes_94 Out AttributeRename_95 In)
  (connect Gather_114 Out AttributeRename_103 In)
  (connect Loop_16 InPoints DeleteAttributes_13 In)
  (connect AttributeRename_14 Out DeleteAttributes_15 In)
  (connect AttributeFilter_21 InsideFilter DeleteAttributes_24 In)
  (connect MergeAttributes_53 Out DeleteAttributes_79 In)
  (connect AttributeCast_28 Out ExecuteBlueprint_26 In)
  (connect DefaultInputNode InPoints FilterDataByAttribute_0 In)
  (connect FilterDataByIndex_96 InsideFilter FilterDataByAttribute_34 In)
  (connect AttributeRename_5 Out FilterDataByIndex_96 In)
  (connect ExecuteBlueprint_26 Out FilterDataByIndex_96 SelectedIndices)
  (connect AddAttribute_104 Out Gather_114 In)
  (connect AddAttribute_105 Out Gather_114 In)
  (connect AddAttribute_106 Out Gather_114 In)
  (connect AddAttribute_107 Out Gather_114 In)
  (connect AddAttribute_108 Out Gather_114 In)
  (connect AddAttribute_109 Out Gather_114 In)
  (connect AddAttribute_110 Out Gather_114 In)
  (connect AddAttribute_111 Out Gather_114 In)
  (connect AddAttribute_112 Out Gather_114 In)
  (connect AddAttribute_113 Out Gather_114 In)
  (connect AddAttribute_11 Out Loop_16 InPoints)
  (connect AttributePartition_9 Out Loop_16 LayerAttributes)
  (connect DeleteAttributes_15 Out MatchAndSetAttributes_0 In)
  (connect Select_80 Out MatchAndSetAttributes_0 Match%20Data)
  (connect FilterDataByAttribute_34 InsideFilter MergeAttributes_53 In)
  (connect GetActorProperty_82 Out MergeAttributes_94 In)
  (connect GetActorProperty_83 Out MergeAttributes_94 In)
  (connect GetActorProperty_84 Out MergeAttributes_94 In)
  (connect GetActorProperty_85 Out MergeAttributes_94 In)
  (connect GetActorProperty_86 Out MergeAttributes_94 In)
  (connect GetActorProperty_87 Out MergeAttributes_94 In)
  (connect GetActorProperty_88 Out MergeAttributes_94 In)
  (connect GetActorProperty_89 Out MergeAttributes_94 In)
  (connect GetActorProperty_90 Out MergeAttributes_94 In)
  (connect GetActorProperty_91 Out MergeAttributes_94 In)
  (connect MatchAndSetAttributes_0 Out MergePoints_4 In)
  (connect SkippedPoints Out MergePoints_4 In)
  (connect AttributeCompareOp_77 Out Select_80 bUseInputB)
  (connect MergeAttributes_53 Out Select_80 Input%20A)
  (connect DeleteAttributes_79 Out Select_80 Input%20B)
  (connect AttributeFilter_75 InsideFilter Subgraph_82 In)
  (output OutPoints Subgraph_82.ValidPoints))
