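; PCG graph that processes one landscape layer for mesh grass, scaling its point attribute by a strength value and updating each point's dominant layer.
; Layer and Strength come from LayerAttributes, and point data lacking that layer passes through untouched.
; Points are split by comparing DominantLayer with the scaled weight, and the outside-filter set receives new DominantLayer and DominantLayerName values.
; Both sets merge and re-sort by a saved OriginalIndex to restore point order.
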(graph PCG_MeshGrass_LandscapeLayers_Processor
  (inputs (InPoints :type Point) (LayerAttributes :type Param))
  (outputs (InPoints :type Point))
  (bind DeleteAttributes_22 (Delete%20Attributes :Operation KeepSelectedAttributes :SelectedAttributes "Layer"))
  (bind DeleteAttributes_25 (Delete%20Attributes :Operation KeepSelectedAttributes :SelectedAttributes "Strength"))
  (bind FilterDataByAttribute_12 (Filter%20Data%20By%20Attribute))
  (bind AttributeMathsOp_30 (Attribute%20Maths%20Op :Operation Multiply))
  (bind DeleteAttributes_10 (Delete%20Attributes :Operation KeepSelectedAttributes))
  (bind PointToAttributeSet_13 (Point%20To%20Attribute%20Set))
  (bind CopyAttributes_18 (Copy%20Attributes :InputSource $Index :OutputTarget OriginalIndex))
  (bind CreateAttribute_21 (Create%20Constant :AttributeTypes (struct :Type String :StringValue "DominantLayer")))
  (bind AttributeFilter_14 (Filter%20Attribute%20Elements :TargetAttribute @Last :ThresholdAttribute @Last))
  (bind GetAttributeList_15 (Get%20Attribute%20List))
  (bind AddAttribute_16 (Add%20Attribute :AttributeTypes (struct :Type String) :OutputTarget DominantLayerName))
  (bind AttributeMathsOp_20 (Attribute%20Maths%20Op :Operation Set))
  (bind MergePoints_17 (Merge%20Points))
  (bind SortAttributes_19 (Sort%20Attributes :bUseStableSort false :SortAttributes [(struct :InputSource OriginalIndex)]))
  (bind CreateAttribute_27 (Create%20Constant :AttributeTypes (struct :DoubleValue 1.0)))
  (bind AttributeMathsOp_26 (Attribute%20Maths%20Op :Operation Divide))
  (connect GetAttributeList_15 Out AddAttribute_16 Attributes)
  (connect AttributeFilter_14 OutsideFilter AddAttribute_16 In)
  (connect PointToAttributeSet_13 Out AttributeFilter_14 Filter)
  (connect CopyAttributes_18 Out AttributeFilter_14 In)
  (connect CreateAttribute_21 Out AttributeFilter_14 TargetAttribute)
  (connect AddAttribute_16 Out AttributeMathsOp_20 InA)
  (connect AddAttribute_16 Out AttributeMathsOp_20 InB)
  (connect CreateAttribute_21 Out AttributeMathsOp_20 InputSource1)
  (connect GetAttributeList_15 Out AttributeMathsOp_20 InputSource2)
  (connect CreateAttribute_27 Out AttributeMathsOp_26 InA)
  (connect DeleteAttributes_25 Out AttributeMathsOp_26 InB)
  (connect FilterDataByAttribute_12 InsideFilter AttributeMathsOp_30 InA)
  (connect DeleteAttributes_25 Out AttributeMathsOp_30 InB)
  (connect DeleteAttributes_22 Out AttributeMathsOp_30 InputSource1)
  (connect AttributeMathsOp_30 Out CopyAttributes_18 Source)
  (connect AttributeMathsOp_30 Out CopyAttributes_18 Target)
  (connect AttributeMathsOp_30 Out DeleteAttributes_10 In)
  (connect DeleteAttributes_22 Out DeleteAttributes_10 SelectedAttributes)
  (connect DefaultInputNode LayerAttributes DeleteAttributes_22 In)
  (connect DefaultInputNode LayerAttributes DeleteAttributes_25 In)
  (connect DeleteAttributes_22 Out FilterDataByAttribute_12 Attribute)
  (connect DefaultInputNode InPoints FilterDataByAttribute_12 In)
  (connect PointToAttributeSet_13 Out GetAttributeList_15 In)
  (connect AttributeFilter_14 InsideFilter MergePoints_17 In)
  (connect AttributeMathsOp_20 Out MergePoints_17 In)
  (connect DeleteAttributes_10 Out PointToAttributeSet_13 In)
  (connect MergePoints_17 Out SortAttributes_19 In)
  (output InPoints FilterDataByAttribute_12.OutsideFilter)
  (output InPoints SortAttributes_19))
