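; PCG graph that samples the landscape into LandscapePoints for biome-driven mesh grass.
; Shared subgraphs resolve biome and attributes, assign grass, apply landscape layers and process points and metadata, with validity branches between early stages.
; Points are then slope-filtered by normal, shrunk and randomly yawed.
; Binary discrete Difference nodes carve out looped FortWaterBody results, sampled landscape splines and cull volume actors before output through a debug print.
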(graph PCG_MeshGrass_LandscapeSampler
  (params (DetailModeOverride :type (Enum /Game/Athena/Helios/Environments/Landscape/WorldBuildingTools/PCGGrass/EMeshGrassDetailMode.EMeshGrassDetailMode)) (PointsPerSquaredMeter :type Double :default 0.5) (Looseness :type Double :default 1.0) (RandomizeRotationOffset :type Bool :default true))
  (outputs (LandscapePoints :type Point))
  (bind GetLandscapeData_1 (Get%20Landscape%20Data :SamplingProperties (struct)))
  (bind SurfaceSampler_2 (Surface%20Sampler :bApplyDensityToPoints false :PointExtents (2e+01 2e+01 2e+01) :PointsPerSquaredMeter 0.5 :PointSteepness 0.5 :Looseness-param Looseness :PointsPerSquaredMeter-param PointsPerSquaredMeter))
  (bind GetActorData_2 (Get%20Actor%20Data :Mode GetSinglePoint))
  (bind Subgraph_0 (Subgraph :Graph /Game/Athena/Helios/Environments/Landscape/WorldBuildingTools/PCGGrass/PCG_MeshGrass_Inputs.PCG_MeshGrass_Inputs))
  (bind Subgraph_3 (Subgraph :Graph /Game/Athena/Helios/Environments/Landscape/WorldBuildingTools/PCGGrass/PCG_MeshGrass_BiomeDetermination.PCG_MeshGrass_BiomeDetermination))
  (bind Subgraph_6 (Subgraph :Graph /Game/Athena/Helios/Environments/Landscape/WorldBuildingTools/PCGGrass/PCG_HasValidPoints.PCG_HasValidPoints))
  (bind Branch_0 (Branch))
  (bind Subgraph_1 (Subgraph :Graph /Game/Athena/Helios/Environments/Landscape/WorldBuildingTools/PCGGrass/PCG_MeshGrass_GrassAssignment.PCG_MeshGrass_GrassAssignment))
  (bind Subgraph_2 (Subgraph :Graph /Game/Athena/Helios/Environments/Landscape/WorldBuildingTools/PCGGrass/PCG_MeshGrass_LandscapeLayers.PCG_MeshGrass_LandscapeLayers))
  (bind Subgraph_8 (Subgraph :Graph /Game/Athena/Helios/Environments/Landscape/WorldBuildingTools/PCGGrass/PCG_HasValidData.PCG_HasValidData))
  (bind Branch_9 (Branch))
  (bind Subgraph_4 (Subgraph :Graph /Game/Athena/Helios/Environments/Landscape/WorldBuildingTools/PCGGrass/PCG_MeshGrass_PointProcessing.PCG_MeshGrass_PointProcessing))
  (bind Subgraph_5 (Subgraph :Graph /Game/Athena/Helios/Environments/Landscape/WorldBuildingTools/PCGGrass/PCG_MeshGrass_MetadataProcessing.PCG_MeshGrass_MetadataProcessing))
  (bind NormalToDensity_11 (Normal%20To%20Density))
  (bind DensityFilter_12 (Density%20Filter :LowerBound 0.85))
  (bind ExtentsModifier_140 (Extents%20Modifier :Extents (0.15 0.15 0.15) :Mode Multiply))
  (bind TransformPoints_13 (Transform%20Points :RotationMax (struct :Yaw 1.8e+02) :RotationMin (struct :Yaw -1.8e+02)))
  (bind GetActorData_15 (Get%20Actor%20Data :ActorSelector (struct :ActorFilter AllWorldActors :bMustOverlapSelf true :ActorSelection ByClass :ActorSelectionClass /Game/Athena/Apollo/Environments/Landscape/Water/FortWaterBodyBP.FortWaterBodyBP_C :bSelectMultiple true) :bComponentsMustOverlapSelf false))
  (bind FilterDataByType_16 (Filter%20Data%20By%20Type :TargetType (struct :Ids [(struct :Struct /Script/PCG.PCGDataTypeInfoSpline)])))
  (bind GetActorData_17 (Get%20Actor%20Data :ActorSelector (struct :ActorFilter AllWorldActors :bMustOverlapSelf true :ActorSelection ByClass :ActorSelectionClass /Game/Athena/Apollo/Environments/Landscape/Water/FortWaterBodyBP.FortWaterBodyBP_C :bSelectMultiple true) :bComponentsMustOverlapSelf false :Mode GetSinglePoint))
  (bind Loop_18 (Loop :bTokenizeOnWhiteSpace true :bUseGraphDefaultPinUsage false :LoopPins "WaterBodyActors,WaterBodySplines" :Graph /Game/Athena/Helios/Environments/Landscape/WorldBuildingTools/PCGGrass/PCG_MeshGrass_WaterBodyProcessor_Default.PCG_MeshGrass_WaterBodyProcessor_Default))
  (bind ExtentsModifier_142 (Extents%20Modifier :Extents (1.0 0.6 1.0) :Mode Multiply))
  (bind Difference_19 (Difference :bDiffMetadata false :DensityFunction Binary :Mode Discrete))
  (bind GetSplineData_1 (Get%20Spline%20Data :ActorSelector (struct :ActorFilter AllWorldActors :bMustOverlapSelf true :ActorSelection ByClass :ActorSelectionClass /Script/Landscape.LandscapeSplineActor :bSelectMultiple true)))
  (bind SplineSampler_2 (Spline%20Sampler :SamplerParams (struct :Mode Distance :DistanceIncrement 5e+02)))
  (bind ExtentsModifier_4 (Extents%20Modifier :Extents (1.5 0.25 2e+01) :Mode Multiply))
  (bind Difference_7 (Difference :bDiffMetadata false :DensityFunction Binary :Mode Discrete))
  (bind GetActorData_0 (Get%20Actor%20Data :ActorSelector (struct :ActorFilter AllWorldActors :bMustOverlapSelf true :ActorSelection ByClass :ActorSelectionClass /Game/Athena/Helios/Environments/Landscape/WorldBuildingTools/PCGGrass/BP_MeshGrass_CullVolume.BP_MeshGrass_CullVolume_C :bSelectMultiple true) :bMergeSinglePointData true :Mode GetSinglePoint))
  (bind Difference_2 (Difference :bDiffMetadata false :DensityFunction Binary :Mode Discrete))
  (bind ToPoint_20 (To%20Point))
  (bind CreateAttribute_0 (Create%20Constant :AttributeTypes (struct :Type Boolean)))
  (bind Print%20String_0 (Print%20String :bPrefixWithNode false :bPrintPerComponent false :bPrintToScreen true :CustomPrefix "PCG_MeshGrass_LandscapeSampler::" :PrintString "UpdateLandscapeGrass" :PrintToScreenColor (struct :B 135 :G 175 :R 255 :A 255) :PrintToScreenDuration 0.2))
  (connect Subgraph_6 IsValid Branch_0 bOutputToB)
  (connect SurfaceSampler_2 Out Branch_0 In)
  (connect Subgraph_8 IsValid Branch_9 bOutputToB)
  (connect Subgraph_2 OutPoints Branch_9 In)
  (connect NormalToDensity_11 Out DensityFilter_12 In)
  (connect GetActorData_0 Out Difference_2 Differences)
  (connect Difference_7 Out Difference_2 Source)
  (connect ExtentsModifier_4 Out Difference_7 Differences)
  (connect Difference_19 Out Difference_7 Source)
  (connect ExtentsModifier_142 Out Difference_19 Differences)
  (connect TransformPoints_13 Out Difference_19 Source)
  (connect SplineSampler_2 Out ExtentsModifier_4 In)
  (connect DensityFilter_12 Out ExtentsModifier_140 In)
  (connect Loop_18 WaterBodyPoints ExtentsModifier_142 In)
  (connect GetActorData_15 Out FilterDataByType_16 In)
  (connect NormalToDensity_11 Out Loop_18 Exec)
  (connect GetActorData_17 Out Loop_18 WaterBodyActors)
  (connect FilterDataByType_16 InsideFilter Loop_18 WaterBodySplines)
  (connect Subgraph_5 OutPoints NormalToDensity_11 In)
  (connect CreateAttribute_0 Out Print%20String_0 bEnablePrint)
  (connect ToPoint_20 Out Print%20String_0 In)
  (connect GetSplineData_1 Out SplineSampler_2 Spline)
  (connect GetActorData_2 Out Subgraph_0 ActorPoint)
  (connect Subgraph_3 ActorPoint Subgraph_1 ActorPoint)
  (connect Subgraph_0 BiomeAttributes Subgraph_1 BiomeAttributes)
  (connect Subgraph_0 GlobalAttributes Subgraph_1 GlobalAttributes)
  (connect Branch_0 Output%20B Subgraph_1 InPoints)
  (connect Subgraph_0 VolumeAttributes Subgraph_1 VolumeAttributes)
  (connect Subgraph_1 OutPoints Subgraph_2 InPoints)
  (connect Subgraph_0 ActorPoint Subgraph_3 ActorPoint)
  (connect Subgraph_0 BiomeAttributes Subgraph_3 BiomeAttributes)
  (connect Subgraph_0 GlobalAttributes Subgraph_3 GlobalAttributes)
  (connect Subgraph_3 ActorPoint Subgraph_4 ActorPoint)
  (connect Subgraph_0 BiomeAttributes Subgraph_4 BiomeAttributes)
  (connect Subgraph_0 GlobalAttributes Subgraph_4 GlobalAttributes)
  (connect Branch_9 Output%20B Subgraph_4 InPoints)
  (connect Subgraph_0 VolumeAttributes Subgraph_4 VolumeAttributes)
  (connect Subgraph_3 ActorPoint Subgraph_5 ActorPoint)
  (connect Subgraph_4 OutPoints Subgraph_5 InPoints)
  (connect SurfaceSampler_2 Out Subgraph_6 In)
  (connect Subgraph_2 OutPoints Subgraph_8 In)
  (connect GetLandscapeData_1 Out SurfaceSampler_2 Surface)
  (connect Difference_2 Out ToPoint_20 In)
  (connect ExtentsModifier_140 Out TransformPoints_13 In)
  (output LandscapePoints Print%20String_0))
