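; PCG subgraph that thins mesh grass points by topology, density and local height.
; Points optionally pass through the FilterTopology subgraph, then get noise-driven density filtering against actor grass density, scaled by biome and optional volume override.
; A FilterWaterBodies branch can add and set a DiscardGrassDensity flag.
; An optional Position.Z range filter uses LocalZMin/Max before valid-point gating; the computed world-offset Z limits are left unconnected.
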(graph PCG_MeshGrass_LocalFiltering
  (inputs (InPoints :type Point) (ActorPoint :type Point) (ActorAttributes :type Param) (VolumeAttributes :type Param) (BiomeAttributes :type Param) (GlobalAttributes :type Param))
  (outputs (OutPoints :type Point))
  ;[0.05 0.05 0.05] Local Z Filtering
  (bind DeleteAttributes_1 (Delete%20Attributes :bTokenizeOnWhiteSpace true :Operation KeepSelectedAttributes :SelectedAttributes "FilterLocalZ"))
  (bind Branch_5 (Branch)) ; LocalZ
  (bind DeleteAttributes_3 (Delete%20Attributes :bTokenizeOnWhiteSpace true :Operation KeepSelectedAttributes :SelectedAttributes "LocalZMin"))
  (bind DeleteAttributes_4 (Delete%20Attributes :bTokenizeOnWhiteSpace true :Operation KeepSelectedAttributes :SelectedAttributes "LocalZMax"))
  (bind AttributeFilterRange_0 (Filter%20Attribute%20Elements%20by%20Range :MaxThreshold (struct :ThresholdAttribute @Last :bUseSpatialQuery false) :MinThreshold (struct :ThresholdAttribute @Last :bUseSpatialQuery false) :TargetAttribute $Position.Z))
  (bind GetAttributeFromPointIndex_6 (Get%20Attribute%20From%20Point%20Index :bForceOutputAttributeToBeInElementsDomain true :InputSource $Position.Z :OutputAttributeName None))
  (bind DeleteAttributes_7 (Delete%20Attributes :bTokenizeOnWhiteSpace true :Operation KeepSelectedAttributes :SelectedAttributes "PointsSource"))
  (bind CreateAttribute_10 (Create%20Constant :AttributeTypes (struct :Type Integer64 :IntValue 1)))
  (bind AttributeCompareOp_9 (Attribute%20Compare%20Op))
  (bind AttributeMathsOp_12 (Attribute%20Maths%20Op))
  (bind Select_11 (Select)) ; might be useful later for world sourced points
  (bind AttributeMathsOp_14 (Attribute%20Maths%20Op))
  (bind Select_13 (Select))
  (bind Subgraph_5 (Subgraph :Graph /Game/Athena/Helios/Environments/Landscape/WorldBuildingTools/PCGGrass/PCG_GateValidPoints.PCG_GateValidPoints))
  ;[0.05 0.05 0.05] Density Modifiers
  (bind DeleteAttributes_50 (Delete%20Attributes :bTokenizeOnWhiteSpace true :Operation KeepSelectedAttributes :SelectedAttributes "GrassDensity"))
  (bind DeleteAttributes_51 (Delete%20Attributes :Operation KeepSelectedAttributes :SelectedAttributes "Density Factor"))
  (bind AttributeMathsOp_55 (Attribute%20Maths%20Op :Operation Multiply))
  (bind DeleteAttributes_52 (Delete%20Attributes :bTokenizeOnWhiteSpace true :Operation KeepSelectedAttributes :SelectedAttributes "VolumeOverride"))
  (bind Select_54 (Select))
  (bind Subgraph_60 (Subgraph :Graph /Game/Athena/Helios/Environments/Landscape/WorldBuildingTools/PCGGrass/PCG_MeshGrass_SelectBiomeAttribute.PCG_MeshGrass_SelectBiomeAttribute))
  (bind AttributeMathsOp_59 (Attribute%20Maths%20Op :Operation Multiply))
  (bind AttributeNoise_61 (Attribute%20Noise))
  (bind DensityFilter_49 (Density%20Filter :LowerBound 0.0))
  (bind Branch_0 (Branch :bOutputToB-param FilterWaterBodies))
  (bind CreateConstant_0 (Create%20Constant :AttributeTypes (struct :Type Boolean) :OutputTarget DiscardGrassDensity))
  (bind AddAttribute_3 (Add%20Attribute :bCopyAllAttributes true))
  (bind AttributeFilter_0 (Filter%20Attribute%20Elements :Operator LesserOrEqual :TargetAttribute $Density :ThresholdAttribute @Last))
  (bind AttributeBooleanOp_1 (Attribute%20Boolean%20Op :Operation Not))
  (bind MergePoints_2 (Merge%20Points))
  (bind AttributeMathsOp_0 (Attribute%20Maths%20Op :InputSource1 $Density :Operation Set))
  (bind Subgraph_6 (Subgraph :Graph /Game/Athena/Helios/Environments/Landscape/WorldBuildingTools/PCGGrass/PCG_GateValidPoints.PCG_GateValidPoints))
  ;[0.05 0.05 0.05] Topological Filters
  (bind DeleteAttributes_0 (Delete%20Attributes :Operation KeepSelectedAttributes :SelectedAttributes "FilterTopologyInWorldSpace"))
  (bind Branch_1 (Branch))
  (bind Subgraph_2 (Subgraph :Graph /Game/Athena/Helios/Environments/Landscape/WorldBuildingTools/PCGGrass/PCG_MeshGrass_FilterTopology.PCG_MeshGrass_FilterTopology))
  (bind Gather_0 (Gather))
  ;[0.05 0.05 0.05] Inputs
  (bind GetAttributeFromPointIndex_72 (Get%20Attribute%20From%20Point%20Index :bForceOutputAttributeToBeInElementsDomain true :InputSource Biome :OutputAttributeName Biome))
  (reroute Actor%20Attributes ActorAttributes)
  (reroute Actor%20Biome GetAttributeFromPointIndex_72.Attribute)
  (reroute Actor%20Point ActorPoint)
  (reroute Biome%20Attributes BiomeAttributes)
  (reroute Global%20Attributes GlobalAttributes)
  (reroute In%20Points InPoints)
  (reroute Volume%20Attributes VolumeAttributes)
  (connect CreateConstant_0 Out AddAttribute_3 Attributes)
  (connect Branch_0 Output%20B AddAttribute_3 In)
  (connect AttributeFilter_0 OutsideFilter AttributeBooleanOp_1 In)
  (connect DeleteAttributes_7 Out AttributeCompareOp_9 InA)
  (connect CreateAttribute_10 Out AttributeCompareOp_9 InB)
  (connect AttributeMathsOp_59 Out AttributeFilter_0 Filter)
  (connect AddAttribute_3 Out AttributeFilter_0 In)
  (connect DeleteAttributes_4 Out AttributeFilterRange_0 FilterMax)
  (connect DeleteAttributes_3 Out AttributeFilterRange_0 FilterMin)
  (connect Branch_5 Output%20B AttributeFilterRange_0 In)
  (connect Branch_0 Output%20A AttributeMathsOp_0 InA)
  (connect MergePoints_2 Out AttributeMathsOp_0 InA)
  (connect DeleteAttributes_3 Out AttributeMathsOp_12 InA)
  (connect GetAttributeFromPointIndex_6 Attribute AttributeMathsOp_12 InB)
  (connect DeleteAttributes_4 Out AttributeMathsOp_14 InA)
  (connect GetAttributeFromPointIndex_6 Attribute AttributeMathsOp_14 InB)
  (connect DeleteAttributes_50 Out AttributeMathsOp_55 InA)
  (connect DeleteAttributes_51 Out AttributeMathsOp_55 InB)
  (connect Select_54 Out AttributeMathsOp_59 InA)
  (connect Subgraph_60 Out%20Attribute AttributeMathsOp_59 InB)
  (connect Gather_0 Out AttributeNoise_61 In)
  (connect DensityFilter_49 Out Branch_0 In)
  (connect DeleteAttributes_0 Out Branch_1 bOutputToB)
  (connect In%20Points Out Branch_1 In)
  (connect DeleteAttributes_1 Out Branch_5 bOutputToB)
  (connect Subgraph_6 ValidPoints Branch_5 In)
  (connect Actor%20Attributes Out DeleteAttributes_0 In)
  (connect Actor%20Attributes Out DeleteAttributes_1 In)
  (connect Actor%20Attributes Out DeleteAttributes_3 In)
  (connect Actor%20Attributes Out DeleteAttributes_4 In)
  (connect Actor%20Attributes Out DeleteAttributes_7 In)
  (connect Actor%20Attributes Out DeleteAttributes_50 In)
  (connect Volume%20Attributes Out DeleteAttributes_51 In)
  (connect Volume%20Attributes Out DeleteAttributes_52 In)
  (connect AttributeNoise_61 Out DensityFilter_49 In)
  (connect AttributeMathsOp_59 Out DensityFilter_49 UpperBound)
  (connect Branch_1 Output%20B Gather_0 In)
  (connect Subgraph_2 OutPoints Gather_0 In)
  (connect Actor%20Point Out GetAttributeFromPointIndex_6 In)
  (connect Actor%20Point Out GetAttributeFromPointIndex_72 In)
  (connect AttributeBooleanOp_1 Out MergePoints_2 In)
  (connect AttributeFilter_0 InsideFilter MergePoints_2 In)
  (connect AttributeCompareOp_9 Out Select_11 bUseInputB)
  (connect DeleteAttributes_3 Out Select_11 Input%20A)
  (connect AttributeMathsOp_12 Out Select_11 Input%20B)
  (connect AttributeCompareOp_9 Out Select_13 bUseInputB)
  (connect DeleteAttributes_4 Out Select_13 Input%20A)
  (connect AttributeMathsOp_14 Out Select_13 Input%20B)
  (connect DeleteAttributes_52 Out Select_54 bUseInputB)
  (connect DeleteAttributes_50 Out Select_54 Input%20A)
  (connect AttributeMathsOp_55 Out Select_54 Input%20B)
  (connect Actor%20Attributes Out Subgraph_2 ActorAttributes)
  (connect Branch_1 Output%20A Subgraph_2 InPoints)
  (connect Volume%20Attributes Out Subgraph_2 VolumeAttributes)
  (connect AttributeFilterRange_0 InsideFilter Subgraph_5 In)
  (connect Branch_5 Output%20A Subgraph_5 In)
  (connect AttributeMathsOp_0 Out Subgraph_6 In)
  (connect Biome%20Attributes Out Subgraph_60 InAttributes)
  (connect Actor%20Biome Out Subgraph_60 TargetBiomeIndex)
  (output OutPoints Subgraph_5.ValidPoints))
