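; PCG graph for the mesh grass tool that Poisson-samples points on a source actor's static mesh.
; It resolves the mesh, derives sampling radius from actor attributes, optional volume factor and inverse scale, and picks LOD by triangle count or forced precision.
; Material ID and vertex color filters, with distance-based dilation or erosion and inversion, flag points in a Discard attribute instead of removing them.
; It prints a hint when the sample cap is reached.
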(graph PCG_MeshGrass_MeshSampling
  (inputs (ActorPoint :type Point) (ActorAttributes :type Param) (VolumeAttributes :type Param) (BiomeAttributes :type Param) (GlobalAttributes :type Param))
  (outputs (OutPoints :type Point) (ActorPoint :type Point))
  ;[0.05 0.05 0.05] Mesh Sampling
  (bind CreateAttribute_11 (Create%20Constant :AttributeTypes (struct :DoubleValue 1.0)))
  (bind AttributeMathsOp_17 (Attribute%20Maths%20Op :InputSource1 $Scale :Operation Abs))
  (bind AttributeMathsOp_9 (Attribute%20Maths%20Op :InputSource1 $Scale.X :InputSource2 $Scale.Y :Operation Max :OutputTarget Max))
  (bind GetAttributeFromPointIndex_16 (Get%20Attribute%20From%20Point%20Index :bForceOutputAttributeToBeInElementsDomain true :OutputAttributeName Max))
  (bind AttributeMathsOp_10 (Attribute%20Maths%20Op :InputSource2 Max :Operation Divide))
  (bind CreateAttribute_14 (Create%20Constant :AttributeTypes (struct :DoubleValue 1.25)))
  (bind CreateAttribute_15 (Create%20Constant :AttributeTypes (struct :DoubleValue 0.75)))
  (bind AttributeMathsOp_13 (Attribute%20Maths%20Op :Operation Clamp)) ; scale based density
  (bind DeleteAttributes_6 (Delete%20Attributes :bTokenizeOnWhiteSpace true :Operation KeepSelectedAttributes :SelectedAttributes "SamplingRadius"))
  (bind DeleteAttributes_103 (Delete%20Attributes :Operation KeepSelectedAttributes :SelectedAttributes "Sampling Radius Factor"))
  (bind AttributeMathsOp_107 (Attribute%20Maths%20Op :Operation Multiply))
  (bind DeleteAttributes_105 (Delete%20Attributes :bTokenizeOnWhiteSpace true :Operation KeepSelectedAttributes :SelectedAttributes "VolumeOverride"))
  (bind Select_104 (Select))
  (bind AttributeMathsOp_12 (Attribute%20Maths%20Op :Operation Multiply))
  (bind GetAttributeFromPointIndex_6 (Get%20Attribute%20From%20Point%20Index :bForceOutputAttributeToBeInElementsDomain true :InputSource ActorReference :OutputAttributeName @SourceName))
  (bind GetActorData_0 (Get%20Actor%20Data :ComponentSelector (struct :ComponentSelection ByClass :ComponentSelectionClass /Script/Engine.StaticMeshComponent) :Mode GetComponentsReference))
  (bind GetElementsCount_1 (Get%20Element%20Count))
  (bind AttributeMathsOp_1 (Attribute%20Maths%20Op :Operation Dec))
  (bind AttributeMathsOp_0 (Attribute%20Maths%20Op :Operation Min :InB SourceMeshIndex))
  (bind GetAttributeFromIndex_0 (Get%20Attribute%20Set%20from%20Index))
  (bind GetPropertyFromObjectPath_0 (Get%20Property%20From%20Object%20Path :InputSource ComponentReference :PropertyName StaticMesh))
  (bind Select_7 (Select :bUseInputB-param VertexColorFromInstance))
  (bind Subgraph_0 (Subgraph :Graph /Game/Athena/Helios/Environments/Landscape/WorldBuildingTools/PCGGrass/PCG_IsValidPath.PCG_IsValidPath))
  (bind Branch_0 (Branch))
  (bind CreateConstant_0 (Create%20Constant :AttributeTypes (struct :Type Integer32) :OutputTarget MaxSamples))
  (bind DeleteAttributes_4 (Delete%20Attributes :bTokenizeOnWhiteSpace true :Operation KeepSelectedAttributes :SelectedAttributes "MaxSamples"))
  (bind Select_0 (Select :bUseInputB-param IgnoreMaxPointsPerMesh))
  (bind CreateAttribute_2 (Create%20Constant :AttributeTypes (struct :Type Integer32 :Int32Value 1000 :DoubleValue 1e+03)))
  (bind ExecuteBlueprint_0 (Execute%20Blueprint :BlueprintElementInstance (class:/Game/Athena/Helios/Environments/Landscape/WorldBuildingTools/PCGGrass/PCGBP_GetNumTriangles.PCGBP_GetNumTriangles_C) :BlueprintElementType /Game/Athena/Helios/Environments/Landscape/WorldBuildingTools/PCGGrass/PCGBP_GetNumTriangles.PCGBP_GetNumTriangles_C))
  (bind AttributeCompareOp_1 (Attribute%20Compare%20Op :Operation Less))
  (bind AttributeBooleanOp_23 (Attribute%20Boolean%20Op :Operation Or :InB VertexColorForcePrecision))
  (bind CreateAttribute_5 (Create%20Constant :AttributeTypes (struct :Type Integer32)))
  (bind CreateAttribute_12 (Create%20Constant :AttributeTypes (struct :Type Integer32 :Int32Value 20000 :DoubleValue 1e+03)))
  (bind AttributeCompareOp_10 (Attribute%20Compare%20Op :Operation Greater))
  (bind CreateAttribute_16 (Create%20Constant :AttributeTypes (struct :Type Integer32 :Int32Value 2)))
  (bind CreateAttribute_18 (Create%20Constant :AttributeTypes (struct :Type Integer32 :Int32Value 1)))
  (bind Select_8 (Select))
  (bind Select_3 (Select))
  (bind MeshSampler_1 (Mesh%20Sampler :bExtractMeshFromInput true :bHasDefaultColorOverride false :bOutputMaterialInfo true :PointSteepness 0.5 :SamplingMethod PoissonSampling :VoxelSize 5e+02 :bOutputMaterialInfo-param FilterMaterialID))
  (bind CopyPoints_53 (Copy%20Points :ColorInheritance Source :ScaleInheritance Source))
  (bind ExtentsModifier_55 (Extents%20Modifier :Extents (15.0 15.0 1e+01)))
  ;[0.05 0.05 0.05] Inputs
  (bind GetAttributeFromPointIndex_119 (Get%20Attribute%20From%20Point%20Index :bForceOutputAttributeToBeInElementsDomain true :InputSource Biome :OutputAttributeName Biome))
  ;[0.05 0.05 0.05] Mesh Filters
  (bind CreateAttribute_28 (Create%20Constant :AttributeTypes (struct :Type Boolean) :OutputTarget DiscardMaterialID))
  (bind AddAttribute_26 (Add%20Attribute :bCopyAllAttributes true))
  (bind CreateAttribute_13 (Create%20Constant :AttributeTypes (struct :Type Boolean) :OutputTarget DiscardVertex))
  (bind AddAttribute_10 (Add%20Attribute :bCopyAllAttributes true))
  ;;[0.05 0.05 0.05] MaterialID Filter
  (bind Branch_36 (Branch :bOutputToB-param FilterMaterialID))
  (bind AttributeFilter_37 (Filter%20Attribute%20Elements :Operator NotEqual :TargetAttribute MaterialID :ThresholdAttribute @Last :Filter MaterialIDTarget))
  (bind CreateAttribute_32 (Create%20Constant :AttributeTypes (struct :Type Boolean :BoolValue true) :OutputTarget DiscardMaterialID))
  (bind AddAttribute_34 (Add%20Attribute :bCopyAllAttributes true))
  (bind MergePoints_35 (Merge%20Points))
  (bind Branch_37 (Branch :bOutputToB-param MaterialIDInvert))
  (bind AttributeBooleanOp_36 (Attribute%20Boolean%20Op :InputSource1 DiscardMaterialID :Operation Not))
  (bind DeleteAttributes_40 (Delete%20Attributes :SelectedAttributes "MaterialID, Material"))
  (bind Gather_39 (Gather))
  ;;[0.05 0.05 0.05] Vertex Color Filter
  (bind Branch_24 (Branch :bOutputToB-param FilterVertexColor))
  (bind CreateAttribute_129 (Create%20Constant :AttributeTypes (struct :DoubleValue 1e-06)))
  (bind AttributeMathsOp_128 (Attribute%20Maths%20Op :Operation Max :InA VertexColorBias))
  (bind CreateAttribute_35 (Create%20Constant))
  (bind AttributeCompareOp_34 (Attribute%20Compare%20Op :Operation NotEqual :InB VertexColorDilation))
  (bind Subgraph_1 (Subgraph :Graph /Game/Athena/Helios/Environments/Landscape/WorldBuildingTools/PCGGrass/PCG_ColorAttributeFromEnum.PCG_ColorAttributeFromEnum :Color%20Enum VertexColorChannel))
  (bind Branch_38 (Branch))
  (bind AttributeFilter_0 (Filter%20Attribute%20Elements :Operator GreaterOrEqual :ThresholdAttribute @Last))
  (bind CreateAttribute_17 (Create%20Constant :AttributeTypes (struct :Type Boolean :BoolValue true) :OutputTarget DiscardVertex))
  (bind AddAttribute_18 (Add%20Attribute :bCopyAllAttributes true))
  (bind MergePoints_7 (Merge%20Points))
  (bind CreateAttribute_86 (Create%20Constant :AttributeTypes (struct :DoubleValue 1.0)))
  (bind AttributeMathsOp_22 (Attribute%20Maths%20Op :Operation Abs :In VertexColorDilation))
  (bind CreateAttribute_19 (Create%20Constant))
  (bind AttributeCompareOp_12 (Attribute%20Compare%20Op :Operation Less :InA VertexColorDilation))
  (bind AttributeFilter_40 (Filter%20Attribute%20Elements :Operator GreaterOrEqual :ThresholdAttribute @Last))
  (bind Select_9 (Select))
  (bind Select_21 (Select))
  (bind Distance_2 (Distance :bOutputToAttribute false :bSetDensity true :MaximumDistance 2e+02 :SourceShape Center :TargetShape Center))
  (bind AttributeFilter_63 (Filter%20Attribute%20Elements :AttributeTypes (struct :DoubleValue 1.0) :bUseConstantThreshold true :Operator Equal :ThresholdAttribute @Last))
  (bind AddAttribute_79 (Add%20Attribute :bCopyAllAttributes true))
  (bind MergePoints_28 (Merge%20Points))
  (bind AttributeBooleanOp_127 (Attribute%20Boolean%20Op :InputSource1 DiscardVertex :Operation Not))
  (bind Select_126 (Select))
  (bind AttributeMathsOp_84 (Attribute%20Maths%20Op :InputSource1 $Density :Operation Set))
  (bind Gather_0 (Gather))
  (bind AttributeBooleanOp_0 (Attribute%20Boolean%20Op :InputSource1 DiscardVertex :Operation Not))
  (bind Select_1 (Select :bUseInputB-param VertexColorInvert))
  (bind Gather_45 (Gather))
  (bind AttributeCast_122 (Attribute%20Cast :InputSource Discard :OutputTarget $Density))
  (bind ExtentsModifier_124 (Extents%20Modifier :Extents (1e+01 1e+01 1e+01)))
  ;;[0.05 0.05 0.05] Clean Up
  (bind AttributeBooleanOp_1 (Attribute%20Boolean%20Op :InputSource1 DiscardMaterialID :InputSource2 DiscardVertex :Operation Or :OutputTarget Discard))
  (bind DeleteAttributes_1 (Delete%20Attributes :SelectedAttributes "DiscardMaterialID, DiscardVertex"))
  ;[0.05 0.05 0.05] Hints
  (bind AttributeCast_0 (Attribute%20Cast :OutputType String))
  (bind CreateConstant_1 (Create%20Constant :AttributeTypes (struct :Type String :StringValue "MaxPointsPerMesh limit of ")))
  (bind AttributeStringOp_0 (Attribute%20String%20Op))
  (bind AttributeStringOp_1 (Attribute%20String%20Op))
  (bind AttributeBooleanOp_2 (Attribute%20Boolean%20Op :Operation Not :In IgnoreMaxPointsPerMesh))
  (bind GetElementsCount_0 (Get%20Element%20Count))
  (bind AttributeCompareOp_0 (Attribute%20Compare%20Op :Operation GreaterOrEqual))
  (bind AttributeBooleanOp_3 (Attribute%20Boolean%20Op))
  (bind Branch_1 (Branch))
  (bind Print%20String_0 (Print%20String :bPrefixWithGraph false :bPrefixWithNode false :bPrintPerComponent false :Verbosity Display))
  (reroute Actor%20Attributes ActorAttributes)
  (reroute Actor%20Biome GetAttributeFromPointIndex_119.Attribute)
  (reroute Actor%20Point ActorPoint)
  (reroute Biome%20Attributes BiomeAttributes)
  (reroute Global%20Attributes GlobalAttributes)
  (reroute Volume%20Attributes VolumeAttributes)
  (connect CreateAttribute_13 Out AddAttribute_10 Attributes)
  (connect AddAttribute_26 Out AddAttribute_10 In)
  (connect CreateAttribute_17 Out AddAttribute_18 Attributes)
  (connect AttributeFilter_0 OutsideFilter AddAttribute_18 In)
  (connect CreateAttribute_28 Out AddAttribute_26 Attributes)
  (connect MeshSampler_1 Out AddAttribute_26 In)
  (connect CreateAttribute_32 Out AddAttribute_34 Attributes)
  (connect AttributeFilter_37 OutsideFilter AddAttribute_34 In)
  (connect CreateAttribute_17 Out AddAttribute_79 Attributes)
  (connect AttributeFilter_63 InsideFilter AddAttribute_79 In)
  (connect Gather_0 Out AttributeBooleanOp_0 In)
  (connect Gather_45 Out AttributeBooleanOp_1 InA)
  (connect Gather_45 Out AttributeBooleanOp_1 InB)
  (connect AttributeCompareOp_0 Out AttributeBooleanOp_3 InA)
  (connect AttributeBooleanOp_2 Out AttributeBooleanOp_3 InB)
  (connect AttributeCompareOp_1 Out AttributeBooleanOp_23 InA)
  (connect Branch_37 Output%20B AttributeBooleanOp_36 In)
  (connect MergePoints_28 Out AttributeBooleanOp_127 In)
  (connect DeleteAttributes_4 Out AttributeCast_0 In)
  (connect AttributeMathsOp_84 Out AttributeCast_122 In)
  (connect GetElementsCount_0 Out AttributeCompareOp_0 InA)
  (connect DeleteAttributes_4 Out AttributeCompareOp_0 InB)
  (connect ExecuteBlueprint_0 Material AttributeCompareOp_1 InA)
  (connect CreateAttribute_2 Out AttributeCompareOp_1 InB)
  (connect ExecuteBlueprint_0 Material AttributeCompareOp_10 InA)
  (connect CreateAttribute_12 Out AttributeCompareOp_10 InB)
  (connect CreateAttribute_19 Out AttributeCompareOp_12 InB)
  (connect CreateAttribute_35 Out AttributeCompareOp_34 InA)
  (connect AttributeMathsOp_128 Out AttributeFilter_0 Filter)
  (connect Branch_38 Output%20A AttributeFilter_0 In)
  (connect Branch_36 Output%20B AttributeFilter_37 In)
  (connect AttributeMathsOp_128 Out AttributeFilter_40 Filter)
  (connect Branch_38 Output%20B AttributeFilter_40 In)
  (connect Distance_2 Out AttributeFilter_63 In)
  (connect AttributeMathsOp_1 Out AttributeMathsOp_0 InA)
  (connect GetElementsCount_1 Out AttributeMathsOp_1 In)
  (connect AttributeMathsOp_17 Out AttributeMathsOp_9 InA)
  (connect AttributeMathsOp_17 Out AttributeMathsOp_9 InB)
  (connect CreateAttribute_11 Out AttributeMathsOp_10 InA)
  (connect GetAttributeFromPointIndex_16 Attribute AttributeMathsOp_10 InB)
  (connect Select_104 Out AttributeMathsOp_12 InA)
  (connect AttributeMathsOp_13 Out AttributeMathsOp_12 InB)
  (connect AttributeMathsOp_10 Out AttributeMathsOp_13 In)
  (connect CreateAttribute_14 Out AttributeMathsOp_13 Max)
  (connect CreateAttribute_15 Out AttributeMathsOp_13 Min)
  (connect Actor%20Point Out AttributeMathsOp_17 In)
  (connect Select_126 Out AttributeMathsOp_84 InA)
  (connect CreateAttribute_86 Out AttributeMathsOp_84 InB)
  (connect DeleteAttributes_6 Out AttributeMathsOp_107 InA)
  (connect DeleteAttributes_103 Out AttributeMathsOp_107 InB)
  (connect CreateAttribute_129 Out AttributeMathsOp_128 InB)
  (connect CreateConstant_1 Out AttributeStringOp_0 InA)
  (connect AttributeCast_0 Out AttributeStringOp_0 InB)
  (connect AttributeStringOp_0 Out AttributeStringOp_1 InA)
  (connect Subgraph_0 IsValid Branch_0 bOutputToB)
  (connect Select_7 Out Branch_0 In)
  (connect AttributeBooleanOp_3 Out Branch_1 bOutputToB)
  (connect AttributeBooleanOp_3 Out Branch_1 In)
  (connect Gather_39 Out Branch_24 In)
  (connect AddAttribute_10 Out Branch_36 In)
  (connect MergePoints_35 Out Branch_37 In)
  (connect AttributeCompareOp_34 Out Branch_38 bOutputToB)
  (connect Subgraph_1 Delete%20Attributes%20Out Branch_38 In)
  (connect MeshSampler_1 Out CopyPoints_53 Source)
  (connect Actor%20Point Out CopyPoints_53 Target)
  (connect AttributeBooleanOp_1 Out DeleteAttributes_1 In)
  (connect Global%20Attributes Out DeleteAttributes_4 In)
  (connect Actor%20Attributes Out DeleteAttributes_6 In)
  (connect AttributeBooleanOp_36 Out DeleteAttributes_40 In)
  (connect Branch_37 Output%20A DeleteAttributes_40 In)
  (connect Volume%20Attributes Out DeleteAttributes_103 In)
  (connect Volume%20Attributes Out DeleteAttributes_105 In)
  (connect AttributeMathsOp_22 Out Distance_2 MaximumDistance)
  (connect Select_9 Out Distance_2 Source)
  (connect Select_21 Out Distance_2 Target)
  (connect CopyPoints_53 Out ExtentsModifier_55 In)
  (connect AttributeCast_122 Out ExtentsModifier_124 In)
  (connect AttributeMathsOp_84 Out Gather_0 In)
  (connect MergePoints_7 Out Gather_0 In)
  (connect Branch_36 Output%20A Gather_39 In)
  (connect DeleteAttributes_40 Out Gather_39 In)
  (connect Branch_24 Output%20A Gather_45 In)
  (connect Select_1 Out Gather_45 In)
  (connect GetActorData_0 Out GetAttributeFromIndex_0 In)
  (connect AttributeMathsOp_0 Out GetAttributeFromIndex_0 Index)
  (connect Actor%20Point Out GetAttributeFromPointIndex_6 In)
  (connect AttributeMathsOp_9 Out GetAttributeFromPointIndex_16 In)
  (connect Actor%20Point Out GetAttributeFromPointIndex_119 In)
  (connect MeshSampler_1 Out GetElementsCount_0 In)
  (connect GetActorData_0 Out GetElementsCount_1 In)
  (connect GetAttributeFromIndex_0 Out GetPropertyFromObjectPath_0 In)
  (connect AddAttribute_18 Out MergePoints_7 In)
  (connect AttributeFilter_0 InsideFilter MergePoints_7 In)
  (connect AddAttribute_79 Out MergePoints_28 In)
  (connect AttributeFilter_63 OutsideFilter MergePoints_28 In)
  (connect Select_21 Out MergePoints_28 In)
  (connect AddAttribute_34 Out MergePoints_35 In)
  (connect AttributeFilter_37 InsideFilter MergePoints_35 In)
  (connect Branch_0 Output%20B MeshSampler_1 In)
  (connect Select_0 Out MeshSampler_1 MaxNumSamples)
  (connect Select_3 Out MeshSampler_1 RequestedLODIndex)
  (connect AttributeMathsOp_12 Out MeshSampler_1 SamplingRadius)
  (connect Branch_1 Output%20B Print%20String_0 In)
  (connect AttributeStringOp_1 Out Print%20String_0 PrintString)
  (connect DeleteAttributes_4 Out Select_0 Input%20A)
  (connect CreateConstant_0 Out Select_0 Input%20B)
  (connect Gather_0 Out Select_1 Input%20A)
  (connect AttributeBooleanOp_0 Out Select_1 Input%20B)
  (connect AttributeBooleanOp_23 Out Select_3 bUseInputB)
  (connect Select_8 Out Select_3 Input%20A)
  (connect CreateAttribute_5 Out Select_3 Input%20B)
  (connect GetPropertyFromObjectPath_0 Out Select_7 Input%20A)
  (connect GetAttributeFromPointIndex_6 Attribute Select_7 Input%20B)
  (connect AttributeCompareOp_10 Out Select_8 bUseInputB)
  (connect CreateAttribute_18 Out Select_8 Input%20A)
  (connect CreateAttribute_16 Out Select_8 Input%20B)
  (connect AttributeCompareOp_12 Out Select_9 bUseInputB)
  (connect AttributeFilter_40 OutsideFilter Select_9 Input%20A)
  (connect AttributeFilter_40 InsideFilter Select_9 Input%20B)
  (connect AttributeCompareOp_12 Out Select_21 bUseInputB)
  (connect AttributeFilter_40 InsideFilter Select_21 Input%20A)
  (connect AttributeFilter_40 OutsideFilter Select_21 Input%20B)
  (connect DeleteAttributes_105 Out Select_104 bUseInputB)
  (connect DeleteAttributes_6 Out Select_104 Input%20A)
  (connect AttributeMathsOp_107 Out Select_104 Input%20B)
  (connect AttributeCompareOp_12 Out Select_126 bUseInputB)
  (connect MergePoints_28 Out Select_126 Input%20A)
  (connect AttributeBooleanOp_127 Out Select_126 Input%20B)
  (connect GetPropertyFromObjectPath_0 Out Subgraph_0 In)
  (connect Branch_24 Output%20B Subgraph_1 In)
  (output ActorPoint Actor%20Point)
  (output OutPoints DeleteAttributes_1))
