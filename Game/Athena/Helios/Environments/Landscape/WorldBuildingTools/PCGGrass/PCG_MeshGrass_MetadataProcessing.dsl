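; PCG graph that stamps source-actor metadata onto points for the MeshGrass world-building tool.
; It copies the ActorReference from ActorPoint onto InPoints, then uses it to look up and attach the actor's data layers.
; ActorPoint is filtered on PCG_MeshGrass_CustomScalability, gaining a default Boolean where it is missing.
; Both branches are gathered so that scalability flag is always copied to OutPoints.
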(graph PCG_MeshGrass_MetadataProcessing
  (inputs (InPoints :type Point) (ActorPoint :type Point))
  (outputs (OutPoints :type Point))
  (bind GetAttributeFromPointIndex_3 (Get%20Attribute%20From%20Point%20Index :bForceOutputAttributeToBeInElementsDomain true :InputSource ActorReference))
  (bind AddAttribute_2 (Add%20Attribute :bCopyAllAttributes true))
  (bind PCGGetActorDataLayers_0 (Get%20Actor%20Data%20Layers))
  (bind AddAttribute_1 (Add%20Attribute :bCopyAllAttributes true))
  (bind FilterDataByAttribute_0 (Filter%20Data%20By%20Attribute :Attribute PCG_MeshGrass_CustomScalability))
  (bind AddAttribute_0 (Add%20Attribute :AttributeTypes (struct :Type Boolean) :OutputTarget PCG_MeshGrass_CustomScalability))
  (bind Gather_0 (Gather))
  (bind CopyAttributes_0 (Copy%20Attributes :InputSource PCG_MeshGrass_CustomScalability))
  (connect FilterDataByAttribute_0 OutsideFilter AddAttribute_0 In)
  (connect PCGGetActorDataLayers_0 Out AddAttribute_1 Attributes)
  (connect AddAttribute_2 Out AddAttribute_1 In)
  (connect GetAttributeFromPointIndex_3 Attribute AddAttribute_2 Attributes)
  (connect DefaultInputNode InPoints AddAttribute_2 In)
  (connect Gather_0 Out CopyAttributes_0 Source)
  (connect AddAttribute_1 Out CopyAttributes_0 Target)
  (connect DefaultInputNode ActorPoint FilterDataByAttribute_0 In)
  (connect AddAttribute_0 Out Gather_0 In)
  (connect FilterDataByAttribute_0 InsideFilter Gather_0 In)
  (connect DefaultInputNode ActorPoint GetAttributeFromPointIndex_3 In)
  (connect AddAttribute_2 Out PCGGetActorDataLayers_0 In)
  (output OutPoints CopyAttributes_0))
