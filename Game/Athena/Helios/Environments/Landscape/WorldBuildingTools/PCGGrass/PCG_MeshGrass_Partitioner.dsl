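; PCG graph that partitions mesh grass points and passes them to the grass spawning subgraph.
; It gathers five point sources from overlapping PCG_MeshGrass-tagged actors, culling all but landscape points to actor bounds when partitioned.
; Points with a Mesh attribute are merged, split by the supplied data layers, and looped through SpawnGrass under a BP_MeshGrass label, suffixed with partition coordinates when partitioned.
; The resulting actor attributes, stamped with date data, form PartitionMetadata.
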(graph PCG_MeshGrass_Partitioner
  (params (PartitionedDataLayers :type (Array String)))
  (outputs (PartitionMetadata :type Param))
  ;[0.05 0.05 0.05] Inputs
  (bind CreateAttribute_6 (Create%20Constant :AttributeTypes (struct :Type String :StringValue "BP_MeshGrass") :OutputTarget ActorLabel))
  (bind GetActorData_5 (Get%20Actor%20Data :bComponentsMustOverlapSelf false :Mode GetActorReference :PropertyName Position))
  (bind ExecuteBlueprint_2 (Execute%20Blueprint :BlueprintElementInstance (class:/Game/Athena/Helios/Environments/Landscape/WorldBuildingTools/PCGGrass/PCGBP_GetPartitionCoordsFromPath.PCGBP_GetPartitionCoordsFromPath_C) :BlueprintElementType /Game/Athena/Helios/Environments/Landscape/WorldBuildingTools/PCGGrass/PCGBP_GetPartitionCoordsFromPath.PCGBP_GetPartitionCoordsFromPath_C))
  (bind AttributeStringOp_3 (Attribute%20String%20Op))
  (bind ExecuteBlueprint_8 (Execute%20Blueprint :BlueprintElementInstance (class:/Game/Athena/Helios/Environments/Landscape/WorldBuildingTools/PCGGrass/PCGBP_IsPartitioned.PCGBP_IsPartitioned_C) :BlueprintElementType /Game/Athena/Helios/Environments/Landscape/WorldBuildingTools/PCGGrass/PCGBP_IsPartitioned.PCGBP_IsPartitioned_C))
  (bind Select_15 (Select))
  (bind GetPCGComponentData_3 (Get%20PCG%20Component%20Data :ActorSelector (struct :ActorFilter AllWorldActors :bMustOverlapSelf true :ActorSelectionTag PCG_MeshGrass :bSelectMultiple true) :ExpectedPins [MeshPoints LandscapePoints SplinePoints TexturePoints CustomPoints]))
  (bind FilterDataByType_9 (Filter%20Data%20By%20Type :TargetType (struct :Ids [(struct :Struct /Script/PCG.PCGDataTypeInfoPoint)])))
  (bind FilterDataByType_5 (Filter%20Data%20By%20Type :TargetType (struct :Ids [(struct :Struct /Script/PCG.PCGDataTypeInfoPoint)])))
  (bind FilterDataByType_7 (Filter%20Data%20By%20Type :TargetType (struct :Ids [(struct :Struct /Script/PCG.PCGDataTypeInfoPoint)])))
  (bind FilterDataByType_6 (Filter%20Data%20By%20Type :TargetType (struct :Ids [(struct :Struct /Script/PCG.PCGDataTypeInfoPoint)])))
  (bind FilterDataByType_1 (Filter%20Data%20By%20Type :TargetType (struct :Ids [(struct :Struct /Script/PCG.PCGDataTypeInfoPoint)])))
  ;[0.05 0.05 0.05] Points Partitioning
  (bind Branch_5 (Branch))
  (bind CullPointsOutsideActorBounds_6 (Cull%20Points%20Outside%20Actor%20Bounds))
  (bind FilterDataByAttribute_19 (Filter%20Data%20By%20Attribute :Attribute Mesh))
  (bind MergePoints_56 (Merge%20Points))
  (bind CreateAttribute_5 (Create%20Constant :AttributeTypes (struct :Type String) :OutputTarget PartitionedDataLayers))
  (bind MergeAttributes_6 (Merge%20Attributes :In PartitionedDataLayers))
  (bind AttributeCast_4 (Attribute%20Cast :OutputType SoftObjectPath))
  (bind PartitionByActorDataLayers_1 (Partition%20by%20Actor%20Data%20Layers :IncludedDataLayers (struct :bAsInput true)))
  (bind Loop_17 (Loop :Graph /Game/Athena/Helios/Environments/Landscape/WorldBuildingTools/PCGGrass/PCG_MeshGrass_SpawnGrass.PCG_MeshGrass_SpawnGrass))
  ;[0.05 0.05 0.05] Output
  (bind ExecuteBlueprint_0 (Execute%20Blueprint :BlueprintElementInstance (class:/Game/Athena/Helios/Environments/Landscape/WorldBuildingTools/PCGGrass/PCGBP_GetDate.PCGBP_GetDate_C) :BlueprintElementType /Game/Athena/Helios/Environments/Landscape/WorldBuildingTools/PCGGrass/PCGBP_GetDate.PCGBP_GetDate_C))
  (bind MergeAttributes_0 (Merge%20Attributes))
  (bind AddAttribute_2 (Add%20Attribute :bCopyAllAttributes true))
  (reroute CustomPoints FilterDataByType_1.InsideFilter)
  (reroute IsPartitioned ExecuteBlueprint_8)
  (reroute LandscapePoints FilterDataByType_9.InsideFilter)
  (reroute MeshPoints FilterDataByType_5.InsideFilter)
  (reroute OutputActorLabel Select_15)
  (reroute PartitionActor GetActorData_5)
  (reroute PartitionAttributes OutputActorLabel)
  (reroute SplinePoints FilterDataByType_7.InsideFilter)
  (reroute TexturePoints FilterDataByType_6.InsideFilter)
  (connect ExecuteBlueprint_0 Out AddAttribute_2 Attributes)
  (connect MergeAttributes_0 Out AddAttribute_2 In)
  (connect MergeAttributes_6 Out AttributeCast_4 In)
  (connect CreateAttribute_6 Out AttributeStringOp_3 InA)
  (connect ExecuteBlueprint_2 Out AttributeStringOp_3 InB)
  (connect IsPartitioned Out Branch_5 bOutputToB)
  (connect CustomPoints Out Branch_5 In)
  (connect TexturePoints Out Branch_5 In)
  (connect SplinePoints Out Branch_5 In)
  (connect MeshPoints Out Branch_5 In)
  (connect Branch_5 Output%20B CullPointsOutsideActorBounds_6 In)
  (connect PartitionActor Out ExecuteBlueprint_2 In)
  (connect Branch_5 Output%20A FilterDataByAttribute_19 In)
  (connect CullPointsOutsideActorBounds_6 Out FilterDataByAttribute_19 In)
  (connect LandscapePoints Out FilterDataByAttribute_19 In)
  (connect GetPCGComponentData_3 CustomPoints FilterDataByType_1 In)
  (connect GetPCGComponentData_3 MeshPoints FilterDataByType_5 In)
  (connect GetPCGComponentData_3 TexturePoints FilterDataByType_6 In)
  (connect GetPCGComponentData_3 SplinePoints FilterDataByType_7 In)
  (connect GetPCGComponentData_3 LandscapePoints FilterDataByType_9 In)
  (connect PartitionByActorDataLayers_1 DataLayerPartitions Loop_17 DataLayers)
  (connect PartitionByActorDataLayers_1 Out Loop_17 InPoints)
  (connect PartitionAttributes Out Loop_17 PartitionAttributes)
  (connect Loop_17 OutActors MergeAttributes_0 In)
  (connect CreateAttribute_5 Out MergeAttributes_6 In)
  (connect FilterDataByAttribute_19 InsideFilter MergePoints_56 In)
  (connect MergePoints_56 Out PartitionByActorDataLayers_1 In)
  (connect AttributeCast_4 Out PartitionByActorDataLayers_1 IncludedDataLayers)
  (connect IsPartitioned Out Select_15 bUseInputB)
  (connect CreateAttribute_6 Out Select_15 Input%20A)
  (connect AttributeStringOp_3 Out Select_15 Input%20B)
  (output PartitionMetadata AddAttribute_2))
