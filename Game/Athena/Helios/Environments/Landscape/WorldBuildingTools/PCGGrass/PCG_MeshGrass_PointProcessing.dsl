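; PCG graph that post-processes mesh grass points, setting their transforms and bounds.
; Variety points get noise-based scale per mesh, with volume overrides, before actor-driven offset, rotation, scale and gradient scaling.
; Bounds are then sized from mesh bounds and Dominance, and points are optionally copied to the actor's location and filtered by topology.
; ForceSurfaceOrientation rebuilds rotation around each point's up axis with the rotation offset, before temporary attributes are removed.
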(graph PCG_MeshGrass_PointProcessing
  (params (TransformToWorld :type Bool :default true) (ForceNoVariety :type Bool) (ForceVariety :type Bool))
  (inputs (InPoints :type Point) (ActorPoint :type Point) (ActorAttributes :type Param) (VolumeAttributes :type Param) (BiomeAttributes :type Param) (GlobalAttributes :type Param))
  (outputs (OutPoints :type Point))
  ;[0.05 0.05 0.05] Copy MeshPoints at Actor location
  (bind FilterDataByAttribute_1 (Filter%20Data%20By%20Attribute :Attribute ForceSurfaceOrientation))
  (bind DeleteAttributes_14 (Delete%20Attributes :Operation KeepSelectedAttributes :SelectedAttributes "ForceSurfaceOrientation"))
  (bind AttributeBooleanOp_1 (Attribute%20Boolean%20Op :Operation Not))
  (bind DeleteAttributes_12 (Delete%20Attributes :Operation KeepSelectedAttributes :SelectedAttributes "WorldSpaceTransform"))
  (bind AttributeBooleanOp_2 (Attribute%20Boolean%20Op))
  (bind Branch_5 (Branch :bOutputToB-param TransformToWorld))
  (bind CopyPoints_0 (Copy%20Points :AttributeInheritance SourceOnly :bCopyEachSourceOnEveryTarget false :ColorInheritance Source :ScaleInheritance Source :TagInheritance Source))
  (bind DeleteAttributes_3 (Delete%20Attributes :Operation KeepSelectedAttributes :SelectedAttributes "LocationOffsetZ"))
  (bind AttributeMathsOp_3 (Attribute%20Maths%20Op :InputSource1 $Position.Z))
  (bind Select_2 (Select))
  (bind Gather_16 (Gather))
  ;[0.05 0.05 0.05] Transform
  (bind DeleteAttributes_2 (Delete%20Attributes :bTokenizeOnWhiteSpace true :Operation KeepSelectedAttributes :SelectedAttributes "ScaleByGradient"))
  (bind DeleteAttributes_18 (Delete%20Attributes :bTokenizeOnWhiteSpace true :Operation KeepSelectedAttributes :SelectedAttributes "ScaleMinMax"))
  (bind BreakVectorAttribute_1 (Break%20Vector%20Attribute))
  (bind DeleteAttributes_4 (Delete%20Attributes :bTokenizeOnWhiteSpace true :Operation KeepSelectedAttributes :SelectedAttributes "RotationOffset"))
  (bind DeleteAttributes_8 (Delete%20Attributes :Operation KeepSelectedAttributes :SelectedAttributes "WorldSpaceTransform"))
  (bind AttributeBooleanOp_0 (Attribute%20Boolean%20Op :InB TransformToWorld))
  (bind FilterDataByAttribute_0 (Filter%20Data%20By%20Attribute :Attribute ForceSurfaceOrientation))
  (bind DeleteAttributes_13 (Delete%20Attributes :Operation KeepSelectedAttributes :SelectedAttributes "ForceSurfaceOrientation"))
  (bind AttributeBooleanOp_5 (Attribute%20Boolean%20Op :Operation Not))
  (bind AttributeBooleanOp_7 (Attribute%20Boolean%20Op))
  (bind CreateConstant_1 (Create%20Constant))
  (bind DeleteAttributes_10 (Delete%20Attributes :Operation KeepSelectedAttributes :SelectedAttributes "LocationOffsetZ"))
  (bind Select_1 (Select))
  (bind MakeVectorAttribute_11 (Make%20Vector%20Attribute :OutputType Vector))
  (bind CreateConstant_0 (Create%20Constant :AttributeTypes (struct :Type Quaternion)))
  (bind ExtentsModifier_0 (Extents%20Modifier))
  (bind Branch_14 (Branch :bOutputToB-param ForceNoVariety))
  (bind ExecuteBlueprint_0 (Execute%20Blueprint :BlueprintElementInstance (class:/Game/Athena/Helios/Environments/Landscape/WorldBuildingTools/PCGGrass/PCGBP_IsTargetClass.PCGBP_IsTargetClass_C :TargetClass /Game/Athena/Helios/Environments/Landscape/WorldBuildingTools/PCGGrass/BP_MeshGrass_TextureGenerator.BP_MeshGrass_TextureGenerator_C) :BlueprintElementType /Game/Athena/Helios/Environments/Landscape/WorldBuildingTools/PCGGrass/PCGBP_IsTargetClass.PCGBP_IsTargetClass_C)) ; Assuming UseVariety for this because the determination would be too complex
  (bind Subgraph_53 (Subgraph :Graph /Game/Athena/Helios/Environments/Landscape/WorldBuildingTools/PCGGrass/PCG_MeshGrass_SelectBiomeAttribute.PCG_MeshGrass_SelectBiomeAttribute))
  (bind AttributeBooleanOp_6 (Attribute%20Boolean%20Op :Operation Or))
  (bind AttributeBooleanOp_3 (Attribute%20Boolean%20Op :Operation Or :InB ForceVariety))
  (bind Branch_52 (Branch))
  (bind DeleteAttributes_29 (Delete%20Attributes :bTokenizeOnWhiteSpace true :Operation KeepSelectedAttributes :SelectedAttributes "VolumeOverride"))
  (bind Branch_31 (Branch))
  (bind DeleteAttributes_30 (Delete%20Attributes :bTokenizeOnWhiteSpace true :Operation KeepSelectedAttributes :SelectedAttributes "VolumePerInstanceOverride"))
  (bind Branch_32 (Branch))
  (bind DeleteAttributes_33 (Delete%20Attributes :Operation KeepSelectedAttributes :SelectedAttributes "Scale Min Factor"))
  (bind AttributeMathsOp_28 (Attribute%20Maths%20Op :InputSource1 ScaleMin :Operation Multiply))
  (bind DeleteAttributes_34 (Delete%20Attributes :Operation KeepSelectedAttributes :SelectedAttributes "Scale Max Factor"))
  (bind AttributeMathsOp_35 (Attribute%20Maths%20Op :InputSource1 ScaleMax :Operation Multiply))
  (bind CreateAttribute_38 (Create%20Constant :AttributeTypes (struct :Type Boolean) :OutputTarget IsLocallyControlled))
  (bind AddAttribute_40 (Add%20Attribute :AttributeTypes (struct :DoubleValue 1.0) :OutputTarget VolumeScaleMin))
  (bind AddAttribute_41 (Add%20Attribute :AttributeTypes (struct :DoubleValue 1.0) :OutputTarget VolumeScaleMax))
  (bind CreateAttribute_39 (Create%20Constant :AttributeTypes (struct :Type Boolean :BoolValue true) :OutputTarget IsLocallyControlled))
  (bind AddAttribute_42 (Add%20Attribute :AttributeTypes (struct :DoubleValue 1.0) :OutputTarget VolumeScaleMin))
  (bind AddAttribute_43 (Add%20Attribute :AttributeTypes (struct :DoubleValue 1.0) :OutputTarget VolumeScaleMax))
  (bind MergeAttributes_44 (Merge%20Attributes))
  (bind MatchAndSetAttributes_37 (Match%20And%20Set%20Attributes :bMatchAttributes true :InputAttribute IsLocallyControlled :MatchAttribute IsLocallyControlled))
  (bind AttributeMathsOp_36 (Attribute%20Maths%20Op :InputSource1 ScaleMin :InputSource2 VolumeScaleMin :Operation Multiply))
  (bind AttributeMathsOp_45 (Attribute%20Maths%20Op :InputSource1 ScaleMax :InputSource2 VolumeScaleMax :Operation Multiply))
  (bind DeleteAttributes_7 (Delete%20Attributes :SelectedAttributes "VolumeScaleMin,VolumeScaleMax"))
  (bind AttributePartition_21 (Attribute%20Partition :bTokenizeOnWhiteSpace true :PartitionAttributeSelectors [Mesh]))
  (bind AttributeNoise_24 (Attribute%20Noise :OutputTarget ScaleNoise))
  (bind AttributeMathsOp_23 (Attribute%20Maths%20Op :InputSource1 ScaleMax :InputSource2 ScaleMin :Operation Subtract :OutputTarget ScaleRatio))
  (bind AttributeMathsOp_25 (Attribute%20Maths%20Op :InputSource1 ScaleRatio :InputSource2 ScaleNoise :Operation Multiply :OutputTarget RandomizedScale))
  (bind AttributeMathsOp_26 (Attribute%20Maths%20Op :InputSource1 RandomizedScale :InputSource2 ScaleMin))
  (bind AttributeMathsOp_27 (Attribute%20Maths%20Op :InputSource1 $Scale :InputSource2 RandomizedScale :Operation Set))
  (bind MergePoints_22 (Merge%20Points))
  (bind DeleteAttributes_0 (Delete%20Attributes :SelectedAttributes "ScaleNoise,ScaleRatio,RandomizedScale,ScaleMin,ScaleMax"))
  (bind Gather_0 (Gather))
  (bind AttributeMathsOp_2 (Attribute%20Maths%20Op :InputSource1 $Rotation :Operation Set))
  (bind Select_0 (Select))
  (bind AttributeRotatorOp_7 (Attribute%20Rotator%20Op :Operation Invert))
  (bind Select_5 (Select :bUseInputB-param RandomizeRotationOffset))
  (bind AttributeBooleanOp_16 (Attribute%20Boolean%20Op))
  (bind Select_6 (Select :bUseInputB-param TransformToWorld))
  (bind TransformPoints_17 (Transform%20Points :RotationMax (struct :Yaw 3.6e+02)))
  (bind Branch_3 (Branch))
  (bind DeleteAttributes_9 (Delete%20Attributes :bTokenizeOnWhiteSpace true :Operation KeepSelectedAttributes :SelectedAttributes "GradientScalePower"))
  (bind AttributeMathsOp_1 (Attribute%20Maths%20Op :InputSource1 ScaleGradient :Operation Pow))
  (bind DeleteAttributes_1 (Delete%20Attributes :bTokenizeOnWhiteSpace true :Operation KeepSelectedAttributes :SelectedAttributes "GradientScaleMinMax"))
  (bind BreakVectorAttribute_2 (Break%20Vector%20Attribute))
  (bind AttributeRemap_3 (Attribute%20Remap :bAllowInverseRange false :InputSource ScaleGradient))
  (bind AttributeMathsOp_0 (Attribute%20Maths%20Op :InputSource1 $Scale :InputSource2 ScaleGradient :Operation Multiply))
  (bind DeleteAttributes_6 (Delete%20Attributes :SelectedAttributes "ScaleGradient"))
  (bind Gather_4 (Gather))
  (bind AttributeBooleanOp_8 (Attribute%20Boolean%20Op))
  ;[0.05 0.05 0.05] Point Bounds
  (bind AttributePartition_2 (Attribute%20Partition :PartitionAttributeSelectors [Mesh]))
  (bind GetAttributeFromPointIndex_3 (Get%20Attribute%20From%20Point%20Index :bForceOutputAttributeToBeInElementsDomain true :InputSource Mesh :OutputAttributeName Mesh))
  (bind MergeAttributes_4 (Merge%20Attributes))
  (bind ExecuteBlueprint_6 (Execute%20Blueprint :BlueprintElementInstance (class:/Game/Athena/Helios/Environments/Landscape/WorldBuildingTools/PCGGrass/PCGBP_GetMeshBounds.PCGBP_GetMeshBounds_C) :BlueprintElementType /Game/Athena/Helios/Environments/Landscape/WorldBuildingTools/PCGGrass/PCGBP_GetMeshBounds.PCGBP_GetMeshBounds_C))
  (bind AddAttribute_9 (Add%20Attribute :bCopyAllAttributes true))
  (bind MatchAndSetAttributes_10 (Match%20And%20Set%20Attributes :bMatchAttributes true :InputAttribute Mesh :MatchAttribute Mesh))
  (bind AttributeMathsOp_6 (Attribute%20Maths%20Op :InputSource1 MeshBounds :InputSource2 Dominance :Operation Multiply :OutputTarget MeshBounds))
  (bind AttributeMathsOp_21 (Attribute%20Maths%20Op :InputSource1 MeshBounds :InputSource2 $Scale :Operation Multiply :OutputTarget ScaledMeshBounds))
  (bind AttributeMathsOp_13 (Attribute%20Maths%20Op :InputSource1 $BoundsMin :Operation Multiply))
  (bind AttributeMathsOp_15 (Attribute%20Maths%20Op :InputSource1 $BoundsMax :Operation Multiply))
  (bind CreateAttribute_18 (Create%20Constant :AttributeTypes (struct :DoubleValue 0.5)))
  (bind CreateAttribute_20 (Create%20Constant :AttributeTypes (struct :DoubleValue 0.36) :OutputTarget MeshBoundsFactor))
  (bind AttributeMathsOp_19 (Attribute%20Maths%20Op :Operation Multiply))
  (bind ExtentsModifier_16 (Extents%20Modifier :Extents (0.5 0.5 0.5) :Mode Multiply))
  (bind DeleteAttributes_5 (Delete%20Attributes :SelectedAttributes "MeshBounds,ScaledMeshBounds"))
  ;[0.05 0.05 0.05] Inputs
  (bind GetAttributeFromPointIndex_66 (Get%20Attribute%20From%20Point%20Index :bForceOutputAttributeToBeInElementsDomain true :InputSource Biome :OutputAttributeName Biome))
  ;[0.05 0.05 0.05] Topological Filters
  (bind DeleteAttributes_19 (Delete%20Attributes :Operation KeepSelectedAttributes :SelectedAttributes "FilterTopologyInWorldSpace"))
  (bind Branch_0 (Branch))
  (bind Subgraph_0 (Subgraph :Graph /Game/Athena/Helios/Environments/Landscape/WorldBuildingTools/PCGGrass/PCG_MeshGrass_FilterTopology.PCG_MeshGrass_FilterTopology))
  (bind Gather_1 (Gather))
  ;[0.05 0.05 0.05] Surface Aligned Random Yaw
  (bind AttributeCast_1 (Attribute%20Cast :InputSource $Rotation :OutputTarget customRotation :OutputType Rotator))
  (bind CreateConstant_3 (Create%20Constant :AttributeTypes (struct :Type Quaternion)))
  (bind AttributeMathsOp_11 (Attribute%20Maths%20Op :InputSource1 $Rotation :Operation Set))
  (bind CopyAttributes_2 (Copy%20Attributes :InputSource $Rotation.Forward :OutputTarget FAxis))
  (bind AttributeMathsOp_16 (Attribute%20Maths%20Op :InputSource1 FAxis :Operation Negate))
  (bind MakeRotatorAttribute_8 (Make%20Rotator%20Attribute :InputSource1 FAxis :InputSource2 customRotation.Up :InputSource3 customRotation.Up :Operation MakeRotFromXZ :OutputTarget $Rotation))
  (bind MakeRotatorAttribute_9 (Make%20Rotator%20Attribute :InputSource1 $Rotation.Right :InputSource2 customRotation.Up :InputSource3 customRotation.Up :Operation MakeRotFromYZ :OutputTarget $Rotation))
  (bind DeleteAttributes_16 (Delete%20Attributes :bTokenizeOnWhiteSpace true :Operation KeepSelectedAttributes :SelectedAttributes "RotationOffset"))
  (bind AttributeRotatorOp_1 (Attribute%20Rotator%20Op :Operation Invert))
  (bind Select_7 (Select :bUseInputB-param RandomizeRotationOffset))
  (bind AttributeRotatorOp_2 (Attribute%20Rotator%20Op))
  (bind FilterDataByAttribute_2 (Filter%20Data%20By%20Attribute :Attribute ForceSurfaceOrientation))
  (bind DeleteAttributes_15 (Delete%20Attributes :Operation KeepSelectedAttributes :SelectedAttributes "ForceSurfaceOrientation"))
  (bind Select_3 (Select))
  (bind MakeRotatorAttribute_3 (Make%20Rotator%20Attribute :InputSource1 $Rotation.Forward :InputSource2 customRotation.Up :InputSource3 customRotation.Up :Operation MakeRotFromXZ :OutputTarget $Rotation))
  (bind MakeRotatorAttribute_5 (Make%20Rotator%20Attribute :InputSource1 $Rotation.Right :InputSource2 customRotation.Up :InputSource3 customRotation.Up :Operation MakeRotFromYZ :OutputTarget $Rotation))
  ;[0.05 0.05 0.05] Cleanup
  (bind DeleteAttributes_11 (Delete%20Attributes :SelectedAttributes "IsLocallyControlled,Variety,TextureStamped"))
  (reroute Actor%20Attributes ActorAttributes)
  (reroute Actor%20Biome GetAttributeFromPointIndex_66.Attribute)
  (reroute Actor%20Point ActorPoint)
  (reroute Biome%20Attributes BiomeAttributes)
  (reroute Global%20Attributes GlobalAttributes)
  (reroute In%20Points InPoints)
  (reroute Volume%20Attributes VolumeAttributes)
  (connect ExecuteBlueprint_6 Bounds AddAttribute_9 Attributes)
  (connect MergeAttributes_4 Out AddAttribute_9 In)
  (connect CreateAttribute_38 Out AddAttribute_40 In)
  (connect AddAttribute_40 Out AddAttribute_41 In)
  (connect DeleteAttributes_33 Out AddAttribute_42 Attributes)
  (connect CreateAttribute_39 Out AddAttribute_42 In)
  (connect DeleteAttributes_34 Out AddAttribute_43 Attributes)
  (connect AddAttribute_42 Out AddAttribute_43 In)
  (connect DeleteAttributes_8 Out AttributeBooleanOp_0 InA)
  (connect DeleteAttributes_14 Out AttributeBooleanOp_1 In)
  (connect DeleteAttributes_12 Out AttributeBooleanOp_2 InA)
  (connect AttributeBooleanOp_1 Out AttributeBooleanOp_2 InB)
  (connect AttributeBooleanOp_6 Out AttributeBooleanOp_3 InA)
  (connect DeleteAttributes_13 Out AttributeBooleanOp_5 In)
  (connect ExecuteBlueprint_0 Out AttributeBooleanOp_6 InA)
  (connect Subgraph_53 Out%20Attribute AttributeBooleanOp_6 InB)
  (connect AttributeBooleanOp_5 Out AttributeBooleanOp_7 InA)
  (connect AttributeBooleanOp_0 Out AttributeBooleanOp_7 InB)
  (connect DeleteAttributes_8 Out AttributeBooleanOp_8 InA)
  (connect DeleteAttributes_13 Out AttributeBooleanOp_8 InB)
  (connect DeleteAttributes_8 Out AttributeBooleanOp_16 InA)
  (connect AttributeBooleanOp_5 Out AttributeBooleanOp_16 InB)
  (connect Gather_1 Out AttributeCast_1 In)
  (connect AttributeMathsOp_1 Out AttributeMathsOp_0 InA)
  (connect AttributeRemap_3 Out AttributeMathsOp_0 InB)
  (connect Branch_3 Output%20B AttributeMathsOp_1 InA)
  (connect DeleteAttributes_9 Out AttributeMathsOp_1 InB)
  (connect Gather_0 Out AttributeMathsOp_2 InA)
  (connect CreateConstant_0 Out AttributeMathsOp_2 InB)
  (connect CopyPoints_0 Out AttributeMathsOp_3 InA)
  (connect DeleteAttributes_3 Out AttributeMathsOp_3 InB)
  (connect MatchAndSetAttributes_10 Out AttributeMathsOp_6 InA)
  (connect MatchAndSetAttributes_10 Out AttributeMathsOp_6 InB)
  (connect AttributeCast_1 Out AttributeMathsOp_11 InA)
  (connect CreateConstant_3 Out AttributeMathsOp_11 InB)
  (connect AttributeMathsOp_21 Out AttributeMathsOp_13 InA)
  (connect AttributeMathsOp_21 Out AttributeMathsOp_13 InB)
  (connect AttributeMathsOp_13 Out AttributeMathsOp_15 InA)
  (connect AttributeMathsOp_21 Out AttributeMathsOp_15 InB)
  (connect CopyAttributes_2 Out AttributeMathsOp_16 In)
  (connect CreateAttribute_18 Out AttributeMathsOp_19 InA)
  (connect CreateAttribute_20 Out AttributeMathsOp_19 InB)
  (connect AttributeMathsOp_6 Out AttributeMathsOp_21 InA)
  (connect AttributeMathsOp_6 Out AttributeMathsOp_21 InB)
  (connect AttributeNoise_24 Out AttributeMathsOp_23 InA)
  (connect AttributeNoise_24 Out AttributeMathsOp_23 InB)
  (connect AttributeMathsOp_23 Out AttributeMathsOp_25 InA)
  (connect AttributeMathsOp_23 Out AttributeMathsOp_25 InB)
  (connect AttributeMathsOp_25 Out AttributeMathsOp_26 InA)
  (connect AttributeMathsOp_25 Out AttributeMathsOp_26 InB)
  (connect AttributeMathsOp_26 Out AttributeMathsOp_27 InA)
  (connect AttributeMathsOp_26 Out AttributeMathsOp_27 InB)
  (connect Branch_32 Output%20A AttributeMathsOp_28 InA)
  (connect DeleteAttributes_33 Out AttributeMathsOp_28 InB)
  (connect AttributeMathsOp_28 Out AttributeMathsOp_35 InA)
  (connect DeleteAttributes_34 Out AttributeMathsOp_35 InB)
  (connect MatchAndSetAttributes_37 Out AttributeMathsOp_36 InA)
  (connect MatchAndSetAttributes_37 Out AttributeMathsOp_36 InB)
  (connect AttributeMathsOp_36 Out AttributeMathsOp_45 InA)
  (connect AttributeMathsOp_36 Out AttributeMathsOp_45 InB)
  (connect AttributePartition_21 Out AttributeNoise_24 In)
  (connect Gather_4 Out AttributePartition_2 In)
  (connect AttributeMathsOp_35 Out AttributePartition_21 In)
  (connect Branch_31 Output%20A AttributePartition_21 In)
  (connect DeleteAttributes_7 Out AttributePartition_21 In)
  (connect AttributeMathsOp_1 Out AttributeRemap_3 In)
  (connect BreakVectorAttribute_2 Y AttributeRemap_3 OutRangeMax)
  (connect BreakVectorAttribute_2 X AttributeRemap_3 OutRangeMin)
  (connect DeleteAttributes_16 Out AttributeRotatorOp_1 In)
  (connect Select_7 Out AttributeRotatorOp_2 InA)
  (connect MakeRotatorAttribute_9 Out AttributeRotatorOp_2 InB)
  (connect DeleteAttributes_4 Out AttributeRotatorOp_7 In)
  (connect DeleteAttributes_19 Out Branch_0 bOutputToB)
  (connect Gather_16 Out Branch_0 In)
  (connect DeleteAttributes_2 Out Branch_3 bOutputToB)
  (connect TransformPoints_17 Out Branch_3 In)
  (connect DeleteAttributes_5 Out Branch_5 In)
  (connect ExtentsModifier_0 Out Branch_14 In)
  (connect DeleteAttributes_29 Out Branch_31 bOutputToB)
  (connect Branch_52 Output%20B Branch_31 In)
  (connect DeleteAttributes_30 Out Branch_32 bOutputToB)
  (connect Branch_31 Output%20B Branch_32 In)
  (connect AttributeBooleanOp_3 Out Branch_52 bOutputToB)
  (connect Branch_14 Output%20A Branch_52 In)
  (connect DeleteAttributes_18 Out BreakVectorAttribute_1 In)
  (connect DeleteAttributes_1 Out BreakVectorAttribute_2 In)
  (connect AttributeMathsOp_11 Out CopyAttributes_2 Source)
  (connect AttributeMathsOp_11 Out CopyAttributes_2 Target)
  (connect AttributeBooleanOp_2 Out CopyPoints_0 RotationInheritance)
  (connect Branch_5 Output%20B CopyPoints_0 Source)
  (connect Actor%20Point Out CopyPoints_0 Target)
  (connect MergePoints_22 Out DeleteAttributes_0 In)
  (connect Branch_3 Output%20B DeleteAttributes_1 Execution%20Dependency)
  (connect Actor%20Attributes Out DeleteAttributes_1 In)
  (connect Actor%20Attributes Out DeleteAttributes_2 In)
  (connect Actor%20Attributes Out DeleteAttributes_3 In)
  (connect Actor%20Attributes Out DeleteAttributes_4 In)
  (connect ExtentsModifier_16 Out DeleteAttributes_5 In)
  (connect AttributeMathsOp_0 Out DeleteAttributes_6 In)
  (connect Branch_3 Output%20A DeleteAttributes_6 In)
  (connect AttributeMathsOp_45 Out DeleteAttributes_7 In)
  (connect Actor%20Attributes Out DeleteAttributes_8 In)
  (connect Actor%20Attributes Out DeleteAttributes_9 In)
  (connect Actor%20Attributes Out DeleteAttributes_10 In)
  (connect Select_3 Out DeleteAttributes_11 In)
  (connect Actor%20Attributes Out DeleteAttributes_12 In)
  (connect FilterDataByAttribute_0 InsideFilter DeleteAttributes_13 In)
  (connect FilterDataByAttribute_1 InsideFilter DeleteAttributes_14 In)
  (connect FilterDataByAttribute_2 InsideFilter DeleteAttributes_15 In)
  (connect Actor%20Attributes Out DeleteAttributes_16 In)
  (connect Actor%20Attributes Out DeleteAttributes_18 In)
  (connect Actor%20Attributes Out DeleteAttributes_19 In)
  (connect Volume%20Attributes Out DeleteAttributes_29 In)
  (connect Volume%20Attributes Out DeleteAttributes_30 In)
  (connect Volume%20Attributes Out DeleteAttributes_33 In)
  (connect Volume%20Attributes Out DeleteAttributes_34 In)
  (connect MergeAttributes_4 Out ExecuteBlueprint_6 Meshes)
  (connect In%20Points Out ExtentsModifier_0 In)
  (connect AttributeMathsOp_19 Out ExtentsModifier_16 Extents)
  (connect AttributeMathsOp_15 Out ExtentsModifier_16 In)
  (connect Actor%20Attributes Out FilterDataByAttribute_0 In)
  (connect Actor%20Attributes Out FilterDataByAttribute_1 In)
  (connect Actor%20Attributes Out FilterDataByAttribute_2 In)
  (connect Branch_14 Output%20B Gather_0 In)
  (connect Branch_52 Output%20A Gather_0 In)
  (connect DeleteAttributes_0 Out Gather_0 In)
  (connect Branch_0 Output%20A Gather_1 In)
  (connect Subgraph_0 OutPoints Gather_1 In)
  (connect DeleteAttributes_6 Out Gather_4 In)
  (connect Branch_5 Output%20A Gather_16 In)
  (connect Select_2 Out Gather_16 In)
  (connect AttributePartition_2 Out GetAttributeFromPointIndex_3 In)
  (connect Actor%20Point Out GetAttributeFromPointIndex_66 In)
  (connect AttributeMathsOp_16 Out MakeRotatorAttribute_3 X)
  (connect AttributeMathsOp_16 Out MakeRotatorAttribute_3 Z)
  (connect MakeRotatorAttribute_3 Out MakeRotatorAttribute_5 Y)
  (connect MakeRotatorAttribute_3 Out MakeRotatorAttribute_5 Z)
  (connect AttributeMathsOp_16 Out MakeRotatorAttribute_8 X)
  (connect AttributeMathsOp_16 Out MakeRotatorAttribute_8 Z)
  (connect MakeRotatorAttribute_8 Out MakeRotatorAttribute_9 Y)
  (connect MakeRotatorAttribute_8 Out MakeRotatorAttribute_9 Z)
  (connect Select_1 Out MakeVectorAttribute_11 Z)
  (connect Gather_4 Out MatchAndSetAttributes_10 In)
  (connect AddAttribute_9 Out MatchAndSetAttributes_10 Match%20Data)
  (connect Branch_32 Output%20B MatchAndSetAttributes_37 In)
  (connect MergeAttributes_44 Out MatchAndSetAttributes_37 Match%20Data)
  (connect GetAttributeFromPointIndex_3 Attribute MergeAttributes_4 In)
  (connect AddAttribute_41 Out MergeAttributes_44 In)
  (connect AddAttribute_43 Out MergeAttributes_44 In)
  (connect AttributeMathsOp_27 Out MergePoints_22 In)
  (connect DeleteAttributes_8 Out Select_0 bUseInputB)
  (connect Gather_0 Out Select_0 Input%20A)
  (connect AttributeMathsOp_2 Out Select_0 Input%20B)
  (connect AttributeBooleanOp_7 Out Select_1 bUseInputB)
  (connect DeleteAttributes_10 Out Select_1 Input%20A)
  (connect CreateConstant_1 Out Select_1 Input%20B)
  (connect AttributeBooleanOp_2 Out Select_2 bUseInputB)
  (connect CopyPoints_0 Out Select_2 Input%20A)
  (connect AttributeMathsOp_3 Out Select_2 Input%20B)
  (connect DeleteAttributes_15 Out Select_3 bUseInputB)
  (connect Gather_1 Out Select_3 Input%20A)
  (connect AttributeRotatorOp_2 Out Select_3 Input%20B)
  (connect DeleteAttributes_4 Out Select_5 Input%20A)
  (connect AttributeRotatorOp_7 Out Select_5 Input%20B)
  (connect DeleteAttributes_8 Out Select_6 Input%20A)
  (connect AttributeBooleanOp_16 Out Select_6 Input%20B)
  (connect DeleteAttributes_16 Out Select_7 Input%20A)
  (connect AttributeRotatorOp_1 Out Select_7 Input%20B)
  (connect Actor%20Attributes Out Subgraph_0 ActorAttributes)
  (connect Branch_0 Output%20B Subgraph_0 InPoints)
  (connect Volume%20Attributes Out Subgraph_0 VolumeAttributes)
  (connect Biome%20Attributes Out Subgraph_53 InAttributes)
  (connect Actor%20Biome Out Subgraph_53 TargetBiomeIndex)
  (connect DeleteAttributes_8 Out TransformPoints_17 bAbsoluteOffset)
  (connect Select_6 Out TransformPoints_17 bAbsoluteRotation)
  (connect Select_0 Out TransformPoints_17 In)
  (connect MakeVectorAttribute_11 Out TransformPoints_17 OffsetMax)
  (connect MakeVectorAttribute_11 Out TransformPoints_17 OffsetMin)
  (connect DeleteAttributes_4 Out TransformPoints_17 RotationMax)
  (connect Select_5 Out TransformPoints_17 RotationMin)
  (connect BreakVectorAttribute_1 Y TransformPoints_17 ScaleMax)
  (connect BreakVectorAttribute_1 X TransformPoints_17 ScaleMin)
  (output OutPoints DeleteAttributes_11))
