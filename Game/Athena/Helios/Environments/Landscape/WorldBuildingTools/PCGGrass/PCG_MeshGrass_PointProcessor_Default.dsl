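; Default PCG point processor for biome mesh grass, setting instance scale, random yaw and density.
; Mesh-matched biome scale ranges, adjusted by optional volume factors (all or only locally controlled instances), are randomized per mesh partition, while actor local overrides can supply the range.
; Actor, volume and biome densities set the upper bound of a noise-driven density filter that thins output points.
; Normal-bias and edge-distance filter nodes exist but never reach the output.
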(graph PCG_MeshGrass_PointProcessor_Default
  (inputs (InPoints :type Point) (ActorAttributes :type Param) (VolumeAttributes :type Param) (GlobalAttributes :type Param) (BiomeAttributes :type Param))
  (outputs (OutPoints :type Point))
  ;[0.05 0.05 0.05] Normal Bias
  (bind FilterAttribute_6 (Delete%20Attributes :bTokenizeOnWhiteSpace true :Operation KeepSelectedAttributes :SelectedAttributes "NormalOffset"))
  (bind FilterAttribute_19 (Delete%20Attributes :bTokenizeOnWhiteSpace true :Operation KeepSelectedAttributes :SelectedAttributes "NormalOffset"))
  (bind AttributeMathsOp_18 (Attribute%20Maths%20Op))
  (bind DeleteAttributes_19 (Delete%20Attributes :bTokenizeOnWhiteSpace true :Operation KeepSelectedAttributes :SelectedAttributes "VolumeOverride"))
  (bind Select_1 (Select))
  (bind NormalToDensity_0 (Normal%20To%20Density))
  (bind PointFilter_13 (Filter%20Attribute%20Elements :AttributeTypes (struct :DoubleValue 0.5) :bUseConstantThreshold true :bUseSpatialQuery true))
  ;[0.05 0.05 0.05] Edge Distance
  (bind CreateAttribute_17 (Create%20Constant :AttributeTypes (struct :DoubleValue 1e-05)))
  (bind DeleteAttributes_26 (Delete%20Attributes :bTokenizeOnWhiteSpace true :Operation KeepSelectedAttributes :SelectedAttributes "VolumeOverride"))
  (bind FilterAttribute_15 (Delete%20Attributes :bTokenizeOnWhiteSpace true :Operation KeepSelectedAttributes :SelectedAttributes "EdgeDistance"))
  (bind FilterAttribute_22 (Delete%20Attributes :bTokenizeOnWhiteSpace true :Operation KeepSelectedAttributes :SelectedAttributes "EdgeDistance"))
  (bind Select_21 (Select))
  (bind AttributeMathsOp_16 (Attribute%20Maths%20Op :Operation Max))
  (bind DeleteAttributes_1 (Delete%20Attributes :bTokenizeOnWhiteSpace true :Operation KeepSelectedAttributes :SelectedAttributes "FilterHeightGradient"))
  (bind Branch_15 (Branch))
  (bind Distance_14 (Distance :bSetDensity true :MaximumDistance 1.0))
  (bind DensityFilter_11 (Density%20Filter))
  (bind AttributeNoise_1 (Attribute%20Noise))
  ;[0.05 0.05 0.05] Filter Points using actor and volume exposed params
  ;;[0.05 0.05 0.05] Density
  (bind FilterAttribute_7 (Delete%20Attributes :bTokenizeOnWhiteSpace true :Operation KeepSelectedAttributes :SelectedAttributes "GrassDensity"))
  (bind FilterAttribute_10 (Delete%20Attributes :Operation KeepSelectedAttributes :SelectedAttributes "Density Factor"))
  (bind AttributeMathsOp_4 (Attribute%20Maths%20Op :Operation Multiply))
  (bind FilterAttribute_20 (Delete%20Attributes :bTokenizeOnWhiteSpace true :Operation KeepSelectedAttributes :SelectedAttributes "VolumeOverride"))
  (bind CreateAttribute_8 (Create%20Constant :AttributeTypes (struct :DoubleValue 1.0)))
  (bind FilterAttribute_1 (Delete%20Attributes :bTokenizeOnWhiteSpace true :Operation KeepSelectedAttributes :SelectedAttributes "LocalOverride"))
  (bind Select_6 (Select))
  (bind Select_3 (Select))
  (bind Subgraph_48 (Subgraph :Graph /Game/Athena/Helios/Environments/Landscape/WorldBuildingTools/PCGGrass/PCG_MeshGrass_SelectBiomeAttribute.PCG_MeshGrass_SelectBiomeAttribute))
  (bind AttributeMathsOp_1 (Attribute%20Maths%20Op :Operation Multiply))
  (bind AttributeNoise_3 (Attribute%20Noise))
  (bind DensityFilter_2 (Density%20Filter :LowerBound 0.0))
  ;;[0.05 0.05 0.05 1 11] Scale
  (bind Subgraph_41 (Subgraph :Graph /Game/Athena/Helios/Environments/Landscape/WorldBuildingTools/PCGGrass/PCG_MeshGrass_SelectBiomeAttribute.PCG_MeshGrass_SelectBiomeAttribute))
  (bind Branch_29 (Branch))
  (bind DeleteAttributes_15 (Delete%20Attributes :bTokenizeOnWhiteSpace true :Operation KeepSelectedAttributes :SelectedAttributes "LocalOverride"))
  (bind DeleteAttributes_12 (Delete%20Attributes :bTokenizeOnWhiteSpace true :Operation KeepSelectedAttributes :SelectedAttributes "VolumeOverride"))
  (bind ExecuteBlueprint_1 (Execute%20Blueprint :BlueprintElementInstance (class:/Game/Athena/Helios/Environments/Landscape/WorldBuildingTools/PCGGrass/PCGBP_GetDataByIndex.PCGBP_GetDataByIndex_C) :BlueprintElementType /Game/Athena/Helios/Environments/Landscape/WorldBuildingTools/PCGGrass/PCGBP_GetDataByIndex.PCGBP_GetDataByIndex_C))
  (bind DeleteAttributes_4 (Delete%20Attributes :bTokenizeOnWhiteSpace true :Operation KeepSelectedAttributes :SelectedAttributes "Mesh,ScaleMin,ScaleMax"))
  (bind MatchAndSetAttributes_10 (Match%20And%20Set%20Attributes :bMatchAttributes true :InputAttribute Mesh :MatchAttribute Mesh))
  (bind Branch_19 (Branch))
  (bind DeleteAttributes_16 (Delete%20Attributes :bTokenizeOnWhiteSpace true :Operation KeepSelectedAttributes :SelectedAttributes "VolumePerInstanceOverride"))
  (bind Branch_21 (Branch))
  (bind DeleteAttributes_23 (Delete%20Attributes :Operation KeepSelectedAttributes :SelectedAttributes "Scale Min Factor"))
  (bind AttributeMathsOp_7 (Attribute%20Maths%20Op :InputSource1 ScaleMin :Operation Multiply))
  (bind DeleteAttributes_25 (Delete%20Attributes :Operation KeepSelectedAttributes :SelectedAttributes "Scale Max Factor"))
  (bind AttributeMathsOp_27 (Attribute%20Maths%20Op :InputSource1 ScaleMax :Operation Multiply))
  (bind CreateAttribute_34 (Create%20Constant :AttributeTypes (struct :Type Boolean) :OutputTarget IsLocallyControlled))
  (bind AddAttribute_38 (Add%20Attribute :AttributeTypes (struct :DoubleValue 1.0) :OutputTarget VolumeScaleMin))
  (bind AddAttribute_40 (Add%20Attribute :AttributeTypes (struct :DoubleValue 1.0) :OutputTarget VolumeScaleMax))
  (bind CreateAttribute_36 (Create%20Constant :AttributeTypes (struct :Type Boolean :BoolValue true) :OutputTarget IsLocallyControlled))
  (bind AddAttribute_42 (Add%20Attribute :AttributeTypes (struct :DoubleValue 1.0) :OutputTarget VolumeScaleMin))
  (bind AddAttribute_44 (Add%20Attribute :AttributeTypes (struct :DoubleValue 1.0) :OutputTarget VolumeScaleMax))
  (bind MergeAttributes_45 (Merge%20Attributes))
  (bind MatchAndSetAttributes_33 (Match%20And%20Set%20Attributes :bMatchAttributes true :InputAttribute IsLocallyControlled :MatchAttribute IsLocallyControlled))
  (bind AttributeMathsOp_29 (Attribute%20Maths%20Op :InputSource1 ScaleMin :InputSource2 VolumeScaleMin :Operation Multiply))
  (bind AttributeMathsOp_47 (Attribute%20Maths%20Op :InputSource1 ScaleMax :InputSource2 VolumeScaleMax :Operation Multiply))
  (bind AttributePartition_48 (Attribute%20Partition :bTokenizeOnWhiteSpace true :PartitionAttributeSelectors [Mesh]))
  (bind AttributeNoise_2 (Attribute%20Noise :OutputTarget ScaleNoise))
  (bind AttributeMathsOp_0 (Attribute%20Maths%20Op :InputSource1 ScaleMax :InputSource2 ScaleMin :Operation Subtract :OutputTarget ScaleRatio))
  (bind AttributeMathsOp_3 (Attribute%20Maths%20Op :InputSource1 ScaleRatio :InputSource2 ScaleNoise :Operation Multiply :OutputTarget RandomizedScale))
  (bind AttributeMathsOp_5 (Attribute%20Maths%20Op :InputSource1 RandomizedScale :InputSource2 ScaleMin))
  (bind AttributeMathsOp_6 (Attribute%20Maths%20Op :InputSource1 $Scale :InputSource2 RandomizedScale :Operation Set))
  (bind Merge_52 (Merge%20Points))
  (bind Branch_1 (Branch))
  (bind FilterAttribute_4 (Delete%20Attributes :bTokenizeOnWhiteSpace true :Operation KeepSelectedAttributes :SelectedAttributes "ScaleMin"))
  (bind FilterAttribute_5 (Delete%20Attributes :bTokenizeOnWhiteSpace true :Operation KeepSelectedAttributes :SelectedAttributes "ScaleMax"))
  (bind TransformPoints_3 (Transform%20Points :RotationMax (struct :Yaw 3.6e+02)))
  (bind TransformPoints_4 (Transform%20Points :RotationMax (struct :Yaw 3.6e+02)))
  (bind Subgraph_0 (Subgraph :Graph /Game/Athena/Helios/Environments/Landscape/WorldBuildingTools/PCGGrass/PCG_MeshGrass_GetGrassData.PCG_MeshGrass_GetGrassData))
  ;[0.05 0.05 0.05] Inputs
  (bind GetAttributeFromPointIndex_6 (Get%20Attribute%20From%20Point%20Index :bForceOutputAttributeToBeInElementsDomain true :InputSource Biome :OutputAttributeName Biome))
  (reroute Actor%20Attributes ActorAttributes)
  (reroute Actor%20Biome GetAttributeFromPointIndex_6.Attribute)
  (reroute Biome%20Attributes BiomeAttributes)
  (reroute Global%20Attributes GlobalAttributes)
  (reroute In%20Points InPoints)
  (reroute Volume%20Attributes VolumeAttributes)
  (connect CreateAttribute_34 Out AddAttribute_38 In)
  (connect AddAttribute_38 Out AddAttribute_40 In)
  (connect DeleteAttributes_23 Out AddAttribute_42 Attributes)
  (connect CreateAttribute_36 Out AddAttribute_42 In)
  (connect DeleteAttributes_25 Out AddAttribute_44 Attributes)
  (connect AddAttribute_42 Out AddAttribute_44 In)
  (connect AttributeNoise_2 Out AttributeMathsOp_0 InA)
  (connect AttributeNoise_2 Out AttributeMathsOp_0 InB)
  (connect Select_3 Out AttributeMathsOp_1 InA)
  (connect Subgraph_48 Out%20Attribute AttributeMathsOp_1 InB)
  (connect AttributeMathsOp_0 Out AttributeMathsOp_3 InA)
  (connect AttributeMathsOp_0 Out AttributeMathsOp_3 InB)
  (connect FilterAttribute_7 Out AttributeMathsOp_4 InA)
  (connect FilterAttribute_10 Out AttributeMathsOp_4 InB)
  (connect AttributeMathsOp_3 Out AttributeMathsOp_5 InA)
  (connect AttributeMathsOp_3 Out AttributeMathsOp_5 InB)
  (connect AttributeMathsOp_5 Out AttributeMathsOp_6 InA)
  (connect AttributeMathsOp_5 Out AttributeMathsOp_6 InB)
  (connect Branch_21 Output%20A AttributeMathsOp_7 InA)
  (connect DeleteAttributes_23 Out AttributeMathsOp_7 InB)
  (connect Select_21 Out AttributeMathsOp_16 InA)
  (connect CreateAttribute_17 Out AttributeMathsOp_16 InB)
  (connect FilterAttribute_6 Out AttributeMathsOp_18 InA)
  (connect FilterAttribute_19 Out AttributeMathsOp_18 InB)
  (connect AttributeMathsOp_7 Out AttributeMathsOp_27 InA)
  (connect DeleteAttributes_25 Out AttributeMathsOp_27 InB)
  (connect MatchAndSetAttributes_33 Out AttributeMathsOp_29 InA)
  (connect MatchAndSetAttributes_33 Out AttributeMathsOp_29 InB)
  (connect AttributeMathsOp_29 Out AttributeMathsOp_47 InA)
  (connect AttributeMathsOp_29 Out AttributeMathsOp_47 InB)
  (connect DensityFilter_11 Out AttributeNoise_1 In)
  (connect AttributePartition_48 Out AttributeNoise_2 In)
  (connect TransformPoints_3 Out AttributeNoise_3 In)
  (connect TransformPoints_4 Out AttributeNoise_3 In)
  (connect AttributeMathsOp_27 Out AttributePartition_48 In)
  (connect AttributeMathsOp_47 Out AttributePartition_48 In)
  (connect Branch_19 Output%20A AttributePartition_48 In)
  (connect DeleteAttributes_15 Out Branch_1 bOutputToB)
  (connect Branch_29 Output%20A Branch_1 In)
  (connect Merge_52 Out Branch_1 In)
  (connect DeleteAttributes_1 Out Branch_15 bOutputToB)
  (connect PointFilter_13 InsideFilter Branch_15 In)
  (connect DeleteAttributes_12 Out Branch_19 bOutputToB)
  (connect MatchAndSetAttributes_10 Out Branch_19 In)
  (connect DeleteAttributes_16 Out Branch_21 bOutputToB)
  (connect Branch_19 Output%20B Branch_21 In)
  (connect Subgraph_41 Out%20Attribute Branch_29 bOutputToB)
  (connect In%20Points Out Branch_29 In)
  (connect Actor%20Attributes Out DeleteAttributes_1 In)
  (connect ExecuteBlueprint_1 OutData DeleteAttributes_4 In)
  (connect Volume%20Attributes Out DeleteAttributes_12 In)
  (connect Actor%20Attributes Out DeleteAttributes_15 In)
  (connect Volume%20Attributes Out DeleteAttributes_16 In)
  (connect Volume%20Attributes Out DeleteAttributes_19 In)
  (connect Volume%20Attributes Out DeleteAttributes_23 In)
  (connect Volume%20Attributes Out DeleteAttributes_25 In)
  (connect Volume%20Attributes Out DeleteAttributes_26 In)
  (connect AttributeNoise_3 Out DensityFilter_2 In)
  (connect AttributeMathsOp_1 Out DensityFilter_2 UpperBound)
  (connect Distance_14 Out DensityFilter_11 In)
  (connect AttributeMathsOp_16 Out Distance_14 MaximumDistance)
  (connect Branch_15 Output%20B Distance_14 Source)
  (connect PointFilter_13 OutsideFilter Distance_14 Target)
  (connect Actor%20Biome Out ExecuteBlueprint_1 DataIndex)
  (connect Actor%20Attributes Out FilterAttribute_1 In)
  (connect Actor%20Attributes Out FilterAttribute_4 In)
  (connect Actor%20Attributes Out FilterAttribute_5 In)
  (connect Actor%20Attributes Out FilterAttribute_6 In)
  (connect Actor%20Attributes Out FilterAttribute_7 In)
  (connect Volume%20Attributes Out FilterAttribute_10 In)
  (connect Volume%20Attributes Out FilterAttribute_15 In)
  (connect Volume%20Attributes Out FilterAttribute_19 In)
  (connect Volume%20Attributes Out FilterAttribute_20 In)
  (connect Actor%20Attributes Out FilterAttribute_22 In)
  (connect In%20Points Out GetAttributeFromPointIndex_6 In)
  (connect Branch_29 Output%20B MatchAndSetAttributes_10 In)
  (connect DeleteAttributes_4 Out MatchAndSetAttributes_10 Match%20Data)
  (connect Branch_21 Output%20B MatchAndSetAttributes_33 In)
  (connect MergeAttributes_45 Out MatchAndSetAttributes_33 Match%20Data)
  (connect AttributeMathsOp_6 Out Merge_52 In)
  (connect AddAttribute_40 Out MergeAttributes_45 In)
  (connect AddAttribute_44 Out MergeAttributes_45 In)
  (connect Select_1 Out NormalToDensity_0 Offset)
  (connect NormalToDensity_0 Out PointFilter_13 In)
  (connect DeleteAttributes_19 Out Select_1 bUseInputB)
  (connect FilterAttribute_6 Out Select_1 Input%20A)
  (connect AttributeMathsOp_18 Out Select_1 Input%20B)
  (connect FilterAttribute_20 Out Select_3 bUseInputB)
  (connect Select_6 Out Select_3 Input%20A)
  (connect AttributeMathsOp_4 Out Select_3 Input%20B)
  (connect FilterAttribute_1 Out Select_6 bUseInputB)
  (connect CreateAttribute_8 Out Select_6 Input%20A)
  (connect FilterAttribute_7 Out Select_6 Input%20B)
  (connect DeleteAttributes_26 Out Select_21 bUseInputB)
  (connect FilterAttribute_22 Out Select_21 Input%20A)
  (connect FilterAttribute_15 Out Select_21 Input%20B)
  (connect Biome%20Attributes Out Subgraph_41 InAttributes)
  (connect Actor%20Biome Out Subgraph_41 TargetBiomeIndex)
  (connect Biome%20Attributes Out Subgraph_48 InAttributes)
  (connect Actor%20Biome Out Subgraph_48 TargetBiomeIndex)
  (connect Branch_1 Output%20B TransformPoints_3 In)
  (connect FilterAttribute_5 Out TransformPoints_3 ScaleMax)
  (connect FilterAttribute_4 Out TransformPoints_3 ScaleMin)
  (connect Branch_1 Output%20A TransformPoints_4 In)
  (output OutPoints DensityFilter_2))
