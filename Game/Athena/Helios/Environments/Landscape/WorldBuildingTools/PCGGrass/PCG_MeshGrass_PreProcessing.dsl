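; PCG graph that pre-processes mesh grass points, pruning unwanted instances before output.
; Cull volumes are subtracted first, then attribute-driven branches filter world intersections, non-self hits (both via world ray-hit surface sampling) and borders (via a grid color-projection mask).
; A selectable looped subgraph supplies water body points to subtract.
; When landscape exists, projection onto it removes points below terrain or over excluded layers, per global attributes.
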(graph PCG_MeshGrass_PreProcessing
  (inputs (Actor%20Point :type Point) (InPoints :type Point) (ActorAttributes :type Param) (VolumeAttributes :type Param) (GlobalAttributes :type Param) (BiomeAttributes :type Param))
  (outputs (OutPoints :type Point))
  (bind BoundsModifier_28 (Bounds%20Modifier :BoundsMin (1e+01 1e+01 1e+01)))
  (bind DeleteAttributes_8 (Delete%20Attributes :bTokenizeOnWhiteSpace true :Operation KeepSelectedAttributes :SelectedAttributes "ExtendNegativeBounds"))
  (bind Select_5 (Select))
  ;[0.05 0.05 0.05] Landscape based processing
  ;;[0.05 0.05 0.05] Landscape Check
  (bind CreateAttribute_0 (Create%20Constant :AttributeTypes (struct :Type Integer32)))
  (bind GetLandscapeData_8 (Get%20Landscape%20Data :bComponentsMustOverlapSelf false :SamplingProperties (struct)))
  (bind DataNum_0 (Data%20Count))
  (bind AttributeCompareOp_1 (Attribute%20Compare%20Op :Operation NotEqual :Tolerance 0.0))
  (bind Branch_8 (Branch))
  ;;[0.05 0.05 0.05] Below terrain handling
  (bind FilterAttribute_16 (Delete%20Attributes :bTokenizeOnWhiteSpace true :Operation KeepSelectedAttributes :SelectedAttributes "RemoveInstancesUnderTerrain"))
  (bind Branch_14 (Branch))
  (bind Projection_7 (Projection :bKeepZeroDensityPoints true :ProjectionParams (struct :bProjectRotations false :AttributeMode IncludeAttributes)))
  (bind FilterAttribute_17 (Delete%20Attributes :bTokenizeOnWhiteSpace true :Operation KeepSelectedAttributes :SelectedAttributes "UnderTerrainBias"))
  (bind AttributeMathsOp_12 (Attribute%20Maths%20Op :InputSource1 $Position.Z))
  (bind PointFilter_10 (Filter%20Attribute%20Elements :AttributeTypes (struct :Type Boolean :BoolValue true) :Operator Lesser :TargetAttribute $Position.Z :ThresholdAttribute $Position.Z))
  ;;[0.05 0.05 0.05] Layer Exclusion
  (bind FilterAttribute_3 (Delete%20Attributes :bTokenizeOnWhiteSpace true :Operation KeepSelectedAttributes :SelectedAttributes "ExcludeByLandscapeLayers"))
  (bind Branch_1 (Branch))
  (bind FilterAttribute_12 (Delete%20Attributes :bTokenizeOnWhiteSpace true :Operation KeepSelectedAttributes :SelectedAttributes "ExcludedLayerNames"))
  (bind Projection_8 (Projection :bKeepZeroDensityPoints true :ProjectionParams (struct :bProjectPositions false :bProjectRotations false :AttributeMode IncludeAttributes)))
  (bind AddAttribute_25 (Add%20Attribute :OutputTarget LayerExclusion))
  (bind AttributeMathsOp_26 (Attribute%20Maths%20Op :InputSource1 LayerExclusion :Operation Set))
  (bind FilterAttribute_21 (Delete%20Attributes :bTokenizeOnWhiteSpace true :Operation KeepSelectedAttributes :SelectedAttributes "LayerExclusionThreshold"))
  (bind PointFilter_9 (Filter%20Attribute%20Elements :Operator LesserOrEqual :TargetAttribute LayerExclusion :ThresholdAttribute @Last))
  ;[0.05 0.05 0.05] Handle water bodies
  (bind FilterAttribute_27 (Delete%20Attributes :bTokenizeOnWhiteSpace true :Operation KeepSelectedAttributes :SelectedAttributes "ProcessInstancesUnderWater"))
  (bind Branch_31 (Branch))
  (bind GetActorData_1 (Get%20Actor%20Data :ActorSelector (struct :ActorFilter AllWorldActors :bMustOverlapSelf true :ActorSelection ByClass :ActorSelectionClass /Game/Athena/Apollo/Environments/Landscape/Water/FortWaterBodyBP.FortWaterBodyBP_C :bSelectMultiple true) :bComponentsMustOverlapSelf false))
  (bind FilterByType_2 (Filter%20Data%20By%20Type :TargetType (struct :Ids [(struct :Struct /Script/PCG.PCGDataTypeInfoSpline)])))
  (bind GetActorData_3 (Get%20Actor%20Data :ActorSelector (struct :ActorFilter AllWorldActors :bMustOverlapSelf true :ActorSelection ByClass :ActorSelectionClass /Game/Athena/Apollo/Environments/Landscape/Water/FortWaterBodyBP.FortWaterBodyBP_C :bSelectMultiple true) :bComponentsMustOverlapSelf false :Mode GetSinglePoint))
  (bind CreateAttribute_20 (Create%20Constant :AttributeTypes (struct :Type SoftObjectPath :SoftObjectPathValue (struct :AssetPath (struct :PackageName /Game/Athena/Helios/Environments/Landscape/WorldBuildingTools/PCGGrass/PCG_MeshGrass_WaterBodyProcessor_Default :AssetName PCG_MeshGrass_WaterBodyProcessor_Default)))))
  (bind CreateAttribute_39 (Create%20Constant :AttributeTypes (struct :Type String :StringValue "None")))
  (bind FilterAttribute_25 (Delete%20Attributes :bTokenizeOnWhiteSpace true :Operation KeepSelectedAttributes :SelectedAttributes "UnderWaterProcessorGraph"))
  (bind ExecuteBlueprint_0 (Execute%20Blueprint :BlueprintElementInstance (class:/Game/Athena/Helios/Environments/Landscape/WorldBuildingTools/PCGGrass/PCGBP_SoftObjectPathToString.PCGBP_SoftObjectPathToString_C) :BlueprintElementType /Game/Athena/Helios/Environments/Landscape/WorldBuildingTools/PCGGrass/PCGBP_SoftObjectPathToString.PCGBP_SoftObjectPathToString_C))
  (bind AttributeCompareOp_43 (Attribute%20Compare%20Op :Operation NotEqual :OutputTarget IsValid))
  (bind FilterAttribute_1 (Delete%20Attributes :bTokenizeOnWhiteSpace true :Operation KeepSelectedAttributes :SelectedAttributes "IsValid"))
  (bind Select_18 (Select))
  (bind Loop_5 (Loop :bTokenizeOnWhiteSpace true :bUseGraphDefaultPinUsage false :LoopPins "WaterBodyActors,WaterBodySplines" :Graph /Game/Athena/Helios/Environments/Landscape/WorldBuildingTools/PCGGrass/PCG_MeshGrass_WaterBodyProcessor_Default.PCG_MeshGrass_WaterBodyProcessor_Default))
  (bind Difference_12 (Difference :bDiffMetadata false :DensityFunction Binary :Mode Discrete))
  (bind ToPoint_14 (To%20Point))
  ;[0.05 0.05 0.05] Handle world intersection
  (bind CreateAttribute_5 (Create%20Constant :AttributeTypes (struct :DoubleValue 1.0)))
  (bind CreateAttribute_4 (Create%20Constant))
  (bind CreateAttribute_32 (Create%20Constant :AttributeTypes (struct :DoubleValue 5e+01)))
  (bind FilterAttribute_30 (Delete%20Attributes :bTokenizeOnWhiteSpace true :Operation KeepSelectedAttributes :SelectedAttributes "WorldIntersectionDistance"))
  (bind MakeVectorAttribute_31 (Make%20Vector%20Attribute :OutputType Vector))
  (bind AttributeMathsOp_25 (Attribute%20Maths%20Op :InputSource1 $Position.Z :InputSource2 $BoundsMax.Z :OutputTarget RayOrigin))
  (bind GetAttributeFromPointIndex_26 (Get%20Attribute%20From%20Point%20Index :bForceOutputAttributeToBeInElementsDomain true :OutputAttributeName None))
  (bind AttributeMathsOp_29 (Attribute%20Maths%20Op))
  (bind CreateAttribute_28 (Create%20Constant))
  (bind MakeVectorAttribute_27 (Make%20Vector%20Attribute :OutputType Vector))
  (bind WorldRayHitQuery_21 (World%20Ray%20Hit%20Query :QueryParams (struct :bOverrideDefaultParams true :RayOrigin (0.0 0.0 8e+03) :RayLength 1e+04 :bIgnorePCGHits true :ActorTagFilter Exclude :ActorTagsList "PCG_MeshGrass" :SelectLandscapeHits Exclude)))
  (bind SurfaceSampler_22 (Surface%20Sampler :bApplyDensityToPoints false :Looseness 0.0 :PointExtents (5e+01 5e+01 5e+02) :PointsPerSquaredMeter 1.0 :PointSteepness 0.5))
  (bind GetPointsCount_0 (Get%20Element%20Count))
  (bind AttributeCompareOp_3 (Attribute%20Compare%20Op :Operation NotEqual))
  (bind DeleteAttributes_5 (Delete%20Attributes :bTokenizeOnWhiteSpace true :Operation KeepSelectedAttributes :SelectedAttributes "FilterWorldIntersection"))
  (bind Branch_10 (Branch))
  (bind Branch_2 (Branch))
  (bind DeleteAttributes_35 (Delete%20Attributes :bTokenizeOnWhiteSpace true :Operation KeepSelectedAttributes :SelectedAttributes "FilterNonSelfHits"))
  (bind Branch_6 (Branch))
  (bind TransformPoints_5 (Transform%20Points :bAbsoluteRotation true))
  (bind Projection_16 (Projection :bKeepZeroDensityPoints true :ProjectionParams (struct :bProjectRotations false)))
  (bind PointFilter_23 (Filter%20Attribute%20Elements :AttributeTypes (struct :BoolValue true) :Operator Lesser :TargetAttribute $Position.Z :ThresholdAttribute $Position.Z))
  (bind Difference_23 (Difference :bDiffMetadata false :DensityFunction Binary :Mode Discrete))
  (bind Branch_17 (Branch))
  (bind CreateAttribute_6 (Create%20Constant :AttributeTypes (struct :Type Vector) :OutputTarget Vector))
  (bind DeleteAttributes_25 (Delete%20Attributes :bTokenizeOnWhiteSpace true :Operation KeepSelectedAttributes :SelectedAttributes "NonSelfTraceZOffset"))
  (bind AttributeMathsOp_7 (Attribute%20Maths%20Op :InputSource1 Vector.Z :Operation Set))
  (bind AttributeMathsOp_10 (Attribute%20Maths%20Op :InputSource1 $Position.Z :InputSource2 $BoundsMax.Z))
  (bind GetAttributeFromPointIndex_8 (Get%20Attribute%20From%20Point%20Index :bForceOutputAttributeToBeInElementsDomain true :InputSource $Position :OutputAttributeName None))
  (bind AttributeMathsOp_19 (Attribute%20Maths%20Op))
  (bind AttributeMathsOp_13 (Attribute%20Maths%20Op :InputSource1 $BoundsMin.Z :InputSource2 $BoundsMin.Z :Operation Abs))
  (bind AttributeMathsOp_18 (Attribute%20Maths%20Op :InputSource1 $BoundsMin.Z :InputSource2 $BoundsMax.Z :OutputTarget RayLenght))
  (bind GetAttributeFromPointIndex_15 (Get%20Attribute%20From%20Point%20Index :bForceOutputAttributeToBeInElementsDomain true :OutputAttributeName None))
  (bind AttributeMathsOp_2 (Attribute%20Maths%20Op))
  (bind CreateAttribute_22 (Create%20Constant :AttributeTypes (struct :DoubleValue 1.0 :VectorValue (0.0 0.0 1e+02))))
  (bind AttributeMathsOp_20 (Attribute%20Maths%20Op))
  (bind DeleteAttributes_3 (Delete%20Attributes :bTokenizeOnWhiteSpace true :Operation KeepSelectedAttributes :SelectedAttributes "NonSelfTraceComplex"))
  (bind WorldRayHitQuery_1 (World%20Ray%20Hit%20Query :QueryParams (struct :bOverrideDefaultParams true :RayOrigin (0.0 0.0 8e+03) :RayLength 1.5e+02 :bIgnoreSelfHits false :ActorTagsList "PCG_MeshGrass" :bGetReferenceToActorHit true)))
  (bind SurfaceSampler_3 (Surface%20Sampler :bApplyDensityToPoints false :Looseness 0.5 :PointExtents (1e+01 1e+01 1e+01) :PointsPerSquaredMeter 8.0 :PointSteepness 0.5))
  (bind Subgraph_0 (Subgraph :Graph /Game/Athena/Helios/Environments/Landscape/WorldBuildingTools/PCGGrass/PCG_HasValidPoints.PCG_HasValidPoints))
  (bind Branch_9 (Branch))
  (bind AttributeCompareOp_0 (Attribute%20Compare%20Op :InputSource1 ActorReference :InputSource2 ActorReference :OutputTarget IsSelf))
  (bind AttributeFilter_3 (Filter%20Attribute%20Elements :AttributeTypes (struct :Type Boolean :BoolValue true) :bUseConstantThreshold true :bUseSpatialQuery true :Operator NotEqual :TargetAttribute IsSelf))
  (bind DeleteAttributes_24 (Delete%20Attributes :bTokenizeOnWhiteSpace true :Operation KeepSelectedAttributes :SelectedAttributes "NonSelfHitDistance"))
  (bind ExtentsModifier_5 (Extents%20Modifier :Extents (2e+01 2e+01 2e+01)))
  (bind Difference_7 (Difference :bDiffMetadata false :DensityFunction Binary :Mode Discrete))
  (bind AttributeMathsOp_4 (Attribute%20Maths%20Op :InputSource1 $Density :Operation Set))
  (bind DeleteAttributes_33 (Delete%20Attributes :bTokenizeOnWhiteSpace true :Operation KeepSelectedAttributes :SelectedAttributes "FilterBorders"))
  (bind Branch_23 (Branch))
  (bind CreateAttribute_33 (Create%20Constant :AttributeTypes (struct :Type Vector)))
  (bind ApplyScaleToBounds_1 (Apply%20Scale%20To%20Bounds))
  (bind GetAttributeFromPointIndex_16 (Get%20Attribute%20From%20Point%20Index :bForceOutputAttributeToBeInElementsDomain true :InputSource $Extents :OutputAttributeName None))
  (bind CreateAttribute_38 (Create%20Constant :AttributeTypes (struct :DoubleValue 0.3)))
  (bind AttributeVectorOp_35 (Attribute%20Vector%20Op :InputSource1 $BoundsMax :Operation Length :OutputTarget Lenght))
  (bind GetAttributeFromPointIndex_36 (Get%20Attribute%20From%20Point%20Index :bForceOutputAttributeToBeInElementsDomain true :OutputAttributeName None))
  (bind AttributeMathsOp_37 (Attribute%20Maths%20Op :Operation Multiply))
  (bind CreateAttribute_30 (Create%20Constant))
  (bind MakeVectorAttribute_25 (Make%20Vector%20Attribute :OutputType Vector))
  (bind AttributeMathsOp_23 (Attribute%20Maths%20Op))
  (bind CreateAttribute_40 (Create%20Constant :AttributeTypes (struct :DoubleValue 1e+02) :OutputTarget CellSize))
  (bind CreatePointsGrid_8 (Create%20Points%20Grid :CellSize (1e+02 1e+02 2e+02) :CoordinateSpace LocalComponent :GridExtents (1e+02 1e+02 5e+01) :PointSteepness 0.5))
  (bind GetAttributeFromPointIndex_12 (Get%20Attribute%20From%20Point%20Index :bForceOutputAttributeToBeInElementsDomain true :InputSource $LocalCenter :OutputAttributeName None))
  (bind TransformPoints_11 (Transform%20Points))
  (bind AttributeMathsOp_31 (Attribute%20Maths%20Op :InputSource1 $Color :Operation Set))
  (bind CreateAttribute_29 (Create%20Constant :AttributeTypes (struct :Type Vector :VectorValue (1.0 1.0 1.0))))
  (bind AttributeMathsOp_27 (Attribute%20Maths%20Op :InputSource1 $Color :Operation Set))
  (bind DeleteAttributes_76 (Delete%20Attributes :bTokenizeOnWhiteSpace true :Operation KeepSelectedAttributes :SelectedAttributes "BorderDistance"))
  (bind ExtentsModifier_37 (Extents%20Modifier :Extents (2e+01 2e+01 2e+01)))
  (bind Projection_20 (Projection :bKeepZeroDensityPoints true :ProjectionParams (struct :bProjectPositions false :bProjectRotations false :ColorBlendMode TargetValue)))
  (bind AttributeFilter_39 (Filter%20Attribute%20Elements :AttributeTypes (struct :Type Vector) :bUseConstantThreshold true :bUseSpatialQuery true :Operator Equal :TargetAttribute $Color))
  (bind CreateAttribute_46 (Create%20Constant :AttributeTypes (struct :DoubleValue 1.0)))
  (bind AttributeMathsOp_44 (Attribute%20Maths%20Op :InputSource1 $Density :Operation Set))
  (bind DeleteAttributes_72 (Delete%20Attributes :bTokenizeOnWhiteSpace true :Operation KeepSelectedAttributes :SelectedAttributes "BorderSize"))
  (bind ExtentsModifier_43 (Extents%20Modifier :Extents (75.0 75.0 5e+01)))
  (bind Difference_14 (Difference :bDiffMetadata false :DensityFunction Binary :Mode Discrete))
  (bind ToPoint_39 (To%20Point))
  ;[0.05 0.05 0.05] Cull Volumes
  (bind GetActorData_13 (Get%20Actor%20Data :ActorSelector (struct :ActorFilter AllWorldActors :ActorSelection ByClass :ActorSelectionClass /Game/Athena/Helios/Environments/Landscape/WorldBuildingTools/PCGGrass/BP_MeshGrass_CullVolume.BP_MeshGrass_CullVolume_C :bSelectMultiple true) :bComponentsMustOverlapSelf false :bMergeSinglePointData true :Mode GetSinglePoint)) ; single point
  (bind ApplyScaleToBounds_0 (Apply%20Scale%20To%20Bounds))
  (bind Difference_24 (Difference :bDiffMetadata false :DensityFunction Binary :Mode Discrete))
  (bind ToPoint_25 (To%20Point))
  ;[0.05 0.05 0.05] Inputs
  (bind GetAttributeFromPointIndex_4 (Get%20Attribute%20From%20Point%20Index :bForceOutputAttributeToBeInElementsDomain true :InputSource Biome :OutputAttributeName Biome))
  (reroute Actor%20Attributes ActorAttributes)
  (reroute Actor%20Biome GetAttributeFromPointIndex_4.Attribute)
  (reroute Biome%20Attributes BiomeAttributes)
  (reroute Global%20Attributes GlobalAttributes)
  (reroute In%20Points InPoints)
  (reroute NamedRerouteDeclaration_0 Actor%20Point)
  (reroute Volume%20Attributes VolumeAttributes)
  (connect Projection_8 Out AddAttribute_25 In)
  (connect GetActorData_13 Out ApplyScaleToBounds_0 In)
  (connect NamedRerouteDeclaration_0 Out ApplyScaleToBounds_1 In)
  (connect Branch_9 Output%20B AttributeCompareOp_0 InA)
  (connect NamedRerouteDeclaration_0 Out AttributeCompareOp_0 InB)
  (connect DataNum_0 Out AttributeCompareOp_1 InA)
  (connect CreateAttribute_0 Out AttributeCompareOp_1 InB)
  (connect GetPointsCount_0 Out AttributeCompareOp_3 InA)
  (connect CreateAttribute_4 Out AttributeCompareOp_3 InB)
  (connect ExecuteBlueprint_0 Out AttributeCompareOp_43 InA)
  (connect CreateAttribute_39 Out AttributeCompareOp_43 InB)
  (connect AttributeCompareOp_0 Out AttributeFilter_3 In)
  (connect Projection_20 Out AttributeFilter_39 In)
  (connect DeleteAttributes_25 Out AttributeMathsOp_2 InA)
  (connect GetAttributeFromPointIndex_15 Attribute AttributeMathsOp_2 InB)
  (connect Difference_7 Out AttributeMathsOp_4 InA)
  (connect CreateAttribute_5 Out AttributeMathsOp_4 InB)
  (connect CreateAttribute_6 Out AttributeMathsOp_7 InA)
  (connect DeleteAttributes_25 Out AttributeMathsOp_7 InB)
  (connect NamedRerouteDeclaration_0 Out AttributeMathsOp_10 InA)
  (connect NamedRerouteDeclaration_0 Out AttributeMathsOp_10 InB)
  (connect Projection_7 Out AttributeMathsOp_12 InA)
  (connect FilterAttribute_17 Out AttributeMathsOp_12 InB)
  (connect NamedRerouteDeclaration_0 Out AttributeMathsOp_13 In)
  (connect AttributeMathsOp_13 Out AttributeMathsOp_18 InA)
  (connect AttributeMathsOp_13 Out AttributeMathsOp_18 InB)
  (connect GetAttributeFromPointIndex_8 Attribute AttributeMathsOp_19 InA)
  (connect AttributeMathsOp_7 Out AttributeMathsOp_19 InB)
  (connect AttributeMathsOp_2 Out AttributeMathsOp_20 InA)
  (connect CreateAttribute_22 Out AttributeMathsOp_20 InB)
  (connect GetAttributeFromPointIndex_16 Attribute AttributeMathsOp_23 InA)
  (connect MakeVectorAttribute_25 Out AttributeMathsOp_23 InB)
  (connect NamedRerouteDeclaration_0 Out AttributeMathsOp_25 InA)
  (connect NamedRerouteDeclaration_0 Out AttributeMathsOp_25 InB)
  (connect AddAttribute_25 Out AttributeMathsOp_26 InA)
  (connect Projection_8 Out AttributeMathsOp_26 InB)
  (connect ToPoint_25 Out AttributeMathsOp_27 InA)
  (connect CreateAttribute_29 Out AttributeMathsOp_27 InB)
  (connect GetAttributeFromPointIndex_26 Attribute AttributeMathsOp_29 InA)
  (connect FilterAttribute_30 Out AttributeMathsOp_29 InB)
  (connect TransformPoints_11 Out AttributeMathsOp_31 InA)
  (connect CreateAttribute_33 Out AttributeMathsOp_31 InB)
  (connect GetAttributeFromPointIndex_36 Attribute AttributeMathsOp_37 InA)
  (connect CreateAttribute_38 Out AttributeMathsOp_37 InB)
  (connect AttributeFilter_39 InsideFilter AttributeMathsOp_44 InA)
  (connect CreateAttribute_46 Out AttributeMathsOp_44 InB)
  (connect ApplyScaleToBounds_1 Out AttributeVectorOp_35 In)
  (connect Branch_10 Output%20B BoundsModifier_28 In)
  (connect FilterAttribute_3 Out Branch_1 bOutputToB)
  (connect Branch_14 Output%20A Branch_1 In)
  (connect PointFilter_10 OutsideFilter Branch_1 In)
  (connect AttributeCompareOp_3 Out Branch_2 bOutputToB)
  (connect Branch_10 Output%20B Branch_2 In)
  (connect AttributeCompareOp_3 Out Branch_6 bOutputToB)
  (connect SurfaceSampler_22 Out Branch_6 In)
  (connect AttributeCompareOp_1 Out Branch_8 bOutputToB)
  (connect Branch_31 Output%20A Branch_8 In)
  (connect ToPoint_14 Out Branch_8 In)
  (connect Subgraph_0 IsValid Branch_9 bOutputToB)
  (connect SurfaceSampler_3 Out Branch_9 In)
  (connect DeleteAttributes_5 Out Branch_10 bOutputToB)
  (connect ToPoint_25 Out Branch_10 In)
  (connect FilterAttribute_16 Out Branch_14 bOutputToB)
  (connect Branch_8 Output%20B Branch_14 In)
  (connect DeleteAttributes_35 Out Branch_17 bOutputToB)
  (connect Branch_2 Output%20A Branch_17 In)
  (connect Branch_10 Output%20A Branch_17 In)
  (connect Difference_23 Out Branch_17 In)
  (connect DeleteAttributes_33 Out Branch_23 bOutputToB)
  (connect AttributeMathsOp_4 Out Branch_23 In)
  (connect Branch_17 Output%20A Branch_23 In)
  (connect FilterAttribute_27 Out Branch_31 bOutputToB)
  (connect ToPoint_39 Out Branch_31 In)
  (connect CreateAttribute_40 Out CreatePointsGrid_8 CellSize)
  (connect AttributeMathsOp_23 Out CreatePointsGrid_8 GridExtents)
  (connect GetLandscapeData_8 Out DataNum_0 In)
  (connect Actor%20Attributes Out DeleteAttributes_3 In)
  (connect Actor%20Attributes Out DeleteAttributes_5 In)
  (connect Actor%20Attributes Out DeleteAttributes_8 In)
  (connect Actor%20Attributes Out DeleteAttributes_24 In)
  (connect Actor%20Attributes Out DeleteAttributes_25 In)
  (connect Actor%20Attributes Out DeleteAttributes_33 In)
  (connect Actor%20Attributes Out DeleteAttributes_35 In)
  (connect Actor%20Attributes Out DeleteAttributes_72 In)
  (connect Actor%20Attributes Out DeleteAttributes_76 In)
  (connect ExtentsModifier_5 Out Difference_7 Differences)
  (connect Branch_17 Output%20B Difference_7 Source)
  (connect Loop_5 WaterBodyPoints Difference_12 Differences)
  (connect Branch_31 Output%20B Difference_12 Source)
  (connect ExtentsModifier_43 Out Difference_14 Differences)
  (connect Branch_23 Output%20B Difference_14 Source)
  (connect PointFilter_23 OutsideFilter Difference_23 Differences)
  (connect Branch_2 Output%20B Difference_23 Source)
  (connect ApplyScaleToBounds_0 Out Difference_24 Differences)
  (connect In%20Points Out Difference_24 Source)
  (connect FilterAttribute_25 Out ExecuteBlueprint_0 In)
  (connect DeleteAttributes_24 Out ExtentsModifier_5 Extents)
  (connect AttributeFilter_3 InsideFilter ExtentsModifier_5 In)
  (connect DeleteAttributes_76 Out ExtentsModifier_37 Extents)
  (connect AttributeMathsOp_27 Out ExtentsModifier_37 In)
  (connect DeleteAttributes_72 Out ExtentsModifier_43 Extents)
  (connect AttributeMathsOp_44 Out ExtentsModifier_43 In)
  (connect AttributeCompareOp_43 Out FilterAttribute_1 In)
  (connect Global%20Attributes Out FilterAttribute_3 In)
  (connect Global%20Attributes Out FilterAttribute_12 In)
  (connect Global%20Attributes Out FilterAttribute_16 In)
  (connect Global%20Attributes Out FilterAttribute_17 In)
  (connect Global%20Attributes Out FilterAttribute_21 In)
  (connect Global%20Attributes Out FilterAttribute_25 In)
  (connect Global%20Attributes Out FilterAttribute_27 In)
  (connect Actor%20Attributes Out FilterAttribute_30 In)
  (connect GetActorData_1 Out FilterByType_2 In)
  (connect NamedRerouteDeclaration_0 Out GetAttributeFromPointIndex_4 In)
  (connect AttributeMathsOp_10 Out GetAttributeFromPointIndex_8 In)
  (connect ApplyScaleToBounds_1 Out GetAttributeFromPointIndex_12 In)
  (connect AttributeMathsOp_18 Out GetAttributeFromPointIndex_15 In)
  (connect ApplyScaleToBounds_1 Out GetAttributeFromPointIndex_16 In)
  (connect AttributeMathsOp_25 Out GetAttributeFromPointIndex_26 In)
  (connect AttributeVectorOp_35 Out GetAttributeFromPointIndex_36 In)
  (connect SurfaceSampler_22 Out GetPointsCount_0 In)
  (connect Select_18 Out Loop_5 SubgraphOverride)
  (connect GetActorData_3 Out Loop_5 WaterBodyActors)
  (connect FilterByType_2 InsideFilter Loop_5 WaterBodySplines)
  (connect AttributeMathsOp_37 Out MakeVectorAttribute_25 X)
  (connect AttributeMathsOp_37 Out MakeVectorAttribute_25 Y)
  (connect CreateAttribute_30 Out MakeVectorAttribute_25 Z)
  (connect CreateAttribute_28 Out MakeVectorAttribute_27 X)
  (connect CreateAttribute_28 Out MakeVectorAttribute_27 Y)
  (connect AttributeMathsOp_29 Out MakeVectorAttribute_27 Z)
  (connect CreateAttribute_32 Out MakeVectorAttribute_31 X)
  (connect CreateAttribute_32 Out MakeVectorAttribute_31 Y)
  (connect FilterAttribute_30 Out MakeVectorAttribute_31 Z)
  (connect FilterAttribute_21 Out PointFilter_9 Filter)
  (connect AttributeMathsOp_26 Out PointFilter_9 In)
  (connect AttributeMathsOp_12 Out PointFilter_10 Filter)
  (connect Branch_14 Output%20B PointFilter_10 In)
  (connect Projection_16 Out PointFilter_23 Filter)
  (connect TransformPoints_5 Out PointFilter_23 In)
  (connect Branch_14 Output%20B Projection_7 In)
  (connect GetLandscapeData_8 Out Projection_7 Projection%20Target)
  (connect FilterAttribute_12 Out Projection_8 AttributeList)
  (connect Branch_1 Output%20B Projection_8 In)
  (connect GetLandscapeData_8 Out Projection_8 Projection%20Target)
  (connect TransformPoints_5 Out Projection_16 In)
  (connect Select_5 Out Projection_16 Projection%20Target)
  (connect AttributeMathsOp_31 Out Projection_20 In)
  (connect ExtentsModifier_37 Out Projection_20 Projection%20Target)
  (connect DeleteAttributes_8 Out Select_5 bUseInputB)
  (connect Branch_10 Output%20B Select_5 Input%20A)
  (connect BoundsModifier_28 Out Select_5 Input%20B)
  (connect FilterAttribute_1 Out Select_18 bUseInputB)
  (connect CreateAttribute_20 Out Select_18 Input%20A)
  (connect FilterAttribute_25 Out Select_18 Input%20B)
  (connect SurfaceSampler_3 Out Subgraph_0 In)
  (connect WorldRayHitQuery_1 Out SurfaceSampler_3 Surface)
  (connect MakeVectorAttribute_31 Out SurfaceSampler_22 PointExtents)
  (connect WorldRayHitQuery_21 Out SurfaceSampler_22 Surface)
  (connect Difference_12 Out ToPoint_14 In)
  (connect Difference_24 Out ToPoint_25 In)
  (connect Branch_23 Output%20A ToPoint_39 In)
  (connect Difference_14 Out ToPoint_39 In)
  (connect Branch_6 Output%20B TransformPoints_5 In)
  (connect CreatePointsGrid_8 Out TransformPoints_11 In)
  (connect GetAttributeFromPointIndex_12 Attribute TransformPoints_11 OffsetMax)
  (connect GetAttributeFromPointIndex_12 Attribute TransformPoints_11 OffsetMin)
  (connect DeleteAttributes_3 Out WorldRayHitQuery_1 bTraceComplex)
  (connect AttributeMathsOp_20 Out WorldRayHitQuery_1 RayLength)
  (connect AttributeMathsOp_19 Out WorldRayHitQuery_1 RayOrigin)
  (connect MakeVectorAttribute_27 Out WorldRayHitQuery_21 RayOrigin)
  (output OutPoints Branch_1.Output%20A)
  (output OutPoints Branch_8.Output%20A)
  (output OutPoints PointFilter_9.InsideFilter))
