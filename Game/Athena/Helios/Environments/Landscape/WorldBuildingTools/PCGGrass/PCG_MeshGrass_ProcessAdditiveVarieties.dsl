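; PCG loop-body subgraph for mesh grass that generates the points for one additive grass variety from a set of source points.
; The variety's Weight attribute sets the ratio of source points randomly chosen, then each chosen point gets a small horizontal offset and random yaw.
; The loop index seeds both random steps, and all VarietyData attributes are copied onto the output points.
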(graph PCG_MeshGrass_ProcessAdditiveVarieties
  (inputs (SourcePoints :type Point) (VarietyData :type Param))
  (outputs (OutPoints :type Point))
  (bind DeleteAttributes_1 (Delete%20Attributes :Operation KeepSelectedAttributes :SelectedAttributes "Weight"))
  (bind GetLoopIndex_2 (Get%20Loop%20Index))
  (bind RandomChoice_0 (Random%20Choice :bFixedMode false :bOutputDiscardedEntries false))
  (bind TransformPoints_3 (Transform%20Points :OffsetMax (1e+01 1e+01 0.0) :OffsetMin (-1e+01 -1e+01 0.0) :RotationMax (struct :Yaw 1.8e+02) :RotationMin (struct :Yaw -1.8e+02)))
  (bind AddAttribute_4 (Add%20Attribute :bCopyAllAttributes true))
  (connect DefaultInputNode VarietyData AddAttribute_4 Attributes)
  (connect TransformPoints_3 Out AddAttribute_4 In)
  (connect DefaultInputNode VarietyData DeleteAttributes_1 In)
  (connect DefaultInputNode SourcePoints RandomChoice_0 In)
  (connect DeleteAttributes_1 Out RandomChoice_0 Ratio)
  (connect GetLoopIndex_2 Out RandomChoice_0 Seed)
  (connect RandomChoice_0 Chosen TransformPoints_3 In)
  (connect GetLoopIndex_2 Out TransformPoints_3 Seed)
  (output OutPoints AddAttribute_4))
