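; PCG subgraph for mesh grass tools that filters points by a world-placed texture and outputs them with that texture.
; It samples the texture at its transform, copies the chosen channel to a temporary TextureValue, and keeps points passing the threshold, optionally inverted.
; A projection overlap test drives a Branch, but that Branch has no data input wired.
; A Bounds Min Max-to-transform chain is also left disconnected.
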(graph PCG_MeshGrass_ProcessWorldTexture
  (inputs (Points :type Point) (WorldTextureData :type Param))
  (outputs (Points :type Point) (Texture :type BaseTexture))
  (bind GetBounds_2 (Get%20Bounds))
  (bind AttributeMathsOp_15 (Attribute%20Maths%20Op :InputSource1 BoundsMax :InputSource2 BoundsMin :OutputTarget avg))
  (bind CreateAttribute_5 (Create%20Constant :AttributeTypes (struct :DoubleValue 0.5)))
  (bind AttributeMathsOp_16 (Attribute%20Maths%20Op :Operation Multiply))
  (bind CreatePoints_0 (Create%20Points))
  (bind AttributeMathsOp_17 (Attribute%20Maths%20Op :InputSource1 $Position :Operation Set))
  (bind AttributeMathsOp_18 (Attribute%20Maths%20Op :InputSource1 BoundsMax :InputSource2 BoundsMin :Operation Subtract :OutputTarget Extents))
  (bind AttributeMathsOp_19 (Attribute%20Maths%20Op :Operation Multiply))
  (bind AttributeMathsOp_20 (Attribute%20Maths%20Op :InputSource1 $Extents :Operation Set))
  (bind AttributeTransformOp_0 (Attribute%20Transform%20Op :InputSource1 $Transform :InputSource2 Transform))
  (bind Projection_0 (Projection :ProjectionParams (struct :bProjectPositions false :bProjectRotations false :AttributeList "$Density" :AttributeMode IncludeAttributes)))
  (bind GetPointsCount_0 (Get%20Element%20Count))
  (bind Branch_0 (Branch))
  (bind DeleteAttributes_0 (Delete%20Attributes :Operation KeepSelectedAttributes :SelectedAttributes "World Texture"))
  (bind DeleteAttributes_2 (Delete%20Attributes :Operation KeepSelectedAttributes :SelectedAttributes "Transform"))
  (bind DeleteAttributes_5 (Delete%20Attributes :Operation KeepSelectedAttributes :SelectedAttributes "Texture Channel"))
  (bind GetTextureData_1 (Get%20Texture%20Data :bForceEditorOnlyCPUSampling true :bUseAbsoluteTransform true :bUseDensitySourceChannel false :ColorChannel Red))
  (bind SampleTexture_1 (Sample%20Texture :bClampOutputDensity false))
  (bind DeleteAttributes_1 (Delete%20Attributes :Operation KeepSelectedAttributes :SelectedAttributes "Bounds Min Max"))
  (bind BreakVectorAttribute_0 (Break%20Vector%20Attribute))
  (bind CreateAttribute_0 (Create%20Constant :AttributeTypes (struct :Type Vector2) :OutputTarget BoundsMin))
  (bind AttributeMathsOp_0 (Attribute%20Maths%20Op :InputSource1 @Last.X :Operation Set))
  (bind AttributeMathsOp_1 (Attribute%20Maths%20Op :InputSource1 BoundsMin.Y :Operation Set))
  (bind CreateAttribute_1 (Create%20Constant :AttributeTypes (struct :Type Vector2) :OutputTarget BoundsMax))
  (bind AttributeMathsOp_2 (Attribute%20Maths%20Op :InputSource1 @Last.X :Operation Set))
  (bind AttributeMathsOp_3 (Attribute%20Maths%20Op :InputSource1 BoundsMax.Y :Operation Set))
  (bind AttributeMathsOp_4 (Attribute%20Maths%20Op :InputSource1 BoundsMin :InputSource2 BoundsMax :OutputTarget Center))
  (bind CreateAttribute_2 (Create%20Constant :AttributeTypes (struct :DoubleValue 2.0)))
  (bind AttributeMathsOp_5 (Attribute%20Maths%20Op :Operation Divide))
  (bind CreateAttribute_3 (Create%20Constant :AttributeTypes (struct :Type Transform) :OutputTarget Transform))
  (bind AttributeMathsOp_7 (Attribute%20Maths%20Op :InputSource1 Transform.Position :InputSource2 Center :Operation Set))
  (bind AttributeMathsOp_9 (Attribute%20Maths%20Op :InputSource1 BoundsMin :Operation Abs))
  (bind AttributeMathsOp_8 (Attribute%20Maths%20Op :InputSource1 Transform.Scale.XY :InputSource2 BoundsMin.XY :Operation Set))
  (bind Subgraph_0 (Subgraph :Graph /Game/Athena/Helios/Environments/Landscape/WorldBuildingTools/PCGGrass/PCG_ColorAttributeFromIndex.PCG_ColorAttributeFromIndex))
  (bind CopyAttributes_0 (Copy%20Attributes :InputSource $Density :OutputTarget TextureValue))
  (bind DeleteAttributes_4 (Delete%20Attributes :Operation KeepSelectedAttributes :SelectedAttributes "Filter Threshold"))
  (bind AttributeFilter_0 (Filter%20Attribute%20Elements :Operator GreaterOrEqual :TargetAttribute TextureValue :ThresholdAttribute @Last))
  (bind DeleteAttributes_3 (Delete%20Attributes :Operation KeepSelectedAttributes :SelectedAttributes "Invert Filter"))
  (bind Select_0 (Select))
  (bind DeleteAttributes_6 (Delete%20Attributes :SelectedAttributes "TextureValue"))
  (bind Gather_0 (Gather))
  (connect DeleteAttributes_4 Out AttributeFilter_0 Filter)
  (connect CopyAttributes_0 Out AttributeFilter_0 In)
  (connect CreateAttribute_0 Out AttributeMathsOp_0 InA)
  (connect BreakVectorAttribute_0 X AttributeMathsOp_0 InB)
  (connect AttributeMathsOp_0 Out AttributeMathsOp_1 InA)
  (connect BreakVectorAttribute_0 Y AttributeMathsOp_1 InB)
  (connect CreateAttribute_1 Out AttributeMathsOp_2 InA)
  (connect BreakVectorAttribute_0 Z AttributeMathsOp_2 InB)
  (connect AttributeMathsOp_2 Out AttributeMathsOp_3 InA)
  (connect BreakVectorAttribute_0 W AttributeMathsOp_3 InB)
  (connect AttributeMathsOp_1 Out AttributeMathsOp_4 InA)
  (connect AttributeMathsOp_3 Out AttributeMathsOp_4 InB)
  (connect AttributeMathsOp_4 Out AttributeMathsOp_5 InA)
  (connect CreateAttribute_2 Out AttributeMathsOp_5 InB)
  (connect CreateAttribute_3 Out AttributeMathsOp_7 InA)
  (connect AttributeMathsOp_5 Out AttributeMathsOp_7 InB)
  (connect AttributeMathsOp_7 Out AttributeMathsOp_8 InA)
  (connect AttributeMathsOp_9 Out AttributeMathsOp_8 InB)
  (connect AttributeMathsOp_4 Out AttributeMathsOp_9 In)
  (connect GetBounds_2 Out AttributeMathsOp_15 InA)
  (connect GetBounds_2 Out AttributeMathsOp_15 InB)
  (connect AttributeMathsOp_15 Out AttributeMathsOp_16 InA)
  (connect CreateAttribute_5 Out AttributeMathsOp_16 InB)
  (connect CreatePoints_0 Out AttributeMathsOp_17 InA)
  (connect AttributeMathsOp_16 Out AttributeMathsOp_17 InB)
  (connect GetBounds_2 Out AttributeMathsOp_18 InA)
  (connect GetBounds_2 Out AttributeMathsOp_18 InB)
  (connect AttributeMathsOp_18 Out AttributeMathsOp_19 InA)
  (connect CreateAttribute_5 Out AttributeMathsOp_19 InB)
  (connect AttributeMathsOp_17 Out AttributeMathsOp_20 InA)
  (connect AttributeMathsOp_19 Out AttributeMathsOp_20 InB)
  (connect CreatePoints_0 Out AttributeTransformOp_0 InA)
  (connect DefaultInputNode WorldTextureData AttributeTransformOp_0 InB)
  (connect GetPointsCount_0 Out Branch_0 bOutputToB)
  (connect DeleteAttributes_1 Out BreakVectorAttribute_0 In)
  (connect Subgraph_0 ColorAttribute CopyAttributes_0 InputSource)
  (connect SampleTexture_1 Out CopyAttributes_0 Source)
  (connect SampleTexture_1 Out CopyAttributes_0 Target)
  (connect DefaultInputNode WorldTextureData DeleteAttributes_0 In)
  (connect DefaultInputNode WorldTextureData DeleteAttributes_2 In)
  (connect DefaultInputNode WorldTextureData DeleteAttributes_3 In)
  (connect DefaultInputNode WorldTextureData DeleteAttributes_4 In)
  (connect DefaultInputNode WorldTextureData DeleteAttributes_5 In)
  (connect Select_0 Out DeleteAttributes_6 In)
  (connect Branch_0 Output%20A Gather_0 In)
  (connect DeleteAttributes_6 Out Gather_0 In)
  (connect DefaultInputNode Points GetBounds_2 In)
  (connect Projection_0 Out GetPointsCount_0 In)
  (connect DeleteAttributes_5 Out GetTextureData_1 ColorChannel)
  (connect DeleteAttributes_0 Out GetTextureData_1 Texture)
  (connect DeleteAttributes_2 Out GetTextureData_1 Transform)
  (connect AttributeMathsOp_20 Out Projection_0 In)
  (connect AttributeTransformOp_0 Out Projection_0 Projection%20Target)
  (connect GetTextureData_1 Out SampleTexture_1 BaseTexture)
  (connect Branch_0 Output%20B SampleTexture_1 Point)
  (connect DefaultInputNode Points SampleTexture_1 Point)
  (connect DeleteAttributes_3 Out Select_0 bUseInputB)
  (connect AttributeFilter_0 InsideFilter Select_0 Input%20A)
  (connect AttributeFilter_0 OutsideFilter Select_0 Input%20B)
  (connect DeleteAttributes_5 Out Subgraph_0 Index)
  (output Points Gather_0)
  (output Texture GetTextureData_1))
