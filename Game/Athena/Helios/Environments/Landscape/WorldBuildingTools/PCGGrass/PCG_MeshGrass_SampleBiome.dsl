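; PCG subgraph for the mesh grass tools that samples one biome's mask texture onto target points and writes an integer Biome attribute.
; Points are sampled only when the biome is active with valid coordinates; otherwise they pass through.
; A subgraph decodes coordinates into texture offset, tiling and channel, supporting atlases, while global controls can override projection size.
; Projected color is scaled by strength, floored and clamped, then multiplied by BiomeIndex in the decoded channel.
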(graph PCG_MeshGrass_SampleBiome
  (inputs (Target%20Point :type Point) (Biome%20Attributes :type Param) (Global%20Attributes :type Param))
  (bind DeleteAttributes_5 (Delete%20Attributes :Operation KeepSelectedAttributes :SelectedAttributes "BiomeTextureIsAtlas"))
  (bind ExecuteBlueprint_0 (Execute%20Blueprint :BlueprintElementInstance (class:/Game/Athena/Helios/Environments/Landscape/WorldBuildingTools/PCGGrass/PCGBP_RenameAttributeSet.PCGBP_RenameAttributeSet_C :Chop true :ReverseIndex true :TruncationIndex 7 :UseTruncation true) :BlueprintElementType /Game/Athena/Helios/Environments/Landscape/WorldBuildingTools/PCGGrass/PCGBP_RenameAttributeSet.PCGBP_RenameAttributeSet_C))
  (bind DeleteAttributes_10 (Delete%20Attributes :bTokenizeOnWhiteSpace true :Operation KeepSelectedAttributes :SelectedAttributes "BiomeCoordinates"))
  (bind Subgraph_0 (Subgraph :Graph /Game/Athena/Helios/Environments/Landscape/WorldBuildingTools/PCGGrass/PCG_MeshGrass_DecodeBiomeCoordinates.PCG_MeshGrass_DecodeBiomeCoordinates))
  (bind DeleteAttributes_9 (Delete%20Attributes :bTokenizeOnWhiteSpace true :Operation KeepSelectedAttributes :SelectedAttributes "Strength"))
  (bind CreateAttribute_4 (Create%20Constant :AttributeTypes (struct :Type Boolean :BoolValue true)))
  (bind DeleteAttributes_2 (Delete%20Attributes :bTokenizeOnWhiteSpace true :Operation KeepSelectedAttributes :SelectedAttributes "Active"))
  (bind AttributeCompareOp_3 (Attribute%20Compare%20Op))
  (bind CreateAttribute_1 (Create%20Constant))
  (bind AttributeCompareOp_0 (Attribute%20Compare%20Op :InputSource1 BiomeCoordinates.Y :Operation GreaterOrEqual :OutputTarget HasValidCoordinates))
  (bind DeleteAttributes_4 (Delete%20Attributes :bTokenizeOnWhiteSpace true :Operation KeepSelectedAttributes :SelectedAttributes "HasValidCoordinates"))
  (bind AttributeBooleanOp_5 (Attribute%20Boolean%20Op))
  (bind Branch_2 (Branch))
  (bind GetActorProperty_0 (Get%20Actor%20Property :ActorSelector (struct :ActorFilter AllWorldActors :ActorSelection ByClass :ActorSelectionClass /Game/Athena/Helios/Environments/Landscape/WorldBuildingTools/PCGGrass/BP_MeshGrass_GlobalControls.BP_MeshGrass_GlobalControls_C) :PropertyName UseCustomBiomeTextureProjectionSize))
  (bind CreateConstant_0 (Create%20Constant :AttributeTypes (struct :Type Vector :VectorValue (260096.0 260096.0 1.0))))
  (bind MakeTransformAttribute_0 (Make%20Transform%20Attribute))
  (bind CreateConstant_5 (Create%20Constant :AttributeTypes (struct :Type Float :FloatValue 1.0)))
  (bind GetActorProperty_1 (Get%20Actor%20Property :ActorSelector (struct :ActorFilter AllWorldActors :ActorSelection ByClass :ActorSelectionClass /Game/Athena/Helios/Environments/Landscape/WorldBuildingTools/PCGGrass/BP_MeshGrass_GlobalControls.BP_MeshGrass_GlobalControls_C) :PropertyName CustomBiomeTextureProjectionSize))
  (bind MakeVectorAttribute_2 (Make%20Vector%20Attribute :OutputType Vector))
  (bind MakeTransformAttribute_1 (Make%20Transform%20Attribute))
  (bind Select_0 (Select))
  (bind CreateAttribute_9 (Create%20Constant :AttributeTypes (struct :Type SoftObjectPath)))
  (bind DeleteAttributes_1 (Delete%20Attributes :bTokenizeOnWhiteSpace true :Operation KeepSelectedAttributes :SelectedAttributes "BiomeTexture"))
  (bind Subgraph_7 (Subgraph :Graph /Game/Athena/Helios/Environments/Landscape/WorldBuildingTools/PCGGrass/PCG_HasValidData.PCG_HasValidData))
  (bind Select_8 (Select))
  (bind GetTextureData_3 (Get%20Texture%20Data :bForceEditorOnlyCPUSampling true :bSynchronousLoad true :bUseAbsoluteTransform true :bUseAdvancedTiling true :bUseDensitySourceChannel false :ColorChannel Red :TexelSize 1e+02 :Texture /Engine/EngineMaterials/Good64x64TilingNoiseHighFreq.Good64x64TilingNoiseHighFreq :Transform (struct :Scale3D (260096.0 260096.0 1.0))))
  (bind Projection_2 (Projection :bKeepZeroDensityPoints true :ProjectionParams (struct :bProjectPositions false :bProjectRotations false :ColorBlendMode Multiply)))
  (bind AddAttribute_8 (Add%20Attribute :OutputTarget BiomeStrength))
  (bind AttributeMathsOp_4 (Attribute%20Maths%20Op :InputSource1 $Color :Operation Multiply :OutputTarget $Color))
  (bind AttributeMathsOp_5 (Attribute%20Maths%20Op :Operation Floor))
  (bind AttributeMathsOp_2 (Attribute%20Maths%20Op :InputSource1 $Color :Operation Min))
  (bind DeleteAttributes_3 (Delete%20Attributes :bTokenizeOnWhiteSpace true :Operation KeepSelectedAttributes :SelectedAttributes "BiomeIndex"))
  (bind AttributeMathsOp_1 (Attribute%20Maths%20Op :InputSource1 $Color.R :Operation Multiply :OutputTarget $Color.R))
  (bind AttributeCast_0 (Attribute%20Cast :OutputTarget Biome :OutputType Integer32))
  (connect DeleteAttributes_9 Out AddAttribute_8 Attributes)
  (connect Projection_2 Out AddAttribute_8 In)
  (connect DeleteAttributes_4 Out AttributeBooleanOp_5 InA)
  (connect AttributeCompareOp_3 Out AttributeBooleanOp_5 InB)
  (connect AttributeMathsOp_1 Out AttributeCast_0 In)
  (connect Subgraph_0 Color%20Attribute AttributeCast_0 InputSource)
  (connect DeleteAttributes_10 Out AttributeCompareOp_0 InA)
  (connect CreateAttribute_1 Out AttributeCompareOp_0 InB)
  (connect DeleteAttributes_2 Out AttributeCompareOp_3 InA)
  (connect CreateAttribute_4 Out AttributeCompareOp_3 InB)
  (connect AttributeMathsOp_2 Out AttributeMathsOp_1 InA)
  (connect DeleteAttributes_3 Out AttributeMathsOp_1 InB)
  (connect Subgraph_0 Color%20Attribute AttributeMathsOp_1 InputSource1)
  (connect Subgraph_0 Color%20Attribute AttributeMathsOp_1 OutputTarget)
  (connect AttributeMathsOp_5 Out AttributeMathsOp_2 InA)
  (connect AddAttribute_8 Out AttributeMathsOp_4 InA)
  (connect AddAttribute_8 Out AttributeMathsOp_4 InB)
  (connect AttributeMathsOp_4 Out AttributeMathsOp_5 In)
  (connect AttributeBooleanOp_5 Out Branch_2 bOutputToB)
  (connect DefaultInputNode Target%20Point Branch_2 In)
  (connect ExecuteBlueprint_0 Out DeleteAttributes_1 In)
  (connect ExecuteBlueprint_0 Out DeleteAttributes_2 In)
  (connect ExecuteBlueprint_0 Out DeleteAttributes_3 In)
  (connect AttributeCompareOp_0 Out DeleteAttributes_4 In)
  (connect DefaultInputNode Global%20Attributes DeleteAttributes_5 In)
  (connect ExecuteBlueprint_0 Out DeleteAttributes_9 In)
  (connect ExecuteBlueprint_0 Out DeleteAttributes_10 In)
  (connect DefaultInputNode Biome%20Attributes ExecuteBlueprint_0 In)
  (connect Subgraph_0 Texture%20Offset GetTextureData_3 CenterOffset)
  (connect Select_8 Out GetTextureData_3 Texture)
  (connect Subgraph_0 Tiling GetTextureData_3 Tiling)
  (connect Select_0 Out GetTextureData_3 Transform)
  (connect CreateConstant_0 Out MakeTransformAttribute_0 Scale)
  (connect MakeVectorAttribute_2 Out MakeTransformAttribute_1 Scale)
  (connect GetActorProperty_1 Out MakeVectorAttribute_2 X)
  (connect GetActorProperty_1 Out MakeVectorAttribute_2 Y)
  (connect CreateConstant_5 Out MakeVectorAttribute_2 Z)
  (connect Branch_2 Output%20B Projection_2 In)
  (connect GetTextureData_3 Out Projection_2 Projection%20Target)
  (connect GetActorProperty_0 Out Select_0 bUseInputB)
  (connect MakeTransformAttribute_0 Out Select_0 Input%20A)
  (connect MakeTransformAttribute_1 Out Select_0 Input%20B)
  (connect Subgraph_7 IsValid Select_8 bUseInputB)
  (connect CreateAttribute_9 Out Select_8 Input%20A)
  (connect DeleteAttributes_1 Out Select_8 Input%20B)
  (connect DeleteAttributes_10 Out Subgraph_0 Biome%20Texture%20Coordinates)
  (connect DeleteAttributes_5 Out Subgraph_0 IsAtlas)
  (connect Branch_2 Output%20B Subgraph_7 In)
  (output Out AttributeCast_0)
  (output Out Branch_2.Output%20A))
