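; PCG graph that spawns a single BP_MeshGrass_ScalabilityManager actor for the mesh grass system.
; IsUnattended and HasScalabilityDLs Blueprint checks are combined to gate execution.
; It gathers all BP_MeshGrass_Partitioner actors in the world and selects the one with the largest scaled horizontal extents.
; Comparing that actor with the graph's own acts as an election, so only the largest partitioner creates the manager, attached in the graph folder.
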(graph PCG_MeshGrass_ScalabilityManager
  (bind ExecuteBlueprint_0 (Execute%20Blueprint :BlueprintElementInstance (class:/Game/Athena/Helios/Environments/Landscape/WorldBuildingTools/PCGGrass/PCGBP_IsUnattended.PCGBP_IsUnattended_C) :BlueprintElementType /Game/Athena/Helios/Environments/Landscape/WorldBuildingTools/PCGGrass/PCGBP_IsUnattended.PCGBP_IsUnattended_C))
  (bind ExecuteBlueprint_1 (Execute%20Blueprint :BlueprintElementInstance (class:/Game/Athena/Helios/Environments/Landscape/WorldBuildingTools/PCGGrass/PCGBP_HasScalabilityDLs.PCGBP_HasScalabilityDLs_C) :BlueprintElementType /Game/Athena/Helios/Environments/Landscape/WorldBuildingTools/PCGGrass/PCGBP_HasScalabilityDLs.PCGBP_HasScalabilityDLs_C))
  (bind AttributeBooleanOp_0 (Attribute%20Boolean%20Op))
  (bind Branch_1 (Branch))
  (bind GetActorData_2 (Get%20Actor%20Data :Mode GetActorReference))
  (bind GetActorData_1 (Get%20Actor%20Data :ActorSelector (struct :ActorFilter AllWorldActors :ActorSelection ByClass :ActorSelectionClass /Game/Athena/Helios/Environments/Landscape/WorldBuildingTools/PCGGrass/BP_MeshGrass_Partitioner.BP_MeshGrass_Partitioner_C :bSelectMultiple true) :bComponentsMustOverlapSelf false :bMergeSinglePointData true :Mode GetSinglePoint))
  (bind ApplyScaleToBounds_1 (Apply%20Scale%20To%20Bounds))
  (bind AttributeSelect_0 (Attribute%20Select :Axis CustomAxis :CustomAxis (1.0 1.0 0.0 0.0) :InputSource $Extents :Operation Max))
  (bind PointToAttributeSet_0 (Point%20To%20Attribute%20Set))
  (bind AttributeCompareOp_0 (Attribute%20Compare%20Op :InputSource1 ActorReference))
  (bind Branch_2 (Branch))
  (bind CreateTargetActor_0 (Create%20Target%20Actor :AttachOptions InGraphFolder :TemplateActorClass /Game/Athena/Helios/Environments/Landscape/WorldBuildingTools/PCGGrass/BP_MeshGrass_ScalabilityManager.BP_MeshGrass_ScalabilityManager_C))
  (bind Gather_1 (Gather))
  (bind CreateConstant_1 (Create%20Constant :AttributeTypes (struct :Type Boolean :BoolValue true)))
  (connect GetActorData_1 Out ApplyScaleToBounds_1 In)
  (connect ExecuteBlueprint_0 Out AttributeBooleanOp_0 InA)
  (connect ExecuteBlueprint_1 Out AttributeBooleanOp_0 InB)
  (connect PointToAttributeSet_0 Out AttributeCompareOp_0 InA)
  (connect GetActorData_2 Out AttributeCompareOp_0 InB)
  (connect ApplyScaleToBounds_1 Out AttributeSelect_0 In)
  (connect AttributeBooleanOp_0 Out Branch_1 bOutputToB)
  (connect AttributeBooleanOp_0 Out Branch_1 In)
  (connect AttributeCompareOp_0 Out Branch_2 bOutputToB)
  (connect Branch_1 Output%20B Branch_2 In)
  (connect Branch_2 Output%20B CreateTargetActor_0 Execution%20Dependency)
  (connect CreateTargetActor_0 Out Gather_1 In)
  (connect AttributeSelect_0 Point PointToAttributeSet_0 In))
