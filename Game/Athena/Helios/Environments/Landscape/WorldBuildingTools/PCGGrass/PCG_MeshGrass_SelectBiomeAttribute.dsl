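; PCG subgraph for mesh grass that selects one named attribute, given by the TargetAttribute parameter, from biome attribute data.
; A TargetBiomeIndex of -1 passes InAttributes straight through; otherwise the FormatBiomeAttributes subgraph formats them and a GetDataByIndex Blueprint picks that biome's entry.
; All attributes except the target are then deleted.
; An attribute count check branches the result so it is output only when the attribute exists.
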(graph PCG_MeshGrass_SelectBiomeAttribute
  (params (TargetAttribute :type String))
  (inputs (InAttributes :type Param) (TargetBiomeIndex :type Param))
  (outputs (Out%20Attribute :type Param))
  (bind CreateAttribute_1 (Create%20Constant :AttributeTypes (struct :Type Integer32 :Int32Value -1 :DoubleValue -1.0)))
  (bind AttributeCompareOp_0 (Attribute%20Compare%20Op))
  (bind Branch_2 (Branch))
  (bind Subgraph_3 (Subgraph :Graph /Game/Athena/Helios/Environments/Landscape/WorldBuildingTools/PCGGrass/PCG_MeshGrass_FormatBiomeAttributes.PCG_MeshGrass_FormatBiomeAttributes))
  (bind ExecuteBlueprint_2 (Execute%20Blueprint :BlueprintElementInstance (class:/Game/Athena/Helios/Environments/Landscape/WorldBuildingTools/PCGGrass/PCGBP_GetDataByIndex.PCGBP_GetDataByIndex_C) :BlueprintElementType /Game/Athena/Helios/Environments/Landscape/WorldBuildingTools/PCGGrass/PCGBP_GetDataByIndex.PCGBP_GetDataByIndex_C))
  (bind DeleteAttributes_5 (Delete%20Attributes :bTokenizeOnWhiteSpace true :Operation KeepSelectedAttributes :SelectedAttributes-param TargetAttribute))
  (bind GetAttributeList_3 (Get%20Attribute%20List))
  (bind GetEntriesCount_4 (Get%20Element%20Count))
  (bind CreateAttribute_8 (Create%20Constant :AttributeTypes (struct :Type Integer32)))
  (bind AttributeCompareOp_7 (Attribute%20Compare%20Op :Operation NotEqual))
  (bind Branch_9 (Branch))
  (connect DefaultInputNode TargetBiomeIndex AttributeCompareOp_0 InA)
  (connect CreateAttribute_1 Out AttributeCompareOp_0 InB)
  (connect GetEntriesCount_4 Out AttributeCompareOp_7 InA)
  (connect CreateAttribute_8 Out AttributeCompareOp_7 InB)
  (connect AttributeCompareOp_0 Out Branch_2 bOutputToB)
  (connect DefaultInputNode InAttributes Branch_2 In)
  (connect AttributeCompareOp_7 Out Branch_9 bOutputToB)
  (connect DeleteAttributes_5 Out Branch_9 In)
  (connect Branch_2 Output%20B DeleteAttributes_5 In)
  (connect ExecuteBlueprint_2 OutData DeleteAttributes_5 In)
  (connect DefaultInputNode TargetBiomeIndex ExecuteBlueprint_2 DataIndex)
  (connect Subgraph_3 BiomeAttributes ExecuteBlueprint_2 InData)
  (connect DeleteAttributes_5 Out GetAttributeList_3 In)
  (connect GetAttributeList_3 Out GetEntriesCount_4 In)
  (connect Branch_2 Output%20A Subgraph_3 InAttributes)
  (output Out%20Attribute Branch_9.Output%20B))
