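; PCG graph that spawns mesh grass into BP_MeshGrass actors, labeled by partition and data layer.
; It sets DetailMode by destructibility and per-mesh point count, then spawns through HISM or default instanced components, split by count, with attach data passed to InitPostSpawn.
; A conditional pass copies DetailMode 1 grass into a separate MobileHighEnd data-layer actor.
; Analytics print component and instance counts, checked against threshold parameters.
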(graph PCG_MeshGrass_SpawnGrass
  (params (ComponentsThreshold :type Int :default 7) (InstancesThreshold :type Int :default 1200))
  (inputs (InPoints :type Point) (DataLayers :type Param) (PartitionAttributes :type Param))
  (outputs (OutActors :type Param))
  ;[0.05 0.05 0.05] Points Processing and Spawning
  (bind CreateAttribute_5 (Create%20Constant :AttributeTypes (struct :DoubleValue 2.0)))
  (bind AttributeCompareOp_1 (Attribute%20Compare%20Op :InputSource1 Interactivity :Operation GreaterOrEqual :OutputTarget Destruction))
  (bind AttributeFilter_0 (Filter%20Attribute%20Elements :AttributeTypes (struct :Type Boolean :BoolValue true) :bGenerateOutputDataEvenIfEmpty false :bUseConstantThreshold true :Operator NotEqual :TargetAttribute Destruction :ThresholdAttribute @Last))
  (bind Gather_1 (Gather))
  (bind AttributePartition_2 (Attribute%20Partition :PartitionAttributeSelectors [Mesh]))
  (bind FilterDataByAttribute_0 (Filter%20Data%20By%20Attribute :Attribute PCG_MeshGrass_CustomScalability :FilterMode FilterByValue :FilterOperator Equal :TargetAttribute PCG_MeshGrass_CustomScalability :Threshold (struct :AttributeTypes (struct :Type Boolean :BoolValue true))))
  (bind AttributeFilter_1 (Filter%20Attribute%20Elements :AttributeTypes (struct :Type Integer32 :BoolValue true) :bGenerateOutputDataEvenIfEmpty false :bUseConstantThreshold true :Operator Equal :TargetAttribute DetailMode :ThresholdAttribute @Last))
  (bind CreateAttribute_1 (Create%20Constant :AttributeTypes (struct :Type Integer32 :Int32Value 10)))
  (bind ExecuteBlueprint_0 (Execute%20Blueprint :BlueprintElementInstance (class:/Game/Athena/Helios/Environments/Landscape/WorldBuildingTools/PCGGrass/PCGBP_FilterPointDataByPointCount.PCGBP_FilterPointDataByPointCount_C) :BlueprintElementType /Game/Athena/Helios/Environments/Landscape/WorldBuildingTools/PCGGrass/PCGBP_FilterPointDataByPointCount.PCGBP_FilterPointDataByPointCount_C))
  (bind CreateAttribute_3 (Create%20Constant :AttributeTypes (struct :Type Integer32 :Int32Value 199)))
  (bind CreateAttribute_7 (Create%20Constant :AttributeTypes (struct :Type Integer64 :DoubleValue 3.0 :IntValue 3) :OutputTarget DetailMode))
  (bind AttributeMathsOp_5 (Attribute%20Maths%20Op :InputSource1 DetailMode :Operation Set))
  (bind MergePoints_0 (Merge%20Points))
  (bind PointToAttributeSet_1 (Point%20To%20Attribute%20Set :PropertiesToConvert (map ($Position Location))))
  (bind DeleteAttributes_3 (Delete%20Attributes :Operation KeepSelectedAttributes :SelectedAttributes "Location,AttachTarget"))
  (bind FilterDataByAttribute_1 (Filter%20Data%20By%20Attribute :Attribute AttachTarget))
  (bind AttributeFilter_3 (Filter%20Attribute%20Elements :AttributeTypes (struct :Type SoftObjectPath) :bUseConstantThreshold true :bWarnOnDataMissingAttribute false :Operator NotEqual :TargetAttribute AttachTarget :ThresholdAttribute @Last))
  (bind CreateTargetActor_1 (Create%20Target%20Actor :AttachOptions InGraphFolder :DataLayerSettings (struct :DataLayerSourceType DataLayerReferences) :TemplateActorClass /Game/Athena/Helios/Environments/Landscape/WorldBuildingTools/PCGGrass/BP_MeshGrass.BP_MeshGrass_C))
  (bind CreateAttribute_20 (Create%20Constant :AttributeTypes (struct :Type Integer32 :Int32Value 300)))
  (bind ExecuteBlueprint_19 (Execute%20Blueprint :BlueprintElementInstance (class:/Game/Athena/Helios/Environments/Landscape/WorldBuildingTools/PCGGrass/PCGBP_FilterPointDataByPointCount.PCGBP_FilterPointDataByPointCount_C) :BlueprintElementType /Game/Athena/Helios/Environments/Landscape/WorldBuildingTools/PCGGrass/PCGBP_FilterPointDataByPointCount.PCGBP_FilterPointDataByPointCount_C))
  (bind StaticMeshSpawner_9 (Static%20Mesh%20Spawner :bAllowDescriptorChanges false :bApplyMeshBoundsToPoints false :MeshSelectorParameters (class:/Script/PCG.PCGMeshSelectorByAttribute :AttributeName Mesh :bUseAttributeMaterialOverrides true :MaterialOverrideAttributes [OverrideMaterial OverrideMaterial1] :TemplateDescriptor (struct :ComponentClass /Script/Engine.HierarchicalInstancedStaticMeshComponent :InstanceStartCullDistance 3500 :InstanceEndCullDistance 5000 :bCastShadow false :bAffectDynamicIndirectLighting false :bAffectDistanceFieldLighting false :bReceivesDecals false :bVisibleInRayTracing false :bIncludeInHLOD false :bForceNavigationObstacle false)) :MeshSelectorType /Script/PCG.PCGMeshSelectorByAttribute :StaticMeshComponentPropertyOverrides [(struct :InputSource DetailMode :PropertyTarget "DetailMode") (struct :InputSource InstanceStartCullDistance :PropertyTarget "InstanceStartCullDistance") (struct :InputSource InstanceEndCullDistance :PropertyTarget "InstanceEndCullDistance") (struct :InputSource WPODisableDistance :PropertyTarget "WorldPositionOffsetDisableDistance") (struct :InputSource RenderingQuality :PropertyTarget "bCastShadow") (struct :InputSource RenderingQuality :PropertyTarget "bAffectDynamicIndirectLighting") (struct :InputSource RenderingQuality :PropertyTarget "bAffectDistanceFieldLighting") (struct :InputSource RenderingQuality :PropertyTarget "bVisibleInRayTracing") (struct :InputSource SDPTags :PropertyTarget "AdditionalCommaSeparatedTags")]))
  (bind StaticMeshSpawner_16 (Static%20Mesh%20Spawner :bAllowDescriptorChanges false :bApplyMeshBoundsToPoints false :MeshSelectorParameters (class:/Script/PCG.PCGMeshSelectorByAttribute :AttributeName Mesh :bUseAttributeMaterialOverrides true :MaterialOverrideAttributes [OverrideMaterial OverrideMaterial1] :TemplateDescriptor (struct :InstanceStartCullDistance 3500 :InstanceEndCullDistance 5000 :bCastShadow false :bAffectDynamicIndirectLighting false :bAffectDistanceFieldLighting false :bReceivesDecals false :bVisibleInRayTracing false :bIncludeInHLOD false :bForceNavigationObstacle false)) :MeshSelectorType /Script/PCG.PCGMeshSelectorByAttribute :StaticMeshComponentPropertyOverrides [(struct :InputSource DetailMode :PropertyTarget "DetailMode") (struct :InputSource InstanceStartCullDistance :PropertyTarget "InstanceStartCullDistance") (struct :InputSource InstanceEndCullDistance :PropertyTarget "InstanceEndCullDistance") (struct :InputSource WPODisableDistance :PropertyTarget "WorldPositionOffsetDisableDistance") (struct :InputSource RenderingQuality :PropertyTarget "bCastShadow") (struct :InputSource RenderingQuality :PropertyTarget "bAffectDynamicIndirectLighting") (struct :InputSource RenderingQuality :PropertyTarget "bAffectDistanceFieldLighting") (struct :InputSource RenderingQuality :PropertyTarget "bVisibleInRayTracing") (struct :InputSource SDPTags :PropertyTarget "AdditionalCommaSeparatedTags")]))
  (bind ApplyOnActor_8 (Apply%20On%20Object :bUseOneOverrideDataPerElement true :PostProcessFunctionNames [InitPostSpawn] :PropertyOverrideDescriptions [(struct :InputSource AttachTarget :PropertyTarget "AttachTargets") (struct :InputSource Location :PropertyTarget "AttachLocations")]))
  (bind ExecuteBlueprint_18 (Execute%20Blueprint :BlueprintElementInstance (class:/Game/Athena/Helios/Environments/Landscape/WorldBuildingTools/PCGGrass/PCGBP_FilterPointDataByPointCount.PCGBP_FilterPointDataByPointCount_C) :BlueprintElementType /Game/Athena/Helios/Environments/Landscape/WorldBuildingTools/PCGGrass/PCGBP_FilterPointDataByPointCount.PCGBP_FilterPointDataByPointCount_C))
  (bind RandomChoice_17 (Random%20Choice :bFixedMode false))
  (bind AddAttribute_10 (Add%20Attribute :AttributeTypes (struct :Type Integer32 :Int32Value 3) :OutputTarget DetailMode))
  (bind DeleteAttributes_23 (Delete%20Attributes :Operation KeepSelectedAttributes :SelectedAttributes "ActorLabel"))
  (bind CreateAttribute_37 (Create%20Constant :AttributeTypes (struct :Type String :StringValue "_")))
  (bind Subgraph_6 (Subgraph :Graph /Game/Athena/Helios/Environments/Landscape/WorldBuildingTools/PCGGrass/PCG_GateValidEntries.PCG_GateValidEntries))
  (bind AttributeCast_30 (Attribute%20Cast :OutputType String))
  (bind CreateAttribute_35 (Create%20Constant :AttributeTypes (struct :Type Integer32 :Int32Value 1 :DoubleValue 1.0)))
  (bind ExecuteBlueprint_32 (Execute%20Blueprint :BlueprintElementInstance (class:/Game/Athena/Helios/Environments/Landscape/WorldBuildingTools/PCGGrass/PCGBP_StringFind.PCGBP_StringFind_C :SearchFromEnd true :SubString "." :TargetAttribute DataLayerReference) :BlueprintElementType /Game/Athena/Helios/Environments/Landscape/WorldBuildingTools/PCGGrass/PCGBP_StringFind.PCGBP_StringFind_C))
  (bind AttributeMathsOp_34 (Attribute%20Maths%20Op))
  (bind ExecuteBlueprint_31 (Execute%20Blueprint :BlueprintElementInstance (class:/Game/Athena/Helios/Environments/Landscape/WorldBuildingTools/PCGGrass/PCGBP_TruncateString.PCGBP_TruncateString_C :Chop true :ReverseIndex true) :BlueprintElementType /Game/Athena/Helios/Environments/Landscape/WorldBuildingTools/PCGGrass/PCGBP_TruncateString.PCGBP_TruncateString_C))
  (bind AttributeStringOp_36 (Attribute%20String%20Op :InputSource2 DataLayerReference))
  (bind AttributeStringOp_24 (Attribute%20String%20Op))
  (bind Subgraph_2 (Subgraph :Graph /Game/Athena/Helios/Environments/Landscape/WorldBuildingTools/PCGGrass/PCG_HasValidEntries.PCG_HasValidEntries))
  (bind Select_38 (Select))
  (bind Gather_14 (Gather))
  (bind AddAttribute_12 (Add%20Attribute :bCopyAllAttributes true))
  (bind CreateAttribute_0 (Create%20Constant))
  (bind AttributeMathsOp_0 (Attribute%20Maths%20Op :InputSource1 DetailMode :Operation Set))
  (bind AttributePartition_0 (Attribute%20Partition :PartitionAttributeSelectors [Mesh]))
  ;[0.05 0.05 0.05] Analytics
  (bind CreateAttribute_76 (Create%20Constant :AttributeTypes (struct :Type String :StringValue " - ")))
  (bind AttributeStringOp_3 (Attribute%20String%20Op))
  (bind CreateAttribute_96 (Create%20Constant :AttributeTypes (struct :Type String :StringValue "Instances: ")))
  (bind MergePoints_83 (Merge%20Points))
  (bind GetPointsCount_84 (Get%20Element%20Count :OutputAttributeName NumInstances))
  (bind AttributeStringOp_94 (Attribute%20String%20Op))
  (bind CreateAttribute_90 (Create%20Constant :AttributeTypes (struct :Type String :StringValue "|")))
  (bind AttributeStringOp_88 (Attribute%20String%20Op))
  (bind AttributeFilter_80 (Filter%20Attribute%20Elements :AttributeTypes (struct :Type Integer32 :Int32Value 3) :bUseConstantThreshold true :Operator Equal :TargetAttribute DetailMode))
  (bind GetPointsCount_86 (Get%20Element%20Count))
  (bind AttributeStringOp_92 (Attribute%20String%20Op))
  (bind CreateAttribute_4 (Create%20Constant :AttributeTypes (struct :Type String :StringValue "Components: ")))
  (bind CreateAttribute_66 (Create%20Constant :AttributeTypes (struct :Type Integer32 :Int32Value 10 :DoubleValue 1e+01)))
  (bind AttributePartition_40 (Attribute%20Partition :PartitionAttributeSelectors [Mesh OverrideMaterial DetailMode]))
  (bind GetAttributeFromPointIndex_42 (Get%20Attribute%20From%20Point%20Index :bForceOutputAttributeToBeInElementsDomain true))
  (bind MergePoints_44 (Merge%20Points))
  (bind GetPointsCount_62 (Get%20Element%20Count :OutputAttributeName NumComponents))
  (bind AttributeCompareOp_64 (Attribute%20Compare%20Op :Operation Less))
  (bind AttributeCast_70 (Attribute%20Cast :OutputType String))
  (bind CreateAttribute_74 (Create%20Constant :AttributeTypes (struct :Type String :StringValue "0")))
  (bind AttributeStringOp_72 (Attribute%20String%20Op))
  (bind Select_68 (Select))
  (bind AttributeStringOp_5 (Attribute%20String%20Op))
  (bind CreateAttribute_50 (Create%20Constant :AttributeTypes (struct :Type String :StringValue "|")))
  (bind AttributeStringOp_48 (Attribute%20String%20Op))
  (bind CreateAttribute_56 (Create%20Constant :AttributeTypes (struct :Type Integer32 :Int32Value 10 :DoubleValue 1e+01)))
  (bind AttributeFilter_45 (Filter%20Attribute%20Elements :AttributeTypes (struct :Type Integer32 :Int32Value 3) :bUseConstantThreshold true :Operator Equal :TargetAttribute DetailMode))
  (bind GetPointsCount_54 (Get%20Element%20Count))
  (bind AttributeCompareOp_55 (Attribute%20Compare%20Op :Operation Less))
  (bind AttributeCast_58 (Attribute%20Cast :OutputType String))
  (bind CreateAttribute_60 (Create%20Constant :AttributeTypes (struct :Type String :StringValue "0")))
  (bind AttributeStringOp_59 (Attribute%20String%20Op))
  (bind Select_57 (Select))
  (bind AttributeStringOp_52 (Attribute%20String%20Op))
  (bind CreateAttribute_100 (Create%20Constant :AttributeTypes (struct :Type String :StringValue " - ")))
  (bind AttributeStringOp_98 (Attribute%20String%20Op))
  (bind AttributeStringOp_102 (Attribute%20String%20Op))
  (bind AttributeStringOp_75 (Attribute%20String%20Op))
  (bind AttributeBooleanOp_105 (Attribute%20Boolean%20Op :Operation Or))
  (bind ExecuteBlueprint_1 (Execute%20Blueprint :BlueprintElementInstance (class:/Game/Athena/Helios/Environments/Landscape/WorldBuildingTools/PCGGrass/PCGBP_IsUnattended.PCGBP_IsUnattended_C) :BlueprintElementType /Game/Athena/Helios/Environments/Landscape/WorldBuildingTools/PCGGrass/PCGBP_IsUnattended.PCGBP_IsUnattended_C))
  (bind Branch_4 (Branch))
  (bind Branch_8 (Branch))
  (bind Branch_109 (Branch))
  (bind Print%20String_1 (Print%20String :bPrefixWithGraph false :bPrefixWithNode false :PrintToScreenColor (struct :G 184 :R 255 :A 255) :Verbosity Display))
  (bind AttributeCompareOp_103 (Attribute%20Compare%20Op :Operation GreaterOrEqual :InB ComponentsThreshold))
  (bind AttributeCompareOp_106 (Attribute%20Compare%20Op :Operation GreaterOrEqual :InB InstancesThreshold))
  (bind Print%20String_2 (Print%20String :bPrefixWithGraph false :bPrefixWithNode false :PrintToScreenColor (struct :G 184 :R 255 :A 255) :Verbosity Display))
  ;[0.05 0.05 0.05] Mobile HighEnd
  (bind ExecuteBlueprint_2 (Execute%20Blueprint :BlueprintElementInstance (class:/Game/Athena/Helios/Environments/Landscape/WorldBuildingTools/PCGGrass/PCGBP_IsUnattended.PCGBP_IsUnattended_C) :BlueprintElementType /Game/Athena/Helios/Environments/Landscape/WorldBuildingTools/PCGGrass/PCGBP_IsUnattended.PCGBP_IsUnattended_C))
  (bind AttributeStringOp_0 (Attribute%20String%20Op))
  (bind ExecuteBlueprint_5 (Execute%20Blueprint :BlueprintElementInstance (class:/Game/Athena/Helios/Environments/Landscape/WorldBuildingTools/PCGGrass/PCGBP_HasScalabilityDLs.PCGBP_HasScalabilityDLs_C) :BlueprintElementType /Game/Athena/Helios/Environments/Landscape/WorldBuildingTools/PCGGrass/PCGBP_HasScalabilityDLs.PCGBP_HasScalabilityDLs_C))
  (bind AttributeBooleanOp_1 (Attribute%20Boolean%20Op))
  (bind Branch_0 (Branch))
  (bind AttributeFilter_2 (Filter%20Attribute%20Elements :AttributeTypes (struct :Type Integer32 :Int32Value 1) :bUseConstantThreshold true :Operator Equal :TargetAttribute DetailMode :ThresholdAttribute @Last))
  (bind RemoveEmptyData_0 (Remove%20Empty%20Data))
  (bind Subgraph_0 (Subgraph :Graph /Game/Athena/Helios/Environments/Landscape/WorldBuildingTools/PCGGrass/PCG_HasValidData.PCG_HasValidData))
  (bind Branch_1 (Branch))
  (bind CreateConstant_3 (Create%20Constant :AttributeTypes (struct :Type SoftObjectPath :SoftObjectPathValue (struct :AssetPath (struct :PackageName /Game/Athena/Helios/Environments/Landscape/WorldBuildingTools/PCGGrass/Resources/DL_MeshGrass_MobileHighEnd :AssetName DL_MeshGrass_MobileHighEnd))) :OutputTarget DataLayerReference))
  (bind MergeAttributes_0 (Merge%20Attributes))
  (bind CreateTargetActor_0 (Create%20Target%20Actor :AttachOptions InGraphFolder :DataLayerSettings (struct :DataLayerSourceType DataLayerReferences) :TemplateActorClass /Game/Athena/Helios/Environments/Landscape/WorldBuildingTools/PCGGrass/BP_MeshGrass.BP_MeshGrass_C))
  (bind CreateConstant_2 (Create%20Constant :AttributeTypes (struct :Type Integer32)))
  (bind AttributeMathsOp_1 (Attribute%20Maths%20Op :InputSource1 DetailMode :Operation Set))
  (bind StaticMeshSpawner_0 (Static%20Mesh%20Spawner :bAllowDescriptorChanges false :bApplyMeshBoundsToPoints false :MeshSelectorParameters (class:/Script/PCG.PCGMeshSelectorByAttribute :AttributeName Mesh :bUseAttributeMaterialOverrides true :MaterialOverrideAttributes [OverrideMaterial OverrideMaterial1] :TemplateDescriptor (struct :InstanceStartCullDistance 3500 :InstanceEndCullDistance 5000 :bCastShadow false :bAffectDynamicIndirectLighting false :bAffectDistanceFieldLighting false :bReceivesDecals false :bVisibleInRayTracing false :bIncludeInHLOD false :bForceNavigationObstacle false)) :MeshSelectorType /Script/PCG.PCGMeshSelectorByAttribute :StaticMeshComponentPropertyOverrides [(struct :InputSource DetailMode :PropertyTarget "DetailMode") (struct :InputSource InstanceStartCullDistance :PropertyTarget "InstanceStartCullDistance") (struct :InputSource InstanceEndCullDistance :PropertyTarget "InstanceEndCullDistance") (struct :InputSource WPODisableDistance :PropertyTarget "WorldPositionOffsetDisableDistance") (struct :InputSource RenderingQuality :PropertyTarget "bCastShadow") (struct :InputSource RenderingQuality :PropertyTarget "bAffectDynamicIndirectLighting") (struct :InputSource RenderingQuality :PropertyTarget "bAffectDistanceFieldLighting") (struct :InputSource RenderingQuality :PropertyTarget "bVisibleInRayTracing") (struct :InputSource SDPTags :PropertyTarget "AdditionalCommaSeparatedTags")]))
  (bind ApplyOnActor_0 (Apply%20On%20Object :PostProcessFunctionNames [InitPostSpawn]))
  (bind CreateConstant_0 (Create%20Constant :AttributeTypes (struct :Type Boolean :BoolValue true)))
  (reroute ActorLabel Select_38)
  (reroute DataLayerReference DataLayers)
  (reroute Destructible AttributePartition_0)
  (reroute ExcessiveComponents AttributeCompareOp_103)
  (reroute ExcessiveInstances AttributeCompareOp_106)
  (reroute NamedRerouteDeclaration_21 PartitionAttributes)
  (reroute NamedRerouteDeclaration_25 InPoints)
  (reroute NumComponents GetPointsCount_62)
  (reroute NumInstances GetPointsCount_84)
  (reroute OutPoints Gather_14)
  (connect RandomChoice_17 Discarded AddAttribute_10 In)
  (connect NumComponents Out AddAttribute_12 Attributes)
  (connect ActorLabel Out AddAttribute_12 Attributes)
  (connect NumInstances Out AddAttribute_12 Attributes)
  (connect CreateTargetActor_1 Out AddAttribute_12 In)
  (connect StaticMeshSpawner_0 Out ApplyOnActor_0 Execution%20Dependency)
  (connect CreateTargetActor_0 Out ApplyOnActor_0 In)
  (connect StaticMeshSpawner_9 Out ApplyOnActor_8 Execution%20Dependency)
  (connect StaticMeshSpawner_16 Out ApplyOnActor_8 Execution%20Dependency)
  (connect CreateTargetActor_1 Out ApplyOnActor_8 In)
  (connect AttributeFilter_3 InsideFilter ApplyOnActor_8 Property%20Overrides)
  (connect ExecuteBlueprint_2 Out AttributeBooleanOp_1 InA)
  (connect ExecuteBlueprint_5 Out AttributeBooleanOp_1 InB)
  (connect ExcessiveComponents Out AttributeBooleanOp_105 InA)
  (connect ExcessiveInstances Out AttributeBooleanOp_105 InB)
  (connect Subgraph_6 ValidEntires AttributeCast_30 In)
  (connect GetPointsCount_54 Out AttributeCast_58 In)
  (connect GetPointsCount_62 Out AttributeCast_70 In)
  (connect NamedRerouteDeclaration_25 Out AttributeCompareOp_1 InA)
  (connect CreateAttribute_5 Out AttributeCompareOp_1 InB)
  (connect GetPointsCount_54 Out AttributeCompareOp_55 InA)
  (connect CreateAttribute_56 Out AttributeCompareOp_55 InB)
  (connect GetPointsCount_62 Out AttributeCompareOp_64 InA)
  (connect CreateAttribute_66 Out AttributeCompareOp_64 InB)
  (connect GetPointsCount_54 Out AttributeCompareOp_103 InA)
  (connect GetPointsCount_86 Out AttributeCompareOp_106 InA)
  (connect AttributeCompareOp_1 Out AttributeFilter_0 In)
  (connect FilterDataByAttribute_0 OutsideFilter AttributeFilter_1 In)
  (connect Branch_0 Output%20B AttributeFilter_2 Execution%20Dependency)
  (connect OutPoints Out AttributeFilter_2 In)
  (connect FilterDataByAttribute_1 InsideFilter AttributeFilter_3 In)
  (connect MergePoints_44 Out AttributeFilter_45 In)
  (connect MergePoints_83 Out AttributeFilter_80 In)
  (connect AttributeFilter_0 OutsideFilter AttributeMathsOp_0 InA)
  (connect CreateAttribute_0 Out AttributeMathsOp_0 InB)
  (connect CreateTargetActor_0 Out AttributeMathsOp_1 Execution%20Dependency)
  (connect AttributeFilter_2 InsideFilter AttributeMathsOp_1 InA)
  (connect CreateConstant_2 Out AttributeMathsOp_1 InB)
  (connect ExecuteBlueprint_0 OutsideFilter AttributeMathsOp_5 InA)
  (connect CreateAttribute_7 Out AttributeMathsOp_5 InB)
  (connect ExecuteBlueprint_32 Out AttributeMathsOp_34 InA)
  (connect CreateAttribute_35 Out AttributeMathsOp_34 InB)
  (connect AttributeMathsOp_0 Out AttributePartition_0 In)
  (connect Gather_1 Out AttributePartition_2 In)
  (connect OutPoints Out AttributePartition_40 In)
  (connect ActorLabel Out AttributeStringOp_0 InA)
  (connect CreateAttribute_76 Out AttributeStringOp_3 InA)
  (connect ActorLabel Out AttributeStringOp_3 InB)
  (connect CreateAttribute_4 Out AttributeStringOp_5 InA)
  (connect Select_68 Out AttributeStringOp_5 InB)
  (connect DeleteAttributes_23 Out AttributeStringOp_24 InA)
  (connect AttributeStringOp_36 Out AttributeStringOp_24 InB)
  (connect CreateAttribute_37 Out AttributeStringOp_36 InA)
  (connect ExecuteBlueprint_31 Out AttributeStringOp_36 InB)
  (connect AttributeStringOp_5 Out AttributeStringOp_48 InA)
  (connect CreateAttribute_50 Out AttributeStringOp_48 InB)
  (connect AttributeStringOp_48 Out AttributeStringOp_52 InA)
  (connect Select_57 Out AttributeStringOp_52 InB)
  (connect CreateAttribute_60 Out AttributeStringOp_59 InA)
  (connect AttributeCast_58 Out AttributeStringOp_59 InB)
  (connect CreateAttribute_74 Out AttributeStringOp_72 InA)
  (connect AttributeCast_70 Out AttributeStringOp_72 InB)
  (connect AttributeStringOp_102 Out AttributeStringOp_75 InA)
  (connect AttributeStringOp_3 Out AttributeStringOp_75 InB)
  (connect AttributeStringOp_94 Out AttributeStringOp_88 InA)
  (connect CreateAttribute_90 Out AttributeStringOp_88 InB)
  (connect AttributeStringOp_88 Out AttributeStringOp_92 InA)
  (connect GetPointsCount_86 Out AttributeStringOp_92 InB)
  (connect CreateAttribute_96 Out AttributeStringOp_94 InA)
  (connect GetPointsCount_84 Out AttributeStringOp_94 InB)
  (connect AttributeStringOp_52 Out AttributeStringOp_98 InA)
  (connect CreateAttribute_100 Out AttributeStringOp_98 InB)
  (connect AttributeStringOp_98 Out AttributeStringOp_102 InA)
  (connect AttributeStringOp_92 Out AttributeStringOp_102 InB)
  (connect AttributeBooleanOp_1 Out Branch_0 bOutputToB)
  (connect ApplyOnActor_8 Out Branch_0 In)
  (connect Subgraph_0 IsValid Branch_1 bOutputToB)
  (connect RemoveEmptyData_0 Out Branch_1 In)
  (connect ExecuteBlueprint_1 Out Branch_4 bOutputToB)
  (connect AttributeStringOp_75 Out Branch_4 In)
  (connect AttributeBooleanOp_105 Out Branch_8 bOutputToB)
  (connect Branch_4 Output%20B Branch_8 In)
  (connect AttributeBooleanOp_105 Out Branch_109 bOutputToB)
  (connect Branch_4 Output%20A Branch_109 In)
  (connect AttributeStringOp_0 Out CreateTargetActor_0 ActorLabel)
  (connect MergeAttributes_0 Out CreateTargetActor_0 DataLayerReference)
  (connect Branch_1 Output%20B CreateTargetActor_0 Execution%20Dependency)
  (connect ActorLabel Out CreateTargetActor_1 ActorLabel)
  (connect DataLayerReference Out CreateTargetActor_1 DataLayerReference)
  (connect PointToAttributeSet_1 Out DeleteAttributes_3 In)
  (connect NamedRerouteDeclaration_21 Out DeleteAttributes_23 In)
  (connect AttributeFilter_1 OutsideFilter ExecuteBlueprint_0 InPoints)
  (connect CreateAttribute_1 Out ExecuteBlueprint_0 Threshold)
  (connect AttributeMathsOp_5 Out ExecuteBlueprint_18 InPoints)
  (connect ExecuteBlueprint_0 InsideFilter ExecuteBlueprint_18 InPoints)
  (connect CreateAttribute_3 Out ExecuteBlueprint_18 Threshold)
  (connect OutPoints Out ExecuteBlueprint_19 InPoints)
  (connect CreateAttribute_20 Out ExecuteBlueprint_19 Threshold)
  (connect AttributeCast_30 Out ExecuteBlueprint_31 In)
  (connect AttributeMathsOp_34 Out ExecuteBlueprint_31 TruncationIndex)
  (connect AttributeCast_30 Out ExecuteBlueprint_32 In)
  (connect AttributePartition_2 Out FilterDataByAttribute_0 In)
  (connect DeleteAttributes_3 Out FilterDataByAttribute_1 In)
  (connect AttributeFilter_0 InsideFilter Gather_1 In)
  (connect AddAttribute_10 Out Gather_14 In)
  (connect AttributeFilter_1 InsideFilter Gather_14 In)
  (connect ExecuteBlueprint_18 OutsideFilter Gather_14 In)
  (connect Destructible Out Gather_14 In)
  (connect RandomChoice_17 Chosen Gather_14 In)
  (connect AttributePartition_40 Out GetAttributeFromPointIndex_42 In)
  (connect AttributeFilter_45 OutsideFilter GetPointsCount_54 In)
  (connect MergePoints_44 Out GetPointsCount_62 In)
  (connect MergePoints_83 Out GetPointsCount_84 In)
  (connect AttributeFilter_80 OutsideFilter GetPointsCount_86 In)
  (connect CreateConstant_3 Out MergeAttributes_0 In)
  (connect DataLayerReference Out MergeAttributes_0 In)
  (connect Destructible Out MergePoints_0 In)
  (connect GetAttributeFromPointIndex_42 Point MergePoints_44 In)
  (connect OutPoints Out MergePoints_83 In)
  (connect MergePoints_0 Out PointToAttributeSet_1 In)
  (connect Branch_8 Output%20A Print%20String_1 In)
  (connect Branch_109 Output%20B Print%20String_1 In)
  (connect AttributeStringOp_75 Out Print%20String_1 PrintString)
  (connect Branch_8 Output%20B Print%20String_2 In)
  (connect AttributeStringOp_75 Out Print%20String_2 PrintString)
  (connect ExecuteBlueprint_18 InsideFilter RandomChoice_17 In)
  (connect AttributeFilter_2 InsideFilter RemoveEmptyData_0 In)
  (connect Subgraph_2 IsValid Select_38 bUseInputB)
  (connect DeleteAttributes_23 Out Select_38 Input%20A)
  (connect AttributeStringOp_24 Out Select_38 Input%20B)
  (connect AttributeCompareOp_55 Out Select_57 bUseInputB)
  (connect GetPointsCount_54 Out Select_57 Input%20A)
  (connect AttributeStringOp_59 Out Select_57 Input%20B)
  (connect AttributeCompareOp_64 Out Select_68 bUseInputB)
  (connect GetPointsCount_62 Out Select_68 Input%20A)
  (connect AttributeStringOp_72 Out Select_68 Input%20B)
  (connect AttributeMathsOp_1 Out StaticMeshSpawner_0 In)
  (connect CreateTargetActor_0 Out StaticMeshSpawner_0 TargetActor)
  (connect ExecuteBlueprint_19 InsideFilter StaticMeshSpawner_9 In)
  (connect FilterDataByAttribute_0 InsideFilter StaticMeshSpawner_9 In)
  (connect CreateTargetActor_1 Out StaticMeshSpawner_9 TargetActor)
  (connect ExecuteBlueprint_19 OutsideFilter StaticMeshSpawner_16 In)
  (connect CreateTargetActor_1 Out StaticMeshSpawner_16 TargetActor)
  (connect RemoveEmptyData_0 Out Subgraph_0 In)
  (connect DataLayerReference Out Subgraph_2 In)
  (connect DataLayerReference Out Subgraph_6 In)
  (output OutActors AddAttribute_12))
