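; PCG subgraph that creates a target actor for the mesh grass world-building tool in the Helios landscape environment.
; It takes an actor label parameter and passes it to a Create Target Actor node using the BP_MeshGrass Blueprint as template.
; The new actor is attached in the graph's folder and returned as the TargetActor output.
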(graph PCG_MeshGrass_SpawnTargetActor
  (inputs (ActorLabel :type Param))
  (outputs (TargetActor :type Param))
  (bind CreateTargetActor_0 (Create%20Target%20Actor :AttachOptions InGraphFolder :TemplateActorClass /Game/Athena/Helios/Environments/Landscape/WorldBuildingTools/PCGGrass/BP_MeshGrass.BP_MeshGrass_C))
  (connect DefaultInputNode ActorLabel CreateTargetActor_0 ActorLabel)
  (output TargetActor CreateTargetActor_0))
