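; PCG graph for the mesh grass world-building tool that turns spline input into scattered points projected onto world surfaces.
; It samples the first spline by interval, prunes duplicates, adds SpawnRadius noise and optionally stores density as ScaleGradient.
; Points are raycast within computed bounds, with spline attributes controlling projection, trace complexity and debug display.
; OrientationMode keeps hit rotation, rebuilds it from neighboring hits or restores source rotation, before optional trace-result processing.
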(graph PCG_MeshGrass_SplineSampling
  (inputs (ActorPoint :type Point) (SplinePoints :type PolyLine) (SplineAttributes :type Param) (ActorAttributes :type Param))
  (outputs (SplinePoints :type Point))
  (bind CreateAttribute_4 (Create%20Constant :AttributeTypes (struct :DoubleValue 1e+02)))
  (bind DeleteAttributes_1 (Delete%20Attributes :Operation KeepSelectedAttributes :SelectedAttributes "SamplingInterval"))
  (bind AttributeMathsOp_2 (Attribute%20Maths%20Op :Operation Divide))
  (bind CreateAttribute_3 (Create%20Constant :AttributeTypes (struct :Type Integer64 :DoubleValue 4.0 :IntValue 25)))
  (bind AttributeMathsOp_11 (Attribute%20Maths%20Op :Operation Divide))
  (bind AttributeMathsOp_7 (Attribute%20Maths%20Op :bForceRoundingOpToInt true :Operation Floor))
  (bind FilterDataByIndex_0 (Filter%20Data%20By%20Index :SelectedIndices "0"))
  (bind SplineSampler_0 (Spline%20Sampler :SamplerParams (struct :Dimension OnHorizontal :Mode Distance :NumPlanarSubdivisions 25 :NumHeightSubdivisions 3)))
  (bind DeleteAttributes_4 (Delete%20Attributes :Operation KeepSelectedAttributes :SelectedAttributes "OrientationMode"))
  (bind Switch_0 (Switch :EnumSelection (struct :Class /Game/Blueprints/Enum/2SidedEnum.2SidedEnum) :IntOptions [1 2]))
  (bind CopyAttributes_2 (Copy%20Attributes :InputSource $Rotation :OutputTarget $Rotation))
  (bind Gather_1 (Gather))
  (bind SortAttributes_0 (Sort%20Attributes :bUseStableSort false :SortAttributes [(struct :InputSource $Index :SortMethod Descending)]))
  (bind AttributeFilter_3 (Filter%20Attribute%20Elements :AttributeTypes (struct :Type Integer32) :bUseConstantThreshold true :TargetAttribute $Index))
  (bind AttributeFilter_1 (Filter%20Attribute%20Elements :AttributeTypes (struct :Type Integer32) :bUseConstantThreshold true :TargetAttribute $Index))
  (bind SortAttributes_1 (Sort%20Attributes :bUseStableSort false :SortAttributes [(struct :InputSource $Index :SortMethod Descending)]))
  (bind AttributeMathsOp_4 (Attribute%20Maths%20Op :InputSource1 $Position :InputSource2 $Position :Operation Subtract))
  (bind MergePoints_0 (Merge%20Points))
  (bind AttributeVectorOp_3 (Attribute%20Vector%20Op :InputSource1 $Position :Operation Normalize :OutputTarget Forward))
  (bind AttributeVectorOp_4 (Attribute%20Vector%20Op :InputSource1 $Rotation.Up :InputSource2 Forward :OutputTarget Cross))
  (bind MakeRotatorAttribute_3 (Make%20Rotator%20Attribute :InputSource1 Forward :InputSource2 Cross :InputSource3 $Rotation.Up :OutputTarget $Rotation))
  (bind AttributeFilter_0 (Filter%20Attribute%20Elements :AttributeTypes (struct :Type Float) :bUseConstantThreshold true :Operator NotEqual :ThresholdAttribute @Last))
  (bind Branch_25 (Branch))
  (bind Subgraph_0 (Subgraph :Graph /Game/Athena/Helios/Environments/Landscape/WorldBuildingTools/PCGGrass/PCG_ProcessTraceResults.PCG_ProcessTraceResults))
  (bind BoundsModifier_0 (Bounds%20Modifier :BoundsMax (25.0 25.0 25.0) :BoundsMin (-25.0 -25.0 -25.0) :Mode Set))
  (bind ApplyScaleToBounds_12 (Apply%20Scale%20To%20Bounds))
  (bind CreateAttribute_71 (Create%20Constant :AttributeTypes (struct :DoubleValue 15.0)))
  (bind AttributeMathsOp_9 (Attribute%20Maths%20Op :Operation Multiply))
  (bind CreateAttribute_74 (Create%20Constant :AttributeTypes (struct :DoubleValue -1.0)))
  (bind AttributeMathsOp_72 (Attribute%20Maths%20Op :Operation Multiply))
  (bind BoundsModifier_11 (Bounds%20Modifier :BoundsMax (25.0 25.0 25.0) :BoundsMin (-25.0 -25.0 -25.0) :Mode Set))
  (bind SelfPruning_8 (Self%20Pruning :Parameters (struct :PruningType AllEqual :ComparisonSource $Position :RadiusSimilarityFactor 0.0 :bRandomizedPruning false)))
  (bind GetBounds_14 (Get%20Bounds))
  (bind DeleteAttributes_22 (Delete%20Attributes :SelectedAttributes "BoundsMin"))
  (bind GetAttributeFromPointIndex_41 (Get%20Attribute%20From%20Point%20Index :bForceOutputAttributeToBeInElementsDomain true :InputSource $Position :OutputAttributeName Position))
  (bind AttributeMathsOp_43 (Attribute%20Maths%20Op :Operation Subtract))
  (bind CreateAttribute_35 (Create%20Constant))
  (bind AttributeMathsOp_31 (Attribute%20Maths%20Op :InputSource1 $Rotation.Yaw :Operation Set))
  (bind AttributeMathsOp_38 (Attribute%20Maths%20Op :InputSource1 $Rotation.Roll :Operation Set))
  (bind AttributeMathsOp_40 (Attribute%20Maths%20Op :InputSource1 $Rotation.Pitch :Operation Set))
  (bind CreateAttribute_48 (Create%20Constant :AttributeTypes (struct :DoubleValue 1.0)))
  (bind AttributeMathsOp_46 (Attribute%20Maths%20Op :InputSource1 $Scale :Operation Set))
  (bind DeleteAttributes_20 (Delete%20Attributes :SelectedAttributes "BoundsMax"))
  (bind AttributeMathsOp_26 (Attribute%20Maths%20Op :Operation Subtract))
  (bind CreateAttribute_50 (Create%20Constant :AttributeTypes (struct :Type Vector) :OutputTarget Vector))
  (bind DeleteAttributes_53 (Delete%20Attributes :Operation KeepSelectedAttributes :SelectedAttributes "SpawnDistance"))
  (bind AttributeMathsOp_51 (Attribute%20Maths%20Op :InputSource1 Vector.Z :Operation Set))
  (bind AttributeMathsOp_49 (Attribute%20Maths%20Op :Operation Subtract))
  (bind BoundsModifier_18 (Bounds%20Modifier :Mode Set))
  (bind DeleteAttributes_57 (Delete%20Attributes :Operation KeepSelectedAttributes :SelectedAttributes "SpawnRadius"))
  (bind CreateAttribute_59 (Create%20Constant :AttributeTypes (struct :DoubleValue -1.0)))
  (bind AttributeMathsOp_60 (Attribute%20Maths%20Op :Operation Multiply))
  (bind CreateAttribute_62 (Create%20Constant :AttributeTypes (struct :DoubleValue 1.0)))
  (bind AttributeMathsOp_64 (Attribute%20Maths%20Op :Operation Multiply))
  (bind AttributeNoise_58 (Attribute%20Noise :InputSource $Position.XY :Mode Add :NoiseMin -1.0))
  (bind DeleteAttributes_3 (Delete%20Attributes :Operation KeepSelectedAttributes :SelectedAttributes "ScaleByGradient"))
  (bind DeleteAttributes_6 (Delete%20Attributes :Operation KeepSelectedAttributes :SelectedAttributes "ProjectionMode"))
  (bind AttributeCompareOp_0 (Attribute%20Compare%20Op :Operation Greater))
  (bind Branch_0 (Branch))
  (bind CopyAttributes_0 (Copy%20Attributes :InputSource $Density :OutputTarget ScaleGradient))
  (bind Gather_2 (Gather))
  (bind CopyAttributes_3 (Copy%20Attributes :InputSource $Rotation.Up :OutputTarget RayDir))
  (bind AttributeMathsOp_3 (Attribute%20Maths%20Op :InputSource1 RayDir :Operation Multiply))
  (bind DeleteAttributes_0 (Delete%20Attributes :Operation KeepSelectedAttributes :SelectedAttributes "ForcePreciseProjection"))
  (bind DeleteAttributes_2 (Delete%20Attributes :Operation KeepSelectedAttributes :SelectedAttributes "SpawnDistance"))
  (bind Subgraph_2 (Subgraph :Graph /Game/Athena/Helios/Environments/Landscape/WorldBuildingTools/PCGGrass/PCG_MeshGrass_InterpretProjectionSetup.PCG_MeshGrass_InterpretProjectionSetup))
  (bind WorldRaycast_0 (World%20Raycast :bKeepOriginalPointOnMiss true :bOverrideRayDirections true :RaycastMode NormalizedWithLength :RayDirectionAttribute RayDir :RayLength 1e+06 :WorldQueryParams (struct :bGetReferenceToStaticMesh true :ActorTagFilter Exclude :ActorTagsList "PCG_MeshGrass_Out" :ActorClassFilter Include :ActorFilterFromInput Include :bGetReferenceToActorHit true)))
  (bind DeleteAttributes_5 (Delete%20Attributes :Operation KeepSelectedAttributes :SelectedAttributes "VisualizeProjection"))
  (bind AttributeMathsOp_5 (Attribute%20Maths%20Op :InputSource1 $Rotation :Operation Set))
  (bind Select_2 (Select))
  (bind Select_1 (Select))
  (bind AttributeMathsOp_1 (Attribute%20Maths%20Op :InputSource1 $Scale.Z :Operation Multiply))
  (bind Debug_1 (Debug))
  (bind ExtentsModifier_0 (Extents%20Modifier :Mode Add))
  (bind Select_3 (Select))
  ;[0.05 0.05 0.05] Inputs
  (bind GetAttributeFromPointIndex_5 (Get%20Attribute%20From%20Point%20Index :bForceOutputAttributeToBeInElementsDomain true :InputSource ActorReference :OutputAttributeName None))
  (reroute Actor%20Point ActorPoint)
  (reroute ActorReference GetAttributeFromPointIndex_5.Attribute)
  (reroute Bounds Select_3)
  (reroute IgnoredMeshes Subgraph_2.IgnoredMeshes)
  (reroute Impacts WorldRaycast_0)
  (reroute NamedRerouteDeclaration_0 ActorAttributes)
  (reroute NamedRerouteDeclaration_36 SplineAttributes)
  (reroute NamedRerouteDeclaration_38 SplinePoints)
  (reroute Origins Gather_2)
  (reroute ProjectionSetup Subgraph_2.ProjectionSetup)
  (reroute RequiresProcessing Subgraph_2.RequiresProcessing)
  (connect SplineSampler_0 Out ApplyScaleToBounds_12 In)
  (connect DeleteAttributes_6 Out AttributeCompareOp_0 InA)
  (connect CopyAttributes_2 Out AttributeFilter_0 In)
  (connect MakeRotatorAttribute_3 Out AttributeFilter_0 In)
  (connect Switch_0 Default AttributeFilter_0 In)
  (connect Gather_1 Out AttributeFilter_1 In)
  (connect SortAttributes_0 Out AttributeFilter_3 In)
  (connect Select_1 Out AttributeMathsOp_1 InA)
  (connect DeleteAttributes_1 Out AttributeMathsOp_2 InA)
  (connect CreateAttribute_4 Out AttributeMathsOp_2 InB)
  (connect CopyAttributes_3 Out AttributeMathsOp_3 InA)
  (connect SortAttributes_1 Out AttributeMathsOp_4 InA)
  (connect AttributeFilter_1 InsideFilter AttributeMathsOp_4 InB)
  (connect Gather_2 Out AttributeMathsOp_5 InA)
  (connect AttributeMathsOp_11 Out AttributeMathsOp_7 In)
  (connect AttributeMathsOp_2 Out AttributeMathsOp_9 InA)
  (connect CreateAttribute_71 Out AttributeMathsOp_9 InB)
  (connect CreateAttribute_3 Out AttributeMathsOp_11 InA)
  (connect AttributeMathsOp_2 Out AttributeMathsOp_11 InB)
  (connect DeleteAttributes_20 Out AttributeMathsOp_26 InA)
  (connect GetAttributeFromPointIndex_41 Attribute AttributeMathsOp_26 InB)
  (connect Actor%20Point Out AttributeMathsOp_31 InA)
  (connect CreateAttribute_35 Out AttributeMathsOp_31 InB)
  (connect AttributeMathsOp_31 Out AttributeMathsOp_38 InA)
  (connect CreateAttribute_35 Out AttributeMathsOp_38 InB)
  (connect AttributeMathsOp_38 Out AttributeMathsOp_40 InA)
  (connect CreateAttribute_35 Out AttributeMathsOp_40 InB)
  (connect DeleteAttributes_22 Out AttributeMathsOp_43 InA)
  (connect GetAttributeFromPointIndex_41 Attribute AttributeMathsOp_43 InB)
  (connect AttributeMathsOp_40 Out AttributeMathsOp_46 InA)
  (connect CreateAttribute_48 Out AttributeMathsOp_46 InB)
  (connect AttributeMathsOp_26 Out AttributeMathsOp_49 InA)
  (connect AttributeMathsOp_51 Out AttributeMathsOp_49 InB)
  (connect CreateAttribute_50 Out AttributeMathsOp_51 InA)
  (connect DeleteAttributes_53 Out AttributeMathsOp_51 InB)
  (connect CreateAttribute_59 Out AttributeMathsOp_60 InA)
  (connect DeleteAttributes_57 Out AttributeMathsOp_60 InB)
  (connect DeleteAttributes_57 Out AttributeMathsOp_64 InA)
  (connect CreateAttribute_62 Out AttributeMathsOp_64 InB)
  (connect CreateAttribute_74 Out AttributeMathsOp_72 InA)
  (connect AttributeMathsOp_9 Out AttributeMathsOp_72 InB)
  (connect SelfPruning_8 Out AttributeNoise_58 In)
  (connect AttributeMathsOp_64 Out AttributeNoise_58 NoiseMax)
  (connect AttributeMathsOp_60 Out AttributeNoise_58 NoiseMin)
  (connect MergePoints_0 Out AttributeVectorOp_3 In)
  (connect Impacts Out AttributeVectorOp_4 InA)
  (connect AttributeVectorOp_3 Out AttributeVectorOp_4 InB)
  (connect Branch_25 Output%20A BoundsModifier_0 In)
  (connect Subgraph_0 OutPoints BoundsModifier_0 In)
  (connect AttributeMathsOp_9 Out BoundsModifier_11 BoundsMax)
  (connect AttributeMathsOp_72 Out BoundsModifier_11 BoundsMin)
  (connect ApplyScaleToBounds_12 Out BoundsModifier_11 In)
  (connect AttributeMathsOp_43 Out BoundsModifier_18 BoundsMax)
  (connect AttributeMathsOp_49 Out BoundsModifier_18 BoundsMin)
  (connect AttributeMathsOp_46 Out BoundsModifier_18 In)
  (connect DeleteAttributes_3 Out Branch_0 bOutputToB)
  (connect AttributeNoise_58 Out Branch_0 In)
  (connect RequiresProcessing Out Branch_25 bOutputToB)
  (connect AttributeFilter_0 InsideFilter Branch_25 In)
  (connect Branch_0 Output%20B CopyAttributes_0 Source)
  (connect Branch_0 Output%20B CopyAttributes_0 Target)
  (connect Origins Out CopyAttributes_2 Source)
  (connect Switch_0 %32 CopyAttributes_2 Target)
  (connect Gather_2 Out CopyAttributes_3 Source)
  (connect Gather_2 Out CopyAttributes_3 Target)
  (connect AttributeMathsOp_1 Out Debug_1 In)
  (connect NamedRerouteDeclaration_36 Out DeleteAttributes_0 In)
  (connect NamedRerouteDeclaration_36 Out DeleteAttributes_1 In)
  (connect NamedRerouteDeclaration_36 Out DeleteAttributes_2 In)
  (connect NamedRerouteDeclaration_0 Out DeleteAttributes_3 In)
  (connect NamedRerouteDeclaration_36 Out DeleteAttributes_4 In)
  (connect NamedRerouteDeclaration_36 Out DeleteAttributes_5 In)
  (connect NamedRerouteDeclaration_36 Out DeleteAttributes_6 In)
  (connect GetBounds_14 Out DeleteAttributes_20 In)
  (connect GetBounds_14 Out DeleteAttributes_22 In)
  (connect NamedRerouteDeclaration_36 Out DeleteAttributes_53 In)
  (connect NamedRerouteDeclaration_36 Out DeleteAttributes_57 In)
  (connect DeleteAttributes_53 Out ExtentsModifier_0 Extents)
  (connect BoundsModifier_18 Out ExtentsModifier_0 In)
  (connect NamedRerouteDeclaration_38 Out FilterDataByIndex_0 In)
  (connect Switch_0 %31 Gather_1 In)
  (connect Branch_0 Output%20A Gather_2 In)
  (connect CopyAttributes_0 Out Gather_2 In)
  (connect Actor%20Point Out GetAttributeFromPointIndex_5 In)
  (connect Actor%20Point Out GetAttributeFromPointIndex_41 In)
  (connect SelfPruning_8 Out GetBounds_14 In)
  (connect AttributeVectorOp_3 Out MakeRotatorAttribute_3 Forward)
  (connect AttributeVectorOp_4 Out MakeRotatorAttribute_3 Right)
  (connect Impacts Out MakeRotatorAttribute_3 Up)
  (connect AttributeMathsOp_4 Out MergePoints_0 In)
  (connect AttributeFilter_3 OutsideFilter MergePoints_0 In2)
  (connect DeleteAttributes_5 Out Select_1 bUseInputB)
  (connect Select_2 Out Select_1 Input%20B)
  (connect AttributeCompareOp_0 Out Select_2 bUseInputB)
  (connect AttributeMathsOp_5 Out Select_2 Input%20A)
  (connect Gather_2 Out Select_2 Input%20B)
  (connect AttributeCompareOp_0 Out Select_3 bUseInputB)
  (connect BoundsModifier_18 Out Select_3 Input%20A)
  (connect ExtentsModifier_0 Out Select_3 Input%20B)
  (connect BoundsModifier_11 Out SelfPruning_8 In)
  (connect Gather_1 Out SortAttributes_0 In)
  (connect AttributeFilter_3 InsideFilter SortAttributes_1 In)
  (connect DeleteAttributes_1 Out SplineSampler_0 DistanceIncrement)
  (connect AttributeMathsOp_7 Out SplineSampler_0 NumPlanarSubdivisions)
  (connect FilterDataByIndex_0 InsideFilter SplineSampler_0 Spline)
  (connect Branch_25 Output%20B Subgraph_0 InPoints)
  (connect ProjectionSetup Out Subgraph_0 ProjectionSetup)
  (connect IgnoredMeshes Out Subgraph_0 ProjectionSetup)
  (connect ActorReference Out Subgraph_2 ActorReference)
  (connect Impacts Out Switch_0 In)
  (connect DeleteAttributes_4 Out Switch_0 IntegerSelection)
  (connect Subgraph_2 CustomTarget WorldRaycast_0 ActorClass)
  (connect Subgraph_2 FilterFromInput WorldRaycast_0 ActorClassFilter)
  (connect Subgraph_2 FilterFromInput WorldRaycast_0 ActorFilterFromInput)
  (connect Bounds Out WorldRaycast_0 Bounding%20Shape)
  (connect AttributeCompareOp_0 Out WorldRaycast_0 bOverrideRayDirections)
  (connect DeleteAttributes_0 Out WorldRaycast_0 bTraceComplex)
  (connect Subgraph_2 FilterActors WorldRaycast_0 FilterActors)
  (connect AttributeMathsOp_3 Out WorldRaycast_0 Origins)
  (connect DeleteAttributes_2 Out WorldRaycast_0 RayLength)
  (connect Subgraph_2 SelectLandscapeHits WorldRaycast_0 SelectLandscapeHits)
  (output SplinePoints BoundsModifier_0))
