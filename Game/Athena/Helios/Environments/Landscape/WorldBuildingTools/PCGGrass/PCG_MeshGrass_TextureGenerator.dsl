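; PCG graph that generates mesh grass points for an actor from its TextureStampingSetup property, output as TexturePoints.
; The actor's point passes only through an IsActive branch, while graph parameters and constant filter attributes merge into actor attributes for a shared inputs subgraph.
; Subgraphs run in sequence: texture sampling, local filtering, grass assignment, point processing, world filtering and metadata processing.
; Validity checks with branches gate the actor point and the results of both filtering stages.
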(graph PCG_MeshGrass_TextureGenerator
  (params (WorldSpaceTransform :type Bool) (LocationOffsetZ :type Double) (RotationOffset :type Rotator :default (struct :Yaw 1.8e+02)) (RandomizeRotationOffset :type Bool :default true) (ScaleMinMax :type Vector2 :default (1.0 1.0)) (EnableMeshOverride :type Bool) (GrassMesh :type (SoftObject /Script/Engine.StaticMesh)) (MaterialOverride :type (Object /Script/Engine.MaterialInterface)) (GrassDensity :type Double :default 1.0) (FilterNormals :type Bool :default true) (NormalOffset :type Double) (ScaleByGradient :type Bool) (GradientScaleMinMax :type Vector2 :default (0.5 1.0)) (GradientScalePower :type Double :default 1.0) (FilterWorldIntersection :type Bool) (WorldIntersectionDistance :type Double :default 1e+02) (ExtendNegativeBounds :type Bool) (FilterBelowTerrain :type Bool) (BelowTerrainBias :type Double :default -1.0) (TraceZOffset :type Double) (UsePruning :type Bool) (FilterLandscapeLayers :type Bool) (LandscapeLayers :type String) (LandscapeLayerThreshold :type Double) (LandscapeLayerDilation :type Double) (FilterWaterBodies :type Bool) (WaterBodyDistance :type Double) (WaterBodyInvertFilter :type Bool) (WaterBodyPrecision :type (Enum /Game/Athena/Helios/Environments/Landscape/WorldBuildingTools/PCGGrass/EMeshGrassGenericQuality.EMeshGrassGenericQuality) :default NewEnumerator1) (DetailModeOverride :type (Enum /Game/Athena/Helios/Environments/Landscape/WorldBuildingTools/PCGGrass/EMeshGrassDetailMode.EMeshGrassDetailMode)) (IgnoreCullVolumes :type Bool))
  (outputs (TexturePoints :type Point))
  (bind Subgraph_49 (Subgraph :Graph /Game/Athena/Helios/Environments/Landscape/WorldBuildingTools/PCGGrass/PCG_MeshGrass_Inputs.PCG_MeshGrass_Inputs))
  (bind Subgraph_0 (Subgraph :Graph /Game/Athena/Helios/Environments/Landscape/WorldBuildingTools/PCGGrass/PCG_HasValidData.PCG_HasValidData))
  (bind Branch_1 (Branch))
  (bind Subgraph_67 (Subgraph :Graph /Game/Athena/Helios/Environments/Landscape/WorldBuildingTools/PCGGrass/PCG_MeshGrass_TextureSampling.PCG_MeshGrass_TextureSampling))
  (bind Subgraph_62 (Subgraph :Graph /Game/Athena/Helios/Environments/Landscape/WorldBuildingTools/PCGGrass/PCG_MeshGrass_LocalFiltering.PCG_MeshGrass_LocalFiltering))
  (bind Subgraph_65 (Subgraph :Graph /Game/Athena/Helios/Environments/Landscape/WorldBuildingTools/PCGGrass/PCG_HasValidData.PCG_HasValidData))
  (bind Branch_66 (Branch))
  (bind Subgraph_61 (Subgraph :Graph /Game/Athena/Helios/Environments/Landscape/WorldBuildingTools/PCGGrass/PCG_MeshGrass_GrassAssignment.PCG_MeshGrass_GrassAssignment))
  (bind Subgraph_63 (Subgraph :Graph /Game/Athena/Helios/Environments/Landscape/WorldBuildingTools/PCGGrass/PCG_MeshGrass_PointProcessing.PCG_MeshGrass_PointProcessing))
  (bind Subgraph_60 (Subgraph :Graph /Game/Athena/Helios/Environments/Landscape/WorldBuildingTools/PCGGrass/PCG_MeshGrass_WorldFiltering.PCG_MeshGrass_WorldFiltering))
  (bind Subgraph_1 (Subgraph :Graph /Game/Athena/Helios/Environments/Landscape/WorldBuildingTools/PCGGrass/PCG_HasValidPoints.PCG_HasValidPoints))
  (bind Branch_0 (Branch))
  (bind Subgraph_2 (Subgraph :Graph /Game/Athena/Helios/Environments/Landscape/WorldBuildingTools/PCGGrass/PCG_MeshGrass_MetadataProcessing.PCG_MeshGrass_MetadataProcessing))
  ;[0.05 0.05 0.05] Actor Attributes
  (bind CreateAttribute_42 (Create%20Constant :AttributeTypes (struct :Type Boolean) :OutputTarget FilterHeightGradient))
  (bind CreateAttribute_43 (Create%20Constant :OutputTarget EdgeDistance))
  (bind CreateAttribute_44 (Create%20Constant :AttributeTypes (struct :Type Boolean) :OutputTarget FilterLocalZ))
  (bind CreateAttribute_45 (Create%20Constant :OutputTarget LocalZMin))
  (bind CreateAttribute_46 (Create%20Constant :OutputTarget LocalZMax))
  (bind CreateAttribute_47 (Create%20Constant :AttributeTypes (struct :Type Boolean) :OutputTarget FilterNonSelfHits))
  (bind CreateAttribute_48 (Create%20Constant :AttributeTypes (struct :Type Boolean) :OutputTarget FilterBorders))
  (bind CreateConstant_0 (Create%20Constant :AttributeTypes (struct :Type Boolean) :OutputTarget FilterTopologyInWorldSpace))
  (bind AddAttribute_21 (Add%20Attribute :AttributeTypes (struct :Type Integer64) :bCopyAllAttributes true :Attributes MaterialOverride :Attributes ScaleMinMax :Attributes NormalOffset :Attributes GrassDensity :Attributes EnableMeshOverride :Attributes FilterWorldIntersection :Attributes TraceZOffset :Attributes WorldIntersectionDistance :Attributes ExtendNegativeBounds :Attributes UsePruning :Attributes WorldSpaceTransform :Attributes ScaleByGradient :Attributes GradientScaleMinMax :Attributes LocationOffsetZ :Attributes RotationOffset :Attributes GradientScalePower :Attributes FilterLandscapeLayers :Attributes LandscapeLayers :Attributes LandscapeLayerThreshold :Attributes LandscapeLayerDilation :In GrassMesh))
  (bind GetActorData_27 (Get%20Actor%20Data :bComponentsMustOverlapSelf false :Mode GetSinglePoint :PropertyName Scale))
  (bind GetAttributeFromPointIndex_32 (Get%20Attribute%20From%20Point%20Index :bForceOutputAttributeToBeInElementsDomain true :InputSource Actorreference :OutputAttributeName None))
  (bind GetActorProperty_39 (Get%20Actor%20Property :bProcessAllComponents false :bSanitizeOutputAttributeName false :OutputAttributeName None :PropertyName IsActive))
  (bind Branch_38 (Branch))
  (bind GetPropertyFromObjectPath_1 (Get%20Property%20From%20Object%20Path :OutputAttributeName None :PropertyName TextureStampingSetup))
  (reroute ActorAttributes AddAttribute_21)
  (reroute ActorPoint Branch_38.Output%20B)
  (reroute Self%20Reference GetAttributeFromPointIndex_32.Attribute)
  (reroute TextureAttributes GetPropertyFromObjectPath_1)
  (connect CreateAttribute_42 Out AddAttribute_21 Attributes)
  (connect CreateAttribute_43 Out AddAttribute_21 Attributes)
  (connect CreateAttribute_44 Out AddAttribute_21 Attributes)
  (connect CreateAttribute_45 Out AddAttribute_21 Attributes)
  (connect CreateAttribute_46 Out AddAttribute_21 Attributes)
  (connect CreateAttribute_47 Out AddAttribute_21 Attributes)
  (connect CreateAttribute_48 Out AddAttribute_21 Attributes)
  (connect CreateConstant_0 Out AddAttribute_21 Attributes)
  (connect Subgraph_1 IsValid Branch_0 bOutputToB)
  (connect Subgraph_60 OutPoints Branch_0 In)
  (connect Subgraph_0 IsValid Branch_1 bOutputToB)
  (connect Subgraph_49 ActorPoint Branch_1 In)
  (connect GetActorProperty_39 Out Branch_38 bOutputToB)
  (connect GetActorData_27 Out Branch_38 In)
  (connect Subgraph_65 IsValid Branch_66 bOutputToB)
  (connect Subgraph_62 OutPoints Branch_66 In)
  (connect ActorPoint Out GetAttributeFromPointIndex_32 In)
  (connect Self%20Reference Out GetPropertyFromObjectPath_1 In)
  (connect Subgraph_49 ActorPoint Subgraph_0 In)
  (connect Subgraph_60 OutPoints Subgraph_1 In)
  (connect Branch_1 Output%20B Subgraph_2 ActorPoint)
  (connect Branch_0 Output%20B Subgraph_2 InPoints)
  (connect ActorAttributes Out Subgraph_49 ActorAttributes)
  (connect ActorPoint Out Subgraph_49 ActorPoint)
  (connect Subgraph_49 ActorAttributes Subgraph_60 ActorAttributes)
  (connect Branch_1 Output%20B Subgraph_60 ActorPoint)
  (connect TextureAttributes Out Subgraph_60 BiomeAttributes)
  (connect Subgraph_49 GlobalAttributes Subgraph_60 GlobalAttributes)
  (connect Subgraph_63 OutPoints Subgraph_60 InPoints)
  (connect Subgraph_49 VolumeAttributes Subgraph_60 VolumeAttributes)
  (connect Subgraph_49 ActorAttributes Subgraph_61 ActorAttributes)
  (connect Branch_1 Output%20B Subgraph_61 ActorPoint)
  (connect TextureAttributes Out Subgraph_61 BiomeAttributes)
  (connect Subgraph_49 GlobalAttributes Subgraph_61 GlobalAttributes)
  (connect Branch_66 Output%20B Subgraph_61 InPoints)
  (connect Subgraph_49 VolumeAttributes Subgraph_61 VolumeAttributes)
  (connect Subgraph_49 ActorAttributes Subgraph_62 ActorAttributes)
  (connect Branch_1 Output%20B Subgraph_62 ActorPoint)
  (connect TextureAttributes Out Subgraph_62 BiomeAttributes)
  (connect Subgraph_49 GlobalAttributes Subgraph_62 GlobalAttributes)
  (connect Subgraph_67 TexturePoints Subgraph_62 InPoints)
  (connect Subgraph_49 VolumeAttributes Subgraph_62 VolumeAttributes)
  (connect Subgraph_49 ActorAttributes Subgraph_63 ActorAttributes)
  (connect Branch_1 Output%20B Subgraph_63 ActorPoint)
  (connect TextureAttributes Out Subgraph_63 BiomeAttributes)
  (connect Subgraph_49 GlobalAttributes Subgraph_63 GlobalAttributes)
  (connect Subgraph_61 OutPoints Subgraph_63 InPoints)
  (connect Subgraph_49 VolumeAttributes Subgraph_63 VolumeAttributes)
  (connect Subgraph_62 OutPoints Subgraph_65 In)
  (connect Subgraph_49 ActorAttributes Subgraph_67 ActorAttributes)
  (connect Branch_1 Output%20B Subgraph_67 ActorPoint)
  (connect TextureAttributes Out Subgraph_67 TextureAttributes)
  (output TexturePoints Subgraph_2.OutPoints))
