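; PCG graph that generates grass points for the mesh grass world-building tool from an actor's settings.
; Surface-samples a world ray hit query configured by a projection setup subgraph, with spacing and extents derived from the actor's capped Density per Square Meter, skipped unless positive.
; Valid points optionally pass through trace-result filtering, then a texture stamping subgraph using the actor point and transform.
; Output points are randomly offset in XY within Spawn Radius.
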(graph PCG_MeshGrass_TextureSampling
  (inputs (ActorPoint :type Point) (ActorAttributes :type Param) (TextureAttributes :type Param))
  (outputs (TexturePoints :type Point))
  (bind Subgraph_3 (Subgraph :Graph /Game/Athena/Helios/Environments/Landscape/WorldBuildingTools/PCGGrass/PCG_MeshGrass_InterpretProjectionSetup.PCG_MeshGrass_InterpretProjectionSetup))
  (bind WorldRayHitQuery_1 (World%20Ray%20Hit%20Query :QueryParams (struct :bGetReferenceToStaticMesh true :ActorTagFilter Exclude :ActorTagsList "PCG_MeshGrass_Out" :ActorClassFilter Include :ActorFilterFromInput Include :SelectLandscapeHits Require :bGetReferenceToActorHit true)))
  (bind CreateAttribute_5 (Create%20Constant :AttributeTypes (struct :DoubleValue 2.0)))
  (bind GetPropertyFromObjectPath_1 (Get%20Property%20From%20Object%20Path :OutputAttributeName None :PropertyName Density%20per%20Square%20Meter))
  (bind AttributeMathsOp_0 (Attribute%20Maths%20Op :Operation Min))
  (bind CreateAttribute_18 (Create%20Constant :AttributeTypes (struct :DoubleValue 25.0)))
  (bind AttributeMathsOp_24 (Attribute%20Maths%20Op :Operation Divide))
  (bind CreateAttribute_23 (Create%20Constant :AttributeTypes (struct :DoubleValue 2.0)))
  (bind AttributeMathsOp_25 (Attribute%20Maths%20Op :Operation Pow))
  (bind CreateAttribute_6 (Create%20Constant))
  (bind AttributeCompareOp_1 (Attribute%20Compare%20Op :Operation Greater))
  (bind Branch_7 (Branch))
  (bind SurfaceSampler_2 (Surface%20Sampler :bApplyDensityToPoints false :PointExtents (25.0 25.0 25.0) :PointsPerSquaredMeter 4.0 :PointSteepness 0.5))
  (bind GetAttributeFromPointIndex_4 (Get%20Attribute%20From%20Point%20Index :bForceOutputAttributeToBeInElementsDomain true :InputSource $Transform :OutputAttributeName None))
  (bind Subgraph_1 (Subgraph :Graph /Game/Athena/Helios/Environments/Landscape/WorldBuildingTools/PCGGrass/PCG_GateValidPoints.PCG_GateValidPoints))
  (bind Branch_4 (Branch))
  (bind Subgraph_2 (Subgraph :Graph /Game/Athena/Helios/Environments/Landscape/WorldBuildingTools/PCGGrass/PCG_ProcessTraceResults.PCG_ProcessTraceResults))
  (bind Gather_5 (Gather))
  (bind Subgraph_0 (Subgraph :Graph /Game/Athena/Helios/Environments/Landscape/WorldBuildingTools/PCGGrass/PCG_MeshGrass_TextureStamping.PCG_MeshGrass_TextureStamping))
  (bind MakeVectorAttribute_0 (Make%20Vector%20Attribute :OutputType Vector))
  (bind AttributeMathsOp_1 (Attribute%20Maths%20Op :Operation Negate))
  (bind MakeVectorAttribute_1 (Make%20Vector%20Attribute :OutputType Vector))
  (bind TransformPoints_0 (Transform%20Points :OffsetMax (5e+01 5e+01 0.0) :OffsetMin (-5e+01 -5e+01 0.0)))
  ;[0.05 0.05 0.05] Inputs
  (bind GetAttributeFromPointIndex_1 (Get%20Attribute%20From%20Point%20Index :bForceOutputAttributeToBeInElementsDomain true :InputSource ActorReference :OutputAttributeName None))
  (bind GetPropertyFromObjectPath_2 (Get%20Property%20From%20Object%20Path :OutputAttributeName @SourceName :PropertyName Spawn%20Radius))
  (reroute ActorReference GetAttributeFromPointIndex_1.Attribute)
  (reroute IgnoredMeshes Subgraph_3.IgnoredMeshes)
  (reroute NamedRerouteDeclaration_0 ActorAttributes)
  (reroute NamedRerouteDeclaration_1 ActorPoint)
  (reroute ProjectionSetup Subgraph_3.ProjectionSetup)
  (reroute RequiresProcessing Subgraph_3.RequiresProcessing)
  (reroute SpawnRadius GetPropertyFromObjectPath_2)
  (reroute TextureStampingSetup TextureAttributes)
  (connect AttributeMathsOp_0 Out AttributeCompareOp_1 InA)
  (connect CreateAttribute_6 Out AttributeCompareOp_1 InB)
  (connect GetPropertyFromObjectPath_1 Out AttributeMathsOp_0 InA)
  (connect CreateAttribute_5 Out AttributeMathsOp_0 InB)
  (connect SpawnRadius Out AttributeMathsOp_1 In)
  (connect CreateAttribute_18 Out AttributeMathsOp_24 InA)
  (connect AttributeMathsOp_0 Out AttributeMathsOp_24 InB)
  (connect AttributeMathsOp_0 Out AttributeMathsOp_25 InA)
  (connect CreateAttribute_23 Out AttributeMathsOp_25 InB)
  (connect RequiresProcessing Out Branch_4 bOutputToB)
  (connect Subgraph_1 ValidPoints Branch_4 In)
  (connect AttributeCompareOp_1 Out Branch_7 bOutputToB)
  (connect WorldRayHitQuery_1 Out Branch_7 In)
  (connect Branch_4 Output%20A Gather_5 In)
  (connect Subgraph_2 OutPoints Gather_5 In)
  (connect NamedRerouteDeclaration_1 Out GetAttributeFromPointIndex_1 In)
  (connect NamedRerouteDeclaration_1 Out GetAttributeFromPointIndex_4 In)
  (connect ActorReference Out GetPropertyFromObjectPath_1 In)
  (connect ActorReference Out GetPropertyFromObjectPath_2 In)
  (connect SpawnRadius Out MakeVectorAttribute_0 X)
  (connect SpawnRadius Out MakeVectorAttribute_0 Y)
  (connect AttributeMathsOp_1 Out MakeVectorAttribute_1 X)
  (connect AttributeMathsOp_1 Out MakeVectorAttribute_1 Y)
  (connect NamedRerouteDeclaration_1 Out Subgraph_0 LocalControls)
  (connect Gather_5 Out Subgraph_0 TargetPoints)
  (connect GetAttributeFromPointIndex_4 Attribute Subgraph_0 TargetTransform)
  (connect SurfaceSampler_2 Out Subgraph_1 In)
  (connect Branch_4 Output%20B Subgraph_2 InPoints)
  (connect ProjectionSetup Out Subgraph_2 ProjectionSetup)
  (connect IgnoredMeshes Out Subgraph_2 ProjectionSetup)
  (connect ActorReference Out Subgraph_3 ActorReference)
  (connect AttributeMathsOp_24 Out SurfaceSampler_2 PointExtents)
  (connect AttributeMathsOp_25 Out SurfaceSampler_2 PointsPerSquaredMeter)
  (connect Branch_7 Output%20B SurfaceSampler_2 Surface)
  (connect Subgraph_0 TargetPoints TransformPoints_0 In)
  (connect MakeVectorAttribute_0 Out TransformPoints_0 OffsetMax)
  (connect MakeVectorAttribute_1 Out TransformPoints_0 OffsetMin)
  (connect Subgraph_3 CustomTarget WorldRayHitQuery_1 ActorClass)
  (connect Subgraph_3 FilterFromInput WorldRayHitQuery_1 ActorClassFilter)
  (connect Subgraph_3 FilterFromInput WorldRayHitQuery_1 ActorFilterFromInput)
  (connect Subgraph_3 FilterActors WorldRayHitQuery_1 FilterActors)
  (connect Subgraph_3 SelectLandscapeHits WorldRayHitQuery_1 SelectLandscapeHits)
  (output TexturePoints TransformPoints_0))
