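; PCG graph that stamps grass mesh setups from a stamping actor's texture onto target points.
; It builds a texel-sized grid over the actor's bounds, samples the stamp and optional mask, thresholds alpha and adds noise-based variety.
; Red, green and blue dominance, plus optional white/black thresholds, select weighted biome mesh setups, whose attributes are projected onto the targets.
; Run outside a loop, it keeps only points with nonzero DensitySum.
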(graph PCG_MeshGrass_TextureStamping
  (params (StampingBoundsScale :type Double :default 1.0) (WorldSpace :type Bool) (NewProperty :type Double))
  (inputs (TargetPoints :type Point) (LocalControls :type Point) (TargetTransform :type Param))
  (outputs (TargetPoints :type Point))
  ;[0.05 0.05 0.05] Projection Setup
  (bind ApplyScaleToBounds_47 (Apply%20Scale%20To%20Bounds))
  (bind GetAttributeFromPointIndex_46 (Get%20Attribute%20From%20Point%20Index :bForceOutputAttributeToBeInElementsDomain true :InputSource $BoundsMax :OutputAttributeName None))
  (bind CreateAttribute_30 (Create%20Constant :AttributeTypes (struct :DoubleValue 5e+01)))
  (bind CreateAttribute_32 (Create%20Constant :AttributeTypes (struct :DoubleValue 1e+04)))
  (bind DeleteAttributes_44 (Delete%20Attributes :Operation KeepSelectedAttributes :SelectedAttributes "TexelSize"))
  (bind AttributeMathsOp_29 (Attribute%20Maths%20Op :Operation Clamp))
  (bind CreateAttribute_57 (Create%20Constant :AttributeTypes (struct :DoubleValue 2.0)))
  (bind GetAttributeFromPointIndex_55 (Get%20Attribute%20From%20Point%20Index :bForceOutputAttributeToBeInElementsDomain true :InputSource $BoundsMax.Z :OutputAttributeName None))
  (bind AttributeMathsOp_56 (Attribute%20Maths%20Op :Operation Multiply))
  (bind MakeVectorAttribute_54 (Make%20Vector%20Attribute :OutputType Vector))
  (bind CreatePointsGrid_44 (Create%20Points%20Grid :PointSteepness 0.5))
  (bind GetAttributeFromPointIndex_51 (Get%20Attribute%20From%20Point%20Index :bForceOutputAttributeToBeInElementsDomain true :InputSource $Scale :OutputAttributeName None))
  (bind AttributeMathsOp_52 (Attribute%20Maths%20Op :Operation Divide))
  (bind AttributeMathsOp_50 (Attribute%20Maths%20Op :InputSource1 $Position :Operation Multiply))
  (bind AttributeVectorOp_49 (Attribute%20Vector%20Op :InputSource1 $Position :InputSource2 $Transform :Operation TransformLocation))
  (bind DeleteAttributes_17 (Delete%20Attributes :Operation KeepSelectedAttributes :SelectedAttributes "StampTexture"))
  (bind GetTextureData_59 (Get%20Texture%20Data :bForceEditorOnlyCPUSampling true :bUseAbsoluteTransform true :bUseDensitySourceChannel false :Filter Point :Tiling (0.01 0.01) :Transform (struct :Scale3D (1e+01 1e+01 1e+01))))
  (bind SampleTexture_58 (Sample%20Texture :TextureMappingMethod UVCoordinates :UVCoordinatesAttribute UVs))
  (bind CreateAttribute_74 (Create%20Constant :AttributeTypes (struct :DoubleValue 2.2)))
  (bind AttributeMathsOp_89 (Attribute%20Maths%20Op :InputSource1 $Color :Operation Pow))
  (bind AddAttribute_81 (Add%20Attribute :AttributeTypes (struct :Type Vector) :OutputTarget ColorRGB))
  (bind AttributeMathsOp_82 (Attribute%20Maths%20Op :InputSource1 ColorRGB :InputSource2 $Color.RGB :Operation Set))
  (bind AddAttribute_83 (Add%20Attribute :AttributeTypes (struct :Type Float) :OutputTarget ColorA))
  (bind DeleteAttributes_24 (Delete%20Attributes :Operation KeepSelectedAttributes :SelectedAttributes "LuminanceAsAlpha"))
  (bind Branch_2 (Branch))
  (bind AttributeMathsOp_6 (Attribute%20Maths%20Op :InputSource1 ColorRGB.X :InputSource2 ColorRGB.Y :Operation Max))
  (bind AttributeMathsOp_8 (Attribute%20Maths%20Op :InputSource1 ColorRGB.X :InputSource2 ColorRGB.Z :Operation Max :OutputTarget $Color.A))
  (bind AttributeMathsOp_84 (Attribute%20Maths%20Op :InputSource1 ColorA :InputSource2 $Color.A :Operation Set))
  (bind Branch_0 (Branch :bOutputToB-param ScaleByGradient))
  (bind CopyAttributes_2 (Copy%20Attributes :InputSource ColorA :OutputTarget ScaleGradient))
  (bind DeleteAttributes_53 (Delete%20Attributes :Operation KeepSelectedAttributes :SelectedAttributes "MaskTexture"))
  (bind Subgraph_55 (Subgraph :Graph /Game/Athena/Helios/Environments/Landscape/WorldBuildingTools/PCGGrass/PCG_IsValidPath.PCG_IsValidPath))
  (bind Branch_56 (Branch))
  (bind CreateAttribute_90 (Create%20Constant :AttributeTypes (struct :DoubleValue 2.2)))
  (bind GetTextureData_61 (Get%20Texture%20Data :bForceEditorOnlyCPUSampling true :bUseAbsoluteTransform true :bUseDensitySourceChannel false :Filter Point :Tiling (0.01 0.01) :Transform (struct :Scale3D (1e+01 1e+01 1e+01))))
  (bind SampleTexture_59 (Sample%20Texture :TextureMappingMethod UVCoordinates :UVCoordinatesAttribute BaseUVs))
  (bind AttributeMathsOp_71 (Attribute%20Maths%20Op :InputSource1 $Color :Operation Pow))
  (bind DeleteAttributes_2 (Delete%20Attributes :Operation KeepSelectedAttributes :SelectedAttributes "MaskTextureChannel"))
  (bind Subgraph_7 (Subgraph :Graph /Game/Athena/Helios/Environments/Landscape/WorldBuildingTools/PCGGrass/PCG_ColorAttributeFromIndex.PCG_ColorAttributeFromIndex))
  (bind CopyAttributes_66 (Copy%20Attributes :InputSource $Color.A :OutputTarget ColorA))
  (bind Gather_16 (Gather))
  (bind AttributeMathsOp_9 (Attribute%20Maths%20Op :InputSource1 ColorA :Operation OneMinus))
  (bind DeleteAttributes_14 (Delete%20Attributes :Operation KeepSelectedAttributes :SelectedAttributes "InvertMask"))
  (bind Select_12 (Select))
  (bind Branch_1 (Branch :bOutputToB-param WorldSpace))
  (bind AttributeVectorOp_72 (Attribute%20Vector%20Op :InputSource1 $Position :Operation InverseTransformLocation))
  (bind CreateAttribute_14 (Create%20Constant :AttributeTypes (struct :DoubleValue 0.999)))
  (bind DeleteAttributes_28 (Delete%20Attributes :Operation KeepSelectedAttributes :SelectedAttributes "MaskingThreshold"))
  (bind AttributeMathsOp_13 (Attribute%20Maths%20Op :Operation Min))
  (bind AttributeFilter_73 (Filter%20Attribute%20Elements :TargetAttribute ColorA :ThresholdAttribute @Last))
  (bind MakeVectorAttribute_88 (Make%20Vector%20Attribute :OutputType Vector :X StampingBoundsScale :Y StampingBoundsScale :Z StampingBoundsScale))
  (bind Subgraph_38 (Subgraph :Graph /Game/Athena/Helios/Environments/Landscape/WorldBuildingTools/PCGGrass/PCG_GateValidPoints.PCG_GateValidPoints))
  (bind ExtentsModifier_86 (Extents%20Modifier :Mode Multiply))
  (bind AddAttribute_97 (Add%20Attribute :AttributeTypes (struct :Type SoftObjectPath) :OutputTarget OverrideMaterial))
  (bind AddAttribute_99 (Add%20Attribute :AttributeTypes (struct :Type SoftObjectPath) :OutputTarget OverrideMaterial1))
  (bind AddAttribute_54 (Add%20Attribute :AttributeTypes (struct :Type SoftObjectPath) :OutputTarget Mesh))
  (bind FilterDataByAttribute_10 (Filter%20Data%20By%20Attribute :Attribute Variety))
  (bind Spatial%20Noise_6 (Spatial%20Noise :bTiling true :Contrast 3.4 :Iterations 3 :Transform (struct :Scale3D (8e+01 8e+01 1.0)) :ValueTarget Variety))
  (bind AttributeReduce_8 (Attribute%20Reduce :Operation Max :OutputAttributeName VarietyMax))
  (bind AttributeReduce_9 (Attribute%20Reduce :Operation Min :OutputAttributeName VarietyMax))
  (bind AttributeRemap_7 (Attribute%20Remap :bAllowInverseRange false :bClampToUnitRange true))
  (bind AddAttribute_32 (Add%20Attribute :AttributeTypes (struct :DoubleValue 1.0) :OutputTarget DensitySum))
  (bind CopyAttributes_0 (Copy%20Attributes :InputSource ColorA :OutputTarget $Color))
  ;;[0.05 0.05 0.05] UVs
  (bind AddAttribute_2 (Add%20Attribute :AttributeTypes (struct :Type Vector2) :OutputTarget UVs))
  (bind AttributeMathsOp_3 (Attribute%20Maths%20Op :InputSource1 UVs.XY :InputSource2 $Position.XY :Operation Set))
  (bind AttributeMathsOp_4 (Attribute%20Maths%20Op :InputSource2 @Last.XY :Operation Divide))
  (bind AttributeMathsOp_5 (Attribute%20Maths%20Op))
  (bind CreateAttribute_8 (Create%20Constant :AttributeTypes (struct :DoubleValue 0.5)))
  (bind AttributeMathsOp_7 (Attribute%20Maths%20Op :Operation Multiply))
  (bind CopyAttributes_67 (Copy%20Attributes :InputSource UVs :OutputTarget BaseUVs))
  (bind DeleteAttributes_21 (Delete%20Attributes :Operation KeepSelectedAttributes :SelectedAttributes "Tiling"))
  (bind AttributeMathsOp_2 (Attribute%20Maths%20Op :InputSource1 UVs :Operation Multiply))
  ;[0.05 0.05 0.05] Attribute Projection
  (bind DataNum_26 (Data%20Count))
  (bind AttributeCompareOp_35 (Attribute%20Compare%20Op :Operation NotEqual))
  (bind AttributeFilter_179 (Filter%20Attribute%20Elements :Operator GreaterOrEqual :TargetAttribute ColorRGB.X :ThresholdAttribute ColorRGB.Y))
  (bind AttributeFilter_181 (Filter%20Attribute%20Elements :Operator GreaterOrEqual :TargetAttribute ColorRGB.X :ThresholdAttribute ColorRGB.Z))
  (bind DataNum_30 (Data%20Count))
  (bind AttributeCompareOp_48 (Attribute%20Compare%20Op :Operation NotEqual))
  (bind GetPointsCount_12 (Get%20Element%20Count))
  (bind AttributeCompareOp_56 (Attribute%20Compare%20Op :Operation NotEqual))
  (bind AttributeBooleanOp_58 (Attribute%20Boolean%20Op))
  (bind Branch_37 (Branch))
  (bind Select_0 (Select))
  (bind Gather_0 (Gather))
  (bind MergePoints_187 (Merge%20Points))
  (bind AttributeFilter_183 (Filter%20Attribute%20Elements :Operator GreaterOrEqual :TargetAttribute ColorRGB.Y :ThresholdAttribute ColorRGB.X))
  (bind AttributeFilter_185 (Filter%20Attribute%20Elements :Operator GreaterOrEqual :TargetAttribute ColorRGB.Y :ThresholdAttribute ColorRGB.Z))
  (bind GetPointsCount_10 (Get%20Element%20Count))
  (bind AttributeCompareOp_20 (Attribute%20Compare%20Op :Operation NotEqual))
  (bind DataNum_28 (Data%20Count))
  (bind AttributeCompareOp_151 (Attribute%20Compare%20Op :Operation NotEqual))
  (bind AttributeBooleanOp_21 (Attribute%20Boolean%20Op))
  (bind Branch_64 (Branch))
  (bind Select_3 (Select))
  (bind Gather_1 (Gather))
  (bind MergePoints_189 (Merge%20Points))
  (bind AttributeFilter_191 (Filter%20Attribute%20Elements :Operator GreaterOrEqual :TargetAttribute ColorRGB.Z :ThresholdAttribute ColorRGB.X))
  (bind AttributeFilter_193 (Filter%20Attribute%20Elements :Operator GreaterOrEqual :TargetAttribute ColorRGB.Z :ThresholdAttribute ColorRGB.Y))
  (bind GetPointsCount_8 (Get%20Element%20Count))
  (bind AttributeCompareOp_44 (Attribute%20Compare%20Op :Operation NotEqual))
  (bind AttributeBooleanOp_46 (Attribute%20Boolean%20Op))
  (bind AddAttribute_3 (Add%20Attribute :OutputTarget DensitySum))
  (bind Subgraph_39 (Subgraph :Graph /Game/Athena/Helios/Environments/Landscape/WorldBuildingTools/PCGGrass/PCG_BranchValidCondition.PCG_BranchValidCondition))
  (bind Branch_139 (Branch))
  (bind GetLoopIndex_0 (Get%20Loop%20Index :bWarnIfCalledOutsideOfLoop false))
  (bind Subgraph_8 (Subgraph :Graph /Game/Athena/Helios/Environments/Landscape/WorldBuildingTools/PCGGrass/PCG_HasValidData.PCG_HasValidData))
  (bind Branch_4 (Branch))
  (bind MatchAndSetAttributes_11 (Match%20And%20Set%20Attributes :bUseInputWeightAttribute true :bUseWeightAttribute true :InputAttribute Biome :InputWeightAttribute Variety :MatchAttribute Biome :WeightAttribute Weight))
  (bind Projection_125 (Projection :bKeepZeroDensityPoints true :ProjectionParams (struct :bProjectPositions false :bProjectRotations false :AttributeList "Mesh,OverrideMaterial,OverrideMaterial1,TextureStamped" :AttributeMode IncludeAttributes))) ; R
  (bind Gather_17 (Gather))
  (bind Branch_147 (Branch))
  (bind MatchAndSetAttributes_27 (Match%20And%20Set%20Attributes :bUseInputWeightAttribute true :bUseWeightAttribute true :InputAttribute Biome :InputWeightAttribute Variety :MatchAttribute Biome :WeightAttribute Weight))
  (bind Projection_123 (Projection :bKeepZeroDensityPoints true :ProjectionParams (struct :bProjectPositions false :bProjectRotations false :AttributeList "Mesh,OverrideMaterial,OverrideMaterial1,TextureStamped" :AttributeMode IncludeAttributes))) ; G
  (bind Gather_19 (Gather))
  (bind Branch_155 (Branch))
  (bind Branch_66 (Branch))
  (bind MatchAndSetAttributes_29 (Match%20And%20Set%20Attributes :bUseInputWeightAttribute true :bUseWeightAttribute true :InputAttribute Biome :InputWeightAttribute Variety :MatchAttribute Biome :WeightAttribute Weight))
  (bind Projection_115 (Projection :bKeepZeroDensityPoints true :ProjectionParams (struct :bProjectPositions false :bProjectRotations false :AttributeList "Mesh,OverrideMaterial,OverrideMaterial1,TextureStamped" :AttributeMode IncludeAttributes))) ; B
  (bind DataNum_24 (Data%20Count))
  (bind AttributeCompareOp_60 (Attribute%20Compare%20Op :Operation NotEqual))
  (bind DeleteAttributes_37 (Delete%20Attributes :Operation KeepSelectedAttributes :SelectedAttributes "UseBlackWhiteSetup"))
  (bind Branch_25 (Branch))
  (bind DeleteAttributes_32 (Delete%20Attributes :Operation KeepSelectedAttributes :SelectedAttributes "BlackWhiteThresholds"))
  (bind BreakVectorAttribute_34 (Break%20Vector%20Attribute))
  (bind MakeVectorAttribute_203 (Make%20Vector%20Attribute :OutputType Vector))
  (bind AttributeFilter_197 (Filter%20Attribute%20Elements :Operator GreaterOrEqual :TargetAttribute ColorRGB :ThresholdAttribute @Last))
  (bind GetPointsCount_6 (Get%20Element%20Count))
  (bind AttributeCompareOp_68 (Attribute%20Compare%20Op :Operation NotEqual))
  (bind AttributeBooleanOp_70 (Attribute%20Boolean%20Op))
  (bind DeleteAttributes_39 (Delete%20Attributes :Operation KeepSelectedAttributes :SelectedAttributes "UseBlackWhiteSetup"))
  (bind Gather_21 (Gather))
  (bind Branch_16 (Branch))
  (bind Branch_163 (Branch))
  (bind Branch_68 (Branch))
  (bind MatchAndSetAttributes_31 (Match%20And%20Set%20Attributes :bUseInputWeightAttribute true :bUseWeightAttribute true :InputAttribute Biome :InputWeightAttribute Variety :MatchAttribute Biome :WeightAttribute Weight))
  (bind Projection_135 (Projection :bKeepZeroDensityPoints true :ProjectionParams (struct :bProjectPositions false :bProjectRotations false :AttributeList "Mesh,OverrideMaterial,OverrideMaterial1,TextureStamped" :AttributeMode IncludeAttributes))) ; W
  (bind DataNum_22 (Data%20Count))
  (bind AttributeCompareOp_72 (Attribute%20Compare%20Op :Operation NotEqual))
  (bind MakeVectorAttribute_201 (Make%20Vector%20Attribute :OutputType Vector))
  (bind AttributeFilter_199 (Filter%20Attribute%20Elements :Operator LesserOrEqual :TargetAttribute ColorRGB :ThresholdAttribute @Last))
  (bind GetPointsCount_1 (Get%20Element%20Count))
  (bind AttributeCompareOp_80 (Attribute%20Compare%20Op :Operation NotEqual))
  (bind AttributeBooleanOp_82 (Attribute%20Boolean%20Op))
  (bind Gather_38 (Gather))
  (bind Branch_171 (Branch))
  (bind Branch_70 (Branch))
  (bind MatchAndSetAttributes_33 (Match%20And%20Set%20Attributes :bUseInputWeightAttribute true :bUseWeightAttribute true :InputAttribute Biome :InputWeightAttribute Variety :MatchAttribute Biome :WeightAttribute Weight))
  (bind Projection_137 (Projection :bKeepZeroDensityPoints true :ProjectionParams (struct :bProjectPositions false :bProjectRotations false :AttributeList "Mesh,OverrideMaterial,OverrideMaterial1,TextureStamped" :AttributeMode IncludeAttributes))) ; B
  (bind Gather_266 (Gather))
  (bind Branch_8 (Branch))
  (bind AttributeFilter_0 (Filter%20Attribute%20Elements :AttributeTypes (struct :Type Float :Int32Value 3) :bUseConstantThreshold true :Operator Equal :TargetAttribute DensitySum :ThresholdAttribute @Last))
  (bind AttributeBooleanOp_0 (Attribute%20Boolean%20Op :InputSource1 IsLocallyControlled :Operation Not))
  (bind MergePoints_0 (Merge%20Points))
  (bind DeleteAttributes_1 (Delete%20Attributes :SelectedAttributes "DensitySum"))
  (bind AttributeFilter_17 (Filter%20Attribute%20Elements :bUseConstantThreshold true :Operator NotEqual :TargetAttribute DensitySum :ThresholdAttribute @Last))
  (bind DeleteAttributes_36 (Delete%20Attributes :SelectedAttributes "DensitySum"))
  (bind AttributeMathsOp_0 (Attribute%20Maths%20Op :InputSource1 $Density :Operation Set))
  ;;[0.15 0 0 0.5] tmp
  (bind AttributeCast_1 (Attribute%20Cast :InputSource InstanceStartCullDistance :OutputType String))
  (bind AttributeCast_2 (Attribute%20Cast :InputSource InstanceEndCullDistance :OutputType String))
  (bind AttributeCast_3 (Attribute%20Cast :InputSource WPODisableDistance :OutputType String))
  (bind AttributeCast_4 (Attribute%20Cast :InputSource DetailMode :OutputType String))
  (bind AttributeCast_5 (Attribute%20Cast :InputSource Dominance :OutputType String))
  (bind AttributeCast_6 (Attribute%20Cast :InputSource Interactivity :OutputType String))
  ;[0.05 0.05 0.05] Inputs
  (bind CreateAttribute_276 (Create%20Constant))
  (bind CreateAttribute_279 (Create%20Constant :AttributeTypes (struct :DoubleValue 1.0)))
  (bind GetAttributeFromPointIndex_5 (Get%20Attribute%20From%20Point%20Index :bForceOutputAttributeToBeInElementsDomain true :InputSource ActorReference :OutputAttributeName ActorReference))
  (bind Subgraph_1 (Subgraph :Graph /Game/Athena/Helios/Environments/Landscape/WorldBuildingTools/PCGGrass/PCG_MeshGrass_TextureStamping_ParseMeshSetup.PCG_MeshGrass_TextureStamping_ParseMeshSetup))
  (bind Subgraph_2 (Subgraph :Graph /Game/Athena/Helios/Environments/Landscape/WorldBuildingTools/PCGGrass/PCG_MeshGrass_TextureStamping_ParseMeshSetup.PCG_MeshGrass_TextureStamping_ParseMeshSetup))
  (bind Subgraph_3 (Subgraph :Graph /Game/Athena/Helios/Environments/Landscape/WorldBuildingTools/PCGGrass/PCG_MeshGrass_TextureStamping_ParseMeshSetup.PCG_MeshGrass_TextureStamping_ParseMeshSetup))
  (bind Subgraph_4 (Subgraph :Graph /Game/Athena/Helios/Environments/Landscape/WorldBuildingTools/PCGGrass/PCG_MeshGrass_TextureStamping_ParseMeshSetup.PCG_MeshGrass_TextureStamping_ParseMeshSetup))
  (bind Subgraph_5 (Subgraph :Graph /Game/Athena/Helios/Environments/Landscape/WorldBuildingTools/PCGGrass/PCG_MeshGrass_TextureStamping_ParseMeshSetup.PCG_MeshGrass_TextureStamping_ParseMeshSetup))
  (bind CreateAttribute_6 (Create%20Constant :AttributeTypes (struct :Type String :StringValue "Mesh,OverrideMaterial,OverrideMaterial1,TextureStamped,ScaleMin,ScaleMax,InstanceEndCullDistance,InstanceStartCullDistance,WPODisableDistance,DetailMode,Dominance,Interactivity,RenderingQuality,ScaleGradient,DensitySum,SDPTags")))
  (bind GetLoopIndex_2 (Get%20Loop%20Index :bWarnIfCalledOutsideOfLoop false))
  (bind Subgraph_6 (Subgraph :Graph /Game/Athena/Helios/Environments/Landscape/WorldBuildingTools/PCGGrass/PCG_HasValidData.PCG_HasValidData))
  (bind AttributeBooleanOp_16 (Attribute%20Boolean%20Op :Operation Not))
  (bind GetPropertyFromObjectPath_1 (Get%20Property%20From%20Object%20Path :OutputAttributeName None :PropertyName TextureStampingSetup))
  (bind DeleteAttributes_0 (Delete%20Attributes :Operation KeepSelectedAttributes :SelectedAttributes "ExclusiveChannels"))
  ;[0.15 0 0 0.5] TMP
  (bind ParseString_1 (Parse%20String :InputSource InstanceStartCullDistance))
  (bind ParseString_2 (Parse%20String :InputSource InstanceEndCullDistance))
  (bind ParseString_3 (Parse%20String :InputSource WPODisableDistance))
  (bind ParseString_4 (Parse%20String :InputSource DetailMode))
  (bind ParseString_5 (Parse%20String :InputSource Dominance :TargetType Double))
  (bind ParseString_8 (Parse%20String :InputSource Interactivity))
  (reroute %30 CreateAttribute_276)
  (reroute %31 CreateAttribute_279)
  (reroute ActorPoint LocalControls)
  (reroute AttributeList CreateAttribute_6)
  (reroute ExclusiveChannels DeleteAttributes_0)
  (reroute InPoints TargetPoints)
  (reroute IsStandalone AttributeBooleanOp_16)
  (reroute MeshSetupBlack Subgraph_5.MeshSetup)
  (reroute MeshSetupBlue Subgraph_3.MeshSetup)
  (reroute MeshSetupGreen Subgraph_2.MeshSetup)
  (reroute MeshSetupRed Subgraph_1.MeshSetup)
  (reroute MeshSetupWhite Subgraph_4.MeshSetup)
  (reroute ProjectionPoints AddAttribute_32)
  (reroute Stamping%20Actor GetAttributeFromPointIndex_5.Attribute)
  (reroute StampingAttributes GetPropertyFromObjectPath_1)
  (reroute Target%20Transform TargetTransform)
  (connect CreatePointsGrid_44 Out AddAttribute_2 In)
  (connect InPoints Out AddAttribute_3 In)
  (connect AttributeRemap_7 Out AddAttribute_32 In)
  (connect AddAttribute_99 Out AddAttribute_54 In)
  (connect AttributeMathsOp_89 Out AddAttribute_81 In)
  (connect AttributeMathsOp_82 Out AddAttribute_83 In)
  (connect ExtentsModifier_86 Out AddAttribute_97 In)
  (connect AddAttribute_97 Out AddAttribute_99 In)
  (connect ActorPoint Out ApplyScaleToBounds_47 In)
  (connect AttributeFilter_0 InsideFilter AttributeBooleanOp_0 In)
  (connect Subgraph_6 IsValid AttributeBooleanOp_16 In)
  (connect AttributeCompareOp_20 Out AttributeBooleanOp_21 InA)
  (connect AttributeCompareOp_151 Out AttributeBooleanOp_21 InB)
  (connect AttributeCompareOp_44 Out AttributeBooleanOp_46 InA)
  (connect AttributeCompareOp_35 Out AttributeBooleanOp_46 InB)
  (connect AttributeCompareOp_56 Out AttributeBooleanOp_58 InA)
  (connect AttributeCompareOp_48 Out AttributeBooleanOp_58 InB)
  (connect AttributeCompareOp_68 Out AttributeBooleanOp_70 InA)
  (connect AttributeCompareOp_60 Out AttributeBooleanOp_70 InB)
  (connect AttributeCompareOp_80 Out AttributeBooleanOp_82 InA)
  (connect AttributeCompareOp_72 Out AttributeBooleanOp_82 InB)
  (connect Branch_4 Output%20B AttributeCast_1 In)
  (connect AttributeCast_1 Out AttributeCast_2 In)
  (connect AttributeCast_2 Out AttributeCast_3 In)
  (connect AttributeCast_3 Out AttributeCast_4 In)
  (connect AttributeCast_4 Out AttributeCast_5 In)
  (connect AttributeCast_5 Out AttributeCast_6 In)
  (connect GetPointsCount_10 Out AttributeCompareOp_20 InA)
  (connect %30 Out AttributeCompareOp_20 InB)
  (connect %30 Out AttributeCompareOp_35 InA)
  (connect DataNum_26 Out AttributeCompareOp_35 InB)
  (connect GetPointsCount_8 Out AttributeCompareOp_44 InA)
  (connect %30 Out AttributeCompareOp_44 InB)
  (connect %30 Out AttributeCompareOp_48 InA)
  (connect DataNum_30 Out AttributeCompareOp_48 InB)
  (connect GetPointsCount_12 Out AttributeCompareOp_56 InA)
  (connect %30 Out AttributeCompareOp_56 InB)
  (connect %30 Out AttributeCompareOp_60 InA)
  (connect DataNum_24 Out AttributeCompareOp_60 InB)
  (connect GetPointsCount_6 Out AttributeCompareOp_68 InA)
  (connect %30 Out AttributeCompareOp_68 InB)
  (connect %30 Out AttributeCompareOp_72 InA)
  (connect DataNum_22 Out AttributeCompareOp_72 InB)
  (connect GetPointsCount_1 Out AttributeCompareOp_80 InA)
  (connect %30 Out AttributeCompareOp_80 InB)
  (connect %30 Out AttributeCompareOp_151 InA)
  (connect DataNum_28 Out AttributeCompareOp_151 InB)
  (connect Branch_8 Output%20A AttributeFilter_0 In)
  (connect Branch_8 Output%20B AttributeFilter_17 In)
  (connect AttributeMathsOp_13 Out AttributeFilter_73 Filter)
  (connect AttributeVectorOp_72 Out AttributeFilter_73 In)
  (connect Branch_1 Output%20B AttributeFilter_73 In)
  (connect ProjectionPoints Out AttributeFilter_179 Filter)
  (connect ProjectionPoints Out AttributeFilter_179 In)
  (connect AttributeFilter_179 InsideFilter AttributeFilter_181 Filter)
  (connect AttributeFilter_179 InsideFilter AttributeFilter_181 In)
  (connect MergePoints_187 Out AttributeFilter_183 Filter)
  (connect MergePoints_187 Out AttributeFilter_183 In)
  (connect AttributeFilter_183 InsideFilter AttributeFilter_185 Filter)
  (connect AttributeFilter_183 InsideFilter AttributeFilter_185 In)
  (connect MergePoints_189 Out AttributeFilter_191 Filter)
  (connect MergePoints_189 Out AttributeFilter_191 In)
  (connect AttributeFilter_191 InsideFilter AttributeFilter_193 Filter)
  (connect AttributeFilter_191 InsideFilter AttributeFilter_193 In)
  (connect MakeVectorAttribute_203 Out AttributeFilter_197 Filter)
  (connect Branch_25 Output%20B AttributeFilter_197 In)
  (connect MakeVectorAttribute_201 Out AttributeFilter_199 Filter)
  (connect Branch_25 Output%20B AttributeFilter_199 In)
  (connect DeleteAttributes_1 Out AttributeMathsOp_0 InA)
  (connect DeleteAttributes_36 Out AttributeMathsOp_0 InA)
  (connect %31 Out AttributeMathsOp_0 InB)
  (connect CopyAttributes_67 Out AttributeMathsOp_2 InA)
  (connect DeleteAttributes_21 Out AttributeMathsOp_2 InB)
  (connect AddAttribute_2 Out AttributeMathsOp_3 InA)
  (connect AddAttribute_2 Out AttributeMathsOp_3 InB)
  (connect AttributeMathsOp_3 Out AttributeMathsOp_4 InA)
  (connect GetAttributeFromPointIndex_46 Attribute AttributeMathsOp_4 InB)
  (connect AttributeMathsOp_4 Out AttributeMathsOp_5 InA)
  (connect %31 Out AttributeMathsOp_5 InB)
  (connect Branch_2 Output%20B AttributeMathsOp_6 InA)
  (connect Branch_2 Output%20B AttributeMathsOp_6 InB)
  (connect AttributeMathsOp_5 Out AttributeMathsOp_7 InA)
  (connect CreateAttribute_8 Out AttributeMathsOp_7 InB)
  (connect AttributeMathsOp_6 Out AttributeMathsOp_8 InA)
  (connect AttributeMathsOp_6 Out AttributeMathsOp_8 InB)
  (connect Gather_16 Out AttributeMathsOp_9 In)
  (connect DeleteAttributes_28 Out AttributeMathsOp_13 InA)
  (connect CreateAttribute_14 Out AttributeMathsOp_13 InB)
  (connect DeleteAttributes_44 Out AttributeMathsOp_29 In)
  (connect CreateAttribute_32 Out AttributeMathsOp_29 Max)
  (connect CreateAttribute_30 Out AttributeMathsOp_29 Min)
  (connect AttributeMathsOp_2 Out AttributeMathsOp_50 InA)
  (connect AttributeMathsOp_52 Out AttributeMathsOp_50 InB)
  (connect %31 Out AttributeMathsOp_52 InA)
  (connect GetAttributeFromPointIndex_51 Attribute AttributeMathsOp_52 InB)
  (connect GetAttributeFromPointIndex_55 Attribute AttributeMathsOp_56 InA)
  (connect CreateAttribute_57 Out AttributeMathsOp_56 InB)
  (connect SampleTexture_59 Out AttributeMathsOp_71 InA)
  (connect CreateAttribute_90 Out AttributeMathsOp_71 InB)
  (connect AddAttribute_81 Out AttributeMathsOp_82 InA)
  (connect AddAttribute_81 Out AttributeMathsOp_82 InB)
  (connect AddAttribute_83 Out AttributeMathsOp_84 InA)
  (connect AttributeMathsOp_8 Out AttributeMathsOp_84 InB)
  (connect Branch_2 Output%20A AttributeMathsOp_84 InB)
  (connect SampleTexture_58 Out AttributeMathsOp_89 InA)
  (connect CreateAttribute_74 Out AttributeMathsOp_89 InB)
  (connect Spatial%20Noise_6 Out AttributeReduce_8 In)
  (connect Spatial%20Noise_6 Out AttributeReduce_9 In)
  (connect Spatial%20Noise_6 Out AttributeRemap_7 In)
  (connect AttributeReduce_8 Out AttributeRemap_7 InRangeMax)
  (connect AttributeReduce_9 Out AttributeRemap_7 InRangeMin)
  (connect AttributeMathsOp_50 Out AttributeVectorOp_49 In)
  (connect ActorPoint Out AttributeVectorOp_49 Transform)
  (connect Branch_1 Output%20A AttributeVectorOp_72 In)
  (connect Target%20Transform Out AttributeVectorOp_72 Transform)
  (connect AttributeMathsOp_84 Out Branch_0 In)
  (connect Select_12 Out Branch_1 In)
  (connect DeleteAttributes_24 Out Branch_2 bOutputToB)
  (connect AddAttribute_83 Out Branch_2 In)
  (connect Subgraph_8 IsValid Branch_4 bOutputToB)
  (connect Branch_139 Output%20B Branch_4 In)
  (connect IsStandalone Out Branch_8 bOutputToB)
  (connect Branch_16 Output%20A Branch_8 In)
  (connect Gather_266 Out Branch_8 In)
  (connect DeleteAttributes_39 Out Branch_16 bOutputToB)
  (connect Gather_21 Out Branch_16 In)
  (connect DeleteAttributes_37 Out Branch_25 bOutputToB)
  (connect ProjectionPoints Out Branch_25 In)
  (connect AttributeBooleanOp_58 Out Branch_37 bOutputToB)
  (connect AttributeFilter_181 InsideFilter Branch_37 In)
  (connect Subgraph_55 IsValid Branch_56 bOutputToB)
  (connect Branch_0 Output%20A Branch_56 In)
  (connect CopyAttributes_2 Out Branch_56 In)
  (connect AttributeBooleanOp_21 Out Branch_64 bOutputToB)
  (connect AttributeFilter_185 InsideFilter Branch_64 In)
  (connect AttributeBooleanOp_46 Out Branch_66 bOutputToB)
  (connect AttributeFilter_193 InsideFilter Branch_66 In)
  (connect AttributeBooleanOp_70 Out Branch_68 bOutputToB)
  (connect AttributeFilter_197 InsideFilter Branch_68 In)
  (connect AttributeBooleanOp_82 Out Branch_70 bOutputToB)
  (connect AttributeFilter_199 InsideFilter Branch_70 In)
  (connect AttributeBooleanOp_58 Out Branch_139 bOutputToB)
  (connect Subgraph_39 Success Branch_139 In)
  (connect AttributeBooleanOp_21 Out Branch_147 bOutputToB)
  (connect Gather_17 Out Branch_147 In)
  (connect AttributeBooleanOp_46 Out Branch_155 bOutputToB)
  (connect Gather_19 Out Branch_155 In)
  (connect AttributeBooleanOp_70 Out Branch_163 bOutputToB)
  (connect Branch_16 Output%20B Branch_163 In)
  (connect AttributeBooleanOp_82 Out Branch_171 bOutputToB)
  (connect Gather_38 Out Branch_171 In)
  (connect DeleteAttributes_32 Out BreakVectorAttribute_34 In)
  (connect Select_12 Out CopyAttributes_0 Source)
  (connect Select_12 Out CopyAttributes_0 Target)
  (connect Branch_0 Output%20B CopyAttributes_2 Source)
  (connect Branch_0 Output%20B CopyAttributes_2 Target)
  (connect Subgraph_7 ColorAttribute CopyAttributes_66 InputSource)
  (connect AttributeMathsOp_71 Out CopyAttributes_66 Source)
  (connect Branch_56 Output%20B CopyAttributes_66 Target)
  (connect AttributeMathsOp_7 Out CopyAttributes_67 Source)
  (connect AttributeMathsOp_7 Out CopyAttributes_67 Target)
  (connect MakeVectorAttribute_54 Out CreatePointsGrid_44 CellSize)
  (connect GetAttributeFromPointIndex_46 Attribute CreatePointsGrid_44 GridExtents)
  (connect MeshSetupBlack Out DataNum_22 In)
  (connect MeshSetupWhite Out DataNum_24 In)
  (connect MeshSetupBlue Out DataNum_26 In)
  (connect MeshSetupGreen Out DataNum_28 In)
  (connect MeshSetupRed Out DataNum_30 In)
  (connect StampingAttributes Out DeleteAttributes_0 In)
  (connect MergePoints_0 Out DeleteAttributes_1 In)
  (connect StampingAttributes Out DeleteAttributes_2 In)
  (connect StampingAttributes Out DeleteAttributes_14 In)
  (connect StampingAttributes Out DeleteAttributes_17 In)
  (connect StampingAttributes Out DeleteAttributes_21 In)
  (connect StampingAttributes Out DeleteAttributes_24 In)
  (connect StampingAttributes Out DeleteAttributes_28 In)
  (connect StampingAttributes Out DeleteAttributes_32 In)
  (connect AttributeFilter_17 InsideFilter DeleteAttributes_36 In)
  (connect StampingAttributes Out DeleteAttributes_37 In)
  (connect StampingAttributes Out DeleteAttributes_39 In)
  (connect StampingAttributes Out DeleteAttributes_44 In)
  (connect StampingAttributes Out DeleteAttributes_53 In)
  (connect MakeVectorAttribute_88 Out ExtentsModifier_86 Extents)
  (connect Subgraph_38 ValidPoints ExtentsModifier_86 In)
  (connect AddAttribute_54 Out FilterDataByAttribute_10 In)
  (connect Branch_37 Output%20A Gather_0 In)
  (connect Select_0 Out Gather_0 In)
  (connect Branch_64 Output%20A Gather_1 In)
  (connect Select_3 Out Gather_1 In)
  (connect Branch_56 Output%20A Gather_16 In)
  (connect CopyAttributes_66 Out Gather_16 In)
  (connect Branch_139 Output%20A Gather_17 In)
  (connect Projection_125 Out Gather_17 In)
  (connect Branch_147 Output%20A Gather_19 In)
  (connect Projection_123 Out Gather_19 In)
  (connect Branch_155 Output%20A Gather_21 In)
  (connect Projection_115 Out Gather_21 In)
  (connect Branch_163 Output%20A Gather_38 In)
  (connect Projection_135 Out Gather_38 In)
  (connect Branch_171 Output%20A Gather_266 In)
  (connect Projection_137 Out Gather_266 In)
  (connect ActorPoint Out GetAttributeFromPointIndex_5 In)
  (connect ApplyScaleToBounds_47 Out GetAttributeFromPointIndex_46 In)
  (connect ActorPoint Out GetAttributeFromPointIndex_51 In)
  (connect ApplyScaleToBounds_47 Out GetAttributeFromPointIndex_55 In)
  (connect AttributeFilter_199 InsideFilter GetPointsCount_1 In)
  (connect AttributeFilter_197 InsideFilter GetPointsCount_6 In)
  (connect AttributeFilter_193 InsideFilter GetPointsCount_8 In)
  (connect AttributeFilter_185 InsideFilter GetPointsCount_10 In)
  (connect AttributeFilter_181 InsideFilter GetPointsCount_12 In)
  (connect Stamping%20Actor Out GetPropertyFromObjectPath_1 In)
  (connect DeleteAttributes_17 Out GetTextureData_59 Texture)
  (connect DeleteAttributes_53 Out GetTextureData_61 Texture)
  (connect AttributeMathsOp_29 Out MakeVectorAttribute_54 X)
  (connect AttributeMathsOp_29 Out MakeVectorAttribute_54 Y)
  (connect AttributeMathsOp_56 Out MakeVectorAttribute_54 Z)
  (connect BreakVectorAttribute_34 X MakeVectorAttribute_201 X)
  (connect BreakVectorAttribute_34 X MakeVectorAttribute_201 Y)
  (connect BreakVectorAttribute_34 X MakeVectorAttribute_201 Z)
  (connect BreakVectorAttribute_34 Y MakeVectorAttribute_203 X)
  (connect BreakVectorAttribute_34 Y MakeVectorAttribute_203 Y)
  (connect BreakVectorAttribute_34 Y MakeVectorAttribute_203 Z)
  (connect Branch_37 Output%20B MatchAndSetAttributes_11 In)
  (connect MeshSetupRed Out MatchAndSetAttributes_11 Match%20Data)
  (connect Branch_64 Output%20B MatchAndSetAttributes_27 In)
  (connect MeshSetupGreen Out MatchAndSetAttributes_27 Match%20Data)
  (connect Branch_66 Output%20B MatchAndSetAttributes_29 In)
  (connect MeshSetupBlue Out MatchAndSetAttributes_29 Match%20Data)
  (connect Branch_68 Output%20B MatchAndSetAttributes_31 In)
  (connect MeshSetupWhite Out MatchAndSetAttributes_31 Match%20Data)
  (connect Branch_70 Output%20B MatchAndSetAttributes_33 In)
  (connect MeshSetupBlack Out MatchAndSetAttributes_33 Match%20Data)
  (connect AttributeBooleanOp_0 Out MergePoints_0 In)
  (connect AttributeFilter_0 OutsideFilter MergePoints_0 In)
  (connect AttributeFilter_179 OutsideFilter MergePoints_187 In)
  (connect AttributeFilter_181 OutsideFilter MergePoints_187 In)
  (connect Gather_0 Out MergePoints_187 In)
  (connect AttributeFilter_183 OutsideFilter MergePoints_189 In)
  (connect AttributeFilter_185 OutsideFilter MergePoints_189 In)
  (connect Gather_1 Out MergePoints_189 In)
  (connect AttributeMathsOp_0 Out ParseString_1 In)
  (connect ParseString_1 Out ParseString_2 In)
  (connect ParseString_2 Out ParseString_3 In)
  (connect ParseString_3 Out ParseString_4 In)
  (connect ParseString_4 Out ParseString_5 In)
  (connect ParseString_5 Out ParseString_8 In)
  (connect AttributeList Out Projection_115 AttributeList)
  (connect Branch_155 Output%20B Projection_115 In)
  (connect MatchAndSetAttributes_29 Out Projection_115 Projection%20Target)
  (connect AttributeList Out Projection_123 AttributeList)
  (connect Branch_147 Output%20B Projection_123 In)
  (connect MatchAndSetAttributes_27 Out Projection_123 Projection%20Target)
  (connect AttributeList Out Projection_125 AttributeList)
  (connect AttributeCast_6 Out Projection_125 In)
  (connect Branch_4 Output%20A Projection_125 In)
  (connect MatchAndSetAttributes_11 Out Projection_125 Projection%20Target)
  (connect AttributeList Out Projection_135 AttributeList)
  (connect Branch_163 Output%20B Projection_135 In)
  (connect MatchAndSetAttributes_31 Out Projection_135 Projection%20Target)
  (connect AttributeList Out Projection_137 AttributeList)
  (connect Branch_171 Output%20B Projection_137 In)
  (connect MatchAndSetAttributes_33 Out Projection_137 Projection%20Target)
  (connect GetTextureData_59 Out SampleTexture_58 BaseTexture)
  (connect AttributeVectorOp_49 Out SampleTexture_58 Point)
  (connect GetTextureData_61 Out SampleTexture_59 BaseTexture)
  (connect Branch_56 Output%20B SampleTexture_59 Point)
  (connect ExclusiveChannels Out Select_0 bUseInputB)
  (connect Branch_37 Output%20B Select_0 Input%20A)
  (connect ExclusiveChannels Out Select_3 bUseInputB)
  (connect Branch_64 Output%20B Select_3 Input%20A)
  (connect DeleteAttributes_14 Out Select_12 bUseInputB)
  (connect Gather_16 Out Select_12 Input%20A)
  (connect AttributeMathsOp_9 Out Select_12 Input%20B)
  (connect FilterDataByAttribute_10 OutsideFilter Spatial%20Noise_6 In)
  (connect Stamping%20Actor Out Subgraph_1 SourceActor)
  (connect Stamping%20Actor Out Subgraph_2 SourceActor)
  (connect Stamping%20Actor Out Subgraph_3 SourceActor)
  (connect Stamping%20Actor Out Subgraph_4 SourceActor)
  (connect Stamping%20Actor Out Subgraph_5 SourceActor)
  (connect GetLoopIndex_2 Out Subgraph_6 In)
  (connect DeleteAttributes_2 Out Subgraph_7 Index)
  (connect GetLoopIndex_0 Out Subgraph_8 In)
  (connect AttributeFilter_73 InsideFilter Subgraph_38 In)
  (connect ProjectionPoints Out Subgraph_39 Condition)
  (connect AddAttribute_3 Out Subgraph_39 Payload)
  (connect DeleteAttributes_53 Out Subgraph_55 In)
  (output TargetPoints ParseString_8))
