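; PCG graph that builds the MeshSetup attribute set for texture-stamped mesh grass in one biome channel.
; PCGBP_GatherBiomeData supplies biome data, and property paths built from "Biome" plus TargetChannel read its VarietyData and MeshData.
; Both sets are normalized, merged, filtered on Mesh, then have settings cast to strings and are flagged TextureStamped.
; An ObjectDependencies lookup of bEnableDensityScaling does not feed the output.
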(graph PCG_MeshGrass_TextureStamping_ParseMeshSetup
  (params (TargetChannel :type String))
  (inputs (SourceActor :type Param))
  (outputs (MeshSetup :type Param))
  (bind CreateAttribute_9 (Create%20Constant :AttributeTypes (struct :Type String :StringValue "Biome")))
  (bind AttributeStringOp_8 (Attribute%20String%20Op :InB TargetChannel))
  (bind CreateAttribute_12 (Create%20Constant :AttributeTypes (struct :Type String :StringValue ".VarietyData")))
  (bind AttributeStringOp_14 (Attribute%20String%20Op))
  (bind CreateAttribute_16 (Create%20Constant :AttributeTypes (struct :Type String :StringValue "MeshSetup")))
  (bind AttributeStringOp_18 (Attribute%20String%20Op :InB TargetChannel))
  (bind ExecuteBlueprint_0 (Execute%20Blueprint :BlueprintElementInstance (class:/Game/Athena/Helios/Environments/Landscape/WorldBuildingTools/PCGGrass/PCGBP_GatherBiomeData.PCGBP_GatherBiomeData_C) :BlueprintElementType /Game/Athena/Helios/Environments/Landscape/WorldBuildingTools/PCGGrass/PCGBP_GatherBiomeData.PCGBP_GatherBiomeData_C))
  (bind FilterDataByType_1 (Filter%20Data%20By%20Type :TargetType (struct :Ids [(struct :Struct /Script/PCG.PCGDataTypeInfoParam)])))
  (bind GetPropertyFromObjectPath_0 (Get%20Property%20From%20Object%20Path :bSanitizeOutputAttributeName false :OutputAttributeName None))
  (bind CreateAttribute_24 (Create%20Constant :AttributeTypes (struct :Type String :StringValue ".MeshData")))
  (bind AttributeStringOp_22 (Attribute%20String%20Op))
  (bind GetPropertyFromObjectPath_1 (Get%20Property%20From%20Object%20Path :bSanitizeOutputAttributeName false :OutputAttributeName MeshSetupRed :PropertyName BiomeRed.MeshData))
  (bind AttributeRename_2 (Attribute%20Rename :AttributeToRename GrassMesh :NewAttributeName Mesh))
  (bind Subgraph_0 (Subgraph :Graph /Game/Athena/Helios/Environments/Landscape/WorldBuildingTools/PCGGrass/PCG_GateValidEntries.PCG_GateValidEntries))
  (bind Subgraph_26 (Subgraph :Graph /Game/Athena/Helios/Environments/Landscape/WorldBuildingTools/PCGGrass/PCG_MeshGrass_ExtraMeshAttributes.PCG_MeshGrass_ExtraMeshAttributes))
  (bind AddAttribute_25 (Add%20Attribute :bCopyAllAttributes true))
  (bind AttributeRename_3 (Attribute%20Rename :AttributeToRename StartCullDistance :NewAttributeName InstanceStartCullDistance))
  (bind AttributeRename_4 (Attribute%20Rename :AttributeToRename EndCullDistance :NewAttributeName InstanceEndCullDistance))
  (bind MergeAttributes_3 (Merge%20Attributes))
  (bind AttributeFilter_9 (Filter%20Attribute%20Elements :AttributeTypes (struct :Type SoftObjectPath) :bUseConstantThreshold true :Operator NotEqual :TargetAttribute Mesh :ThresholdAttribute @Last))
  (bind Subgraph_5 (Subgraph :Graph /Game/Athena/Helios/Environments/Landscape/WorldBuildingTools/PCGGrass/PCG_GateValidEntries.PCG_GateValidEntries))
  (bind AddAttribute_4 (Add%20Attribute :AttributeTypes (struct :Type Boolean :BoolValue true) :OutputTarget TextureStamped))
  (bind CreateConstant_0 (Create%20Constant :AttributeTypes (struct :Type String :StringValue ".ObjectDependencies")))
  (bind AttributeStringOp_0 (Attribute%20String%20Op))
  (bind GetPropertyFromObjectPath_2 (Get%20Property%20From%20Object%20Path))
  (bind GetPropertyFromObjectPath_3 (Get%20Property%20From%20Object%20Path :PropertyName bEnableDensityScaling))
  ;[0.15 0 0 0.5] TMP
  (bind AttributeCast_0 (Attribute%20Cast :InputSource InstanceStartCullDistance :OutputType String))
  (bind AttributeCast_7 (Attribute%20Cast :InputSource InstanceEndCullDistance :OutputType String))
  (bind AttributeCast_9 (Attribute%20Cast :InputSource WPODisableDistance :OutputType String))
  (bind AttributeCast_11 (Attribute%20Cast :InputSource DetailMode :OutputType String))
  (bind AttributeCast_13 (Attribute%20Cast :InputSource Dominance :OutputType String))
  (bind AttributeCast_1 (Attribute%20Cast :InputSource Interactivity :OutputType String))
  (connect AttributeCast_1 Out AddAttribute_4 In)
  (connect Subgraph_26 MeshAttributes AddAttribute_25 Attributes)
  (connect Subgraph_0 ValidEntires AddAttribute_25 In)
  (connect Subgraph_5 ValidEntires AttributeCast_0 In)
  (connect AttributeCast_13 Out AttributeCast_1 In)
  (connect AttributeCast_0 Out AttributeCast_7 In)
  (connect AttributeCast_7 Out AttributeCast_9 In)
  (connect AttributeCast_9 Out AttributeCast_11 In)
  (connect AttributeCast_11 Out AttributeCast_13 In)
  (connect MergeAttributes_3 Out AttributeFilter_9 In)
  (connect GetPropertyFromObjectPath_0 Out AttributeRename_2 In)
  (connect AttributeRename_2 Out AttributeRename_3 In)
  (connect AttributeRename_3 Out AttributeRename_4 In)
  (connect AttributeStringOp_8 Out AttributeStringOp_0 InA)
  (connect CreateConstant_0 Out AttributeStringOp_0 InB)
  (connect CreateAttribute_9 Out AttributeStringOp_8 InA)
  (connect AttributeStringOp_8 Out AttributeStringOp_14 InA)
  (connect CreateAttribute_12 Out AttributeStringOp_14 InB)
  (connect CreateAttribute_16 Out AttributeStringOp_18 InA)
  (connect AttributeStringOp_8 Out AttributeStringOp_22 InA)
  (connect CreateAttribute_24 Out AttributeStringOp_22 InB)
  (connect DefaultInputNode SourceActor ExecuteBlueprint_0 In)
  (connect ExecuteBlueprint_0 Out FilterDataByType_1 In)
  (connect FilterDataByType_1 InsideFilter GetPropertyFromObjectPath_0 In)
  (connect AttributeStringOp_18 Out GetPropertyFromObjectPath_0 OutputAttributeName)
  (connect AttributeStringOp_14 Out GetPropertyFromObjectPath_0 PropertyName)
  (connect FilterDataByType_1 InsideFilter GetPropertyFromObjectPath_1 In)
  (connect AttributeStringOp_18 Out GetPropertyFromObjectPath_1 OutputAttributeName)
  (connect AttributeStringOp_22 Out GetPropertyFromObjectPath_1 PropertyName)
  (connect FilterDataByType_1 InsideFilter GetPropertyFromObjectPath_2 In)
  (connect AttributeStringOp_0 Out GetPropertyFromObjectPath_2 PropertyName)
  (connect GetPropertyFromObjectPath_2 Out GetPropertyFromObjectPath_3 In)
  (connect AddAttribute_25 Out MergeAttributes_3 In)
  (connect AttributeRename_4 Out MergeAttributes_3 In)
  (connect GetPropertyFromObjectPath_1 Out Subgraph_0 In)
  (connect AttributeFilter_9 InsideFilter Subgraph_5 In)
  (output MeshSetup AddAttribute_4))
