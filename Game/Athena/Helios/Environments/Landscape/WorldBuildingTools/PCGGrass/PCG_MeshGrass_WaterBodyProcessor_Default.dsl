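; PCG subgraph for mesh grass that turns a water body's splines into water body points, using river- or lake-specific processing.
; The first actor point's ActorReference is string-matched against "River" and then "Lake" to branch the splines into the matching Helios subgraph.
; Splines matching neither name are not processed.
; Subgraph results are gathered and pass through an element-count check, so points are output only when non-empty.
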(graph PCG_MeshGrass_WaterBodyProcessor_Default
  (inputs (Exec :type Any) (WaterBodyActors :type Point) (WaterBodySplines :type Spline))
  (outputs (WaterBodyPoints :type Point))
  (bind GetAttributeFromPointIndex_0 (Get%20Attribute%20From%20Point%20Index :bForceOutputAttributeToBeInElementsDomain true :InputSource ActorReference :OutputAttributeName ActorReference))
  (bind ExecuteBlueprint_2 (Execute%20Blueprint :BlueprintElementInstance (class:/Game/Athena/Helios/Environments/Landscape/WorldBuildingTools/PCGGrass/PCGBP_StringContains.PCGBP_StringContains_C :Substring "River" :TargetAttribute ActorReference) :BlueprintElementType /Game/Athena/Helios/Environments/Landscape/WorldBuildingTools/PCGGrass/PCGBP_StringContains.PCGBP_StringContains_C))
  (bind FilterAttributesByName_14 (Delete%20Attributes :bTokenizeOnWhiteSpace true :Operation KeepSelectedAttributes :SelectedAttributes "ContainsSubstring"))
  (bind Branch_3 (Branch))
  (bind ExecuteBlueprint_7 (Execute%20Blueprint :BlueprintElementInstance (class:/Game/Athena/Helios/Environments/Landscape/WorldBuildingTools/PCGGrass/PCGBP_StringContains.PCGBP_StringContains_C :Substring "Lake" :TargetAttribute ActorReference) :BlueprintElementType /Game/Athena/Helios/Environments/Landscape/WorldBuildingTools/PCGGrass/PCGBP_StringContains.PCGBP_StringContains_C))
  (bind FilterAttributesByName_9 (Delete%20Attributes :bTokenizeOnWhiteSpace true :Operation KeepSelectedAttributes :SelectedAttributes "ContainsSubstring"))
  (bind Branch_11 (Branch))
  (bind Subgraph_4 (Subgraph :Graph /Game/Athena/Helios/Environments/Landscape/WorldBuildingTools/PCGGrass/PCG_MeshGrass_WaterBodyProcessor_Lakes_Helios.PCG_MeshGrass_WaterBodyProcessor_Lakes_Helios))
  (bind Subgraph_5 (Subgraph :Graph /Game/Athena/Helios/Environments/Landscape/WorldBuildingTools/PCGGrass/PCG_MeshGrass_WaterBodyProcessor_Rivers_Helios.PCG_MeshGrass_WaterBodyProcessor_Rivers_Helios))
  (bind Gather_6 (Gather))
  (bind GetPointsCount_2 (Get%20Element%20Count))
  (bind CreateAttribute_4 (Create%20Constant))
  (bind AttributeCompareOp_3 (Attribute%20Compare%20Op :Operation NotEqual))
  (bind Branch_5 (Branch))
  (connect GetPointsCount_2 Out AttributeCompareOp_3 InA)
  (connect CreateAttribute_4 Out AttributeCompareOp_3 InB)
  (connect FilterAttributesByName_14 Out Branch_3 bOutputToB)
  (connect DefaultInputNode WaterBodySplines Branch_3 In)
  (connect AttributeCompareOp_3 Out Branch_5 bOutputToB)
  (connect Gather_6 Out Branch_5 In)
  (connect FilterAttributesByName_9 Out Branch_11 bOutputToB)
  (connect Branch_3 Output%20A Branch_11 In)
  (connect GetAttributeFromPointIndex_0 Attribute ExecuteBlueprint_2 InString)
  (connect GetAttributeFromPointIndex_0 Attribute ExecuteBlueprint_7 InString)
  (connect ExecuteBlueprint_7 Out FilterAttributesByName_9 In)
  (connect ExecuteBlueprint_2 Out FilterAttributesByName_14 In)
  (connect Subgraph_4 WaterBodyPoints Gather_6 In)
  (connect Subgraph_5 WaterBodyPoints Gather_6 In)
  (connect DefaultInputNode WaterBodyActors GetAttributeFromPointIndex_0 In)
  (connect Gather_6 Out GetPointsCount_2 In)
  (connect Branch_11 Output%20B Subgraph_4 InSplineData)
  (connect Branch_3 Output%20B Subgraph_5 InSplineData)
  (output WaterBodyPoints Branch_5.Output%20B))
