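; PCG subgraph that turns a lake spline into water body points for the Helios mesh grass tools.
; A Spline Sampler fills the spline interior with points projected onto the surface, bounded by a single point from Get Actor Data.
; Base interior and border spacings are divided by a Max-clamped WaterBodyPrecision, so higher precision gives denser sampling.
; A Bounds Modifier reshapes the point bounds before output as WaterBodyPoints.
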(graph PCG_MeshGrass_WaterBodyProcessor_Lakes_Helios
  (inputs (InSplineData :type Spline))
  (outputs (WaterBodyPoints :type Point))
  (bind AttributeMathsOp_2 (Attribute%20Maths%20Op :Operation Max :InA WaterBodyPrecision))
  (bind CreateConstant_0 (Create%20Constant :AttributeTypes (struct :Type Float :FloatValue 5e+02)))
  (bind AttributeMathsOp_0 (Attribute%20Maths%20Op :Operation Divide))
  (bind CreateConstant_1 (Create%20Constant :AttributeTypes (struct :Type Float :FloatValue 2e+03)))
  (bind AttributeMathsOp_1 (Attribute%20Maths%20Op :Operation Divide))
  (bind GetActorData_1 (Get%20Actor%20Data :bComponentsMustOverlapSelf false :Mode GetSinglePoint))
  (bind SplineSampler_2 (Spline%20Sampler :SamplerParams (struct :Dimension OnInterior :Mode Distance :DistanceIncrement 5e+02 :InteriorSampleSpacing 5e+02 :InteriorBorderSampleSpacing 2e+03 :bProjectOntoSurface true)))
  (bind BoundsModifier_1 (Bounds%20Modifier :BoundsMax (0.25 0.25 0.0) :BoundsMin (0.25 0.25 2.0)))
  (connect CreateConstant_0 Out AttributeMathsOp_0 InA)
  (connect AttributeMathsOp_2 Out AttributeMathsOp_0 InB)
  (connect CreateConstant_1 Out AttributeMathsOp_1 InA)
  (connect AttributeMathsOp_2 Out AttributeMathsOp_1 InB)
  (connect SplineSampler_2 Out BoundsModifier_1 In)
  (connect GetActorData_1 Out SplineSampler_2 Bounding%20Shape)
  (connect AttributeMathsOp_1 Out SplineSampler_2 InteriorBorderSampleSpacing)
  (connect AttributeMathsOp_0 Out SplineSampler_2 InteriorSampleSpacing)
  (connect DefaultInputNode InSplineData SplineSampler_2 Spline)
  (output WaterBodyPoints BoundsModifier_1))
