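; PCG graph that turns river splines into water body points for the Helios landscape mesh grass tools.
; It samples the input spline at fixed distance intervals, using single-point actor data as the sampler's bounding shape.
; Each point swaps its X and Y scale, then a Bounds Modifier reshapes its bounds, with a far larger minimum Z extending them vertically.
; The resulting points are output as WaterBodyPoints.
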(graph PCG_MeshGrass_WaterBodyProcessor_Rivers_Helios
  (inputs (InSplineData :type Spline))
  (outputs (WaterBodyPoints :type Point))
  (bind GetActorData_2 (Get%20Actor%20Data :bComponentsMustOverlapSelf false :Mode GetSinglePoint))
  (bind SplineSampler_2 (Spline%20Sampler :SamplerParams (struct :Mode Distance :DistanceIncrement 5e+02)))
  (bind CopyAttribute_0 (Copy%20Attributes :InputSource $Scale.XY :OutputTarget $Scale.YX))
  (bind BoundsModifier_1 (Bounds%20Modifier :BoundsMax (15.0 0.4 1.0) :BoundsMin (15.0 0.4 5e+02)))
  (connect CopyAttribute_0 Out BoundsModifier_1 In)
  (connect SplineSampler_2 Out CopyAttribute_0 Source)
  (connect SplineSampler_2 Out CopyAttribute_0 Target)
  (connect GetActorData_2 Out SplineSampler_2 Bounding%20Shape)
  (connect DefaultInputNode InSplineData SplineSampler_2 Spline)
  (output WaterBodyPoints BoundsModifier_1))
