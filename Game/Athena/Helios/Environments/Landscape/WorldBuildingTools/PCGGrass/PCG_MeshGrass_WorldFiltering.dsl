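; PCG subgraph that filters mesh grass points against the world, with stages toggled by actor attributes, global attributes and graph parameters.
; Points pass through self-pruning, cull-volume difference, intersection, trace and border filters, then water-body distance filtering against landscape-projected spline points.
; Landscape stages drop below-terrain points and excluded layers, using blur for dilation and optional gradient scaling.
; A final optional loop processes world textures from the GlobalControls actor.
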(graph PCG_MeshGrass_WorldFiltering
  (inputs (InPoints :type Point) (ActorPoint :type Point) (ActorAttributes :type Param) (VolumeAttributes :type Param) (BiomeAttributes :type Param) (GlobalAttributes :type Param))
  (outputs (OutPoints :type Point))
  ;[0.05 0.05 0.05] Landscape based processing
  (bind Gather_31 (Gather))
  ;;[0.05 0.05 0.05] Landscape Check
  (bind CreateAttribute_0 (Create%20Constant :AttributeTypes (struct :Type Integer32)))
  (bind DataNum_0 (Data%20Count))
  (bind AttributeCompareOp_1 (Attribute%20Compare%20Op :Operation NotEqual :Tolerance 0.0))
  (bind Branch_8 (Branch))
  ;;[0.05 0.05 0.05] Below terrain handling
  (bind Branch_14 (Branch :bOutputToB-param FilterBelowTerrain))
  (bind Projection_7 (Projection :bKeepZeroDensityPoints true :ProjectionParams (struct :bProjectRotations false :AttributeMode IncludeAttributes)))
  (bind AttributeMathsOp_12 (Attribute%20Maths%20Op :InputSource1 $Position.Z :InB BelowTerrainBias))
  (bind PointFilter_10 (Filter%20Attribute%20Elements :AttributeTypes (struct :Type Boolean :BoolValue true) :Operator Lesser :TargetAttribute $Position.Z :ThresholdAttribute $Position.Z))
  (bind Subgraph_5 (Subgraph :Graph /Game/Athena/Helios/Environments/Landscape/WorldBuildingTools/PCGGrass/PCG_GateValidPoints.PCG_GateValidPoints))
  (bind Gather_7 (Gather))
  ;;[0.05 0.05 0.05] Layer Exclusion
  (bind CreateAttribute_10 (Create%20Constant :AttributeTypes (struct :Type String)))
  (bind DeleteAttributes_6 (Delete%20Attributes :Operation KeepSelectedAttributes :SelectedAttributes "LandscapeLayers"))
  (bind AttributeCompareOp_8 (Attribute%20Compare%20Op :Operation NotEqual))
  (bind DeleteAttributes_1 (Delete%20Attributes :Operation KeepSelectedAttributes :SelectedAttributes "FilterLandscapeLayers"))
  (bind AttributeBooleanOp_12 (Attribute%20Boolean%20Op))
  (bind CreateAttribute_46 (Create%20Constant :AttributeTypes (struct :Type String)))
  (bind FilterAttribute_12 (Delete%20Attributes :bTokenizeOnWhiteSpace true :Operation KeepSelectedAttributes :SelectedAttributes "ExcludedLayerNames"))
  (bind AttributeCompareOp_45 (Attribute%20Compare%20Op :Operation NotEqual))
  (bind FilterAttribute_3 (Delete%20Attributes :bTokenizeOnWhiteSpace true :Operation KeepSelectedAttributes :SelectedAttributes "ExcludeByLandscapeLayers"))
  (bind AttributeBooleanOp_47 (Attribute%20Boolean%20Op))
  (bind AttributeBooleanOp_2 (Attribute%20Boolean%20Op :Operation Or))
  (bind Branch_1 (Branch))
  (bind AttributeBooleanOp_3 (Attribute%20Boolean%20Op))
  (bind AttributeRename_6 (Attribute%20Rename :AttributeToRename LandscapeLayers :NewAttributeName ExcludedLayerNames))
  (bind CreateAttribute_16 (Create%20Constant :AttributeTypes (struct :Type String :StringValue ",")))
  (bind AttributeStringOp_15 (Attribute%20String%20Op))
  (bind AttributeStringOp_18 (Attribute%20String%20Op))
  (bind Select_5 (Select))
  (bind Select_0 (Select))
  (bind ExecuteBlueprint_1 (Execute%20Blueprint :BlueprintElementInstance (class:/Game/Athena/Helios/Environments/Landscape/WorldBuildingTools/PCGGrass/PCGBP_TrimString.PCGBP_TrimString_C :Remove%20All%20Whitespaces true) :BlueprintElementType /Game/Athena/Helios/Environments/Landscape/WorldBuildingTools/PCGGrass/PCGBP_TrimString.PCGBP_TrimString_C))
  (bind Projection_8 (Projection :bKeepZeroDensityPoints true :ProjectionParams (struct :bProjectPositions false :bProjectRotations false :AttributeMode IncludeAttributes)))
  (bind DeleteAttributes_29 (Delete%20Attributes :Operation KeepSelectedAttributes :SelectedAttributes "LandscapeLayerDilation"))
  (bind AddAttribute_25 (Add%20Attribute :OutputTarget LayerExclusion))
  (bind ExecuteBlueprint_8 (Execute%20Blueprint :BlueprintElementInstance (class:/Game/Athena/Helios/Environments/Landscape/WorldBuildingTools/PCGGrass/PCGBP_ParseString.PCGBP_ParseString_C :Delimiter "," :SourceAttribute "ParsedStrings") :BlueprintElementType /Game/Athena/Helios/Environments/Landscape/WorldBuildingTools/PCGGrass/PCGBP_ParseString.PCGBP_ParseString_C))
  (bind AttributePartition_40 (Attribute%20Partition))
  (bind Loop_43 (Loop :Graph /Game/Athena/Helios/Environments/Landscape/WorldBuildingTools/PCGGrass/PCG_AttributesMax.PCG_AttributesMax))
  (bind DeleteAttributes_30 (Delete%20Attributes))
  (bind BlurElement_27 (Blur :SearchDistance 0.0))
  (bind DeleteAttributes_26 (Delete%20Attributes :Operation KeepSelectedAttributes :SelectedAttributes "LandscapeLayerThreshold"))
  (bind PointFilter_9 (Filter%20Attribute%20Elements :Operator LesserOrEqual :TargetAttribute LayerExclusion :ThresholdAttribute @Last))
  (bind CreateAttribute_37 (Create%20Constant :AttributeTypes (struct :DoubleValue 1.0)))
  (bind AttributeMathsOp_36 (Attribute%20Maths%20Op :Operation Divide))
  (bind DeleteAttributes_51 (Delete%20Attributes :bTokenizeOnWhiteSpace true :Operation KeepSelectedAttributes :SelectedAttributes "ScalebyGradient"))
  (bind Branch_48 (Branch :bOutputToB true))
  (bind AttributeMathsOp_35 (Attribute%20Maths%20Op :Operation Multiply))
  (bind AttributeMathsOp_38 (Attribute%20Maths%20Op :Operation OneMinus))
  (bind DeleteAttributes_43 (Delete%20Attributes :bTokenizeOnWhiteSpace true :Operation KeepSelectedAttributes :SelectedAttributes "GradientScaleMinMax"))
  (bind BreakVectorAttribute_41 (Break%20Vector%20Attribute))
  (bind AttributeRemap_40 (Attribute%20Remap :bAllowInverseRange false))
  (bind AttributeMathsOp_39 (Attribute%20Maths%20Op :InputSource1 $Scale :Operation Multiply))
  (bind DeleteAttributes_34 (Delete%20Attributes :SelectedAttributes "LayerExclusion"))
  (bind Subgraph_9 (Subgraph :Graph /Game/Athena/Helios/Environments/Landscape/WorldBuildingTools/PCGGrass/PCG_GateValidPoints.PCG_GateValidPoints))
  ;[0.05 0.05 0.05] Handle water bodies
  (bind AttributeMathsOp_0 (Attribute%20Maths%20Op :Operation Multiply :InA WaterBodyDistance))
  (bind Select_1 (Select :bUseInputB-param WaterBodyInvertFilter :Input%20A WaterBodyDistance))
  (bind AttributeMathsOp_37 (Attribute%20Maths%20Op :Operation Abs))
  (bind AttributeMathsOp_40 (Attribute%20Maths%20Op :Operation Max))
  (bind CreateAttribute_26 (Create%20Constant :AttributeTypes (struct :DoubleValue -1.0)))
  (bind Branch_31 (Branch :bOutputToB-param FilterWaterBodies))
  (bind GetActorData_2 (Get%20Actor%20Data :ActorSelector (struct :ActorFilter AllWorldActors :bMustOverlapSelf true :ActorSelection ByClass :ActorSelectionClass /Game/Athena/Apollo/Environments/Landscape/Water/FortWaterBodyBP.FortWaterBodyBP_C :bSelectMultiple true) :bComponentsMustOverlapSelf false :Mode GetSinglePoint))
  (bind GetActorData_4 (Get%20Actor%20Data :ActorSelector (struct :ActorFilter FromInput :bMustOverlapSelf true :ActorSelection ByClass :ActorSelectionClass /Game/Athena/Apollo/Environments/Landscape/Water/FortWaterBodyBP.FortWaterBodyBP_C :bSelectMultiple true) :bComponentsMustOverlapSelf false :ComponentSelector (struct :ComponentSelection ByClass :ComponentSelectionClass /Script/Engine.SplineComponent)))
  (bind FilterByType_2 (Filter%20Data%20By%20Type :TargetType (struct :Ids [(struct :Struct /Script/PCG.PCGDataTypeInfoSpline)])))
  (bind Loop_5 (Loop :bTokenizeOnWhiteSpace true :bUseGraphDefaultPinUsage false :LoopPins "WaterBodyActors,WaterBodySplines" :Graph /Game/Athena/Helios/Environments/Landscape/WorldBuildingTools/PCGGrass/PCG_MeshGrass_WaterBodyProcessor_Default.PCG_MeshGrass_WaterBodyProcessor_Default))
  (bind MergePoints_24 (Merge%20Points))
  (bind Projection_12 (Projection :bKeepZeroDensityPoints true :ProjectionParams (struct :bProjectRotations false :AttributeMode IncludeAttributes)))
  (bind AttributeMathsOp_16 (Attribute%20Maths%20Op :InputSource1 $Position.Z))
  (bind AttributeFilter_14 (Filter%20Attribute%20Elements :AttributeTypes (struct :Type Boolean :BoolValue true) :Operator Lesser :TargetAttribute $Position.Z :ThresholdAttribute $Position.Z))
  (bind AttributeMathsOp_1 (Attribute%20Maths%20Op :InputSource1 $Position.Z :InB LocationOffsetZ))
  (bind Subgraph_7 (Subgraph :Graph /Game/Athena/Helios/Environments/Landscape/WorldBuildingTools/PCGGrass/PCG_HasValidPoints.PCG_HasValidPoints))
  (bind Subgraph_52 (Subgraph :Graph /Game/Athena/Helios/Environments/Landscape/WorldBuildingTools/PCGGrass/PCG_HasValidData.PCG_HasValidData))
  (bind AttributeBooleanOp_8 (Attribute%20Boolean%20Op))
  (bind Branch_53 (Branch))
  (bind Distance_2 (Distance :MaximumDistance 2e+03 :OutputAttribute WaterDistance :SourceShape BoxBounds :TargetShape BoxBounds))
  (bind AttributeFilter_8 (Filter%20Attribute%20Elements :AttributeTypes (struct :DoubleValue 2e+02) :Operator GreaterOrEqual :TargetAttribute @Last :ThresholdAttribute @Last))
  (bind AttributeCompareOp_44 (Attribute%20Compare%20Op :Operation Less))
  (bind Branch_51 (Branch))
  (bind Distance_33 (Distance :MaximumDistance 2e+03 :OutputAttribute WaterDistance :SourceShape BoxBounds :TargetShape BoxBounds))
  (bind AttributeFilter_35 (Filter%20Attribute%20Elements :AttributeTypes (struct :DoubleValue 2e+02) :Operator GreaterOrEqual :TargetAttribute @Last :ThresholdAttribute @Last))
  (bind Branch_49 (Branch))
  (bind MergePoints_42 (Merge%20Points))
  (bind Select_2 (Select :bUseInputB-param WaterBodyInvertFilter))
  (bind AttributeFilter_0 (Filter%20Attribute%20Elements :AttributeTypes (struct :Type Boolean) :bUseConstantThreshold true :Operator Equal :TargetAttribute DiscardGrassDensity)) ; deferred to here to preserve integrity of distance calculcations
  (bind DeleteAttributes_55 (Delete%20Attributes :SelectedAttributes "WaterDistance,DiscardGrassDensity"))
  (bind Gather_54 (Gather))
  (bind Subgraph_4 (Subgraph :Graph /Game/Athena/Helios/Environments/Landscape/WorldBuildingTools/PCGGrass/PCG_GateValidPoints.PCG_GateValidPoints))
  (bind CreateAttribute_20 (Create%20Constant :AttributeTypes (struct :Type SoftObjectPath :SoftObjectPathValue (struct :AssetPath (struct :PackageName /Game/Athena/Helios/Environments/Landscape/WorldBuildingTools/PCGGrass/PCG_MeshGrass_WaterBodyProcessor_Default :AssetName PCG_MeshGrass_WaterBodyProcessor_Default)))))
  (bind CreateAttribute_39 (Create%20Constant :AttributeTypes (struct :Type String :StringValue "None")))
  (bind FilterAttribute_25 (Delete%20Attributes :bTokenizeOnWhiteSpace true :Operation KeepSelectedAttributes :SelectedAttributes "UnderWaterProcessorGraph"))
  (bind ExecuteBlueprint_0 (Execute%20Blueprint :BlueprintElementInstance (class:/Game/Athena/Helios/Environments/Landscape/WorldBuildingTools/PCGGrass/PCGBP_SoftObjectPathToString.PCGBP_SoftObjectPathToString_C) :BlueprintElementType /Game/Athena/Helios/Environments/Landscape/WorldBuildingTools/PCGGrass/PCGBP_SoftObjectPathToString.PCGBP_SoftObjectPathToString_C))
  (bind AttributeCompareOp_43 (Attribute%20Compare%20Op :Operation NotEqual :OutputTarget IsValid))
  (bind FilterAttribute_1 (Delete%20Attributes :bTokenizeOnWhiteSpace true :Operation KeepSelectedAttributes :SelectedAttributes "IsValid"))
  (bind Select_18 (Select))
  ;[0.05 0.05 0.05] Pruning
  (bind DeleteAttributes_4 (Delete%20Attributes :bTokenizeOnWhiteSpace true :Operation KeepSelectedAttributes :SelectedAttributes "UsePruning"))
  (bind Branch_3 (Branch))
  (bind SelfPruning_0 (Self%20Pruning :Parameters (struct :PruningType AllEqual :ComparisonSource $Extents :RadiusSimilarityFactor 1e-05)))
  (bind Gather_6 (Gather))
  ;[0.05 0.05 0.05] Spatial Filters
  (bind DeleteAttributes_5 (Delete%20Attributes :bTokenizeOnWhiteSpace true :Operation KeepSelectedAttributes :SelectedAttributes "FilterWorldIntersection"))
  (bind Branch_10 (Branch))
  (bind DeleteAttributes_35 (Delete%20Attributes :bTokenizeOnWhiteSpace true :Operation KeepSelectedAttributes :SelectedAttributes "FilterNonSelfHits"))
  (bind DeleteAttributes_8 (Delete%20Attributes :bTokenizeOnWhiteSpace true :Operation KeepSelectedAttributes :SelectedAttributes "ExtendNegativeBounds"))
  (bind FilterAttribute_30 (Delete%20Attributes :bTokenizeOnWhiteSpace true :Operation KeepSelectedAttributes :SelectedAttributes "WorldIntersectionDistance"))
  (bind Subgraph_3 (Subgraph :Graph /Game/Athena/Helios/Environments/Landscape/WorldBuildingTools/PCGGrass/PCG_MeshGrass_FilterWorldIntersection.PCG_MeshGrass_FilterWorldIntersection))
  (bind Branch_17 (Branch))
  (bind DeleteAttributes_33 (Delete%20Attributes :bTokenizeOnWhiteSpace true :Operation KeepSelectedAttributes :SelectedAttributes "FilterBorders"))
  (bind DeleteAttributes_3 (Delete%20Attributes :bTokenizeOnWhiteSpace true :Operation KeepSelectedAttributes :SelectedAttributes "NonSelfTraceComplex"))
  (bind DeleteAttributes_24 (Delete%20Attributes :bTokenizeOnWhiteSpace true :Operation KeepSelectedAttributes :SelectedAttributes "NonSelfHitDistance"))
  (bind DeleteAttributes_25 (Delete%20Attributes :bTokenizeOnWhiteSpace true :Operation KeepSelectedAttributes :SelectedAttributes "NonSelfTraceZOffset"))
  (bind Subgraph_2 (Subgraph :Graph /Game/Athena/Helios/Environments/Landscape/WorldBuildingTools/PCGGrass/PCG_MeshGrass_FilterNonSelfHits.PCG_MeshGrass_FilterNonSelfHits))
  (bind Branch_23 (Branch))
  (bind DeleteAttributes_72 (Delete%20Attributes :bTokenizeOnWhiteSpace true :Operation KeepSelectedAttributes :SelectedAttributes "BorderSize"))
  (bind DeleteAttributes_76 (Delete%20Attributes :bTokenizeOnWhiteSpace true :Operation KeepSelectedAttributes :SelectedAttributes "BorderDistance"))
  (bind Subgraph_1 (Subgraph :Graph /Game/Athena/Helios/Environments/Landscape/WorldBuildingTools/PCGGrass/PCG_MeshGrass_FilterBorders.PCG_MeshGrass_FilterBorders))
  (bind ToPoint_39 (To%20Point))
  (bind Subgraph_0 (Subgraph :Graph /Game/Athena/Helios/Environments/Landscape/WorldBuildingTools/PCGGrass/PCG_GateValidPoints.PCG_GateValidPoints))
  ;;[0.05 0.05 0.05] Cull Volumes
  (bind Branch_2 (Branch :bOutputToB-param IgnoreCullVolumes))
  (bind GetActorData_13 (Get%20Actor%20Data :ActorSelector (struct :ActorFilter AllWorldActors :bMustOverlapSelf true :ActorSelection ByClass :ActorSelectionClass /Game/Athena/Helios/Environments/Landscape/WorldBuildingTools/PCGGrass/BP_MeshGrass_CullVolume.BP_MeshGrass_CullVolume_C :bSelectMultiple true) :bComponentsMustOverlapSelf false :bMergeSinglePointData true :Mode GetSinglePoint)) ; single point
  (bind ApplyScaleToBounds_0 (Apply%20Scale%20To%20Bounds))
  (bind Difference_24 (Difference :bDiffMetadata false :DensityFunction Binary :Mode Discrete))
  (bind ToPoint_25 (To%20Point))
  ;[0.05 0.05 0.05] Inputs
  (bind AttributeBooleanOp_0 (Attribute%20Boolean%20Op :Operation Or :InA FilterWaterBodies :InB FilterBelowTerrain))
  (bind AttributeBooleanOp_1 (Attribute%20Boolean%20Op :Operation Or))
  (bind Branch_4 (Branch))
  (bind GetActorData_0 (Get%20Actor%20Data :ActorSelector (struct :ActorFilter AllWorldActors :bMustOverlapSelf true :ActorSelection ByClass :ActorSelectionClass /Script/Landscape.Landscape) :Mode GetActorReference))
  (bind GetLandscapeData_8 (Get%20Landscape%20Data :ActorSelector (struct :ActorFilter FromInput :bMustOverlapSelf true :ActorSelection ByClass :ActorSelectionClass /Script/Landscape.LandscapeProxy :bSelectMultiple true) :bComponentsMustOverlapSelf false :SamplingProperties (struct)))
  (bind GetAttributeFromPointIndex_4 (Get%20Attribute%20From%20Point%20Index :bForceOutputAttributeToBeInElementsDomain true :InputSource Biome :OutputAttributeName Biome))
  ;[0.04971 0.04971 0.04971] World Texture Filter
  (bind GetActorProperty_0 (Get%20Actor%20Property :ActorSelector (struct :ActorFilter AllWorldActors :ActorSelection ByClass :ActorSelectionClass /Game/Athena/Helios/Environments/Landscape/WorldBuildingTools/PCGGrass/BP_MeshGrass_GlobalControls.BP_MeshGrass_GlobalControls_C) :bSanitizeOutputAttributeName false :OutputAttributeName None :PropertyName WorldTextureData))
  (bind AttributePartition_0 (Attribute%20Partition :PartitionAttributeSelectors [$Index]))
  (bind GetActorProperty_1 (Get%20Actor%20Property :ActorSelector (struct :ActorFilter AllWorldActors :ActorSelection ByClass :ActorSelectionClass /Game/Athena/Helios/Environments/Landscape/WorldBuildingTools/PCGGrass/BP_MeshGrass_GlobalControls.BP_MeshGrass_GlobalControls_C) :bSanitizeOutputAttributeName false :OutputAttributeName None :PropertyName ExcludeByWorldTextures))
  (bind Branch_0 (Branch))
  (bind Loop_0 (Loop :Graph /Game/Athena/Helios/Environments/Landscape/WorldBuildingTools/PCGGrass/PCG_MeshGrass_ProcessWorldTexture.PCG_MeshGrass_ProcessWorldTexture))
  (reroute Actor%20Attributes ActorAttributes)
  (reroute Actor%20Biome GetAttributeFromPointIndex_4.Attribute)
  (reroute Actor%20Point ActorPoint)
  (reroute Biome%20Attributes BiomeAttributes)
  (reroute Global%20Attributes GlobalAttributes)
  (reroute In%20Points InPoints)
  (reroute Landscape GetLandscapeData_8)
  (reroute LandscapeLayerFilteringEnabled AttributeBooleanOp_2)
  (reroute Volume%20Attributes VolumeAttributes)
  (connect Projection_8 Out AddAttribute_25 In)
  (connect GetActorData_13 Out ApplyScaleToBounds_0 In)
  (connect AttributeBooleanOp_0 Out AttributeBooleanOp_1 InA)
  (connect LandscapeLayerFilteringEnabled Out AttributeBooleanOp_1 InB)
  (connect AttributeBooleanOp_47 Out AttributeBooleanOp_2 InA)
  (connect AttributeBooleanOp_12 Out AttributeBooleanOp_2 InB)
  (connect AttributeBooleanOp_47 Out AttributeBooleanOp_3 InA)
  (connect AttributeBooleanOp_12 Out AttributeBooleanOp_3 InB)
  (connect Subgraph_52 IsValid AttributeBooleanOp_8 InA)
  (connect Subgraph_7 IsValid AttributeBooleanOp_8 InB)
  (connect DeleteAttributes_1 Out AttributeBooleanOp_12 InA)
  (connect AttributeCompareOp_8 Out AttributeBooleanOp_12 InB)
  (connect FilterAttribute_3 Out AttributeBooleanOp_47 InA)
  (connect AttributeCompareOp_45 Out AttributeBooleanOp_47 InB)
  (connect DataNum_0 Out AttributeCompareOp_1 InA)
  (connect CreateAttribute_0 Out AttributeCompareOp_1 InB)
  (connect DeleteAttributes_6 Out AttributeCompareOp_8 InA)
  (connect CreateAttribute_10 Out AttributeCompareOp_8 InB)
  (connect ExecuteBlueprint_0 Out AttributeCompareOp_43 InA)
  (connect CreateAttribute_39 Out AttributeCompareOp_43 InB)
  (connect Select_1 Out AttributeCompareOp_44 InA)
  (connect FilterAttribute_12 Out AttributeCompareOp_45 InA)
  (connect CreateAttribute_46 Out AttributeCompareOp_45 InB)
  (connect Select_2 Out AttributeFilter_0 In)
  (connect AttributeMathsOp_40 Out AttributeFilter_8 Filter)
  (connect Distance_2 Out AttributeFilter_8 In)
  (connect AttributeMathsOp_16 Out AttributeFilter_14 Filter)
  (connect MergePoints_24 Out AttributeFilter_14 In)
  (connect AttributeMathsOp_37 Out AttributeFilter_35 Filter)
  (connect Distance_33 Out AttributeFilter_35 In)
  (connect AttributeFilter_14 OutsideFilter AttributeMathsOp_1 InA)
  (connect Projection_7 Out AttributeMathsOp_12 InA)
  (connect Projection_12 Out AttributeMathsOp_16 InA)
  (connect CreateAttribute_26 Out AttributeMathsOp_16 InB)
  (connect Branch_48 Output%20B AttributeMathsOp_35 InA)
  (connect AttributeMathsOp_36 Out AttributeMathsOp_35 InB)
  (connect CreateAttribute_37 Out AttributeMathsOp_36 InA)
  (connect DeleteAttributes_26 Out AttributeMathsOp_36 InB)
  (connect Select_1 Out AttributeMathsOp_37 In)
  (connect AttributeMathsOp_35 Out AttributeMathsOp_38 In)
  (connect AttributeRemap_40 Out AttributeMathsOp_39 InA)
  (connect AttributeRemap_40 Out AttributeMathsOp_39 InB)
  (connect Select_1 Out AttributeMathsOp_40 InA)
  (connect GetActorProperty_0 Out AttributePartition_0 In)
  (connect ExecuteBlueprint_8 Out AttributePartition_40 In)
  (connect AttributeMathsOp_38 Out AttributeRemap_40 In)
  (connect CreateAttribute_37 Out AttributeRemap_40 OutRangeMax)
  (connect BreakVectorAttribute_41 X AttributeRemap_40 OutRangeMin)
  (connect DeleteAttributes_6 Out AttributeRename_6 In)
  (connect FilterAttribute_12 Out AttributeStringOp_15 InA)
  (connect CreateAttribute_16 Out AttributeStringOp_15 InB)
  (connect AttributeStringOp_15 Out AttributeStringOp_18 InA)
  (connect AttributeRename_6 Out AttributeStringOp_18 InB)
  (connect DeleteAttributes_30 Out BlurElement_27 In)
  (connect DeleteAttributes_29 Out BlurElement_27 SearchDistance)
  (connect GetActorProperty_1 Out Branch_0 bOutputToB)
  (connect Gather_31 Out Branch_0 In)
  (connect Subgraph_9 ValidPoints Branch_0 In)
  (connect AttributeBooleanOp_2 Out Branch_1 bOutputToB)
  (connect Gather_7 Out Branch_1 In)
  (connect Gather_6 Out Branch_2 In)
  (connect DeleteAttributes_4 Out Branch_3 bOutputToB)
  (connect In%20Points Out Branch_3 In)
  (connect AttributeBooleanOp_1 Out Branch_4 bOutputToB)
  (connect Actor%20Point Out Branch_4 In)
  (connect AttributeCompareOp_1 Out Branch_8 bOutputToB)
  (connect Subgraph_4 ValidPoints Branch_8 In)
  (connect DeleteAttributes_5 Out Branch_10 bOutputToB)
  (connect Branch_2 Output%20B Branch_10 In)
  (connect Difference_24 Out Branch_10 In)
  (connect Branch_8 Output%20B Branch_14 In)
  (connect DeleteAttributes_35 Out Branch_17 bOutputToB)
  (connect Branch_10 Output%20A Branch_17 In)
  (connect Subgraph_3 OutSpatial Branch_17 In)
  (connect DeleteAttributes_33 Out Branch_23 bOutputToB)
  (connect Branch_17 Output%20A Branch_23 In)
  (connect Subgraph_2 OutSpatial Branch_23 In)
  (connect Subgraph_0 ValidPoints Branch_31 In)
  (connect DeleteAttributes_51 Out Branch_48 bOutputToB)
  (connect PointFilter_9 InsideFilter Branch_48 In)
  (connect AttributeCompareOp_44 Out Branch_49 bOutputToB)
  (connect AttributeFilter_8 InsideFilter Branch_49 In)
  (connect AttributeCompareOp_44 Out Branch_51 bOutputToB)
  (connect AttributeFilter_8 OutsideFilter Branch_51 In)
  (connect AttributeBooleanOp_8 Out Branch_53 bOutputToB)
  (connect Branch_31 Output%20B Branch_53 In)
  (connect DeleteAttributes_43 Out BreakVectorAttribute_41 In)
  (connect Landscape Out DataNum_0 In)
  (connect Actor%20Attributes Out DeleteAttributes_1 In)
  (connect Actor%20Attributes Out DeleteAttributes_3 In)
  (connect Actor%20Attributes Out DeleteAttributes_4 In)
  (connect Actor%20Attributes Out DeleteAttributes_5 In)
  (connect Actor%20Attributes Out DeleteAttributes_6 In)
  (connect Actor%20Attributes Out DeleteAttributes_8 In)
  (connect Actor%20Attributes Out DeleteAttributes_24 In)
  (connect Actor%20Attributes Out DeleteAttributes_25 In)
  (connect Actor%20Attributes Out DeleteAttributes_26 In)
  (connect Actor%20Attributes Out DeleteAttributes_29 In)
  (connect Loop_43 InPoints DeleteAttributes_30 In)
  (connect Select_0 Out DeleteAttributes_30 SelectedAttributes)
  (connect Actor%20Attributes Out DeleteAttributes_33 In)
  (connect AttributeMathsOp_39 Out DeleteAttributes_34 In)
  (connect Branch_48 Output%20A DeleteAttributes_34 In)
  (connect Actor%20Attributes Out DeleteAttributes_35 In)
  (connect Branch_48 Output%20B DeleteAttributes_43 Execution%20Dependency)
  (connect Actor%20Attributes Out DeleteAttributes_43 In)
  (connect Actor%20Attributes Out DeleteAttributes_51 In)
  (connect AttributeFilter_0 InsideFilter DeleteAttributes_55 In)
  (connect Actor%20Attributes Out DeleteAttributes_72 In)
  (connect Actor%20Attributes Out DeleteAttributes_76 In)
  (connect ApplyScaleToBounds_0 Out Difference_24 Differences)
  (connect Branch_2 Output%20A Difference_24 Source)
  (connect Branch_53 Output%20B Distance_2 Source)
  (connect AttributeMathsOp_1 Out Distance_2 Target)
  (connect Branch_51 Output%20A Distance_33 Source)
  (connect Branch_51 Output%20B Distance_33 Source)
  (connect AttributeFilter_8 InsideFilter Distance_33 Target)
  (connect FilterAttribute_25 Out ExecuteBlueprint_0 In)
  (connect Select_0 Out ExecuteBlueprint_1 In)
  (connect CreateAttribute_16 Out ExecuteBlueprint_8 Delimiter)
  (connect ExecuteBlueprint_1 Out ExecuteBlueprint_8 In)
  (connect AttributeCompareOp_43 Out FilterAttribute_1 In)
  (connect Global%20Attributes Out FilterAttribute_3 In)
  (connect Global%20Attributes Out FilterAttribute_12 In)
  (connect Global%20Attributes Out FilterAttribute_25 In)
  (connect Actor%20Attributes Out FilterAttribute_30 In)
  (connect GetActorData_4 Out FilterByType_2 In)
  (connect Branch_3 Output%20A Gather_6 In)
  (connect SelfPruning_0 Out Gather_6 In)
  (connect Branch_14 Output%20A Gather_7 In)
  (connect Subgraph_5 ValidPoints Gather_7 In)
  (connect Branch_1 Output%20A Gather_31 In)
  (connect Branch_8 Output%20A Gather_31 In)
  (connect Branch_31 Output%20A Gather_54 In)
  (connect Branch_53 Output%20A Gather_54 In)
  (connect Branch_4 Output%20B GetActorData_0 Execution%20Dependency)
  (connect Branch_31 Output%20B GetActorData_2 Execution%20Dependency)
  (connect GetActorData_2 Out GetActorData_4 In)
  (connect Actor%20Point Out GetAttributeFromPointIndex_4 In)
  (connect Branch_4 Output%20B GetLandscapeData_8 Execution%20Dependency)
  (connect GetActorData_0 Out GetLandscapeData_8 In)
  (connect Branch_0 Output%20B Loop_0 Points)
  (connect AttributePartition_0 Out Loop_0 WorldTextureData)
  (connect Branch_31 Output%20B Loop_5 Exec)
  (connect GetActorData_2 Out Loop_5 WaterBodyActors)
  (connect FilterByType_2 InsideFilter Loop_5 WaterBodySplines)
  (connect AddAttribute_25 Out Loop_43 InPoints)
  (connect AttributePartition_40 Out Loop_43 SourceAttributes)
  (connect Loop_5 WaterBodyPoints MergePoints_24 In)
  (connect AttributeFilter_35 OutsideFilter MergePoints_42 In)
  (connect Branch_49 Output%20B MergePoints_42 In)
  (connect DeleteAttributes_26 Out PointFilter_9 Filter)
  (connect BlurElement_27 Out PointFilter_9 In)
  (connect AttributeMathsOp_12 Out PointFilter_10 Filter)
  (connect Branch_14 Output%20B PointFilter_10 In)
  (connect Branch_14 Output%20B Projection_7 In)
  (connect Landscape Out Projection_7 Projection%20Target)
  (connect ExecuteBlueprint_1 Out Projection_8 AttributeList)
  (connect Branch_1 Output%20B Projection_8 In)
  (connect Landscape Out Projection_8 Projection%20Target)
  (connect MergePoints_24 Out Projection_12 In)
  (connect Landscape Out Projection_12 Projection%20Target)
  (connect AttributeBooleanOp_3 Out Select_0 bUseInputB)
  (connect Select_5 Out Select_0 Input%20A)
  (connect AttributeStringOp_18 Out Select_0 Input%20B)
  (connect AttributeMathsOp_0 Out Select_1 Input%20B)
  (connect Branch_49 Output%20A Select_2 Input%20A)
  (connect MergePoints_42 Out Select_2 Input%20A)
  (connect AttributeFilter_35 InsideFilter Select_2 Input%20B)
  (connect AttributeBooleanOp_47 Out Select_5 bUseInputB)
  (connect AttributeRename_6 Out Select_5 Input%20A)
  (connect FilterAttribute_12 Out Select_5 Input%20B)
  (connect FilterAttribute_1 Out Select_18 bUseInputB)
  (connect CreateAttribute_20 Out Select_18 Input%20A)
  (connect FilterAttribute_25 Out Select_18 Input%20B)
  (connect Branch_3 Output%20B SelfPruning_0 In)
  (connect ToPoint_39 Out Subgraph_0 In)
  (connect Actor%20Point Out Subgraph_1 ActorPoint)
  (connect DeleteAttributes_76 Out Subgraph_1 BorderDistance)
  (connect DeleteAttributes_72 Out Subgraph_1 BorderSize)
  (connect Branch_23 Output%20B Subgraph_1 InSpatial)
  (connect ToPoint_25 Out Subgraph_1 OriginalPoints)
  (connect Actor%20Point Out Subgraph_2 ActorPoint)
  (connect Branch_17 Output%20B Subgraph_2 InSpatial)
  (connect DeleteAttributes_24 Out Subgraph_2 NonSelfHitDistance)
  (connect DeleteAttributes_3 Out Subgraph_2 NonSelfTraceComplex)
  (connect DeleteAttributes_25 Out Subgraph_2 NonSelfTraceZOffset)
  (connect Actor%20Point Out Subgraph_3 ActorPoint)
  (connect DeleteAttributes_8 Out Subgraph_3 ExtendNegativeBounds)
  (connect Branch_10 Output%20B Subgraph_3 InSpatial)
  (connect FilterAttribute_30 Out Subgraph_3 WorldIntersectionDistance)
  (connect DeleteAttributes_55 Out Subgraph_4 In)
  (connect Gather_54 Out Subgraph_4 In)
  (connect PointFilter_10 OutsideFilter Subgraph_5 In)
  (connect AttributeFilter_14 OutsideFilter Subgraph_7 In)
  (connect DeleteAttributes_34 Out Subgraph_9 In)
  (connect Loop_5 WaterBodyPoints Subgraph_52 In)
  (connect Branch_2 Output%20B ToPoint_25 In)
  (connect Difference_24 Out ToPoint_25 In)
  (connect Branch_23 Output%20A ToPoint_39 In)
  (connect Subgraph_1 OutSpatial ToPoint_39 In)
  (output OutPoints Branch_0.Output%20A)
  (output OutPoints Loop_0.Points))
