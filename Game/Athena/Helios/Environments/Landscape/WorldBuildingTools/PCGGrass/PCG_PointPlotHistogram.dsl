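; PCG loop subgraph in the grass world-building tools that plots one bucket of a histogram as a bar of grid points, over a background grid.
; Bucket value is divided by MaxValue and floored to tenths to set bar height; cell size rescales with bucket count.
; Loop index positions each bar along X, aligned to the canvas edge and shifted by HistogramOffset.
; Bar density encodes bucket position, while the background gets fixed density.
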(graph PCG_PointPlotHistogram
  (params (HistogramHeight :type Double :default 0.2) (HistogramCellSize :type Double :default 1e+02) (HistogramOffset :type Vector))
  (inputs (HistogramBuckets :type Any) (Buckets :type Any) (MaxValue :type Param))
  (outputs (HistogramBucketPoints :type Point) (HistogramBackgroundPoints :type Point))
  ;[0.05 0.05 0.05] Histogram
  (bind CreateAttribute_42 (Create%20Constant :AttributeTypes (struct :DoubleValue 0.5)))
  (bind AttributeMathsOp_41 (Attribute%20Maths%20Op :Operation Multiply))
  (bind MakeVectorAttribute_2 (Make%20Vector%20Attribute :OutputType Vector))
  (bind CreatePointsGrid_0 (Create%20Points%20Grid :CoordinateSpace LocalComponent :PointSteepness 0.5))
  (bind CreateAttribute_124 (Create%20Constant :AttributeTypes (struct :DoubleValue 1e+01)))
  (bind AttributeMathsOp_147 (Attribute%20Maths%20Op :Operation Multiply))
  (bind AttributeMathsOp_125 (Attribute%20Maths%20Op :Operation Floor))
  (bind AttributeMathsOp_126 (Attribute%20Maths%20Op :Operation Divide))
  (bind CreateAttribute_156 (Create%20Constant :AttributeTypes (struct :DoubleValue 1.0)))
  (bind AttributeMathsOp_155 (Attribute%20Maths%20Op :Operation Divide))
  (bind CreateAttribute_151 (Create%20Constant :AttributeTypes (struct :DoubleValue 1e+02)))
  (bind AttributeMathsOp_153 (Attribute%20Maths%20Op :Operation Multiply))
  (bind AttributeMathsOp_150 (Attribute%20Maths%20Op :Operation Multiply))
  (bind AttributeMathsOp_5 (Attribute%20Maths%20Op :Operation Multiply))
  (bind AttributeMathsOp_9 (Attribute%20Maths%20Op :Operation Multiply))
  (bind AttributeMathsOp_98 (Attribute%20Maths%20Op))
  (bind CreateAttribute_96 (Create%20Constant :AttributeTypes (struct :DoubleValue -1.0)))
  (bind AttributeMathsOp_95 (Attribute%20Maths%20Op :Operation Multiply))
  (bind AttributeMathsOp_107 (Attribute%20Maths%20Op))
  (bind CreateAttribute_1 (Create%20Constant))
  (bind MakeVectorAttribute_11 (Make%20Vector%20Attribute :OutputType Vector))
  (bind AttributeMathsOp_50 (Attribute%20Maths%20Op :Operation Multiply))
  (bind CreateAttribute_54 (Create%20Constant :AttributeTypes (struct :DoubleValue -0.5)))
  (bind AttributeMathsOp_53 (Attribute%20Maths%20Op :Operation Multiply))
  (bind CreateAttribute_57 (Create%20Constant))
  (bind MakeVectorAttribute_56 (Make%20Vector%20Attribute :OutputType Vector))
  (bind AttributeMathsOp_48 (Attribute%20Maths%20Op))
  (bind AttributeMathsOp_3 (Attribute%20Maths%20Op :InB HistogramOffset))
  (bind TransformPoints_6 (Transform%20Points))
  (bind AttributeMathsOp_13 (Attribute%20Maths%20Op :Operation Multiply :InB HistogramHeight))
  ;[0.05 0.05 0.05] Inputs
  (bind FilterDataByType_4 (Filter%20Data%20By%20Type :TargetType (struct :Ids [(struct :Struct /Script/PCG.PCGDataTypeInfoParam)])))
  (bind GetAttributeFromIndex_3 (Get%20Attribute%20Set%20from%20Index))
  (bind GetLoopIndex_7 (Get%20Loop%20Index))
  (bind DataNum_49 (Data%20Count))
  (bind AttributeMathsOp_102 (Attribute%20Maths%20Op :Operation Multiply))
  (bind CreateAttribute_120 (Create%20Constant :AttributeTypes (struct :DoubleValue 2.0)))
  (bind AttributeMathsOp_119 (Attribute%20Maths%20Op :Operation Divide))
  (bind CreateAttribute_131 (Create%20Constant :AttributeTypes (struct :DoubleValue 1e+01)))
  (bind AttributeMathsOp_130 (Attribute%20Maths%20Op :Operation Divide))
  (bind AttributeMathsOp_134 (Attribute%20Maths%20Op :Operation Multiply :InA HistogramCellSize))
  (bind AttributeMathsOp_146 (Attribute%20Maths%20Op :bForceOpToDouble true :Operation Divide))
  ;[0.05 0.05 0.05] Background
  (bind CreateAttribute_116 (Create%20Constant :AttributeTypes (struct :DoubleValue 2.0)))
  (bind MakeVectorAttribute_144 (Make%20Vector%20Attribute :OutputType Vector))
  (bind AttributeMathsOp_115 (Attribute%20Maths%20Op :Operation Divide))
  (bind CreatePointsGrid_26 (Create%20Points%20Grid :CellSize (1e+03 1e+03 5e+01) :CoordinateSpace LocalComponent :GridExtents (5e+02 5e+02 25.0) :PointSteepness 0.5))
  (bind TransformPoints_47 (Transform%20Points :OffsetMax (0.0 0.0 5e+02) :OffsetMin (0.0 0.0 5e+02) :ScaleMax (1.025 1.025 1.025) :ScaleMin (1.025 1.025 1.025) :OffsetMax-param HistogramOffset :OffsetMin-param HistogramOffset))
  ;[0.05 0.05 0.05] Density
  (bind CreateAttribute_28 (Create%20Constant :AttributeTypes (struct :DoubleValue 0.475)))
  (bind AttributeMathsOp_27 (Attribute%20Maths%20Op :InputSource1 $Density :Operation Set))
  (bind CreateAttribute_62 (Create%20Constant))
  (bind AttributeMathsOp_63 (Attribute%20Maths%20Op :Operation Set))
  (bind AttributeMathsOp_60 (Attribute%20Maths%20Op :Operation Divide))
  (bind CreateAttribute_67 (Create%20Constant :AttributeTypes (struct :DoubleValue 1.0)))
  (bind AttributeMathsOp_66 (Attribute%20Maths%20Op :Operation Divide))
  (bind AttributeMathsOp_64 (Attribute%20Maths%20Op))
  (bind AttributeMathsOp_30 (Attribute%20Maths%20Op :InputSource1 $Density :Operation Set))
  (reroute BucketCount DataNum_49)
  (reroute CanvasSize AttributeMathsOp_102)
  (reroute CellExtent AttributeMathsOp_119)
  (reroute GlobalScale AttributeMathsOp_130)
  (reroute LoopIndex GetLoopIndex_7)
  (reroute Normalized%20Cell%20Height AttributeMathsOp_146)
  (reroute NormalizedCellSize AttributeMathsOp_134)
  (reroute ScaledHistogramHeight AttributeMathsOp_13)
  (connect AttributeMathsOp_48 Out AttributeMathsOp_3 InA)
  (connect NormalizedCellSize Out AttributeMathsOp_5 InA)
  (connect AttributeMathsOp_150 Out AttributeMathsOp_5 InB)
  (connect NormalizedCellSize Out AttributeMathsOp_9 InA)
  (connect LoopIndex Out AttributeMathsOp_9 InB)
  (connect AttributeMathsOp_5 Out AttributeMathsOp_13 InA)
  (connect TransformPoints_47 Out AttributeMathsOp_27 InA)
  (connect CreateAttribute_28 Out AttributeMathsOp_27 InB)
  (connect TransformPoints_6 Out AttributeMathsOp_30 InA)
  (connect AttributeMathsOp_64 Out AttributeMathsOp_30 InB)
  (connect NormalizedCellSize Out AttributeMathsOp_41 InA)
  (connect CreateAttribute_42 Out AttributeMathsOp_41 InB)
  (connect MakeVectorAttribute_11 Out AttributeMathsOp_48 InA)
  (connect MakeVectorAttribute_56 Out AttributeMathsOp_48 InB)
  (connect BucketCount Out AttributeMathsOp_50 InA)
  (connect NormalizedCellSize Out AttributeMathsOp_50 InB)
  (connect AttributeMathsOp_50 Out AttributeMathsOp_53 InA)
  (connect CreateAttribute_54 Out AttributeMathsOp_53 InB)
  (connect AttributeMathsOp_63 Out AttributeMathsOp_60 InA)
  (connect BucketCount Out AttributeMathsOp_60 InB)
  (connect CreateAttribute_62 Out AttributeMathsOp_63 InA)
  (connect LoopIndex Out AttributeMathsOp_63 InB)
  (connect AttributeMathsOp_60 Out AttributeMathsOp_64 InA)
  (connect AttributeMathsOp_66 Out AttributeMathsOp_64 InB)
  (connect CreateAttribute_67 Out AttributeMathsOp_66 InA)
  (connect BucketCount Out AttributeMathsOp_66 InB)
  (connect ScaledHistogramHeight Out AttributeMathsOp_95 InA)
  (connect CreateAttribute_96 Out AttributeMathsOp_95 InB)
  (connect AttributeMathsOp_9 Out AttributeMathsOp_98 InA)
  (connect CellExtent Out AttributeMathsOp_98 InB)
  (connect BucketCount Out AttributeMathsOp_102 InA)
  (connect NormalizedCellSize Out AttributeMathsOp_102 InB)
  (connect AttributeMathsOp_95 Out AttributeMathsOp_107 InA)
  (connect CanvasSize Out AttributeMathsOp_107 InB)
  (connect MakeVectorAttribute_144 Out AttributeMathsOp_115 InA)
  (connect CreateAttribute_116 Out AttributeMathsOp_115 InB)
  (connect NormalizedCellSize Out AttributeMathsOp_119 InA)
  (connect CreateAttribute_120 Out AttributeMathsOp_119 InB)
  (connect AttributeMathsOp_147 Out AttributeMathsOp_125 In)
  (connect AttributeMathsOp_125 Out AttributeMathsOp_126 InA)
  (connect CreateAttribute_124 Out AttributeMathsOp_126 InB)
  (connect CreateAttribute_131 Out AttributeMathsOp_130 InA)
  (connect BucketCount Out AttributeMathsOp_130 InB)
  (connect GlobalScale Out AttributeMathsOp_134 InB)
  (connect GetAttributeFromIndex_3 Out AttributeMathsOp_146 InA)
  (connect DefaultInputNode MaxValue AttributeMathsOp_146 InB)
  (connect Normalized%20Cell%20Height Out AttributeMathsOp_147 InA)
  (connect CreateAttribute_124 Out AttributeMathsOp_147 InB)
  (connect AttributeMathsOp_126 Out AttributeMathsOp_150 InA)
  (connect AttributeMathsOp_153 Out AttributeMathsOp_150 InB)
  (connect CreateAttribute_151 Out AttributeMathsOp_153 InA)
  (connect AttributeMathsOp_155 Out AttributeMathsOp_153 InB)
  (connect CreateAttribute_156 Out AttributeMathsOp_155 InA)
  (connect GlobalScale Out AttributeMathsOp_155 InB)
  (connect NormalizedCellSize Out CreatePointsGrid_0 CellSize)
  (connect MakeVectorAttribute_2 Out CreatePointsGrid_0 GridExtents)
  (connect MakeVectorAttribute_144 Out CreatePointsGrid_26 CellSize)
  (connect AttributeMathsOp_115 Out CreatePointsGrid_26 GridExtents)
  (connect DefaultInputNode Buckets DataNum_49 In)
  (connect DefaultInputNode HistogramBuckets FilterDataByType_4 In)
  (connect FilterDataByType_4 InsideFilter GetAttributeFromIndex_3 In)
  (connect AttributeMathsOp_41 Out MakeVectorAttribute_2 X)
  (connect ScaledHistogramHeight Out MakeVectorAttribute_2 Y)
  (connect AttributeMathsOp_41 Out MakeVectorAttribute_2 Z)
  (connect AttributeMathsOp_98 Out MakeVectorAttribute_11 X)
  (connect AttributeMathsOp_107 Out MakeVectorAttribute_11 Y)
  (connect CreateAttribute_1 Out MakeVectorAttribute_11 Z)
  (connect AttributeMathsOp_53 Out MakeVectorAttribute_56 X)
  (connect AttributeMathsOp_53 Out MakeVectorAttribute_56 Y)
  (connect CreateAttribute_57 Out MakeVectorAttribute_56 Z)
  (connect CanvasSize Out MakeVectorAttribute_144 X)
  (connect CanvasSize Out MakeVectorAttribute_144 Y)
  (connect CellExtent Out MakeVectorAttribute_144 Z)
  (connect CreatePointsGrid_0 Out TransformPoints_6 In)
  (connect AttributeMathsOp_3 Out TransformPoints_6 OffsetMax)
  (connect AttributeMathsOp_3 Out TransformPoints_6 OffsetMin)
  (connect CreatePointsGrid_26 Out TransformPoints_47 In)
  (output HistogramBackgroundPoints AttributeMathsOp_27)
  (output HistogramBucketPoints AttributeMathsOp_30))
