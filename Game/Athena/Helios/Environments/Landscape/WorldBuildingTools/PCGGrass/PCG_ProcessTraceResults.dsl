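; PCG subgraph for the Helios grass tool that filters trace-result points to landscape surfaces, seeing through a list of ignored meshes from ProjectionSetup.
; Without valid ignore entries, points pass through with only trace attributes stripped.
; Otherwise, hits without a static mesh are kept if their actor name contains "LandscapeStreamingProxy", tested per actor partition via Blueprint.
; Ignored-mesh hits are raycast again requiring landscape, other mesh hits are dropped, and both surviving sets merge.
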(graph PCG_ProcessTraceResults
  (inputs (InPoints :type Point) (ProjectionSetup :type Param))
  (outputs (OutPoints :type Point))
  (bind Subgraph_27 (Subgraph :Graph /Game/Athena/Helios/Environments/Landscape/WorldBuildingTools/PCGGrass/PCG_HasValidEntries.PCG_HasValidEntries))
  (bind Branch_20 (Branch))
  (bind AttributeFilter_0 (Filter%20Attribute%20Elements :AttributeTypes (struct :Type SoftObjectPath) :bUseConstantThreshold true :Operator Equal :TargetAttribute ImpactStaticMesh :ThresholdAttribute @Last))
  (bind AttributePartition_1 (Attribute%20Partition :PartitionAttributeSelectors [ActorReference]))
  (bind GetAttributeFromPointIndex_2 (Get%20Attribute%20From%20Point%20Index :bForceOutputAttributeToBeInElementsDomain true :InputSource ActorReference :OutputAttributeName @SourceName))
  (bind AttributeCast_3 (Attribute%20Cast :OutputType String))
  (bind ExecuteBlueprint_4 (Execute%20Blueprint :BlueprintElementInstance (class:/Game/Athena/Helios/Environments/Landscape/WorldBuildingTools/PCGGrass/PCGBP_StringContains.PCGBP_StringContains_C :Substring "LandscapeStreamingProxy") :BlueprintElementType /Game/Athena/Helios/Environments/Landscape/WorldBuildingTools/PCGGrass/PCGBP_StringContains.PCGBP_StringContains_C))
  (bind AttributeRename_5 (Attribute%20Rename :AttributeToRename ContainsSubstring :NewAttributeName IsLandscape))
  (bind CopyAttributes_6 (Copy%20Attributes :bCopyAllAttributes true :InputSource $Density :OutputTarget None))
  (bind MergePoints_7 (Merge%20Points))
  (bind AttributeFilter_8 (Filter%20Attribute%20Elements :AttributeTypes (struct :Type Boolean :BoolValue true) :bUseConstantThreshold true :Operator Equal :TargetAttribute IsLandscape :ThresholdAttribute @Last))
  (bind AddAttribute_10 (Add%20Attribute :AttributeTypes (struct :Type Boolean :BoolValue true) :OutputTarget MeshMatch))
  (bind AddAttribute_11 (Add%20Attribute :AttributeTypes (struct :Type Boolean) :OutputTarget MeshMatch))
  (bind MatchAndSetAttributes_9 (Match%20And%20Set%20Attributes :bMatchAttributes true :InputAttribute ImpactStaticMesh :MatchAttribute IgnoredMeshes))
  (bind AttributeFilter_12 (Filter%20Attribute%20Elements :AttributeTypes (struct :Type Boolean :BoolValue true) :bUseConstantThreshold true :Operator Equal :TargetAttribute MeshMatch :ThresholdAttribute @Last))
  (bind Subgraph_16 (Subgraph :Graph /Game/Athena/Helios/Environments/Landscape/WorldBuildingTools/PCGGrass/PCG_GateValidPoints.PCG_GateValidPoints))
  (bind DeleteAttributes_17 (Delete%20Attributes :SelectedAttributes "MeshMatch"))
  (bind WorldRaycast_13 (World%20Raycast :RayLength 1e+04 :WorldQueryParams (struct :SelectLandscapeHits Require)))
  (bind DeleteAttributes_18 (Delete%20Attributes :SelectedAttributes "IsLandscape"))
  (bind MergePoints_14 (Merge%20Points))
  (bind DeleteAttributes_19 (Delete%20Attributes :SelectedAttributes "ActorReference,ImpactStaticMesh"))
  (bind FilterDataByIndex_30 (Filter%20Data%20By%20Index :SelectedIndices "1"))
  (reroute IgnoredMeshes FilterDataByIndex_30.InsideFilter)
  (connect IgnoredMeshes Out AddAttribute_10 In)
  (connect AttributeFilter_0 OutsideFilter AddAttribute_11 In)
  (connect GetAttributeFromPointIndex_2 Attribute AttributeCast_3 In)
  (connect Branch_20 Output%20B AttributeFilter_0 In)
  (connect MergePoints_7 Out AttributeFilter_8 In)
  (connect MatchAndSetAttributes_9 Out AttributeFilter_12 In)
  (connect AttributeFilter_0 InsideFilter AttributePartition_1 In)
  (connect ExecuteBlueprint_4 Out AttributeRename_5 In)
  (connect Subgraph_27 IsValid Branch_20 bOutputToB)
  (connect DefaultInputNode InPoints Branch_20 In)
  (connect AttributeRename_5 Out CopyAttributes_6 Source)
  (connect AttributePartition_1 Out CopyAttributes_6 Target)
  (connect Subgraph_16 ValidPoints DeleteAttributes_17 In)
  (connect AttributeFilter_8 InsideFilter DeleteAttributes_18 In)
  (connect Branch_20 Output%20A DeleteAttributes_19 In)
  (connect MergePoints_14 Out DeleteAttributes_19 In)
  (connect AttributeCast_3 Out ExecuteBlueprint_4 InString)
  (connect DefaultInputNode ProjectionSetup FilterDataByIndex_30 In)
  (connect AttributePartition_1 Out GetAttributeFromPointIndex_2 In)
  (connect AddAttribute_11 Out MatchAndSetAttributes_9 In)
  (connect AddAttribute_10 Out MatchAndSetAttributes_9 Match%20Data)
  (connect CopyAttributes_6 Out MergePoints_7 In)
  (connect DeleteAttributes_18 Out MergePoints_14 In)
  (connect WorldRaycast_13 Out MergePoints_14 In)
  (connect AttributeFilter_12 InsideFilter Subgraph_16 In)
  (connect IgnoredMeshes Out Subgraph_27 In)
  (connect DeleteAttributes_17 Out WorldRaycast_13 Origins)
  (output OutPoints DeleteAttributes_19))
