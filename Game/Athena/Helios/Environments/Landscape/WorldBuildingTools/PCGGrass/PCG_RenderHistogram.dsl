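; PCG debug graph that renders a histogram of a named point attribute as plotted points, apparently for grass world-building tools.
; A points grid sized by BucketCount is partitioned by index into bucket attribute sets, which a looped PCG_AttributeHistogram subgraph fills from the input points.
; A Branch gates the input on a non-empty attribute name.
; Looped PCG_PointPlotHistogram outputs bar and background points using the maximum bucket value, cell size and render offset.
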(graph PCG_RenderHistogram
  (params (TargetAttribute :type String) (BucketCount :type Int :default 10) (CellSize :type Double :default 5e+01) (RenderOffset :type Vector :default (0.0 0.0 5e+02)))
  (inputs (InPoints :type Point))
  (outputs (HistogramBucketPoints :type Point) (HistogramBackgroundPoints :type Point))
  (bind CreateAttribute_8 (Create%20Constant :AttributeTypes (struct :Type String)))
  (bind AttributeCompareOp_6 (Attribute%20Compare%20Op :Operation NotEqual :InA TargetAttribute))
  (bind Branch_9 (Branch))
  (bind CreateAttribute_7 (Create%20Constant :AttributeTypes (struct :Type Vector :VectorValue (0.0 1.0 1.0))))
  (bind AttributeMathsOp_8 (Attribute%20Maths%20Op :InputSource1 @Last.X :Operation Set :InB BucketCount))
  (bind CreatePointsGrid_6 (Create%20Points%20Grid :CellSize (2.0 2.0 2.0) :CoordinateSpace LocalComponent :GridExtents (1.0 1.0 1.0) :PointSteepness 0.5))
  (bind AttributePartition_1 (Attribute%20Partition :PartitionAttributeSelectors [$Index]))
  (bind DataNum_5 (Data%20Count))
  (bind AddAttribute_3 (Add%20Attribute :AttributeTypes (struct :DoubleValue 2e+01) :OutputTarget bucket))
  (bind PointToAttributeSet_2 (Point%20To%20Attribute%20Set))
  (bind Loop_0 (Loop :Graph /Game/Athena/Helios/Environments/Landscape/WorldBuildingTools/PCGGrass/PCG_AttributeHistogram.PCG_AttributeHistogram :TargetAttribute TargetAttribute))
  (bind MergeAttributes_10 (Merge%20Attributes))
  (bind AttributeReduce_9 (Attribute%20Reduce :bMergeOutputAttributes true :Operation Max))
  (bind Loop_4 (Loop :Graph /Game/Athena/Helios/Environments/Landscape/WorldBuildingTools/PCGGrass/PCG_PointPlotHistogram.PCG_PointPlotHistogram :HistogramCellSize CellSize :HistogramOffset RenderOffset))
  (connect DataNum_5 Out AddAttribute_3 Attributes)
  (connect AttributePartition_1 Out AddAttribute_3 In)
  (connect CreateAttribute_8 Out AttributeCompareOp_6 InB)
  (connect CreateAttribute_7 Out AttributeMathsOp_8 InA)
  (connect CreatePointsGrid_6 Out AttributePartition_1 In)
  (connect MergeAttributes_10 Out AttributeReduce_9 In)
  (connect AttributeCompareOp_6 Out Branch_9 bOutputToB)
  (connect DefaultInputNode InPoints Branch_9 In)
  (connect AttributeMathsOp_8 Out CreatePointsGrid_6 GridExtents)
  (connect AttributePartition_1 Out DataNum_5 In)
  (connect PointToAttributeSet_2 Out Loop_0 HistogramBuckets)
  (connect PointToAttributeSet_2 Out Loop_0 HistogramBucketsLooping)
  (connect Branch_9 Output%20B Loop_0 InPoint)
  (connect Loop_0 OutBuckets Loop_4 Buckets)
  (connect Loop_0 OutBuckets Loop_4 HistogramBuckets)
  (connect AttributeReduce_9 Out Loop_4 MaxValue)
  (connect Loop_0 OutBuckets MergeAttributes_10 In)
  (connect AddAttribute_3 Out PointToAttributeSet_2 In)
  (output HistogramBackgroundPoints Loop_4.HistogramBackgroundPoints)
  (output HistogramBucketPoints Loop_4.HistogramBucketPoints))
